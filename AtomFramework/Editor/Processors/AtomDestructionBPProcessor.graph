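# AtomDestructionBPProcessor is an Atom Framework editor utility Blueprint, named as a processor for destruction Blueprints.
# Its only logic overrides GetModelImporterTemplateSettings, passing through the parent implementation's model importer template structure.
# The override always returns false, discarding the parent's success flag.
# The event graph is empty and the Blueprint declares no variables.
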EditorUtilityBlueprint /AtomFramework/Editor/Processors/AtomDestructionBPProcessor
graphs [
 EventGraph {links [], nodes []}
 GetModelImporterTemplateSettings {nodes [
   GetModelImporterTemplateSettings() {is=title K2Node_FunctionEntry}
   > Return Node!(
    execute
    OutModelImporterTemplate:Atom Model Importer Template Structure=Parent: GetModelImporterTemplateSettings!() -> (OutModelImporterTemplate:Atom Model Importer Template Structure, ReturnValue:bool) {is=title K2Node_CallParentFunction}
    ReturnValue:bool=false
   )
  ]}
]
variables []
