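# Dataflow graph that builds Atom Framework destruction Geometry Collections from source geometry, with clustering, convex collision, proximity, anchor and remove-on-break setup.
# Leaf hulls are created and simplified per size band, then feed cluster and root hull generation with optional negative-space shape protection.
# Level 1 cluster counts derive from hull volume, bounding-box grid resolution and fill ratio.
# Branch nodes switch between these adaptive results and user overrides for cluster and convex counts.
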Dataflow /AtomFramework/Editor/Dataflow/DF_GC_Global_V4_ATOM
name=DF_GC_Global_V4_ATOM
connections [
 {
  fromNode=Branch Use CleanUp
  fromPin=ChosenCollection
  fromPinGuid=#1
  toNode=BranchCollection_9
  toPin=FalseCollection
  toPinGuid=#2
 }
 {
  fromNode=FloatToDouble_5
  fromPin=Double
  fromPinGuid=#3
  toNode=ClusterConvexes - Auto
  toPin=MinRadius
  toPinGuid=#4
 }
 {
  fromNode=ClusterConvexes - Auto
  fromPin=Collection
  fromPinGuid=#5
  toNode=MergeConvexHulls_0_copy
  toPin=Collection
  toPinGuid=#6
 }
 {
  fromNode="Level 1 - by  Cluster Number"
  fromPin=Int
  fromPinGuid=#7
  toNode=AutoCluster - L1 User Defined
  toPin=ClusterSites
  toPinGuid=#8
 }
 {
  fromNode=CollectionTransformSelectRoot
  fromPin=TransformSelection
  fromPinGuid=#9
  toNode=AutoCluster - L1 User Defined
  toPin=TransformSelection
  toPinGuid=#10
 }
 {
  fromNode=FloatToInt_8
  fromPin=Int
  fromPinGuid=#11
  toNode=IntToBool_2
  toPin=Int
  toPinGuid=#12
 }
 {
  fromNode=Multiply_9
  fromPin=ReturnValue
  fromPinGuid=#13
  toNode=Clamp_2
  toPin=Float
  toPinGuid=#14
 }
 {
  fromNode=FloatToInt_1
  fromPin=Int
  fromPinGuid=#15
  toNode=MergeConvexHulls_0_copy
  toPin=TargetNumSamples
  toPinGuid=#16
 }
 {
  fromNode=Maximum Divisor
  fromPin=Float
  fromPinGuid=#17
  toNode=SafeDivide_Max
  toPin=FloatB
  toPinGuid=#18
 }
 {
  fromNode=CollectionTransformSelectRoot_1
  fromPin=TransformSelection
  fromPinGuid=#19
  toNode=SetAnchorState
  toPin=TransformSelection
  toPinGuid=#20
 }
 {
  fromNode=CollectionTransformSelectRoot_1
  fromPin=Collection
  fromPinGuid=#21
  toNode=SetAnchorState
  toPin=Collection
  toPinGuid=#22
 }
 {
  fromNode=UpdateVolumeAttributes_2
  fromPin=Collection
  fromPinGuid=#23
  toNode=CollectionTransformSelectRoot_11
  toPin=Collection
  toPinGuid=#24
 }
 {
  fromNode=Ceil_0
  fromPin=ReturnValue
  fromPinGuid=#25
  toNode=FloatToInt_10
  toPin=Float
  toPinGuid=#26
 }
 {
  fromNode=Min3
  fromPin=ReturnValue
  fromPinGuid=#27
  toNode=Divide by 2
  toPin=FloatA
  toPinGuid=#28
 }
 {
  fromNode=SimplifyConvexHulls QSlim
  fromPin=Collection
  fromPinGuid=#29
  toNode=BranchCollection_5
  toPin=FalseCollection
  toPinGuid=#30
 }
 {
  fromNode=Convert Mesh to OBJ String
  fromPin=StringOBJ
  fromPinGuid=#31
  toNode=Write String to File
  toPin=FileContents
  toPinGuid=#32
 }
 {
  fromNode=CollectionTransformSelectRoot_2
  fromPin=TransformSelection
  fromPinGuid=#33
  toNode=CollectionTransformSelectInvert
  toPin=TransformSelection
  toPinGuid=#34
 }
 {
  fromNode=GetNumElementsInCollectionGroup
  fromPin=NumElements
  fromPinGuid=#35
  toNode=IntToFloat
  toPin=Int
  toPinGuid=#36
 }
 {
  fromNode=IntToFloat_3
  fromPin=Float
  fromPinGuid=#37
  toNode=Subtract
  toPin=FloatA
  toPinGuid=#38
 }
 {
  fromNode="Level 1 - by  Cluster Number"
  fromPin=IsOverriden
  fromPinGuid=#39
  toNode=BranchCollection_2
  toPin=bCondition
  toPinGuid=#40
 }
 {
  fromNode=AutoCluster
  fromPin=Collection
  fromPinGuid=#41
  toNode=BranchCollection_0
  toPin=FalseCollection
  toPinGuid=#42
 }
 {
  fromNode=Root Convexes by Error Tolerance
  fromPin=Float
  fromPinGuid=#43
  toNode=FloatToDouble_2
  toPin=Float
  toPinGuid=#44
 }
 {
  fromNode=IntToFloat
  fromPin=Float
  fromPinGuid=#45
  toNode=SafeDivide_Min
  toPin=FloatA
  toPinGuid=#46
 }
 {
  fromNode=CollectionTransformSelectBySize_Small
  fromPin=TransformSelection
  fromPinGuid=#47
  toNode=CreateLeafConvexHulls_Small
  toPin=OptionalSelectionFilter
  toPinGuid=#48
 }
 {
  fromNode=Divide_13
  fromPin=ReturnValue
  fromPinGuid=#49
  toNode=Ceil_1
  toPin=Float
  toPinGuid=#50
 }
 {
  fromNode=SetAnchorState
  fromPin=Collection
  fromPinGuid=#51
  toNode=CollectionTransformSelectRoot_2
  toPin=Collection
  toPinGuid=#52
 }
 {
  fromNode=Multiply
  fromPin=ReturnValue
  fromPinGuid=#53
  toNode=SafeDivide_1
  toPin=FloatA
  toPinGuid=#54
 }
 {
  fromNode=GenerateClusterConvexHullsFromLeafHulls_4
  fromPin=Collection
  fromPinGuid=#55
  toNode=SimplifyConvexHulls
  toPin=Collection
  toPinGuid=#56
 }
 {
  fromNode=FloatToString_0
  fromPin=String
  fromPinGuid=#57
  toNode=PrintString_0
  toPin=String
  toPinGuid=#58
 }
 {
  fromNode=ExpandVector_0
  fromPin=Z
  fromPinGuid=#59
  toNode=Multiply_6
  toPin=FloatA
  toPinGuid=#60
 }
 {
  fromNode=Max Clusters
  fromPin=Float
  fromPinGuid=#61
  toNode=Clamp_2
  toPin=Max
  toPinGuid=#62
 }
 {
  fromNode=Add
  fromPin=ReturnValue
  fromPinGuid=#63
  toNode=Multiply
  toPin=FloatA
  toPinGuid=#64
 }
 {
  fromNode=Minimum Divisor
  fromPin=Float
  fromPinGuid=#65
  toNode=SafeDivide_Min
  toPin=FloatB
  toPinGuid=#66
 }
 {
  fromNode=SafeDivide_Max
  fromPin=ReturnValue
  fromPinGuid=#67
  toNode=Add
  toPin=FloatA
  toPinGuid=#68
 }
 {
  fromNode=IntToFloat
  fromPin=Float
  fromPinGuid=#45
  toNode=SafeDivide_1
  toPin=FloatB
  toPinGuid=#69
 }
 {
  fromNode=IntToFloat
  fromPin=Float
  fromPinGuid=#45
  toNode=SafeDivide_Max
  toPin=FloatA
  toPinGuid=#70
 }
 {
  fromNode=Multiply_9
  fromPin=ReturnValue
  fromPinGuid=#13
  toNode=FloatToString_23
  toPin=Float
  toPinGuid=#71
 }
 {
  fromNode=SafeDivide_Min
  fromPin=ReturnValue
  fromPinGuid=#72
  toNode=Add
  toPin=FloatB
  toPinGuid=#73
 }
 {
  fromNode=Divide_2
  fromPin=ReturnValue
  fromPinGuid=#74
  toNode=Cubed
  toPin=Base
  toPinGuid=#75
 }
 {
  fromNode=Floor_3
  fromPin=ReturnValue
  fromPinGuid=#76
  toNode=One Minus_copy
  toPin=FloatB
  toPinGuid=#77
 }
 {
  fromNode=Clamp_2
  fromPin=ReturnValue
  fromPinGuid=#78
  toNode=FloatToInt
  toPin=Float
  toPinGuid=#79
 }
 {
  fromNode=Pow_2
  fromPin=ReturnValue
  fromPinGuid=#80
  toNode=Divide_14
  toPin=FloatB
  toPinGuid=#81
 }
 {
  fromNode=SafeDivide_1
  fromPin=ReturnValue
  fromPinGuid=#82
  toNode=Clamp_0
  toPin=Float
  toPinGuid=#83
 }
 {
  fromNode=BranchCollection_8
  fromPin=ChosenCollection
  fromPinGuid=#84
  toNode=CollectionTransformSelectRoot_4
  toPin=Collection
  toPinGuid=#85
 }
 {
  fromNode=Floor_0
  fromPin=ReturnValue
  fromPinGuid=#86
  toNode=FloatToString_copy
  toPin=Float
  toPinGuid=#87
 }
 {
  fromNode=x100
  fromPin=ReturnValue
  fromPinGuid=#88
  toNode=FloatToString_20
  toPin=Float
  toPinGuid=#89
 }
 {
  fromNode=One Minus
  fromPin=ReturnValue
  fromPinGuid=#90
  toNode=Add_0
  toPin=FloatB
  toPinGuid=#91
 }
 {
  fromNode=BranchCollection_9
  fromPin=ChosenCollection
  fromPinGuid=#92
  toNode=ClusterIsolatedRoots
  toPin=Collection
  toPinGuid=#93
 }
 {
  fromNode=Adaptive Tolerance
  fromPin=String
  fromPinGuid=#94
  toNode=Log - Adaptive Tolerance
  toPin=String
  toPinGuid=#95
 }
 {
  fromNode=FloatToInt_10
  fromPin=Int
  fromPinGuid=#96
  toNode=AutoCluster_1
  toPin=ClusterGridDepth
  toPinGuid=#97
 }
 {
  fromNode=Clamp_0
  fromPin=ReturnValue
  fromPinGuid=#98
  toNode=AutoCluster_0
  toPin=ClusterFraction
  toPinGuid=#99
 }
 {
  fromNode=CollectionTransformSelectInvert
  fromPin=TransformSelection
  fromPinGuid=#100
  toNode=RemoveOnBreak_0
  toPin=TransformSelection
  toPinGuid=#101
 }
 {
  fromNode=GetGeometryCollectionSources
  fromPin=Sources
  fromPinGuid=#102
  toNode=CreateGeometryCollectionFromSources
  toPin=Sources
  toPinGuid=#103
 }
 {
  fromNode=FloatToInt_7
  fromPin=Int
  fromPinGuid=#104
  toNode=IntToBool_1
  toPin=Int
  toPinGuid=#105
 }
 {
  fromNode=Divide_7
  fromPin=ReturnValue
  fromPinGuid=#106
  toNode=Clamp_7
  toPin=Float
  toPinGuid=#107
 }
 {
  fromNode=CollectionTransformSelectionFromIndexArray
  fromPin=Collection
  fromPinGuid=#108
  toNode=CreateLeafConvexHulls_Tiny_copy
  toPin=Collection
  toPinGuid=#109
 }
 {
  fromNode=Ceil
  fromPin=ReturnValue
  fromPinGuid=#110
  toNode=Min3
  toPin=FloatA
  toPinGuid=#111
 }
 {
  fromNode=MakeLiteralVector
  fromPin=Vector
  fromPinGuid=#112
  toNode=VectorToString
  toPin=Vector
  toPinGuid=#113
 }
 {
  fromNode=CollectionTransformSelectionFromIndexArray
  fromPin=TransformSelection
  fromPinGuid=#114
  toNode=SimplifyConvexHulls_Leaf_QSlim_Tiny_copy
  toPin=OptionalSelectionFilter
  toPinGuid=#115
 }
 {
  fromNode=FloatToInt_0
  fromPin=Int
  fromPinGuid=#116
  toNode=IntToBool_3
  toPin=Int
  toPinGuid=#117
 }
 {
  fromNode=GetNumElementsInCollectionGroup_1
  fromPin=NumElements
  fromPinGuid=#118
  toNode=IntToFloat_3
  toPin=Int
  toPinGuid=#119
 }
 {
  fromNode=REROUTE - Box Res
  fromPin=ReturnValue
  fromPinGuid=#120
  toNode=Divide_11
  toPin=FloatB
  toPinGuid=#121
 }
 {
  fromNode=OneMinus_0
  fromPin=ReturnValue
  fromPinGuid=#122
  toNode=FloatToInt_4_copy
  toPin=Float
  toPinGuid=#123
 }
 {
  fromNode=Cleanup_L1_bySize
  fromPin=IsOverriden
  fromPinGuid=#124
  toNode=Branch Use CleanUp
  toPin=bCondition
  toPinGuid=#125
 }
 {
  fromNode=ExpandVector_0
  fromPin=Y
  fromPinGuid=#126
  toNode=Multiply_5
  toPin=FloatA
  toPinGuid=#127
 }
 {
  fromNode=MergeConvexHulls_copy
  fromPin=Collection
  fromPinGuid=#128
  toNode=SimplifyConvexHulls_Leaf_QSlim
  toPin=Collection
  toPinGuid=#129
 }
 {
  fromNode=AutoCluster_0
  fromPin=Collection
  fromPinGuid=#130
  toNode=BranchCollection
  toPin=FalseCollection
  toPinGuid=#131
 }
 {
  fromNode=IntToBool_3_copy_0
  fromPin=Bool
  fromPinGuid=#132
  toNode=BoolToString_5_copy_0
  toPin=Bool
  toPinGuid=#133
 }
 {
  fromNode=FloatToString_28_copy
  fromPin=String
  fromPinGuid=#134
  toNode=StringAppend_20_copy_0_0
  toPin=String2
  toPinGuid=#135
 }
 {
  fromNode=FloatToString_copy
  fromPin=String
  fromPinGuid=#136
  toNode=LogString_5_copy
  toPin=String
  toPinGuid=#137
 }
 {
  fromNode=CollectionTransformSelectRoot
  fromPin=Collection
  fromPinGuid=#138
  toNode=AutoCluster Volume Numeric
  toPin=Collection
  toPinGuid=#139
 }
 {
  fromNode=Percentage Differential Tolerance
  fromPin=Float
  fromPinGuid=#140
  toNode=Differential Driver %
  toPin=FloatB
  toPinGuid=#141
 }
 {
  fromNode=FloatToString_2
  fromPin=String
  fromPinGuid=#142
  toNode=LogString_4
  toPin=String
  toPinGuid=#143
 }
 {
  fromNode=GetConvexHullVolume
  fromPin=Volume
  fromPinGuid=#144
  toNode=Pow_0
  toPin=Base
  toPinGuid=#145
 }
 {
  fromNode=CreateLeafConvexHulls_Large
  fromPin=Collection
  fromPinGuid=#146
  toNode=MergeConvexHulls_copy_0
  toPin=Collection
  toPinGuid=#147
 }
 {
  fromNode=Proximity - Contact Threshold
  fromPin=Float
  fromPinGuid=#148
  toNode=Convex - Sharp Contact
  toPin=ContactThreshold
  toPinGuid=#149
 }
 {
  fromNode=Root - Convexes by Number
  fromPin=IsOverriden
  fromPinGuid=#150
  toNode=BoolToInt
  toPin=Bool
  toPinGuid=#151
 }
 {
  fromNode=CompareFloat
  fromPin=Result
  fromPinGuid=#152
  toNode=BoolToString_0
  toPin=Bool
  toPinGuid=#153
 }
 {
  fromNode="Level 1 - by  Cluster Number"
  fromPin=Int
  fromPinGuid=#7
  toNode=IntToString_3
  toPin=Int
  toPinGuid=#154
 }
 {
  fromNode=SimplifyConvexHulls_Angle_Tiny_copy
  fromPin=Collection
  fromPinGuid=#155
  toNode=SimplifyConvexHulls_Angle_Tiny_copy_0
  toPin=Collection
  toPinGuid=#156
 }
 {
  fromNode=CompareFloat
  fromPin=Result
  fromPinGuid=#152
  toNode=BranchCollection_1
  toPin=bCondition
  toPinGuid=#157
 }
 {
  fromNode=CollectionTransformSelectChildren_1
  fromPin=Collection
  fromPinGuid=#158
  toNode=ClusterConvexes - Manual
  toPin=Collection
  toPinGuid=#159
 }
 {
  fromNode=Multiply_8
  fromPin=ReturnValue
  fromPinGuid=#160
  toNode=Add_5
  toPin=FloatB
  toPinGuid=#161
 }
 {
  fromNode=Tiny Leaf Size
  fromPin=Float
  fromPinGuid=#162
  toNode=CollectionTransformSelectBySize_Small
  toPin=SizeMin
  toPinGuid=#163
 }
 {
  fromNode=Ceil_0
  fromPin=ReturnValue
  fromPinGuid=#25
  toNode=MakeLiteralVector
  toPin=Y
  toPinGuid=#164
 }
 {
  fromNode=CollectionTransformSelectRoot_10
  fromPin=TransformSelection
  fromPinGuid=#165
  toNode=AutoCluster_2
  toPin=TransformSelection
  toPinGuid=#166
 }
 {
  fromNode=Ceil
  fromPin=ReturnValue
  fromPinGuid=#110
  toNode=MakeLiteralVector
  toPin=X
  toPinGuid=#167
 }
 {
  fromNode=Root Convexes by Error Tolerance
  fromPin=IsOverriden
  fromPinGuid=#168
  toNode=BoolToInt_0
  toPin=Bool
  toPinGuid=#169
 }
 {
  fromNode=CollectionTransformSelectRoot
  fromPin=Collection
  fromPinGuid=#138
  toNode=AutoCluster - L1 User Defined
  toPin=Collection
  toPinGuid=#170
 }
 {
  fromNode=Proximity - Contact Threshold
  fromPin=Float
  fromPinGuid=#148
  toNode=Proximity_5_copy_1
  toPin=ContactThreshold
  toPinGuid=#171
 }
 {
  fromNode=Proximity - Precise Contact
  fromPin=Collection
  fromPinGuid=#172
  toNode=BranchCollection_DerivedProximityMethod
  toPin=FalseCollection
  toPinGuid=#173
 }
 {
  fromNode=One Minus_copy
  fromPin=ReturnValue
  fromPinGuid=#174
  toNode=Add_0
  toPin=FloatA
  toPinGuid=#175
 }
 {
  fromNode=AutoCluster_Explicit
  fromPin=Collection
  fromPinGuid=#176
  toNode=BranchCollection_0
  toPin=TrueCollection
  toPinGuid=#177
 }
 {
  fromNode=CollectionTransformSelectRoot_4_copy
  fromPin=Collection
  fromPinGuid=#178
  toNode=CollectionTransformSelectChildren_1_copy
  toPin=Collection
  toPinGuid=#179
 }
 {
  fromNode=FloatToInt
  fromPin=Int
  fromPinGuid=#180
  toNode=AutoCluster Volume Numeric
  toPin=ClusterSites
  toPinGuid=#181
 }
 {
  fromNode=Clamp_9
  fromPin=ReturnValue
  fromPinGuid=#182
  toNode=Floor_1
  toPin=Float
  toPinGuid=#183
 }
 {
  fromNode=DEBUG -- Test L1s
  fromPin=Bool
  fromPinGuid=#184
  toNode=BranchCollection_11
  toPin=bCondition
  toPinGuid=#185
 }
 {
  fromNode=Cubed
  fromPin=ReturnValue
  fromPinGuid=#186
  toNode=Add_5
  toPin=FloatA
  toPinGuid=#187
 }
 {
  fromNode=Floor_0
  fromPin=ReturnValue
  fromPinGuid=#86
  toNode=Add_2
  toPin=FloatB
  toPinGuid=#188
 }
 {
  fromNode=FloatToString_3
  fromPin=String
  fromPinGuid=#189
  toNode=LogString_11
  toPin=String
  toPinGuid=#190
 }
 {
  fromNode=2x single plate
  fromPin=Float
  fromPinGuid=#191
  toNode=AutoCluster_2
  toPin=MinimumSize
  toPinGuid=#192
 }
 {
  fromNode=BranchCollection_6
  fromPin=ChosenCollection
  fromPinGuid=#193
  toNode=BranchCollection_2
  toPin=FalseCollection
  toPinGuid=#194
 }
 {
  fromNode=Add_5
  fromPin=ReturnValue
  fromPinGuid=#195
  toNode=Divide_10
  toPin=FloatA
  toPinGuid=#196
 }
 {
  fromNode=Clamp_0
  fromPin=ReturnValue
  fromPinGuid=#98
  toNode=FloatToString_0
  toPin=Float
  toPinGuid=#197
 }
 {
  fromNode=CollectionTransformSelectRoot_7
  fromPin=Collection
  fromPinGuid=#198
  toNode=CollectionTransformSelectChildren_2
  toPin=Collection
  toPinGuid=#199
 }
 {
  fromNode=CollectionTransformSelectChildren_0
  fromPin=TransformSelection
  fromPinGuid=#200
  toNode=GenerateClusterConvexHullsFromLeafHulls_0
  toPin=OptionalSelectionFilter
  toPinGuid=#201
 }
 {
  fromNode=ExpandBoundingBox_1
  fromPin=Volume
  fromPinGuid=#202
  toNode=Pow_2
  toPin=Base
  toPinGuid=#203
 }
 {
  fromNode=CollectionTransformSelectRoot_12
  fromPin=Collection
  fromPinGuid=#204
  toNode=GetConvexHullVolume_0
  toPin=Collection
  toPinGuid=#205
 }
 {
  fromNode=IntToString
  fromPin=String
  fromPinGuid=#206
  toNode=Incoming Cells
  toPin=String
  toPinGuid=#207
 }
 {
  fromNode=Add_2
  fromPin=ReturnValue
  fromPinGuid=#208
  toNode=FloatToInt_4
  toPin=Float
  toPinGuid=#209
 }
 {
  fromNode=Clamp_0
  fromPin=ReturnValue
  fromPinGuid=#98
  toNode=AutoCluster
  toPin=ClusterFraction
  toPinGuid=#210
 }
 {
  fromNode=CollectionTransformSelectRoot_4_copy
  fromPin=TransformSelection
  fromPinGuid=#211
  toNode=CollectionTransformSelectChildren_1_copy
  toPin=TransformSelection
  toPinGuid=#212
 }
 {
  fromNode=CollectionTransformSelectBySize_NarrowWindow
  fromPin=Collection
  fromPinGuid=#213
  toNode=CreateLeafConvexHulls_Mid_copy
  toPin=Collection
  toPinGuid=#214
 }
 {
  fromNode=BranchCollection_1
  fromPin=ChosenCollection
  fromPinGuid=#215
  toNode=CollectionTransformSelectRoot_4_copy
  toPin=Collection
  toPinGuid=#216
 }
 {
  fromNode=GetBoxLengths
  fromPin=Lengths
  fromPinGuid=#217
  toNode=SelectFloatArrayIndicesInRange
  toPin=Values
  toPinGuid=#218
 }
 {
  fromNode=Multiply_0
  fromPin=ReturnValue
  fromPinGuid=#219
  toNode=FloatToString_2_copy
  toPin=Float
  toPinGuid=#220
 }
 {
  fromNode=BoolToString_5
  fromPin=String
  fromPinGuid=#221
  toNode=Convex - Shape Protection Root
  toPin=Default
  toPinGuid=#222
 }
 {
  fromNode=Multiply_8
  fromPin=ReturnValue
  fromPinGuid=#160
  toNode=FloatToString_25
  toPin=Float
  toPinGuid=#223
 }
 {
  fromNode=Clamp
  fromPin=ReturnValue
  fromPinGuid=#224
  toNode=Floor_2
  toPin=Float
  toPinGuid=#225
 }
 {
  fromNode=Divide_13
  fromPin=ReturnValue
  fromPinGuid=#49
  toNode=Multiply_8
  toPin=FloatB
  toPinGuid=#226
 }
 {
  fromNode=CollectionTransformSelectChildren
  fromPin=Collection
  fromPinGuid=#227
  toNode=CollectionTransformSelectChildren_0
  toPin=Collection
  toPinGuid=#228
 }
 {
  fromNode=CollectionTransformSelectRoot_3
  fromPin=TransformSelection
  fromPinGuid=#229
  toNode=CollectionTransformSelectChildren
  toPin=TransformSelection
  toPinGuid=#230
 }
 {
  fromNode=CollectionTransformSelectRoot_3
  fromPin=Collection
  fromPinGuid=#231
  toNode=CollectionTransformSelectChildren
  toPin=Collection
  toPinGuid=#232
 }
 {
  fromNode=GenerateClusterConvexHullsFromLeafHulls_1
  fromPin=Collection
  fromPinGuid=#233
  toNode=BranchCollection_1
  toPin=TrueCollection
  toPinGuid=#234
 }
 {
  fromNode=Clamp_6
  fromPin=ReturnValue
  fromPinGuid=#235
  toNode=Floor_0
  toPin=Float
  toPinGuid=#236
 }
 {
  fromNode=CollectionTransformSelectRoot_9
  fromPin=Collection
  fromPinGuid=#237
  toNode=GenerateClusterConvexHullsFromLeafHulls_1
  toPin=Collection
  toPinGuid=#238
 }
 {
  fromNode=CollectionTransformSelectChildren_0
  fromPin=Collection
  fromPinGuid=#239
  toNode=GenerateClusterConvexHullsFromLeafHulls_0
  toPin=Collection
  toPinGuid=#240
 }
 {
  fromNode=CollectionTransformSelectRoot_4
  fromPin=Collection
  fromPinGuid=#241
  toNode=CollectionTransformSelectChildren_1
  toPin=Collection
  toPinGuid=#242
 }
 {
  fromNode=CollectionTransformSelectRoot_4
  fromPin=TransformSelection
  fromPinGuid=#243
  toNode=CollectionTransformSelectChildren_1
  toPin=TransformSelection
  toPinGuid=#244
 }
 {
  fromNode=Proximity_1
  fromPin=Collection
  fromPinGuid=#245
  toNode=BranchCollection
  toPin=TrueCollection
  toPinGuid=#246
 }
 {
  fromNode=SelectFloatArrayIndicesInRange
  fromPin=Indices
  fromPinGuid=#247
  toNode=CollectionTransformSelectionFromIndexArray
  toPin=BoneIndices
  toPinGuid=#248
 }
 {
  fromNode=Num Root Shapes
  fromPin=Int
  fromPinGuid=#249
  toNode=GenerateClusterConvexHullsFromLeafHulls
  toPin=ConvexCount
  toPinGuid=#250
 }
 {
  fromNode=CollectionTransformSelectChildren_2
  fromPin=TransformSelection
  fromPinGuid=#251
  toNode=AutoCluster_0
  toPin=TransformSelection
  toPinGuid=#252
 }
 {
  fromNode=CollectionTransformSelectChildren_2
  fromPin=Collection
  fromPinGuid=#253
  toNode=AutoCluster_0
  toPin=Collection
  toPinGuid=#254
 }
 {
  fromNode=CollectionTransformSelectRoot_14
  fromPin=TransformSelection
  fromPinGuid=#255
  toNode=MergeConvexHulls_0
  toPin=OptionalSelectionFilter
  toPinGuid=#256
 }
 {
  fromNode=FloatToDouble_2
  fromPin=Double
  fromPinGuid=#257
  toNode=GenerateClusterConvexHullsFromLeafHulls_1
  toPin=ErrorTolerance
  toPinGuid=#258
 }
 {
  fromNode="Level 1 - by  Cluster Number"
  fromPin=Int
  fromPinGuid=#7
  toNode=AutoCluster_Explicit
  toPin=ClusterSites
  toPinGuid=#259
 }
 {
  fromNode=ClusterConvexes - Manual
  fromPin=Collection
  fromPinGuid=#260
  toNode=MergeConvexHulls_0_copy_0
  toPin=Collection
  toPinGuid=#261
 }
 {
  fromNode=Divide_12
  fromPin=ReturnValue
  fromPinGuid=#262
  toNode=Ceil_0
  toPin=Float
  toPinGuid=#263
 }
 {
  fromNode=UpdateVolumeAttributes
  fromPin=Collection
  fromPinGuid=#264
  toNode=CollectionTransformSelectRoot_10
  toPin=Collection
  toPinGuid=#265
 }
 {
  fromNode="Level 1 - by  Cluster Number"
  fromPin=IsOverriden
  fromPinGuid=#39
  toNode=BranchCollection_0
  toPin=bCondition
  toPinGuid=#266
 }
 {
  fromNode=IntToFloat_2
  fromPin=Float
  fromPinGuid=#267
  toNode=Add_4
  toPin=FloatB
  toPinGuid=#268
 }
 {
  fromNode=CollectionTransformSelectRoot
  fromPin=TransformSelection
  fromPinGuid=#9
  toNode=AutoCluster Volume Numeric
  toPin=TransformSelection
  toPinGuid=#269
 }
 {
  fromNode=Threshold
  fromPin=Float
  fromPinGuid=#270
  toNode=Divide_3_copy
  toPin=FloatB
  toPinGuid=#271
 }
 {
  fromNode=Proximity - Distance Threshold
  fromPin=Float
  fromPinGuid=#272
  toNode=Convex - Sharp Contact
  toPin=DistanceThreshold
  toPinGuid=#273
 }
 {
  fromNode=FloatToInt_11
  fromPin=Int
  fromPinGuid=#274
  toNode=AutoCluster_1
  toPin=ClusterGridHeight
  toPinGuid=#275
 }
 {
  fromNode=CollectionTransformSelectRoot_11
  fromPin=TransformSelection
  fromPinGuid=#276
  toNode=GenerateClusterConvexHullsFromLeafHulls_4
  toPin=OptionalSelectionFilter
  toPinGuid=#277
 }
 {
  fromNode=Multiply_6
  fromPin=ReturnValue
  fromPinGuid=#278
  toNode=Divide_13
  toPin=FloatA
  toPinGuid=#279
 }
 {
  fromNode=BoolToInt_0
  fromPin=Int
  fromPinGuid=#280
  toNode=IntToFloat_2
  toPin=Int
  toPinGuid=#281
 }
 {
  fromNode=CreateGeometryCollectionFromSources
  fromPin=Collection
  fromPinGuid=#282
  toNode=CollectionTransformSelectBySize_Small_copy_0
  toPin=Collection
  toPinGuid=#283
 }
 {
  fromNode=CollectionTransformSelectRoot_10
  fromPin=Collection
  fromPinGuid=#284
  toNode=AutoCluster_2
  toPin=Collection
  toPinGuid=#285
 }
 {
  fromNode=MakeLiteralInt
  fromPin=Int
  fromPinGuid=#286
  toNode=GenerateClusterConvexHullsFromLeafHulls_4
  toPin=ConvexCount
  toPinGuid=#287
 }
 {
  fromNode=Add_4
  fromPin=ReturnValue
  fromPinGuid=#288
  toNode=CompareFloat
  toPin=FloatA
  toPinGuid=#289
 }
 {
  fromNode=PURGE L2s
  fromPin=Bool
  fromPinGuid=#290
  toNode=BranchCollection_9
  toPin=bCondition
  toPinGuid=#291
 }
 {
  fromNode=FloatToInt
  fromPin=Int
  fromPinGuid=#180
  toNode=IntToString
  toPin=Int
  toPinGuid=#292
 }
 {
  fromNode=CollectionTransformSelectRoot_9
  fromPin=TransformSelection
  fromPinGuid=#293
  toNode=GenerateClusterConvexHullsFromLeafHulls_1
  toPin=OptionalSelectionFilter
  toPinGuid=#294
 }
 {
  fromNode=BranchCollection_2
  fromPin=ChosenCollection
  fromPinGuid=#295
  toNode=Proximity_1
  toPin=Collection
  toPinGuid=#296
 }
 {
  fromNode=1 Over 3 - Convex test
  fromPin=ReturnValue
  fromPinGuid=#297
  toNode=Pow_1
  toPin=Exp
  toPinGuid=#298
 }
 {
  fromNode=CollectionTransformSelectRoot_11
  fromPin=Collection
  fromPinGuid=#299
  toNode=GenerateClusterConvexHullsFromLeafHulls_4
  toPin=Collection
  toPinGuid=#300
 }
 {
  fromNode=CollectionTransformSelectRoot_11
  fromPin=TransformSelection
  fromPinGuid=#276
  toNode=GetConvexHullVolume
  toPin=TransformSelection
  toPinGuid=#301
 }
 {
  fromNode=AutoCluster Volume Numeric
  fromPin=Collection
  fromPinGuid=#302
  toNode=BranchCollection_6
  toPin=TrueCollection
  toPinGuid=#303
 }
 {
  fromNode=CollectionTransformSelectRoot_11
  fromPin=Collection
  fromPinGuid=#299
  toNode=GetConvexHullVolume
  toPin=Collection
  toPinGuid=#304
 }
 {
  fromNode=CollectionTransformSelectRoot_12
  fromPin=TransformSelection
  fromPinGuid=#305
  toNode=GetConvexHullVolume_0
  toPin=TransformSelection
  toPinGuid=#306
 }
 {
  fromNode=Ceil_1
  fromPin=ReturnValue
  fromPinGuid=#307
  toNode=Min3
  toPin=FloatC
  toPinGuid=#308
 }
 {
  fromNode=Proximity_Precise
  fromPin=Collection
  fromPinGuid=#309
  toNode=BranchCollection_4
  toPin=FalseCollection
  toPinGuid=#310
 }
 {
  fromNode=FloatToString_22
  fromPin=String
  fromPinGuid=#311
  toNode=Sites - by Leaf Volumes
  toPin=String
  toPinGuid=#312
 }
 {
  fromNode=Differential Driver %
  fromPin=ReturnValue
  fromPinGuid=#313
  toNode=Clamp_9
  toPin=Float
  toPinGuid=#314
 }
 {
  fromNode=Floor_2
  fromPin=ReturnValue
  fromPinGuid=#315
  toNode=FloatToInt_8
  toPin=Float
  toPinGuid=#316
 }
 {
  fromNode=BranchCollection_UseManualProximity
  fromPin=ChosenCollection
  fromPinGuid=#317
  toNode=CollectionTransformSelectRoot
  toPin=Collection
  toPinGuid=#318
 }
 {
  fromNode=Floor_1
  fromPin=ReturnValue
  fromPinGuid=#319
  toNode=FloatToInt_7
  toPin=Float
  toPinGuid=#320
 }
 {
  fromNode=Multiply_1
  fromPin=ReturnValue
  fromPinGuid=#321
  toNode=BranchFloat_copy
  toPin=B
  toPinGuid=#322
 }
 {
  fromNode=SP - Base Samples
  fromPin=Float
  fromPinGuid=#323
  toNode=FloatToInt_1
  toPin=Float
  toPinGuid=#324
 }
 {
  fromNode=Ceil_0
  fromPin=ReturnValue
  fromPinGuid=#25
  toNode=Min3
  toPin=FloatB
  toPinGuid=#325
 }
 {
  fromNode=Cubed
  fromPin=ReturnValue
  fromPinGuid=#186
  toNode=FloatToString_22
  toPin=Float
  toPinGuid=#326
 }
 {
  fromNode=FloatToDouble_3
  fromPin=Double
  fromPinGuid=#327
  toNode=MergeConvexHulls_0
  toPin=NegativeSpaceTolerance
  toPinGuid=#328
 }
 {
  fromNode=Ceil
  fromPin=ReturnValue
  fromPinGuid=#110
  toNode=FloatToInt_9
  toPin=Float
  toPinGuid=#329
 }
 {
  fromNode=FloatToString_27
  fromPin=String
  fromPinGuid=#330
  toNode=LogString_16
  toPin=String
  toPinGuid=#331
 }
 {
  fromNode=FloatToString
  fromPin=String
  fromPinGuid=#332
  toNode=LogString_5
  toPin=String
  toPinGuid=#333
 }
 {
  fromNode=BranchCollection_7_copy
  fromPin=ChosenCollection
  fromPinGuid=#334
  toNode=CollectionTransformSelectRoot_6
  toPin=Collection
  toPinGuid=#335
 }
 {
  fromNode=x100
  fromPin=ReturnValue
  fromPinGuid=#88
  toNode=Divide_7
  toPin=FloatA
  toPinGuid=#336
 }
 {
  fromNode=1 Over 3 - Convex test
  fromPin=ReturnValue
  fromPinGuid=#297
  toNode=Pow_2
  toPin=Exp
  toPinGuid=#337
 }
 {
  fromNode=CollectionTransformSelectBySize
  fromPin=TransformSelection
  fromPinGuid=#338
  toNode=CreateLeafConvexHulls_Tiny
  toPin=OptionalSelectionFilter
  toPinGuid=#339
 }
 {
  fromNode=FloatToString_1
  fromPin=String
  fromPinGuid=#340
  toNode=Volume - by Root
  toPin=String
  toPinGuid=#341
 }
 {
  fromNode=CollectionTransformSelectBySize_Small
  fromPin=TransformSelection
  fromPinGuid=#47
  toNode=SimplifyConvexHulls_Leaf_QSlim_Small
  toPin=OptionalSelectionFilter
  toPinGuid=#342
 }
 {
  fromNode=MergeConvexHulls_2
  fromPin=Collection
  fromPinGuid=#343
  toNode=SimplifyConvexHulls_Leaf_QSlim_Tiny
  toPin=Collection
  toPinGuid=#344
 }
 {
  fromNode=GetGeometryCollectionAsset
  fromPin=Asset
  fromPinGuid=#345
  toNode=AtomDestructionFixISMMaterialsDataflowNode
  toPin=GCAsset
  toPinGuid=#346
 }
 {
  fromNode=AutoCluster_1
  fromPin=Collection
  fromPinGuid=#347
  toNode=BranchCollection_6
  toPin=FalseCollection
  toPinGuid=#348
 }
 {
  fromNode=SetAnchorState_0
  fromPin=Collection
  fromPinGuid=#349
  toNode=Terminal Early Out
  toPin=Collection
  toPinGuid=#350
 }
 {
  fromNode=Divide_10
  fromPin=ReturnValue
  fromPinGuid=#351
  toNode=Multiply_9
  toPin=FloatA
  toPinGuid=#352
 }
 {
  fromNode=CollectionTransformSelectRoot_5
  fromPin=Collection
  fromPinGuid=#353
  toNode=GenerateClusterConvexHullsFromLeafHulls_3
  toPin=Collection
  toPinGuid=#354
 }
 {
  fromNode=CollectionTransformSelectRoot_8
  fromPin=Collection
  fromPinGuid=#355
  toNode=GenerateClusterConvexHullsFromLeafHulls
  toPin=Collection
  toPinGuid=#356
 }
 {
  fromNode=Level 1 - Cluster Convexes
  fromPin=IsOverriden
  fromPinGuid=#357
  toNode=BranchCollection_1_copy
  toPin=bCondition
  toPinGuid=#358
 }
 {
  fromNode=Single Convex Tolerance %
  fromPin=Float
  fromPinGuid=#359
  toNode=Divide_7
  toPin=FloatB
  toPinGuid=#360
 }
 {
  fromNode="Protection Ratio (95%)"
  fromPin=Float
  fromPinGuid=#361
  toNode=Divide_0
  toPin=FloatB
  toPinGuid=#362
 }
 {
  fromNode=CollectionTransformSelectRoot_6
  fromPin=TransformSelection
  fromPinGuid=#363
  toNode=AutoCluster
  toPin=TransformSelection
  toPinGuid=#364
 }
 {
  fromNode=Clamp_5
  fromPin=ReturnValue
  fromPinGuid=#365
  toNode=Floor
  toPin=Float
  toPinGuid=#366
 }
 {
  fromNode=BoolToString_5
  fromPin=String
  fromPinGuid=#221
  toNode=Level 1 - Shape Protection
  toPin=Default
  toPinGuid=#367
 }
 {
  fromNode=CollectionTransformSelectRoot
  fromPin=TransformSelection
  fromPinGuid=#9
  toNode=AutoCluster_1
  toPin=TransformSelection
  toPinGuid=#368
 }
 {
  fromNode="Minimum Volume Threshold (Cubic Root Volume)"
  fromPin=Float
  fromPinGuid=#369
  toNode=Divide_4
  toPin=FloatB
  toPinGuid=#370
 }
 {
  fromNode=Divide_7
  fromPin=ReturnValue
  fromPinGuid=#106
  toNode=FloatToString_16
  toPin=Float
  toPinGuid=#371
 }
 {
  fromNode=Divide by 2
  fromPin=ReturnValue
  fromPinGuid=#372
  toNode=Clamp
  toPin=Float
  toPinGuid=#373
 }
 {
  fromNode=StringAppend_1
  fromPin=String
  fromPinGuid=#374
  toNode=StringAppend
  toPin=String2
  toPinGuid=#375
 }
 {
  fromNode="Minimum Threshold (##Leaf Nodes)"
  fromPin=Float
  fromPinGuid=#376
  toNode=Divide_3
  toPin=FloatB
  toPinGuid=#377
 }
 {
  fromNode="Level 1 - by  Cluster Number"
  fromPin=IsOverriden
  fromPinGuid=#39
  toNode=BoolToString_2
  toPin=Bool
  toPinGuid=#378
 }
 {
  fromNode=Divide_3
  fromPin=ReturnValue
  fromPinGuid=#379
  toNode=Clamp_5
  toPin=Float
  toPinGuid=#380
 }
 {
  fromNode=Floor
  fromPin=ReturnValue
  fromPinGuid=#381
  toNode=Add_2
  toPin=FloatA
  toPinGuid=#382
 }
 {
  fromNode=Floor_0
  fromPin=ReturnValue
  fromPinGuid=#86
  toNode=Multiply_3
  toPin=FloatB
  toPinGuid=#383
 }
 {
  fromNode=CollectionTransformSelectRoot
  fromPin=Collection
  fromPinGuid=#138
  toNode=AutoCluster_1
  toPin=Collection
  toPinGuid=#384
 }
 {
  fromNode=GetConvexHullVolume_0
  fromPin=Volume
  fromPinGuid=#385
  toNode=Pow_1
  toPin=Base
  toPinGuid=#386
 }
 {
  fromNode=StringAppend_20_copy_0_0
  fromPin=String
  fromPinGuid=#387
  toNode=StringAppend_20_copy_1_0
  toPin=String1
  toPinGuid=#388
 }
 {
  fromNode=FloatToInt_4
  fromPin=Int
  fromPinGuid=#389
  toNode=IntToBool
  toPin=Int
  toPinGuid=#390
 }
 {
  fromNode=CollectionTransformSelectRoot_6
  fromPin=TransformSelection
  fromPinGuid=#363
  toNode=AutoCluster_Explicit
  toPin=TransformSelection
  toPinGuid=#391
 }
 {
  fromNode=Divide_4
  fromPin=ReturnValue
  fromPinGuid=#392
  toNode=Clamp_6
  toPin=Float
  toPinGuid=#393
 }
 {
  fromNode=IntToString_3
  fromPin=String
  fromPinGuid=#394
  toNode=LogString_9
  toPin=String
  toPinGuid=#395
 }
 {
  fromNode=BranchCollection_0
  fromPin=ChosenCollection
  fromPinGuid=#396
  toNode=CollectionTransformSelectRoot_7
  toPin=Collection
  toPinGuid=#397
 }
 {
  fromNode=FloatToInt_9
  fromPin=Int
  fromPinGuid=#398
  toNode=AutoCluster_1
  toPin=ClusterGridWidth
  toPinGuid=#399
 }
 {
  fromNode=CollectionTransformSelectBySize_Small_copy
  fromPin=TransformSelection
  fromPinGuid=#400
  toNode=MergeConvexHulls_copy
  toPin=OptionalSelectionFilter
  toPinGuid=#401
 }
 {
  fromNode=Level 1 - Shape Protection
  fromPin=Bool
  fromPinGuid=#402
  toNode=ClusterConvexes - Auto
  toPin=bProtectNegativeSpace
  toPinGuid=#403
 }
 {
  fromNode=BoolToString_3
  fromPin=String
  fromPinGuid=#404
  toNode=StringAppend_6
  toPin=String2
  toPinGuid=#405
 }
 {
  fromNode=BranchCollection_4
  fromPin=ChosenCollection
  fromPinGuid=#406
  toNode=UpdateVolumeAttributes
  toPin=Collection
  toPinGuid=#407
 }
 {
  fromNode=BoolToInt
  fromPin=Int
  fromPinGuid=#408
  toNode=IntToFloat_1
  toPin=Int
  toPinGuid=#409
 }
 {
  fromNode=CreateGeometryCollectionFromSources
  fromPin=Materials
  fromPinGuid=#410
  toNode=Terminal Early Out
  toPin=Materials
  toPinGuid=#411
 }
 {
  fromNode=FloatToString_26
  fromPin=String
  fromPinGuid=#412
  toNode=GC Fill Ratio
  toPin=String
  toPinGuid=#413
 }
 {
  fromNode=Divide_14
  fromPin=ReturnValue
  fromPinGuid=#414
  toNode=FloatToString_26
  toPin=Float
  toPinGuid=#415
 }
 {
  fromNode=Multiply_2
  fromPin=ReturnValue
  fromPinGuid=#416
  toNode=Divide_11
  toPin=FloatA
  toPinGuid=#417
 }
 {
  fromNode=Level 1 - Cluster Convexes
  fromPin=IsOverriden
  fromPinGuid=#357
  toNode=BoolToInt_1
  toPin=Bool
  toPinGuid=#418
 }
 {
  fromNode=IntToBool
  fromPin=Bool
  fromPinGuid=#419
  toNode=BoolToString_3
  toPin=Bool
  toPinGuid=#420
 }
 {
  fromNode=Floor_2
  fromPin=ReturnValue
  fromPinGuid=#315
  toNode=FloatToString_27
  toPin=Float
  toPinGuid=#421
 }
 {
  fromNode="Level 1 - by Target Resolution "
  fromPin=Float
  fromPinGuid=#422
  toNode=REROUTE - Box Res
  toPin=FloatA
  toPinGuid=#423
 }
 {
  fromNode=Divide_10
  fromPin=ReturnValue
  fromPinGuid=#351
  toNode=FloatToString_24
  toPin=Float
  toPinGuid=#424
 }
 {
  fromNode=StringAppend_11
  fromPin=String
  fromPinGuid=#425
  toNode=StringAppend_12
  toPin=String1
  toPinGuid=#426
 }
 {
  fromNode=IntToBool_3
  fromPin=Bool
  fromPinGuid=#427
  toNode=BoolToString_5
  toPin=Bool
  toPinGuid=#428
 }
 {
  fromNode=Divide_14
  fromPin=ReturnValue
  fromPinGuid=#414
  toNode=Multiply_9
  toPin=FloatB
  toPinGuid=#429
 }
 {
  fromNode=StringAppend_19
  fromPin=String
  fromPinGuid=#430
  toNode=Path Report
  toPin=String
  toPinGuid=#431
 }
 {
  fromNode=x100
  fromPin=ReturnValue
  fromPinGuid=#88
  toNode=Differential Driver %
  toPin=FloatA
  toPinGuid=#432
 }
 {
  fromNode=REROUTE - Box Res
  fromPin=ReturnValue
  fromPinGuid=#120
  toNode=Divide_12
  toPin=FloatB
  toPinGuid=#433
 }
 {
  fromNode=ExpandVector_0
  fromPin=X
  fromPinGuid=#434
  toNode=Multiply_2
  toPin=FloatA
  toPinGuid=#435
 }
 {
  fromNode=BoundingBox_1
  fromPin=BoundingBox
  fromPinGuid=#436
  toNode=ExpandBoundingBox_1
  toPin=BoundingBox
  toPinGuid=#437
 }
 {
  fromNode=GenerateClusterConvexHullsFromLeafHulls_0
  fromPin=Collection
  fromPinGuid=#438
  toNode=BranchCollection_8
  toPin=FalseCollection
  toPinGuid=#439
 }
 {
  fromNode=IntToFloat_1
  fromPin=Float
  fromPinGuid=#440
  toNode=Add_4
  toPin=FloatA
  toPinGuid=#441
 }
 {
  fromNode=BoolToString_9_copy
  fromPin=String
  fromPinGuid=#442
  toNode=LogString_8_copy
  toPin=String
  toPinGuid=#443
 }
 {
  fromNode=Multiply_5
  fromPin=ReturnValue
  fromPinGuid=#444
  toNode=Divide_12
  toPin=FloatA
  toPinGuid=#445
 }
 {
  fromNode=FloatToString_25
  fromPin=String
  fromPinGuid=#446
  toNode=Sites - Volume Grid Cells
  toPin=String
  toPinGuid=#447
 }
 {
  fromNode=Multiply_7
  fromPin=ReturnValue
  fromPinGuid=#448
  toNode=Multiply_8
  toPin=FloatA
  toPinGuid=#449
 }
 {
  fromNode=Convex - Protection Radius
  fromPin=Float
  fromPinGuid=#450
  toNode="REROUTER "
  toPin=FloatA
  toPinGuid=#451
 }
 {
  fromNode="Level 1 - by Target Resolution "
  fromPin=Float
  fromPinGuid=#422
  toNode=Divide_2
  toPin=FloatB
  toPinGuid=#452
 }
 {
  fromNode=StringAppend_20_copy_1
  fromPin=String
  fromPinGuid=#453
  toNode=StringAppend_21
  toPin=String2
  toPinGuid=#454
 }
 {
  fromNode=CollectionTransformSelectRoot_5
  fromPin=TransformSelection
  fromPinGuid=#455
  toNode=GenerateClusterConvexHullsFromLeafHulls_3
  toPin=OptionalSelectionFilter
  toPinGuid=#456
 }
 {
  fromNode=REROUTE - Box Res
  fromPin=ReturnValue
  fromPinGuid=#120
  toNode=Divide_13
  toPin=FloatB
  toPinGuid=#457
 }
 {
  fromNode=ExpandBoundingBox_1
  fromPin=HalfExtents
  fromPinGuid=#458
  toNode=ExpandVector_0
  toPin=Vector
  toPinGuid=#459
 }
 {
  fromNode=Sphere Covering to Mesh
  fromPin=Mesh
  fromPinGuid=#460
  toNode=Convert Mesh to OBJ String
  toPin=Mesh
  toPinGuid=#461
 }
 {
  fromNode=CollectionTransformSelectChildren
  fromPin=TransformSelection
  fromPinGuid=#462
  toNode=CollectionTransformSelectChildren_0
  toPin=TransformSelection
  toPinGuid=#463
 }
 {
  fromNode=FloatToString_23
  fromPin=String
  fromPinGuid=#464
  toNode=Sites - Averaged and then Ratio Scaled
  toPin=String
  toPinGuid=#465
 }
 {
  fromNode=Root - Convexes by Number
  fromPin=Int
  fromPinGuid=#466
  toNode=GenerateClusterConvexHullsFromLeafHulls_1
  toPin=ConvexCount
  toPinGuid=#467
 }
 {
  fromNode=FloatToString_24
  fromPin=String
  fromPinGuid=#468
  toNode=Sites Averaged
  toPin=String
  toPinGuid=#469
 }
 {
  fromNode=Divide_11
  fromPin=ReturnValue
  fromPinGuid=#470
  toNode=Multiply_7
  toPin=FloatA
  toPinGuid=#471
 }
 {
  fromNode=AutoCluster - L1 User Defined
  fromPin=Collection
  fromPinGuid=#472
  toNode=BranchCollection_2
  toPin=TrueCollection
  toPinGuid=#473
 }
 {
  fromNode=CollectionTransformSelectRoot_7
  fromPin=TransformSelection
  fromPinGuid=#474
  toNode=CollectionTransformSelectChildren_2
  toPin=TransformSelection
  toPinGuid=#475
 }
 {
  fromNode=GenerateClusterConvexHullsFromLeafHulls
  fromPin=Collection
  fromPinGuid=#476
  toNode=Use Specific Root Shapes End
  toPin=TrueCollection
  toPinGuid=#477
 }
 {
  fromNode=CollectionTransformSelectRoot_8
  fromPin=TransformSelection
  fromPinGuid=#478
  toNode=GenerateClusterConvexHullsFromLeafHulls
  toPin=OptionalSelectionFilter
  toPinGuid=#479
 }
 {
  fromNode=Ceil_1
  fromPin=ReturnValue
  fromPinGuid=#307
  toNode=MakeLiteralVector
  toPin=Z
  toPinGuid=#480
 }
 {
  fromNode=Ceil_1
  fromPin=ReturnValue
  fromPinGuid=#307
  toNode=FloatToInt_11
  toPin=Float
  toPinGuid=#481
 }
 {
  fromNode=FloatToInt_5
  fromPin=Int
  fromPinGuid=#482
  toNode=IntToBool_0
  toPin=Int
  toPinGuid=#483
 }
 {
  fromNode=FloatToInt_5
  fromPin=Int
  fromPinGuid=#482
  toNode=IntToString_2
  toPin=Int
  toPinGuid=#484
 }
 {
  fromNode=Proximity - Distance Threshold
  fromPin=Float
  fromPinGuid=#272
  toNode=Proximity_5_copy_1
  toPin=DistanceThreshold
  toPinGuid=#485
 }
 {
  fromNode=FloatToDouble_1
  fromPin=Double
  fromPinGuid=#486
  toNode=MergeConvexHulls_0_copy_0
  toPin=NegativeSpaceTolerance
  toPinGuid=#487
 }
 {
  fromNode=Divide_11
  fromPin=ReturnValue
  fromPinGuid=#470
  toNode=Ceil
  toPin=Float
  toPinGuid=#488
 }
 {
  fromNode=Divide_12
  fromPin=ReturnValue
  fromPinGuid=#262
  toNode=Multiply_7
  toPin=FloatB
  toPinGuid=#489
 }
 {
  fromNode=StringAppend_14
  fromPin=String
  fromPinGuid=#490
  toNode=StringAppend_15
  toPin=String1
  toPinGuid=#491
 }
 {
  fromNode=FloatToString_16
  fromPin=String
  fromPinGuid=#492
  toNode=Log Divide Differential by Tolerance %
  toPin=String
  toPinGuid=#493
 }
 {
  fromNode=IntToString_2
  fromPin=String
  fromPinGuid=#494
  toNode=Log Divide Differential by Tolerance % Int
  toPin=String
  toPinGuid=#495
 }
 {
  fromNode=Clamp_7
  fromPin=ReturnValue
  fromPinGuid=#496
  toNode=FloatToInt_5
  toPin=Float
  toPinGuid=#497
 }
 {
  fromNode=IntToBool_2
  fromPin=Bool
  fromPinGuid=#498
  toNode=BranchCollection_6
  toPin=bCondition
  toPinGuid=#499
 }
 {
  fromNode=IntToBool_2
  fromPin=Bool
  fromPinGuid=#498
  toNode=BoolToString_4
  toPin=Bool
  toPinGuid=#500
 }
 {
  fromNode=StringAppend_2
  fromPin=String
  fromPinGuid=#501
  toNode=StringAppend_3
  toPin=String1
  toPinGuid=#502
 }
 {
  fromNode=AtomDestructionFixISMMaterialsDataflowNode
  fromPin=InstancedMeshes
  fromPinGuid=#503
  toNode=Terminal Early Out
  toPin=InstancedMeshes
  toPinGuid=#504
 }
 {
  fromNode=FloatToInt_2
  fromPin=Int
  fromPinGuid=#505
  toNode=GenerateClusterConvexHullsFromLeafHulls_3
  toPin=TargetNumSamples
  toPinGuid=#506
 }
 {
  fromNode=RemoveOnBreak_0_copy
  fromPin=Collection
  fromPinGuid=#507
  toNode=RemoveOnBreak_0
  toPin=Collection
  toPinGuid=#508
 }
 {
  fromNode=Level 1 - Cluster Convexes
  fromPin=Int
  fromPinGuid=#509
  toNode=ClusterConvexes - Manual
  toPin=ConvexCount
  toPinGuid=#510
 }
 {
  fromNode=FloatToDouble_1
  fromPin=Double
  fromPinGuid=#486
  toNode=MergeConvexHulls_0_copy
  toPin=NegativeSpaceTolerance
  toPinGuid=#511
 }
 {
  fromNode=CollectionTransformSelectBySize
  fromPin=TransformSelection
  fromPinGuid=#338
  toNode=SimplifyConvexHulls_QSlim_Tiny_2_copy
  toPin=OptionalSelectionFilter
  toPinGuid=#512
 }
 {
  fromNode=BranchFloat_copy
  fromPin=ReturnValue
  fromPinGuid=#513
  toNode=FloatToString_3
  toPin=Float
  toPinGuid=#514
 }
 {
  fromNode=CollectionTransformSelectChildren_1
  fromPin=Collection
  fromPinGuid=#158
  toNode=ClusterConvexes - Auto
  toPin=Collection
  toPinGuid=#515
 }
 {
  fromNode=CreateLeafConvexHulls_Tiny
  fromPin=Collection
  fromPinGuid=#516
  toNode=MergeConvexHulls_2
  toPin=Collection
  toPinGuid=#517
 }
 {
  fromNode=MakeConvexDecompositionSettings
  fromPin=DecompositionSettings
  fromPinGuid=#518
  toNode=CreateLeafConvexHulls
  toPin=ConvexDecompositionSettings
  toPinGuid=#519
 }
 {
  fromNode=FloatToDouble_5
  fromPin=Double
  fromPinGuid=#3
  toNode=MergeConvexHulls_0_copy_0
  toPin=MinRadius
  toPinGuid=#520
 }
 {
  fromNode=Floor_3
  fromPin=ReturnValue
  fromPinGuid=#76
  toNode=FloatToString_2
  toPin=Float
  toPinGuid=#521
 }
 {
  fromNode=Divide_0
  fromPin=ReturnValue
  fromPinGuid=#522
  toNode=Clamp_1
  toPin=Float
  toPinGuid=#523
 }
 {
  fromNode=Shape Protection Radius
  fromPin=String
  fromPinGuid=#524
  toNode=Convex - Protection Radius
  toPin=Default
  toPinGuid=#525
 }
 {
  fromNode=GenerateClusterConvexHullsFromLeafHulls_3
  fromPin=Collection
  fromPinGuid=#526
  toNode=CollectionTransformSelectRoot_14
  toPin=Collection
  toPinGuid=#527
 }
 {
  fromNode=FloatToInt_2
  fromPin=Int
  fromPinGuid=#505
  toNode=MergeConvexHulls_0
  toPin=TargetNumSamples
  toPinGuid=#528
 }
 {
  fromNode=CollectionTransformSelectionSetOperation
  fromPin=TransformSelection
  fromPinGuid=#529
  toNode=ClusterMergeToNeighbors_copy
  toPin=TransformSelection
  toPinGuid=#530
 }
 {
  fromNode=CollectionTransformSelectRoot_14
  fromPin=Collection
  fromPinGuid=#531
  toNode=MergeConvexHulls_0
  toPin=Collection
  toPinGuid=#532
 }
 {
  fromNode=Convert Mesh to OBJ String
  fromPin=StringOBJ
  fromPinGuid=#31
  toNode=LogString_3
  toPin=String
  toPinGuid=#533
 }
 {
  fromNode=StringAppend_4
  fromPin=String
  fromPinGuid=#534
  toNode=StringAppend_8
  toPin=String1
  toPinGuid=#535
 }
 {
  fromNode=Level 1 - Shape Protection
  fromPin=BoolDefault
  fromPinGuid=#536
  toNode=BoolToString_6
  toPin=Bool
  toPinGuid=#537
 }
 {
  fromNode=CollectionTransformSelectBySize_NarrowWindow_copy
  fromPin=Collection
  fromPinGuid=#538
  toNode=CreateLeafConvexHulls_Mid_copy_0
  toPin=Collection
  toPinGuid=#539
 }
 {
  fromNode=CollectionTransformSelectionFromIndexArray
  fromPin=TransformSelection
  fromPinGuid=#114
  toNode=SimplifyConvexHulls_Angle_Tiny_copy_0
  toPin=OptionalSelectionFilter
  toPinGuid=#540
 }
 {
  fromNode=Clamp_1
  fromPin=ReturnValue
  fromPinGuid=#541
  toNode=Floor_3
  toPin=Float
  toPinGuid=#542
 }
 {
  fromNode=Convex - Shape Protection Root
  fromPin=BoolDefault
  fromPinGuid=#543
  toNode=BoolToString_7
  toPin=Bool
  toPinGuid=#544
 }
 {
  fromNode=BoolToString_7
  fromPin=String
  fromPinGuid=#545
  toNode=Enable Protection? - Root
  toPin=String
  toPinGuid=#546
 }
 {
  fromNode=FloatToDouble_0
  fromPin=Double
  fromPinGuid=#547
  toNode=MergeConvexHulls_0_copy_0
  toPin=MinSampleSpacing
  toPinGuid=#548
 }
 {
  fromNode=Small Leaf Size
  fromPin=Float
  fromPinGuid=#549
  toNode=CollectionTransformSelectBySize_Small_copy
  toPin=SizeMin
  toPinGuid=#550
 }
 {
  fromNode=StringAppend_13
  fromPin=String
  fromPinGuid=#551
  toNode=StringAppend_19
  toPin=String1
  toPinGuid=#552
 }
 {
  fromNode=Convert Mesh to OBJ String_0
  fromPin=StringOBJ
  fromPinGuid=#553
  toNode=Write String to File_1
  toPin=FileContents
  toPinGuid=#554
 }
 {
  fromNode=ClusterConvexes - Auto
  fromPin=SphereCovering
  fromPinGuid=#555
  toNode=Sphere Covering to Mesh_0
  toPin=SphereCovering
  toPinGuid=#556
 }
 {
  fromNode=MergeConvexHulls_1
  fromPin=Collection
  fromPinGuid=#557
  toNode=SimplifyConvexHulls_Angle_Tiny_copy
  toPin=Collection
  toPinGuid=#558
 }
 {
  fromNode=Subtract_0
  fromPin=ReturnValue
  fromPinGuid=#559
  toNode=Divide_0_copy
  toPin=FloatB
  toPinGuid=#560
 }
 {
  fromNode=Sphere Covering to Mesh_0
  fromPin=Mesh
  fromPinGuid=#561
  toNode=Convert Mesh to OBJ String_0
  toPin=Mesh
  toPinGuid=#562
 }
 {
  fromNode=FloatToDouble_1
  fromPin=Double
  fromPinGuid=#486
  toNode=ClusterConvexes - Auto
  toPin=NegativeSpaceTolerance
  toPinGuid=#563
 }
 {
  fromNode=UpdateVolumeAttributes_0_copy
  fromPin=Collection
  fromPinGuid=#564
  toNode=BranchCollection_7_copy
  toPin=TrueCollection
  toPinGuid=#565
 }
 {
  fromNode=SP - Tolerance
  fromPin=Float
  fromPinGuid=#566
  toNode=Adaptive Tolerance
  toPin=Float
  toPinGuid=#567
 }
 {
  fromNode=UpdateVolumeAttributes_1
  fromPin=Collection
  fromPinGuid=#568
  toNode=Proximity
  toPin=Collection
  toPinGuid=#569
 }
 {
  fromNode=BoolToString_1
  fromPin=String
  fromPinGuid=#570
  toNode=LogString_1
  toPin=String
  toPinGuid=#571
 }
 {
  fromNode=BoolToString_1
  fromPin=String
  fromPinGuid=#570
  toNode="Used Light Proximity to start = "
  toPin=String2
  toPinGuid=#572
 }
 {
  fromNode=BoolToString_4
  fromPin=String
  fromPinGuid=#573
  toNode=LogString_0
  toPin=String
  toPinGuid=#574
 }
 {
  fromNode=BoolToString_4
  fromPin=String
  fromPinGuid=#573
  toNode=Was Grid Clustered
  toPin=String2
  toPinGuid=#575
 }
 {
  fromNode=StringAppend_0
  fromPin=String
  fromPinGuid=#576
  toNode=StringAppend
  toPin=String1
  toPinGuid=#577
 }
 {
  fromNode="Used Light Proximity to start = "
  fromPin=String
  fromPinGuid=#578
  toNode=StringAppend_0
  toPin=String1
  toPinGuid=#579
 }
 {
  fromNode=Was Grid Clustered
  fromPin=String
  fromPinGuid=#580
  toNode=StringAppend_1
  toPin=String1
  toPinGuid=#581
 }
 {
  fromNode=BoolToString_2
  fromPin=String
  fromPinGuid=#582
  toNode=LogString_15
  toPin=String
  toPinGuid=#583
 }
 {
  fromNode=BoolToString_2
  fromPin=String
  fromPinGuid=#582
  toNode=StringAppend_2
  toPin=String2
  toPinGuid=#584
 }
 {
  fromNode=StringAppend_3
  fromPin=String
  fromPinGuid=#585
  toNode=StringAppend_4
  toPin=String2
  toPinGuid=#586
 }
 {
  fromNode=StringAppend
  fromPin=String
  fromPinGuid=#587
  toNode=StringAppend_4
  toPin=String1
  toPinGuid=#588
 }
 {
  fromNode=StringAppend_7
  fromPin=String
  fromPinGuid=#589
  toNode=StringAppend_8
  toPin=String2
  toPinGuid=#590
 }
 {
  fromNode=StringAppend_9
  fromPin=String
  fromPinGuid=#591
  toNode=StringAppend_10
  toPin=String2
  toPinGuid=#592
 }
 {
  fromNode=StringAppend_8
  fromPin=String
  fromPinGuid=#593
  toNode=StringAppend_10
  toPin=String1
  toPinGuid=#594
 }
 {
  fromNode=StringAppend_6
  fromPin=String
  fromPinGuid=#595
  toNode=StringAppend_7
  toPin=String1
  toPinGuid=#596
 }
 {
  fromNode=BoolToString_6
  fromPin=String
  fromPinGuid=#597
  toNode=Enable Protection? - Level 1
  toPin=String
  toPinGuid=#598
 }
 {
  fromNode=BoolToString_6
  fromPin=String
  fromPinGuid=#597
  toNode=StringAppend_5
  toPin=String2
  toPinGuid=#599
 }
 {
  fromNode=StringAppend_5
  fromPin=String
  fromPinGuid=#600
  toNode=StringAppend_9
  toPin=String1
  toPinGuid=#601
 }
 {
  fromNode=BoolToString_3
  fromPin=String
  fromPinGuid=#404
  toNode=LogString_2
  toPin=String
  toPinGuid=#602
 }
 {
  fromNode=BoolToString
  fromPin=String
  fromPinGuid=#603
  toNode=LogString_7
  toPin=String
  toPinGuid=#604
 }
 {
  fromNode=BoolToString
  fromPin=String
  fromPinGuid=#603
  toNode=StringAppend_11
  toPin=String2
  toPinGuid=#605
 }
 {
  fromNode=StringAppend_12
  fromPin=String
  fromPinGuid=#606
  toNode=StringAppend_13
  toPin=String2
  toPinGuid=#607
 }
 {
  fromNode=StringAppend_10
  fromPin=String
  fromPinGuid=#608
  toNode=StringAppend_13
  toPin=String1
  toPinGuid=#609
 }
 {
  fromNode=IntToBool_0
  fromPin=Bool
  fromPinGuid=#610
  toNode=Use Specific Root Shapes End
  toPin=bCondition
  toPinGuid=#611
 }
 {
  fromNode=IntToBool_0
  fromPin=Bool
  fromPinGuid=#610
  toNode=BoolToString_8
  toPin=Bool
  toPinGuid=#612
 }
 {
  fromNode=BoolToString_8
  fromPin=String
  fromPinGuid=#613
  toNode=LogString_6
  toPin=String
  toPinGuid=#614
 }
 {
  fromNode=BoolToString_8
  fromPin=String
  fromPinGuid=#613
  toNode=StringAppend_14
  toPin=String2
  toPinGuid=#615
 }
 {
  fromNode=StringAppend_15
  fromPin=String
  fromPinGuid=#616
  toNode=StringAppend_16
  toPin=String2
  toPinGuid=#617
 }
 {
  fromNode=StringAppend_17
  fromPin=String
  fromPinGuid=#618
  toNode=StringAppend_18
  toPin=String1
  toPinGuid=#619
 }
 {
  fromNode=BoolToString_0
  fromPin=String
  fromPinGuid=#620
  toNode=LogString_10
  toPin=String
  toPinGuid=#621
 }
 {
  fromNode=BoolToString_0
  fromPin=String
  fromPinGuid=#620
  toNode=StringAppend_17
  toPin=String2
  toPinGuid=#622
 }
 {
  fromNode=StringAppend_18
  fromPin=String
  fromPinGuid=#623
  toNode=StringAppend_19
  toPin=String2
  toPinGuid=#624
 }
 {
  fromNode=Proximity_Standard
  fromPin=Collection
  fromPinGuid=#625
  toNode=BranchCollection_4
  toPin=TrueCollection
  toPinGuid=#626
 }
 {
  fromNode=Subtract
  fromPin=ReturnValue
  fromPinGuid=#627
  toNode=Divide_3
  toPin=FloatA
  toPinGuid=#628
 }
 {
  fromNode=Subtract
  fromPin=ReturnValue
  fromPinGuid=#627
  toNode=FloatToString
  toPin=Float
  toPinGuid=#629
 }
 {
  fromNode=StringAppend_20_copy_1_0
  fromPin=String
  fromPinGuid=#630
  toNode=StringAppend_21_copy
  toPin=String2
  toPinGuid=#631
 }
 {
  fromNode=StringAppend_21
  fromPin=String
  fromPinGuid=#632
  toNode=StringAppend_21_copy
  toPin=String1
  toPinGuid=#633
 }
 {
  fromNode=CollectionTransformSelectBySize_Small
  fromPin=TransformSelection
  fromPinGuid=#47
  toNode=SimplifyConvexHulls_Leaf_QSlim_Small_copy
  toPin=OptionalSelectionFilter
  toPinGuid=#634
 }
 {
  fromNode=MergeConvexHulls_0
  fromPin=Collection
  fromPinGuid=#635
  toNode=BranchCollection_1
  toPin=FalseCollection
  toPinGuid=#636
 }
 {
  fromNode=Pow_0
  fromPin=ReturnValue
  fromPinGuid=#637
  toNode=Divide_3_copy
  toPin=FloatA
  toPinGuid=#638
 }
 {
  fromNode=CreateLeafConvexHulls_Mid
  fromPin=Collection
  fromPinGuid=#639
  toNode=CollectionTransformSelectBySize_NarrowWindow
  toPin=Collection
  toPinGuid=#640
 }
 {
  fromNode=MergeConvexHulls_0_copy_0
  fromPin=Collection
  fromPinGuid=#641
  toNode=BranchCollection_3
  toPin=FalseCollection
  toPinGuid=#642
 }
 {
  fromNode=DEBUG -- BYPASS FINAL SIMPLIFICATION
  fromPin=Bool
  fromPinGuid=#643
  toNode=BranchCollection_5
  toPin=bCondition
  toPinGuid=#644
 }
 {
  fromNode=SimplifyConvexHulls_Leaf_QSlim
  fromPin=Collection
  fromPinGuid=#645
  toNode=SimplifyConvexHulls_Leaf_Angle
  toPin=Collection
  toPinGuid=#646
 }
 {
  fromNode=CollectionTransformSelectBySize_Small_copy_0
  fromPin=TransformSelection
  fromPinGuid=#647
  toNode=MergeConvexHulls_copy_0
  toPin=OptionalSelectionFilter
  toPinGuid=#648
 }
 {
  fromNode=FloatToDouble_3
  fromPin=Double
  fromPinGuid=#327
  toNode=GenerateClusterConvexHullsFromLeafHulls_3
  toPin=NegativeSpaceTolerance
  toPinGuid=#649
 }
 {
  fromNode=StringAppend_20_copy
  fromPin=String
  fromPinGuid=#650
  toNode=StringAppend_21
  toPin=String1
  toPinGuid=#651
 }
 {
  fromNode=Pow_1
  fromPin=ReturnValue
  fromPinGuid=#652
  toNode=Subtract_0
  toPin=FloatA
  toPinGuid=#653
 }
 {
  fromNode=SP - Base Samples
  fromPin=Float
  fromPinGuid=#323
  toNode=FloatToInt_2
  toPin=Float
  toPinGuid=#654
 }
 {
  fromNode=CollectionTransformSelectRoot_0
  fromPin=Collection
  fromPinGuid=#655
  toNode=SetAnchorState_0
  toPin=Collection
  toPinGuid=#656
 }
 {
  fromNode=BranchCollection_DerivedProximityMethod
  fromPin=ChosenCollection
  fromPinGuid=#657
  toNode=BranchCollection_UseManualProximity
  toPin=FalseCollection
  toPinGuid=#658
 }
 {
  fromNode=SimplifyConvexHulls_Leaf_QSlim_Small_copy
  fromPin=Collection
  fromPinGuid=#659
  toNode=SimplifyConvexHulls_Leaf_QSlim_Small
  toPin=Collection
  toPinGuid=#660
 }
 {
  fromNode=CollectionTransformSelectionFromIndexArray
  fromPin=TransformSelection
  fromPinGuid=#114
  toNode=MergeConvexHulls_1
  toPin=OptionalSelectionFilter
  toPinGuid=#661
 }
 {
  fromNode=BranchCollection_7
  fromPin=ChosenCollection
  fromPinGuid=#662
  toNode=UpdateVolumeAttributes_0
  toPin=Collection
  toPinGuid=#663
 }
 {
  fromNode=One Minus_copy
  fromPin=ReturnValue
  fromPinGuid=#174
  toNode=FloatToInt_0_copy_0
  toPin=Float
  toPinGuid=#664
 }
 {
  fromNode=Pow_1
  fromPin=ReturnValue
  fromPinGuid=#652
  toNode=FloatToString_1
  toPin=Float
  toPinGuid=#665
 }
 {
  fromNode=ClusterIsolatedRoots
  fromPin=Collection
  fromPinGuid=#666
  toNode=CollectionTransformSelectRoot_3
  toPin=Collection
  toPinGuid=#667
 }
 {
  fromNode=CollectionTransformSelectRoot_2
  fromPin=TransformSelection
  fromPinGuid=#33
  toNode=RemoveOnBreak_0_copy
  toPin=TransformSelection
  toPinGuid=#668
 }
 {
  fromNode=Max Leaf Range
  fromPin=Float
  fromPinGuid=#669
  toNode=CollectionTransformSelectBySize_Small_copy_0
  toPin=SizeMax
  toPinGuid=#670
 }
 {
  fromNode=BoolToString_5
  fromPin=String
  fromPinGuid=#221
  toNode=Gets Shape Protection
  toPin=String
  toPinGuid=#671
 }
 {
  fromNode=SimplifyConvexHulls_Leaf_QSlim_Tiny
  fromPin=Collection
  fromPinGuid=#672
  toNode=SimplifyConvexHulls_QSlim_Tiny_2_copy
  toPin=Collection
  toPinGuid=#673
 }
 {
  fromNode=GenerateClusterConvexHullsFromLeafHulls_3
  fromPin=SphereCovering
  fromPinGuid=#674
  toNode=Sphere Covering to Mesh
  toPin=SphereCovering
  toPinGuid=#675
 }
 {
  fromNode=CollectionTransformSelectRoot_2
  fromPin=Collection
  fromPinGuid=#676
  toNode=RemoveOnBreak_0_copy
  toPin=Collection
  toPinGuid=#677
 }
 {
  fromNode=Level 1 - Cluster Convexes
  fromPin=IsOverriden
  fromPinGuid=#357
  toNode=BoolToString_6_copy
  toPin=Bool
  toPinGuid=#678
 }
 {
  fromNode=GetBoundingBoxesFromCollection
  fromPin=BoundingBoxes
  fromPinGuid=#679
  toNode=GetBoxLengths
  toPin=Boxes
  toPinGuid=#680
 }
 {
  fromNode=SimplifyConvexHulls_QSlim_Tiny_2_copy
  fromPin=Collection
  fromPinGuid=#681
  toNode=GetBoundingBoxesFromCollection
  toPin=Collection
  toPinGuid=#682
 }
 {
  fromNode=Divide_8
  fromPin=ReturnValue
  fromPinGuid=#683
  toNode=Divide_0
  toPin=FloatA
  toPinGuid=#684
 }
 {
  fromNode=Pow_0
  fromPin=ReturnValue
  fromPinGuid=#637
  toNode=Divide_4
  toPin=FloatA
  toPinGuid=#685
 }
 {
  fromNode=Subtract_0
  fromPin=ReturnValue
  fromPinGuid=#559
  toNode=FloatToString_28_copy
  toPin=Float
  toPinGuid=#686
 }
 {
  fromNode=SimplifyConvexHulls
  fromPin=Collection
  fromPinGuid=#687
  toNode=CollectionTransformSelectRoot_12
  toPin=Collection
  toPinGuid=#688
 }
 {
  fromNode=BranchCollection_5
  fromPin=ChosenCollection
  fromPinGuid=#689
  toNode=Proximity_0
  toPin=Collection
  toPinGuid=#690
 }
 {
  fromNode=CollectionTransformSelectBySize_0
  fromPin=TransformSelection
  fromPinGuid=#691
  toNode=CollectionTransformSelectionSetOperation
  toPin=TransformSelectionB
  toPinGuid=#692
 }
 {
  fromNode=Level 1 - Cluster Convexes
  fromPin=Int
  fromPinGuid=#509
  toNode=ClusterConvexes - Auto
  toPin=ConvexCount
  toPinGuid=#693
 }
 {
  fromNode=SimplifyConvexHulls_Angle_Tiny_copy_0
  fromPin=Collection
  fromPinGuid=#694
  toNode=SimplifyConvexHulls_Leaf_QSlim_Tiny_copy
  toPin=Collection
  toPinGuid=#695
 }
 {
  fromNode=CollectionTransformSelectBySize_Small
  fromPin=TransformSelection
  fromPinGuid=#47
  toNode=MergeConvexHulls_3
  toPin=OptionalSelectionFilter
  toPinGuid=#696
 }
 {
  fromNode=Convex - Protection Tolerance
  fromPin=IsOverriden
  fromPinGuid=#697
  toNode=BranchFloat
  toPin=bCondition
  toPinGuid=#698
 }
 {
  fromNode=MergeConvexHulls_3
  fromPin=Collection
  fromPinGuid=#699
  toNode=SimplifyConvexHulls_Leaf_QSlim_Small_copy
  toPin=Collection
  toPinGuid=#700
 }
 {
  fromNode=FloatToDouble_5
  fromPin=Double
  fromPinGuid=#3
  toNode=MergeConvexHulls_0_copy
  toPin=MinRadius
  toPinGuid=#701
 }
 {
  fromNode=Convex - Shape Protection Root
  fromPin=Bool
  fromPinGuid=#702
  toNode=GenerateClusterConvexHullsFromLeafHulls_3
  toPin=bProtectNegativeSpace
  toPinGuid=#703
 }
 {
  fromNode=CollectionTransformSelectBySize_Small_copy
  fromPin=Collection
  fromPinGuid=#704
  toNode=CreateLeafConvexHulls_Mid
  toPin=Collection
  toPinGuid=#705
 }
 {
  fromNode=Level 1 - Shape Protection
  fromPin=Bool
  fromPinGuid=#402
  toNode=MergeConvexHulls_0_copy
  toPin=bProtectNegativeSpace
  toPinGuid=#706
 }
 {
  fromNode=CollectionTransformSelectBySize
  fromPin=TransformSelection
  fromPinGuid=#338
  toNode=SimplifyConvexHulls_Leaf_QSlim_Tiny
  toPin=OptionalSelectionFilter
  toPinGuid=#707
 }
 {
  fromNode=Proximity_5
  fromPin=Collection
  fromPinGuid=#708
  toNode=BranchCollection_DerivedProximityMethod
  toPin=TrueCollection
  toPinGuid=#709
 }
 {
  fromNode=Convex - Protection Tolerance
  fromPin=Float
  fromPinGuid=#710
  toNode=Add_1
  toPin=FloatA
  toPinGuid=#711
 }
 {
  fromNode=CollectionTransformSelectBySize_Small_copy
  fromPin=TransformSelection
  fromPinGuid=#400
  toNode=CreateLeafConvexHulls_Mid
  toPin=OptionalSelectionFilter
  toPinGuid=#712
 }
 {
  fromNode=Convex - Shape Protection Root
  fromPin=Bool
  fromPinGuid=#702
  toNode=MergeConvexHulls_0
  toPin=bProtectNegativeSpace
  toPinGuid=#713
 }
 {
  fromNode=Small Leaf Size
  fromPin=Float
  fromPinGuid=#549
  toNode=CollectionTransformSelectBySize_Small
  toPin=SizeMax
  toPinGuid=#714
 }
 {
  fromNode=SimplifyConvexHulls_QSlim_Tiny_2_copy
  fromPin=Collection
  fromPinGuid=#681
  toNode=CollectionTransformSelectionFromIndexArray
  toPin=Collection
  toPinGuid=#715
 }
 {
  fromNode=Tiny Leaf Size
  fromPin=Float
  fromPinGuid=#162
  toNode=CollectionTransformSelectBySize
  toPin=SizeMax
  toPinGuid=#716
 }
 {
  fromNode=SimplifyConvexHulls_Leaf_QSlim_Small
  fromPin=Collection
  fromPinGuid=#717
  toNode=CollectionTransformSelectBySize
  toPin=Collection
  toPinGuid=#718
 }
 {
  fromNode=Divide_8
  fromPin=ReturnValue
  fromPinGuid=#683
  toNode=x100
  toPin=FloatA
  toPinGuid=#719
 }
 {
  fromNode=BoolToString_5_copy_0
  fromPin=String
  fromPinGuid=#720
  toNode=LogString_3_copy_0
  toPin=String
  toPinGuid=#721
 }
 {
  fromNode=CollectionTransformSelectBySize
  fromPin=Collection
  fromPinGuid=#722
  toNode=CreateLeafConvexHulls_Tiny
  toPin=Collection
  toPinGuid=#723
 }
 {
  fromNode=CollectionTransformSelectBySize_Small_copy_0
  fromPin=Collection
  fromPinGuid=#724
  toNode=CreateLeafConvexHulls_Large
  toPin=Collection
  toPinGuid=#725
 }
 {
  fromNode=Mid Leaf Size
  fromPin=Float
  fromPinGuid=#726
  toNode=CollectionTransformSelectBySize_Small_copy_0
  toPin=SizeMin
  toPinGuid=#727
 }
 {
  fromNode=Mid Leaf Size
  fromPin=Float
  fromPinGuid=#726
  toNode=CollectionTransformSelectBySize_Small_copy
  toPin=SizeMax
  toPinGuid=#728
 }
 {
  fromNode=MergeConvexHulls_0
  fromPin=SphereCovering
  fromPinGuid=#729
  toNode=Get Sphere Covering Sphere Count
  toPin=SphereCovering
  toPinGuid=#730
 }
 {
  fromNode=CollectionTransformSelectBySize
  fromPin=TransformSelection
  fromPinGuid=#338
  toNode=MergeConvexHulls_2
  toPin=OptionalSelectionFilter
  toPinGuid=#731
 }
 {
  fromNode=CollectionTransformSelectBySize_Small_copy_0
  fromPin=TransformSelection
  fromPinGuid=#647
  toNode=CreateLeafConvexHulls_Large
  toPin=OptionalSelectionFilter
  toPinGuid=#732
 }
 {
  fromNode=ClusterMergeToNeighbors
  fromPin=Collection
  fromPinGuid=#733
  toNode=Branch Use CleanUp
  toPin=TrueCollection
  toPinGuid=#734
 }
 {
  fromNode=StringAppend_19
  fromPin=String
  fromPinGuid=#430
  toNode=StringAppend_16
  toPin=String1
  toPinGuid=#735
 }
 {
  fromNode=Proximity_5_copy_0
  fromPin=Collection
  fromPinGuid=#736
  toNode=BranchCollection_7
  toPin=FalseCollection
  toPinGuid=#737
 }
 {
  fromNode=Pow_1
  fromPin=ReturnValue
  fromPinGuid=#652
  toNode=Divide_8
  toPin=FloatB
  toPinGuid=#738
 }
 {
  fromNode=FloatToString_28_copy
  fromPin=String
  fromPinGuid=#134
  toNode=Size of Empty Space
  toPin=String
  toPinGuid=#739
 }
 {
  fromNode=FloatToString_28
  fromPin=String
  fromPinGuid=#740
  toNode=Volume - by Leaf Convexes
  toPin=String
  toPinGuid=#741
 }
 {
  fromNode=StringAppend_20_copy_0
  fromPin=String
  fromPinGuid=#742
  toNode=StringAppend_20_copy_1
  toPin=String1
  toPinGuid=#743
 }
 {
  fromNode=CollectionTransformSelectionSetOperation
  fromPin=TransformSelection
  fromPinGuid=#529
  toNode=ClusterMergeToNeighbors
  toPin=TransformSelection
  toPinGuid=#744
 }
 {
  fromNode=CollectionTransformSelectionFromIndexArray
  fromPin=TransformSelection
  fromPinGuid=#114
  toNode=CreateLeafConvexHulls_Tiny_copy
  toPin=OptionalSelectionFilter
  toPinGuid=#745
 }
 {
  fromNode=CreateLeafConvexHulls_Tiny_copy
  fromPin=Collection
  fromPinGuid=#746
  toNode=MergeConvexHulls_1
  toPin=Collection
  toPinGuid=#747
 }
 {
  fromNode=Convert Mesh to OBJ String
  fromPin=StringOBJ
  fromPinGuid=#31
  toNode=Write String to File_0
  toPin=FileContents
  toPinGuid=#748
 }
 {
  fromNode=FloatToInt_0_copy_0
  fromPin=Int
  fromPinGuid=#749
  toNode=IntToBool_3_copy_0
  toPin=Int
  toPinGuid=#750
 }
 {
  fromNode=FloatToString_28
  fromPin=String
  fromPinGuid=#740
  toNode=StringAppend_20_copy_0
  toPin=String2
  toPinGuid=#751
 }
 {
  fromNode=StringAppend_21_copy
  fromPin=String
  fromPinGuid=#752
  toNode=LogAttributes
  toPin=String
  toPinGuid=#753
 }
 {
  fromNode=1 Over 3 - Convex test
  fromPin=ReturnValue
  fromPinGuid=#297
  toNode=Pow_0
  toPin=Exp
  toPinGuid=#754
 }
 {
  fromNode=SimplifyConvexHulls Angle Tolerance
  fromPin=Collection
  fromPinGuid=#755
  toNode=SimplifyConvexHulls QSlim
  toPin=Collection
  toPinGuid=#756
 }
 {
  fromNode=FloatToString_21
  fromPin=String
  fromPinGuid=#757
  toNode=Volume Ratio
  toPin=String
  toPinGuid=#758
 }
 {
  fromNode=FloatToString_20
  fromPin=String
  fromPinGuid=#759
  toNode=Volume Differential %
  toPin=String
  toPinGuid=#760
 }
 {
  fromNode=FloatToString_20
  fromPin=String
  fromPinGuid=#759
  toNode=StringAppend_20
  toPin=String2
  toPinGuid=#761
 }
 {
  fromNode=StringAppend_20
  fromPin=String
  fromPinGuid=#762
  toNode=StringAppend_20_copy
  toPin=String1
  toPinGuid=#763
 }
 {
  fromNode=Add_0
  fromPin=ReturnValue
  fromPinGuid=#764
  toNode=FloatToInt_0
  toPin=Float
  toPinGuid=#765
 }
 {
  fromNode=Adaptive Tolerance
  fromPin=String
  fromPinGuid=#94
  toNode=Convex - Protection Tolerance
  toPin=Default
  toPinGuid=#766
 }
 {
  fromNode=CollectionTransformSelectRoot_0
  fromPin=TransformSelection
  fromPinGuid=#767
  toNode=SetAnchorState_0
  toPin=TransformSelection
  toPinGuid=#768
 }
 {
  fromNode=CreateGeometryCollectionFromSources
  fromPin=Materials
  fromPinGuid=#410
  toNode=GeometryCollectionTerminal
  toPin=Materials
  toPinGuid=#769
 }
 {
  fromNode=CollectionTransformSelectionFromIndexArray
  fromPin=TransformSelection
  fromPinGuid=#114
  toNode=SimplifyConvexHulls_Angle_Tiny_copy
  toPin=OptionalSelectionFilter
  toPinGuid=#770
 }
 {
  fromNode=Divide_0_copy
  fromPin=ReturnValue
  fromPinGuid=#771
  toNode=Clamp_1_copy
  toPin=Float
  toPinGuid=#772
 }
 {
  fromNode=Clamp_1_copy
  fromPin=ReturnValue
  fromPinGuid=#773
  toNode=Floor_3_copy
  toPin=Float
  toPinGuid=#774
 }
 {
  fromNode=Floor_3_copy
  fromPin=ReturnValue
  fromPinGuid=#775
  toNode=One Minus
  toPin=FloatB
  toPinGuid=#776
 }
 {
  fromNode=FloatToInt_0_copy
  fromPin=Int
  fromPinGuid=#777
  toNode=IntToBool_3_copy
  toPin=Int
  toPinGuid=#778
 }
 {
  fromNode=IntToBool_3_copy
  fromPin=Bool
  fromPinGuid=#779
  toNode=BoolToString_5_copy
  toPin=Bool
  toPinGuid=#780
 }
 {
  fromNode=BoolToString_5_copy
  fromPin=String
  fromPinGuid=#781
  toNode=LogString_3_copy
  toPin=String
  toPinGuid=#782
 }
 {
  fromNode=Empty Space Driver
  fromPin=Float
  fromPinGuid=#783
  toNode=Divide_0_copy
  toPin=FloatA
  toPinGuid=#784
 }
 {
  fromNode=Divide_8
  fromPin=ReturnValue
  fromPinGuid=#683
  toNode=FloatToString_21
  toPin=Float
  toPinGuid=#785
 }
 {
  fromNode=One Minus
  fromPin=ReturnValue
  fromPinGuid=#90
  toNode=FloatToInt_0_copy
  toPin=Float
  toPinGuid=#786
 }
 {
  fromNode=CollectionTransformSelectBySize_Small
  fromPin=Collection
  fromPinGuid=#787
  toNode=CreateLeafConvexHulls_Small
  toPin=Collection
  toPinGuid=#788
 }
 {
  fromNode=Pow_0
  fromPin=ReturnValue
  fromPinGuid=#637
  toNode=FloatToString_28
  toPin=Float
  toPinGuid=#789
 }
 {
  fromNode=Pow_0
  fromPin=ReturnValue
  fromPinGuid=#637
  toNode=Subtract_0
  toPin=FloatB
  toPinGuid=#790
 }
 {
  fromNode=Branch Use CleanUp
  fromPin=ChosenCollection
  fromPinGuid=#1
  toNode=CollectionTransformSelectCluster
  toPin=Collection
  toPinGuid=#791
 }
 {
  fromNode=Proximity - Contact Threshold
  fromPin=IsOverriden
  fromPinGuid=#792
  toNode=BoolToInt_0_copy
  toPin=Bool
  toPinGuid=#793
 }
 {
  fromNode=IntToBool
  fromPin=Bool
  fromPinGuid=#419
  toNode=BranchCollection
  toPin=bCondition
  toPinGuid=#794
 }
 {
  fromNode=CollectionTransformSelectChildren_1_copy
  fromPin=TransformSelection
  fromPinGuid=#795
  toNode=CreateLeafConvexHulls
  toPin=OptionalSelectionFilter
  toPinGuid=#796
 }
 {
  fromNode=BranchCollection_3
  fromPin=ChosenCollection
  fromPinGuid=#797
  toNode=CollectionTransformSelectRoot_9
  toPin=Collection
  toPinGuid=#798
 }
 {
  fromNode=BoolToString_6_copy
  fromPin=String
  fromPinGuid=#799
  toNode=Enable L1 Convex Count?
  toPin=String
  toPinGuid=#800
 }
 {
  fromNode=FloatToString_2_copy
  fromPin=String
  fromPinGuid=#801
  toNode=LogString_4_copy
  toPin=String
  toPinGuid=#802
 }
 {
  fromNode=FloatToDouble_0
  fromPin=Double
  fromPinGuid=#547
  toNode=MergeConvexHulls_0_copy
  toPin=MinSampleSpacing
  toPinGuid=#803
 }
 {
  fromNode=BoolToInt_1
  fromPin=Int
  fromPinGuid=#804
  toNode=IntToFloat_0
  toPin=Int
  toPinGuid=#805
 }
 {
  fromNode=OneMinus
  fromPin=ReturnValue
  fromPinGuid=#806
  toNode=Clamp_3
  toPin=Float
  toPinGuid=#807
 }
 {
  fromNode=IntToFloat_0
  fromPin=Float
  fromPinGuid=#808
  toNode=OneMinus
  toPin=FloatB
  toPinGuid=#809
 }
 {
  fromNode=Clamp_3
  fromPin=ReturnValue
  fromPinGuid=#810
  toNode=Multiply_0
  toPin=FloatB
  toPinGuid=#811
 }
 {
  fromNode=32
  fromPin=Float
  fromPinGuid=#812
  toNode=Multiply_0
  toPin=FloatA
  toPinGuid=#813
 }
 {
  fromNode=Multiply_0
  fromPin=ReturnValue
  fromPinGuid=#219
  toNode=FloatToDouble_6
  toPin=Float
  toPinGuid=#814
 }
 {
  fromNode=Level 1 - Cluster Convexes
  fromPin=Int
  fromPinGuid=#509
  toNode=IntToString_1
  toPin=Int
  toPinGuid=#815
 }
 {
  fromNode=Level 1 - Cluster Convexes
  fromPin=Int
  fromPinGuid=#509
  toNode=MakeConvexDecompositionSettings
  toPin=MaxHullsPerGeometry
  toPinGuid=#816
 }
 {
  fromNode=Min Size
  fromPin=Float
  fromPinGuid=#817
  toNode=AutoCluster Volume Numeric
  toPin=MinimumSize
  toPinGuid=#818
 }
 {
  fromNode=Min Size
  fromPin=Float
  fromPinGuid=#817
  toNode=AutoCluster_1
  toPin=MinimumSize
  toPinGuid=#819
 }
 {
  fromNode=Convex - Protection Tolerance
  fromPin=Float
  fromPinGuid=#710
  toNode=FloatToDouble_3
  toPin=Float
  toPinGuid=#820
 }
 {
  fromNode="SP - Spacing Multiplier "
  fromPin=Float
  fromPinGuid=#821
  toNode=Multiply_1_copy
  toPin=FloatB
  toPinGuid=#822
 }
 {
  fromNode=CompareFloat_copy_0
  fromPin=Result
  fromPinGuid=#823
  toNode=BoolToString_9_copy_0_0
  toPin=Bool
  toPinGuid=#824
 }
 {
  fromNode=CompareFloat_copy
  fromPin=Result
  fromPinGuid=#825
  toNode=BranchCollection_7_copy
  toPin=bCondition
  toPinGuid=#826
 }
 {
  fromNode=IntToString_0
  fromPin=String
  fromPinGuid=#827
  toNode=LogString_3_copy_1
  toPin=String
  toPinGuid=#828
 }
 {
  fromNode=Get Sphere Covering Sphere Count
  fromPin=NumSpheres
  fromPinGuid=#829
  toNode=IntToString_0
  toPin=Int
  toPinGuid=#830
 }
 {
  fromNode=Convex - Protection Radius
  fromPin=IsOverriden
  fromPinGuid=#831
  toNode=BranchFloat_copy
  toPin=bCondition
  toPinGuid=#832
 }
 {
  fromNode=Proximity - Distance Threshold
  fromPin=IsOverriden
  fromPinGuid=#833
  toNode=BoolToInt_copy
  toPin=Bool
  toPinGuid=#834
 }
 {
  fromNode=Convex - Protection Tolerance
  fromPin=Float
  fromPinGuid=#710
  toNode=FloatToDouble_1
  toPin=Float
  toPinGuid=#835
 }
 {
  fromNode=Proximity_5_copy_1
  fromPin=Collection
  fromPinGuid=#836
  toNode=UpdateVolumeAttributes_0_copy
  toPin=Collection
  toPinGuid=#837
 }
 {
  fromNode=Cleanup - SimplificationAngle
  fromPin=Float
  fromPinGuid=#838
  toNode=SimplifyConvexHulls Angle Tolerance
  toPin=SimplificationAngleThreshold
  toPinGuid=#839
 }
 {
  fromNode=Proximity_5_copy
  fromPin=Collection
  fromPinGuid=#840
  toNode=BranchCollection_7
  toPin=TrueCollection
  toPinGuid=#841
 }
 {
  fromNode=BoolToString_9
  fromPin=String
  fromPinGuid=#842
  toNode=LogString_8
  toPin=String
  toPinGuid=#843
 }
 {
  fromNode=FloatToInt_4_copy
  fromPin=Int
  fromPinGuid=#844
  toNode=IntToBool_copy
  toPin=Int
  toPinGuid=#845
 }
 {
  fromNode=Clamp_5_copy
  fromPin=ReturnValue
  fromPinGuid=#846
  toNode=Floor_copy
  toPin=Float
  toPinGuid=#847
 }
 {
  fromNode=Divide_3_copy
  fromPin=ReturnValue
  fromPinGuid=#848
  toNode=Clamp_5_copy
  toPin=Float
  toPinGuid=#849
 }
 {
  fromNode=Pow_0
  fromPin=ReturnValue
  fromPinGuid=#637
  toNode=Divide_8
  toPin=FloatA
  toPinGuid=#850
 }
 {
  fromNode=Pow_0
  fromPin=ReturnValue
  fromPinGuid=#637
  toNode=Divide_2
  toPin=FloatA
  toPinGuid=#851
 }
 {
  fromNode=Pow_0
  fromPin=ReturnValue
  fromPinGuid=#637
  toNode=Divide_14
  toPin=FloatA
  toPinGuid=#852
 }
 {
  fromNode=Floor_copy
  fromPin=ReturnValue
  fromPinGuid=#853
  toNode=OneMinus_0
  toPin=A
  toPinGuid=#854
 }
 {
  fromNode=IntToBool_copy
  fromPin=Bool
  fromPinGuid=#855
  toNode=BoolToString_9
  toPin=Bool
  toPinGuid=#856
 }
 {
  fromNode=IntToBool_copy
  fromPin=Bool
  fromPinGuid=#855
  toNode=BranchCollection_7
  toPin=bCondition
  toPinGuid=#857
 }
 {
  fromNode=CompareFloat_copy
  fromPin=Result
  fromPinGuid=#825
  toNode=BoolToString_9_copy_0
  toPin=Bool
  toPinGuid=#858
 }
 {
  fromNode=BoolToString_9_copy_0
  fromPin=String
  fromPinGuid=#859
  toNode=LogString_8_copy_0
  toPin=String
  toPinGuid=#860
 }
 {
  fromNode=Proximity - Distance Threshold
  fromPin=IsOverriden
  fromPinGuid=#833
  toNode=BoolToInt_copy_0
  toPin=Bool
  toPinGuid=#861
 }
 {
  fromNode=BoolToString_9_copy_0_0
  fromPin=String
  fromPinGuid=#862
  toNode=LogString_8_copy_0_0
  toPin=String
  toPinGuid=#863
 }
 {
  fromNode=Add_4_copy
  fromPin=ReturnValue
  fromPinGuid=#864
  toNode=CompareFloat_copy
  toPin=FloatA
  toPinGuid=#865
 }
 {
  fromNode=IntToFloat_2_copy
  fromPin=Float
  fromPinGuid=#866
  toNode=Add_4_copy
  toPin=FloatB
  toPinGuid=#867
 }
 {
  fromNode=IntToFloat_1_copy
  fromPin=Float
  fromPinGuid=#868
  toNode=Add_4_copy
  toPin=FloatA
  toPinGuid=#869
 }
 {
  fromNode=IntToBool_1
  fromPin=Bool
  fromPinGuid=#870
  toNode=BranchCollection_4
  toPin=bCondition
  toPinGuid=#871
 }
 {
  fromNode=IntToBool_1
  fromPin=Bool
  fromPinGuid=#870
  toNode=BoolToString_1
  toPin=Bool
  toPinGuid=#872
 }
 {
  fromNode=IntToBool_1
  fromPin=Bool
  fromPinGuid=#870
  toNode=BranchCollection_DerivedProximityMethod
  toPin=bCondition
  toPinGuid=#873
 }
 {
  fromNode=IntToBool_1
  fromPin=Bool
  fromPinGuid=#870
  toNode=BoolToString_9_copy
  toPin=Bool
  toPinGuid=#874
 }
 {
  fromNode=Floor
  fromPin=ReturnValue
  fromPinGuid=#381
  toNode=Multiply_3
  toPin=FloatA
  toPinGuid=#875
 }
 {
  fromNode=Level 1 - Cluster Convexes
  fromPin=IsOverriden
  fromPinGuid=#357
  toNode=BoolToString
  toPin=Bool
  toPinGuid=#876
 }
 {
  fromNode=Multiply_1_copy
  fromPin=ReturnValue
  fromPinGuid=#877
  toNode=FloatToDouble_0
  toPin=Float
  toPinGuid=#878
 }
 {
  fromNode=BranchCollection_3
  fromPin=ChosenCollection
  fromPinGuid=#797
  toNode=CollectionTransformSelectRoot_5
  toPin=Collection
  toPinGuid=#879
 }
 {
  fromNode=IntToString_1
  fromPin=String
  fromPinGuid=#880
  toNode=LogString
  toPin=String
  toPinGuid=#881
 }
 {
  fromNode=Add_1
  fromPin=ReturnValue
  fromPinGuid=#882
  toNode=BranchFloat
  toPin=B
  toPinGuid=#883
 }
 {
  fromNode=GetGeometryCollectionAsset
  fromPin=Asset
  fromPinGuid=#345
  toNode=GetGeometryCollectionSources
  toPin=Asset
  toPinGuid=#884
 }
 {
  fromNode=Convex - Protection Tolerance
  fromPin=Float
  fromPinGuid=#710
  toNode=BranchFloat
  toPin=A
  toPinGuid=#885
 }
 {
  fromNode=Cleanup_L1_bySize
  fromPin=Float
  fromPinGuid=#886
  toNode=CollectionTransformSelectBySize_0
  toPin=SizeMax
  toPinGuid=#887
 }
 {
  fromNode=Proximity Distance - ATOM Default
  fromPin=String
  fromPinGuid=#888
  toNode=Proximity - Distance Threshold
  toPin=Default
  toPinGuid=#889
 }
 {
  fromNode=CollectionTransformSelectTargetLevel
  fromPin=TransformSelection
  fromPinGuid=#890
  toNode=CollectionTransformSelectionSetOperation
  toPin=TransformSelectionA
  toPinGuid=#891
 }
 {
  fromNode=Convex - Sharp Contact
  fromPin=Collection
  fromPinGuid=#892
  toNode=BranchCollection_UseManualProximity
  toPin=TrueCollection
  toPinGuid=#893
 }
 {
  fromNode=MergeConvexHulls_0_copy_0
  fromPin=Collection
  fromPinGuid=#641
  toNode=BranchCollection_3
  toPin=TrueCollection
  toPinGuid=#894
 }
 {
  fromNode=Cleanup_L1_bySize
  fromPin=Float
  fromPinGuid=#886
  toNode=ClusterMergeToNeighbors
  toPin=MinVolumeCubeRoot
  toPinGuid=#895
 }
 {
  fromNode=CreateLeafConvexHulls_Small
  fromPin=Collection
  fromPinGuid=#896
  toNode=MergeConvexHulls_3
  toPin=Collection
  toPinGuid=#897
 }
 {
  fromNode=Branch Use CleanUp
  fromPin=ChosenCollection
  fromPinGuid=#1
  toNode=Flatten
  toPin=Collection
  toPinGuid=#898
 }
 {
  fromNode=FloatToInt_1
  fromPin=Int
  fromPinGuid=#15
  toNode=MergeConvexHulls_0_copy_0
  toPin=TargetNumSamples
  toPinGuid=#899
 }
 {
  fromNode=PURGE L2s
  fromPin=Bool
  fromPinGuid=#290
  toNode=BranchCollection_8
  toPin=bCondition
  toPinGuid=#900
 }
 {
  fromNode=Cleanup - SimplificationDist
  fromPin=Float
  fromPinGuid=#901
  toNode=SimplifyConvexHulls QSlim
  toPin=SimplificationDistanceThreshold
  toPinGuid=#902
 }
 {
  fromNode=Cleanup - SimplificationMinFace
  fromPin=Int
  fromPinGuid=#903
  toNode=SimplifyConvexHulls Angle Tolerance
  toPin=MinTargetTriangleCount
  toPinGuid=#904
 }
 {
  fromNode=Cleanup - SimplificationMinFace
  fromPin=Int
  fromPinGuid=#903
  toNode=SimplifyConvexHulls QSlim
  toPin=MinTargetTriangleCount
  toPinGuid=#905
 }
 {
  fromNode=CollectionTransformSelectRoot_6
  fromPin=Collection
  fromPinGuid=#906
  toNode=AutoCluster
  toPin=Collection
  toPinGuid=#907
 }
 {
  fromNode=CollectionTransformSelectRoot_6
  fromPin=Collection
  fromPinGuid=#906
  toNode=AutoCluster_Explicit
  toPin=Collection
  toPinGuid=#908
 }
 {
  fromNode=Add_4_copy_0
  fromPin=ReturnValue
  fromPinGuid=#909
  toNode=CompareFloat_copy_0
  toPin=FloatA
  toPinGuid=#910
 }
 {
  fromNode=CollectionTransformSelectChildren_4
  fromPin=TransformSelection
  fromPinGuid=#911
  toNode=CollectionTransformSelectionSetOperation_0
  toPin=TransformSelectionB
  toPinGuid=#912
 }
 {
  fromNode=Proximity_copy
  fromPin=Collection
  fromPinGuid=#913
  toNode=CollectionTransformSelectRoot_0
  toPin=Collection
  toPinGuid=#914
 }
 {
  fromNode=SimplifyConvexHulls_Leaf_QSlim_Tiny_copy
  fromPin=Collection
  fromPinGuid=#915
  toNode=Proximity_copy
  toPin=Collection
  toPinGuid=#916
 }
 {
  fromNode=SimplifyConvexHulls_Leaf_Angle
  fromPin=Collection
  fromPinGuid=#917
  toNode=CollectionTransformSelectBySize_Small
  toPin=Collection
  toPinGuid=#918
 }
 {
  fromNode=BoolToInt_copy
  fromPin=Int
  fromPinGuid=#919
  toNode=IntToFloat_1_copy
  toPin=Int
  toPinGuid=#920
 }
 {
  fromNode=BoolToInt_0_copy
  fromPin=Int
  fromPinGuid=#921
  toNode=IntToFloat_2_copy
  toPin=Int
  toPinGuid=#922
 }
 {
  fromNode=Proximity - Distance Threshold
  fromPin=Float
  fromPinGuid=#272
  toNode=Proximity_Standard
  toPin=DistanceThreshold
  toPinGuid=#923
 }
 {
  fromNode=Proximity - Distance Threshold
  fromPin=Float
  fromPinGuid=#272
  toNode=Proximity_5
  toPin=DistanceThreshold
  toPinGuid=#924
 }
 {
  fromNode=Proximity - Distance Threshold
  fromPin=Float
  fromPinGuid=#272
  toNode=Proximity_1
  toPin=DistanceThreshold
  toPinGuid=#925
 }
 {
  fromNode=Proximity
  fromPin=Collection
  fromPinGuid=#926
  toNode=GetNumElementsInCollectionGroup
  toPin=Collection
  toPinGuid=#927
 }
 {
  fromNode=Proximity
  fromPin=Collection
  fromPinGuid=#926
  toNode=Proximity_Precise
  toPin=Collection
  toPinGuid=#928
 }
 {
  fromNode=Proximity
  fromPin=Collection
  fromPinGuid=#926
  toNode=BoundingBox_1
  toPin=Collection
  toPinGuid=#929
 }
 {
  fromNode=Proximity
  fromPin=Collection
  fromPinGuid=#926
  toNode=GetNumElementsInCollectionGroup_1
  toPin=Collection
  toPinGuid=#930
 }
 {
  fromNode=Proximity
  fromPin=Collection
  fromPinGuid=#926
  toNode=Proximity_Standard
  toPin=Collection
  toPinGuid=#931
 }
 {
  fromNode=Proximity
  fromPin=Collection
  fromPinGuid=#926
  toNode=UpdateVolumeAttributes_2
  toPin=Collection
  toPinGuid=#932
 }
 {
  fromNode=Proximity
  fromPin=Collection
  fromPinGuid=#926
  toNode=Proximity_5_copy
  toPin=Collection
  toPinGuid=#933
 }
 {
  fromNode=Proximity
  fromPin=Collection
  fromPinGuid=#926
  toNode=Proximity_5_copy_0
  toPin=Collection
  toPinGuid=#934
 }
 {
  fromNode=Proximity
  fromPin=Collection
  fromPinGuid=#926
  toNode=Proximity_5_copy_1
  toPin=Collection
  toPinGuid=#935
 }
 {
  fromNode=UpdateVolumeAttributes_0
  fromPin=Collection
  fromPinGuid=#936
  toNode=BranchCollection_7_copy
  toPin=FalseCollection
  toPinGuid=#937
 }
 {
  fromNode=Flatten
  fromPin=Collection
  fromPinGuid=#938
  toNode=BranchCollection_9
  toPin=TrueCollection
  toPinGuid=#939
 }
 {
  fromNode=ClusterIsolatedRoots
  fromPin=Collection
  fromPinGuid=#666
  toNode=BranchCollection_8
  toPin=TrueCollection
  toPinGuid=#940
 }
 {
  fromNode=Proximity_0
  fromPin=Collection
  fromPinGuid=#941
  toNode=BranchCollection_10
  toPin=FalseCollection
  toPinGuid=#942
 }
 {
  fromNode=Proximity_0
  fromPin=Collection
  fromPinGuid=#941
  toNode=CollectionTransformSelectRoot_15
  toPin=Collection
  toPinGuid=#943
 }
 {
  fromNode=CollectionTransformSelectInvert_0
  fromPin=TransformSelection
  fromPinGuid=#944
  toNode=ClearConvexHulls
  toPin=TransformSelection
  toPinGuid=#945
 }
 {
  fromNode=ClearConvexHulls
  fromPin=Collection
  fromPinGuid=#946
  toNode=BranchCollection_10
  toPin=TrueCollection
  toPinGuid=#947
 }
 {
  fromNode=Proximity Contact - ATOM Default
  fromPin=String
  fromPinGuid=#948
  toNode=Proximity - Contact Threshold
  toPin=Default
  toPinGuid=#949
 }
 {
  fromNode=Proximity_0
  fromPin=Collection
  fromPinGuid=#941
  toNode=CollectionTransformSelectRoot_15_copy
  toPin=Collection
  toPinGuid=#950
 }
 {
  fromNode=Purge Leaf Convexes
  fromPin=Bool
  fromPinGuid=#951
  toNode=BranchCollection_10
  toPin=bCondition
  toPinGuid=#952
 }
 {
  fromNode=CollectionTransformSelectRoot_15_copy
  fromPin=TransformSelection
  fromPinGuid=#953
  toNode=CollectionTransformSelectionSetOperation_0
  toPin=TransformSelectionA
  toPinGuid=#954
 }
 {
  fromNode=CollectionTransformSelectRoot_15
  fromPin=TransformSelection
  fromPinGuid=#955
  toNode=CollectionTransformSelectChildren_4
  toPin=TransformSelection
  toPinGuid=#956
 }
 {
  fromNode=CollectionTransformSelectRoot_15
  fromPin=Collection
  fromPinGuid=#957
  toNode=CollectionTransformSelectChildren_4
  toPin=Collection
  toPinGuid=#958
 }
 {
  fromNode=CollectionTransformSelectionSetOperation_0
  fromPin=TransformSelection
  fromPinGuid=#959
  toNode=CollectionTransformSelectInvert_0
  toPin=TransformSelection
  toPinGuid=#960
 }
 {
  fromNode=MergeConvexHulls_copy_0
  fromPin=Collection
  fromPinGuid=#961
  toNode=CollectionTransformSelectBySize_Small_copy
  toPin=Collection
  toPinGuid=#962
 }
 {
  fromNode=CollectionTransformSelectChildren_4
  fromPin=Collection
  fromPinGuid=#963
  toNode=ClearConvexHulls
  toPin=Collection
  toPinGuid=#964
 }
 {
  fromNode=CollectionTransformSelectionSetOperation_1
  fromPin=TransformSelection
  fromPinGuid=#965
  toNode=Flatten
  toPin=OptionalTransformSelection
  toPinGuid=#966
 }
 {
  fromNode=CollectionTransformSelectCluster
  fromPin=TransformSelection
  fromPinGuid=#967
  toNode=CollectionTransformSelectionSetOperation_1
  toPin=TransformSelectionB
  toPinGuid=#968
 }
 {
  fromNode=CollectionTransformSelectChildren_3
  fromPin=TransformSelection
  fromPinGuid=#969
  toNode=CollectionTransformSelectionSetOperation_1
  toPin=TransformSelectionA
  toPinGuid=#970
 }
 {
  fromNode=CollectionTransformSelectRoot_13
  fromPin=Collection
  fromPinGuid=#971
  toNode=CollectionTransformSelectChildren_3
  toPin=Collection
  toPinGuid=#972
 }
 {
  fromNode=CollectionTransformSelectRoot_13
  fromPin=TransformSelection
  fromPinGuid=#973
  toNode=CollectionTransformSelectChildren_3
  toPin=TransformSelection
  toPinGuid=#974
 }
 {
  fromNode=Branch Use CleanUp
  fromPin=ChosenCollection
  fromPinGuid=#1
  toNode=CollectionTransformSelectRoot_13
  toPin=Collection
  toPinGuid=#975
 }
 {
  fromNode=SimplifyConvexHulls_Leaf_QSlim_Tiny_copy
  fromPin=Collection
  fromPinGuid=#915
  toNode=UpdateVolumeAttributes_1
  toPin=Collection
  toPinGuid=#976
 }
 {
  fromNode=RemoveOnBreak_0
  fromPin=Collection
  fromPinGuid=#977
  toNode=GeometryCollectionTerminal
  toPin=Collection
  toPinGuid=#978
 }
 {
  fromNode=CollectionTransformSelectRoot_16
  fromPin=Collection
  fromPinGuid=#979
  toNode=ClearConvexHulls_copy
  toPin=Collection
  toPinGuid=#980
 }
 {
  fromNode=CollectionTransformSelectRoot_16
  fromPin=TransformSelection
  fromPinGuid=#981
  toNode=ClearConvexHulls_copy
  toPin=TransformSelection
  toPinGuid=#982
 }
 {
  fromNode=BranchCollection_10
  fromPin=ChosenCollection
  fromPinGuid=#983
  toNode=CollectionTransformSelectRoot_16
  toPin=Collection
  toPinGuid=#984
 }
 {
  fromNode=BranchCollection_10
  fromPin=ChosenCollection
  fromPinGuid=#983
  toNode=BranchCollection_11
  toPin=FalseCollection
  toPinGuid=#985
 }
 {
  fromNode=BranchCollection_11
  fromPin=ChosenCollection
  fromPinGuid=#986
  toNode=CollectionTransformSelectRoot_1
  toPin=Collection
  toPinGuid=#987
 }
 {
  fromNode=ClearConvexHulls_copy
  fromPin=Collection
  fromPinGuid=#988
  toNode=BranchCollection_11
  toPin=TrueCollection
  toPinGuid=#989
 }
 {
  fromNode=CompareFloat_copy_0
  fromPin=Result
  fromPinGuid=#823
  toNode=BranchCollection_UseManualProximity
  toPin=bCondition
  toPinGuid=#990
 }
 {
  fromNode=AutoCluster_2
  fromPin=Collection
  fromPinGuid=#991
  toNode=Convex - Sharp Contact
  toPin=Collection
  toPinGuid=#992
 }
 {
  fromNode=AutoCluster_2
  fromPin=Collection
  fromPinGuid=#991
  toNode=Proximity - Precise Contact
  toPin=Collection
  toPinGuid=#993
 }
 {
  fromNode=BoolToInt_copy_0
  fromPin=Int
  fromPinGuid=#994
  toNode=IntToFloat_1_copy_0
  toPin=Int
  toPinGuid=#995
 }
 {
  fromNode=BoolToInt_0_copy_0
  fromPin=Int
  fromPinGuid=#996
  toNode=IntToFloat_2_copy_0
  toPin=Int
  toPinGuid=#997
 }
 {
  fromNode=IntToFloat_1_copy_0
  fromPin=Float
  fromPinGuid=#998
  toNode=Add_4_copy_0
  toPin=FloatA
  toPinGuid=#999
 }
 {
  fromNode=IntToFloat_2_copy_0
  fromPin=Float
  fromPinGuid=#1000
  toNode=Add_4_copy_0
  toPin=FloatB
  toPinGuid=#1001
 }
 {
  fromNode=Proximity - Contact Threshold
  fromPin=IsOverriden
  fromPinGuid=#792
  toNode=BoolToInt_0_copy_0
  toPin=Bool
  toPinGuid=#1002
 }
 {
  fromNode=Proximity - Distance Threshold
  fromPin=Float
  fromPinGuid=#272
  toNode=Proximity_0
  toPin=DistanceThreshold
  toPinGuid=#1003
 }
 {
  fromNode=Proximity - Contact Threshold
  fromPin=Float
  fromPinGuid=#148
  toNode=Proximity_0
  toPin=ContactThreshold
  toPinGuid=#1004
 }
 {
  fromNode=Proximity - Contact Threshold
  fromPin=Float
  fromPinGuid=#148
  toNode=Proximity_1
  toPin=ContactThreshold
  toPinGuid=#1005
 }
 {
  fromNode=Proximity - Contact Threshold
  fromPin=Float
  fromPinGuid=#148
  toNode=Proximity_5
  toPin=ContactThreshold
  toPinGuid=#1006
 }
 {
  fromNode=Proximity - Contact Threshold
  fromPin=Float
  fromPinGuid=#148
  toNode=Proximity_Standard
  toPin=ContactThreshold
  toPinGuid=#1007
 }
 {
  fromNode=AutoCluster_2
  fromPin=Collection
  fromPinGuid=#991
  toNode=Proximity_5
  toPin=Collection
  toPinGuid=#1008
 }
 {
  fromNode=CollectionTransformSelectBySize_NarrowWindow
  fromPin=TransformSelection
  fromPinGuid=#1009
  toNode=CreateLeafConvexHulls_Mid_copy
  toPin=OptionalSelectionFilter
  toPinGuid=#1010
 }
 {
  fromNode=Multiply_1_copy
  fromPin=ReturnValue
  fromPinGuid=#877
  toNode=FloatToDouble
  toPin=Float
  toPinGuid=#1011
 }
 {
  fromNode=FloatToDouble_4
  fromPin=Double
  fromPinGuid=#1012
  toNode=MergeConvexHulls_0
  toPin=MinRadius
  toPinGuid=#1013
 }
 {
  fromNode=FloatToDouble_4
  fromPin=Double
  fromPinGuid=#1012
  toNode=GenerateClusterConvexHullsFromLeafHulls_3
  toPin=MinRadius
  toPinGuid=#1014
 }
 {
  fromNode=FloatToDouble
  fromPin=Double
  fromPinGuid=#1015
  toNode=GenerateClusterConvexHullsFromLeafHulls_3
  toPin=MinSampleSpacing
  toPinGuid=#1016
 }
 {
  fromNode=FloatToDouble
  fromPin=Double
  fromPinGuid=#1015
  toNode=MergeConvexHulls_0
  toPin=MinSampleSpacing
  toPinGuid=#1017
 }
 {
  fromNode=CollectionTransformSelectChildren_1
  fromPin=TransformSelection
  fromPinGuid=#1018
  toNode=ClusterConvexes - Manual
  toPin=OptionalSelectionFilter
  toPinGuid=#1019
 }
 {
  fromNode=CollectionTransformSelectChildren_1
  fromPin=TransformSelection
  fromPinGuid=#1018
  toNode=ClusterConvexes - Auto
  toPin=OptionalSelectionFilter
  toPinGuid=#1020
 }
 {
  fromNode=CollectionTransformSelectChildren_1
  fromPin=TransformSelection
  fromPinGuid=#1018
  toNode=MergeConvexHulls_0_copy
  toPin=OptionalSelectionFilter
  toPinGuid=#1021
 }
 {
  fromNode=CollectionTransformSelectChildren_1
  fromPin=TransformSelection
  fromPinGuid=#1018
  toNode=MergeConvexHulls_0_copy_0
  toPin=OptionalSelectionFilter
  toPinGuid=#1022
 }
 {
  fromNode=ClusterMergeToNeighbors_copy
  fromPin=Collection
  fromPinGuid=#1023
  toNode=Branch Use CleanUp
  toPin=FalseCollection
  toPinGuid=#1024
 }
 {
  fromNode=BranchCollection
  fromPin=ChosenCollection
  fromPinGuid=#1025
  toNode=ClusterMergeToNeighbors_copy
  toPin=Collection
  toPinGuid=#1026
 }
 {
  fromNode=BranchCollection
  fromPin=ChosenCollection
  fromPinGuid=#1025
  toNode=CollectionTransformSelectBySize_0
  toPin=Collection
  toPinGuid=#1027
 }
 {
  fromNode=CollectionTransformSelectChildren_1_copy
  fromPin=Collection
  fromPinGuid=#1028
  toNode=CreateLeafConvexHulls
  toPin=Collection
  toPinGuid=#1029
 }
 {
  fromNode=L1 by Size - ATOM Default
  fromPin=Float
  fromPinGuid=#1030
  toNode=ClusterMergeToNeighbors_copy
  toPin=MinVolumeCubeRoot
  toPinGuid=#1031
 }
 {
  fromNode=BranchCollection
  fromPin=ChosenCollection
  fromPinGuid=#1025
  toNode=ClusterMergeToNeighbors
  toPin=Collection
  toPinGuid=#1032
 }
 {
  fromNode="REROUTER "
  fromPin=ReturnValue
  fromPinGuid=#1033
  toNode=FloatToDouble_5
  toPin=Float
  toPinGuid=#1034
 }
 {
  fromNode="REROUTER "
  fromPin=ReturnValue
  fromPinGuid=#1033
  toNode=Multiply_1_copy
  toPin=FloatA
  toPinGuid=#1035
 }
 {
  fromNode=BranchCollection_1_copy
  fromPin=ChosenCollection
  fromPinGuid=#1036
  toNode=SimplifyConvexHulls Angle Tolerance
  toPin=Collection
  toPinGuid=#1037
 }
 {
  fromNode=BranchCollection_1_copy
  fromPin=ChosenCollection
  fromPinGuid=#1036
  toNode=BranchCollection_5
  toPin=TrueCollection
  toPinGuid=#1038
 }
 {
  fromNode=CollectionTransformSelectBySize_NarrowWindow_copy
  fromPin=TransformSelection
  fromPinGuid=#1039
  toNode=CreateLeafConvexHulls_Mid_copy_0
  toPin=OptionalSelectionFilter
  toPinGuid=#1040
 }
 {
  fromNode=CreateLeafConvexHulls_Mid_copy
  fromPin=Collection
  fromPinGuid=#1041
  toNode=CollectionTransformSelectBySize_NarrowWindow_copy
  toPin=Collection
  toPinGuid=#1042
 }
 {
  fromNode=CreateLeafConvexHulls_Mid_copy_0
  fromPin=Collection
  fromPinGuid=#1043
  toNode=MergeConvexHulls_copy
  toPin=Collection
  toPinGuid=#1044
 }
 {
  fromNode=CreateGeometryCollectionFromSources
  fromPin=InstancedMeshes
  fromPinGuid=#1045
  toNode=AtomDestructionFixISMMaterialsDataflowNode
  toPin=InstancedMeshes
  toPinGuid=#1046
 }
 {
  fromNode=AtomDestructionFixISMMaterialsDataflowNode
  fromPin=InstancedMeshes
  fromPinGuid=#503
  toNode=GeometryCollectionTerminal
  toPin=InstancedMeshes
  toPinGuid=#1047
 }
 {
  fromNode=BranchCollection
  fromPin=ChosenCollection
  fromPinGuid=#1025
  toNode=CollectionTransformSelectTargetLevel
  toPin=Collection
  toPinGuid=#1048
 }
 {
  fromNode="REROUTER "
  fromPin=ReturnValue
  fromPinGuid=#1033
  toNode=FloatToDouble_4
  toPin=Float
  toPinGuid=#1049
 }
 {
  fromNode="REROUTER "
  fromPin=ReturnValue
  fromPinGuid=#1033
  toNode=Multiply_1
  toPin=FloatA
  toPinGuid=#1050
 }
 {
  fromNode="REROUTER "
  fromPin=ReturnValue
  fromPinGuid=#1033
  toNode=BranchFloat_copy
  toPin=A
  toPinGuid=#1051
 }
 {
  fromNode=CreateLeafConvexHulls
  fromPin=Collection
  fromPinGuid=#1052
  toNode=BranchCollection_1_copy
  toPin=TrueCollection
  toPinGuid=#1053
 }
 {
  fromNode=CreateLeafConvexHulls
  fromPin=Collection
  fromPinGuid=#1052
  toNode=BranchCollection_1_copy
  toPin=FalseCollection
  toPinGuid=#1054
 }
]
nodes [
 GetGeometryCollectionAsset {
  type=FGetGeometryCollectionAssetDataflowNode
  guid=#1055
  inputPins []
  outputPins [Asset {type="TObjectPtr<UGeometryCollection>"}]
 }
 GetGeometryCollectionSources {
  type=FGetGeometryCollectionSourcesDataflowNode
  guid=#1056
  inputPins [Asset {type="TObjectPtr<UGeometryCollection>"}]
  outputPins [Sources {type="TArray<FGeometryCollectionSource>"}]
 }
 Proximity {
  type=FProximityDataflowNode
  guid=#1057
  inputPins [
   Collection {type=FManagedArrayCollection}
   DistanceThreshold {type=float}
   ContactThreshold {type=float}
  ]
  outputPins [Collection {type=FManagedArrayCollection}]
  properties {
   ContactAreaMethod=Dataflow_ConnectionContactAreaMethod_None
   ContactThreshold=3.2
   DistanceThreshold=1.5
   FilterContactMethod=Dataflow_ProximityContactFilteringMethod_ConvexHullSharp
   ProximityMethod=Dataflow_ProximityMethod_ConvexHull
   bRecomputeConvexHulls=False
   bUseAsConnectionGraph=False
  }
 }
 CollectionTransformSelectRoot {
  type=FCollectionTransformSelectionRootDataflowNode
  guid=#1058
  inputPins [Collection {type=FManagedArrayCollection}]
  outputPins [
   Collection {type=FManagedArrayCollection}
   TransformSelection {type=FDataflowTransformSelection}
  ]
 }
 CollectionTransformSelectRoot_1 {
  type=FCollectionTransformSelectionRootDataflowNode
  guid=#1059
  inputPins [Collection {type=FManagedArrayCollection}]
  outputPins [
   Collection {type=FManagedArrayCollection}
   TransformSelection {type=FDataflowTransformSelection}
  ]
 }
 SetAnchorState {
  type=FSetAnchorStateDataflowNode
  guid=#1060
  inputPins [
   Collection {type=FManagedArrayCollection}
   TransformSelection {type=FDataflowTransformSelection}
  ]
  outputPins [Collection {type=FManagedArrayCollection}]
  properties {AnchorState=Dataflow_AnchorState_Anchored, bSetNotSelectedBonesToOppositeState=False}
 }
 CreateGeometryCollectionFromSources {
  type=FCreateGeometryCollectionFromSourcesDataflowNode
  guid=#1061
  inputPins [Sources {type="TArray<FGeometryCollectionSource>"}]
  outputPins [
   Collection {type=FManagedArrayCollection}
   Materials {type="TArray<TObjectPtr<UMaterial>>"}
   MaterialInstances {type="TArray<TObjectPtr<UMaterialInterface>>"}
   InstancedMeshes {type="TArray<FGeometryCollectionAutoInstanceMesh>"}
  ]
 }
 Proximity_0 {
  type=FProximityDataflowNode
  guid=#1062
  inputPins [
   Collection {type=FManagedArrayCollection}
   DistanceThreshold {type=float}
   ContactThreshold {type=float}
  ]
  outputPins [Collection {type=FManagedArrayCollection}]
  properties {
   ContactAreaMethod=Dataflow_ProximityContactFilteringMethod_ConvexHullArea
   ContactThreshold=3.2
   DistanceThreshold=2
   FilterContactMethod=Dataflow_ProximityContactFilteringMethod_ConvexHullSharp
   ProximityMethod=Dataflow_ProximityMethod_ConvexHull
   bRecomputeConvexHulls=False
   bUseAsConnectionGraph=True
  }
 }
 GetNumElementsInCollectionGroup {
  type=FGetNumElementsInCollectionGroupDataflowNode
  guid=#1063
  inputPins [Collection {type=FManagedArrayCollection}]
  outputPins [NumElements {type=int32}]
  properties {CustomGroupName="", GroupName=Dataflow_EStandardGroupNameEnum_Transform}
 }
 IntToFloat {
  type=FIntToFloatDataflowNode
  guid=#1064
  inputPins [Int {type=int32}]
  outputPins [Float {type=float}]
  properties {Int=0}
 }
 CollectionTransformSelectRoot_2 {
  type=FCollectionTransformSelectionRootDataflowNode
  guid=#1065
  inputPins [Collection {type=FManagedArrayCollection}]
  outputPins [
   Collection {type=FManagedArrayCollection}
   TransformSelection {type=FDataflowTransformSelection}
  ]
 }
 Minimum Divisor {
  type=FMakeLiteralFloatDataflowNode
  guid=#1066
  inputPins []
  outputPins [Float {type=float}]
  properties {Value=6}
 }
 SafeDivide_Min {
  type=FSafeDivideDataflowNode
  guid=#1067
  inputPins [
   FloatA {type=float}
   FloatB {type=float}
  ]
  outputPins [ReturnValue {type=float}]
  properties {FloatA=0, FloatB=0}
 }
 Maximum Divisor {
  type=FMakeLiteralFloatDataflowNode
  guid=#1068
  inputPins []
  outputPins [Float {type=float}]
  properties {Value=32}
 }
 SafeDivide_Max {
  type=FSafeDivideDataflowNode
  guid=#1069
  inputPins [
   FloatA {type=float}
   FloatB {type=float}
  ]
  outputPins [ReturnValue {type=float}]
  properties {FloatA=0, FloatB=0}
 }
 FloatToString_0 {
  type=FFloatToStringDataflowNode
  guid=#1070
  inputPins [Float {type=float}]
  outputPins [String {type=FString}]
  properties {Float=0}
 }
 PrintString_0 {
  type=FPrintStringDataflowNode
  guid=#1071
  inputPins [String {type=FString}]
  outputPins []
  properties {Color=(B=255,G=255,R=255,A=255), Duration=2, String="", bPrintToLog=True, bPrintToScreen=True}
 }
 SafeDivide_1 {
  type=FSafeDivideDataflowNode
  guid=#1072
  inputPins [
   FloatA {type=float}
   FloatB {type=float}
  ]
  outputPins [ReturnValue {type=float}]
  properties {FloatA=0, FloatB=0}
 }
 RemoveOnBreak_0 {
  type=FRemoveOnBreakDataflowNode
  guid=#1073
  inputPins [
   Collection {type=FManagedArrayCollection}
   TransformSelection {type=FDataflowTransformSelection}
   bEnabledRemoval {type=bool}
   PostBreakTimer {type=FVector2f}
   RemovalTimer {type=FVector2f}
   bClusterCrumbling {type=bool}
  ]
  outputPins [Collection {type=FManagedArrayCollection}]
  properties {PostBreakTimer=(X=1,Y=2), RemovalTimer=(X=0,Y=0), bClusterCrumbling=False, bEnabledRemoval=True}
 }
 Add {
  type=FAddDataflowNode
  guid=#1074
  inputPins [
   FloatA {type=float}
   FloatB {type=float}
  ]
  outputPins [ReturnValue {type=float}]
  properties {FloatA=0, FloatB=0}
 }
 Multiply {
  type=FMultiplyDataflowNode
  guid=#1075
  inputPins [
   FloatA {type=float}
   FloatB {type=float}
  ]
  outputPins [ReturnValue {type=float}]
  properties {FloatA=0, FloatB=0.5}
 }
 CollectionTransformSelectInvert {
  type=FCollectionTransformSelectionInvertDataflowNode
  guid=#1076
  inputPins [TransformSelection {type=FDataflowTransformSelection}]
  outputPins [TransformSelection {type=FDataflowTransformSelection}]
 }
 Clamp_0 {
  type=FClampDataflowNode
  guid=#1077
  inputPins [
   Float {type=float}
   Min {type=float}
   Max {type=float}
  ]
  outputPins [ReturnValue {type=float}]
  properties {Float=0, Max=0.5, Min=0.25}
 }
 MakeLiteralVector {
  type=FMakeLiteralVectorDataflowNode
  guid=#1078
  inputPins [
   X {type=float}
   Y {type=float}
   Z {type=float}
  ]
  outputPins [Vector {type=FVector}]
  properties {X=0, Y=0, Z=0}
 }
 VectorToString {
  type=FVectorToStringDataflowNode
  guid=#1079
  inputPins [Vector {type=FVector}]
  outputPins [String {type=FString}]
  properties {Vector=(X=0,Y=0,Z=0)}
 }
 Ceil {
  type=FCeilDataflowNode
  guid=#1080
  inputPins [Float {type=float}]
  outputPins [ReturnValue {type=float}]
  properties {Float=0}
 }
 Ceil_0 {
  type=FCeilDataflowNode
  guid=#1081
  inputPins [Float {type=float}]
  outputPins [ReturnValue {type=float}]
  properties {Float=0}
 }
 Ceil_1 {
  type=FCeilDataflowNode
  guid=#1082
  inputPins [Float {type=float}]
  outputPins [ReturnValue {type=float}]
  properties {Float=0}
 }
 CollectionTransformSelectRoot_3 {
  type=FCollectionTransformSelectionRootDataflowNode
  guid=#1083
  inputPins [Collection {type=FManagedArrayCollection}]
  outputPins [
   Collection {type=FManagedArrayCollection}
   TransformSelection {type=FDataflowTransformSelection}
  ]
 }
 Proximity_1 {
  type=FProximityDataflowNode
  guid=#1084
  inputPins [
   Collection {type=FManagedArrayCollection}
   DistanceThreshold {type=float}
   ContactThreshold {type=float}
  ]
  outputPins [Collection {type=FManagedArrayCollection}]
  properties {
   ContactAreaMethod=Dataflow_ProximityContactFilteringMethod_ConvexHullArea
   ContactThreshold=8
   DistanceThreshold=3
   FilterContactMethod=Dataflow_ProximityContactFilteringMethod_ConvexHullSharp
   ProximityMethod=Dataflow_ProximityMethod_ConvexHull
   bRecomputeConvexHulls=False
   bUseAsConnectionGraph=False
  }
 }
 CollectionTransformSelectChildren {
  type=FCollectionTransformSelectionChildrenDataflowNode
  guid=#1085
  inputPins [
   Collection {type=FManagedArrayCollection}
   TransformSelection {type=FDataflowTransformSelection}
  ]
  outputPins [
   Collection {type=FManagedArrayCollection}
   TransformSelection {type=FDataflowTransformSelection}
  ]
 }
 GenerateClusterConvexHullsFromLeafHulls_0 {
  type=FGenerateClusterConvexHullsFromLeafHullsDataflowNode
  guid=#1086
  inputPins [
   Collection {type=FManagedArrayCollection}
   ConvexCount {type=int32}
   ErrorTolerance {type=double}
   OptionalSelectionFilter {type=FDataflowTransformSelection}
   bAllowMergingLeafHulls {type=bool}
   RequireNegativeSpaceCovering {type=FDataflowSphereCovering}
   bProtectNegativeSpace {type=bool}
   TargetNumSamples {type=int32}
   MinSampleSpacing {type=double}
   NegativeSpaceTolerance {type=double}
   MinRadius {type=double}
  ]
  outputPins [
   Collection {type=FManagedArrayCollection}
   SphereCovering {type=FDataflowSphereCovering}
  ]
  properties {
   AllowMerges=Any
   ColorRandomSeed=0
   ConvexCount=3
   DebugDrawRenderSettings=(bTranslucent=True,Color=(R=0,G=1,B=0,A=1),LineWidthMultiplier=2)
   ErrorTolerance=16
   MergeProximityDistanceThreshold=0.1
   MergeProximityFilter=None
   MinRadius=10
   MinSampleSpacing=1
   NegativeSpaceTolerance=2
   SampleMethod=Uniform
   SphereCoveringDebugDrawRenderSettings=(bTranslucent=True,LineWidthMultiplier=0.25,Color=(R=1,G=0,B=0,A=1),ColorA=(R=1,G=0,B=0,A=1),ColorB=(R=0,G=0,B=1,A=1))
   TargetNumSamples=50
   bAllowMergingLeafHulls=False
   bOnlyConnectedToHull=False
   bPreferExternalCollisionShapes=False
   bProtectNegativeSpace=False
   bRandomizeColor=True
   bRequireSearchSampleCoverage=False
  }
 }
 CollectionTransformSelectRoot_4 {
  type=FCollectionTransformSelectionRootDataflowNode
  guid=#1087
  inputPins [Collection {type=FManagedArrayCollection}]
  outputPins [
   Collection {type=FManagedArrayCollection}
   TransformSelection {type=FDataflowTransformSelection}
  ]
 }
 CollectionTransformSelectChildren_1 {
  type=FCollectionTransformSelectionChildrenDataflowNode
  guid=#1088
  inputPins [
   Collection {type=FManagedArrayCollection}
   TransformSelection {type=FDataflowTransformSelection}
  ]
  outputPins [
   Collection {type=FManagedArrayCollection}
   TransformSelection {type=FDataflowTransformSelection}
  ]
 }
 CollectionTransformSelectRoot_5 {
  type=FCollectionTransformSelectionRootDataflowNode
  guid=#1089
  inputPins [Collection {type=FManagedArrayCollection}]
  outputPins [
   Collection {type=FManagedArrayCollection}
   TransformSelection {type=FDataflowTransformSelection}
  ]
 }
 AutoCluster {
  type=FAutoClusterDataflowNode
  guid=#1090
  inputPins [
   Collection {type=FManagedArrayCollection}
   TransformSelection {type=FDataflowTransformSelection}
   ClusterSites {type=int32}
   ClusterFraction {type=float}
   SiteSize {type=float}
   ClusterGridWidth {type=int32}
   ClusterGridDepth {type=int32}
   ClusterGridHeight {type=int32}
   MinimumSize {type=float}
   bPreferConvexity {type=bool}
   ConcavityTolerance {type=float}
  ]
  outputPins [Collection {type=FManagedArrayCollection}]
  properties {
   AutoCluster=True
   AvoidIsolated=True
   ClusterFraction=0.25
   ClusterGridDepth=2
   ClusterGridHeight=2
   ClusterGridWidth=2
   ClusterSites=3
   ClusterSizeMethod=Dataflow_ClusterSizeMethod_ByFractionOfInput
   ConcavityTolerance=0
   DriftIterations=0
   EnforceSiteParameters=True
   MinimumSize=0
   SiteSize=1
   bPreferConvexity=False
  }
 }
 Divide_3 {
  type=FDivideDataflowNode
  guid=#1091
  inputPins [
   FloatA {type=float}
   FloatB {type=float}
  ]
  outputPins [ReturnValue {type=float}]
  properties {FloatA=0, FloatB=0}
 }
 Clamp_5 {
  type=FClampDataflowNode
  guid=#1092
  inputPins [
   Float {type=float}
   Min {type=float}
   Max {type=float}
  ]
  outputPins [ReturnValue {type=float}]
  properties {Float=0, Max=1, Min=0}
 }
 FloatToInt_4 {
  type=FFloatToIntDataflowNode
  guid=#1093
  inputPins [Float {type=float}]
  outputPins [Int {type=int32}]
  properties {Float=0, Function=Dataflow_FloatToInt_Function_Floor}
 }
 IntToBool {
  type=FIntToBoolDataflowNode
  guid=#1094
  inputPins [Int {type=int32}]
  outputPins [Bool {type=bool}]
  properties {Int=0}
 }
 BranchCollection {
  type=FBranchCollectionDataflowNode
  guid=#1095
  inputPins [
   TrueCollection {type=FManagedArrayCollection}
   FalseCollection {type=FManagedArrayCollection}
   bCondition {type=bool}
  ]
  outputPins [ChosenCollection {type=FManagedArrayCollection}]
  properties {bCondition=False}
 }
 "Minimum Threshold (##Leaf Nodes)" {
  type=FMakeLiteralFloatDataflowNode
  guid=#1096
  inputPins []
  outputPins [Float {type=float}]
  properties {Value=32}
 }
 AutoCluster_0 {
  type=FAutoClusterDataflowNode
  guid=#1097
  inputPins [
   Collection {type=FManagedArrayCollection}
   TransformSelection {type=FDataflowTransformSelection}
   ClusterSites {type=int32}
   ClusterFraction {type=float}
   SiteSize {type=float}
   ClusterGridWidth {type=int32}
   ClusterGridDepth {type=int32}
   ClusterGridHeight {type=int32}
   MinimumSize {type=float}
   bPreferConvexity {type=bool}
   ConcavityTolerance {type=float}
  ]
  outputPins [Collection {type=FManagedArrayCollection}]
  properties {
   AutoCluster=True
   AvoidIsolated=True
   ClusterFraction=0.25
   ClusterGridDepth=2
   ClusterGridHeight=2
   ClusterGridWidth=2
   ClusterSites=10
   ClusterSizeMethod=Dataflow_ClusterSizeMethod_ByFractionOfInput
   ConcavityTolerance=0
   DriftIterations=0
   EnforceSiteParameters=True
   MinimumSize=0
   SiteSize=1
   bPreferConvexity=False
  }
 }
 CollectionTransformSelectRoot_6 {
  type=FCollectionTransformSelectionRootDataflowNode
  guid=#1098
  inputPins [Collection {type=FManagedArrayCollection}]
  outputPins [
   Collection {type=FManagedArrayCollection}
   TransformSelection {type=FDataflowTransformSelection}
  ]
 }
 CollectionTransformSelectRoot_7 {
  type=FCollectionTransformSelectionRootDataflowNode
  guid=#1099
  inputPins [Collection {type=FManagedArrayCollection}]
  outputPins [
   Collection {type=FManagedArrayCollection}
   TransformSelection {type=FDataflowTransformSelection}
  ]
 }
 Divide_4 {
  type=FDivideDataflowNode
  guid=#1100
  inputPins [
   FloatA {type=float}
   FloatB {type=float}
  ]
  outputPins [ReturnValue {type=float}]
  properties {FloatA=0, FloatB=0}
 }
 "Minimum Volume Threshold (Cubic Root Volume)" {
  type=FMakeLiteralFloatDataflowNode
  guid=#1101
  inputPins []
  outputPins [Float {type=float}]
  properties {Value=52}
 }
 Clamp_6 {
  type=FClampDataflowNode
  guid=#1102
  inputPins [
   Float {type=float}
   Min {type=float}
   Max {type=float}
  ]
  outputPins [ReturnValue {type=float}]
  properties {Float=0, Max=1, Min=0}
 }
 Add_2 {
  type=FAddDataflowNode
  guid=#1103
  inputPins [
   FloatA {type=float}
   FloatB {type=float}
  ]
  outputPins [ReturnValue {type=float}]
  properties {FloatA=0, FloatB=0}
 }
 LogString_2 {
  type=FLogStringDataflowNode
  guid=#1104
  inputPins [String {type=FString}]
  outputPins []
  properties {String="", bPrintToLog=True}
 }
 Floor {
  type=FFloorDataflowNode
  guid=#1105
  inputPins [Float {type=float}]
  outputPins [ReturnValue {type=float}]
  properties {Float=0}
 }
 Floor_0 {
  type=FFloorDataflowNode
  guid=#1106
  inputPins [Float {type=float}]
  outputPins [ReturnValue {type=float}]
  properties {Float=0}
 }
 Proximity_Precise {
  type=FProximityDataflowNode
  guid=#1107
  inputPins [
   Collection {type=FManagedArrayCollection}
   DistanceThreshold {type=float}
   ContactThreshold {type=float}
  ]
  outputPins [Collection {type=FManagedArrayCollection}]
  properties {
   ContactAreaMethod=Dataflow_ConnectionContactAreaMethod_None
   ContactThreshold=2
   DistanceThreshold=2
   FilterContactMethod=Dataflow_ProximityContactFilteringMethod_ProjectedBoundsOverlap
   ProximityMethod=Dataflow_ProximityMethod_Precise
   bRecomputeConvexHulls=False
   bUseAsConnectionGraph=False
  }
 }
 CreateLeafConvexHulls_Small {
  type=FCreateLeafConvexHullsDataflowNode
  guid=#1108
  inputPins [
   Collection {type=FManagedArrayCollection}
   OptionalSelectionFilter {type=FDataflowTransformSelection}
   SimplificationDistanceThreshold {type=float}
   ConvexDecompositionSettings {type=FDataflowConvexDecompositionSettings}
   RequireNegativeSpaceCovering {type=FDataflowSphereCovering}
  ]
  outputPins [
   Collection {type=FManagedArrayCollection}
   SphereCovering {type=FDataflowSphereCovering}
  ]
  properties {
   ColorRandomSeed=0
   ConvexDecompositionSettings=(MinSizeToDecompose=12,MaxGeoToHullVolumeRatioToDecompose=0.515,ErrorTolerance=8,MaxHullsPerGeometry=4,MinThicknessTolerance=2,NumAdditionalSplits=3,bOnlyConnectedToHull=True,NegativeSpaceTolerance=2,NegativeSpaceMinRadius=10)
   DebugDrawRenderSettings=(bTranslucent=True,Color=(R=0,G=1,B=0,A=1),LineWidthMultiplier=2)
   GenerateMethod=IntersectExternalWithComputed
   IntersectIfComputedIsSmallerByFactor=1
   MinExternalVolumeToIntersect=3000
   SimplificationDistanceThreshold=0
   SphereCoveringDebugDrawRenderSettings=(bTranslucent=True,LineWidthMultiplier=0.25,Color=(R=1,G=0,B=0,A=1),ColorA=(R=1,G=0,B=0,A=1),ColorB=(R=0,G=0,B=1,A=1))
   bComputeIntersectionsBeforeHull=True
   bRandomizeColor=True
  }
 }
 CollectionTransformSelectChildren_0 {
  type=FCollectionTransformSelectionChildrenDataflowNode
  guid=#1109
  inputPins [
   Collection {type=FManagedArrayCollection}
   TransformSelection {type=FDataflowTransformSelection}
  ]
  outputPins [
   Collection {type=FManagedArrayCollection}
   TransformSelection {type=FDataflowTransformSelection}
  ]
 }
 GenerateClusterConvexHullsFromLeafHulls {
  type=FGenerateClusterConvexHullsFromLeafHullsDataflowNode
  guid=#1110
  inputPins [
   Collection {type=FManagedArrayCollection}
   ConvexCount {type=int32}
   ErrorTolerance {type=double}
   OptionalSelectionFilter {type=FDataflowTransformSelection}
   bAllowMergingLeafHulls {type=bool}
   RequireNegativeSpaceCovering {type=FDataflowSphereCovering}
   bProtectNegativeSpace {type=bool}
   TargetNumSamples {type=int32}
   MinSampleSpacing {type=double}
   NegativeSpaceTolerance {type=double}
   MinRadius {type=double}
  ]
  outputPins [
   Collection {type=FManagedArrayCollection}
   SphereCovering {type=FDataflowSphereCovering}
  ]
  properties {
   AllowMerges=Any
   ColorRandomSeed=0
   ConvexCount=1
   DebugDrawRenderSettings=(bTranslucent=True,Color=(R=0,G=1,B=0,A=1),LineWidthMultiplier=2)
   ErrorTolerance=0
   MergeProximityDistanceThreshold=0.1
   MergeProximityFilter=None
   MinRadius=10
   MinSampleSpacing=1
   NegativeSpaceTolerance=2
   SampleMethod=Uniform
   SphereCoveringDebugDrawRenderSettings=(bTranslucent=True,LineWidthMultiplier=0.25,Color=(R=1,G=0,B=0,A=1),ColorA=(R=1,G=0,B=0,A=1),ColorB=(R=0,G=0,B=1,A=1))
   TargetNumSamples=50
   bAllowMergingLeafHulls=False
   bOnlyConnectedToHull=False
   bPreferExternalCollisionShapes=False
   bProtectNegativeSpace=False
   bRandomizeColor=True
   bRequireSearchSampleCoverage=False
  }
 }
 CollectionTransformSelectRoot_8 {
  type=FCollectionTransformSelectionRootDataflowNode
  guid=#1111
  inputPins [Collection {type=FManagedArrayCollection}]
  outputPins [
   Collection {type=FManagedArrayCollection}
   TransformSelection {type=FDataflowTransformSelection}
  ]
 }
 GeometryCollectionTerminal {
  type=FGeometryCollectionTerminalDataflowNode
  guid=#1112
  inputPins [
   Collection {type=FManagedArrayCollection}
   Materials {type="TArray<TObjectPtr<UMaterial>>"}
   MaterialInstances {type="TArray<TObjectPtr<UMaterialInterface>>"}
   InstancedMeshes {type="TArray<FGeometryCollectionAutoInstanceMesh>"}
  ]
  outputPins [
   Collection {type=FManagedArrayCollection}
   Materials {type="TArray<TObjectPtr<UMaterial>>"}
   MaterialInstances {type="TArray<TObjectPtr<UMaterialInterface>>"}
   InstancedMeshes {type="TArray<FGeometryCollectionAutoInstanceMesh>"}
  ]
 }
 SimplifyConvexHulls QSlim {
  type=FSimplifyConvexHullsDataflowNode
  guid=#1113
  inputPins [
   Collection {type=FManagedArrayCollection}
   OptionalSelectionFilter {type=FDataflowTransformSelection}
   SimplificationAngleThreshold {type=float}
   SimplificationDistanceThreshold {type=float}
   MinTargetTriangleCount {type=int32}
  ]
  outputPins [Collection {type=FManagedArrayCollection}]
  properties {
   ColorRandomSeed=0
   DebugDrawRenderSettings=(bTranslucent=True,Color=(R=0,G=1,B=0,A=1),LineWidthMultiplier=2)
   MinTargetTriangleCount=16
   SimplificationAngleThreshold=10
   SimplificationDistanceThreshold=20
   SimplifyMethod=MeshQSlim
   bRandomizeColor=True
   bUseExistingVertices=True
  }
 }
 Simplification Distance {
  type=FMakeLiteralFloatDataflowNode
  guid=#1114
  inputPins []
  outputPins [Float {type=float}]
  properties {Value=0.5}
 }
 Num Root Shapes {type=FMakeLiteralIntDataflowNode, guid=#1115, inputPins [], outputPins [Int {type=int32}], properties {Value=1}}
 Use Specific Root Shapes End {
  type=FBranchCollectionDataflowNode
  guid=#1116
  inputPins [
   TrueCollection {type=FManagedArrayCollection}
   FalseCollection {type=FManagedArrayCollection}
   bCondition {type=bool}
  ]
  outputPins [ChosenCollection {type=FManagedArrayCollection}]
  properties {bCondition=False}
 }
 Divide_7 {
  type=FDivideDataflowNode
  guid=#1117
  inputPins [
   FloatA {type=float}
   FloatB {type=float}
  ]
  outputPins [ReturnValue {type=float}]
  properties {FloatA=0, FloatB=0}
 }
 Single Convex Tolerance % {
  type=FMakeLiteralFloatDataflowNode
  guid=#1118
  inputPins []
  outputPins [Float {type=float}]
  properties {Value=92}
 }
 FloatToString_16 {
  type=FFloatToStringDataflowNode
  guid=#1119
  inputPins [Float {type=float}]
  outputPins [String {type=FString}]
  properties {Float=0}
 }
 Log Divide Differential by Tolerance % Int {
  type=FLogStringDataflowNode
  guid=#1120
  inputPins [String {type=FString}]
  outputPins []
  properties {String="", bPrintToLog=True}
 }
 FloatToInt_5 {
  type=FFloatToIntDataflowNode
  guid=#1121
  inputPins [Float {type=float}]
  outputPins [Int {type=int32}]
  properties {Float=0, Function=Dataflow_FloatToInt_Function_Floor}
 }
 Clamp_7 {
  type=FClampDataflowNode
  guid=#1122
  inputPins [
   Float {type=float}
   Min {type=float}
   Max {type=float}
  ]
  outputPins [ReturnValue {type=float}]
  properties {Float=0, Max=1, Min=0}
 }
 IntToString_2 {
  type=FIntToStringDataflowNode
  guid=#1123
  inputPins [Int {type=int32}]
  outputPins [String {type=FString}]
  properties {Int=0}
 }
 Log Divide Differential by Tolerance % {
  type=FLogStringDataflowNode
  guid=#1124
  inputPins [String {type=FString}]
  outputPins []
  properties {String="", bPrintToLog=True}
 }
 IntToBool_0 {
  type=FIntToBoolDataflowNode
  guid=#1125
  inputPins [Int {type=int32}]
  outputPins [Bool {type=bool}]
  properties {Int=0}
 }
 AutoCluster_Explicit {
  type=FAutoClusterDataflowNode
  guid=#1126
  inputPins [
   Collection {type=FManagedArrayCollection}
   TransformSelection {type=FDataflowTransformSelection}
   ClusterSites {type=int32}
   ClusterFraction {type=float}
   SiteSize {type=float}
   ClusterGridWidth {type=int32}
   ClusterGridDepth {type=int32}
   ClusterGridHeight {type=int32}
   MinimumSize {type=float}
   bPreferConvexity {type=bool}
   ConcavityTolerance {type=float}
  ]
  outputPins [Collection {type=FManagedArrayCollection}]
  properties {
   AutoCluster=True
   AvoidIsolated=False
   ClusterFraction=0.25
   ClusterGridDepth=2
   ClusterGridHeight=2
   ClusterGridWidth=2
   ClusterSites=10
   ClusterSizeMethod=Dataflow_ClusterSizeMethod_ByNumber
   ConcavityTolerance=0
   DriftIterations=0
   EnforceSiteParameters=True
   MinimumSize=0
   SiteSize=1
   bPreferConvexity=False
  }
 }
 "Level 1 - by  Cluster Number" {
  type=FGetIntOverrideFromAssetDataflowNode
  guid=#1127
  inputPins [
   Key {type=FName}
   Default {type=FString}
  ]
  outputPins [
   IsOverriden {type=bool}
   Int {type=int32}
   IntDefault {type=int32}
  ]
  properties {Default=3, Key="Level 1 - by  Cluster Number"}
 }
 BranchCollection_0 {
  type=FBranchCollectionDataflowNode
  guid=#1128
  inputPins [
   TrueCollection {type=FManagedArrayCollection}
   FalseCollection {type=FManagedArrayCollection}
   bCondition {type=bool}
  ]
  outputPins [ChosenCollection {type=FManagedArrayCollection}]
  properties {bCondition=False}
 }
 IntToString_3 {
  type=FIntToStringDataflowNode
  guid=#1129
  inputPins [Int {type=int32}]
  outputPins [String {type=FString}]
  properties {Int=0}
 }
 LogString_9 {
  type=FLogStringDataflowNode
  guid=#1130
  inputPins [String {type=FString}]
  outputPins []
  properties {String="", bPrintToLog=True}
 }
 GenerateClusterConvexHullsFromLeafHulls_1 {
  type=FGenerateClusterConvexHullsFromLeafHullsDataflowNode
  guid=#1131
  inputPins [
   Collection {type=FManagedArrayCollection}
   ConvexCount {type=int32}
   ErrorTolerance {type=double}
   OptionalSelectionFilter {type=FDataflowTransformSelection}
   bAllowMergingLeafHulls {type=bool}
   RequireNegativeSpaceCovering {type=FDataflowSphereCovering}
   bProtectNegativeSpace {type=bool}
   TargetNumSamples {type=int32}
   MinSampleSpacing {type=double}
   NegativeSpaceTolerance {type=double}
   MinRadius {type=double}
  ]
  outputPins [
   Collection {type=FManagedArrayCollection}
   SphereCovering {type=FDataflowSphereCovering}
  ]
  properties {
   AllowMerges=Any
   ColorRandomSeed=0
   ConvexCount=2
   DebugDrawRenderSettings=(bTranslucent=True,Color=(R=0,G=1,B=0,A=1),LineWidthMultiplier=2)
   ErrorTolerance=0
   MergeProximityDistanceThreshold=0.1
   MergeProximityFilter=None
   MinRadius=10
   MinSampleSpacing=1
   NegativeSpaceTolerance=2
   SampleMethod=Uniform
   SphereCoveringDebugDrawRenderSettings=(bTranslucent=True,LineWidthMultiplier=0.25,Color=(R=1,G=0,B=0,A=1),ColorA=(R=1,G=0,B=0,A=1),ColorB=(R=0,G=0,B=1,A=1))
   TargetNumSamples=50
   bAllowMergingLeafHulls=False
   bOnlyConnectedToHull=False
   bPreferExternalCollisionShapes=False
   bProtectNegativeSpace=False
   bRandomizeColor=True
   bRequireSearchSampleCoverage=False
  }
 }
 Root - Convexes by Number {
  type=FGetIntOverrideFromAssetDataflowNode
  guid=#1132
  inputPins [
   Key {type=FName}
   Default {type=FString}
  ]
  outputPins [
   IsOverriden {type=bool}
   Int {type=int32}
   IntDefault {type=int32}
  ]
  properties {Default=1, Key=Root - Convexes by Number}
 }
 BranchCollection_1 {
  type=FBranchCollectionDataflowNode
  guid=#1133
  inputPins [
   TrueCollection {type=FManagedArrayCollection}
   FalseCollection {type=FManagedArrayCollection}
   bCondition {type=bool}
  ]
  outputPins [ChosenCollection {type=FManagedArrayCollection}]
  properties {bCondition=False}
 }
 BoolToInt {
  type=FBoolToIntDataflowNode
  guid=#1134
  inputPins [Bool {type=bool}]
  outputPins [Int {type=int32}]
  properties {Bool=False}
 }
 Root Convexes by Error Tolerance {
  type=FGetFloatOverrideFromAssetDataflowNode
  guid=#1135
  inputPins [
   Key {type=FName}
   Default {type=FString}
  ]
  outputPins [
   IsOverriden {type=bool}
   Float {type=float}
   FloatDefault {type=float}
  ]
  properties {Default=0, Key=Root - Convexes by Error Tolerance}
 }
 BoolToInt_0 {
  type=FBoolToIntDataflowNode
  guid=#1136
  inputPins [Bool {type=bool}]
  outputPins [Int {type=int32}]
  properties {Bool=False}
 }
 IntToFloat_1 {
  type=FIntToFloatDataflowNode
  guid=#1137
  inputPins [Int {type=int32}]
  outputPins [Float {type=float}]
  properties {Int=0}
 }
 IntToFloat_2 {
  type=FIntToFloatDataflowNode
  guid=#1138
  inputPins [Int {type=int32}]
  outputPins [Float {type=float}]
  properties {Int=0}
 }
 CompareFloat {
  type=FCompareFloatDataflowNode
  guid=#1139
  inputPins [
   FloatA {type=float}
   FloatB {type=float}
  ]
  outputPins [Result {type=bool}]
  properties {FloatA=0, FloatB=0, Operation=Dataflow_Compare_Greater}
 }
 FloatToDouble_2 {
  type=FFloatToDoubleDataflowNode
  guid=#1140
  inputPins [Float {type=float}]
  outputPins [Double {type=double}]
  properties {Float=0}
 }
 Add_4 {
  type=FAddDataflowNode
  guid=#1141
  inputPins [
   FloatA {type=float}
   FloatB {type=float}
  ]
  outputPins [ReturnValue {type=float}]
  properties {FloatA=0, FloatB=0}
 }
 AutoCluster - L1 User Defined {
  type=FAutoClusterDataflowNode
  guid=#1142
  inputPins [
   Collection {type=FManagedArrayCollection}
   TransformSelection {type=FDataflowTransformSelection}
   ClusterSites {type=int32}
   ClusterFraction {type=float}
   SiteSize {type=float}
   ClusterGridWidth {type=int32}
   ClusterGridDepth {type=int32}
   ClusterGridHeight {type=int32}
   MinimumSize {type=float}
   bPreferConvexity {type=bool}
   ConcavityTolerance {type=float}
  ]
  outputPins [Collection {type=FManagedArrayCollection}]
  properties {
   AutoCluster=True
   AvoidIsolated=True
   ClusterFraction=0.25
   ClusterGridDepth=2
   ClusterGridHeight=2
   ClusterGridWidth=2
   ClusterSites=10
   ClusterSizeMethod=Dataflow_ClusterSizeMethod_ByNumber
   ConcavityTolerance=0
   DriftIterations=0
   EnforceSiteParameters=True
   MinimumSize=0
   SiteSize=1
   bPreferConvexity=False
  }
 }
 BranchCollection_2 {
  type=FBranchCollectionDataflowNode
  guid=#1143
  inputPins [
   TrueCollection {type=FManagedArrayCollection}
   FalseCollection {type=FManagedArrayCollection}
   bCondition {type=bool}
  ]
  outputPins [ChosenCollection {type=FManagedArrayCollection}]
  properties {bCondition=False}
 }
 BranchCollection_3 {
  type=FBranchCollectionDataflowNode
  guid=#1144
  inputPins [
   TrueCollection {type=FManagedArrayCollection}
   FalseCollection {type=FManagedArrayCollection}
   bCondition {type=bool}
  ]
  outputPins [ChosenCollection {type=FManagedArrayCollection}]
  properties {bCondition=False}
 }
 Level 1 - Cluster Convexes {
  type=FGetIntOverrideFromAssetDataflowNode
  guid=#1145
  inputPins [
   Key {type=FName}
   Default {type=FString}
  ]
  outputPins [
   IsOverriden {type=bool}
   Int {type=int32}
   IntDefault {type=int32}
  ]
  properties {Default=1, Key=Level 1 - Cluster Convexes}
 }
 BoolToString {
  type=FBoolToStringDataflowNode
  guid=#1146
  inputPins [Bool {type=bool}]
  outputPins [String {type=FString}]
  properties {Bool=False}
 }
 LogString_7 {
  type=FLogStringDataflowNode
  guid=#1147
  inputPins [String {type=FString}]
  outputPins []
  properties {String="", bPrintToLog=True}
 }
 ClusterConvexes - Manual {
  type=FGenerateClusterConvexHullsFromLeafHullsDataflowNode
  guid=#1148
  inputPins [
   Collection {type=FManagedArrayCollection}
   ConvexCount {type=int32}
   ErrorTolerance {type=double}
   OptionalSelectionFilter {type=FDataflowTransformSelection}
   bAllowMergingLeafHulls {type=bool}
   RequireNegativeSpaceCovering {type=FDataflowSphereCovering}
   bProtectNegativeSpace {type=bool}
   TargetNumSamples {type=int32}
   MinSampleSpacing {type=double}
   NegativeSpaceTolerance {type=double}
   MinRadius {type=double}
  ]
  outputPins [
   Collection {type=FManagedArrayCollection}
   SphereCovering {type=FDataflowSphereCovering}
  ]
  properties {
   AllowMerges=Any
   ColorRandomSeed=0
   ConvexCount=1
   DebugDrawRenderSettings=(bTranslucent=True,Color=(R=0,G=1,B=0,A=1),LineWidthMultiplier=2)
   ErrorTolerance=0
   MergeProximityDistanceThreshold=0.1
   MergeProximityFilter=None
   MinRadius=10
   MinSampleSpacing=1
   NegativeSpaceTolerance=2
   SampleMethod=VoxelSearch
   SphereCoveringDebugDrawRenderSettings=(bTranslucent=True,LineWidthMultiplier=0.25,Color=(R=1,G=0,B=0,A=1),ColorA=(R=1,G=0,B=0,A=1),ColorB=(R=0,G=0,B=1,A=1))
   TargetNumSamples=50
   bAllowMergingLeafHulls=False
   bOnlyConnectedToHull=False
   bPreferExternalCollisionShapes=False
   bProtectNegativeSpace=False
   bRandomizeColor=True
   bRequireSearchSampleCoverage=True
  }
 }
 BoolToString_0 {
  type=FBoolToStringDataflowNode
  guid=#1149
  inputPins [Bool {type=bool}]
  outputPins [String {type=FString}]
  properties {Bool=False}
 }
 LogString_10 {
  type=FLogStringDataflowNode
  guid=#1150
  inputPins [String {type=FString}]
  outputPins []
  properties {String="", bPrintToLog=True}
 }
 CollectionTransformSelectRoot_9 {
  type=FCollectionTransformSelectionRootDataflowNode
  guid=#1151
  inputPins [Collection {type=FManagedArrayCollection}]
  outputPins [
   Collection {type=FManagedArrayCollection}
   TransformSelection {type=FDataflowTransformSelection}
  ]
 }
 "Level 1 - by Target Resolution " {
  type=FGetFloatOverrideFromAssetDataflowNode
  guid=#1152
  inputPins [
   Key {type=FName}
   Default {type=FString}
  ]
  outputPins [
   IsOverriden {type=bool}
   Float {type=float}
   FloatDefault {type=float}
  ]
  properties {Default=50, Key=Level 1 - by Target Resolution}
 }
 Convex - Shape Protection Root {
  type=FGetBoolOverrideFromAssetDataflowNode
  guid=#1153
  inputPins [
   Key {type=FName}
   Default {type=FString}
  ]
  outputPins [
   IsOverriden {type=bool}
   Bool {type=bool}
   BoolDefault {type=bool}
  ]
  properties {Default=true, Key=Convex - Shape Protection}
 }
 Convex - Protection Tolerance {
  type=FGetFloatOverrideFromAssetDataflowNode
  guid=#1154
  inputPins [
   Key {type=FName}
   Default {type=FString}
  ]
  outputPins [
   IsOverriden {type=bool}
   Float {type=float}
   FloatDefault {type=float}
  ]
  properties {Default=4, Key=Convex - Protection Tolerance}
 }
 Convex - Protection Radius {
  type=FGetFloatOverrideFromAssetDataflowNode
  guid=#1155
  inputPins [
   Key {type=FName}
   Default {type=FString}
  ]
  outputPins [
   IsOverriden {type=bool}
   Float {type=float}
   FloatDefault {type=float}
  ]
  properties {Default=8, Key=Convex - Protection Radius}
 }
 FloatToDouble {
  type=FFloatToDoubleDataflowNode
  guid=#1156
  inputPins [Float {type=float}]
  outputPins [Double {type=double}]
  properties {Float=0}
 }
 FloatToDouble_3 {
  type=FFloatToDoubleDataflowNode
  guid=#1157
  inputPins [Float {type=float}]
  outputPins [Double {type=double}]
  properties {Float=0}
 }
 FloatToDouble_4 {
  type=FFloatToDoubleDataflowNode
  guid=#1158
  inputPins [Float {type=float}]
  outputPins [Double {type=double}]
  properties {Float=0}
 }
 GenerateClusterConvexHullsFromLeafHulls_3 {
  type=FGenerateClusterConvexHullsFromLeafHullsDataflowNode
  guid=#1159
  inputPins [
   Collection {type=FManagedArrayCollection}
   ConvexCount {type=int32}
   ErrorTolerance {type=double}
   OptionalSelectionFilter {type=FDataflowTransformSelection}
   bAllowMergingLeafHulls {type=bool}
   RequireNegativeSpaceCovering {type=FDataflowSphereCovering}
   bProtectNegativeSpace {type=bool}
   TargetNumSamples {type=int32}
   MinSampleSpacing {type=double}
   NegativeSpaceTolerance {type=double}
   MinRadius {type=double}
  ]
  outputPins [
   Collection {type=FManagedArrayCollection}
   SphereCovering {type=FDataflowSphereCovering}
  ]
  properties {
   AllowMerges=Any
   ColorRandomSeed=0
   ConvexCount=2
   DebugDrawRenderSettings=(bTranslucent=True,Color=(R=0,G=1,B=0,A=1),LineWidthMultiplier=2)
   ErrorTolerance=0
   MergeProximityDistanceThreshold=0.1
   MergeProximityFilter=None
   MinRadius=10
   MinSampleSpacing=1
   NegativeSpaceTolerance=2
   SampleMethod=VoxelSearch
   SphereCoveringDebugDrawRenderSettings=(bTranslucent=True,LineWidthMultiplier=0.25,Color=(R=1,G=0,B=0,A=1),ColorA=(R=1,G=0,B=0,A=1),ColorB=(R=0,G=0,B=1,A=1))
   TargetNumSamples=50
   bAllowMergingLeafHulls=False
   bOnlyConnectedToHull=False
   bPreferExternalCollisionShapes=False
   bProtectNegativeSpace=True
   bRandomizeColor=True
   bRequireSearchSampleCoverage=True
  }
 }
 GetConvexHullVolume {
  type=FGetConvexHullVolumeDataflowNode
  guid=#1160
  inputPins [
   Collection {type=FManagedArrayCollection}
   TransformSelection {type=FDataflowTransformSelection}
  ]
  outputPins [Volume {type=float}]
  properties {
   ColorRandomSeed=0
   DebugDrawRenderSettings=(bTranslucent=True,Color=(R=0,G=1,B=0,A=1),LineWidthMultiplier=2)
   bRandomizeColor=True
   bSumChildrenForClustersWithoutHulls=True
   bVolumeOfUnion=True
  }
 }
 CollectionTransformSelectRoot_11 {
  type=FCollectionTransformSelectionRootDataflowNode
  guid=#1161
  inputPins [Collection {type=FManagedArrayCollection}]
  outputPins [
   Collection {type=FManagedArrayCollection}
   TransformSelection {type=FDataflowTransformSelection}
  ]
 }
 GenerateClusterConvexHullsFromLeafHulls_4 {
  type=FGenerateClusterConvexHullsFromLeafHullsDataflowNode
  guid=#1162
  inputPins [
   Collection {type=FManagedArrayCollection}
   ConvexCount {type=int32}
   ErrorTolerance {type=double}
   OptionalSelectionFilter {type=FDataflowTransformSelection}
   bAllowMergingLeafHulls {type=bool}
   RequireNegativeSpaceCovering {type=FDataflowSphereCovering}
   bProtectNegativeSpace {type=bool}
   TargetNumSamples {type=int32}
   MinSampleSpacing {type=double}
   NegativeSpaceTolerance {type=double}
   MinRadius {type=double}
  ]
  outputPins [
   Collection {type=FManagedArrayCollection}
   SphereCovering {type=FDataflowSphereCovering}
  ]
  properties {
   AllowMerges=Any
   ColorRandomSeed=0
   ConvexCount=1
   DebugDrawRenderSettings=(bTranslucent=True,Color=(R=0,G=1,B=0,A=1),LineWidthMultiplier=2)
   ErrorTolerance=0
   MergeProximityDistanceThreshold=0.1
   MergeProximityFilter=None
   MinRadius=10
   MinSampleSpacing=1
   NegativeSpaceTolerance=2
   SampleMethod=Uniform
   SphereCoveringDebugDrawRenderSettings=(bTranslucent=True,LineWidthMultiplier=0.25,Color=(R=1,G=0,B=0,A=1),ColorA=(R=1,G=0,B=0,A=1),ColorB=(R=0,G=0,B=1,A=1))
   TargetNumSamples=50
   bAllowMergingLeafHulls=False
   bOnlyConnectedToHull=False
   bPreferExternalCollisionShapes=False
   bProtectNegativeSpace=False
   bRandomizeColor=True
   bRequireSearchSampleCoverage=False
  }
 }
 CollectionTransformSelectRoot_12 {
  type=FCollectionTransformSelectionRootDataflowNode
  guid=#1163
  inputPins [Collection {type=FManagedArrayCollection}]
  outputPins [
   Collection {type=FManagedArrayCollection}
   TransformSelection {type=FDataflowTransformSelection}
  ]
 }
 GetConvexHullVolume_0 {
  type=FGetConvexHullVolumeDataflowNode
  guid=#1164
  inputPins [
   Collection {type=FManagedArrayCollection}
   TransformSelection {type=FDataflowTransformSelection}
  ]
  outputPins [Volume {type=float}]
  properties {
   ColorRandomSeed=0
   DebugDrawRenderSettings=(bTranslucent=True,Color=(R=0,G=1,B=0,A=1),LineWidthMultiplier=2)
   bRandomizeColor=True
   bSumChildrenForClustersWithoutHulls=True
   bVolumeOfUnion=True
  }
 }
 1 Over 3 - Convex test {
  type=FDivideDataflowNode
  guid=#1165
  inputPins [
   FloatA {type=float}
   FloatB {type=float}
  ]
  outputPins [ReturnValue {type=float}]
  properties {FloatA=1, FloatB=3}
 }
 Pow_0 {
  type=FPowDataflowNode
  guid=#1166
  inputPins [
   Base {type=float}
   Exp {type=float}
  ]
  outputPins [ReturnValue {type=float}]
  properties {Base=0, Exp=0}
 }
 Pow_1 {
  type=FPowDataflowNode
  guid=#1167
  inputPins [
   Base {type=float}
   Exp {type=float}
  ]
  outputPins [ReturnValue {type=float}]
  properties {Base=0, Exp=0}
 }
 Divide_8 {
  type=FDivideDataflowNode
  guid=#1168
  inputPins [
   FloatA {type=float}
   FloatB {type=float}
  ]
  outputPins [ReturnValue {type=float}]
  properties {FloatA=0, FloatB=0}
 }
 x100 {
  type=FMultiplyDataflowNode
  guid=#1169
  inputPins [
   FloatA {type=float}
   FloatB {type=float}
  ]
  outputPins [ReturnValue {type=float}]
  properties {FloatA=0, FloatB=100}
 }
 FloatToString_20 {
  type=FFloatToStringDataflowNode
  guid=#1170
  inputPins [Float {type=float}]
  outputPins [String {type=FString}]
  properties {Float=0}
 }
 Volume Differential % {
  type=FLogStringDataflowNode
  guid=#1171
  inputPins [String {type=FString}]
  outputPins []
  properties {String="", bPrintToLog=True}
 }
 Differential Driver % {
  type=FDivideDataflowNode
  guid=#1172
  inputPins [
   FloatA {type=float}
   FloatB {type=float}
  ]
  outputPins [ReturnValue {type=float}]
  properties {FloatA=0, FloatB=0}
 }
 Percentage Differential Tolerance {
  type=FMakeLiteralFloatDataflowNode
  guid=#1173
  inputPins []
  outputPins [Float {type=float}]
  properties {Value=60}
 }
 Clamp_9 {
  type=FClampDataflowNode
  guid=#1174
  inputPins [
   Float {type=float}
   Min {type=float}
   Max {type=float}
  ]
  outputPins [ReturnValue {type=float}]
  properties {Float=0, Max=1, Min=0}
 }
 Floor_1 {
  type=FFloorDataflowNode
  guid=#1175
  inputPins [Float {type=float}]
  outputPins [ReturnValue {type=float}]
  properties {Float=0}
 }
 FloatToInt_7 {
  type=FFloatToIntDataflowNode
  guid=#1176
  inputPins [Float {type=float}]
  outputPins [Int {type=int32}]
  properties {Float=0, Function=Dataflow_FloatToInt_Function_Round}
 }
 IntToBool_1 {
  type=FIntToBoolDataflowNode
  guid=#1177
  inputPins [Int {type=int32}]
  outputPins [Bool {type=bool}]
  properties {Int=0}
 }
 BranchCollection_4 {
  type=FBranchCollectionDataflowNode
  guid=#1178
  inputPins [
   TrueCollection {type=FManagedArrayCollection}
   FalseCollection {type=FManagedArrayCollection}
   bCondition {type=bool}
  ]
  outputPins [ChosenCollection {type=FManagedArrayCollection}]
  properties {bCondition=False}
 }
 FloatToString_21 {
  type=FFloatToStringDataflowNode
  guid=#1179
  inputPins [Float {type=float}]
  outputPins [String {type=FString}]
  properties {Float=0}
 }
 Volume Ratio {
  type=FLogStringDataflowNode
  guid=#1180
  inputPins [String {type=FString}]
  outputPins []
  properties {String="", bPrintToLog=True}
 }
 BoolToString_1 {
  type=FBoolToStringDataflowNode
  guid=#1181
  inputPins [Bool {type=bool}]
  outputPins [String {type=FString}]
  properties {Bool=False}
 }
 LogString_1 {
  type=FLogStringDataflowNode
  guid=#1182
  inputPins [String {type=FString}]
  outputPins []
  properties {String="", bPrintToLog=True}
 }
 FloatToString_22 {
  type=FFloatToStringDataflowNode
  guid=#1183
  inputPins [Float {type=float}]
  outputPins [String {type=FString}]
  properties {Float=0}
 }
 Sites - by Leaf Volumes {
  type=FLogStringDataflowNode
  guid=#1184
  inputPins [String {type=FString}]
  outputPins []
  properties {String="", bPrintToLog=True}
 }
 Divide_2 {
  type=FDivideDataflowNode
  guid=#1185
  inputPins [
   FloatA {type=float}
   FloatB {type=float}
  ]
  outputPins [ReturnValue {type=float}]
  properties {FloatA=0, FloatB=0}
 }
 Cubed {
  type=FPowDataflowNode
  guid=#1186
  inputPins [
   Base {type=float}
   Exp {type=float}
  ]
  outputPins [ReturnValue {type=float}]
  properties {Base=0, Exp=3}
 }
 AutoCluster Volume Numeric {
  type=FAutoClusterDataflowNode
  guid=#1187
  inputPins [
   Collection {type=FManagedArrayCollection}
   TransformSelection {type=FDataflowTransformSelection}
   ClusterSites {type=int32}
   ClusterFraction {type=float}
   SiteSize {type=float}
   ClusterGridWidth {type=int32}
   ClusterGridDepth {type=int32}
   ClusterGridHeight {type=int32}
   MinimumSize {type=float}
   bPreferConvexity {type=bool}
   ConcavityTolerance {type=float}
  ]
  outputPins [Collection {type=FManagedArrayCollection}]
  properties {
   AutoCluster=True
   AvoidIsolated=True
   ClusterFraction=0.25
   ClusterGridDepth=2
   ClusterGridHeight=2
   ClusterGridWidth=2
   ClusterSites=10
   ClusterSizeMethod=Dataflow_ClusterSizeMethod_ByNumber
   ConcavityTolerance=0
   DriftIterations=0
   EnforceSiteParameters=True
   MinimumSize=0
   SiteSize=1
   bPreferConvexity=False
  }
 }
 FloatToInt {
  type=FFloatToIntDataflowNode
  guid=#1188
  inputPins [Float {type=float}]
  outputPins [Int {type=int32}]
  properties {Float=0, Function=Dataflow_FloatToInt_Function_Round}
 }
 FloatToString_23 {
  type=FFloatToStringDataflowNode
  guid=#1189
  inputPins [Float {type=float}]
  outputPins [String {type=FString}]
  properties {Float=0}
 }
 Sites - Averaged and then Ratio Scaled {
  type=FLogStringDataflowNode
  guid=#1190
  inputPins [String {type=FString}]
  outputPins []
  properties {String="", bPrintToLog=True}
 }
 Divide_10 {
  type=FDivideDataflowNode
  guid=#1191
  inputPins [
   FloatA {type=float}
   FloatB {type=float}
  ]
  outputPins [ReturnValue {type=float}]
  properties {FloatA=0, FloatB=2}
 }
 Add_5 {
  type=FAddDataflowNode
  guid=#1192
  inputPins [
   FloatA {type=float}
   FloatB {type=float}
  ]
  outputPins [ReturnValue {type=float}]
  properties {FloatA=0, FloatB=0}
 }
 FloatToString_24 {
  type=FFloatToStringDataflowNode
  guid=#1193
  inputPins [Float {type=float}]
  outputPins [String {type=FString}]
  properties {Float=0}
 }
 Sites Averaged {
  type=FLogStringDataflowNode
  guid=#1194
  inputPins [String {type=FString}]
  outputPins []
  properties {String="", bPrintToLog=True}
 }
 BoundingBox_1 {
  type=FBoundingBoxDataflowNode
  guid=#1195
  inputPins [Collection {type=FManagedArrayCollection}]
  outputPins [
   BoundingBox {type=FBox}
   Center {type=FVector}
   Dimensions {type=FVector}
  ]
 }
 ExpandBoundingBox_1 {
  type=FExpandBoundingBoxDataflowNode
  guid=#1196
  inputPins [BoundingBox {type=FBox}]
  outputPins [
   Min {type=FVector}
   Max {type=FVector}
   Center {type=FVector}
   HalfExtents {type=FVector}
   Dimensions {type=FVector}
   Volume {type=float}
  ]
 }
 ExpandVector_0 {
  type=FExpandVectorDataflowNode
  guid=#1197
  inputPins [Vector {type=FVector}]
  outputPins [
   X {type=float}
   Y {type=float}
   Z {type=float}
  ]
 }
 Divide_11 {
  type=FDivideDataflowNode
  guid=#1198
  inputPins [
   FloatA {type=float}
   FloatB {type=float}
  ]
  outputPins [ReturnValue {type=float}]
  properties {FloatA=0, FloatB=0}
 }
 Divide_12 {
  type=FDivideDataflowNode
  guid=#1199
  inputPins [
   FloatA {type=float}
   FloatB {type=float}
  ]
  outputPins [ReturnValue {type=float}]
  properties {FloatA=0, FloatB=0}
 }
 Divide_13 {
  type=FDivideDataflowNode
  guid=#1200
  inputPins [
   FloatA {type=float}
   FloatB {type=float}
  ]
  outputPins [ReturnValue {type=float}]
  properties {FloatA=0, FloatB=0}
 }
 Multiply_2 {
  type=FMultiplyDataflowNode
  guid=#1201
  inputPins [
   FloatA {type=float}
   FloatB {type=float}
  ]
  outputPins [ReturnValue {type=float}]
  properties {FloatA=0, FloatB=2}
 }
 Multiply_5 {
  type=FMultiplyDataflowNode
  guid=#1202
  inputPins [
   FloatA {type=float}
   FloatB {type=float}
  ]
  outputPins [ReturnValue {type=float}]
  properties {FloatA=0, FloatB=2}
 }
 Multiply_6 {
  type=FMultiplyDataflowNode
  guid=#1203
  inputPins [
   FloatA {type=float}
   FloatB {type=float}
  ]
  outputPins [ReturnValue {type=float}]
  properties {FloatA=0, FloatB=2}
 }
 REROUTE - Box Res {
  type=FMultiplyDataflowNode
  guid=#1204
  inputPins [
   FloatA {type=float}
   FloatB {type=float}
  ]
  outputPins [ReturnValue {type=float}]
  properties {FloatA=0, FloatB=1}
 }
 Sites - Volume Grid Cells {
  type=FLogStringDataflowNode
  guid=#1205
  inputPins [String {type=FString}]
  outputPins []
  properties {String="", bPrintToLog=True}
 }
 FloatToString_25 {
  type=FFloatToStringDataflowNode
  guid=#1206
  inputPins [Float {type=float}]
  outputPins [String {type=FString}]
  properties {Float=0}
 }
 Multiply_7 {
  type=FMultiplyDataflowNode
  guid=#1207
  inputPins [
   FloatA {type=float}
   FloatB {type=float}
  ]
  outputPins [ReturnValue {type=float}]
  properties {FloatA=0, FloatB=0}
 }
 Multiply_8 {
  type=FMultiplyDataflowNode
  guid=#1208
  inputPins [
   FloatA {type=float}
   FloatB {type=float}
  ]
  outputPins [ReturnValue {type=float}]
  properties {FloatA=0, FloatB=0}
 }
 Pow_2 {
  type=FPowDataflowNode
  guid=#1209
  inputPins [
   Base {type=float}
   Exp {type=float}
  ]
  outputPins [ReturnValue {type=float}]
  properties {Base=0, Exp=0}
 }
 Divide_14 {
  type=FDivideDataflowNode
  guid=#1210
  inputPins [
   FloatA {type=float}
   FloatB {type=float}
  ]
  outputPins [ReturnValue {type=float}]
  properties {FloatA=0, FloatB=0}
 }
 FloatToString_26 {
  type=FFloatToStringDataflowNode
  guid=#1211
  inputPins [Float {type=float}]
  outputPins [String {type=FString}]
  properties {Float=0}
 }
 GC Fill Ratio {
  type=FLogStringDataflowNode
  guid=#1212
  inputPins [String {type=FString}]
  outputPins []
  properties {String="", bPrintToLog=True}
 }
 Multiply_9 {
  type=FMultiplyDataflowNode
  guid=#1213
  inputPins [
   FloatA {type=float}
   FloatB {type=float}
  ]
  outputPins [ReturnValue {type=float}]
  properties {FloatA=0, FloatB=0}
 }
 BoolToString_2 {
  type=FBoolToStringDataflowNode
  guid=#1214
  inputPins [Bool {type=bool}]
  outputPins [String {type=FString}]
  properties {Bool=False}
 }
 LogString_15 {
  type=FLogStringDataflowNode
  guid=#1215
  inputPins [String {type=FString}]
  outputPins []
  properties {String="", bPrintToLog=True}
 }
 BoolToString_3 {
  type=FBoolToStringDataflowNode
  guid=#1216
  inputPins [Bool {type=bool}]
  outputPins [String {type=FString}]
  properties {Bool=False}
 }
 CollectionTransformSelectChildren_2 {
  type=FCollectionTransformSelectionChildrenDataflowNode
  guid=#1217
  inputPins [
   Collection {type=FManagedArrayCollection}
   TransformSelection {type=FDataflowTransformSelection}
  ]
  outputPins [
   Collection {type=FManagedArrayCollection}
   TransformSelection {type=FDataflowTransformSelection}
  ]
 }
 Min3 {
  type=FMin3DataflowNode
  guid=#1218
  inputPins [
   FloatA {type=float}
   FloatB {type=float}
   FloatC {type=float}
  ]
  outputPins [ReturnValue {type=float}]
  properties {FloatA=0, FloatB=0, FloatC=0}
 }
 FloatToString_27 {
  type=FFloatToStringDataflowNode
  guid=#1219
  inputPins [Float {type=float}]
  outputPins [String {type=FString}]
  properties {Float=0}
 }
 LogString_16 {
  type=FLogStringDataflowNode
  guid=#1220
  inputPins [String {type=FString}]
  outputPins []
  properties {String="", bPrintToLog=True}
 }
 Divide by 2 {
  type=FDivideDataflowNode
  guid=#1221
  inputPins [
   FloatA {type=float}
   FloatB {type=float}
  ]
  outputPins [ReturnValue {type=float}]
  properties {FloatA=0, FloatB=2}
 }
 Clamp {
  type=FClampDataflowNode
  guid=#1222
  inputPins [
   Float {type=float}
   Min {type=float}
   Max {type=float}
  ]
  outputPins [ReturnValue {type=float}]
  properties {Float=0, Max=1, Min=0}
 }
 Floor_2 {
  type=FFloorDataflowNode
  guid=#1223
  inputPins [Float {type=float}]
  outputPins [ReturnValue {type=float}]
  properties {Float=0}
 }
 FloatToInt_8 {
  type=FFloatToIntDataflowNode
  guid=#1224
  inputPins [Float {type=float}]
  outputPins [Int {type=int32}]
  properties {Float=0, Function=Dataflow_FloatToInt_Function_Round}
 }
 IntToBool_2 {
  type=FIntToBoolDataflowNode
  guid=#1225
  inputPins [Int {type=int32}]
  outputPins [Bool {type=bool}]
  properties {Int=0}
 }
 AutoCluster_1 {
  type=FAutoClusterDataflowNode
  guid=#1226
  inputPins [
   Collection {type=FManagedArrayCollection}
   TransformSelection {type=FDataflowTransformSelection}
   ClusterSites {type=int32}
   ClusterFraction {type=float}
   SiteSize {type=float}
   ClusterGridWidth {type=int32}
   ClusterGridDepth {type=int32}
   ClusterGridHeight {type=int32}
   MinimumSize {type=float}
   bPreferConvexity {type=bool}
   ConcavityTolerance {type=float}
  ]
  outputPins [Collection {type=FManagedArrayCollection}]
  properties {
   AutoCluster=True
   AvoidIsolated=True
   ClusterFraction=0.25
   ClusterGridDepth=2
   ClusterGridHeight=2
   ClusterGridWidth=2
   ClusterSites=10
   ClusterSizeMethod=Dataflow_ClusterSizeMethod_ByNumber
   ConcavityTolerance=0
   DriftIterations=0
   EnforceSiteParameters=True
   MinimumSize=22
   SiteSize=1
   bPreferConvexity=False
  }
 }
 FloatToInt_9 {
  type=FFloatToIntDataflowNode
  guid=#1227
  inputPins [Float {type=float}]
  outputPins [Int {type=int32}]
  properties {Float=0, Function=Dataflow_FloatToInt_Function_Round}
 }
 FloatToInt_10 {
  type=FFloatToIntDataflowNode
  guid=#1228
  inputPins [Float {type=float}]
  outputPins [Int {type=int32}]
  properties {Float=0, Function=Dataflow_FloatToInt_Function_Round}
 }
 FloatToInt_11 {
  type=FFloatToIntDataflowNode
  guid=#1229
  inputPins [Float {type=float}]
  outputPins [Int {type=int32}]
  properties {Float=0, Function=Dataflow_FloatToInt_Function_Round}
 }
 BranchCollection_6 {
  type=FBranchCollectionDataflowNode
  guid=#1230
  inputPins [
   TrueCollection {type=FManagedArrayCollection}
   FalseCollection {type=FManagedArrayCollection}
   bCondition {type=bool}
  ]
  outputPins [ChosenCollection {type=FManagedArrayCollection}]
  properties {bCondition=False}
 }
 Min Size {
  type=FMakeLiteralFloatDataflowNode
  guid=#1231
  inputPins []
  outputPins [Float {type=float}]
  properties {Value=20}
 }
 FloatToString_28 {
  type=FFloatToStringDataflowNode
  guid=#1232
  inputPins [Float {type=float}]
  outputPins [String {type=FString}]
  properties {Float=0}
 }
 Volume - by Leaf Convexes {
  type=FLogStringDataflowNode
  guid=#1233
  inputPins [String {type=FString}]
  outputPins []
  properties {String="", bPrintToLog=True}
 }
 GetNumElementsInCollectionGroup_1 {
  type=FGetNumElementsInCollectionGroupDataflowNode
  guid=#1234
  inputPins [Collection {type=FManagedArrayCollection}]
  outputPins [NumElements {type=int32}]
  properties {CustomGroupName="", GroupName=Dataflow_EStandardGroupNameEnum_Transform}
 }
 IntToFloat_3 {
  type=FIntToFloatDataflowNode
  guid=#1235
  inputPins [Int {type=int32}]
  outputPins [Float {type=float}]
  properties {Int=0}
 }
 Subtract {
  type=FSubtractDataflowNode
  guid=#1236
  inputPins [
   FloatA {type=float}
   FloatB {type=float}
  ]
  outputPins [ReturnValue {type=float}]
  properties {FloatA=0, FloatB=1}
 }
 BoolToString_4 {
  type=FBoolToStringDataflowNode
  guid=#1237
  inputPins [Bool {type=bool}]
  outputPins [String {type=FString}]
  properties {Bool=False}
 }
 LogString_0 {
  type=FLogStringDataflowNode
  guid=#1238
  inputPins [String {type=FString}]
  outputPins []
  properties {String="", bPrintToLog=True}
 }
 FloatToInt_1 {
  type=FFloatToIntDataflowNode
  guid=#1239
  inputPins [Float {type=float}]
  outputPins [Int {type=int32}]
  properties {Float=0, Function=Dataflow_FloatToInt_Function_Round}
 }
 FloatToInt_2 {
  type=FFloatToIntDataflowNode
  guid=#1240
  inputPins [Float {type=float}]
  outputPins [Int {type=int32}]
  properties {Float=0, Function=Dataflow_FloatToInt_Function_Round}
 }
 Level 1 - Shape Protection {
  type=FGetBoolOverrideFromAssetDataflowNode
  guid=#1241
  inputPins [
   Key {type=FName}
   Default {type=FString}
  ]
  outputPins [
   IsOverriden {type=bool}
   Bool {type=bool}
   BoolDefault {type=bool}
  ]
  properties {Default=0, Key=Level 1 - Shape Protection}
 }
 FloatToDouble_0 {
  type=FFloatToDoubleDataflowNode
  guid=#1242
  inputPins [Float {type=float}]
  outputPins [Double {type=double}]
  properties {Float=0}
 }
 FloatToDouble_1 {
  type=FFloatToDoubleDataflowNode
  guid=#1243
  inputPins [Float {type=float}]
  outputPins [Double {type=double}]
  properties {Float=0}
 }
 FloatToDouble_5 {
  type=FFloatToDoubleDataflowNode
  guid=#1244
  inputPins [Float {type=float}]
  outputPins [Double {type=double}]
  properties {Float=0}
 }
 ClusterConvexes - Auto {
  type=FGenerateClusterConvexHullsFromLeafHullsDataflowNode
  guid=#1245
  inputPins [
   Collection {type=FManagedArrayCollection}
   ConvexCount {type=int32}
   ErrorTolerance {type=double}
   OptionalSelectionFilter {type=FDataflowTransformSelection}
   bAllowMergingLeafHulls {type=bool}
   RequireNegativeSpaceCovering {type=FDataflowSphereCovering}
   bProtectNegativeSpace {type=bool}
   TargetNumSamples {type=int32}
   MinSampleSpacing {type=double}
   NegativeSpaceTolerance {type=double}
   MinRadius {type=double}
  ]
  outputPins [
   Collection {type=FManagedArrayCollection}
   SphereCovering {type=FDataflowSphereCovering}
  ]
  properties {
   AllowMerges=Any
   ColorRandomSeed=0
   ConvexCount=2
   DebugDrawRenderSettings=(bTranslucent=True,Color=(R=0,G=1,B=0,A=1),LineWidthMultiplier=2)
   ErrorTolerance=32
   MergeProximityDistanceThreshold=0.1
   MergeProximityFilter=None
   MinRadius=8
   MinSampleSpacing=16
   NegativeSpaceTolerance=4
   SampleMethod=VoxelSearch
   SphereCoveringDebugDrawRenderSettings=(bTranslucent=True,LineWidthMultiplier=0.25,Color=(R=1,G=0,B=0,A=1),ColorA=(R=1,G=0,B=0,A=1),ColorB=(R=0,G=0,B=1,A=1))
   TargetNumSamples=40
   bAllowMergingLeafHulls=False
   bOnlyConnectedToHull=False
   bPreferExternalCollisionShapes=False
   bProtectNegativeSpace=True
   bRandomizeColor=True
   bRequireSearchSampleCoverage=True
  }
 }
 "Protection Ratio (95%)" {
  type=FMakeLiteralFloatDataflowNode
  guid=#1246
  inputPins []
  outputPins [Float {type=float}]
  properties {Value=95}
 }
 Divide_0 {
  type=FDivideDataflowNode
  guid=#1247
  inputPins [
   FloatA {type=float}
   FloatB {type=float}
  ]
  outputPins [ReturnValue {type=float}]
  properties {FloatA=0, FloatB=0}
 }
 Clamp_1 {
  type=FClampDataflowNode
  guid=#1248
  inputPins [
   Float {type=float}
   Min {type=float}
   Max {type=float}
  ]
  outputPins [ReturnValue {type=float}]
  properties {Float=0, Max=1, Min=0}
 }
 Floor_3 {
  type=FFloorDataflowNode
  guid=#1249
  inputPins [Float {type=float}]
  outputPins [ReturnValue {type=float}]
  properties {Float=0}
 }
 FloatToInt_0 {
  type=FFloatToIntDataflowNode
  guid=#1250
  inputPins [Float {type=float}]
  outputPins [Int {type=int32}]
  properties {Float=0, Function=Dataflow_FloatToInt_Function_Round}
 }
 IntToBool_3 {
  type=FIntToBoolDataflowNode
  guid=#1251
  inputPins [Int {type=int32}]
  outputPins [Bool {type=bool}]
  properties {Int=0}
 }
 BoolToString_5 {
  type=FBoolToStringDataflowNode
  guid=#1252
  inputPins [Bool {type=bool}]
  outputPins [String {type=FString}]
  properties {Bool=False}
 }
 Gets Shape Protection {
  type=FLogStringDataflowNode
  guid=#1253
  inputPins [String {type=FString}]
  outputPins []
  properties {String="", bPrintToLog=True}
 }
 BoolToString_6 {
  type=FBoolToStringDataflowNode
  guid=#1254
  inputPins [Bool {type=bool}]
  outputPins [String {type=FString}]
  properties {Bool=False}
 }
 Enable Protection? - Level 1 {
  type=FLogStringDataflowNode
  guid=#1255
  inputPins [String {type=FString}]
  outputPins []
  properties {String="", bPrintToLog=True}
 }
 BoolToString_7 {
  type=FBoolToStringDataflowNode
  guid=#1256
  inputPins [Bool {type=bool}]
  outputPins [String {type=FString}]
  properties {Bool=False}
 }
 Enable Protection? - Root {
  type=FLogStringDataflowNode
  guid=#1257
  inputPins [String {type=FString}]
  outputPins []
  properties {String="", bPrintToLog=True}
 }
 Shape Protection Radius {
  type=FMakeLiteralStringDataflowNode
  guid=#1258
  inputPins []
  outputPins [String {type=FString}]
  properties {Value=8}
 }
 MakeLiteralInt {type=FMakeLiteralIntDataflowNode, guid=#1259, inputPins [], outputPins [Int {type=int32}], properties {Value=1}}
 SimplifyConvexHulls_Leaf_QSlim {
  type=FSimplifyConvexHullsDataflowNode
  guid=#1260
  inputPins [
   Collection {type=FManagedArrayCollection}
   OptionalSelectionFilter {type=FDataflowTransformSelection}
   SimplificationAngleThreshold {type=float}
   SimplificationDistanceThreshold {type=float}
   MinTargetTriangleCount {type=int32}
  ]
  outputPins [Collection {type=FManagedArrayCollection}]
  properties {
   ColorRandomSeed=0
   DebugDrawRenderSettings=(bTranslucent=True,Color=(R=0,G=1,B=0,A=1),LineWidthMultiplier=2)
   MinTargetTriangleCount=16
   SimplificationAngleThreshold=30
   SimplificationDistanceThreshold=2
   SimplifyMethod=MeshQSlim
   bRandomizeColor=True
   bUseExistingVertices=True
  }
 }
 Proximity_Standard {
  type=FProximityDataflowNode
  guid=#1261
  inputPins [
   Collection {type=FManagedArrayCollection}
   DistanceThreshold {type=float}
   ContactThreshold {type=float}
  ]
  outputPins [Collection {type=FManagedArrayCollection}]
  properties {
   ContactAreaMethod=Dataflow_ConnectionContactAreaMethod_None
   ContactThreshold=8
   DistanceThreshold=3
   FilterContactMethod=Dataflow_ProximityContactFilteringMethod_ConvexHullSharp
   ProximityMethod=Dataflow_ProximityMethod_ConvexHull
   bRecomputeConvexHulls=False
   bUseAsConnectionGraph=False
  }
 }
 CollectionTransformSelectRoot_10 {
  type=FCollectionTransformSelectionRootDataflowNode
  guid=#1262
  inputPins [Collection {type=FManagedArrayCollection}]
  outputPins [
   Collection {type=FManagedArrayCollection}
   TransformSelection {type=FDataflowTransformSelection}
  ]
 }
 AutoCluster_2 {
  type=FAutoClusterDataflowNode
  guid=#1263
  inputPins [
   Collection {type=FManagedArrayCollection}
   TransformSelection {type=FDataflowTransformSelection}
   ClusterSites {type=int32}
   ClusterFraction {type=float}
   SiteSize {type=float}
   ClusterGridWidth {type=int32}
   ClusterGridDepth {type=int32}
   ClusterGridHeight {type=int32}
   MinimumSize {type=float}
   bPreferConvexity {type=bool}
   ConcavityTolerance {type=float}
  ]
  outputPins [Collection {type=FManagedArrayCollection}]
  properties {
   AutoCluster=True
   AvoidIsolated=True
   ClusterFraction=0.5
   ClusterGridDepth=2
   ClusterGridHeight=2
   ClusterGridWidth=2
   ClusterSites=10
   ClusterSizeMethod=Dataflow_ClusterSizeMethod_ByFractionOfInput
   ConcavityTolerance=0
   DriftIterations=0
   EnforceSiteParameters=True
   MinimumSize=0
   SiteSize=1
   bPreferConvexity=False
  }
 }
 Proximity_5 {
  type=FProximityDataflowNode
  guid=#1264
  inputPins [
   Collection {type=FManagedArrayCollection}
   DistanceThreshold {type=float}
   ContactThreshold {type=float}
  ]
  outputPins [Collection {type=FManagedArrayCollection}]
  properties {
   ContactAreaMethod=Dataflow_ConnectionContactAreaMethod_None
   ContactThreshold=8
   DistanceThreshold=3
   FilterContactMethod=Dataflow_ProximityContactFilteringMethod_ConvexHullSharp
   ProximityMethod=Dataflow_ProximityMethod_ConvexHull
   bRecomputeConvexHulls=False
   bUseAsConnectionGraph=False
  }
 }
 2x single plate {
  type=FMakeLiteralFloatDataflowNode
  guid=#1265
  inputPins []
  outputPins [Float {type=float}]
  properties {Value=14}
 }
 IntToString {
  type=FIntToStringDataflowNode
  guid=#1266
  inputPins [Int {type=int32}]
  outputPins [String {type=FString}]
  properties {Int=0}
 }
 Incoming Cells {
  type=FLogStringDataflowNode
  guid=#1267
  inputPins [String {type=FString}]
  outputPins []
  properties {String="", bPrintToLog=True}
 }
 UpdateVolumeAttributes {
  type=FUpdateVolumeAttributesDataflowNode
  guid=#1268
  inputPins [Collection {type=FManagedArrayCollection}]
  outputPins [Collection {type=FManagedArrayCollection}]
 }
 "Used Light Proximity to start = " {
  type=FStringAppendDataflowNode
  guid=#1269
  inputPins [
   String1 {type=FString}
   String2 {type=FString}
  ]
  outputPins [String {type=FString}]
  properties {String1="Chose Regular Proximity = ", String2=""}
 }
 Was Grid Clustered {
  type=FStringAppendDataflowNode
  guid=#1270
  inputPins [
   String1 {type=FString}
   String2 {type=FString}
  ]
  outputPins [String {type=FString}]
  properties {String1="Skipped Grid Clustering = ", String2=""}
 }
 Path Report {
  type=FLogStringDataflowNode
  guid=#1271
  inputPins [String {type=FString}]
  outputPins []
  properties {String="", bPrintToLog=True}
 }
 StringAppend {
  type=FStringAppendDataflowNode
  guid=#1272
  inputPins [
   String1 {type=FString}
   String2 {type=FString}
  ]
  outputPins [String {type=FString}]
  properties {String1="", String2=""}
 }
 StringAppend_0 {
  type=FStringAppendDataflowNode
  guid=#1273
  inputPins [
   String1 {type=FString}
   String2 {type=FString}
  ]
  outputPins [String {type=FString}]
  properties {String1="", String2=" , "}
 }
 StringAppend_1 {
  type=FStringAppendDataflowNode
  guid=#1274
  inputPins [
   String1 {type=FString}
   String2 {type=FString}
  ]
  outputPins [String {type=FString}]
  properties {String1="", String2=" , "}
 }
 StringAppend_2 {
  type=FStringAppendDataflowNode
  guid=#1275
  inputPins [
   String1 {type=FString}
   String2 {type=FString}
  ]
  outputPins [String {type=FString}]
  properties {String1="MANUAL OVERRIDE CLUSTERING = ", String2=""}
 }
 StringAppend_3 {
  type=FStringAppendDataflowNode
  guid=#1276
  inputPins [
   String1 {type=FString}
   String2 {type=FString}
  ]
  outputPins [String {type=FString}]
  properties {String1="", String2=" , "}
 }
 StringAppend_4 {
  type=FStringAppendDataflowNode
  guid=#1277
  inputPins [
   String1 {type=FString}
   String2 {type=FString}
  ]
  outputPins [String {type=FString}]
  properties {String1="", String2=""}
 }
 StringAppend_5 {
  type=FStringAppendDataflowNode
  guid=#1278
  inputPins [
   String1 {type=FString}
   String2 {type=FString}
  ]
  outputPins [String {type=FString}]
  properties {String1="Activated Shape Protection = ", String2=""}
 }
 StringAppend_6 {
  type=FStringAppendDataflowNode
  guid=#1279
  inputPins [
   String1 {type=FString}
   String2 {type=FString}
  ]
  outputPins [String {type=FString}]
  properties {String1="Skipped SMALL OBJECT PATH = ", String2=""}
 }
 StringAppend_7 {
  type=FStringAppendDataflowNode
  guid=#1280
  inputPins [
   String1 {type=FString}
   String2 {type=FString}
  ]
  outputPins [String {type=FString}]
  properties {String1="", String2=" , "}
 }
 StringAppend_8 {
  type=FStringAppendDataflowNode
  guid=#1281
  inputPins [
   String1 {type=FString}
   String2 {type=FString}
  ]
  outputPins [String {type=FString}]
  properties {String1="", String2=""}
 }
 StringAppend_9 {
  type=FStringAppendDataflowNode
  guid=#1282
  inputPins [
   String1 {type=FString}
   String2 {type=FString}
  ]
  outputPins [String {type=FString}]
  properties {String1="", String2=" , "}
 }
 StringAppend_10 {
  type=FStringAppendDataflowNode
  guid=#1283
  inputPins [
   String1 {type=FString}
   String2 {type=FString}
  ]
  outputPins [String {type=FString}]
  properties {String1="", String2=""}
 }
 StringAppend_11 {
  type=FStringAppendDataflowNode
  guid=#1284
  inputPins [
   String1 {type=FString}
   String2 {type=FString}
  ]
  outputPins [String {type=FString}]
  properties {String1="MANUAL OVERRIDE L1 Convexes = ", String2=""}
 }
 StringAppend_12 {
  type=FStringAppendDataflowNode
  guid=#1285
  inputPins [
   String1 {type=FString}
   String2 {type=FString}
  ]
  outputPins [String {type=FString}]
  properties {String1="", String2=" , "}
 }
 StringAppend_13 {
  type=FStringAppendDataflowNode
  guid=#1286
  inputPins [
   String1 {type=FString}
   String2 {type=FString}
  ]
  outputPins [String {type=FString}]
  properties {String1="", String2=""}
 }
 BoolToString_8 {
  type=FBoolToStringDataflowNode
  guid=#1287
  inputPins [Bool {type=bool}]
  outputPins [String {type=FString}]
  properties {Bool=False}
 }
 StringAppend_14 {
  type=FStringAppendDataflowNode
  guid=#1288
  inputPins [
   String1 {type=FString}
   String2 {type=FString}
  ]
  outputPins [String {type=FString}]
  properties {String1="Chose Root Single Convex = ", String2=""}
 }
 LogString_6 {
  type=FLogStringDataflowNode
  guid=#1289
  inputPins [String {type=FString}]
  outputPins []
  properties {String="", bPrintToLog=True}
 }
 StringAppend_15 {
  type=FStringAppendDataflowNode
  guid=#1290
  inputPins [
   String1 {type=FString}
   String2 {type=FString}
  ]
  outputPins [String {type=FString}]
  properties {String1="", String2=" , "}
 }
 StringAppend_16 {
  type=FStringAppendDataflowNode
  guid=#1291
  inputPins [
   String1 {type=FString}
   String2 {type=FString}
  ]
  outputPins [String {type=FString}]
  properties {String1="", String2=""}
 }
 StringAppend_17 {
  type=FStringAppendDataflowNode
  guid=#1292
  inputPins [
   String1 {type=FString}
   String2 {type=FString}
  ]
  outputPins [String {type=FString}]
  properties {String1="MANUAL OVERRIDE ROOT CONVEX = ", String2=""}
 }
 StringAppend_18 {
  type=FStringAppendDataflowNode
  guid=#1293
  inputPins [
   String1 {type=FString}
   String2 {type=FString}
  ]
  outputPins [String {type=FString}]
  properties {String1="", String2=" , "}
 }
 StringAppend_19 {
  type=FStringAppendDataflowNode
  guid=#1294
  inputPins [
   String1 {type=FString}
   String2 {type=FString}
  ]
  outputPins [String {type=FString}]
  properties {String1="", String2=""}
 }
 UpdateVolumeAttributes_0 {
  type=FUpdateVolumeAttributesDataflowNode
  guid=#1295
  inputPins [Collection {type=FManagedArrayCollection}]
  outputPins [Collection {type=FManagedArrayCollection}]
 }
 UpdateVolumeAttributes_1 {
  type=FUpdateVolumeAttributesDataflowNode
  guid=#1296
  inputPins [Collection {type=FManagedArrayCollection}]
  outputPins [Collection {type=FManagedArrayCollection}]
 }
 FloatToString {
  type=FFloatToStringDataflowNode
  guid=#1297
  inputPins [Float {type=float}]
  outputPins [String {type=FString}]
  properties {Float=0}
 }
 LogString_5 {
  type=FLogStringDataflowNode
  guid=#1298
  inputPins [String {type=FString}]
  outputPins []
  properties {String="", bPrintToLog=True}
 }
 FloatToString_1 {
  type=FFloatToStringDataflowNode
  guid=#1299
  inputPins [Float {type=float}]
  outputPins [String {type=FString}]
  properties {Float=0}
 }
 Volume - by Root {
  type=FLogStringDataflowNode
  guid=#1300
  inputPins [String {type=FString}]
  outputPins []
  properties {String="", bPrintToLog=True}
 }
 Sphere Covering to Mesh {
  type=FSphereCoveringToMeshDataflowNode
  guid=#1301
  inputPins [SphereCovering {type=FDataflowSphereCovering}]
  outputPins [Mesh {type="TObjectPtr<UDynamicMesh>"}]
  properties {VerticesAlongEachSide=8}
 }
 Convert Mesh to OBJ String {
  type=FMeshToOBJStringDebugDataflowNode
  guid=#1302
  inputPins [
   Mesh {type="TObjectPtr<UDynamicMesh>"}
   bInvertFaces {type=bool}
  ]
  outputPins [StringOBJ {type=FString}]
  properties {bInvertFaces=False}
 }
 Write String to File {
  type=FWriteStringToFile
  guid=#1303
  inputPins [
   FileContents {type=FString}
   FilePath {type=FString}
  ]
  outputPins []
  properties {FileContents="", FilePath="C:\\Juno\\DF Mesh Exports\\SphereProtection_Output.obj"}
 }
 Simplification Angle {
  type=FMakeLiteralFloatDataflowNode
  guid=#1304
  inputPins []
  outputPins [Float {type=float}]
  properties {Value=15}
 }
 UpdateVolumeAttributes_2 {
  type=FUpdateVolumeAttributesDataflowNode
  guid=#1305
  inputPins [Collection {type=FManagedArrayCollection}]
  outputPins [Collection {type=FManagedArrayCollection}]
 }
 Empty Space Driver {
  type=FMakeLiteralFloatDataflowNode
  guid=#1306
  inputPins []
  outputPins [Float {type=float}]
  properties {Value=5}
 }
 Write String to File_0 {
  type=FWriteStringToFile
  guid=#1307
  inputPins [
   FileContents {type=FString}
   FilePath {type=FString}
  ]
  outputPins []
  properties {FileContents="", FilePath="C:\\Juno\\DF Mesh Exports\\SphereProtection_Output.obj"}
 }
 Sphere Covering to Mesh_0 {
  type=FSphereCoveringToMeshDataflowNode
  guid=#1308
  inputPins [SphereCovering {type=FDataflowSphereCovering}]
  outputPins [Mesh {type="TObjectPtr<UDynamicMesh>"}]
  properties {VerticesAlongEachSide=8}
 }
 Convert Mesh to OBJ String_0 {
  type=FMeshToOBJStringDebugDataflowNode
  guid=#1309
  inputPins [
   Mesh {type="TObjectPtr<UDynamicMesh>"}
   bInvertFaces {type=bool}
  ]
  outputPins [StringOBJ {type=FString}]
  properties {bInvertFaces=False}
 }
 Write String to File_1 {
  type=FWriteStringToFile
  guid=#1310
  inputPins [
   FileContents {type=FString}
   FilePath {type=FString}
  ]
  outputPins []
  properties {FileContents="", FilePath="C:\\Juno\\DF Mesh Exports\\SphereProtection_L1Clusters_Output.obj"}
 }
 SimplifyConvexHulls Angle Tolerance {
  type=FSimplifyConvexHullsDataflowNode
  guid=#1311
  inputPins [
   Collection {type=FManagedArrayCollection}
   OptionalSelectionFilter {type=FDataflowTransformSelection}
   SimplificationAngleThreshold {type=float}
   SimplificationDistanceThreshold {type=float}
   MinTargetTriangleCount {type=int32}
  ]
  outputPins [Collection {type=FManagedArrayCollection}]
  properties {
   ColorRandomSeed=0
   DebugDrawRenderSettings=(bTranslucent=True,Color=(R=0,G=1,B=0,A=1),LineWidthMultiplier=2)
   MinTargetTriangleCount=16
   SimplificationAngleThreshold=22.5
   SimplificationDistanceThreshold=50
   SimplifyMethod=AngleTolerance
   bRandomizeColor=True
   bUseExistingVertices=True
  }
 }
 BranchCollection_5 {
  type=FBranchCollectionDataflowNode
  guid=#1312
  inputPins [
   TrueCollection {type=FManagedArrayCollection}
   FalseCollection {type=FManagedArrayCollection}
   bCondition {type=bool}
  ]
  outputPins [ChosenCollection {type=FManagedArrayCollection}]
  properties {bCondition=False}
 }
 CreateLeafConvexHulls_Mid {
  type=FCreateLeafConvexHullsDataflowNode
  guid=#1313
  inputPins [
   Collection {type=FManagedArrayCollection}
   OptionalSelectionFilter {type=FDataflowTransformSelection}
   SimplificationDistanceThreshold {type=float}
   ConvexDecompositionSettings {type=FDataflowConvexDecompositionSettings}
   RequireNegativeSpaceCovering {type=FDataflowSphereCovering}
  ]
  outputPins [
   Collection {type=FManagedArrayCollection}
   SphereCovering {type=FDataflowSphereCovering}
  ]
  properties {
   ColorRandomSeed=0
   ConvexDecompositionSettings=(MaxGeoToHullVolumeRatioToDecompose=1,ErrorTolerance=9.4,MaxHullsPerGeometry=10,MinThicknessTolerance=2,NumAdditionalSplits=3,bOnlyConnectedToHull=True,NegativeSpaceTolerance=2,NegativeSpaceMinRadius=10)
   DebugDrawRenderSettings=(bTranslucent=True,Color=(R=0,G=1,B=0,A=1),LineWidthMultiplier=2)
   GenerateMethod=IntersectExternalWithComputed
   IntersectIfComputedIsSmallerByFactor=1
   MinExternalVolumeToIntersect=0
   SimplificationDistanceThreshold=0
   SphereCoveringDebugDrawRenderSettings=(bTranslucent=True,LineWidthMultiplier=0.25,Color=(R=1,G=0,B=0,A=1),ColorA=(R=1,G=0,B=0,A=1),ColorB=(R=0,G=0,B=1,A=1))
   bComputeIntersectionsBeforeHull=True
   bRandomizeColor=True
  }
 }
 Small Leaf Size {
  type=FMakeLiteralFloatDataflowNode
  guid=#1314
  inputPins []
  outputPins [Float {type=float}]
  properties {Value=22}
 }
 Mid Leaf Size {
  type=FMakeLiteralFloatDataflowNode
  guid=#1315
  inputPins []
  outputPins [Float {type=float}]
  properties {Value=34}
 }
 CreateLeafConvexHulls_Large {
  type=FCreateLeafConvexHullsDataflowNode
  guid=#1316
  inputPins [
   Collection {type=FManagedArrayCollection}
   OptionalSelectionFilter {type=FDataflowTransformSelection}
   SimplificationDistanceThreshold {type=float}
   ConvexDecompositionSettings {type=FDataflowConvexDecompositionSettings}
   RequireNegativeSpaceCovering {type=FDataflowSphereCovering}
  ]
  outputPins [
   Collection {type=FManagedArrayCollection}
   SphereCovering {type=FDataflowSphereCovering}
  ]
  properties {
   ColorRandomSeed=0
   ConvexDecompositionSettings=(MaxGeoToHullVolumeRatioToDecompose=0.4,ErrorTolerance=9.4,MaxHullsPerGeometry=16,MinThicknessTolerance=1.6,NumAdditionalSplits=4,bOnlyConnectedToHull=True,NegativeSpaceTolerance=2,NegativeSpaceMinRadius=10)
   DebugDrawRenderSettings=(bTranslucent=True,Color=(R=0,G=1,B=0,A=1),LineWidthMultiplier=2)
   GenerateMethod=IntersectExternalWithComputed
   IntersectIfComputedIsSmallerByFactor=1
   MinExternalVolumeToIntersect=0
   SimplificationDistanceThreshold=0
   SphereCoveringDebugDrawRenderSettings=(bTranslucent=True,LineWidthMultiplier=0.25,Color=(R=1,G=0,B=0,A=1),ColorA=(R=1,G=0,B=0,A=1),ColorB=(R=0,G=0,B=1,A=1))
   bComputeIntersectionsBeforeHull=True
   bRandomizeColor=True
  }
 }
 Max Leaf Range {
  type=FMakeLiteralFloatDataflowNode
  guid=#1317
  inputPins []
  outputPins [Float {type=float}]
  properties {Value=20000}
 }
 DEBUG -- BYPASS FINAL SIMPLIFICATION {
  type=FMakeLiteralBoolDataflowNode
  guid=#1318
  inputPins []
  outputPins [Bool {type=bool}]
  properties {Value=False}
 }
 CreateLeafConvexHulls_Tiny {
  type=FCreateLeafConvexHullsDataflowNode
  guid=#1319
  inputPins [
   Collection {type=FManagedArrayCollection}
   OptionalSelectionFilter {type=FDataflowTransformSelection}
   SimplificationDistanceThreshold {type=float}
   ConvexDecompositionSettings {type=FDataflowConvexDecompositionSettings}
   RequireNegativeSpaceCovering {type=FDataflowSphereCovering}
  ]
  outputPins [
   Collection {type=FManagedArrayCollection}
   SphereCovering {type=FDataflowSphereCovering}
  ]
  properties {
   ColorRandomSeed=0
   ConvexDecompositionSettings=(MinSizeToDecompose=14.5,MaxGeoToHullVolumeRatioToDecompose=0.7,ErrorTolerance=9.4,MaxHullsPerGeometry=3,MinThicknessTolerance=5,bOnlyConnectedToHull=True,NegativeSpaceTolerance=2,NegativeSpaceMinRadius=10)
   DebugDrawRenderSettings=(bTranslucent=True,Color=(R=0,G=1,B=0,A=1),LineWidthMultiplier=2)
   GenerateMethod=IntersectExternalWithComputed
   IntersectIfComputedIsSmallerByFactor=1
   MinExternalVolumeToIntersect=3000
   SimplificationDistanceThreshold=0
   SphereCoveringDebugDrawRenderSettings=(bTranslucent=True,LineWidthMultiplier=0.25,Color=(R=1,G=0,B=0,A=1),ColorA=(R=1,G=0,B=0,A=1),ColorB=(R=0,G=0,B=1,A=1))
   bComputeIntersectionsBeforeHull=True
   bRandomizeColor=True
  }
 }
 Tiny Leaf Size {
  type=FMakeLiteralFloatDataflowNode
  guid=#1320
  inputPins []
  outputPins [Float {type=float}]
  properties {Value=19.45}
 }
 SimplifyConvexHulls_Leaf_Angle {
  type=FSimplifyConvexHullsDataflowNode
  guid=#1321
  inputPins [
   Collection {type=FManagedArrayCollection}
   OptionalSelectionFilter {type=FDataflowTransformSelection}
   SimplificationAngleThreshold {type=float}
   SimplificationDistanceThreshold {type=float}
   MinTargetTriangleCount {type=int32}
  ]
  outputPins [Collection {type=FManagedArrayCollection}]
  properties {
   ColorRandomSeed=0
   DebugDrawRenderSettings=(bTranslucent=True,Color=(R=0,G=1,B=0,A=1),LineWidthMultiplier=2)
   MinTargetTriangleCount=6
   SimplificationAngleThreshold=20
   SimplificationDistanceThreshold=1.6
   SimplifyMethod=AngleTolerance
   bRandomizeColor=True
   bUseExistingVertices=True
  }
 }
 MergeConvexHulls_copy {
  type=FMergeConvexHullsDataflowNode
  guid=#1322
  inputPins [
   Collection {type=FManagedArrayCollection}
   MaxConvexCount {type=int32}
   ErrorTolerance {type=double}
   OptionalSelectionFilter {type=FDataflowTransformSelection}
   bProtectNegativeSpace {type=bool}
   RequireNegativeSpaceCovering {type=FDataflowSphereCovering}
   TargetNumSamples {type=int32}
   MinSampleSpacing {type=double}
   NegativeSpaceTolerance {type=double}
   MinRadius {type=double}
  ]
  outputPins [
   Collection {type=FManagedArrayCollection}
   SphereCovering {type=FDataflowSphereCovering}
  ]
  properties {
   ColorRandomSeed=0
   DebugDrawRenderSettings=(bTranslucent=True,Color=(R=0,G=1,B=0,A=1),LineWidthMultiplier=2)
   ErrorTolerance=64
   MaxConvexCount=-1
   MergeProximityDistanceThreshold=0.1
   MergeProximityFilter=None
   MinRadius=8
   MinSampleSpacing=8
   NegativeSpaceTolerance=2
   SampleMethod=VoxelSearch
   SphereCoveringDebugDrawRenderSettings=(bTranslucent=True,LineWidthMultiplier=0.25,Color=(R=1,G=0,B=0,A=1),ColorA=(R=1,G=0,B=0,A=1),ColorB=(R=0,G=0,B=1,A=1))
   TargetNumSamples=50
   bComputeNegativeSpacePerBone=True
   bOnlyConnectedToHull=False
   bProtectNegativeSpace=True
   bRandomizeColor=True
   bRequireSearchSampleCoverage=True
  }
 }
 CollectionTransformSelectRoot_14 {
  type=FCollectionTransformSelectionRootDataflowNode
  guid=#1323
  inputPins [Collection {type=FManagedArrayCollection}]
  outputPins [
   Collection {type=FManagedArrayCollection}
   TransformSelection {type=FDataflowTransformSelection}
  ]
 }
 MergeConvexHulls_0 {
  type=FMergeConvexHullsDataflowNode
  guid=#1324
  inputPins [
   Collection {type=FManagedArrayCollection}
   MaxConvexCount {type=int32}
   ErrorTolerance {type=double}
   OptionalSelectionFilter {type=FDataflowTransformSelection}
   bProtectNegativeSpace {type=bool}
   RequireNegativeSpaceCovering {type=FDataflowSphereCovering}
   TargetNumSamples {type=int32}
   MinSampleSpacing {type=double}
   NegativeSpaceTolerance {type=double}
   MinRadius {type=double}
  ]
  outputPins [
   Collection {type=FManagedArrayCollection}
   SphereCovering {type=FDataflowSphereCovering}
  ]
  properties {
   ColorRandomSeed=0
   DebugDrawRenderSettings=(bTranslucent=True,Color=(R=0,G=1,B=0,A=1),LineWidthMultiplier=2)
   ErrorTolerance=0
   MaxConvexCount=-1
   MergeProximityDistanceThreshold=0.1
   MergeProximityFilter=None
   MinRadius=1.6
   MinSampleSpacing=1.6
   NegativeSpaceTolerance=3.2
   SampleMethod=VoxelSearch
   SphereCoveringDebugDrawRenderSettings=(bTranslucent=True,LineWidthMultiplier=0.25,Color=(R=1,G=0,B=0,A=1),ColorA=(R=1,G=0,B=0,A=1),ColorB=(R=0,G=0,B=1,A=1))
   TargetNumSamples=40
   bComputeNegativeSpacePerBone=False
   bOnlyConnectedToHull=False
   bProtectNegativeSpace=True
   bRandomizeColor=True
   bRequireSearchSampleCoverage=True
  }
 }
 MergeConvexHulls_0_copy {
  type=FMergeConvexHullsDataflowNode
  guid=#1325
  inputPins [
   Collection {type=FManagedArrayCollection}
   MaxConvexCount {type=int32}
   ErrorTolerance {type=double}
   OptionalSelectionFilter {type=FDataflowTransformSelection}
   bProtectNegativeSpace {type=bool}
   RequireNegativeSpaceCovering {type=FDataflowSphereCovering}
   TargetNumSamples {type=int32}
   MinSampleSpacing {type=double}
   NegativeSpaceTolerance {type=double}
   MinRadius {type=double}
  ]
  outputPins [
   Collection {type=FManagedArrayCollection}
   SphereCovering {type=FDataflowSphereCovering}
  ]
  properties {
   ColorRandomSeed=0
   DebugDrawRenderSettings=(bTranslucent=True,Color=(R=0,G=1,B=0,A=1),LineWidthMultiplier=2)
   ErrorTolerance=0
   MaxConvexCount=-1
   MergeProximityDistanceThreshold=0.1
   MergeProximityFilter=None
   MinRadius=8
   MinSampleSpacing=16
   NegativeSpaceTolerance=4
   SampleMethod=VoxelSearch
   SphereCoveringDebugDrawRenderSettings=(bTranslucent=True,LineWidthMultiplier=0.25,Color=(R=1,G=0,B=0,A=1),ColorA=(R=1,G=0,B=0,A=1),ColorB=(R=0,G=0,B=1,A=1))
   TargetNumSamples=40
   bComputeNegativeSpacePerBone=True
   bOnlyConnectedToHull=False
   bProtectNegativeSpace=True
   bRandomizeColor=True
   bRequireSearchSampleCoverage=False
  }
 }
 Subtract_0 {
  type=FSubtractDataflowNode
  guid=#1326
  inputPins [
   FloatA {type=float}
   FloatB {type=float}
  ]
  outputPins [ReturnValue {type=float}]
  properties {FloatA=0, FloatB=0}
 }
 FloatToString_28_copy {
  type=FFloatToStringDataflowNode
  guid=#1327
  inputPins [Float {type=float}]
  outputPins [String {type=FString}]
  properties {Float=0}
 }
 Size of Empty Space {
  type=FLogStringDataflowNode
  guid=#1328
  inputPins [String {type=FString}]
  outputPins []
  properties {String="", bPrintToLog=True}
 }
 SimplifyConvexHulls_Leaf_QSlim_Tiny {
  type=FSimplifyConvexHullsDataflowNode
  guid=#1329
  inputPins [
   Collection {type=FManagedArrayCollection}
   OptionalSelectionFilter {type=FDataflowTransformSelection}
   SimplificationAngleThreshold {type=float}
   SimplificationDistanceThreshold {type=float}
   MinTargetTriangleCount {type=int32}
  ]
  outputPins [Collection {type=FManagedArrayCollection}]
  properties {
   ColorRandomSeed=0
   DebugDrawRenderSettings=(bTranslucent=True,Color=(R=0,G=1,B=0,A=1),LineWidthMultiplier=2)
   MinTargetTriangleCount=16
   SimplificationAngleThreshold=30
   SimplificationDistanceThreshold=1.6
   SimplifyMethod=MeshQSlim
   bRandomizeColor=True
   bUseExistingVertices=True
  }
 }
 SimplifyConvexHulls_Leaf_QSlim_Small {
  type=FSimplifyConvexHullsDataflowNode
  guid=#1330
  inputPins [
   Collection {type=FManagedArrayCollection}
   OptionalSelectionFilter {type=FDataflowTransformSelection}
   SimplificationAngleThreshold {type=float}
   SimplificationDistanceThreshold {type=float}
   MinTargetTriangleCount {type=int32}
  ]
  outputPins [Collection {type=FManagedArrayCollection}]
  properties {
   ColorRandomSeed=0
   DebugDrawRenderSettings=(bTranslucent=True,Color=(R=0,G=1,B=0,A=1),LineWidthMultiplier=2)
   MinTargetTriangleCount=16
   SimplificationAngleThreshold=30
   SimplificationDistanceThreshold=1.6
   SimplifyMethod=MeshQSlim
   bRandomizeColor=True
   bUseExistingVertices=True
  }
 }
 CollectionTransformSelectBySize {
  type=FCollectionTransformSelectionBySizeDataflowNode
  guid=#1331
  inputPins [
   Collection {type=FManagedArrayCollection}
   SizeMin {type=float}
   SizeMax {type=float}
  ]
  outputPins [
   Collection {type=FManagedArrayCollection}
   TransformSelection {type=FDataflowTransformSelection}
  ]
  properties {RangeSetting=Dataflow_RangeSetting_InsideRange, SizeMax=1000, SizeMin=0, bInclusive=True, bUseRelativeSize=False}
 }
 CollectionTransformSelectBySize_Small {
  type=FCollectionTransformSelectionBySizeDataflowNode
  guid=#1332
  inputPins [
   Collection {type=FManagedArrayCollection}
   SizeMin {type=float}
   SizeMax {type=float}
  ]
  outputPins [
   Collection {type=FManagedArrayCollection}
   TransformSelection {type=FDataflowTransformSelection}
  ]
  properties {RangeSetting=Dataflow_RangeSetting_InsideRange, SizeMax=1000, SizeMin=0, bInclusive=True, bUseRelativeSize=False}
 }
 CollectionTransformSelectBySize_Small_copy {
  type=FCollectionTransformSelectionBySizeDataflowNode
  guid=#1333
  inputPins [
   Collection {type=FManagedArrayCollection}
   SizeMin {type=float}
   SizeMax {type=float}
  ]
  outputPins [
   Collection {type=FManagedArrayCollection}
   TransformSelection {type=FDataflowTransformSelection}
  ]
  properties {RangeSetting=Dataflow_RangeSetting_InsideRange, SizeMax=1000, SizeMin=0, bInclusive=True, bUseRelativeSize=False}
 }
 CollectionTransformSelectBySize_Small_copy_0 {
  type=FCollectionTransformSelectionBySizeDataflowNode
  guid=#1334
  inputPins [
   Collection {type=FManagedArrayCollection}
   SizeMin {type=float}
   SizeMax {type=float}
  ]
  outputPins [
   Collection {type=FManagedArrayCollection}
   TransformSelection {type=FDataflowTransformSelection}
  ]
  properties {RangeSetting=Dataflow_RangeSetting_InsideRange, SizeMax=1000, SizeMin=0, bInclusive=True, bUseRelativeSize=False}
 }
 StringAppend_20 {
  type=FStringAppendDataflowNode
  guid=#1335
  inputPins [
   String1 {type=FString}
   String2 {type=FString}
  ]
  outputPins [String {type=FString}]
  properties {String1="Volume Differential = ", String2=""}
 }
 StringAppend_20_copy {
  type=FStringAppendDataflowNode
  guid=#1336
  inputPins [
   String1 {type=FString}
   String2 {type=FString}
  ]
  outputPins [String {type=FString}]
  properties {String1="", String2="%, "}
 }
 LogAttributes {
  type=FLogStringDataflowNode
  guid=#1337
  inputPins [String {type=FString}]
  outputPins []
  properties {String="", bPrintToLog=True}
 }
 StringAppend_20_copy_0 {
  type=FStringAppendDataflowNode
  guid=#1338
  inputPins [
   String1 {type=FString}
   String2 {type=FString}
  ]
  outputPins [String {type=FString}]
  properties {String1="Size Convexes = ", String2=""}
 }
 StringAppend_20_copy_1 {
  type=FStringAppendDataflowNode
  guid=#1339
  inputPins [
   String1 {type=FString}
   String2 {type=FString}
  ]
  outputPins [String {type=FString}]
  properties {String1="", String2=", "}
 }
 StringAppend_20_copy_0_0 {
  type=FStringAppendDataflowNode
  guid=#1340
  inputPins [
   String1 {type=FString}
   String2 {type=FString}
  ]
  outputPins [String {type=FString}]
  properties {String1="Empty Space = ", String2=""}
 }
 StringAppend_20_copy_1_0 {
  type=FStringAppendDataflowNode
  guid=#1341
  inputPins [
   String1 {type=FString}
   String2 {type=FString}
  ]
  outputPins [String {type=FString}]
  properties {String1="", String2=", "}
 }
 StringAppend_21 {
  type=FStringAppendDataflowNode
  guid=#1342
  inputPins [
   String1 {type=FString}
   String2 {type=FString}
  ]
  outputPins [String {type=FString}]
  properties {String1="", String2=""}
 }
 StringAppend_21_copy {
  type=FStringAppendDataflowNode
  guid=#1343
  inputPins [
   String1 {type=FString}
   String2 {type=FString}
  ]
  outputPins [String {type=FString}]
  properties {String1="", String2=""}
 }
 SimplifyConvexHulls {
  type=FSimplifyConvexHullsDataflowNode
  guid=#1344
  inputPins [
   Collection {type=FManagedArrayCollection}
   OptionalSelectionFilter {type=FDataflowTransformSelection}
   SimplificationAngleThreshold {type=float}
   SimplificationDistanceThreshold {type=float}
   MinTargetTriangleCount {type=int32}
  ]
  outputPins [Collection {type=FManagedArrayCollection}]
  properties {
   ColorRandomSeed=0
   DebugDrawRenderSettings=(bTranslucent=True,Color=(R=0,G=1,B=0,A=1),LineWidthMultiplier=2)
   MinTargetTriangleCount=6
   SimplificationAngleThreshold=5
   SimplificationDistanceThreshold=10
   SimplifyMethod=AngleTolerance
   bRandomizeColor=True
   bUseExistingVertices=True
  }
 }
 GetBoundingBoxesFromCollection {
  type=FGetBoundingBoxesFromCollectionDataflowNode
  guid=#1345
  inputPins [
   Collection {type=FManagedArrayCollection}
   TransformSelection {type=FDataflowTransformSelection}
  ]
  outputPins [BoundingBoxes {type="TArray<FBox>"}]
 }
 GetBoxLengths {
  type=FGetBoxLengthsDataflowNode
  guid=#1346
  inputPins [Boxes {type="TArray<FBox>"}]
  outputPins [Lengths {type="TArray<float>"}]
  properties {MeasurementMethod=ShortestAxis}
 }
 SelectFloatArrayIndicesInRange {
  type=FSelectFloatArrayIndicesInRangeDataflowNode
  guid=#1347
  inputPins [
   Values {type="TArray<float>"}
   Min {type=float}
   Max {type=float}
  ]
  outputPins [Indices {type="TArray<int32>"}]
  properties {Max=10, Min=6, RangeSetting=Dataflow_RangeSetting_InsideRange, bInclusive=True}
 }
 CollectionTransformSelectionFromIndexArray {
  type=FCollectionTransformSelectionFromIndexArrayDataflowNode
  guid=#1348
  inputPins [
   Collection {type=FManagedArrayCollection}
   BoneIndices {type="TArray<int32>"}
  ]
  outputPins [
   Collection {type=FManagedArrayCollection}
   TransformSelection {type=FDataflowTransformSelection}
  ]
  properties {BoneIndices=""}
 }
 CreateLeafConvexHulls_Tiny_copy {
  type=FCreateLeafConvexHullsDataflowNode
  guid=#1349
  inputPins [
   Collection {type=FManagedArrayCollection}
   OptionalSelectionFilter {type=FDataflowTransformSelection}
   SimplificationDistanceThreshold {type=float}
   ConvexDecompositionSettings {type=FDataflowConvexDecompositionSettings}
   RequireNegativeSpaceCovering {type=FDataflowSphereCovering}
  ]
  outputPins [
   Collection {type=FManagedArrayCollection}
   SphereCovering {type=FDataflowSphereCovering}
  ]
  properties {
   ColorRandomSeed=0
   ConvexDecompositionSettings=(MaxGeoToHullVolumeRatioToDecompose=0.7,ErrorTolerance=8,MaxHullsPerGeometry=9,MinThicknessTolerance=4,NumAdditionalSplits=2,bOnlyConnectedToHull=True,NegativeSpaceTolerance=2,NegativeSpaceMinRadius=10)
   DebugDrawRenderSettings=(bTranslucent=True,Color=(R=0,G=1,B=0,A=1),LineWidthMultiplier=2)
   GenerateMethod=IntersectExternalWithComputed
   IntersectIfComputedIsSmallerByFactor=1
   MinExternalVolumeToIntersect=3000
   SimplificationDistanceThreshold=0
   SphereCoveringDebugDrawRenderSettings=(bTranslucent=True,LineWidthMultiplier=0.25,Color=(R=1,G=0,B=0,A=1),ColorA=(R=1,G=0,B=0,A=1),ColorB=(R=0,G=0,B=1,A=1))
   bComputeIntersectionsBeforeHull=False
   bRandomizeColor=True
  }
 }
 MergeConvexHulls_1 {
  type=FMergeConvexHullsDataflowNode
  guid=#1350
  inputPins [
   Collection {type=FManagedArrayCollection}
   MaxConvexCount {type=int32}
   ErrorTolerance {type=double}
   OptionalSelectionFilter {type=FDataflowTransformSelection}
   bProtectNegativeSpace {type=bool}
   RequireNegativeSpaceCovering {type=FDataflowSphereCovering}
   TargetNumSamples {type=int32}
   MinSampleSpacing {type=double}
   NegativeSpaceTolerance {type=double}
   MinRadius {type=double}
  ]
  outputPins [
   Collection {type=FManagedArrayCollection}
   SphereCovering {type=FDataflowSphereCovering}
  ]
  properties {
   ColorRandomSeed=0
   DebugDrawRenderSettings=(bTranslucent=True,Color=(R=0,G=1,B=0,A=1),LineWidthMultiplier=2)
   ErrorTolerance=64
   MaxConvexCount=-1
   MergeProximityDistanceThreshold=0.1
   MergeProximityFilter=None
   MinRadius=4
   MinSampleSpacing=1
   NegativeSpaceTolerance=1.6
   SampleMethod=Uniform
   SphereCoveringDebugDrawRenderSettings=(bTranslucent=True,LineWidthMultiplier=0.25,Color=(R=1,G=0,B=0,A=1),ColorA=(R=1,G=0,B=0,A=1),ColorB=(R=0,G=0,B=1,A=1))
   TargetNumSamples=10
   bComputeNegativeSpacePerBone=True
   bOnlyConnectedToHull=False
   bProtectNegativeSpace=True
   bRandomizeColor=True
   bRequireSearchSampleCoverage=False
  }
 }
 SimplifyConvexHulls_Leaf_QSlim_Tiny_copy {
  type=FSimplifyConvexHullsDataflowNode
  guid=#1351
  inputPins [
   Collection {type=FManagedArrayCollection}
   OptionalSelectionFilter {type=FDataflowTransformSelection}
   SimplificationAngleThreshold {type=float}
   SimplificationDistanceThreshold {type=float}
   MinTargetTriangleCount {type=int32}
  ]
  outputPins [Collection {type=FManagedArrayCollection}]
  properties {
   ColorRandomSeed=0
   DebugDrawRenderSettings=(bTranslucent=True,Color=(R=0,G=1,B=0,A=1),LineWidthMultiplier=2)
   MinTargetTriangleCount=12
   SimplificationAngleThreshold=30
   SimplificationDistanceThreshold=1.6
   SimplifyMethod=MeshQSlim
   bRandomizeColor=True
   bUseExistingVertices=True
  }
 }
 SimplifyConvexHulls_Angle_Tiny_copy {
  type=FSimplifyConvexHullsDataflowNode
  guid=#1352
  inputPins [
   Collection {type=FManagedArrayCollection}
   OptionalSelectionFilter {type=FDataflowTransformSelection}
   SimplificationAngleThreshold {type=float}
   SimplificationDistanceThreshold {type=float}
   MinTargetTriangleCount {type=int32}
  ]
  outputPins [Collection {type=FManagedArrayCollection}]
  properties {
   ColorRandomSeed=0
   DebugDrawRenderSettings=(bTranslucent=True,Color=(R=0,G=1,B=0,A=1),LineWidthMultiplier=2)
   MinTargetTriangleCount=12
   SimplificationAngleThreshold=10
   SimplificationDistanceThreshold=10
   SimplifyMethod=AngleTolerance
   bRandomizeColor=True
   bUseExistingVertices=True
  }
 }
 SimplifyConvexHulls_Angle_Tiny_copy_0 {
  type=FSimplifyConvexHullsDataflowNode
  guid=#1353
  inputPins [
   Collection {type=FManagedArrayCollection}
   OptionalSelectionFilter {type=FDataflowTransformSelection}
   SimplificationAngleThreshold {type=float}
   SimplificationDistanceThreshold {type=float}
   MinTargetTriangleCount {type=int32}
  ]
  outputPins [Collection {type=FManagedArrayCollection}]
  properties {
   ColorRandomSeed=0
   DebugDrawRenderSettings=(bTranslucent=True,Color=(R=0,G=1,B=0,A=1),LineWidthMultiplier=2)
   MinTargetTriangleCount=12
   SimplificationAngleThreshold=22.5
   SimplificationDistanceThreshold=10
   SimplifyMethod=AngleTolerance
   bRandomizeColor=True
   bUseExistingVertices=True
  }
 }
 MergeConvexHulls_2 {
  type=FMergeConvexHullsDataflowNode
  guid=#1354
  inputPins [
   Collection {type=FManagedArrayCollection}
   MaxConvexCount {type=int32}
   ErrorTolerance {type=double}
   OptionalSelectionFilter {type=FDataflowTransformSelection}
   bProtectNegativeSpace {type=bool}
   RequireNegativeSpaceCovering {type=FDataflowSphereCovering}
   TargetNumSamples {type=int32}
   MinSampleSpacing {type=double}
   NegativeSpaceTolerance {type=double}
   MinRadius {type=double}
  ]
  outputPins [
   Collection {type=FManagedArrayCollection}
   SphereCovering {type=FDataflowSphereCovering}
  ]
  properties {
   ColorRandomSeed=0
   DebugDrawRenderSettings=(bTranslucent=True,Color=(R=0,G=1,B=0,A=1),LineWidthMultiplier=2)
   ErrorTolerance=16
   MaxConvexCount=-1
   MergeProximityDistanceThreshold=0.1
   MergeProximityFilter=None
   MinRadius=4
   MinSampleSpacing=1
   NegativeSpaceTolerance=2
   SampleMethod=VoxelSearch
   SphereCoveringDebugDrawRenderSettings=(bTranslucent=True,LineWidthMultiplier=0.25,Color=(R=1,G=0,B=0,A=1),ColorA=(R=1,G=0,B=0,A=1),ColorB=(R=0,G=0,B=1,A=1))
   TargetNumSamples=10
   bComputeNegativeSpacePerBone=True
   bOnlyConnectedToHull=False
   bProtectNegativeSpace=True
   bRandomizeColor=True
   bRequireSearchSampleCoverage=False
  }
 }
 SimplifyConvexHulls_QSlim_Tiny_2_copy {
  type=FSimplifyConvexHullsDataflowNode
  guid=#1355
  inputPins [
   Collection {type=FManagedArrayCollection}
   OptionalSelectionFilter {type=FDataflowTransformSelection}
   SimplificationAngleThreshold {type=float}
   SimplificationDistanceThreshold {type=float}
   MinTargetTriangleCount {type=int32}
  ]
  outputPins [Collection {type=FManagedArrayCollection}]
  properties {
   ColorRandomSeed=0
   DebugDrawRenderSettings=(bTranslucent=True,Color=(R=0,G=1,B=0,A=1),LineWidthMultiplier=2)
   MinTargetTriangleCount=8
   SimplificationAngleThreshold=50
   SimplificationDistanceThreshold=3.2
   SimplifyMethod=AngleTolerance
   bRandomizeColor=True
   bUseExistingVertices=True
  }
 }
 Terminal Early Out {
  type=FGeometryCollectionTerminalDataflowNode
  guid=#1356
  inputPins [
   Collection {type=FManagedArrayCollection}
   Materials {type="TArray<TObjectPtr<UMaterial>>"}
   MaterialInstances {type="TArray<TObjectPtr<UMaterialInterface>>"}
   InstancedMeshes {type="TArray<FGeometryCollectionAutoInstanceMesh>"}
  ]
  outputPins [
   Collection {type=FManagedArrayCollection}
   Materials {type="TArray<TObjectPtr<UMaterial>>"}
   MaterialInstances {type="TArray<TObjectPtr<UMaterialInterface>>"}
   InstancedMeshes {type="TArray<FGeometryCollectionAutoInstanceMesh>"}
  ]
 }
 Divide_0_copy {
  type=FDivideDataflowNode
  guid=#1357
  inputPins [
   FloatA {type=float}
   FloatB {type=float}
  ]
  outputPins [ReturnValue {type=float}]
  properties {FloatA=0, FloatB=0}
 }
 Clamp_1_copy {
  type=FClampDataflowNode
  guid=#1358
  inputPins [
   Float {type=float}
   Min {type=float}
   Max {type=float}
  ]
  outputPins [ReturnValue {type=float}]
  properties {Float=0, Max=1, Min=0}
 }
 Floor_3_copy {
  type=FFloorDataflowNode
  guid=#1359
  inputPins [Float {type=float}]
  outputPins [ReturnValue {type=float}]
  properties {Float=0}
 }
 FloatToInt_0_copy {
  type=FFloatToIntDataflowNode
  guid=#1360
  inputPins [Float {type=float}]
  outputPins [Int {type=int32}]
  properties {Float=0, Function=Dataflow_FloatToInt_Function_Round}
 }
 IntToBool_3_copy {
  type=FIntToBoolDataflowNode
  guid=#1361
  inputPins [Int {type=int32}]
  outputPins [Bool {type=bool}]
  properties {Int=0}
 }
 BoolToString_5_copy {
  type=FBoolToStringDataflowNode
  guid=#1362
  inputPins [Bool {type=bool}]
  outputPins [String {type=FString}]
  properties {Bool=False}
 }
 LogString_3_copy {
  type=FLogStringDataflowNode
  guid=#1363
  inputPins [String {type=FString}]
  outputPins []
  properties {String="", bPrintToLog=True}
 }
 One Minus {
  type=FSubtractDataflowNode
  guid=#1364
  inputPins [
   FloatA {type=float}
   FloatB {type=float}
  ]
  outputPins [ReturnValue {type=float}]
  properties {FloatA=1, FloatB=0}
 }
 Add_0 {
  type=FAddDataflowNode
  guid=#1365
  inputPins [
   FloatA {type=float}
   FloatB {type=float}
  ]
  outputPins [ReturnValue {type=float}]
  properties {FloatA=0, FloatB=0}
 }
 FloatToInt_0_copy_0 {
  type=FFloatToIntDataflowNode
  guid=#1366
  inputPins [Float {type=float}]
  outputPins [Int {type=int32}]
  properties {Float=0, Function=Dataflow_FloatToInt_Function_Round}
 }
 IntToBool_3_copy_0 {
  type=FIntToBoolDataflowNode
  guid=#1367
  inputPins [Int {type=int32}]
  outputPins [Bool {type=bool}]
  properties {Int=0}
 }
 BoolToString_5_copy_0 {
  type=FBoolToStringDataflowNode
  guid=#1368
  inputPins [Bool {type=bool}]
  outputPins [String {type=FString}]
  properties {Bool=False}
 }
 LogString_3_copy_0 {
  type=FLogStringDataflowNode
  guid=#1369
  inputPins [String {type=FString}]
  outputPins []
  properties {String="", bPrintToLog=True}
 }
 Adaptive Tolerance {
  type=FFloatToStringDataflowNode
  guid=#1370
  inputPins [Float {type=float}]
  outputPins [String {type=FString}]
  properties {Float=0}
 }
 Log - Adaptive Tolerance {
  type=FLogStringDataflowNode
  guid=#1371
  inputPins [String {type=FString}]
  outputPins []
  properties {String="", bPrintToLog=True}
 }
 CollectionTransformSelectRoot_0 {
  type=FCollectionTransformSelectionRootDataflowNode
  guid=#1372
  inputPins [Collection {type=FManagedArrayCollection}]
  outputPins [
   Collection {type=FManagedArrayCollection}
   TransformSelection {type=FDataflowTransformSelection}
  ]
 }
 SetAnchorState_0 {
  type=FSetAnchorStateDataflowNode
  guid=#1373
  inputPins [
   Collection {type=FManagedArrayCollection}
   TransformSelection {type=FDataflowTransformSelection}
  ]
  outputPins [Collection {type=FManagedArrayCollection}]
  properties {AnchorState=Dataflow_AnchorState_Anchored, bSetNotSelectedBonesToOppositeState=False}
 }
 Clamp_2 {
  type=FClampDataflowNode
  guid=#1374
  inputPins [
   Float {type=float}
   Min {type=float}
   Max {type=float}
  ]
  outputPins [ReturnValue {type=float}]
  properties {Float=0, Max=1, Min=0}
 }
 Max Clusters {
  type=FMakeLiteralFloatDataflowNode
  guid=#1375
  inputPins []
  outputPins [Float {type=float}]
  properties {Value=25}
 }
 MergeConvexHulls_3 {
  type=FMergeConvexHullsDataflowNode
  guid=#1376
  inputPins [
   Collection {type=FManagedArrayCollection}
   MaxConvexCount {type=int32}
   ErrorTolerance {type=double}
   OptionalSelectionFilter {type=FDataflowTransformSelection}
   bProtectNegativeSpace {type=bool}
   RequireNegativeSpaceCovering {type=FDataflowSphereCovering}
   TargetNumSamples {type=int32}
   MinSampleSpacing {type=double}
   NegativeSpaceTolerance {type=double}
   MinRadius {type=double}
  ]
  outputPins [
   Collection {type=FManagedArrayCollection}
   SphereCovering {type=FDataflowSphereCovering}
  ]
  properties {
   ColorRandomSeed=0
   DebugDrawRenderSettings=(bTranslucent=True,Color=(R=0,G=1,B=0,A=1),LineWidthMultiplier=2)
   ErrorTolerance=0
   MaxConvexCount=-1
   MergeProximityDistanceThreshold=0.1
   MergeProximityFilter=None
   MinRadius=3
   MinSampleSpacing=1
   NegativeSpaceTolerance=2
   SampleMethod=VoxelSearch
   SphereCoveringDebugDrawRenderSettings=(bTranslucent=True,LineWidthMultiplier=0.25,Color=(R=1,G=0,B=0,A=1),ColorA=(R=1,G=0,B=0,A=1),ColorB=(R=0,G=0,B=1,A=1))
   TargetNumSamples=20
   bComputeNegativeSpacePerBone=True
   bOnlyConnectedToHull=False
   bProtectNegativeSpace=True
   bRandomizeColor=True
   bRequireSearchSampleCoverage=False
  }
 }
 SimplifyConvexHulls_Leaf_QSlim_Small_copy {
  type=FSimplifyConvexHullsDataflowNode
  guid=#1377
  inputPins [
   Collection {type=FManagedArrayCollection}
   OptionalSelectionFilter {type=FDataflowTransformSelection}
   SimplificationAngleThreshold {type=float}
   SimplificationDistanceThreshold {type=float}
   MinTargetTriangleCount {type=int32}
  ]
  outputPins [Collection {type=FManagedArrayCollection}]
  properties {
   ColorRandomSeed=0
   DebugDrawRenderSettings=(bTranslucent=True,Color=(R=0,G=1,B=0,A=1),LineWidthMultiplier=2)
   MinTargetTriangleCount=8
   SimplificationAngleThreshold=15
   SimplificationDistanceThreshold=1.6
   SimplifyMethod=AngleTolerance
   bRandomizeColor=True
   bUseExistingVertices=True
  }
 }
 MergeConvexHulls_copy_0 {
  type=FMergeConvexHullsDataflowNode
  guid=#1378
  inputPins [
   Collection {type=FManagedArrayCollection}
   MaxConvexCount {type=int32}
   ErrorTolerance {type=double}
   OptionalSelectionFilter {type=FDataflowTransformSelection}
   bProtectNegativeSpace {type=bool}
   RequireNegativeSpaceCovering {type=FDataflowSphereCovering}
   TargetNumSamples {type=int32}
   MinSampleSpacing {type=double}
   NegativeSpaceTolerance {type=double}
   MinRadius {type=double}
  ]
  outputPins [
   Collection {type=FManagedArrayCollection}
   SphereCovering {type=FDataflowSphereCovering}
  ]
  properties {
   ColorRandomSeed=0
   DebugDrawRenderSettings=(bTranslucent=True,Color=(R=0,G=1,B=0,A=1),LineWidthMultiplier=2)
   ErrorTolerance=64
   MaxConvexCount=-1
   MergeProximityDistanceThreshold=0.1
   MergeProximityFilter=None
   MinRadius=4
   MinSampleSpacing=1
   NegativeSpaceTolerance=2
   SampleMethod=VoxelSearch
   SphereCoveringDebugDrawRenderSettings=(bTranslucent=True,LineWidthMultiplier=0.25,Color=(R=1,G=0,B=0,A=1),ColorA=(R=1,G=0,B=0,A=1),ColorB=(R=0,G=0,B=1,A=1))
   TargetNumSamples=50
   bComputeNegativeSpacePerBone=True
   bOnlyConnectedToHull=False
   bProtectNegativeSpace=True
   bRandomizeColor=True
   bRequireSearchSampleCoverage=True
  }
 }
 LogString_3 {
  type=FLogStringDataflowNode
  guid=#1379
  inputPins [String {type=FString}]
  outputPins []
  properties {String="", bPrintToLog=True}
 }
 FloatToString_2 {
  type=FFloatToStringDataflowNode
  guid=#1380
  inputPins [Float {type=float}]
  outputPins [String {type=FString}]
  properties {Float=0}
 }
 LogString_4 {
  type=FLogStringDataflowNode
  guid=#1381
  inputPins [String {type=FString}]
  outputPins []
  properties {String="", bPrintToLog=True}
 }
 One Minus_copy {
  type=FSubtractDataflowNode
  guid=#1382
  inputPins [
   FloatA {type=float}
   FloatB {type=float}
  ]
  outputPins [ReturnValue {type=float}]
  properties {FloatA=1, FloatB=0}
 }
 "SP - Spacing Multiplier " {
  type=FMakeLiteralFloatDataflowNode
  guid=#1383
  inputPins []
  outputPins [Float {type=float}]
  properties {Value=2}
 }
 SP - Base Samples {
  type=FMakeLiteralFloatDataflowNode
  guid=#1384
  inputPins []
  outputPins [Float {type=float}]
  properties {Value=40}
 }
 RemoveOnBreak_0_copy {
  type=FRemoveOnBreakDataflowNode
  guid=#1385
  inputPins [
   Collection {type=FManagedArrayCollection}
   TransformSelection {type=FDataflowTransformSelection}
   bEnabledRemoval {type=bool}
   PostBreakTimer {type=FVector2f}
   RemovalTimer {type=FVector2f}
   bClusterCrumbling {type=bool}
  ]
  outputPins [Collection {type=FManagedArrayCollection}]
  properties {PostBreakTimer=(X=1,Y=2), RemovalTimer=(X=0,Y=0), bClusterCrumbling=True, bEnabledRemoval=True}
 }
 "REROUTER " {
  type=FMultiplyDataflowNode
  guid=#1386
  inputPins [
   FloatA {type=float}
   FloatB {type=float}
  ]
  outputPins [ReturnValue {type=float}]
  properties {FloatA=0, FloatB=1}
 }
 IntToFloat_0 {
  type=FIntToFloatDataflowNode
  guid=#1387
  inputPins [Int {type=int32}]
  outputPins [Float {type=float}]
  properties {Int=0}
 }
 Clamp_3 {
  type=FClampDataflowNode
  guid=#1388
  inputPins [
   Float {type=float}
   Min {type=float}
   Max {type=float}
  ]
  outputPins [ReturnValue {type=float}]
  properties {Float=0, Max=1, Min=0}
 }
 FloatToString_2_copy {
  type=FFloatToStringDataflowNode
  guid=#1389
  inputPins [Float {type=float}]
  outputPins [String {type=FString}]
  properties {Float=0}
 }
 LogString_4_copy {
  type=FLogStringDataflowNode
  guid=#1390
  inputPins [String {type=FString}]
  outputPins []
  properties {String="", bPrintToLog=True}
 }
 FloatToDouble_6 {
  type=FFloatToDoubleDataflowNode
  guid=#1391
  inputPins [Float {type=float}]
  outputPins [Double {type=double}]
  properties {Float=0}
 }
 BoolToInt_1 {
  type=FBoolToIntDataflowNode
  guid=#1392
  inputPins [Bool {type=bool}]
  outputPins [Int {type=int32}]
  properties {Bool=False}
 }
 BoolToString_6_copy {
  type=FBoolToStringDataflowNode
  guid=#1393
  inputPins [Bool {type=bool}]
  outputPins [String {type=FString}]
  properties {Bool=False}
 }
 Enable L1 Convex Count? {
  type=FLogStringDataflowNode
  guid=#1394
  inputPins [String {type=FString}]
  outputPins []
  properties {String="", bPrintToLog=True}
 }
 32 {
  type=FMakeLiteralFloatDataflowNode
  guid=#1395
  inputPins []
  outputPins [Float {type=float}]
  properties {Value=32}
 }
 SP - Tolerance {
  type=FMakeLiteralFloatDataflowNode
  guid=#1396
  inputPins []
  outputPins [Float {type=float}]
  properties {Value=8}
 }
 OneMinus {
  type=FSubtractDataflowNode
  guid=#1397
  inputPins [
   FloatA {type=float}
   FloatB {type=float}
  ]
  outputPins [ReturnValue {type=float}]
  properties {FloatA=1, FloatB=0}
 }
 Multiply_0 {
  type=FMultiplyDataflowNode
  guid=#1398
  inputPins [
   FloatA {type=float}
   FloatB {type=float}
  ]
  outputPins [ReturnValue {type=float}]
  properties {FloatA=0, FloatB=0}
 }
 Multiply_1 {
  type=FMultiplyDataflowNode
  guid=#1399
  inputPins [
   FloatA {type=float}
   FloatB {type=float}
  ]
  outputPins [ReturnValue {type=float}]
  properties {FloatA=0, FloatB=2}
 }
 BranchCollection_DerivedProximityMethod {
  type=FBranchCollectionDataflowNode
  guid=#1400
  inputPins [
   TrueCollection {type=FManagedArrayCollection}
   FalseCollection {type=FManagedArrayCollection}
   bCondition {type=bool}
  ]
  outputPins [ChosenCollection {type=FManagedArrayCollection}]
  properties {bCondition=False}
 }
 Proximity - Precise Contact {
  type=FProximityDataflowNode
  guid=#1401
  inputPins [
   Collection {type=FManagedArrayCollection}
   DistanceThreshold {type=float}
   ContactThreshold {type=float}
  ]
  outputPins [Collection {type=FManagedArrayCollection}]
  properties {
   ContactAreaMethod=Dataflow_ConnectionContactAreaMethod_None
   ContactThreshold=2
   DistanceThreshold=2
   FilterContactMethod=Dataflow_ProximityContactFilteringMethod_ConvexHullSharp
   ProximityMethod=Dataflow_ProximityMethod_Precise
   bRecomputeConvexHulls=False
   bUseAsConnectionGraph=False
  }
 }
 Get Sphere Covering Sphere Count {
  type=FSphereCoveringCountSpheresNode
  guid=#1402
  inputPins [SphereCovering {type=FDataflowSphereCovering}]
  outputPins [NumSpheres {type=int32}]
 }
 IntToString_0 {
  type=FIntToStringDataflowNode
  guid=#1403
  inputPins [Int {type=int32}]
  outputPins [String {type=FString}]
  properties {Int=0}
 }
 LogString_3_copy_1 {
  type=FLogStringDataflowNode
  guid=#1404
  inputPins [String {type=FString}]
  outputPins []
  properties {String="", bPrintToLog=True}
 }
 Add_1 {
  type=FAddDataflowNode
  guid=#1405
  inputPins [
   FloatA {type=float}
   FloatB {type=float}
  ]
  outputPins [ReturnValue {type=float}]
  properties {FloatA=0, FloatB=2}
 }
 ClusterIsolatedRoots {
  type=FClusterIsolatedRootsDataflowNode
  guid=#1406
  inputPins [Collection {type=FManagedArrayCollection}]
  outputPins [Collection {type=FManagedArrayCollection}]
 }
 Proximity_5_copy {
  type=FProximityDataflowNode
  guid=#1407
  inputPins [
   Collection {type=FManagedArrayCollection}
   DistanceThreshold {type=float}
   ContactThreshold {type=float}
  ]
  outputPins [Collection {type=FManagedArrayCollection}]
  properties {
   ContactAreaMethod=Dataflow_ConnectionContactAreaMethod_None
   ContactThreshold=1.6
   DistanceThreshold=1.6
   FilterContactMethod=Dataflow_ProximityContactFilteringMethod_ConvexHullSharp
   ProximityMethod=Dataflow_ProximityMethod_ConvexHull
   bRecomputeConvexHulls=False
   bUseAsConnectionGraph=False
  }
 }
 Proximity_copy {
  type=FProximityDataflowNode
  guid=#1408
  inputPins [
   Collection {type=FManagedArrayCollection}
   DistanceThreshold {type=float}
   ContactThreshold {type=float}
  ]
  outputPins [Collection {type=FManagedArrayCollection}]
  properties {
   ContactAreaMethod=Dataflow_ConnectionContactAreaMethod_None
   ContactThreshold=3.2
   DistanceThreshold=3.2
   FilterContactMethod=Dataflow_ProximityContactFilteringMethod_ConvexHullSharp
   ProximityMethod=Dataflow_ProximityMethod_ConvexHull
   bRecomputeConvexHulls=False
   bUseAsConnectionGraph=False
  }
 }
 Divide_3_copy {
  type=FDivideDataflowNode
  guid=#1409
  inputPins [
   FloatA {type=float}
   FloatB {type=float}
  ]
  outputPins [ReturnValue {type=float}]
  properties {FloatA=0, FloatB=0}
 }
 Clamp_5_copy {
  type=FClampDataflowNode
  guid=#1410
  inputPins [
   Float {type=float}
   Min {type=float}
   Max {type=float}
  ]
  outputPins [ReturnValue {type=float}]
  properties {Float=0, Max=1, Min=0}
 }
 Floor_copy {
  type=FFloorDataflowNode
  guid=#1411
  inputPins [Float {type=float}]
  outputPins [ReturnValue {type=float}]
  properties {Float=0}
 }
 FloatToInt_4_copy {
  type=FFloatToIntDataflowNode
  guid=#1412
  inputPins [Float {type=float}]
  outputPins [Int {type=int32}]
  properties {Float=0, Function=Dataflow_FloatToInt_Function_Floor}
 }
 IntToBool_copy {
  type=FIntToBoolDataflowNode
  guid=#1413
  inputPins [Int {type=int32}]
  outputPins [Bool {type=bool}]
  properties {Int=0}
 }
 Threshold {
  type=FMakeLiteralFloatDataflowNode
  guid=#1414
  inputPins []
  outputPins [Float {type=float}]
  properties {Value=40}
 }
 Proximity_5_copy_0 {
  type=FProximityDataflowNode
  guid=#1415
  inputPins [
   Collection {type=FManagedArrayCollection}
   DistanceThreshold {type=float}
   ContactThreshold {type=float}
  ]
  outputPins [Collection {type=FManagedArrayCollection}]
  properties {
   ContactAreaMethod=Dataflow_ConnectionContactAreaMethod_None
   ContactThreshold=1.6
   DistanceThreshold=3.2
   FilterContactMethod=Dataflow_ProximityContactFilteringMethod_ConvexHullSharp
   ProximityMethod=Dataflow_ProximityMethod_ConvexHull
   bRecomputeConvexHulls=False
   bUseAsConnectionGraph=False
  }
 }
 BranchCollection_7 {
  type=FBranchCollectionDataflowNode
  guid=#1416
  inputPins [
   TrueCollection {type=FManagedArrayCollection}
   FalseCollection {type=FManagedArrayCollection}
   bCondition {type=bool}
  ]
  outputPins [ChosenCollection {type=FManagedArrayCollection}]
  properties {bCondition=False}
 }
 BoolToString_9 {
  type=FBoolToStringDataflowNode
  guid=#1417
  inputPins [Bool {type=bool}]
  outputPins [String {type=FString}]
  properties {Bool=False}
 }
 LogString_8 {
  type=FLogStringDataflowNode
  guid=#1418
  inputPins [String {type=FString}]
  outputPins []
  properties {String="", bPrintToLog=True}
 }
 OneMinus_0 {
  type=FOneMinusDataflowNode
  guid=#1419
  inputPins [A {type=float}]
  outputPins [ReturnValue {type=float}]
  properties {A=0}
 }
 BoolToString_9_copy {
  type=FBoolToStringDataflowNode
  guid=#1420
  inputPins [Bool {type=bool}]
  outputPins [String {type=FString}]
  properties {Bool=False}
 }
 LogString_8_copy {
  type=FLogStringDataflowNode
  guid=#1421
  inputPins [String {type=FString}]
  outputPins []
  properties {String="", bPrintToLog=True}
 }
 FloatToString_copy {
  type=FFloatToStringDataflowNode
  guid=#1422
  inputPins [Float {type=float}]
  outputPins [String {type=FString}]
  properties {Float=0}
 }
 LogString_5_copy {
  type=FLogStringDataflowNode
  guid=#1423
  inputPins [String {type=FString}]
  outputPins []
  properties {String="", bPrintToLog=True}
 }
 Multiply_3 {
  type=FMultiplyDataflowNode
  guid=#1424
  inputPins [
   FloatA {type=float}
   FloatB {type=float}
  ]
  outputPins [ReturnValue {type=float}]
  properties {FloatA=0, FloatB=0}
 }
 BranchFloat {
  type=FBranchFloatDataflowNode
  guid=#1425
  inputPins [
   A {type=float}
   B {type=float}
   bCondition {type=bool}
  ]
  outputPins [ReturnValue {type=float}]
  properties {A=0, B=0, bCondition=False}
 }
 BranchFloat_copy {
  type=FBranchFloatDataflowNode
  guid=#1426
  inputPins [
   A {type=float}
   B {type=float}
   bCondition {type=bool}
  ]
  outputPins [ReturnValue {type=float}]
  properties {A=0, B=0, bCondition=False}
 }
 FloatToString_3 {
  type=FFloatToStringDataflowNode
  guid=#1427
  inputPins [Float {type=float}]
  outputPins [String {type=FString}]
  properties {Float=0}
 }
 LogString_11 {
  type=FLogStringDataflowNode
  guid=#1428
  inputPins [String {type=FString}]
  outputPins []
  properties {String="", bPrintToLog=True}
 }
 CollectionTransformSelectTargetLevel {
  type=FCollectionTransformSelectionTargetLevelDataflowNode
  guid=#1429
  inputPins [
   Collection {type=FManagedArrayCollection}
   TargetLevel {type=int32}
  ]
  outputPins [
   Collection {type=FManagedArrayCollection}
   TransformSelection {type=FDataflowTransformSelection}
  ]
  properties {TargetLevel=1, bSkipEmbedded=False}
 }
 ClusterMergeToNeighbors {
  type=FClusterMergeToNeighborsDataflowNode
  guid=#1430
  inputPins [
   Collection {type=FManagedArrayCollection}
   TransformSelection {type=FDataflowTransformSelection}
   MinVolumeCubeRoot {type=float}
   bOnlyToConnected {type=bool}
   bOnlySameParent {type=bool}
  ]
  outputPins [Collection {type=FManagedArrayCollection}]
  properties {
   MinVolumeCubeRoot=16
   NeighborSelectionMethod=Dataflow_ClusterNeighborSelectionMethod_NearestCenter
   bOnlySameParent=True
   bOnlyToConnected=True
  }
 }
 CollectionTransformSelectionSetOperation {
  type=FCollectionTransformSelectionSetOperationDataflowNode
  guid=#1431
  inputPins [
   TransformSelectionA {type=FDataflowTransformSelection}
   TransformSelectionB {type=FDataflowTransformSelection}
  ]
  outputPins [TransformSelection {type=FDataflowTransformSelection}]
  properties {Operation=Dataflow_SetOperation_AND}
 }
 Branch Use CleanUp {
  type=FBranchCollectionDataflowNode
  guid=#1432
  inputPins [
   TrueCollection {type=FManagedArrayCollection}
   FalseCollection {type=FManagedArrayCollection}
   bCondition {type=bool}
  ]
  outputPins [ChosenCollection {type=FManagedArrayCollection}]
  properties {bCondition=False}
 }
 MergeConvexHulls_0_copy_0 {
  type=FMergeConvexHullsDataflowNode
  guid=#1433
  inputPins [
   Collection {type=FManagedArrayCollection}
   MaxConvexCount {type=int32}
   ErrorTolerance {type=double}
   OptionalSelectionFilter {type=FDataflowTransformSelection}
   bProtectNegativeSpace {type=bool}
   RequireNegativeSpaceCovering {type=FDataflowSphereCovering}
   TargetNumSamples {type=int32}
   MinSampleSpacing {type=double}
   NegativeSpaceTolerance {type=double}
   MinRadius {type=double}
  ]
  outputPins [
   Collection {type=FManagedArrayCollection}
   SphereCovering {type=FDataflowSphereCovering}
  ]
  properties {
   ColorRandomSeed=0
   DebugDrawRenderSettings=(bTranslucent=True,Color=(R=0,G=1,B=0,A=1),LineWidthMultiplier=2)
   ErrorTolerance=0
   MaxConvexCount=1
   MergeProximityDistanceThreshold=0.1
   MergeProximityFilter=None
   MinRadius=8
   MinSampleSpacing=16
   NegativeSpaceTolerance=4
   SampleMethod=VoxelSearch
   SphereCoveringDebugDrawRenderSettings=(bTranslucent=True,LineWidthMultiplier=0.25,Color=(R=1,G=0,B=0,A=1),ColorA=(R=1,G=0,B=0,A=1),ColorB=(R=0,G=0,B=1,A=1))
   TargetNumSamples=40
   bComputeNegativeSpacePerBone=True
   bOnlyConnectedToHull=False
   bProtectNegativeSpace=True
   bRandomizeColor=True
   bRequireSearchSampleCoverage=True
  }
 }
 CollectionTransformSelectBySize_0 {
  type=FCollectionTransformSelectionBySizeDataflowNode
  guid=#1434
  inputPins [
   Collection {type=FManagedArrayCollection}
   SizeMin {type=float}
   SizeMax {type=float}
  ]
  outputPins [
   Collection {type=FManagedArrayCollection}
   TransformSelection {type=FDataflowTransformSelection}
  ]
  properties {RangeSetting=Dataflow_RangeSetting_InsideRange, SizeMax=1000, SizeMin=0, bInclusive=True, bUseRelativeSize=False}
 }
 Cleanup_L1_bySize {
  type=FGetFloatOverrideFromAssetDataflowNode
  guid=#1435
  inputPins [
   Key {type=FName}
   Default {type=FString}
  ]
  outputPins [
   IsOverriden {type=bool}
   Float {type=float}
   FloatDefault {type=float}
  ]
  properties {Default=16, Key=Cleanup_L1_bySize}
 }
 Cleanup - SimplificationAngle {
  type=FGetFloatOverrideFromAssetDataflowNode
  guid=#1436
  inputPins [
   Key {type=FName}
   Default {type=FString}
  ]
  outputPins [
   IsOverriden {type=bool}
   Float {type=float}
   FloatDefault {type=float}
  ]
  properties {Default=15, Key=Cleanup - SimplificationAngle}
 }
 Cleanup - SimplificationDist {
  type=FGetFloatOverrideFromAssetDataflowNode
  guid=#1437
  inputPins [
   Key {type=FName}
   Default {type=FString}
  ]
  outputPins [
   IsOverriden {type=bool}
   Float {type=float}
   FloatDefault {type=float}
  ]
  properties {Default=1.5, Key=Cleanup - SimplificationDist}
 }
 Cleanup - SimplificationMinFace {
  type=FGetIntOverrideFromAssetDataflowNode
  guid=#1438
  inputPins [
   Key {type=FName}
   Default {type=FString}
  ]
  outputPins [
   IsOverriden {type=bool}
   Int {type=int32}
   IntDefault {type=int32}
  ]
  properties {Default=16, Key=Cleanup - SimplificationMinFace}
 }
 Proximity - Distance Threshold {
  type=FGetFloatOverrideFromAssetDataflowNode
  guid=#1439
  inputPins [
   Key {type=FName}
   Default {type=FString}
  ]
  outputPins [
   IsOverriden {type=bool}
   Float {type=float}
   FloatDefault {type=float}
  ]
  properties {Default=3, Key=Proximity - Distance Threshold}
 }
 Proximity - Contact Threshold {
  type=FGetFloatOverrideFromAssetDataflowNode
  guid=#1440
  inputPins [
   Key {type=FName}
   Default {type=FString}
  ]
  outputPins [
   IsOverriden {type=bool}
   Float {type=float}
   FloatDefault {type=float}
  ]
  properties {Default=8, Key=Proximity - Contact Threshold}
 }
 BranchCollection_7_copy {
  type=FBranchCollectionDataflowNode
  guid=#1441
  inputPins [
   TrueCollection {type=FManagedArrayCollection}
   FalseCollection {type=FManagedArrayCollection}
   bCondition {type=bool}
  ]
  outputPins [ChosenCollection {type=FManagedArrayCollection}]
  properties {bCondition=False}
 }
 BoolToInt_copy {
  type=FBoolToIntDataflowNode
  guid=#1442
  inputPins [Bool {type=bool}]
  outputPins [Int {type=int32}]
  properties {Bool=False}
 }
 BoolToInt_0_copy {
  type=FBoolToIntDataflowNode
  guid=#1443
  inputPins [Bool {type=bool}]
  outputPins [Int {type=int32}]
  properties {Bool=False}
 }
 IntToFloat_1_copy {
  type=FIntToFloatDataflowNode
  guid=#1444
  inputPins [Int {type=int32}]
  outputPins [Float {type=float}]
  properties {Int=0}
 }
 IntToFloat_2_copy {
  type=FIntToFloatDataflowNode
  guid=#1445
  inputPins [Int {type=int32}]
  outputPins [Float {type=float}]
  properties {Int=0}
 }
 CompareFloat_copy {
  type=FCompareFloatDataflowNode
  guid=#1446
  inputPins [
   FloatA {type=float}
   FloatB {type=float}
  ]
  outputPins [Result {type=bool}]
  properties {FloatA=0, FloatB=0, Operation=Dataflow_Compare_Greater}
 }
 Add_4_copy {
  type=FAddDataflowNode
  guid=#1447
  inputPins [
   FloatA {type=float}
   FloatB {type=float}
  ]
  outputPins [ReturnValue {type=float}]
  properties {FloatA=0, FloatB=0}
 }
 BoolToString_9_copy_0 {
  type=FBoolToStringDataflowNode
  guid=#1448
  inputPins [Bool {type=bool}]
  outputPins [String {type=FString}]
  properties {Bool=False}
 }
 LogString_8_copy_0 {
  type=FLogStringDataflowNode
  guid=#1449
  inputPins [String {type=FString}]
  outputPins []
  properties {String="", bPrintToLog=True}
 }
 Proximity_5_copy_1 {
  type=FProximityDataflowNode
  guid=#1450
  inputPins [
   Collection {type=FManagedArrayCollection}
   DistanceThreshold {type=float}
   ContactThreshold {type=float}
  ]
  outputPins [Collection {type=FManagedArrayCollection}]
  properties {
   ContactAreaMethod=Dataflow_ConnectionContactAreaMethod_None
   ContactThreshold=1.6
   DistanceThreshold=1.6
   FilterContactMethod=Dataflow_ProximityContactFilteringMethod_ConvexHullSharp
   ProximityMethod=Dataflow_ProximityMethod_ConvexHull
   bRecomputeConvexHulls=False
   bUseAsConnectionGraph=False
  }
 }
 UpdateVolumeAttributes_0_copy {
  type=FUpdateVolumeAttributesDataflowNode
  guid=#1451
  inputPins [Collection {type=FManagedArrayCollection}]
  outputPins [Collection {type=FManagedArrayCollection}]
 }
 CollectionTransformSelectRoot_13 {
  type=FCollectionTransformSelectionRootDataflowNode
  guid=#1452
  inputPins [Collection {type=FManagedArrayCollection}]
  outputPins [
   Collection {type=FManagedArrayCollection}
   TransformSelection {type=FDataflowTransformSelection}
  ]
 }
 Flatten {
  type=FClusterFlattenDataflowNode
  guid=#1453
  inputPins [
   Collection {type=FManagedArrayCollection}
   OptionalTransformSelection {type=FDataflowTransformSelection}
  ]
  outputPins [Collection {type=FManagedArrayCollection}]
 }
 BranchCollection_9 {
  type=FBranchCollectionDataflowNode
  guid=#1454
  inputPins [
   TrueCollection {type=FManagedArrayCollection}
   FalseCollection {type=FManagedArrayCollection}
   bCondition {type=bool}
  ]
  outputPins [ChosenCollection {type=FManagedArrayCollection}]
  properties {bCondition=False}
 }
 PURGE L2s {
  type=FMakeLiteralBoolDataflowNode
  guid=#1455
  inputPins []
  outputPins [Bool {type=bool}]
  properties {Value=True}
 }
 ClearConvexHulls {
  type=FClearConvexHullsDataflowNode
  guid=#1456
  inputPins [
   Collection {type=FManagedArrayCollection}
   TransformSelection {type=FDataflowTransformSelection}
  ]
  outputPins [Collection {type=FManagedArrayCollection}]
 }
 BranchCollection_10 {
  type=FBranchCollectionDataflowNode
  guid=#1457
  inputPins [
   TrueCollection {type=FManagedArrayCollection}
   FalseCollection {type=FManagedArrayCollection}
   bCondition {type=bool}
  ]
  outputPins [ChosenCollection {type=FManagedArrayCollection}]
  properties {bCondition=False}
 }
 Purge Leaf Convexes {
  type=FMakeLiteralBoolDataflowNode
  guid=#1458
  inputPins []
  outputPins [Bool {type=bool}]
  properties {Value=True}
 }
 CollectionTransformSelectRoot_15 {
  type=FCollectionTransformSelectionRootDataflowNode
  guid=#1459
  inputPins [Collection {type=FManagedArrayCollection}]
  outputPins [
   Collection {type=FManagedArrayCollection}
   TransformSelection {type=FDataflowTransformSelection}
  ]
 }
 CollectionTransformSelectChildren_4 {
  type=FCollectionTransformSelectionChildrenDataflowNode
  guid=#1460
  inputPins [
   Collection {type=FManagedArrayCollection}
   TransformSelection {type=FDataflowTransformSelection}
  ]
  outputPins [
   Collection {type=FManagedArrayCollection}
   TransformSelection {type=FDataflowTransformSelection}
  ]
 }
 CollectionTransformSelectInvert_0 {
  type=FCollectionTransformSelectionInvertDataflowNode
  guid=#1461
  inputPins [TransformSelection {type=FDataflowTransformSelection}]
  outputPins [TransformSelection {type=FDataflowTransformSelection}]
 }
 CollectionTransformSelectRoot_15_copy {
  type=FCollectionTransformSelectionRootDataflowNode
  guid=#1462
  inputPins [Collection {type=FManagedArrayCollection}]
  outputPins [
   Collection {type=FManagedArrayCollection}
   TransformSelection {type=FDataflowTransformSelection}
  ]
 }
 CollectionTransformSelectionSetOperation_0 {
  type=FCollectionTransformSelectionSetOperationDataflowNode
  guid=#1463
  inputPins [
   TransformSelectionA {type=FDataflowTransformSelection}
   TransformSelectionB {type=FDataflowTransformSelection}
  ]
  outputPins [TransformSelection {type=FDataflowTransformSelection}]
  properties {Operation=Dataflow_SetOperation_OR}
 }
 CollectionTransformSelectCluster {
  type=FCollectionTransformSelectionClusterDataflowNode
  guid=#1464
  inputPins [Collection {type=FManagedArrayCollection}]
  outputPins [
   Collection {type=FManagedArrayCollection}
   TransformSelection {type=FDataflowTransformSelection}
  ]
 }
 CollectionTransformSelectChildren_3 {
  type=FCollectionTransformSelectionChildrenDataflowNode
  guid=#1465
  inputPins [
   Collection {type=FManagedArrayCollection}
   TransformSelection {type=FDataflowTransformSelection}
  ]
  outputPins [
   Collection {type=FManagedArrayCollection}
   TransformSelection {type=FDataflowTransformSelection}
  ]
 }
 CollectionTransformSelectionSetOperation_1 {
  type=FCollectionTransformSelectionSetOperationDataflowNode
  guid=#1466
  inputPins [
   TransformSelectionA {type=FDataflowTransformSelection}
   TransformSelectionB {type=FDataflowTransformSelection}
  ]
  outputPins [TransformSelection {type=FDataflowTransformSelection}]
  properties {Operation=Dataflow_SetOperation_XOR}
 }
 CollectionTransformSelectRoot_16 {
  type=FCollectionTransformSelectionRootDataflowNode
  guid=#1467
  inputPins [Collection {type=FManagedArrayCollection}]
  outputPins [
   Collection {type=FManagedArrayCollection}
   TransformSelection {type=FDataflowTransformSelection}
  ]
 }
 ClearConvexHulls_copy {
  type=FClearConvexHullsDataflowNode
  guid=#1468
  inputPins [
   Collection {type=FManagedArrayCollection}
   TransformSelection {type=FDataflowTransformSelection}
  ]
  outputPins [Collection {type=FManagedArrayCollection}]
 }
 BranchCollection_11 {
  type=FBranchCollectionDataflowNode
  guid=#1469
  inputPins [
   TrueCollection {type=FManagedArrayCollection}
   FalseCollection {type=FManagedArrayCollection}
   bCondition {type=bool}
  ]
  outputPins [ChosenCollection {type=FManagedArrayCollection}]
  properties {bCondition=False}
 }
 DEBUG -- Test L1s {
  type=FMakeLiteralBoolDataflowNode
  guid=#1470
  inputPins []
  outputPins [Bool {type=bool}]
  properties {Value=False}
 }
 BranchCollection_UseManualProximity {
  type=FBranchCollectionDataflowNode
  guid=#1471
  inputPins [
   TrueCollection {type=FManagedArrayCollection}
   FalseCollection {type=FManagedArrayCollection}
   bCondition {type=bool}
  ]
  outputPins [ChosenCollection {type=FManagedArrayCollection}]
  properties {bCondition=False}
 }
 BoolToInt_copy_0 {
  type=FBoolToIntDataflowNode
  guid=#1472
  inputPins [Bool {type=bool}]
  outputPins [Int {type=int32}]
  properties {Bool=False}
 }
 BoolToInt_0_copy_0 {
  type=FBoolToIntDataflowNode
  guid=#1473
  inputPins [Bool {type=bool}]
  outputPins [Int {type=int32}]
  properties {Bool=False}
 }
 IntToFloat_1_copy_0 {
  type=FIntToFloatDataflowNode
  guid=#1474
  inputPins [Int {type=int32}]
  outputPins [Float {type=float}]
  properties {Int=0}
 }
 IntToFloat_2_copy_0 {
  type=FIntToFloatDataflowNode
  guid=#1475
  inputPins [Int {type=int32}]
  outputPins [Float {type=float}]
  properties {Int=0}
 }
 CompareFloat_copy_0 {
  type=FCompareFloatDataflowNode
  guid=#1476
  inputPins [
   FloatA {type=float}
   FloatB {type=float}
  ]
  outputPins [Result {type=bool}]
  properties {FloatA=0, FloatB=0, Operation=Dataflow_Compare_Greater}
 }
 Add_4_copy_0 {
  type=FAddDataflowNode
  guid=#1477
  inputPins [
   FloatA {type=float}
   FloatB {type=float}
  ]
  outputPins [ReturnValue {type=float}]
  properties {FloatA=0, FloatB=0}
 }
 BoolToString_9_copy_0_0 {
  type=FBoolToStringDataflowNode
  guid=#1478
  inputPins [Bool {type=bool}]
  outputPins [String {type=FString}]
  properties {Bool=False}
 }
 LogString_8_copy_0_0 {
  type=FLogStringDataflowNode
  guid=#1479
  inputPins [String {type=FString}]
  outputPins []
  properties {String="", bPrintToLog=True}
 }
 Convex - Sharp Contact {
  type=FProximityDataflowNode
  guid=#1480
  inputPins [
   Collection {type=FManagedArrayCollection}
   DistanceThreshold {type=float}
   ContactThreshold {type=float}
  ]
  outputPins [Collection {type=FManagedArrayCollection}]
  properties {
   ContactAreaMethod=Dataflow_ConnectionContactAreaMethod_None
   ContactThreshold=2
   DistanceThreshold=2
   FilterContactMethod=Dataflow_ProximityContactFilteringMethod_ConvexHullSharp
   ProximityMethod=Dataflow_ProximityMethod_ConvexHull
   bRecomputeConvexHulls=False
   bUseAsConnectionGraph=False
  }
 }
 CreateLeafConvexHulls_Mid_copy {
  type=FCreateLeafConvexHullsDataflowNode
  guid=#1481
  inputPins [
   Collection {type=FManagedArrayCollection}
   OptionalSelectionFilter {type=FDataflowTransformSelection}
   SimplificationDistanceThreshold {type=float}
   ConvexDecompositionSettings {type=FDataflowConvexDecompositionSettings}
   RequireNegativeSpaceCovering {type=FDataflowSphereCovering}
  ]
  outputPins [
   Collection {type=FManagedArrayCollection}
   SphereCovering {type=FDataflowSphereCovering}
  ]
  properties {
   ColorRandomSeed=0
   ConvexDecompositionSettings=(MaxGeoToHullVolumeRatioToDecompose=1,ErrorTolerance=9.4,MaxHullsPerGeometry=10,MinThicknessTolerance=2,NumAdditionalSplits=3,bOnlyConnectedToHull=True,NegativeSpaceTolerance=2,NegativeSpaceMinRadius=10)
   DebugDrawRenderSettings=(bTranslucent=True,Color=(R=0,G=1,B=0,A=1),LineWidthMultiplier=2)
   GenerateMethod=IntersectExternalWithComputed
   IntersectIfComputedIsSmallerByFactor=1
   MinExternalVolumeToIntersect=0
   SimplificationDistanceThreshold=0
   SphereCoveringDebugDrawRenderSettings=(bTranslucent=True,LineWidthMultiplier=0.25,Color=(R=1,G=0,B=0,A=1),ColorA=(R=1,G=0,B=0,A=1),ColorB=(R=0,G=0,B=1,A=1))
   bComputeIntersectionsBeforeHull=False
   bRandomizeColor=True
  }
 }
 CollectionTransformSelectBySize_NarrowWindow {
  type=FCollectionTransformSelectionBySizeDataflowNode
  guid=#1482
  inputPins [
   Collection {type=FManagedArrayCollection}
   SizeMin {type=float}
   SizeMax {type=float}
  ]
  outputPins [
   Collection {type=FManagedArrayCollection}
   TransformSelection {type=FDataflowTransformSelection}
  ]
  properties {RangeSetting=Dataflow_RangeSetting_InsideRange, SizeMax=22.42, SizeMin=22.4, bInclusive=True, bUseRelativeSize=False}
 }
 Multiply_1_copy {
  type=FMultiplyDataflowNode
  guid=#1483
  inputPins [
   FloatA {type=float}
   FloatB {type=float}
  ]
  outputPins [ReturnValue {type=float}]
  properties {FloatA=0, FloatB=1}
 }
 BranchCollection_8 {
  type=FBranchCollectionDataflowNode
  guid=#1484
  inputPins [
   TrueCollection {type=FManagedArrayCollection}
   FalseCollection {type=FManagedArrayCollection}
   bCondition {type=bool}
  ]
  outputPins [ChosenCollection {type=FManagedArrayCollection}]
  properties {bCondition=False}
 }
 CollectionTransformSelectRoot_4_copy {
  type=FCollectionTransformSelectionRootDataflowNode
  guid=#1485
  inputPins [Collection {type=FManagedArrayCollection}]
  outputPins [
   Collection {type=FManagedArrayCollection}
   TransformSelection {type=FDataflowTransformSelection}
  ]
 }
 CollectionTransformSelectChildren_1_copy {
  type=FCollectionTransformSelectionChildrenDataflowNode
  guid=#1486
  inputPins [
   Collection {type=FManagedArrayCollection}
   TransformSelection {type=FDataflowTransformSelection}
  ]
  outputPins [
   Collection {type=FManagedArrayCollection}
   TransformSelection {type=FDataflowTransformSelection}
  ]
 }
 CreateLeafConvexHulls {
  type=FCreateLeafConvexHullsDataflowNode
  guid=#1487
  inputPins [
   Collection {type=FManagedArrayCollection}
   OptionalSelectionFilter {type=FDataflowTransformSelection}
   SimplificationDistanceThreshold {type=float}
   ConvexDecompositionSettings {type=FDataflowConvexDecompositionSettings}
   RequireNegativeSpaceCovering {type=FDataflowSphereCovering}
  ]
  outputPins [
   Collection {type=FManagedArrayCollection}
   SphereCovering {type=FDataflowSphereCovering}
  ]
  properties {
   ColorRandomSeed=0
   ConvexDecompositionSettings=(MaxGeoToHullVolumeRatioToDecompose=1,MaxHullsPerGeometry=-1,NumAdditionalSplits=4,bOnlyConnectedToHull=True,NegativeSpaceTolerance=2,NegativeSpaceMinRadius=10)
   DebugDrawRenderSettings=(bTranslucent=True,Color=(R=0,G=1,B=0,A=1),LineWidthMultiplier=2)
   GenerateMethod=IntersectExternalWithComputed
   IntersectIfComputedIsSmallerByFactor=1
   MinExternalVolumeToIntersect=0
   SimplificationDistanceThreshold=1
   SphereCoveringDebugDrawRenderSettings=(bTranslucent=True,LineWidthMultiplier=0.25,Color=(R=1,G=0,B=0,A=1),ColorA=(R=1,G=0,B=0,A=1),ColorB=(R=0,G=0,B=1,A=1))
   bComputeIntersectionsBeforeHull=True
   bRandomizeColor=True
  }
 }
 MakeConvexDecompositionSettings {
  type=FMakeDataflowConvexDecompositionSettingsNode
  guid=#1488
  inputPins [
   MinSizeToDecompose {type=float}
   MaxGeoToHullVolumeRatioToDecompose {type=float}
   ErrorTolerance {type=float}
   MaxHullsPerGeometry {type=int32}
   MinThicknessTolerance {type=float}
   NumAdditionalSplits {type=int32}
   bProtectNegativeSpace {type=bool}
   bOnlyConnectedToHull {type=bool}
   NegativeSpaceTolerance {type=float}
   NegativeSpaceMinRadius {type=float}
  ]
  outputPins [DecompositionSettings {type=FDataflowConvexDecompositionSettings}]
  properties {
   ErrorTolerance=9.4
   MaxGeoToHullVolumeRatioToDecompose=1
   MaxHullsPerGeometry=-1
   MinSizeToDecompose=0
   MinThicknessTolerance=0
   NegativeSpaceMinRadius=10
   NegativeSpaceTolerance=2
   NumAdditionalSplits=4
   bOnlyConnectedToHull=True
   bProtectNegativeSpace=False
  }
 }
 BranchCollection_1_copy {
  type=FBranchCollectionDataflowNode
  guid=#1489
  inputPins [
   TrueCollection {type=FManagedArrayCollection}
   FalseCollection {type=FManagedArrayCollection}
   bCondition {type=bool}
  ]
  outputPins [ChosenCollection {type=FManagedArrayCollection}]
  properties {bCondition=False}
 }
 CreateLeafConvexHulls_Mid_copy_0 {
  type=FCreateLeafConvexHullsDataflowNode
  guid=#1490
  inputPins [
   Collection {type=FManagedArrayCollection}
   OptionalSelectionFilter {type=FDataflowTransformSelection}
   SimplificationDistanceThreshold {type=float}
   ConvexDecompositionSettings {type=FDataflowConvexDecompositionSettings}
   RequireNegativeSpaceCovering {type=FDataflowSphereCovering}
  ]
  outputPins [
   Collection {type=FManagedArrayCollection}
   SphereCovering {type=FDataflowSphereCovering}
  ]
  properties {
   ColorRandomSeed=0
   ConvexDecompositionSettings=(MaxGeoToHullVolumeRatioToDecompose=1,MaxHullsPerGeometry=30,MinThicknessTolerance=2,NumAdditionalSplits=8,bOnlyConnectedToHull=True,NegativeSpaceTolerance=2,NegativeSpaceMinRadius=10)
   DebugDrawRenderSettings=(bTranslucent=True,Color=(R=0,G=1,B=0,A=1),LineWidthMultiplier=2)
   GenerateMethod=IntersectExternalWithComputed
   IntersectIfComputedIsSmallerByFactor=1
   MinExternalVolumeToIntersect=0
   SimplificationDistanceThreshold=0
   SphereCoveringDebugDrawRenderSettings=(bTranslucent=True,LineWidthMultiplier=0.25,Color=(R=1,G=0,B=0,A=1),ColorA=(R=1,G=0,B=0,A=1),ColorB=(R=0,G=0,B=1,A=1))
   bComputeIntersectionsBeforeHull=False
   bRandomizeColor=True
  }
 }
 CollectionTransformSelectBySize_NarrowWindow_copy {
  type=FCollectionTransformSelectionBySizeDataflowNode
  guid=#1491
  inputPins [
   Collection {type=FManagedArrayCollection}
   SizeMin {type=float}
   SizeMax {type=float}
  ]
  outputPins [
   Collection {type=FManagedArrayCollection}
   TransformSelection {type=FDataflowTransformSelection}
  ]
  properties {RangeSetting=Dataflow_RangeSetting_InsideRange, SizeMax=31.51, SizeMin=31.49, bInclusive=True, bUseRelativeSize=False}
 }
 AtomDestructionFixISMMaterialsDataflowNode {
  type=FAtomDestructionFixISMMaterialsDataflowNode
  guid=#1492
  inputPins [
   GCAsset {type="TObjectPtr<UGeometryCollection>"}
   InstancedMeshes {type="TArray<FGeometryCollectionAutoInstanceMesh>"}
  ]
  outputPins [InstancedMeshes {type="TArray<FGeometryCollectionAutoInstanceMesh>"}]
  properties {ISMMaterialDataTable=None}
 }
 Proximity Distance - ATOM Default {
  type=FMakeLiteralStringDataflowNode
  guid=#1493
  inputPins []
  outputPins [String {type=FString}]
  properties {Value=2}
 }
 Proximity Contact - ATOM Default {
  type=FMakeLiteralStringDataflowNode
  guid=#1494
  inputPins []
  outputPins [String {type=FString}]
  properties {Value=8}
 }
 ClusterMergeToNeighbors_copy {
  type=FClusterMergeToNeighborsDataflowNode
  guid=#1495
  inputPins [
   Collection {type=FManagedArrayCollection}
   TransformSelection {type=FDataflowTransformSelection}
   MinVolumeCubeRoot {type=float}
   bOnlyToConnected {type=bool}
   bOnlySameParent {type=bool}
  ]
  outputPins [Collection {type=FManagedArrayCollection}]
  properties {
   MinVolumeCubeRoot=16
   NeighborSelectionMethod=Dataflow_ClusterNeighborSelectionMethod_NearestCenter
   bOnlySameParent=True
   bOnlyToConnected=True
  }
 }
 L1 by Size - ATOM Default {
  type=FMakeLiteralFloatDataflowNode
  guid=#1496
  inputPins []
  outputPins [Float {type=float}]
  properties {Value=16}
 }
 IntToString_1 {
  type=FIntToStringDataflowNode
  guid=#1497
  inputPins [Int {type=int32}]
  outputPins [String {type=FString}]
  properties {Int=0}
 }
 LogString {
  type=FLogStringDataflowNode
  guid=#1498
  inputPins [String {type=FString}]
  outputPins []
  properties {String="", bPrintToLog=True}
 }
]
