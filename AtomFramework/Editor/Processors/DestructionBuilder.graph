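# Editor utility that converts selected Atom (LEGO) models, or each of a model's selection sets, into destructible Geometry Collection assets and optional Blueprints.
# Output paths come from folder and prefix swaps, unless a Kragle or supplied path applies.
# Each collection gets root proxy meshes, runs an asset-specific, folder-local or global Dataflow in that priority, then receives standard properties and reset overrides that retain listed keys.
# Blueprints get the collection component with replication, collision and break-notify settings before compiling.
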EditorUtilityBlueprint /AtomFramework/Editor/Processors/DestructionBuilder
graphs [
 AssignAndExecuteDataFlowToGC {nodes [
   n1 = DoesAssetExist(AssetPath=Get Global_DF_Asset_Name() -> (Global_DF_Asset_Name:String))
   n2 = Get GlobalDataFlowAsset() -> (GlobalDataFlowAsset:Dataflow Object Reference)
   n3 = Branch(execute=[@n1.then, @n19.else], Condition=IsValid(Object=@n2))
   n4 = Inputs!() -> (execute, GC Asset:Geometry Collection Object Reference)
   > n5 = GetPathNameForLoadedAsset(LoadedAsset=@n4)
   n6 = Split!(InPath:String=@n5) -> (PathPart:String, FilenamePart:String, ExtensionPart:String)
   n7 = Replace(SourceString=@n6, From=Get Folder_GC() -> (Folder_GC:String), To=Get Folder_DF() -> (Folder_DF:String))
   n8 = Append!(A=@n7, B=/, C:String=Get Local_DF_Asset_Name() -> (Local_DF_Asset_Name:String)) -> (ReturnValue)
   n9 = Append!(
    A=@n7
    B=Get Folder_DF_Specific() -> (Folder_DF_Specific:String)
    C=/
    D:String=Get DF_Prefix() -> (DF_Prefix:String)
    E:String=@n6.FilenamePart
   ) -> (ReturnValue)
   n10 = Merge Execution with Data!(Option A:Exec=@n21.then, Option A Data:String=@n9, Option B:Exec=@n19.then, Option B Data:String=@n8) -> (Then:Exec, Data:String) {is=title K2Node_MacroInstance}
   n11 = Get Object Path String(Object=@n2)
   n12 = LeftChop(SourceString=@n11, Count=Len(S=GetExtension!(InPath:String=@n11, bIncludeDot:bool=true) -> (ReturnValue:String)))
   > n13 = Merge Execution with Data(Option A Data=@n10, Option B=@n3.then, Option B Data=@n12)
   > n14 = LoadAsset(AssetPath=@n13)
   > n15 = Cast To Dataflow!(execute, Object=@n14) -> (then, CastFailed, AsDataflow:Dataflow Object Reference) {is=title K2Node_DynamicCast}
   > n16 = Set DataflowAsset!(execute, DataflowAsset:Dataflow Object Reference=@n15, self:Geometry Collection Object Reference=@n4) -> (then, Output_Get:Dataflow Object Reference)
   > EvaluateTerminalNodeByName!(execute, Dataflow:Dataflow Object Reference=@n16, TerminalNodeName:Name=GeometryCollectionTerminal, ResultAsset:Object Reference=@n4) -> (then)
   > Outputs!(Done:Exec, Failed:Exec=[@n15.CastFailed, @n17.then])
   n17 = PrintString(execute=@n3.else, InString=Append!(A="Could not find DataFlow asset : ", B=@n12) -> (ReturnValue))
   n18 = DoesAssetExist(execute=@n21.else, AssetPath=@n8)
   > n19 = Branch(Condition=@n18)
   n20 = DoesAssetExist(execute=@n5.then, AssetPath=@n9)
   > n21 = Branch(Condition=@n20)
  ]}
 AssignRootMeshProxyToGC {nodes [
   n1 = AssignRootMeshProxyToGC!() -> (then, GCAsset:Geometry Collection Object Reference) {is=title K2Node_FunctionEntry}
   n2 = Get!(self:Geometry Collection Object Reference=@n1) -> (
    RootProxyData_ProxyMeshes:Array of Static Mesh Object References
    RootProxyData_MeshTransforms:Array of Transform 3f Structures
    RootProxyData_MeshOverrideMaterials:Array of Geometry Collection Proxy Mesh Materials Structures
   )
   > Clear(TargetArray:Array of Static Mesh Object References=@n2)
   > n3 = GetPathNameForLoadedAsset(LoadedAsset=@n1)
   > n4 = Sequence()
   n5 = Split(InPath=@n3)
   n6 = Append!(
    A=Replace(SourceString=@n5, From=Get Folder_GC(), To=Get Folder_SM() -> (Folder_SM:String), SearchCase=CaseSensitive)
    B=/
    C:String=Replace(SourceString=@n5.FilenamePart, From=Get GC_Prefix() -> (GC_Prefix:String), To=Get SM_Prefix() -> (SM_Prefix:String))
   ) -> (ReturnValue)
   > n7 = LoadAsset(AssetPath=@n6)
   > n8 = Is Valid(InputObject=@n7)
   n9 = PrintString(execute=@n8.Is Not Valid, InString=Append!(A="WARNING :  Could not find  root mesh proxy ", B=@n6) -> (ReturnValue))
   Return Node(execute=[@n9.then, @n10.CastFailed])
   n10 = Cast To StaticMesh(execute=@n8.Is Valid, Object=@n7)
   > Add Unique(TargetArray:Array of Static Mesh Object References=@n2, NewItem:"Static Mesh Object Reference (by ref)"=@n10)
   n11 = For Each Loop(Exec=@n4.then_1, Array=Get ValidSuffixes() -> (ValidSuffixes:Array of Strings))
   n12 = Append(A=@n6, B=@n11)
   > n13 = DoesAssetExist(AssetPath=@n12)
   > Branch(Condition=@n13)
   > n14 = LoadAsset(AssetPath=@n12)
   > Is Valid(InputObject=@n14)
   > n15 = Cast To StaticMesh(Object=@n14)
   > Add Unique(TargetArray=@n2, NewItem=@n15)
  ]}
 ConvertGCPathTOBPPath {nodes [
   n1 = Split(InPath=Inputs!() -> (GC Asset Path:String))
   n2 = Get Folder_BP() -> (Folder_BP:String)
   n3 = Get Folder_GC()
   n4 = Get BP_Prefix() -> (BP_Prefix:String)
   n5 = Get GC_Prefix()
   Outputs!(BP Asset Path:String=Append!(
     A=Replace(SourceString=@n1, From=@n3, To=@n2, SearchCase=CaseSensitive)
     B=/
     C:String=Replace(SourceString=@n1.FilenamePart, From=@n5, To=@n4)
    ) -> (ReturnValue))
  ]}
 ConvertLMPathToGCPath {nodes [
   n1 = Inputs!() -> (LegoModel Asset Path:String, SelectionSet:Name)
   n2 = Split(InPath=@n1)
   n3 = Get Folder_GC()
   n4 = Get Folder_LM() -> (Folder_LM:String)
   n5 = Replace(SourceString="To String (Name)"(InName=@n1.SelectionSet), From=" ", To=_)
   n6 = Replace(
    SourceString=SelectString(A=@n2.FilenamePart, B=Append(A=@n2.FilenamePart, B=_, C=@n5), bPickA="Equal (Name)"(A=@n1.SelectionSet))
    From=Get LM_Prefix() -> (LM_Prefix:String)
    To=Get GC_Prefix()
    SearchCase=CaseSensitive
   )
   n7 = Contains(SearchIn=@n6, Substring=Get GC_Prefix(), bUseCase=true)
   n8 = Append!(A=Get GC_Prefix(), B=@n6) -> (ReturnValue)
   Outputs!(GC Asset Path:String=Append!(
     A=Replace(SourceString=@n2, From=@n4, To=@n3, SearchCase=CaseSensitive)
     B=/
     C:String=Select(Option 0:String=@n8, Option 1:String=@n6, Index=@n7) -> (ReturnValue:String)
    ) -> (ReturnValue))
  ]}
 ConvertLegoModelToBP {nodes [
   ConvertLegoModelToBP() {is=title K2Node_FunctionEntry}
   > PrintString(InString="=== Convert Lego Model to BP - START ===")
   > n1 = GetSelectedAssets()
   > n2 = For Each Loop(Array:Array of Object References=@n1) -> (Array Element:Object Reference)
   PrintString(execute=@n2.Completed, InString="=== Convert Lego Model to BP - END ===")
   > Return Node()
   n3 = Cast To AtomModel!(execute=@n2.LoopBody, Object=@n2) -> (then, CastFailed, AsAtom Model:Atom Model Object Reference) {is=title K2Node_DynamicCast}
   > Make GC From Atom Model!(
    execute
    self
    LM Asset:Atom Model Object Reference=@n3
    SelectionSet:Name=None
    UpdateOrCreateBlueprint:bool=true
    OptionalGCAssetPath:Name=None
    Custom Primitive Processor:Atom Model Processor Object Reference
    Procesed Primitive Settings:Atom On Process Primitive Settings Structure
    Material Factory:Atom Material Factory Object Reference
    BPCreationArguments:"Atom Model Blueprint Creation Arguments Structure (by ref)"=Make Atom Model Blueprint Creation Arguments!(ParentClass:Object Class Reference=&Actor) -> (AtomModelBlueprintCreationArguments:Atom Model Blueprint Creation Arguments Structure) {is=title K2Node_MakeStruct}
   ) -> (then, GCCreated:Geometry Collection Object Reference, BPCreated:Object Reference, bSuccess:bool)
  ]}
 ConvertLegoModelToGC {nodes [
   ConvertLegoModelToGC() {is=title K2Node_FunctionEntry}
   > PrintString(InString="=== Convert LM to GC - START ===")
   > n1 = GetSelectedAssets()
   > n2 = For Each Loop(Array=@n1)
   PrintString(execute=@n2.Completed, InString="=== Convert LM to GC  - END ===")
   > Return Node()
   n3 = Cast To AtomModel(execute=@n2.LoopBody, Object=@n2)
   > Make GC From Atom Model(
    LM Asset=@n3
    SelectionSet=None
    UpdateOrCreateBlueprint=false
    OptionalGCAssetPath=None
    BPCreationArguments=Make Atom Model Blueprint Creation Arguments(ParentClass=&Actor)
   )
  ]}
 EventGraph {links [], nodes []}
 GetOrCreateBPAsset {nodes [
   n1 = Inputs!() -> (InputPin:Exec, BP Asset Path:String, CreationArguments:"Atom Model Blueprint Creation Arguments Structure (by ref)")
   > n2 = GetOrCreateBlueprintAsset!(execute, BPAssetPath:String=@n1, CreationArguments:"Atom Model Blueprint Creation Arguments Structure (by ref)"=@n1.CreationArguments) -> (then, bWasCreatedThisCall:bool, ReturnValue:Blueprint Object Reference)
   > n3 = Is Valid(InputObject=@n2.ReturnValue)
   > Outputs!(Done:Exec, Failed:Exec=@n3.Is Not Valid, BP Asset:Blueprint Object Reference=@n2.ReturnValue, bWasCreatedThisCall:bool=@n2)
  ]}
 GetOrCreateGCAsset {nodes [
   n1 = Inputs!() -> (InputPin, GC Asset Path:String)
   > n2 = DoesAssetExist(AssetPath=@n1)
   > n3 = Branch(Condition=@n2)
   > n4 = LoadAsset(AssetPath=@n1)
   > n5 = Cast To GeometryCollection!(execute, Object=@n4) -> (then, CastFailed, AsGeometry Collection:Geometry Collection Object Reference) {is=title K2Node_DynamicCast}
   n6 = GetAssetTools(execute=@n3.else)
   n7 = Split(InPath=@n1)
   > n8 = CreateAsset(self=@n6, AssetName=@n7.FilenamePart, PackagePath=@n7, AssetClass=/Script/GeometryCollectionEngine.GeometryCollection)
   > n9 = Cast To GeometryCollection(Object=@n8)
   n10 = Merge Execution with Data(
    Option A=@n5.then
    Option A Data:Geometry Collection Object Reference=@n5
    Option B=@n9.then
    Option B Data:Geometry Collection Object Reference=@n9
   ) -> (Data:Geometry Collection Object Reference)
   > Outputs!(Done, GC Asset:Geometry Collection Object Reference=@n10, Failed=[@n5.CastFailed, @n9.CastFailed])
  ]}
 Make GC From Atom Model {nodes [
   n1 = Make GC From Atom Model!() -> (
    then
    LM Asset:Atom Model Object Reference
    SelectionSet:Name
    UpdateOrCreateBlueprint:bool=~
    OptionalGCAssetPath:Name
    Custom Primitive Processor:Atom Model Processor Object Reference
    Procesed Primitive Settings:Atom On Process Primitive Settings Structure
    Material Factory:Atom Material Factory Object Reference
    BPCreationArguments:"Atom Model Blueprint Creation Arguments Structure (by ref)"
   ) {is=title K2Node_FunctionEntry}
   > n2 = Branch(Condition="Equal (Name)"(A=@n1.OptionalGCAssetPath))
   > n3 = GetPathNameForLoadedAsset(LoadedAsset=@n1)
   > n4 = Set GCAssetPath(GCAssetPath:String=Convert LMPath to GCPath!(LegoModel Asset Path:String=@n3, SelectionSet:Name=@n1.SelectionSet) -> (GC Asset Path:String) {is=title K2Node_MacroInstance}) -> (Output_Get:String)
   n5 = TryGetKragleGCPath!(execute=[@n4.then, @n19.then], self, AtomModel:Atom Model Object Reference=@n1) -> (then, Success:bool, ReturnValue:String)
   n6 = SelectString(A=@n5.ReturnValue, B=Get GCAssetPath() -> (GCAssetPath:String), bPickA=@n5)
   > n7 = Get or Create GCAsset!(InputPin:Exec, GC Asset Path:String=@n6) -> (Done:Exec, GC Asset:Geometry Collection Object Reference, Failed:Exec) {is=title K2Node_MacroInstance}
   > AssignRootMeshProxyToGC!(execute, self, GCAsset:Geometry Collection Object Reference=@n7) -> (then) {is=title K2Node_CallFunction}
   > SetGeometryCollectionSourceFromAtomModel!(
    execute
    GeometryCollection:Geometry Collection Object Reference=@n7
    AtomModel:Atom Model Object Reference=@n1
    CustomPrimitiveProcessor:Atom Model Processor Object Reference=Get Custom Primitive Processor() -> (Custom Primitive Processor:Atom Model Processor Object Reference)
    PrimitiveSettings:"Atom On Process Primitive Settings Structure (by ref)"=Get Procesed Primitive Settings() -> (Procesed Primitive Settings:Atom On Process Primitive Settings Structure)
    MaterialFactory:Atom Material Factory Object Reference=Get Material Factory() -> (Material Factory:Atom Material Factory Object Reference)
   ) -> (then)
   > n8 = Assign and Execute Data Flow to GC!(execute:Exec, GC Asset:Geometry Collection Object Reference=@n7) -> (Done:Exec, Failed:Exec) {is=title K2Node_MacroInstance}
   > UpdateGCProperties!(execute, self, GC asset:Geometry Collection Object Reference=@n7) -> (then)
   > SetAtomModelGeometryCollectionImportAsset!(
    execute
    GeometryCollection:Geometry Collection Object Reference=@n7
    AtomModel:Atom Model Object Reference=@n1
    SelectionSet:Name=@n1.SelectionSet
   ) -> (then)
   > n9 = Branch(Condition=@n1.UpdateOrCreateBlueprint)
   > n10 = Get or Create BPAsset!(
    InputPin:Exec
    BP Asset Path:String=Convert GCPath TOBPPath!(GC Asset Path:String=@n6) -> (BP Asset Path:String) {is=title K2Node_MacroInstance}
    CreationArguments:"Atom Model Blueprint Creation Arguments Structure (by ref)"=Get BPCreationArguments() -> (BPCreationArguments:"Atom Model Blueprint Creation Arguments Structure (by ref)")
   ) -> (Done:Exec, Failed:Exec, BP Asset:Blueprint Object Reference, bWasCreatedThisCall:bool=~) {is=title K2Node_MacroInstance}
   > Set WasCreatedThisCall(WasCreatedThisCall:bool=@n10.bWasCreatedThisCall) -> (Output_Get:bool)
   > n11 = Gather Subobject Data For Blueprint(self=Get SubobjectDataSubsystem(), Context=@n10)
   > n12 = For Each Loop with Break!(Exec:Exec, Array:Array of Subobject Data Handle Structures=@n11, Break:Exec=@n16.then) -> (LoopBody:Exec, Array Element:Subobject Data Handle Structure, Array Index:int, Completed:Exec)
   > n13 = Cast To GeometryCollectionComponent!(
    execute
    Object=GetObject!(Data:"Subobject Data Structure (by ref)"=GetData(DataHandle=@n12), bEvenIfPendingKill:bool=false) -> (ReturnValue:Object Reference)
   ) -> (then, CastFailed, AsGeometry Collection Component:Geometry Collection Component Object Reference) {is=title K2Node_DynamicCast}
   n14 = "To String (Name)"(InName=Get GC_Component_Name() -> (GC_Component_Name:Name=None))
   > Branch(Condition="Equal Exactly (String)"(A=GetObjectName(Object=@n13), B=Append!(A=@n14, B=_GEN_VARIABLE) -> (ReturnValue)))
   > n15 = Set GC_Component(GC_Component:Geometry Collection Component Object Reference=@n13) -> (Output_Get:Geometry Collection Component Object Reference)
   > n16 = SetRestCollection!(
    execute
    self:Geometry Collection Component Object Reference=@n15
    RestCollectionIn:Geometry Collection Object Reference=@n7
    bApplyAssetDefaults:bool=true
   ) -> (then)
   n17 = Is Valid(exec=@n12.Completed, InputObject=Get GC_Component() -> (GC_Component:Geometry Collection Component Object Reference))
   n18 = AddGeometryCollectionComponentToBlueprint!(execute=@n17.Is Not Valid, Blueprint:Blueprint Object Reference=@n10, GeometryCollection:Geometry Collection Object Reference=@n7) -> (then)
   UpdateBPProperties!(
    execute=[@n18.then, @n17.Is Valid]
    self
    Blueprint Asset:Blueprint Object Reference=@n10
    WasCreatedThisTime:bool=Get WasCreatedThisCall() -> (WasCreatedThisCall:bool)
   ) -> (then)
   > SetEditorProperty(
    Object=@n10
    PropertyName=bRunConstructionScriptOnDrag
    PropertyValue=Get Run Construction Script on Drag() -> (Run Construction Script on Drag:bool)
   )
   > CompileBlueprint!(execute, Blueprint:Blueprint Object Reference=@n10) -> (then, ReturnValue:bool)
   > Return Node!(execute, GCCreated:Geometry Collection Object Reference=@n7, BPCreated:Object Reference=@n10, bSuccess:bool=true)
   ReportProcessorError!(
    execute=@n10.Failed
    Message:"Text (by ref)"=Format Text(Format="LOCTEXT(\"Fail to get or create the blueprint asset\")")
    Asset:Object Reference=Get LM Asset() -> (LM Asset:Atom Model Object Reference)
   ) -> (then)
   > Return Node(GCCreated=@n7, BPCreated=@n10, bSuccess=false)
   Return Node(execute=@n9.else, GCCreated=@n7, bSuccess=true)
   ReportProcessorError(execute=@n8.Failed, Message=Format Text(Format="LOCTEXT(\"Failed to assign and execute Dataflow\")"), Asset=Get LM Asset())
   > Return Node(GCCreated=@n7, bSuccess=false)
   ReportProcessorError(execute=@n7.Failed, Message=Format Text(Format="LOCTEXT(\"Fail to get or create the GC asset\")"), Asset=Get LM Asset())
   > Return Node(GCCreated=@n7, bSuccess=true)
   n19 = Set GCAssetPath(execute=@n2.else, GCAssetPath="To String (Name)"(InName=@n1.OptionalGCAssetPath))
  ]}
 MakeGCfromSelectionSet {nodes [
   n1 = MakeGCfromSelectionSet!() -> (then, LM Asset:Object Reference, Convert BP:bool=~) {is=title K2Node_FunctionEntry}
   > n2 = Cast To AtomModel(Object=@n1)
   > n3 = For Each Loop(Array:Array of Atom Model Selection Set Structures=Get SelectionSets!(self:Atom Model Object Reference=@n2) -> (SelectionSets:Array of Atom Model Selection Set Structures)) -> (Array Element:Atom Model Selection Set Structure)
   n4 = Break Atom Model Selection Set!(AtomModelSelectionSet:"Atom Model Selection Set Structure (by ref)"=@n3) -> (PrimitiveIds:Set of Guid Structures, SelectionSetName:Name) {is=title K2Node_BreakStruct}
   > PrintString(InString=Append(A="Exporting selection set : ", B="To String (Name)"(InName=@n4.SelectionSetName)))
   > Make GC From Atom Model(
    LM Asset=@n2
    SelectionSet=@n4.SelectionSetName
    UpdateOrCreateBlueprint=@n1.Convert BP
    OptionalGCAssetPath=None
    BPCreationArguments=Make Atom Model Blueprint Creation Arguments(ParentClass=&Actor)
   ) {is=title K2Node_CallFunction}
  ]}
 MergeExecutionWithData {nodes [
   n1 = Inputs!() -> (Option A:Exec, Option A Data:Wildcard, Option B:Exec, Option B Data:Wildcard)
   n2 = Local Boolean!() -> (Variable:bool=~) {is=title K2Node_TemporaryVariable}
   > n3 = Assign!(execute:Exec, Variable:bool=@n2, Value:bool) -> (then:Exec) {is=title K2Node_AssignmentStatement}
   Outputs!(
    Then:Exec=[@n4.then, @n3.then]
    Data:Wildcard=Select(Option 0:Wildcard=@n1, Option 1:Wildcard=@n1.Option B Data, Index=@n2) -> (ReturnValue:Wildcard)
   )
   n4 = Assign(execute=@n1.Option B, Variable=@n2, Value=true)
  ]}
 TryGetKragleGCPath {nodes [
   n1 = TryGetKragleGCPath!() -> (then, AtomModel:Atom Model Object Reference) {is=title K2Node_FunctionEntry}
   > n2 = TryGetKraglePath!(execute, AtomModel:Atom Model Object Reference=@n1) -> (then, OutKraglePath:String, ReturnValue:bool)
   > n3 = Branch(Condition=@n2.ReturnValue)
   > n4 = TryGetKragleName!(execute, AtomModel:Atom Model Object Reference=@n1) -> (then, OutKragleName:String, ReturnValue:bool)
   > n5 = Branch(Condition=@n4.ReturnValue)
   > Return Node!(execute, Success:bool=true, ReturnValue:String=Append!(A=@n2, B=/, C:String=GC_, D:String=@n4) -> (ReturnValue))
   Return Node(execute=[@n3.else, @n5.else], Success=false)
  ]}
 UpdateBPProperties {nodes [
   n1 = UpdateBPProperties!() -> (then, Blueprint Asset:Blueprint Object Reference, WasCreatedThisTime:bool=~) {is=title K2Node_FunctionEntry}
   > n2 = Is Valid(InputObject=@n1)
   PrintString(execute=@n2.Is Not Valid, InString=Invalid blueprint asset)
   n3 = Gather Subobject Data For Blueprint(execute=@n2.Is Valid, self=Get SubobjectDataSubsystem(), Context=@n1)
   > n4 = For Each Loop(Array:Array of Subobject Data Handle Structures=@n3) -> (Array Element:Subobject Data Handle Structure)
   > n5 = Sequence()
   n6 = GetObject(Data=GetData(DataHandle=@n4), bEvenIfPendingKill=false)
   > n7 = Cast To StaticMeshComponent(Object=@n6)
   > SetStaticMesh(self=@n7, NewMesh=Get NoStaticMesh() -> (NoStaticMesh:Static Mesh Object Reference))
   n8 = Cast To GeometryCollectionComponent(execute=@n5.then_1, Object=@n6)
   > Set bEnableReplication!(execute, bEnableReplication:bool=true, self:Geometry Collection Component Object Reference=@n8) -> (then, Output_Get:bool)
   > Set ReplicationAbandonAfterLevel!(execute, ReplicationAbandonAfterLevel:int=1, self:Geometry Collection Component Object Reference=@n8) -> (then, Output_Get:int=0)
   > Set bEnableAbandonAfterLevel!(execute, bEnableAbandonAfterLevel:bool=true, self:Geometry Collection Component Object Reference=@n8) -> (then, Output_Get:bool)
   > Set ReplicationMaxPositionAndVelocityCorrectionLevel!(execute, ReplicationMaxPositionAndVelocityCorrectionLevel:int=0, self:Geometry Collection Component Object Reference=@n8) -> (then, Output_Get:int=0)
   > Set OneWayInteractionLevel!(execute, OneWayInteractionLevel:int=1, self:Geometry Collection Component Object Reference=@n8) -> (then, Output_Get:int=0)
   > Set CollisionProfilePerLevel!(
    execute
    CollisionProfilePerLevel:Array of Names=Get PerLevelProfile() -> (PerLevelProfile:Array of Names)
    self:Geometry Collection Component Object Reference=@n8
   ) -> (then, Output_Get:Array of Names)
   > SetCollisionProfileName!(
    execute
    self:Primitive Component Object Reference=@n8
    InCollisionProfileName:Name=Get GC Collision Profile Name() -> (GC Collision Profile Name:Name=None)
    bUpdateOverlaps:bool=false
   ) -> (then)
   > SetEnableDamageFromCollision!(execute, self:Geometry Collection Component Object Reference=@n8, bValue:bool=false) -> (then)
   > Set InitialVelocityType!(
    execute
    InitialVelocityType:EInitialVelocityTypeEnum Enum=Chaos_Initial_Velocity_None
    self:Geometry Collection Component Object Reference=@n8
   ) -> (then, Output_Get:EInitialVelocityTypeEnum Enum=Chaos_Initial_Velocity_User_Defined)
   > Set MaxSimulatedLevel!(execute, MaxSimulatedLevel:int=1, self:Geometry Collection Component Object Reference=@n8) -> (then, Output_Get:int=0)
   > SetNotifyGlobalBreaks!(execute, self:Geometry Collection Component Object Reference=@n8, bNewNotifyGlobalBreaks:bool=true) -> (then)
   > SetNotifyGlobalCollision!(execute, self:Geometry Collection Component Object Reference=@n8, bNewNotifyGlobalCollisions:bool=true) -> (then)
   SetComponentAffectNavigationGeneration!(execute, Component:Actor Component Object Reference, bAffectNavigation:bool=false) -> (then)
  ]}
 UpdateGCProperties {nodes [
   n1 = UpdateGCProperties!() -> (then, GC asset:Geometry Collection Object Reference) {is=title K2Node_FunctionEntry}
   > Set bImportCollisionFromSource!(execute, bImportCollisionFromSource:bool=false, self:Geometry Collection Object Reference=@n1) -> (then, Output_Get:bool)
   > Set Mass!(execute, Mass:float=100, self:Geometry Collection Object Reference=@n1) -> (then, Output_Get:float)
   > Set bStripOnCook!(execute, bStripOnCook:bool=true, self:Geometry Collection Object Reference=@n1) -> (then, Output_Get:bool)
   > Set bStripRenderDataOnCook!(execute, bStripRenderDataOnCook:bool=true, self:Geometry Collection Object Reference=@n1) -> (then, Output_Get:bool)
   > Set bScaleOnRemoval!(execute, bScaleOnRemoval:bool=false, self:Geometry Collection Object Reference=@n1) -> (then, Output_Get:bool)
   > Set CustomRendererType!(execute, CustomRendererType:Object Class Reference=/Script/AtomRuntime.AtomISMPoolRenderer, self:Geometry Collection Object Reference=@n1) -> (then, Output_Get:Object Class Reference)
   n2 = Get Overrides!(self:Geometry Collection Object Reference=@n1) -> (Overrides:Map of Strings to Strings)
   > Set Old DataFlow Overrides(Old DataFlow Overrides:Map of Strings to Strings=@n2) -> (Output_Get:Map of Strings to Strings)
   > Clear!(execute, TargetMap:Map of Strings to Strings=@n2) -> (then) {is=title K2Node_CallFunction}
   > n3 = For Each Loop(Array:Array of Strings=Get GC_MakeOverrides() -> (GC_MakeOverrides:Array of Strings)) -> (Array Element:String)
   n4 = Find(TargetMap=Get Old DataFlow Overrides() -> (Old DataFlow Overrides:Map of Strings to Strings), Key=@n3)
   > n5 = Branch(Condition=@n4.ReturnValue)
   > Add!(execute, TargetMap:Map of Strings to Strings=@n2, Key:"String (by ref)"=@n3, Value:"String (by ref)"=@n4) -> (then) {is=title K2Node_CallFunction}
   Add(execute=@n5.else, TargetMap=@n2, Key=@n3)
  ]}
 UpdateGCProxyMeshes {nodes [
   n1 = UpdateGCProxyMeshes!() -> (then, GC Asset:Geometry Collection Object Reference) {is=title K2Node_FunctionEntry}
   > AssignRootMeshProxyToGC(GCAsset=@n1)
  ]}
]
variables [
 BP_Prefix {type=string, default=BP_}
 DF_Prefix {type=string, default=DF_}
 Folder_BP {type=string, default=/Blueprints}
 Folder_DF {type=string, default=/DF}
 Folder_DF_Specific {type=string, default=/Assets}
 Folder_GC {type=string, default=/GC}
 Folder_LM {type=string, default=/Meshes}
 Folder_SM {type=string, default=/Meshes}
 FoldersToExport {type="array<string>", default []}
 GC Collision Profile Name {type=string, default=FortBuildingMeshPhysics}
 GC_Component_Name {type=string, default=GeometryCollection}
 GC_MakeOverrides {
  type="array<string>"
  default [
   Root - Convexes by Number
   Root - Convexes by Error Tolerance
   Proximity - Distance Threshold
   Proximity - Contact Threshold
   Level 1 - by Target Resolution
   "Level 1 - by  Cluster Number"
   Level 1 - Cluster Convexes
   Convex - Protection Tolerance
   Convex - Protection Radius
   Cleanup_L1_bySize
   Cleanup - SimplificationAngle
   Cleanup - SimplificationDist
   Cleanup - SimplificationMinFace
  ]
 }
 GC_Prefix {type=string, default=GC_}
 Global_DF_Asset_Name {type=string, default=/AtomFramework/Editor/Dataflow/DF_GC_Global_V4}
 GlobalDataFlowAsset {type=Dataflow, default=/AtomFramework/Editor/Dataflow/DF_GC_Global_V4_ATOM}
 InvalidSelectionSet {type=string, default=None}
 LM_Prefix {type=string, default=LEGO_}
 Local_DF_Asset_Name {type=string, default=DF_GC_Local_All}
 Run Construction Script on Drag {type=boolean}
 SM_Prefix {type=string, default=SM_}
 Variable {type=boolean}
]
