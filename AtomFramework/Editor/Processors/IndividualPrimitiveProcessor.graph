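# Editor utility processor for Atom models, named for individual primitives, that supplies default materials and mesh merging settings.
# GetDefaultMaterialForType switches on the Atom material type, returning Game_Qual primitive material instances for Standard, Transparent, Glitter, Opalescent and Metallic; None is unconnected.
# GetMeshMergingSettings calls the parent, sets members on broken-out sub-structs for requested, simplified, approximate, coarse and common-part LODs, then returns the parent's struct.
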EditorUtilityBlueprint /AtomFramework/Editor/Processors/IndividualPrimitiveProcessor
graphs [
 EventGraph {links [], nodes []}
 GetDefaultMaterialForType {nodes [
   n1 = GetDefaultMaterialForType!() -> (then, "Atom Material  Type":EAtomMaterialType Enum=None, Part Info:"Atom Model Part Instance Info Structure (by ref)") {is=title K2Node_FunctionEntry}
   > n2 = Switch on EAtomMaterialType!(execute, Selection:EAtomMaterialType Enum=@n1) -> (None:Exec, Standard:Exec, Transparent:Exec, Glitter:Exec, Opalescent:Exec, Metallic:Exec) {is=title K2Node_SwitchEnum}
   Return Node!(execute=@n2.Glitter, Material:Material Interface Object Reference=/AtomFramework/Materials/Game_Qual/MI_Primitive_Glit)
   Return Node(execute=@n2.Metallic, Material=/AtomFramework/Materials/Game_Qual/MI_Primitive_Metal)
   Return Node(execute=@n2.Opalescent, Material=/AtomFramework/Materials/Game_Qual/MI_Primitive_Opal)
   Return Node(execute=@n2.Standard, Material=/AtomFramework/Materials/Game_Qual/MI_Primitive_Standard)
   Return Node(execute=@n2.Transparent, Material=/AtomFramework/Materials/Game_Qual/MI_Primitive_Trans)
  ]}
 GetMeshMergingSettings {nodes [
   n1 = GetMeshMergingSettings!() -> (then, AtomModel:Atom Model Object Reference) {is=title K2Node_FunctionEntry}
   > n2 = Parent: GetMeshMergingSettings!(execute, AtomModel:Atom Model Object Reference=@n1) -> (then, Regular Mesh Merging Settings:Atom Mesh Merging Settings Structure) {is=title K2Node_CallParentFunction}
   n3 = Break Atom Mesh Merging Settings!(AtomMeshMergingSettings:"Atom Mesh Merging Settings Structure (by ref)"=@n2) -> (
    RequestedLODsSettings:Atom Mesh Merging Requested LODs Settings Structure
    SimplifiedLODsSettings:Atom Mesh Merging Simplified LODs Settings Structure
    ApproximateLODsSettings:Atom Mesh Merging Approximate LODs Settings Structure
    CoarseLODsSettings:Atom Mesh Merging Coarse LODs Settings Structure
    CommonPartsSettings:Atom Mesh Merging Common Parts Settings Structure
   ) {is=title K2Node_BreakStruct}
   > Set members in Atom Mesh Merging Common Parts Settings!(
    execute:Exec
    StructRef:"Atom Mesh Merging Common Parts Settings Structure (by ref)"=@n3.CommonPartsSettings
    FilterCommonPartsLODLevel:int=3
    ApproximateCommonPartLODs:int=0
   ) -> (then:Exec, StructOut:"Atom Mesh Merging Common Parts Settings Structure (by ref)") {is=title K2Node_SetFieldsInStruct}
   > Set members in Atom Mesh Merging Coarse LODs Settings!(
    execute:Exec
    StructRef:"Atom Mesh Merging Coarse LODs Settings Structure (by ref)"=@n3.CoarseLODsSettings
    NumCoarseLODs:int=1
    CoarseLODBaseTolerance:float=0.5
   ) -> (then:Exec, StructOut:"Atom Mesh Merging Coarse LODs Settings Structure (by ref)") {is=title K2Node_SetFieldsInStruct}
   > Set members in Atom Mesh Merging Approximate LODs Settings!(
    execute:Exec
    StructRef:"Atom Mesh Merging Approximate LODs Settings Structure (by ref)"=@n3.ApproximateLODsSettings
    ApproximationSourceLOD:int=1
   ) -> (then:Exec, StructOut:"Atom Mesh Merging Approximate LODs Settings Structure (by ref)") {is=title K2Node_SetFieldsInStruct}
   > Set members in Atom Mesh Merging Simplified LODs Settings!(execute:Exec, StructRef:"Atom Mesh Merging Simplified LODs Settings Structure (by ref)"=@n3.SimplifiedLODsSettings, NumSimplifiedLODs:int=1) -> (then:Exec, StructOut:"Atom Mesh Merging Simplified LODs Settings Structure (by ref)") {is=title K2Node_SetFieldsInStruct}
   > Set members in Atom Mesh Merging Requested LODs Settings!(execute:Exec, StructRef:"Atom Mesh Merging Requested LODs Settings Structure (by ref)"=@n3, NumLODs:int=4) -> (then:Exec, StructOut:"Atom Mesh Merging Requested LODs Settings Structure (by ref)") {is=title K2Node_SetFieldsInStruct}
   > Return Node!(execute, Regular Mesh Merging Settings:Atom Mesh Merging Settings Structure=@n2)
  ]}
]
variables []
