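# Editor utility Atom model processor that turns models into Geometry Collections for destruction, optionally with Blueprints, when finalized or through manual generate functions.
# It instantiates configurable builder, primitive processor and material factory classes, processes primitives lacking target assets, then has the Destruction Builder create the assets.
# Finalization picks Blueprint creation arguments by model path, with a default fallback, and appends created objects to the result.
# Reprocess Primitives force-rebuilds deduplicated primitives from selected models.
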EditorUtilityBlueprint /AtomFramework/Editor/Processors/AtomGeometryCollectionAndBlueprintWriter
graphs [
 EventGraph {links [], nodes []}
 Generate Blueprint {nodes [
   Generate Blueprint() {is=title K2Node_FunctionEntry}
   > GenerateInternal!(
    execute
    self
    GenerateBP:bool=true
    AtomModel:Atom Model Object Reference
    Parent Class:Object Class Reference
    BPCreationArguments:"Atom Model Blueprint Creation Arguments Structure (by ref)"
   ) -> (then, GCCreated:Geometry Collection Object Reference, BPCreated:Object Reference, bSuccess:bool)
  ]}
 Generate Geometry Collection {nodes [
   Generate Geometry Collection() {is=title K2Node_FunctionEntry}
   > GenerateInternal(GenerateBP=false)
  ]}
 GenerateInternal {nodes [
   n1 = GenerateInternal!() -> (
    then
    GenerateBP:bool=~
    AtomModel:Atom Model Object Reference
    Parent Class:Object Class Reference
    BPCreationArguments:"Atom Model Blueprint Creation Arguments Structure (by ref)"
   ) {is=title K2Node_FunctionEntry}
   > InitializeDestructionBuilder()
   > n2 = Sequence()
   > n3 = GetModelPrimitiveAssets!(execute, Model:Atom Model Object Reference=@n1.AtomModel) -> (then, ReturnValue:Array of Atom Primitive Object References)
   > ProcessPrimitives!(execute, self, Primitives:Array of Atom Primitive Object References=@n3, Force Process:bool=false) -> (then)
   n4 = Make GC From Atom Model!(
    execute=@n2.then_1
    self:Destruction Builder Object Reference=Get Destruction Builder() -> (Destruction Builder:Destruction Builder Object Reference)
    LM Asset:Atom Model Object Reference=@n1.AtomModel
    SelectionSet:Name=None
    UpdateOrCreateBlueprint:bool=Get GenerateBP() -> (GenerateBP:bool)
    OptionalGCAssetPath:Name=None
    Custom Primitive Processor:Atom Model Processor Object Reference=Get Primitive Processor() -> (Primitive Processor:Atom Model Processor Object Reference)
    Procesed Primitive Settings:Atom On Process Primitive Settings Structure=Get Processed Primitive Settings() -> (Processed Primitive Settings:Atom On Process Primitive Settings Structure)
    Material Factory:Atom Material Factory Object Reference=Get Material Factory() -> (Material Factory:Atom Material Factory Object Reference)
    BPCreationArguments:"Atom Model Blueprint Creation Arguments Structure (by ref)"=@n1.BPCreationArguments
   ) -> (then, GCCreated:Geometry Collection Object Reference, BPCreated:Object Reference, bSuccess:bool)
   > Return Node!(execute, GCCreated:Geometry Collection Object Reference=@n4, BPCreated:Object Reference=@n4.BPCreated, bSuccess:bool=@n4.bSuccess)
  ]}
 InitializeDestructionBuilder {nodes [
   InitializeDestructionBuilder() {is=title K2Node_FunctionEntry}
   > n1 = Construct Object from Class(Class=Get Destruction Builder Class() -> (Destruction Builder Class:Destruction Builder Class Reference)) -> (ReturnValue:Destruction Builder Object Reference)
   > Set Destruction Builder(Destruction Builder:Destruction Builder Object Reference=@n1) -> (Output_Get:Destruction Builder Object Reference)
   > n2 = Construct Object from Class(Class=Get Primitive Processor Class() -> (Primitive Processor Class:Atom Model Processor Class Reference)) -> (ReturnValue:Atom Model Processor Object Reference)
   > Set Primitive Processor(Primitive Processor:Atom Model Processor Object Reference=@n2) -> (Output_Get:Atom Model Processor Object Reference)
   > n3 = Construct Object from Class(Class=Get Material Factory Class() -> (Material Factory Class:Atom Material Factory Class Reference)) -> (ReturnValue:Atom Material Factory Object Reference)
   > Set Material Factory(Material Factory:Atom Material Factory Object Reference=@n3) -> (Output_Get:Atom Material Factory Object Reference)
  ]}
 OnFinalizeModel {nodes [
   n1 = OnFinalizeModel!() -> (
    then
    Model:Atom Model Object Reference
    AtomModelPartsCollection:"Atom Model Parts Collection Structure (by ref)"
    InputObjects:Array of Object References
   ) {is=title K2Node_FunctionEntry}
   > n2 = Parent: OnFinalizeModel!(
    execute
    Model:Atom Model Object Reference=@n1
    AtomModelPartsCollection:"Atom Model Parts Collection Structure (by ref)"=@n1.AtomModelPartsCollection
    InputObjects:Array of Object References=@n1.InputObjects
   ) -> (then, ReturnValue:Atom Processor Result Structure) {is=title K2Node_CallParentFunction}
   > n3 = GenerateInternal(
    GenerateBP=Get Generate BP() -> (Generate BP:bool)
    AtomModel=@n1
    BPCreationArguments=FindMatchingBlueprintCreationArguments!(
     NameToMatch:String=Get Object Path String(Object=@n1)
     PossibleMatches:Array of Atom Model Blueprint Creation Arguments And String Structures=Get AlternateBlueprintCreationArgs() -> (AlternateBlueprintCreationArgs:Array of Atom Model Blueprint Creation Arguments And String Structures)
     DefaultMatch:"Atom Model Blueprint Creation Arguments Structure (by ref)"=Get DefaultBlueprintCreationArgs() -> (DefaultBlueprintCreationArgs:Atom Model Blueprint Creation Arguments Structure)
    ) -> (ReturnValue:Atom Model Blueprint Creation Arguments Structure)
   ) {is=title K2Node_CallFunction}
   > n4 = Set CreatedObjects(CreatedObjects:Array of Object References=Make Array("[0]":Object Reference=@n3) -> (Array:Array of Object References)) -> (Output_Get:Array of Object References)
   > n5 = Branch(Condition=IsValid(Object=@n3.BPCreated))
   > n6 = Add(TargetArray:Array of Object References=@n4, NewItem:"Object Reference (by ref)"=@n3.BPCreated) {is=title K2Node_CallArrayFunction}
   n7 = AppendAtomProcessorResult!(
    execute=[@n6.then, @n5.else]
    Result:"Atom Processor Result Structure (by ref)"=@n2
    ResultToAppend:"Atom Processor Result Structure (by ref)"=Make Atom Processor Result!(
     bSuccess:bool=@n3.bSuccess
     ProcessedObjects:Array of Object References=Get CreatedObjects() -> (CreatedObjects:Array of Object References)
     SharedAssets:Array of Object References
    ) -> (AtomProcessorResult:Atom Processor Result Structure) {is=title K2Node_MakeStruct}
   ) -> (then, ReturnValue:Atom Processor Result Structure)
   > Return Node!(execute, ReturnValue:Atom Processor Result Structure=@n7)
  ]}
 ProcessPrimitives {nodes [
   n1 = ProcessPrimitives!() -> (then, Primitives:Array of Atom Primitive Object References, Force Process:bool=~) {is=title K2Node_FunctionEntry}
   > n2 = For Each Loop(Array:Array of Atom Primitive Object References=@n1) -> (Array Element:Atom Primitive Object Reference)
   > n3 = Branch(Condition=Get Force Process() -> (Force Process:bool))
   RunCustomPrimitiveProcessor!(
    execute=[@n3.then, @n5.Is Not Valid]
    Processor:Atom Model Processor Object Reference=Get Primitive Processor()
    Primitive:Atom Primitive Object Reference=@n2
    PrimitiveSettings:"Atom On Process Primitive Settings Structure (by ref)"=Get Processed Primitive Settings()
   ) -> (then, ReturnValue:Atom Processor Result Structure)
   n4 = GetCustomPrimitiveProcessorTargetAsset!(
    execute=@n3.else
    Processor:Atom Model Processor Object Reference=Get Primitive Processor()
    Primitive:Atom Primitive Object Reference=@n2
    PrimitiveSettings:"Atom On Process Primitive Settings Structure (by ref)"=Get Processed Primitive Settings()
   ) -> (then, ReturnValue:Object Reference)
   > n5 = Is Valid(InputObject=@n4)
  ]}
 Reprocess Primitives {nodes [
   Reprocess Primitives() {is=title K2Node_FunctionEntry}
   > InitializeDestructionBuilder() {is=title K2Node_CallFunction}
   > n1 = GetSelectedAssets()
   > n2 = For Each Loop(Array:Array of Object References=@n1) -> (Array Element:Object Reference)
   n3 = To Array!(
    execute=@n2.Completed
    A:Set of Atom Primitive Object References=Get PrimitivesSet() -> (PrimitivesSet:Set of Atom Primitive Object References)
   ) -> (then, Result:Array of Atom Primitive Object References)
   > ProcessPrimitives(Primitives=@n3, Force Process=true) {is=title K2Node_CallFunction}
   n4 = Cast To AtomModel!(execute=@n2.LoopBody, Object=@n2) -> (then, CastFailed, AsAtom Model:Atom Model Object Reference) {is=title K2Node_DynamicCast}
   > n5 = GetModelPrimitiveAssets(Model=@n4)
   > Add Items!(execute, TargetSet:Set of Atom Primitive Object References=Get PrimitivesSet(), NewItems:Array of Atom Primitive Object References=@n5) -> (then)
  ]}
]
variables [
 AlternateBlueprintCreationArgs {type="array<AtomModelBlueprintCreationArgumentsAndString>", default []}
 DefaultBlueprintCreationArgs {type=AtomModelBlueprintCreationArguments, default {parentClass=&Actor}}
 Destruction Builder {type=DestructionBuilder_C, default=null}
 Destruction Builder Class {type=Class, default=&/DestructionBuilder.DestructionBuilder_C}
 Generate BP {type=boolean, default=true}
 Material Factory {type=AtomMaterialFactory, default=null}
 Material Factory Class {type=Class, default=/Script/AtomModelImporterEditor.AtomMaterialDefaultFactory}
 Primitive Processor {type=AtomModelProcessor, default=null}
 Primitive Processor Class {type=Class, default=&/IndividualPrimitiveProcessor.IndividualPrimitiveProcessor_C}
 Processed Primitive Settings {type=AtomOnProcessPrimitiveSettings, default {bSupportDecorations=true, customLocation=""}}
]
