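# Editor processor step that writes an Atom model's merged geometry out as static mesh assets.
# OnFinalizeModel creates one mesh per container geometry, with names and paths from model or primitive settings, recording baked flex meshes as per-part overrides.
# A second pass sets LOD screen sizes, imported sockets, optional detailed collision, asset metadata and, for the unnamed model mesh, primitive instance data.
# FinalizeMesh and the pass-through OnPostMeshMerge are separate functions within this flow.
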EditorUtilityBlueprint /AtomFramework/Editor/Processors/AtomMergedMeshWriter
graphs [
 EventGraph {links [], nodes []}
 FinalizeMesh {nodes [
   n1 = FinalizeMesh!() -> (then, Static Mesh:Static Mesh Object Reference) {is=title K2Node_FunctionEntry}
   > CalculateMergedMeshScreenSizes!(
    execute
    LODSettings:"Atom Merged Mesh LODDistance Settings Structure (by ref)"=Make Atom Merged Mesh LODDistance Settings!(
     bOverrideLODScreenSizes:bool=true
     BaseLODScreenSize:float=0.25
     BaseLODScreenSizeScaling:float=1.05
     LODDistances:"Array of Float (single-precision)s"=Make Array!("[0]":float=1, "[1]":float=225, "[2]":float=300, "[3]":float=450, "[4]":float=700, "[5]":float=800, "[6]":float=1000, "[7]":float=1200) -> (Array:"Array of Float (single-precision)s")
    ) -> (AtomMergedMeshLODDistanceSettings:Atom Merged Mesh LODDistance Settings Structure) {is=title K2Node_MakeStruct}
    StaticMesh:Static Mesh Object Reference=@n1
   ) -> (then)
   > ApplyImportedSocketsToStaticMesh!(
    execute
    StaticMesh:Static Mesh Object Reference
    AtomModel:Atom Model Object Reference=Get AtomModel() -> (AtomModel:Atom Model Object Reference)
   ) -> (then)
  ]}
 FinalizeMeshInternal {nodes [
   n1 = FinalizeMeshInternal!() -> (
    then
    Static Mesh:Static Mesh Object Reference
    Merged Mesh Name:String
    Primitive Mesh Overrides:Map of Atom Model Part Guid Structures to Static Mesh Object References
    DetailedCollision:Atom Primitive Collision Geometry Structure
   ) {is=title K2Node_FunctionEntry}
   > FinalizeMesh!(execute, self, Static Mesh:Static Mesh Object Reference=@n1) -> (then) {is=title K2Node_CallFunction}
   > n2 = Branch(Condition=Get IsProcessingPrimitive() -> (IsProcessingPrimitive:bool))
   n3 = Get AtomModel()
   n4 = Get Static Mesh() -> (Static Mesh:Static Mesh Object Reference)
   > AddAtomPrimitiveAssetUserDataToProcessedPrimitiveMesh!(execute, StaticMesh:Static Mesh Object Reference=@n4, AtomModel:Atom Model Object Reference=@n3) -> (then)
   n5 = Get GenerateDetailedCollision() -> (GenerateDetailedCollision:bool)
   > n6 = Branch(Condition=@n5)
   n7 = Get Static Mesh()
   n8 = Get DetailedCollision() -> (DetailedCollision:Atom Primitive Collision Geometry Structure)
   > n9 = AddAtomDetailedCollisionToStaticMesh!(execute, StaticMesh:Static Mesh Object Reference=@n7, DetailedCollision:"Atom Primitive Collision Geometry Structure (by ref)"=@n8) -> (then)
   n10 = Get ProcessorImportJson!(self:Atom Merged Geometry Container Object Reference=Get MergedGeoContainer() -> (MergedGeoContainer:Atom Merged Geometry Container Object Reference)) -> (ProcessorImportJson:String)
   n11 = Get AtomModel()
   n12 = Get Static Mesh()
   ApplyProcessorAssetImportDataToStaticMeshWithJson!(
    execute=[@n9.then, @n6.else]
    StaticMesh:Static Mesh Object Reference=@n12
    ProcessorJson:String=@n10
    AtomModel:Atom Model Object Reference=@n11
   ) -> (then)
   n13 = IsEmpty(InString=Get Merged Mesh Name() -> (Merged Mesh Name:String))
   AddAtomModelReferenceToStaticMesh!(execute=@n2.else, StaticMesh:Static Mesh Object Reference=@n4, AtomModel:Atom Model Object Reference=@n3, bSaveConnectivityState:bool=@n13) -> (then)
   > n14 = Branch(Condition=@n5)
   > n15 = AddAtomDetailedCollisionToStaticMesh(StaticMesh=@n7, DetailedCollision=@n8)
   ApplyProcessorAssetImportDataToStaticMeshWithJson(execute=[@n15.then, @n14.else], StaticMesh=@n12, ProcessorJson=@n10, AtomModel=@n11)
   > Branch(Condition=@n13)
   n16 = Get MaterialFactory!(self:Atom Merged Geometry Container Object Reference=Get MergedGeoContainer()) -> (MaterialFactory:Atom Material Factory Object Reference)
   > n17 = Is Valid(InputObject=@n16)
   AddLegacyPrimitiveInstanceDataToStaticMesh!(
    execute=@n17.Is Not Valid
    StaticMesh:Static Mesh Object Reference=Get Static Mesh()
    AtomModel:Atom Model Object Reference=Get AtomModel()
    PrimitiveMeshOverrides:Map of Atom Model Part Guid Structures to Static Mesh Object References=Get Primitive Mesh Overrides() -> (Primitive Mesh Overrides:Map of Atom Model Part Guid Structures to Static Mesh Object References)
    MaterialOverrides:Map of Integers to Material Interface Object References=Get Material ID Overrides Map() -> (Material ID Overrides Map:Map of Integers to Material Interface Object References)
   ) -> (then)
   AddPrimitiveInstanceDataToStaticMesh!(
    execute=@n17.Is Valid
    StaticMesh:Static Mesh Object Reference=Get Static Mesh()
    MaterialFactory:Atom Material Factory Object Reference=@n16
    InstanceDataMap:Map of Atom Model Part Guid Structures to Atom Primitive Instance Data Structures=Get PrimitiveInstanceDataMap!(self:Atom Merged Geometry Container Object Reference=Get MergedGeoContainer()) -> (PrimitiveInstanceDataMap:Map of Atom Model Part Guid Structures to Atom Primitive Instance Data Structures)
   ) -> (then)
  ]}
 GetProcessedPrimitivePath {nodes [
   GetProcessedPrimitivePath() {is=title K2Node_FunctionEntry}
   > n1 = GetImporterSettings!(execute) -> (then, ReturnValue:Atom Model Importer Settings Object Reference)
   n2 = Get ProcessedPrimitiveContentFolder!(self:Atom Model Importer Settings Object Reference=@n1) -> (ProcessedPrimitiveContentFolder:Directory Path Structure)
   n3 = Get Custom Primitive Location() -> (Custom Primitive Location:String)
   > Return Node!(
    execute
    Processed Primitive Path:String=SelectString(
     A=Break Directory Path!(DirectoryPath:"Directory Path Structure (by ref)"=@n2) -> (Path:String) {is=title K2Node_BreakStruct}
     B=@n3
     bPickA=IsEmpty(InString=@n3)
    )
   )
  ]}
 GetStaticMeshName {nodes [
   n1 = GetStaticMeshName!() -> (then, Model:Atom Model Object Reference, Parts Collection:"Atom Model Parts Collection Structure (by ref)") {is=title K2Node_FunctionEntry}
   > n2 = TryGetKragleName!(execute, AtomModel:Atom Model Object Reference=@n1) -> (then, OutKragleName:String, ReturnValue:bool)
   > n3 = Branch(Condition=AND Boolean(A=Get IsProcessingPrimitive(), B=Get SupportDecorations() -> (SupportDecorations:bool)))
   n4 = Append(
    A=SM_
    B=SelectString(
     A=@n2
     B=Break Atom Model Parts Collection!(AtomModelPartsCollection:"Atom Model Parts Collection Structure (by ref)"=@n1.Parts Collection) -> (Name:String, Pivot:Transform, Parts:Array of Atom Model Part Instance Info Structures) {is=title K2Node_BreakStruct}
     bPickA=@n2.ReturnValue
    )
   )
   > Return Node!(execute, Result:String=Append(A=@n4, B=_DECO))
   Return Node(execute=@n3.else, Result=@n4)
  ]}
 GetStaticMeshPath {nodes [
   n1 = GetStaticMeshPath!() -> (then, Model:Atom Model Object Reference, Parts Collection:"Atom Model Parts Collection Structure (by ref)") {is=title K2Node_FunctionEntry}
   > n2 = Branch(Condition=Get IsProcessingPrimitive())
   > Return Node(Result=GetProcessedPrimitivePath!() -> (Processed Primitive Path:String) {is=title K2Node_CallFunction})
   n3 = TryGetKraglePath!(execute=@n2.else, AtomModel:Atom Model Object Reference=@n1) -> (then, OutKraglePath:String, ReturnValue:bool)
   > Return Node(Result=SelectString(A=@n3, B=GetModelPath!(self:Atom Model Object Reference=@n1) -> (ReturnValue:String), bPickA=@n3.ReturnValue))
  ]}
 LocalMergeExecutionWithData {nodes [
   n1 = Inputs!() -> (Option A:Exec, Option A Data:Wildcard, Option B:Exec, Option B Data:Wildcard)
   n2 = Local Boolean!() -> (Variable:bool=~) {is=title K2Node_TemporaryVariable}
   > n3 = Assign!(execute:Exec, Variable:bool=@n2, Value:bool) -> (then:Exec) {is=title K2Node_AssignmentStatement}
   Outputs!(
    Then:Exec=[@n4.then, @n3.then]
    Data:Wildcard=Select(Option 0:Wildcard=@n1, Option 1:Wildcard=@n1.Option B Data, Index=@n2) -> (ReturnValue:Wildcard)
   )
   n4 = Assign(execute=@n1.Option B, Variable=@n2, Value=true)
  ]}
 OnFinalizeModel {nodes [
   n1 = OnFinalizeModel!() -> (
    then
    Model:Atom Model Object Reference
    AtomModelPartsCollection:"Atom Model Parts Collection Structure (by ref)"
    InputObjects:Array of Object References
   ) {is=title K2Node_FunctionEntry}
   > n2 = Set AtomModel(AtomModel:Atom Model Object Reference=@n1) -> (Output_Get:Atom Model Object Reference)
   > n3 = Set Current Parts Collection(Current Parts Collection:Atom Model Parts Collection Structure=@n1.AtomModelPartsCollection) -> (Output_Get:Atom Model Parts Collection Structure)
   > n4 = Parent: OnFinalizeModel!(
    execute
    Model:Atom Model Object Reference=@n2
    AtomModelPartsCollection:"Atom Model Parts Collection Structure (by ref)"=@n3
    InputObjects:Array of Object References=@n1.InputObjects
   ) -> (then, ReturnValue:Atom Processor Result Structure) {is=title K2Node_CallParentFunction}
   > n5 = For Each Loop(Array:Array of Object References=@n1.InputObjects) -> (Array Element:Object Reference)
   n6 = Branch(execute=@n5.Completed, Condition=IsValid(Object=Get MergedGeoContainer()))
   n7 = Get MergedGeometries!(self:Atom Merged Geometry Container Object Reference=Get MergedGeoContainer()) -> (MergedGeometries:Map of Strings to Atom Merged Geometry Structures)
   > n8 = Keys!(execute, TargetMap:Map of Strings to Atom Merged Geometry Structures=@n7) -> (then, Keys:Array of Strings)
   > n9 = For Each Loop(Array:Array of Strings=@n8) -> (Array Element:String)
   n10 = Break Atom Merged Geometry!(AtomMergedGeometry:"Atom Merged Geometry Structure (by ref)"=Find(TargetMap:Map of Strings to Atom Merged Geometry Structures=@n7, Key=@n9) -> (Value:Atom Merged Geometry Structure)) -> (GeometryLODs:Atom Primitive Geometry LODs Structure, Materials:Array of Material Interface Object References) {is=title K2Node_BreakStruct}
   > n11 = OnPostMeshMerge!(
    execute
    self
    PrimitiveGeometryLODs:"Atom Primitive Geometry LODs Structure (by ref)"=@n10
    Materials:Array of Material Interface Object References=@n10.Materials
   ) -> (then, OutPrimitiveGeometryLODs:Atom Primitive Geometry LODs Structure, OutMaterials:Array of Material Interface Object References)
   > n12 = GetStaticMeshPath!(
    execute
    self
    Model:Atom Model Object Reference=Get AtomModel()
    Parts Collection:"Atom Model Parts Collection Structure (by ref)"=Get Current Parts Collection() -> (Current Parts Collection:Atom Model Parts Collection Structure)
   ) -> (then, Result:String) {is=title K2Node_CallFunction}
   > n13 = GetStaticMeshName!(
    execute
    self
    Model:Atom Model Object Reference=Get AtomModel()
    Parts Collection:"Atom Model Parts Collection Structure (by ref)"=Get Current Parts Collection()
   ) -> (then, Result:String) {is=title K2Node_CallFunction}
   > n14 = Branch(Condition=IsEmpty(InString=@n9))
   > n15 = Local Merge Execution with Data!(Option A:Exec, Option A Data:String=@n13, Option B:Exec=@n14.else, Option B Data:String=Append!(A=@n13, B=_, C:String=@n9) -> (ReturnValue)) -> (Then:Exec, Data:String) {is=title K2Node_MacroInstance}
   > PrintString(InString=@n15)
   > n16 = CreateStaticMeshFromGeometry!(
    execute
    GeometryLODs:"Atom Primitive Geometry LODs Structure (by ref)"=@n11
    NaniteHiResLOD:Atom Primitive Geometry Object Reference
    MaterialArray:Array of Material Interface Object References=@n11.OutMaterials
    Subpath:String=@n12
    StaticMeshName:String=@n15
    bPreserveExistingCollision:bool=Get Preserve Existing Collision() -> (Preserve Existing Collision:bool)
    bPreserveExistingSockets:bool=Get Preserve Existing Sockets() -> (Preserve Existing Sockets:bool)
    bGenerateMeshDistanceFields:bool=true
    TraceFlag:ECollisionTraceFlag Enum=CTF_UseSimpleAsComplex
    bSupportRaytracing:bool=true
   ) -> (then, ReturnValue:Static Mesh Object Reference)
   > Add!(
    execute
    TargetMap:Map of Strings to Static Mesh Object References=Get NameMergedMeshMap() -> (NameMergedMeshMap:Map of Strings to Static Mesh Object References)
    Key:"String (by ref)"=@n9
    Value:"Static Mesh Object Reference (by ref)"=@n16
   ) -> (then) {is=title K2Node_CallFunction}
   > Add(
    TargetMap:Map of Strings to Atom Primitive Collision Geometry Structures=Get NameMergedMeshDetailedCollision() -> (NameMergedMeshDetailedCollision:Map of Strings to Atom Primitive Collision Geometry Structures)
    Key=@n9
    Value:"Atom Primitive Collision Geometry Structure (by ref)"=Break Atom Primitive Geometry LODs!(AtomPrimitiveGeometryLODs:"Atom Primitive Geometry LODs Structure (by ref)"=@n11) -> (DetailedCollision:Atom Primitive Collision Geometry Structure) {is=title K2Node_BreakStruct}
   )
   n17 = Get MergedGeoContainer()
   > Branch(Condition=Contains Item(
     TargetArray:Array of Strings=Get BakedFlexPrimitivesNames!(self:Atom Merged Geometry Container Object Reference=@n17) -> (BakedFlexPrimitivesNames:Array of Strings)
     ItemToFind:"String (by ref)"=@n9
    ))
   n18 = To Atom Model Part Guid!(InString:String=@n9) -> (ReturnValue:Atom Model Part Guid Structure)
   > Add(
    TargetMap:Map of Atom Model Part Guid Structures to Static Mesh Object References=Get PrimitiveMeshOverrides() -> (PrimitiveMeshOverrides:Map of Atom Model Part Guid Structures to Static Mesh Object References)
    Key:"Atom Model Part Guid Structure (by ref)"=@n18
    Value:"Static Mesh Object Reference (by ref)"=@n16
   )
   > RegisterPrimitiveMeshOverrideData!(
    execute
    self
    Primitive Insance Data Map:Map of Atom Model Part Guid Structures to Atom Primitive Instance Data Structures=Get PrimitiveInstanceDataMap(self=Get MergedGeoContainer())
    Part Guid:Atom Model Part Guid Structure=@n18
    Mesh Override:Static Mesh Object Reference=@n16
   ) -> (then)
   n19 = Keys(execute=@n9.Completed, TargetMap:Map of Strings to Static Mesh Object References=Get NameMergedMeshMap())
   > n20 = For Each Loop(Array=@n19)
   Return Node!(
    execute=@n20.Completed
    ReturnValue:Atom Processor Result Structure=Make Atom Processor Result!(
     bSuccess:bool=True
     ProcessedObjects:Array of Object References=Get ProcessedObjects() -> (ProcessedObjects:Array of Object References)
     SharedAssets:Array of Object References
    ) -> (AtomProcessorResult:Atom Processor Result Structure) {is=title K2Node_MakeStruct}
   )
   n21 = Find(TargetMap:Map of Strings to Static Mesh Object References=Get NameMergedMeshMap(), Key=@n20) -> (Value:Static Mesh Object Reference)
   > Add(TargetArray:Array of Object References=Get ProcessedObjects(), NewItem:"Object Reference (by ref)"=@n21) {is=title K2Node_CallArrayFunction}
   > FinalizeMeshInternal!(
    execute
    self
    Static Mesh:Static Mesh Object Reference=@n21
    Merged Mesh Name:String=@n20
    Primitive Mesh Overrides:Map of Atom Model Part Guid Structures to Static Mesh Object References=Get PrimitiveMeshOverrides()
    DetailedCollision:Atom Primitive Collision Geometry Structure=Find(TargetMap:Map of Strings to Atom Primitive Collision Geometry Structures=Get NameMergedMeshDetailedCollision(), Key=@n20) -> (Value:Atom Primitive Collision Geometry Structure)
   ) -> (then) {is=title K2Node_CallFunction}
   PrintString(execute=@n6.else, InString="No merged meshes generated!!")
   > Return Node(ReturnValue=@n4)
   n22 = Cast To AtomMergedGeometryContainer!(execute=@n5.LoopBody, Object=@n5) -> (then, CastFailed, AsAtom Merged Geometry Container:Atom Merged Geometry Container Object Reference) {is=title K2Node_DynamicCast}
   > Set MergedGeoContainer(MergedGeoContainer:Atom Merged Geometry Container Object Reference=@n22) -> (Output_Get:Atom Merged Geometry Container Object Reference)
  ]}
 OnPostMeshMerge {nodes [
   n1 = OnPostMeshMerge!() -> (then, PrimitiveGeometryLODs:"Atom Primitive Geometry LODs Structure (by ref)", Materials:Array of Material Interface Object References) {is=title K2Node_FunctionEntry}
   > Return Node!(
    execute
    OutPrimitiveGeometryLODs:Atom Primitive Geometry LODs Structure=@n1
    OutMaterials:Array of Material Interface Object References=@n1.Materials
   )
  ]}
 RegisterPrimitiveMeshOverrideData {nodes [
   RegisterPrimitiveMeshOverrideData!() -> (
    then
    Primitive Insance Data Map:Map of Atom Model Part Guid Structures to Atom Primitive Instance Data Structures
    Part Guid:Atom Model Part Guid Structure
    Mesh Override:Static Mesh Object Reference
   ) {is=title K2Node_FunctionEntry}
   n1 = Find(
    TargetMap:Map of Atom Model Part Guid Structures to Atom Primitive Instance Data Structures=Get Primitive Insance Data Map() -> (Primitive Insance Data Map:Map of Atom Model Part Guid Structures to Atom Primitive Instance Data Structures)
    Key:"Atom Model Part Guid Structure (by ref)"=Get Part Guid() -> (Part Guid:Atom Model Part Guid Structure)
   ) -> (Value:Atom Primitive Instance Data Structure)
   > Branch(Condition=@n1.ReturnValue)
   > n2 = Set members in Atom Primitive Instance Data!(
    execute:Exec
    StructRef:"Atom Primitive Instance Data Structure (by ref)"=@n1
    MeshOverride:Static Mesh Soft Object Reference=Make Soft Reference!(Input:Static Mesh Object Reference=Get Mesh Override() -> (Mesh Override:Static Mesh Object Reference)) -> (Output:Static Mesh Soft Object Reference)
   ) -> (then:Exec, StructOut:"Atom Primitive Instance Data Structure (by ref)") {is=title K2Node_SetFieldsInStruct}
   > Add(
    TargetMap:Map of Atom Model Part Guid Structures to Atom Primitive Instance Data Structures=Get Primitive Insance Data Map()
    Key=Get Part Guid()
    Value:"Atom Primitive Instance Data Structure (by ref)"=@n2
   ) {is=title K2Node_CallFunction}
  ]}
]
variables [
 AtomModel {type=AtomModel, default=null}
 Current Parts Collection {type=AtomModelPartsCollection, default {name="", parts [], pivot {location=<0 0 0>, rotation=<rot 0 0 0>, scale=<1 1 1>}}}
 Custom Primitive Location {type=string, default=""}
 GenerateDetailedCollision {type=boolean}
 IsProcessingPrimitive {type=boolean}
 MergedGeoContainer {type=AtomMergedGeometryContainer, default=null}
 Preserve Existing Collision {type=boolean, default=true}
 Preserve Existing Sockets {type=boolean, default=true}
 SupportDecorations {type=boolean, default=true}
]
