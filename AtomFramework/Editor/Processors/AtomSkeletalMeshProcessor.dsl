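; Editor asset processor for skeletal meshes in AtomFramework, as its name and path indicate.
; It overrides GetModelImporterTemplateSettings, calling the parent implementation and returning its result together with a constant true.
; No other logic is present, so the template settings come entirely from the parent class.
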(fn GetModelImporterTemplateSettings
  (return (|CallParentFunction) true))
