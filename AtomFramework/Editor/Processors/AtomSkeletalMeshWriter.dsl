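; Editor processor for Atom model import that turns merged part geometries into a skeletal mesh and skeleton asset, named with SKM_ and SK_ prefixes.
; On finalize, it converts vertex colors from linear to sRGB and seeds each LOD with bones from an existing skeleton or the model's groups.
; It binds each group's vertices to its matching bone, appends them per LOD, gathers unique materials and generates both assets.
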(fn GetAssetPath
  (params (AtomModel :type Atom%20Model%20Object%20Reference))
  (return (AtomModel|GetModelPath AtomModel)))

(fn GetSkeletalMeshAssetName
  (params (AtomModel :type Atom%20Model%20Object%20Reference))
  (return (Utilities|String|Append "SKM_" (AtomModel|GetModelName AtomModel))))

(fn GetSkeletonAssetName
  (params (AtomModel :type Atom%20Model%20Object%20Reference))
  (return (Utilities|String|Append "SK_" (AtomModel|GetModelName AtomModel))))

(fn GetTemplateSkeleton
  (params (Atom%20Model :type Atom%20Model%20Object%20Reference))
  (if (Variables|Assets|GetBuild%20Skeletal%20Mesh%20From%20Existing%20Skeleton)
    (if (Utilities|IsValid (Variables|Assets|GetSkeleton) :!class /Script/Engine.KismetSystemLibrary)
      (LocalVariables|SetTemplateSkeleton (Variables|Assets|GetSkeleton))
      (else
        (bind _returnvalue (EditorScripting|Asset|LoadAsset (Utilities|String|Append (Name|GetAssetPath (Variables|GetAtom%20Model)) "/" (Name|GetSkeletonAssetName (Variables|GetAtom%20Model))) :!class /Script/EditorScriptingUtilities.EditorAssetLibrary))
        (bind _asskeleton (Utilities|Casting|CastToSkeleton _returnvalue))
        (LocalVariables|SetTemplateSkeleton _asskeleton))))
  (return (LocalVariables|GetTemplateSkeleton)))

(fn OnFinalizeModel
  (params (Model :type Atom%20Model%20Object%20Reference) (AtomModelPartsCollection :type Atom%20Model%20Parts%20Collection%20Structure%20%28by%20ref%29) (InputObjects :type Array%20of%20Object%20References))
  (bind _skeletalmeshgeometrylods (LocalVariables|GetSkeletalMeshGeometryLODs))
  (bind _geometrylods (Utilities|Struct|BreakAtomPrimitiveGeometryLODs _skeletalmeshgeometrylods))
  (bind _mergedgeocontainer (LocalVariables|GetMergedGeoContainer))
  (bind _mergedgeometries (Variables|/Script/AtomModelImporterEditor.AtomMergedGeometryContainer|GetMergedGeometries _mergedgeocontainer))
  ;[1 1 1] Get Merged Geometry Container
  (for _array_element InputObjects
    (bind _asatom_merged_geometry_container (Utilities|Casting|CastToAtomMergedGeometryContainer _array_element))
    (LocalVariables|SetMergedGeoContainer _asatom_merged_geometry_container))
  (if (Utilities|IsValid (LocalVariables|GetMergedGeoContainer) :!class /Script/Engine.KismetSystemLibrary)
    (Utilities|FlowControl|Sequence
      (:then_0
        ;[1 1 1] Apply VertexColor correction since FSkeletalMesh assume linear throughput.
        (bind _keys (Utilities|Map|Keys (Variables|/Script/AtomModelImporterEditor.AtomMergedGeometryContainer|GetMergedGeometries (LocalVariables|GetMergedGeoContainer))))
        (for _array_element_2 _keys
          (for _array_element_3 (Utilities|Struct|BreakAtomPrimitiveGeometryLODs (Utilities|Struct|BreakAtomMergedGeometry (Utilities|Map|Find (Variables|/Script/AtomModelImporterEditor.AtomMergedGeometryContainer|GetMergedGeometries (LocalVariables|GetMergedGeoContainer)) _array_element_2)))
            (AtomGeometry|ConvertVertexColorsLineartoSRGB _array_element_3))))
      (:then_1
        ;[1 1 1] Initialize SkeletalMesh Geometry LODs
        (bind _values (Utilities|Map|Values (Variables|/Script/AtomModelImporterEditor.AtomMergedGeometryContainer|GetMergedGeometries (LocalVariables|GetMergedGeoContainer))))
        (for _ (Utilities|Struct|BreakAtomPrimitiveGeometryLODs (Utilities|Struct|BreakAtomMergedGeometry (Utilities|Array|Get(aref) _values)))
          (bind _returnvalue (AtomGeometry|CreateEmptyAtomGeometry 2))
          (if (Variables|Assets|GetBuild%20Skeletal%20Mesh%20From%20Existing%20Skeleton)
            (AtomModel|AddSkeletontoGeometry (Variables|GetModel) (Private|GetTemplateSkeleton (Variables|GetModel)) (Variables|GetAtomModelPartsCollection) _returnvalue)
            (Utilities|Array|Add _geometrylods _returnvalue)
            (else
              (AtomModel|AddSkeletontoGeometryfromGroups (Variables|GetModel) (Variables|GetAtomModelPartsCollection) _returnvalue)
              (Utilities|Array|Add _geometrylods _returnvalue))))
        ;[1 1 1] Build geometry and materials
        (bind _keys_2 (Utilities|Map|Keys (Variables|/Script/AtomModelImporterEditor.AtomMergedGeometryContainer|GetMergedGeometries (LocalVariables|GetMergedGeoContainer))))
        (for _array_element_4 _keys_2
          ;;[1 1 1] Build material array
          (for _array_element_5 (.Materials (Utilities|Struct|BreakAtomMergedGeometry (Utilities|Map|Find _mergedgeometries _array_element_4)))
            (Utilities|Array|AddUnique (LocalVariables|GetSkeletalMaterials) _array_element_5))
          (for (_array_element_6 _array_index) (Utilities|Struct|BreakAtomPrimitiveGeometryLODs (Utilities|Struct|BreakAtomMergedGeometry (Utilities|Map|Find _mergedgeometries _array_element_4)))
            (AtomGeometry|AttachVerticestoBoneIndex _array_element_6 (AtomModel|GetBoneIndexforGroupName (Variables|GetModel) _array_element_4 (Utilities|Array|Get(acopy) (Utilities|Struct|BreakAtomPrimitiveGeometryLODs (LocalVariables|GetSkeletalMeshGeometryLODs)))))
            (AtomGeometry|Append (Utilities|Array|Get(acopy) (Utilities|Struct|BreakAtomPrimitiveGeometryLODs (LocalVariables|GetSkeletalMeshGeometryLODs)) _array_index) _array_element_6 (Math|Transform|MakeTransform "0, 0, 0"))))
        ;[1 1 1] Generate assets
        (bind (_outskeletalmesh _outskeleton) (AtomGeometry|CreateSkeletalMeshfromGeometry (Utilities|Struct|BreakAtomPrimitiveGeometryLODs (LocalVariables|GetSkeletalMeshGeometryLODs)) (LocalVariables|GetSkeletalMaterials) (Private|GetTemplateSkeleton (Variables|GetModel)) (Name|GetAssetPath (Variables|GetModel)) (Name|GetSkeletonAssetName (Variables|GetModel)) (Name|GetSkeletalMeshAssetName (Variables|GetModel))))
        (return (Utilities|Struct|MakeAtomProcessorResult true (Utilities|Array|MakeArray _outskeletalmesh _outskeleton)))))))
