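# Editor utility that batch-configures selected Geometry Collection Blueprints for replicated destruction.
# It validates that each has a Geometry Collection component, then runs Python intended to edit class defaults, though that script lacks a line break between statements.
# A subobject loop swaps static meshes to NoStaticMesh and sets each collection's replication, collision, simulation and notification properties.
# Compile and save run inside that loop, once per collection component.
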EditorUtilityBlueprint /AtomFramework/Editor/Processors/BPGCAssetActions
graphs [
 Check GC Blueprint Valid {nodes [
   n1 = Check GC Blueprint Valid!() -> (then, Object:Object Reference) {is=title K2Node_FunctionEntry}
   > n2 = GetBlueprintAsset(Object=@n1)
   > n3 = GetPathNameForLoadedAsset(LoadedAsset=@n2)
   > n4 = Is Valid(InputObject=@n2)
   PrintString(execute=@n4.Is Not Valid, InString=Append(A="Error - Invalid blueprint asset | ", B=@n3), TextColor=(R=1,G=0,B=0,A=1))
   > Return Node!(execute, Valid:bool=false)
   Set NumGCs(execute=@n4.Is Valid, NumGCs:int=0) -> (Output_Get:int=0)
   > n5 = Gather Subobject Data For Blueprint(self=Get SubobjectDataSubsystem(), Context=@n2)
   > n6 = For Each Loop(Array:Array of Subobject Data Handle Structures=@n5) -> (Array Element:Subobject Data Handle Structure)
   n7 = Branch(execute=@n6.Completed, Condition="integer > integer"(A=Get NumGCs() -> (NumGCs:int=0)))
   > Return Node()
   PrintString(execute=@n7.else, InString=Append(A="Error: This blueprint has no geometry collection component | ", B=@n3))
   > Return Node(Valid=false)
   Cast To GeometryCollectionComponent!(
    execute=@n6.LoopBody
    Object=GetObject!(Data:"Subobject Data Structure (by ref)"=GetData(DataHandle=@n6), bEvenIfPendingKill:bool=false) -> (ReturnValue:Object Reference)
   ) -> (then, CastFailed, AsGeometry Collection Component:Geometry Collection Component Object Reference) {is=title K2Node_DynamicCast}
   > Set NumGCs(NumGCs=int + int(A=Get NumGCs() -> (NumGCs=0))) -> (Output_Get=0)
  ]}
 EventGraph {links [], nodes []}
 Update GC Blueprint Properties {nodes [
   n1 = Update GC Blueprint Properties!() -> (then, Blueprint Asset:Object Reference) {is=title K2Node_FunctionEntry}
   > n2 = Check GC Blueprint Valid!(execute, self, Object:Object Reference=@n1) -> (then, Valid:bool) {is=title K2Node_CallFunction}
   > Branch(Condition=@n2)
   > n3 = GetBlueprintAsset(Object=@n1)
   > n4 = Is Valid(InputObject=@n3)
   PrintString(execute=@n4.Is Not Valid, InString=Invalid blueprint asset)
   n5 = Sequence(execute=@n4.Is Valid)
   > Execute Python Script!(
    execute
    PythonScript="import unreal as ue\r\n\r\nbpgen=ue.load_object(None,path)\r\nbpdef=ue.get_default_object(bpgen)\r\nbpdef.set_editor_property(\"StaticMesh\", None) bpdef.set_editor_property(\"bHasGeometryCollectionComponent\", True)"
    path:String=Append(A=Get Object Path String(Object=@n3), B=_C)
   ) -> (then, ReturnValue) {inputs [path]}
   n6 = Gather Subobject Data For Blueprint(execute=@n5.then_1, self=Get SubobjectDataSubsystem(), Context=@n3)
   > n7 = For Each Loop(Array=@n6)
   > n8 = Sequence()
   n9 = GetObject(Data=GetData(DataHandle=@n7), bEvenIfPendingKill=false)
   > n10 = Cast To StaticMeshComponent(Object=@n9)
   > SetStaticMesh(self=@n10, NewMesh=Get NoStaticMesh() -> (NoStaticMesh:Static Mesh Object Reference))
   n11 = Cast To GeometryCollectionComponent(execute=@n8.then_1, Object=@n9)
   > Set bEnableReplication!(execute, bEnableReplication:bool=true, self:Geometry Collection Component Object Reference=@n11) -> (then, Output_Get:bool)
   > Set ReplicationAbandonAfterLevel!(execute, ReplicationAbandonAfterLevel:int=1, self:Geometry Collection Component Object Reference=@n11) -> (then, Output_Get:int=0)
   > Set bEnableAbandonAfterLevel!(execute, bEnableAbandonAfterLevel:bool=true, self:Geometry Collection Component Object Reference=@n11) -> (then, Output_Get:bool)
   > Set ReplicationMaxPositionAndVelocityCorrectionLevel!(execute, ReplicationMaxPositionAndVelocityCorrectionLevel:int=0, self:Geometry Collection Component Object Reference=@n11) -> (then, Output_Get:int=0)
   > Set OneWayInteractionLevel!(execute, OneWayInteractionLevel:int=1, self:Geometry Collection Component Object Reference=@n11) -> (then, Output_Get:int=0)
   > Set CollisionProfilePerLevel!(
    execute
    CollisionProfilePerLevel:Array of Names=Get ProfilePerLevel() -> (ProfilePerLevel:Array of Names)
    self:Geometry Collection Component Object Reference=@n11
   ) -> (then, Output_Get:Array of Names)
   > SetCollisionProfileName!(execute, self:Primitive Component Object Reference=@n11, InCollisionProfileName:Name=FortBuildingMeshPhysics, bUpdateOverlaps:bool=false) -> (then)
   > SetEnableDamageFromCollision!(execute, self:Geometry Collection Component Object Reference=@n11, bValue:bool=false) -> (then)
   > Set InitialVelocityType!(
    execute
    InitialVelocityType:EInitialVelocityTypeEnum Enum=Chaos_Initial_Velocity_None
    self:Geometry Collection Component Object Reference=@n11
   ) -> (then, Output_Get:EInitialVelocityTypeEnum Enum=Chaos_Initial_Velocity_User_Defined)
   > Set MaxSimulatedLevel!(execute, MaxSimulatedLevel:int=1, self:Geometry Collection Component Object Reference=@n11) -> (then, Output_Get:int=0)
   > SetNotifyGlobalBreaks!(execute, self:Geometry Collection Component Object Reference=@n11, bNewNotifyGlobalBreaks:bool=true) -> (then)
   > SetNotifyGlobalCollision!(execute, self:Geometry Collection Component Object Reference=@n11, bNewNotifyGlobalCollisions:bool=true) -> (then)
   > CompileBlueprint!(execute, Blueprint:Blueprint Object Reference=@n3) -> (then, ReturnValue:bool)
   > SaveLoadedAsset(AssetToSave=@n3, bOnlyIfIsDirty=false)
   SetComponentAffectNavigationGeneration!(execute, Component:Actor Component Object Reference, bAffectNavigation:bool=false) -> (then)
  ]}
 Update Selected GC Blueprint properties {nodes [
   Update Selected GC Blueprint properties() {is=title K2Node_FunctionEntry}
   > n1 = GetSelectedAssets()
   > n2 = For Each Loop(Array:Array of Object References=@n1) -> (Array Element:Object Reference)
   > n3 = GetBlueprintAsset(Object=@n2)
   > Update GC Blueprint Properties!(execute, self, Blueprint Asset:Object Reference=@n3) -> (then) {is=title K2Node_CallFunction}
  ]}
]
variables []
