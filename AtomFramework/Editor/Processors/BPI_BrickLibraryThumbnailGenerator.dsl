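; Blueprint interface for editor-side generators of brick library thumbnails in AtomFramework.
; It declares two functions that take a brick library data table: one to create thumbnails and update the library using a supplied generator actor class, the other to generate thumbnails from the table alone.
; It holds only signatures, so implementers supply all capture and update logic.
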(fn CreateThumbnailsAndUpdateBrickLibrary
  (params (BrickLibraryDataTable :type Data%20Table%20Object%20Reference) (ThumbnailGeneratorClass :type Actor%20Class%20Reference)))

(fn GenerateThumbnails
  (params (BrickLibraryDataTable :type Data%20Table%20Object%20Reference)))
