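; Editor primitive processor for the Atom brick library that adds detailed collision to the static meshes it processes.
; It overrides GenerateDetailedCollision to return true, enabling that collision data.
; OnPostMeshMerge passes through the parent's results and caches the detailed collision from the primitive geometry LODs struct in a variable.
; FinalizeMesh runs the parent logic, then applies the cached collision to the static mesh as Atom detailed collision.
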(fn FinalizeMesh
  (params (Static%20Mesh :type Static%20Mesh%20Object%20Reference))
  (|CallParentFunction Static%20Mesh)
  (AtomModel|AddAtomDetailedCollisiontoStaticMesh Static%20Mesh (Variables|Default|GetDetailedCollision)))

(fn GenerateDetailedCollision
  (return true))

(fn OnPostMeshMerge
  (params (PrimitiveGeometryLODs :type Atom%20Primitive%20Geometry%20LODs%20Structure%20%28by%20ref%29) (Materials :type Array%20of%20Material%20Interface%20Object%20References) (Mesh%20Name :type String))
  (bind (_geometrylods _geometrylodsmaterials _hiresnanitegeomery _hiresnanitegeomerymaterials _collision _detailedcollision) (Utilities|Struct|BreakAtomPrimitiveGeometryLODs PrimitiveGeometryLODs))
  (bind (_outprimitivegeometrylods _outmaterials) (|CallParentFunction PrimitiveGeometryLODs Materials Mesh%20Name))
  (Variables|Default|SetDetailedCollision _detailedcollision)
  (return _outprimitivegeometrylods _outmaterials))
