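# Editor model-import processor that builds a brick library Data Table from an imported model for use in the Atom Model Editor.
# It collects primitives by unique part ID, optionally duplicates them to an override path, and force-processes them through a custom primitive processor into static meshes.
# The primitive-to-mesh map feeds the `DT_<ModelName>` table, preserving existing content.
# When configured, an interface-checked thumbnail action then updates the library.
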EditorUtilityBlueprint /AtomFramework/Editor/Processors/BrickLibraryProcessor
graphs [
 CreateThumbnails {nodes [
   CreateThumbnails!() -> (then, PrimitiveDataTable:Data Table Object Reference) {is=title K2Node_FunctionEntry}
   > n1 = Construct Object from Class(Class=Get ThumbnailGeneratorAction() -> (ThumbnailGeneratorAction:BP Create Primitive Thumbnails Action Class Reference)) -> (ReturnValue:BP Create Primitive Thumbnails Action Object Reference)
   > Set ThumbnailGeneratorActionObject(ThumbnailGeneratorActionObject:Object Reference=@n1) -> (Output_Get:Object Reference)
   > n2 = Branch(Condition=Does Object Implement Interface!(
     TestObject:Object Reference=@n1
     Interface:Interface Interface=&/BPI_BrickLibraryThumbnailGenerator.BPI_BrickLibraryThumbnailGenerator_C
    ) -> (ReturnValue:bool))
   > CreateThumbnailsAndUpdateBrickLibrary!(
    execute:Exec
    self:Object Reference=Get ThumbnailGeneratorActionObject() -> (ThumbnailGeneratorActionObject:Object Reference)
    BrickLibraryDataTable:Data Table Object Reference=Get PrimitiveDataTable() -> (PrimitiveDataTable:Data Table Object Reference)
    ThumbnailGeneratorClass:Actor Class Reference=Get ThumbnailGeneratorClass() -> (ThumbnailGeneratorClass:BP Default Atom Primitive Thumbnail Generator Class Reference)
   ) -> (then:Exec) {is=title K2Node_Message}
   n3 = Get ThumbnailGeneratorAction()
   ReportProcessorError!(
    execute=@n2.else
    Message:"Text (by ref)"=Format Text!(Format="LOCTEXT(\"{Class} does not implement Brick Library Thumbnail Generator Interface\")", Class:String=Get Class Display Name(Class=@n3)) -> (Result)
    Asset:Object Reference=Get PrimitiveDataTable()
   ) -> (then)
  ]}
 EventGraph {links [], nodes []}
 GetModelImporterTemplateSettings {nodes [
   GetModelImporterTemplateSettings() {is=title K2Node_FunctionEntry}
   > Return Node!(
    execute
    OutModelImporterTemplate:Atom Model Importer Template Structure=Get Out Model Importer Template() -> (Out Model Importer Template:Atom Model Importer Template Structure)
    ReturnValue:bool=true
   )
   Parent: GetModelImporterTemplateSettings!() -> (OutModelImporterTemplate:Atom Model Importer Template Structure, ReturnValue:bool) {is=title K2Node_CallParentFunction}
  ]}
 GetProcessedPrimitivePath {nodes [
   GetProcessedPrimitivePath() {is=title K2Node_FunctionEntry}
   > n1 = GetImporterSettings!(execute) -> (then, ReturnValue:Atom Model Importer Settings Object Reference)
   > Return Node!(
    execute
    Processed Primitive Path:String=Break Directory Path!(DirectoryPath:"Directory Path Structure (by ref)"=Get ProcessedPrimitiveContentFolder!(self:Atom Model Importer Settings Object Reference=@n1) -> (ProcessedPrimitiveContentFolder:Directory Path Structure)) -> (Path:String) {is=title K2Node_BreakStruct}
   )
  ]}
 OnProcessModel {nodes [
   n1 = OnProcessModel!() -> (
    then
    Model:Atom Model Object Reference
    AtomModelPartsCollection:"Atom Model Parts Collection Structure (by ref)"
    ExistingObjects:Array of Soft Object References
   ) {is=title K2Node_FunctionEntry}
   > n2 = Construct Object from Class(Class=Get Primitive Processor Class() -> (Primitive Processor Class:Atom Model Processor Class Reference)) -> (ReturnValue:Atom Model Processor Object Reference)
   > Set Primitive Processor(Primitive Processor:Atom Model Processor Object Reference=@n2) -> (Output_Get:Atom Model Processor Object Reference)
   > n3 = Parent: OnProcessModel!(
    execute
    Model:Atom Model Object Reference=@n1
    AtomModelPartsCollection:"Atom Model Parts Collection Structure (by ref)"=@n1.AtomModelPartsCollection
    ExistingObjects:Array of Soft Object References=@n1.ExistingObjects
   ) -> (then, ReturnValue:Atom Processor Result Structure) {is=title K2Node_CallParentFunction}
   > Return Node!(execute, ReturnValue:Atom Processor Result Structure=@n3)
  ]}
 OnProcessPartsCollection {nodes [
   n1 = OnProcessPartsCollection!() -> (then, Parts Collection:"Atom Model Parts Collection Structure (by ref)", Model:Atom Model Object Reference) {is=title K2Node_FunctionEntry}
   > n2 = Parent: OnProcessPartsCollection!(execute, Parts Collection:"Atom Model Parts Collection Structure (by ref)"=@n1, Model:Atom Model Object Reference=@n1.Model) -> (then, Result:Atom Processor Result Structure) {is=title K2Node_CallParentFunction}
   > n3 = For Each Loop(Array:Array of Atom Model Part Instance Info Structures=Break Atom Model Parts Collection!(AtomModelPartsCollection:"Atom Model Parts Collection Structure (by ref)"=@n1) -> (Name:String, Pivot:Transform, Parts:Array of Atom Model Part Instance Info Structures) {is=title K2Node_BreakStruct}.Parts) -> (Array Element:Atom Model Part Instance Info Structure)
   ProcessPrimitives!(
    execute=@n3.Completed
    self
    Primitives:Array of Atom Primitive Object References=Get Parts() -> (Parts:Array of Atom Primitive Object References)
    Force Process:bool=true
    PrimitiveToMesh:Map of Atom Primitive Object References to Static Mesh Object References=Get PrimtiveToMesh() -> (PrimtiveToMesh:Map of Atom Primitive Object References to Static Mesh Object References)
   ) -> (then)
   > n4 = FindOrCreateBrickLibraryDataTableAsset!(
    execute
    PrimitiveArray:Array of Atom Primitive Object References=Get Parts()
    Subpath:String=GetModelPath!(self:Atom Model Object Reference=@n1.Model) -> (ReturnValue:String)
    LibraryName:String=Append(A=DT_, B=GetModelName!(self:Atom Model Object Reference=@n1.Model) -> (ReturnValue:String))
    PrimitiveMeshes:Map of Atom Primitive Object References to Static Mesh Object References=Get PrimtiveToMesh()
    bPreserveExisting:bool=true
   ) -> (then, ReturnValue:Data Table Object Reference)
   > n5 = Sequence()
   Return Node!(execute=@n5.then_1, Result:Atom Processor Result Structure=@n2)
   n6 = Get ThumbnailGeneratorClass()
   n7 = Get ThumbnailGeneratorAction()
   > Branch(Condition=AND Boolean(A=IsValidClass!(Class:Object Class Reference=@n6) -> (ReturnValue:bool), B=IsValidClass(Class=@n7)))
   > CreateThumbnails!(execute, self, PrimitiveDataTable:Data Table Object Reference=@n4) -> (then) {is=title K2Node_CallFunction}
   n8 = Break Atom Model Part Instance Info!(AtomModelPartInstanceInfo:"Atom Model Part Instance Info Structure (by ref)"=@n3) -> (
    PartGuid:Atom Model Part Guid Structure
    PartId:int
    Primitive:Atom Primitive Object Reference
    Transforms:Array of Transforms
    Decorations:Array of Atom Model Part Decoration Info Structures
    Colors:Array of Atom Model Part Color Info Structures
    SelectionSets:Array of Strings
    Group:String
    ParentGroup:String
    CommonParts:Array of Atom Common Part And Transform Structures
    bIsUndersideVisible:bool=~
   ) {is=title K2Node_BreakStruct}
   n9 = Branch(
    execute=@n3.LoopBody
    Condition=Contains Item!(TargetSet:Set of Integers=Get UniqueParts() -> (UniqueParts:Set of Integers), ItemToFind:"Integer (by ref)"=@n8.PartId) -> (ReturnValue:bool) {is=title K2Node_CallFunction}
   )
   Add!(execute=@n9.else, TargetSet:Set of Integers=Get UniqueParts(), NewItem:"Integer (by ref)"=@n8.PartId) -> (then) {is=title K2Node_CallFunction}
   > n10 = Branch(Condition=IsEmpty(InString=Get Override AtomPrimitive Path() -> (Override AtomPrimitive Path:String)))
   > Add(TargetArray:Array of Atom Primitive Object References=Get Parts(), NewItem:"Atom Primitive Object Reference (by ref)"=@n8.Primitive) {is=title K2Node_CallArrayFunction}
   n11 = DuplicateAtomPrimitive!(execute=@n10.else, Src:Atom Primitive Object Reference=@n8.Primitive, NewBasePath:String=Get Override AtomPrimitive Path()) -> (then, ReturnValue:Atom Primitive Object Reference)
   > Set bAllowUseInCookedEditor!(execute, bAllowUseInCookedEditor:bool=true, self:Atom Primitive Object Reference=@n11) -> (then, Output_Get:bool)
   > Add(TargetArray=Get Parts(), NewItem=@n11)
   GetProcessedPrimitivePath!() -> (Processed Primitive Path:String) {is=title K2Node_CallFunction}
   Get Parts()
  ]}
 ProcessPrimitives {nodes [
   n1 = ProcessPrimitives!() -> (
    then
    Primitives:Array of Atom Primitive Object References
    Force Process:bool=~
    PrimitiveToMesh:Map of Atom Primitive Object References to Static Mesh Object References
   ) {is=title K2Node_FunctionEntry}
   > n2 = For Each Loop(Array:Array of Atom Primitive Object References=@n1) -> (Array Element:Atom Primitive Object Reference)
   > n3 = Branch(Condition=Get Force Process() -> (Force Process:bool))
   n4 = RunCustomPrimitiveProcessor!(
    execute=[@n3.then, @n8.Is Not Valid]
    Processor:Atom Model Processor Object Reference=Get Primitive Processor() -> (Primitive Processor:Atom Model Processor Object Reference)
    Primitive:Atom Primitive Object Reference=@n2
    PrimitiveSettings:"Atom On Process Primitive Settings Structure (by ref)"=Get Processed Primitive Settings() -> (Processed Primitive Settings:Atom On Process Primitive Settings Structure)
   ) -> (then, ReturnValue:Atom Processor Result Structure)
   > n5 = For Each Loop(Array:Array of Object References=Break Atom Processor Result!(AtomProcessorResult:"Atom Processor Result Structure (by ref)"=@n4) -> (bSuccess:bool=~, ProcessedObjects:Array of Object References, SharedAssets:Array of Object References) {is=title K2Node_BreakStruct}.ProcessedObjects) -> (Array Element:Object Reference)
   > n6 = Cast To StaticMesh(Object=@n5)
   > Branch()
   > Add!(
    execute
    TargetMap:Map of Atom Primitive Object References to Static Mesh Object References=Get PrimitiveToMesh() -> (PrimitiveToMesh:Map of Atom Primitive Object References to Static Mesh Object References)
    Key:"Atom Primitive Object Reference (by ref)"=@n2
    Value:"Static Mesh Object Reference (by ref)"=@n6
   ) -> (then) {is=title K2Node_CallFunction}
   n7 = GetCustomPrimitiveProcessorTargetAsset!(
    execute=@n3.else
    Processor:Atom Model Processor Object Reference=Get Primitive Processor()
    Primitive:Atom Primitive Object Reference=@n2
    PrimitiveSettings:"Atom On Process Primitive Settings Structure (by ref)"=Get Processed Primitive Settings()
   ) -> (then, ReturnValue:Object Reference)
   > n8 = Is Valid(InputObject=@n7)
   Reroute Node(InputPin:Wildcard) -> (OutputPin:Wildcard)
  ]}
]
variables [
 Out Model Importer Template {
  type=AtomModelImporterTemplate
  default {
   category=UE
   description="Go through each unique brick in the model, generate meshes for each, then create a default datatable for the primitives for use with the Atom Model Editor"
   label=Brick Library Data Table
   thumbnailObject=&/SimpleBrickIcon64
  }
 }
 Override AtomPrimitive Path {type=string, default=""}
 Primitive Processor {type=AtomModelProcessor, default=null}
 Primitive Processor Class {type=Class, default=&/BrickLibraryPrimitiveProcessor.BrickLibraryPrimitiveProcessor_C}
 Processed Primitive Settings {type=AtomOnProcessPrimitiveSettings, default {bSupportDecorations=true, customLocation=""}}
 ThumbnailGeneratorAction {type=Class, default=null}
 ThumbnailGeneratorClass {type=Class, default=null}
]
