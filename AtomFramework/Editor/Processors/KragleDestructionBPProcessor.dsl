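; Editor-side Atom model processor for Kragle destruction that overrides how static meshes are named, placed and generated from Atom models.
; Mesh name and path come from Kragle lookups on the model, and detailed collision generation is always enabled.
; The material factory prefers the Fortnite LimeMaterialFactory class when it is found, otherwise falling back to the parent's choice.
; The model importer template settings override returns only fixed default values.
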(fn GenerateDetailedCollision
  (return true))

(fn GetMaterialFactoryClass
  ; If we are in Fortnite, try to use the FN material factory instead.
  (bind _returnvalue (AtomEditor|FindMaterialFactoryGeneratedClass "LimeMaterialFactory"))
  (if (Utilities|IsValidClass _returnvalue)
    (return _returnvalue)
    (else
      (return (|CallParentFunction)))))

(fn GetModelImporterTemplateSettings
  (return 0 false))

(fn GetStaticMeshName
  (params (Model :type Atom%20Model%20Object%20Reference) (Parts%20Collection :type Atom%20Model%20Parts%20Collection%20Structure%20%28by%20ref%29))
  (bind _outkraglename (Kragle|TryGetKragleName Model))
  (return _outkraglename))

(fn GetStaticMeshPath
  (params (Model :type Atom%20Model%20Object%20Reference) (Parts%20Collection :type Atom%20Model%20Parts%20Collection%20Structure%20%28by%20ref%29))
  (bind _outkraglepath (Kragle|TryGetKraglePath Model))
  (return _outkraglepath))
