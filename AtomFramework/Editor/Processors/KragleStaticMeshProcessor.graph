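# Editor-only static mesh processor for Atom models that names and places generated meshes using Kragle data.
# Mesh names are "SM_" plus the model's Kragle name, and paths come straight from its Kragle path; neither lookup's success flag is checked.
# It uses the LimeMaterialFactory generated class, falling back to the parent's factory if that class is invalid.
# Detailed collision is enabled, and no model importer template is supplied.
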EditorUtilityBlueprint /AtomFramework/Editor/Processors/KragleStaticMeshProcessor
graphs [
 EventGraph {links [], nodes []}
 GenerateDetailedCollision {nodes [
   GenerateDetailedCollision() {is=title K2Node_FunctionEntry}
   > Return Node!(execute, GenerateDetailedCollision:bool=true)
  ]}
 GetMaterialFactoryClass {nodes [
   GetMaterialFactoryClass() {is=title K2Node_FunctionEntry}
   > n1 = FindMaterialFactoryGeneratedClass!(execute, MaterialFactoryName:String=LimeMaterialFactory) -> (then, ReturnValue:Atom Material Factory Class Reference)
   > n2 = Branch(Condition=IsValidClass!(Class:Object Class Reference=@n1) -> (ReturnValue:bool))
   > Return Node!(execute, AtomMaterialFactoryClass:Atom Material Factory Class Reference=@n1)
   Return Node(
    execute=@n2.else
    AtomMaterialFactoryClass=Parent: GetMaterialFactoryClass!() -> (AtomMaterialFactoryClass:Atom Material Factory Class Reference) {is=title K2Node_CallParentFunction}
   )
  ]}
 GetModelImporterTemplateSettings {nodes [
   GetModelImporterTemplateSettings() {is=title K2Node_FunctionEntry}
   > Return Node!(execute, OutModelImporterTemplate:Atom Model Importer Template Structure, ReturnValue:bool=false)
  ]}
 GetStaticMeshName {nodes [
   n1 = GetStaticMeshName!() -> (then, Model:Atom Model Object Reference, Parts Collection:"Atom Model Parts Collection Structure (by ref)") {is=title K2Node_FunctionEntry}
   > n2 = TryGetKragleName!(execute, AtomModel:Atom Model Object Reference=@n1) -> (then, OutKragleName:String, ReturnValue:bool)
   > Return Node!(execute, Result:String=Append(A=SM_, B=@n2))
  ]}
 GetStaticMeshPath {nodes [
   n1 = GetStaticMeshPath!() -> (then, Model:Atom Model Object Reference, Parts Collection:"Atom Model Parts Collection Structure (by ref)") {is=title K2Node_FunctionEntry}
   > n2 = TryGetKraglePath!(execute, AtomModel:Atom Model Object Reference=@n1) -> (then, OutKraglePath:String, ReturnValue:bool)
   > Return Node(Result=@n2)
  ]}
]
variables []
