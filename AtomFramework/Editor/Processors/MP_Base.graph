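# Base editor-utility processor for AtomFramework models and primitives, providing an overridable pipeline with placeholder defaults.
# Model and primitive processing share RunProcessorPipeline, with the primitive entry point caching ProcessedPrimitive first.
# The pipeline calls a pre-process hook, builds parts collections (by default just the input), then processes each.
# Helper objects registered during each collection feed the child processor's finalize call, and results accumulate into ProcessorResult.
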EditorUtilityBlueprint /AtomFramework/Editor/Processors/MP_Base
graphs [
 EventGraph {links [], nodes []}
 GetIsProcessingPrimitive {nodes [
   GetIsProcessingPrimitive() {is=title K2Node_FunctionEntry}
   > Return Node!(execute, Is Processing Primitive:bool=IsValid(Object=Get ProcessedPrimitive() -> (ProcessedPrimitive:Atom Primitive Object Reference)))
  ]}
 GetProcessedPrimitive {nodes [
   GetProcessedPrimitive() {is=title K2Node_FunctionEntry}
   > Return Node!(execute, AtomPrimitive:Atom Primitive Object Reference=Get ProcessedPrimitive())
   Get StaticMeshEditorSubsystem()
  ]}
 OnBuildPartsCollections {nodes [
   n1 = OnBuildPartsCollections!() -> (then, Base Parts Collection:"Atom Model Parts Collection Structure (by ref)", Model:Atom Model Object Reference) {is=title K2Node_FunctionEntry}
   > Return Node!(
    execute
    Array of Filtered Parts Collections:Array of Atom Model Parts Collection Structures=Make Array("[0]":Atom Model Parts Collection Structure=@n1) -> (Array:Array of Atom Model Parts Collection Structures)
   )
  ]}
 OnGetProcessPrimitiveTargetAssetPath {nodes [
   n1 = OnGetProcessPrimitiveTargetAssetPath!() -> (
    then
    Model:Atom Model Object Reference
    Primitive:Atom Primitive Object Reference
    AtomModelPartsCollection:"Atom Model Parts Collection Structure (by ref)"
    Settings:"Atom On Process Primitive Settings Structure (by ref)"
   ) {is=title K2Node_FunctionEntry}
   > Set ProcessedPrimitive(ProcessedPrimitive:Atom Primitive Object Reference=@n1.Primitive) -> (Output_Get:Atom Primitive Object Reference)
   > Return Node!(execute, ReturnValue:String)
  ]}
 OnProcessModel {nodes [
   n1 = OnProcessModel!() -> (
    then
    Model:Atom Model Object Reference
    AtomModelPartsCollection:"Atom Model Parts Collection Structure (by ref)"
    ExistingObjects:Array of Soft Object References
   ) {is=title K2Node_FunctionEntry}
   > n2 = RunProcessorPipeline!(execute, self, AtomModel:Atom Model Object Reference=@n1, PartsCollection:Atom Model Parts Collection Structure=@n1.AtomModelPartsCollection) -> (then, ReturnValue:Atom Processor Result Structure)
   > Return Node(ReturnValue:Atom Processor Result Structure=@n2)
  ]}
 OnProcessPartsCollection {nodes [
   OnProcessPartsCollection!() -> (then, Parts Collection:"Atom Model Parts Collection Structure (by ref)", Model:Atom Model Object Reference) {is=title K2Node_FunctionEntry}
   > Return Node!(
    execute
    Result:Atom Processor Result Structure=Make Atom Processor Result!(bSuccess:bool=True, ProcessedObjects:Array of Object References, SharedAssets:Array of Object References) -> (AtomProcessorResult:Atom Processor Result Structure) {is=title K2Node_MakeStruct}
   )
  ]}
 OnProcessPrimitive {nodes [
   n1 = OnProcessPrimitive!() -> (
    then
    DummyModel:Atom Model Object Reference
    Primitive:Atom Primitive Object Reference
    AtomModelPartsCollection:"Atom Model Parts Collection Structure (by ref)"
    Settings:"Atom On Process Primitive Settings Structure (by ref)"
   ) {is=title K2Node_FunctionEntry}
   > Set ProcessedPrimitive(ProcessedPrimitive=@n1.Primitive)
   > n2 = RunProcessorPipeline(AtomModel=@n1, PartsCollection=@n1.AtomModelPartsCollection)
   > Return Node!(execute, ReturnValue:Atom Processor Result Structure=@n2)
  ]}
 PreProcessorPipeline {nodes [PreProcessorPipeline!() -> (then, AtomModel:Atom Model Object Reference, PartsCollection:"Atom Model Parts Collection Structure (by ref)") {is=title K2Node_FunctionEntry}]}
 RegisterChildProcessorHelperObject {nodes [
   n1 = RegisterChildProcessorHelperObject!() -> (then, Input:Object Reference) {is=title K2Node_FunctionEntry}
   > Add(
    TargetArray:Array of Object References=Get ChildProcessorObjects() -> (ChildProcessorObjects:Array of Object References)
    NewItem:"Object Reference (by ref)"=@n1
   ) {is=title K2Node_CallArrayFunction}
  ]}
 RunChildProcessorPipelineFinalize {nodes [
   n1 = RunChildProcessorPipelineFinalize!() -> (
    then
    InputObjects:Array of Object References
    Model:Atom Model Object Reference
    InputPartsCollections:Atom Model Parts Collection Structure
    InputResults:Atom Processor Result Structure
   ) {is=title K2Node_FunctionEntry}
   > Set CurrentResult(CurrentResult:Atom Processor Result Structure=@n1.InputResults) -> (Output_Get:Atom Processor Result Structure)
   > n2 = InvokeChildProcessorOnFinalizeModels!(
    execute
    self:Editor Atom Model Processor Object Reference
    Model:Atom Model Object Reference=@n1.Model
    AtomModelPartsCollection:"Atom Model Parts Collection Structure (by ref)"=@n1.InputPartsCollections
    InputObjects:Array of Object References=@n1
   ) -> (then, ReturnValue:Atom Processor Result Structure)
   > n3 = AppendAtomProcessorResult!(
    execute
    Result:"Atom Processor Result Structure (by ref)"=Get CurrentResult() -> (CurrentResult:Atom Processor Result Structure)
    ResultToAppend:"Atom Processor Result Structure (by ref)"=@n2
   ) -> (then, ReturnValue:Atom Processor Result Structure)
   > Return Node!(execute, OutputResults:Atom Processor Result Structure=@n3)
  ]}
 RunProcessorPipeline {nodes [
   n1 = RunProcessorPipeline!() -> (then, AtomModel:Atom Model Object Reference, PartsCollection:Atom Model Parts Collection Structure) {is=title K2Node_FunctionEntry}
   > PreProcessorPipeline!(
    execute
    self
    AtomModel:Atom Model Object Reference=@n1
    PartsCollection:"Atom Model Parts Collection Structure (by ref)"=@n1.PartsCollection
   ) -> (then) {is=title K2Node_CallFunction}
   > n2 = OnBuildPartsCollections!(
    execute
    self
    Base Parts Collection:"Atom Model Parts Collection Structure (by ref)"=@n1.PartsCollection
    Model:Atom Model Object Reference=@n1
   ) -> (then, Array of Filtered Parts Collections:Array of Atom Model Parts Collection Structures) {is=title K2Node_CallFunction}
   > n3 = For Each Loop(Array:Array of Atom Model Parts Collection Structures=@n2) -> (Array Element:Atom Model Parts Collection Structure)
   Clear(execute=@n3.Completed, TargetArray:Array of Object References=Get ChildProcessorObjects())
   > Return Node!(execute, ReturnValue:Atom Processor Result Structure=Get ProcessorResult() -> (ProcessorResult:Atom Processor Result Structure))
   Clear(execute=@n3.LoopBody, TargetArray=Get ChildProcessorObjects())
   > n4 = OnProcessPartsCollection!(execute, self, Parts Collection:"Atom Model Parts Collection Structure (by ref)"=@n3, Model:Atom Model Object Reference=@n1) -> (then, Result:Atom Processor Result Structure) {is=title K2Node_CallFunction}
   > n5 = AppendAtomProcessorResult(Result=Get ProcessorResult(), ResultToAppend=@n4)
   > n6 = RunChildProcessorPipelineFinalize!(
    execute
    self
    InputObjects:Array of Object References=Get ChildProcessorObjects()
    Model:Atom Model Object Reference=@n1
    InputPartsCollections:Atom Model Parts Collection Structure=@n3
    InputResults:Atom Processor Result Structure=@n5
   ) -> (then, OutputResults:Atom Processor Result Structure) {is=title K2Node_CallFunction}
   > Set ProcessorResult(ProcessorResult:Atom Processor Result Structure=@n6) -> (Output_Get:Atom Processor Result Structure)
  ]}
]
variables [
 ChildProcessorObjects {type="array<Object>", default []}
 ProcessedPrimitive {type=AtomPrimitive, default=null}
]
