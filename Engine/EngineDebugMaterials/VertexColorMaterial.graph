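# Engine debug material for visualizing mesh vertex colors on a two-sided, Default Lit surface.
# A VertexColor node feeds the DiffuseColor input of the ConvertFromDiffSpec function, which converts diffuse/specular values to the metallic workflow.
# Usage flags cover skeletal, instanced static and deformed meshes, plus geometry caches and geometry collections.
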Material /Engine/EngineDebugMaterials/VertexColorMaterial
graph {
 properties {
  bUsedWithGeometryCache=true
  bUsedWithGeometryCollections=true
  bUsedWithInstancedStaticMeshes=true
  bUsedWithMeshDeformer=true
  bUsedWithSkeletalMesh=true
  pixelDepthOffsetMode=PDOM_Legacy
  twoSided=true
  usedShadingModels=Default Lit
 }
 nodes [MaterialFunctionCall(DiffuseColor=VertexColor()) {materialFunction=&ConvertFromDiffSpec}]
}
