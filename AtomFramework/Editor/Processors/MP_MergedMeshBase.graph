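# Editor base processor that turns Atom model parts and their common parts into merged static meshes.
# Builds per-part merge infos from LOD and Nanite geometry style settings, adding material features, decorations, vertex colors and collision.
# With the new merge API, LODs beyond the configured styles reuse the last one with an offset geometry LOD.
# Finalization writes LOD screen sizes, sockets, import and instance data; a debug mode swaps in vertex-color materials.
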EditorUtilityBlueprint /AtomFramework/Editor/Processors/MP_MergedMeshBase
graphs [
 AppendGroupFilterModeToMeshName {nodes [
   n1 = AppendGroupFilterModeToMeshName!() -> (then, PartInfo:"Atom Merged Mesh Part Info Structure (by ref)", PartInstanceInfo:"Atom Model Part Instance Info Structure (by ref)") {is=title K2Node_FunctionEntry}
   > n2 = Switch on EAtomGroupFilterMode!(execute, Selection:EAtomGroupFilterMode Enum=Get GroupFilter() -> (GroupFilter:EAtomGroupFilterMode Enum=Disabled)) -> (Disabled:Exec, UseDefaultGroupFiltering:Exec) {is=title K2Node_SwitchEnum}
   n3 = GetFirstValidGroupName!(
    execute=@n2.UseDefaultGroupFiltering
    Model:Atom Model Object Reference=GetAtomModel!() -> (Atom Model:Atom Model Object Reference)
    PartInfo:"Atom Model Part Instance Info Structure (by ref)"=@n1.PartInstanceInfo
    InvalidPrefixes:Array of Strings=Get Invalid Group Prefixes() -> (Invalid Group Prefixes:Array of Strings)
   ) -> (then, ReturnValue:String)
   > Branch(Condition=NOT Boolean(A=IsEmpty(InString=@n3)))
   > Set members in Atom Merged Mesh Part Info!(
    execute:Exec
    StructRef:"Atom Merged Mesh Part Info Structure (by ref)"=@n1
    MeshName:String=Append(
     A=Break Atom Merged Mesh Part Info!(AtomMergedMeshPartInfo:"Atom Merged Mesh Part Info Structure (by ref)"=@n1) -> (
      bHasConstantOverrideVertexColor:bool=~
      PreventMergingMaterials:Array of Material Interface Object References
      MaterialFeatures:Array of Atom Material Features Structures
      bAllowSimplification:bool=~
      bIsOccluder:bool=~
      MeshName:String
     ) {is=title K2Node_BreakStruct}.MeshName
     B=@n3
    )
   ) -> (then:Exec, StructOut:"Atom Merged Mesh Part Info Structure (by ref)") {is=title K2Node_SetFieldsInStruct}
  ]}
 CreateGeometryCollection {nodes [
   CreateGeometryCollection!() -> (then, Proxy Mesh:Static Mesh Object Reference, Brick Static Mesh Path:String) {is=title K2Node_FunctionEntry}
   > Return Node!(execute, Geometry Collection:Geometry Collection Object Reference)
  ]}
 EventGraph {links [], nodes []}
 FilterCommonParts {nodes [
   FilterCommonParts!() -> (then, InPartsCollection:"Atom Model Parts Collection Structure (by ref)") {is=title K2Node_FunctionEntry}
   > n1 = Sequence!(execute) -> (then_0, then_1, then_2:Exec)
   n2 = Get UsedStyles() -> (UsedStyles:Array of Names)
   n3 = Branch(execute=@n1.then_2, Condition=NOT Boolean(A=Contains Item(TargetArray=@n2, ItemToFind=Game)))
   n4 = Break Atom Processor Geometry Style Settings!(AtomProcessorGeometryStyleSettings:"Atom Processor Geometry Style Settings Structure (by ref)"=Get Geometry Style Settings For LOD!(LOD:int=0) -> (Style Settings:Atom Processor Geometry Style Settings Structure) {is=title K2Node_CallFunction}) -> (GeometryStylePreferences_4_#1:Array of Strings, FakeTechnicHole_41_#2:bool=~) {is=title K2Node_BreakStruct}
   n5 = Get InPartsCollection() -> (InPartsCollection:"Atom Model Parts Collection Structure (by ref)")
   > n6 = FilterCommonPartsToStyle!(
    execute
    PartsCollection:"Atom Model Parts Collection Structure (by ref)"=@n5
    GeometryStyle:Name=String To Name(InString="Get (a ref)"(Array:Array of Strings=@n4, Dimension 1=0) -> (Output:"String (by ref)"))
   ) -> (then, ReturnValue:Atom Model Parts Collection Structure)
   > Return Node!(execute, OutPartsCollection:Atom Model Parts Collection Structure=@n6)
   Return Node(execute=@n3.else, OutPartsCollection=@n5)
   n7 = For Each Loop(
    Exec=@n1.then_0
    Array=Break Atom Processor Geometry Style Settings!(AtomProcessorGeometryStyleSettings=Get Nanite Geometry Style Settings() -> (Nanite Geometry Style Settings:Atom Processor Geometry Style Settings Structure)) -> (
     GeometryStylePreferences_4_#1
     UseImprovedBevelAlgorithm_37_#3:bool=~
     OnlyBevelBlock-LikePrimitives_39_#4:bool=~
     FakeTechnicHole_41_#2=~
    )
   )
   > Add Unique(TargetArray:Array of Names=Get UsedStyles(), NewItem:"Name (by ref)"=String To Name(InString=@n7))
   n8 = For Each Loop(
    Exec=@n1.then_1
    Array:Array of Atom Processor Geometry Style Settings Structures=Get LOD Geometry Style Settings() -> (LOD Geometry Style Settings:Array of Atom Processor Geometry Style Settings Structures)
   ) -> (Array Element:Atom Processor Geometry Style Settings Structure)
   > n9 = For Each Loop(Array:Array of Strings=Break Atom Processor Geometry Style Settings!(AtomProcessorGeometryStyleSettings=@n8) -> (GeometryStylePreferences_4_#1, FakeTechnicHole_41_#2=~)) -> (Array Element:String)
   > Add Unique(TargetArray=Get UsedStyles(), NewItem=String To Name(InString=@n9))
  ]}
 FinalizeCommonPartMergedMeshPartInfo {nodes [
   n1 = FinalizeCommonPartMergedMeshPartInfo!() -> (
    then
    CommonPartMergedMeshPartInfo:"Atom Merged Mesh Part Info Structure (by ref)"
    PrimitivePartInfo:"Atom Model Part Instance Info Structure (by ref)"
    CommonPartIndex:int
   ) {is=title K2Node_FunctionEntry}
   > n2 = Branch(Condition="Not Equal (Enum)"!(
     A:EAtomMeshMergingVertexColorMappingMode Enum=Get VertexColorDebugMode() -> (VertexColorDebugMode:EAtomMeshMergingVertexColorMappingMode Enum=None)
     B:EAtomMeshMergingVertexColorMappingMode Enum=None
    ) -> (ReturnValue:bool))
   n3 = Break Atom Merged Mesh Part Info!(AtomMergedMeshPartInfo=@n1) -> (Materials:Array of Material Interface Object References, PreventMergingMaterials, bAllowSimplification=~)
   > n4 = For Each Loop(Array:Array of Material Interface Object References=@n3) -> (Array Element:Material Interface Object Reference)
   Return Node!(execute=[@n4.Completed, @n2.else], FinalizedPartInfo:Atom Merged Mesh Part Info Structure=@n1)
   > Set Array Elem!(
    execute
    TargetArray:Array of Material Interface Object References=@n3
    Index:int=@n4.Array Index
    Item:"Material Interface Object Reference (by ref)"=/Engine/EngineDebugMaterials/VertexColorMaterial
    bSizeToFit:bool=false
   ) -> (then)
  ]}
 FinalizeMergedMeshPart {nodes [
   n1 = FinalizeMergedMeshPart!() -> (then, MergedMeshPartInfo:"Atom Merged Mesh Part Info Structure (by ref)", PartInfo:Atom Model Part Instance Info Structure) {is=title K2Node_FunctionEntry}
   > n2 = Branch(Condition="Not Equal (Enum)"(A=Get VertexColorDebugMode() -> (VertexColorDebugMode=None), B=None))
   n3 = Break Atom Merged Mesh Part Info(AtomMergedMeshPartInfo=@n1) -> (bAllowSimplification=~)
   > n4 = For Each Loop(Array=@n3)
   Return Node!(execute=[@n4.Completed, @n2.else], Out MergedMeshPartInfo:Atom Merged Mesh Part Info Structure=@n1)
   > Set Array Elem(TargetArray=@n3, Index=@n4.Array Index, Item=&VertexColorMaterial, bSizeToFit=false)
  ]}
 FinalizeMesh {nodes [
   n1 = FinalizeMesh!() -> (then, Static Mesh:Static Mesh Object Reference) {is=title K2Node_FunctionEntry}
   > CalculateMergedMeshScreenSizes!(
    execute
    LODSettings:"Atom Merged Mesh LODDistance Settings Structure (by ref)"=Make Atom Merged Mesh LODDistance Settings!(
     bOverrideLODScreenSizes:bool=true
     BaseLODScreenSize:float=0.25
     BaseLODScreenSizeScaling:float=1.05
     LODDistances:"Array of Float (single-precision)s"=Make Array!("[0]":float=1, "[1]":float=225, "[2]":float=300, "[3]":float=450, "[4]":float=700, "[5]":float=800, "[6]":float=1000, "[7]":float=1200) -> (Array:"Array of Float (single-precision)s")
    ) -> (AtomMergedMeshLODDistanceSettings:Atom Merged Mesh LODDistance Settings Structure) {is=title K2Node_MakeStruct}
    StaticMesh:Static Mesh Object Reference=@n1
   ) -> (then)
   > ApplyImportedSocketsToStaticMesh!(execute, StaticMesh:Static Mesh Object Reference=@n1, AtomModel:Atom Model Object Reference=GetAtomModel()) -> (then)
  ]}
 FinalizeMeshInternal {nodes [
   n1 = FinalizeMeshInternal!() -> (
    then
    Static Mesh:Static Mesh Object Reference
    Merged Mesh Name:String
    Primitive Mesh Overrides:Map of Atom Model Part Guid Structures to Static Mesh Object References
    Primitive Instance Data:Map of Atom Model Part Guid Structures to Atom Primitive Instance Data Structures
    Detailed Collision:Atom Primitive Collision Geometry Structure
   ) {is=title K2Node_FunctionEntry}
   > n2 = Sequence!(execute) -> (then_0, then_1)
   > n3 = Branch(Condition=GetIsProcessingPrimitive!() -> (Is Processing Primitive:bool))
   n4 = GetAtomModel()
   n5 = Get Static Mesh() -> (Static Mesh:Static Mesh Object Reference)
   > AddAtomPrimitiveAssetUserDataToProcessedPrimitiveMesh!(execute, StaticMesh:Static Mesh Object Reference=@n5, AtomModel:Atom Model Object Reference=@n4) -> (then)
   n6 = GenerateDetailedCollision!() -> (GenerateDetailedCollision:bool)
   > n7 = Branch(Condition=@n6)
   n8 = Get Static Mesh()
   n9 = Get Detailed Collision() -> (Detailed Collision:Atom Primitive Collision Geometry Structure)
   > n10 = AddAtomDetailedCollisionToStaticMesh!(execute, StaticMesh:Static Mesh Object Reference=@n8, DetailedCollision:"Atom Primitive Collision Geometry Structure (by ref)"=@n9) -> (then)
   n11 = Self-Reference()
   n12 = Get Static Mesh()
   ApplyProcessorAssetImportDataToStaticMesh!(
    execute=[@n10.then, @n7.else]
    StaticMesh:Static Mesh Object Reference=@n12
    Processor:Atom Model Processor Object Reference=@n11
    AtomModel:Atom Model Object Reference
   ) -> (then)
   n13 = IsEmpty(InString=Get Merged Mesh Name() -> (Merged Mesh Name:String))
   AddAtomModelReferenceToStaticMesh!(execute=@n3.else, StaticMesh:Static Mesh Object Reference=@n5, AtomModel:Atom Model Object Reference=@n4, bSaveConnectivityState:bool=@n13) -> (then)
   > n14 = Branch(Condition=@n6)
   > n15 = AddAtomDetailedCollisionToStaticMesh(StaticMesh=@n8, DetailedCollision=@n9)
   ApplyProcessorAssetImportDataToStaticMesh(execute=[@n15.then, @n14.else], StaticMesh=@n12, Processor=@n11, AtomModel=GetAtomModel())
   > Branch(Condition=@n13)
   > n16 = Branch(Condition=UseMaterialFactory!() -> (UseMaterialFactory:bool))
   > AddPrimitiveInstanceDataToStaticMesh!(
    execute
    StaticMesh:Static Mesh Object Reference=Get Static Mesh()
    MaterialFactory:Atom Material Factory Object Reference=GetMaterialFactory!() -> (Material Factory:Atom Material Factory Object Reference)
    InstanceDataMap:Map of Atom Model Part Guid Structures to Atom Primitive Instance Data Structures=Get Primitive Instance Data() -> (Primitive Instance Data:Map of Atom Model Part Guid Structures to Atom Primitive Instance Data Structures)
   ) -> (then)
   n17 = Get Material ID Overrides!() -> (Material ID Overrides:Map of Integers to Atom Material Override Structures) {is=title K2Node_CallFunction}
   n18 = Keys!(execute=@n16.else, TargetMap:Map of Integers to Atom Material Override Structures=@n17) -> (then, Keys:Array of Integers)
   > n19 = For Each Loop(Array:Array of Integers=@n18) -> (Array Element:int)
   > Add!(
    execute
    TargetMap:Map of Integers to Material Interface Object References=Get Material ID Overrides Map() -> (Material ID Overrides Map:Map of Integers to Material Interface Object References)
    Key:"Integer (by ref)"=@n19
    Value:"Material Interface Object Reference (by ref)"=Break Atom Material Override!(AtomMaterialOverride:"Atom Material Override Structure (by ref)"=Find(TargetMap:Map of Integers to Atom Material Override Structures=@n17, Key:"Integer (by ref)"=@n19) -> (Value:Atom Material Override Structure)) -> (Material:Material Interface Object Reference, bManualColorSource:bool=~) {is=title K2Node_BreakStruct}
   ) -> (then) {is=title K2Node_CallFunction}
   AddLegacyPrimitiveInstanceDataToStaticMesh!(
    execute=@n19.Completed
    StaticMesh:Static Mesh Object Reference=Get Static Mesh()
    AtomModel:Atom Model Object Reference=GetAtomModel()
    PrimitiveMeshOverrides:Map of Atom Model Part Guid Structures to Static Mesh Object References=Get Primitive Mesh Overrides() -> (Primitive Mesh Overrides:Map of Atom Model Part Guid Structures to Static Mesh Object References)
    MaterialOverrides:Map of Integers to Material Interface Object References=Get Material ID Overrides Map()
   ) -> (then)
   FinalizeMesh!(execute=@n2.then_1, self, Static Mesh:Static Mesh Object Reference=@n1) -> (then) {is=title K2Node_CallFunction}
  ]}
 FormatAdditionalStaticMeshName {nodes [
   n1 = FormatAdditionalStaticMeshName!() -> (then, Base Static Mesh Name:String, Additional Mesh Name:String, Baked Flex Primitives Names:Array of Strings) {is=title K2Node_FunctionEntry}
   n2 = Contains Item(TargetArray:Array of Strings=@n1.Baked Flex Primitives Names, ItemToFind:"String (by ref)"=@n1.Additional Mesh Name)
   n3 = Append(A=Flex_, B=Left!(SourceString:String=@n1.Additional Mesh Name, Count:int=4) -> (ReturnValue:String))
   > Return Node!(execute, MeshName:String=Append!(A=@n1, B=_, C:String=SelectString(A=@n3, B=@n1.Additional Mesh Name, bPickA=@n2)) -> (ReturnValue))
  ]}
 GenerateDetailedCollision {nodes [
   GenerateDetailedCollision() {is=title K2Node_FunctionEntry}
   > Return Node!(execute, GenerateDetailedCollision:bool=false)
  ]}
 Get CP Geometry Material Assignments For Geometry Style {nodes [
   n1 = Get CP Geometry Material Assignments For Geometry Style!() -> (
    then
    Part Info:Atom Model Part Instance Info Structure
    Common Part Index:int
    Primitive Material Assignments:Atom Primitive Geometry Material Assignments Structure
    Common Part Geometry:Atom Primitive Geometry Object Reference
    Geometry Style Settings:Atom Processor Geometry Style Settings Structure
    Style Name:String
   ) {is=title K2Node_FunctionEntry}
   > n2 = Set members in Atom Material Features!(
    execute:Exec
    StructRef:"Atom Material Features Structure (by ref)"="Get (a copy)"(Array:Array of Atom Material Features Structures=Break Atom Primitive Geometry Material Assignments!(AtomPrimitiveGeometryMaterialAssignments:"Atom Primitive Geometry Material Assignments Structure (by ref)"=@n1.Primitive Material Assignments) -> (MaterialSlotName:Array of Strings, MaterialFeatures:Array of Atom Material Features Structures) {is=title K2Node_BreakStruct}.MaterialFeatures) -> (Output:Atom Material Features Structure)
    UDIMDecorationMap:Map of Integers to Texture Object References=Get Empty Decoration Map() -> (Empty Decoration Map:Map of Integers to Texture Object References)
   ) -> (then:Exec, StructOut:"Atom Material Features Structure (by ref)") {is=title K2Node_SetFieldsInStruct}
   > Return Node!(
    execute
    Material Assignments:Atom Primitive Geometry Material Assignments Structure=Make Atom Primitive Geometry Material Assignments!(
     MaterialSlotName:Array of Strings=Make Array!("[0]":String=CP_Default) -> (Array:Array of Strings)
     MaterialFeatures:Array of Atom Material Features Structures=Make Array("[0]":Atom Material Features Structure=@n2) -> (Array:Array of Atom Material Features Structures)
    ) -> (AtomPrimitiveGeometryMaterialAssignments:Atom Primitive Geometry Material Assignments Structure) {is=title K2Node_MakeStruct}
   )
  ]}
 Get CP Geometry Material Assignments For LOD {nodes [
   n1 = Get CP Geometry Material Assignments For LOD!() -> (
    then
    Part Info:Atom Model Part Instance Info Structure
    Common Part Index:int
    "Primitive  Merged Mesh Part Info":Atom Merged Mesh Part Info Structure
    Common Part Geometry LODs:Atom Primitive Geometry LODs Structure
    LOD:int
   ) {is=title K2Node_FunctionEntry}
   n2 = Break Atom Merged Mesh Part Info!(AtomMergedMeshPartInfo=@n1."Primitive  Merged Mesh Part Info") -> (
    GeometryLODs:Atom Primitive Geometry LODs Structure
    PreventMergingMaterials
    MaterialFeatures:Array of Atom Material Features Structures
    bAllowSimplification=~
   )
   > n3 = Get CP Geometry Material Assignments For Geometry Style!(
    execute
    self
    Part Info:Atom Model Part Instance Info Structure=@n1
    Common Part Index:int=@n1.Common Part Index
    Primitive Material Assignments:Atom Primitive Geometry Material Assignments Structure=Make Atom Primitive Geometry Material Assignments(MaterialFeatures=@n2.MaterialFeatures)
    Common Part Geometry:Atom Primitive Geometry Object Reference="Get (a copy)"(
     Array:Array of Atom Primitive Geometry Object References=Break Atom Primitive Geometry LODs!(AtomPrimitiveGeometryLODs:"Atom Primitive Geometry LODs Structure (by ref)"=@n1.Common Part Geometry LODs) -> (GeometryLODs:Array of Atom Primitive Geometry Object References, DetailedCollision:Atom Primitive Collision Geometry Structure) {is=title K2Node_BreakStruct}
     Dimension 1=@n1.LOD
    ) -> (Output:Atom Primitive Geometry Object Reference)
    Geometry Style Settings:Atom Processor Geometry Style Settings Structure=Get Geometry Style Settings For LOD(LOD=@n1.LOD)
    Style Name:String=Append!(A=LOD, B="To String (Integer)"(InInt=@n1.LOD)) -> (ReturnValue)
   ) -> (then, Material Assignments:Atom Primitive Geometry Material Assignments Structure) {is=title K2Node_CallFunction}
   > Return Node(Material Assignments=@n3)
   n4 = Break Atom Primitive Geometry LODs!(AtomPrimitiveGeometryLODs=@n2) -> (GeometryLODsMaterials:Array of Atom Primitive Geometry Material Assignments Structures, DetailedCollision)
   n5 = Length(TargetArray:Array of Atom Primitive Geometry Material Assignments Structures=@n4)
   "Get (a ref)"(
    Array:Array of Atom Primitive Geometry Material Assignments Structures=@n4
    Dimension 1="Clamp (Integer)"(Value=@n1.LOD, Min=0, Max=int - int(A=@n5)) -> (ReturnValue=0)
   ) -> (Output:"Atom Primitive Geometry Material Assignments Structure (by ref)")
  ]}
 Get CP Geometry Material Assignments For Nanite {nodes [
   n1 = Get CP Geometry Material Assignments For Nanite!() -> (
    then
    Part Info:Atom Model Part Instance Info Structure
    Common Part Index:int
    "Primitive  Merged Mesh Part Info":Atom Merged Mesh Part Info Structure
    Common Part Merged Mesh Part Info:Atom Primitive Geometry LODs Structure
   ) {is=title K2Node_FunctionEntry}
   > n2 = Get CP Geometry Material Assignments For Geometry Style(
    Part Info=@n1
    Common Part Index=@n1.Common Part Index
    Primitive Material Assignments=Make Atom Primitive Geometry Material Assignments(MaterialFeatures=Break Atom Merged Mesh Part Info!(AtomMergedMeshPartInfo=@n1."Primitive  Merged Mesh Part Info") -> (PreventMergingMaterials, MaterialFeatures, bAllowSimplification=~).MaterialFeatures)
    Common Part Geometry=Break Atom Primitive Geometry LODs!(AtomPrimitiveGeometryLODs=@n1.Common Part Merged Mesh Part Info) -> (HiResNaniteGeomery:Atom Primitive Geometry Object Reference, DetailedCollision)
    Geometry Style Settings=Get Nanite Geometry Style Settings()
    Style Name=Nanite
   )
   > Return Node(Material Assignments=@n2)
  ]}
 Get Collision Geometry for Part {nodes [
   Get Collision Geometry for Part!() -> (
    then
    Geometry:Atom Primitive Geometry Object Reference
    CollisionType:EPrimitiveAutoCollisionType Enum=Box
    PartInfo:Atom Model Part Instance Info Structure
   ) {is=title K2Node_FunctionEntry}
   n1 = Get Primitive Geometry Container Collision!(Part Info:Atom Model Part Instance Info Structure=Get PartInfo() -> (PartInfo:Atom Model Part Instance Info Structure)) -> (Success:bool, Geometry Container:Atom Primitive Geometry Container Object Reference) {is=title K2Node_CallFunction}
   > n2 = Branch(Condition=@n1)
   > Return Node!(
    execute
    Collision Geometry:Atom Primitive Collision Geometry Structure=MakePrimitiveCollisionGeometry!(
     self:Atom Primitive Geometry Container Object Reference=@n1.Geometry Container
     Scale:float=GetDefaultPrimitiveScale!() -> (ReturnValue:float)
    ) -> (ReturnValue:Atom Primitive Collision Geometry Structure)
   )
   n3 = Break Atom Model Part Instance Info!(AtomModelPartInstanceInfo:"Atom Model Part Instance Info Structure (by ref)"=Get PartInfo()) -> (Primitive:Atom Primitive Object Reference) {is=title K2Node_BreakStruct}
   n4 = Branch(
    execute=@n2.else
    Condition=AND Boolean(
     A=Get bUseImportedCollisions!(self:Atom Primitive Object Reference=@n3) -> (bUseImportedCollisions:bool)
     B=Get bUseImportedPrimitiveCollision() -> (bUseImportedPrimitiveCollision:bool)
    )
   )
   n5 = Get PartInfo()
   > n6 = GetPrimitiveSourceCollision!(
    execute
    Settings:"Primitive Source Collision Geometry Settings Structure (by ref)"=Make Primitive Source Collision Geometry Settings!(
     Primitive:Atom Primitive Object Reference=Break Atom Model Part Instance Info(AtomModelPartInstanceInfo=@n5)
     PrimitiveScale:float=GetDefaultPrimitiveScale()
     BoneTransforms:Array of Transforms=Break Atom Model Part Instance Info!(AtomModelPartInstanceInfo=@n5) -> (Transforms:Array of Transforms)
     CylinderColliderNumberOfSides:int=Get Cylinder Colliders Number Of Sides!() -> (Side Count:int=0) {is=title K2Node_CallFunction}
     bSplitNarrowColliders:bool=False
    ) -> (PrimitiveSourceCollisionGeometrySettings:Primitive Source Collision Geometry Settings Structure) {is=title K2Node_MakeStruct}
   ) -> (then, ReturnValue:Atom Primitive Collision Geometry Structure)
   > Return Node(Collision Geometry=@n6)
   n7 = GenerateCollisionGeometry!(
    execute=@n4.else
    Geometry:Atom Primitive Geometry Object Reference=Get Geometry() -> (Geometry:Atom Primitive Geometry Object Reference)
    CollisionType:EPrimitiveAutoCollisionType Enum=Get CollisionType() -> (CollisionType:EPrimitiveAutoCollisionType Enum=Box)
   ) -> (then, ReturnValue:Atom Primitive Collision Geometry Structure)
   > Return Node(Collision Geometry=@n7)
  ]}
 Get Common Part Geometry Style {nodes [
   n1 = Get Common Part Geometry Style!() -> (then, Common Part Type:EAtomCommonPartType Enum=None, bIsNanite:bool=~) {is=title K2Node_FunctionEntry}
   n2 = Select(
    Option 0:"Atom Processor Geometry Style Settings Structure (by ref)"=Get Geometry Style Settings For LOD(LOD=0)
    Option 1:"Atom Processor Geometry Style Settings Structure (by ref)"=Get Nanite Geometry Style Settings()
    Index=@n1.bIsNanite
   ) -> (ReturnValue:"Atom Processor Geometry Style Settings Structure (by ref)")
   > Return Node!(
    execute
    Geometry Style:String=GetCommonPartBestExportStyle!(
     GeometryStylePriorityList:Array of Strings=Break Atom Processor Geometry Style Settings!(AtomProcessorGeometryStyleSettings=@n2) -> (
      GeometryStylePreferences_4_#1
      PreferredGeometryLOD_30_#5:int
      ShouldBevel_8_#6:bool=~
      BevelWidth_11_#7:double=~
      BevelAngleThreshold_13_#8:double=~
      UseImprovedBevelAlgorithm_37_#3:bool=~
      OnlyBevelBlock-LikePrimitives_39_#4:bool=~
      NumberofSubdivisionsToApply_16_#9:int
      FakeBevel_23_#10:bool=~
      FakeUnderside_25_#11:bool=~
      AnalyticFakeUnderside_32_#12:bool=~
      ForceFlatCap_35_#13:bool=~
      FakeTechnicHole_41_#2=~
      AnalyticCommonParts_27_#14:bool=~
     )
     CommonPartType:EAtomCommonPartType Enum=Get Common Part Type() -> (Common Part Type:EAtomCommonPartType Enum=None)
     CommonPartsMeshPath:String=GetCustomCommonPartMeshPath!(self) -> (Common Part Mesh Path:String)
    ) -> (OutGeometryStyle:String, ReturnValue:bool)
   )
  ]}
 Get Cylinder Colliders Number Of Sides {nodes [
   Get Cylinder Colliders Number Of Sides() {is=title K2Node_FunctionEntry}
   > Return Node!(execute, Side Count:int=8)
  ]}
 Get Geometry Material Assignments For Geometry Style {nodes [
   Get Geometry Material Assignments For Geometry Style!() -> (
    then
    PartInfo:Atom Model Part Instance Info Structure
    Geometry Style Settings:Atom Processor Geometry Style Settings Structure
    StyleName:String
    Color Change Index:int
   ) {is=title K2Node_FunctionEntry}
   n1 = Get PartInfo()
   n2 = Break Atom Model Part Color Info!(AtomModelPartColorInfo:"Atom Model Part Color Info Structure (by ref)"="Get (a ref)"(
     Array:Array of Atom Model Part Color Info Structures=Break Atom Model Part Instance Info!(AtomModelPartInstanceInfo=@n1) -> (Colors:Array of Atom Model Part Color Info Structures)
     Dimension 1=Get Color Change Index() -> (Color Change Index:int=0)
    ) -> (Output:"Atom Model Part Color Info Structure (by ref)")) -> (ColorId:int, MaterialType:EAtomMaterialType Enum) {is=title K2Node_BreakStruct}
   n3 = Find(TargetMap=Get Material ID Overrides(), Key=@n2)
   > n4 = Branch(Condition=@n3.ReturnValue)
   n5 = Make Atom Material Features!(
    MaterialType:EAtomMaterialType Enum=None
    UDIMDecorationMap:Map of Integers to Texture Object References
    NormalMap:Texture Object Reference
    AlphaMask:Texture Object Reference
    TechnicNormalMap:Texture Object Reference
    TechnicMasks:Texture Object Reference
    LookUpTableAtlasTexture:Texture Object Reference=Get LUT Atlas Override!() -> (AsTexture 2D:Texture 2D Object Reference) {is=title K2Node_CallFunction}
    ColorSource:EAtomMaterialColorSource Enum=Select(
     Option 0:EAtomMaterialColorSource Enum=Get Material Color Source() -> (Material Color Source:EAtomMaterialColorSource Enum=VertexColor)
     Option 1:EAtomMaterialColorSource Enum=Material
     Index=Break Atom Material Override(AtomMaterialOverride=@n3) -> (bManualColorSource=~).bManualColorSource
    ) -> (ReturnValue:EAtomMaterialColorSource Enum)
    ColorId:int=@n2
    UVFeature:EAtomMaterialUVFeature Enum=None
    TranslucencyType:EAtomMaterialTranslucencyType Enum=Translucent
    MaterialOverride:Atom Material Override Structure=@n3
    AdditionalFeatures:Array of Instanced Struct Structures
   ) -> (AtomMaterialFeatures:Atom Material Features Structure) {is=title K2Node_MakeStruct}
   n6 = Append(A=Get StyleName() -> (StyleName:String), B=_Default)
   > Return Node!(
    execute
    Material Assignments:Atom Primitive Geometry Material Assignments Structure=Make Atom Primitive Geometry Material Assignments(
     MaterialSlotName=Make Array("[0]":String=@n6) -> (Array:Array of Strings)
     MaterialFeatures=Make Array("[0]":Atom Material Features Structure=@n5) -> (Array:Array of Atom Material Features Structures)
    )
   )
   n7 = GetIsProcessingPrimitive()
   n8 = Branch(execute=@n4.else, Condition=AND Boolean(A=Get Support Decorations() -> (Support Decorations:bool), B=NOT Boolean(A=@n7)))
   > n9 = For Each Loop(Array:Array of Atom Model Part Decoration Info Structures=Break Atom Model Part Instance Info!(AtomModelPartInstanceInfo=Get PartInfo()) -> (Decorations:Array of Atom Model Part Decoration Info Structures)) -> (Array Element:Atom Model Part Decoration Info Structure)
   n10 = Break Atom Model Part Decoration Info!(AtomModelPartDecorationInfo:"Atom Model Part Decoration Info Structure (by ref)"=@n9) -> (Texture:Texture Object Reference, PrimitiveSurfaceIndex:int) {is=title K2Node_BreakStruct}
   > Add(
    TargetMap:Map of Integers to Texture Object References=Get UDIMDecorationMap() -> (UDIMDecorationMap:Map of Integers to Texture Object References)
    Key=@n10.PrimitiveSurfaceIndex
    Value:"Texture Object Reference (by ref)"=@n10
   )
   n11 = GetPartColorInfo!(Part Info:Atom Model Part Instance Info Structure=Get PartInfo(), Color Change Index:int=Get Color Change Index() -> (Color Change Index=0)) -> (Color Id:int=0, Material Type:EAtomMaterialType Enum=None)
   n12 = Get UDIMDecorationMap()
   n13 = Make Atom Material Features(
    MaterialType=@n11.Material Type
    UDIMDecorationMap=Get UDIMDecorationMap()
    LookUpTableAtlasTexture=Get LUT Atlas Override()
    ColorSource=Get Material Color Source() -> (Material Color Source=VertexColor)
    ColorId=@n11
    UVFeature=Select(
     Option 0:EAtomMaterialUVFeature Enum=Literal enum EAtomMaterialUVFeature!(Enum:EAtomMaterialUVFeature Enum=None) -> (ReturnValue:EAtomMaterialUVFeature Enum) {is=title K2Node_EnumLiteral}
     Option 1:EAtomMaterialUVFeature Enum=Literal enum EAtomMaterialUVFeature(Enum=FakeBevels)
     Index=Is Empty!(TargetMap:Map of Integers to Texture Object References=@n12) -> (ReturnValue:bool) {is=title K2Node_CallFunction}
    ) -> (ReturnValue:EAtomMaterialUVFeature Enum)
    TranslucencyType=Translucent
    MaterialOverride=(Material=None,bManualColorSource=False)
   )
   n14 = Append(A=Get StyleName(), B=_Default)
   Return Node(
    execute=[@n8.else, @n9.Completed]
    Material Assignments=Make Atom Primitive Geometry Material Assignments(
     MaterialSlotName=Make Array("[0]":String=@n14) -> (Array:Array of Strings)
     MaterialFeatures=Make Array("[0]":Atom Material Features Structure=@n13) -> (Array:Array of Atom Material Features Structures)
    )
   )
  ]}
 Get Geometry Material Assignments For LOD {nodes [
   n1 = Get Geometry Material Assignments For LOD!() -> (then, PartInfo:Atom Model Part Instance Info Structure, LOD:int, Color Change Index:int) {is=title K2Node_FunctionEntry}
   > n2 = Get Geometry Material Assignments For Geometry Style!(
    execute
    self
    PartInfo:Atom Model Part Instance Info Structure=@n1
    Geometry Style Settings:Atom Processor Geometry Style Settings Structure=Get Geometry Style Settings For LOD(LOD=@n1.LOD)
    StyleName:String=Append(A=LOD, B="To String (Integer)"(InInt=@n1.LOD))
    Color Change Index:int=@n1.Color Change Index
   ) -> (then, Material Assignments:Atom Primitive Geometry Material Assignments Structure) {is=title K2Node_CallFunction}
   > Return Node(Material Assignments=@n2)
  ]}
 Get Geometry Material Assignments For Nanite {nodes [
   n1 = Get Geometry Material Assignments For Nanite!() -> (then, PartInfo:Atom Model Part Instance Info Structure, Color Change Index:int) {is=title K2Node_FunctionEntry}
   > n2 = Get Geometry Material Assignments For Geometry Style(PartInfo=@n1, Geometry Style Settings=Get Nanite Geometry Style Settings(), StyleName=Nanite, Color Change Index=@n1.Color Change Index)
   > Return Node(Material Assignments=@n2)
  ]}
 Get Geometry Style Settings For LOD {nodes [
   n1 = Get Geometry Style Settings For LOD!() -> (then, LOD:int) {is=title K2Node_FunctionEntry}
   n2 = Get UseNewMergeAPI() -> (UseNewMergeAPI:bool)
   n3 = Get LOD Geometry Style Settings()
   n4 = Last Index!(TargetArray:Array of Atom Processor Geometry Style Settings Structures=@n3) -> (ReturnValue=0)
   n5 = int - int(A=@n1, B=@n4)
   > n6 = Branch(Condition=OR Boolean(A=NOT Boolean(A=@n2), B="integer <= integer"(A=@n5)))
   > Return Node!(
    execute
    Style Settings:Atom Processor Geometry Style Settings Structure="Get (a ref)"(Array:Array of Atom Processor Geometry Style Settings Structures=@n3, Dimension 1=@n1) -> (Output:"Atom Processor Geometry Style Settings Structure (by ref)")
   )
   n7 = "Get (a ref)"(Array=@n3, Dimension 1=@n4)
   n8 = Set members in Atom Processor Geometry Style Settings!(
    execute:Exec=@n6.else
    StructRef:"Atom Processor Geometry Style Settings Structure (by ref)"=Copy!(Item:Atom Processor Geometry Style Settings Structure=@n7) -> (ReturnValue:Atom Processor Geometry Style Settings Structure) {is=title K2Node_Copy}
    PreferredGeometryLOD_30_#5:int=int + int(
     A=@n5
     B=Break Atom Processor Geometry Style Settings(AtomProcessorGeometryStyleSettings=@n7) -> (
      ShouldBevel_8_#6=~
      BevelWidth_11_#7=~
      BevelAngleThreshold_13_#8=~
      UseImprovedBevelAlgorithm_37_#3=~
      OnlyBevelBlock-LikePrimitives_39_#4=~
      FakeBevel_23_#10=~
      FakeUnderside_25_#11=~
      AnalyticFakeUnderside_32_#12=~
      ForceFlatCap_35_#13=~
      FakeTechnicHole_41_#2=~
      AnalyticCommonParts_27_#14=~
     ).PreferredGeometryLOD_30_#5
    )
   ) -> (then:Exec, StructOut:"Atom Processor Geometry Style Settings Structure (by ref)") {is=title K2Node_SetFieldsInStruct}
   > Return Node(Style Settings=@n8)
  ]}
 Get Geometry Style Settings Last LOD {nodes [
   n1 = Get Geometry Style Settings Last LOD!() -> (then, Part Info:Atom Model Part Instance Info Structure) {is=title K2Node_FunctionEntry}
   n2 = Break Atom Model Part Instance Info!(AtomModelPartInstanceInfo=@n1) -> (Primitive:Atom Primitive Object Reference)
   n3 = Get LOD Geometry Style Settings()
   n4 = Last Index(TargetArray=@n3) -> (ReturnValue=0)
   n5 = Break Atom Processor Geometry Style Settings(AtomProcessorGeometryStyleSettings="Get (a ref)"(Array=@n3, Dimension 1=@n4)) -> (
    ShouldBevel_8_#6=~
    BevelWidth_11_#7=~
    BevelAngleThreshold_13_#8=~
    UseImprovedBevelAlgorithm_37_#3=~
    OnlyBevelBlock-LikePrimitives_39_#4=~
    FakeBevel_23_#10=~
    FakeUnderside_25_#11=~
    AnalyticFakeUnderside_32_#12=~
    ForceFlatCap_35_#13=~
    FakeTechnicHole_41_#2=~
    AnalyticCommonParts_27_#14=~
   )
   n6 = GetGeometryContainerForBestExportStyle!(
    self:Atom Primitive Object Reference=@n2
    GeometryStylePriorityList:Array of Strings=@n5
    LODIndex:int=@n5.PreferredGeometryLOD_30_#5
   ) -> (ReturnValue:Atom Primitive Geometry Container Object Reference)
   n7 = GetExportStyleMaxLOD!(
    self:Atom Primitive Object Reference=@n2
    ExportStyle:String=GetExportStyleName!(self:Atom Primitive Geometry Container Object Reference=@n6) -> (ReturnValue:String)
   ) -> (ReturnValue:int=0)
   n8 = GetLOD!(self:Atom Primitive Geometry Container Object Reference=@n6) -> (ReturnValue:int=0)
   > Return Node!(execute, Num Style Settings:int=int + int(A=@n4, B=int - int(A=@n7, B=@n8)))
  ]}
 Get LUT Atlas Override {nodes [
   Get LUT Atlas Override() {is=title K2Node_FunctionEntry}
   n1 = Get ColorThemeLibrary!(self:Atom Model Object Reference=GetAtomModel()) -> (ColorThemeLibrary:Atom Color Theme Library Object Reference)
   > n2 = Is Valid(InputObject=@n1)
   Return Node!(execute=[@n4.CastFailed, @n2.Is Not Valid], AsTexture 2D:Texture 2D Object Reference)
   > n3 = LoadAsset_Blocking(Asset=Get GeneratedTexture!(self:Atom Color Theme Library Object Reference=@n1) -> (GeneratedTexture:Texture 2D Soft Object Reference))
   > n4 = Cast To Texture2D(Object=@n3)
   > Return Node(AsTexture 2D=@n4)
  ]}
 Get Material ID Overrides {nodes [
   Get Material ID Overrides() {is=title K2Node_FunctionEntry}
   > Return Node!(execute, Material ID Overrides:Map of Integers to Atom Material Override Structures)
  ]}
 Get Merged Parts Infos {nodes [
   n1 = Get Merged Parts Infos!() -> (then, AtomModelPartsCollection:"Atom Model Parts Collection Structure (by ref)") {is=title K2Node_FunctionEntry}
   n2 = Break Atom Model Parts Collection!(AtomModelPartsCollection:"Atom Model Parts Collection Structure (by ref)"=@n1) -> (Name:String, Pivot:Transform, Parts:Array of Atom Model Part Instance Info Structures) {is=title K2Node_BreakStruct}
   > n3 = For Each Loop(Array:Array of Atom Model Part Instance Info Structures=@n2.Parts) -> (Array Element:Atom Model Part Instance Info Structure)
   Return Node!(
    execute=@n3.Completed
    Merged Mesh Part Infos:Array of Atom Merged Mesh Part Info Structures=Get MergedMeshPartInfos() -> (MergedMeshPartInfos:Array of Atom Merged Mesh Part Info Structures)
    Baked Flex Primitives Names:Array of Strings=Get BakedFlexPrimitivesNames() -> (BakedFlexPrimitivesNames:Array of Strings)
    Primitive Instance Data Map:Map of Atom Model Part Guid Structures to Atom Primitive Instance Data Structures=Get PrimitiveInstanceDataMap() -> (PrimitiveInstanceDataMap:Map of Atom Model Part Guid Structures to Atom Primitive Instance Data Structures)
    Primitive Mesh Overrides:Map of Atom Model Part Guid Structures to Static Mesh Object References=Get PrimitiveMeshOverrides() -> (PrimitiveMeshOverrides:Map of Atom Model Part Guid Structures to Static Mesh Object References)
   )
   > n4 = Set Atom Model Part Info(Atom Model Part Info:Atom Model Part Instance Info Structure=@n3) -> (Output_Get:Atom Model Part Instance Info Structure)
   > Branch(Condition=IsValid(Object=Break Atom Model Part Instance Info!(AtomModelPartInstanceInfo=@n4) -> (Primitive, Group:String, ParentGroup:String)))
   > n5 = Sequence!(execute) -> (then_0, then_1, then_2:Exec, then_3:Exec, then_4:Exec, then_5:Exec, then_6:Exec, then_7:Exec)
   > Set Merged Mesh Part Info(Merged Mesh Part Info:Atom Merged Mesh Part Info Structure=Make Atom Merged Mesh Part Info!(
     bHasConstantOverrideVertexColor:bool=True
     PreventMergingMaterials:Array of Material Interface Object References
     MaterialFeatures:Array of Atom Material Features Structures
     bAllowSimplification:bool=True
     bIsOccluder:bool=True
     FilterLODLevel:int=-1
     MeshName:String
    ) -> (AtomMergedMeshPartInfo:Atom Merged Mesh Part Info Structure) {is=title K2Node_MakeStruct}) -> (Output_Get:Atom Merged Mesh Part Info Structure)
   > Add(
    TargetMap:Map of Atom Model Part Guid Structures to Atom Primitive Instance Data Structures=Get PrimitiveInstanceDataMap()
    Key:"Atom Model Part Guid Structure (by ref)"=Break Atom Model Part Instance Info!(AtomModelPartInstanceInfo=Get Atom Model Part Info() -> (Atom Model Part Info:Atom Model Part Instance Info Structure)) -> (PartGuid:Atom Model Part Guid Structure)
    Value:"Atom Primitive Instance Data Structure (by ref)"=Make Atom Primitive Instance Data!(MaterialAssignments:Array of Atom Primitive Material Assignment Structures, MeshOverride:Static Mesh Soft Object Reference) -> (AtomPrimitiveInstanceData:Atom Primitive Instance Data Structure) {is=title K2Node_MakeStruct}
   )
   n6 = Branch(
    execute=@n5.then_2
    Condition=Has Custom Baked Vertex Color!(PartInfo:Atom Model Part Instance Info Structure=Get Atom Model Part Info()) -> (HasCustomBakedVertexColor:bool)
   )
   > Set members in Atom Merged Mesh Part Info!(
    execute
    StructRef=Get Merged Mesh Part Info() -> (Merged Mesh Part Info:Atom Merged Mesh Part Info Structure)
    bHasConstantOverrideVertexColor:bool=false
   ) -> (then, StructOut)
   n7 = GetVertexColorForPart!(execute=@n6.else, self, Part Info:"Atom Model Part Instance Info Structure (by ref)"=Get Atom Model Part Info(), Color Change Index:int=0) -> (then, Color:Color Structure)
   > Set members in Atom Merged Mesh Part Info!(execute, StructRef=Get Merged Mesh Part Info(), VertexColor:Color Structure=@n7) -> (then, StructOut)
   > Set members in Atom Merged Mesh Part Info!(execute, StructRef=Get Merged Mesh Part Info(), bHasConstantOverrideVertexColor:bool=True) -> (then, StructOut)
   GetCommonPartsGeometry!(
    execute=@n5.then_7
    self
    Part Info:"Atom Model Part Instance Info Structure (by ref)"=Get Atom Model Part Info()
    MergedMeshPartInfo:Atom Merged Mesh Part Info Structure=Get Merged Mesh Part Info()
    Common Part To Geometry LODs:Map of Strings to Atom Primitive Geometry LODs Structures=Get Common Parts Geometry Map() -> (Common Parts Geometry Map:Map of Strings to Atom Primitive Geometry LODs Structures)
   ) -> (then)
   > n8 = For Each Loop(Array:Array of Atom Common Part And Transform Structures=Break Atom Model Part Instance Info!(AtomModelPartInstanceInfo=Get Atom Model Part Info()) -> (CommonParts:Array of Atom Common Part And Transform Structures)) -> (Array Element:Atom Common Part And Transform Structure)
   n9 = Break Atom Common Part and Transform!(AtomCommonPartAndTransform:"Atom Common Part And Transform Structure (by ref)"=@n8) -> (Type:EAtomCommonPartType Enum, Transform:Transform, GeometryStyles:Array of Names) {is=title K2Node_BreakStruct}
   > Set Common Part Type(Common Part Type:EAtomCommonPartType Enum=@n9) -> (Output_Get:EAtomCommonPartType Enum=None)
   n10 = Get Merged Mesh Part Info()
   > n11 = Set Common Part Transform(Common Part Transform:Transform=ComposeTransforms!(
     A:"Transform (by ref)"=@n9.Transform
     B:"Transform (by ref)"=Break Atom Merged Mesh Part Info!(AtomMergedMeshPartInfo=@n10) -> (Transform:Transform, PreventMergingMaterials, bAllowSimplification=~)
    ) -> (ReturnValue:Transform)) -> (Output_Get:Transform)
   > n12 = Get Unique Name for Common Part Geometry LODs!(
    execute
    self
    PartInfo:Atom Model Part Instance Info Structure=Get Atom Model Part Info()
    MergedMeshPartInfo:Atom Merged Mesh Part Info Structure=Get Merged Mesh Part Info()
    CommonPartIndex:int=@n8.Array Index
   ) -> (then, Name:String) {is=title K2Node_CallFunction}
   n13 = Find(TargetMap:Map of Strings to Atom Primitive Geometry LODs Structures=Get Common Parts Geometry Map(), Key:"String (by ref)"=@n12) -> (Value:Atom Primitive Geometry LODs Structure)
   > n14 = GetMaterialsForCommonPart!(
    execute
    self
    Common Part Geometry LODs:"Atom Primitive Geometry LODs Structure (by ref)"=@n13
    Shell Merged Mesh Part Info:"Atom Merged Mesh Part Info Structure (by ref)"=Get Merged Mesh Part Info()
    PartInfo:"Atom Model Part Instance Info Structure (by ref)"=Get Atom Model Part Info()
    CommonPartIndex:int=@n8.Array Index
   ) -> (then, Common Part Materials:Array of Material Interface Object References, Material Features:Array of Atom Material Features Structures)
   > n15 = GetVertexColorForCommonPart!(
    execute
    self
    Part Info:"Atom Model Part Instance Info Structure (by ref)"=Get Atom Model Part Info()
    Merged Mesh Part Info:"Atom Merged Mesh Part Info Structure (by ref)"=Get Merged Mesh Part Info()
    Common Part Index:int=@n8.Array Index
   ) -> (then, Vertex Color:Color Structure)
   > n16 = Set Common Part Merged Mesh Part Info(Common Part Merged Mesh Part Info:Atom Merged Mesh Part Info Structure=Make Atom Merged Mesh Part Info!(
     GeometryLODs:Atom Primitive Geometry LODs Structure=@n13
     Transform:Transform=@n11
     bHasConstantOverrideVertexColor=True
     VertexColor:Color Structure=@n15
     Materials:Array of Material Interface Object References=@n14
     PreventMergingMaterials
     MaterialFeatures=@n14.Material Features
     MeshDetailLevel:EAtomMergedMeshPartDetailLevel Enum=Decorative
     bPreserveUVs:bool=true
     bAllowSimplification=True
     bIsOccluder=True
     FilterLODLevel=-1
     MeshName
    ) -> (AtomMergedMeshPartInfo)) -> (Output_Get:Atom Merged Mesh Part Info Structure)
   n17 = Get Atom Model Part Info()
   > SetCommonPartMeshName!(
    execute
    self
    Merged Mesh Part Info:"Atom Merged Mesh Part Info Structure (by ref)"=@n16
    Atom Model Part Info:"Atom Model Part Instance Info Structure (by ref)"=@n17
    Common Part Index:int=@n8.Array Index
    Common Part Type:"EAtomCommonPartType Enum (by ref)"=Get Common Part Type() -> (Common Part Type=None)
   ) -> (then)
   > n18 = FinalizeCommonPartMergedMeshPartInfo!(
    execute
    self
    CommonPartMergedMeshPartInfo:"Atom Merged Mesh Part Info Structure (by ref)"=@n16
    PrimitivePartInfo:"Atom Model Part Instance Info Structure (by ref)"=@n17
    CommonPartIndex:int=@n8.Array Index
   ) -> (then, FinalizedPartInfo:Atom Merged Mesh Part Info Structure) {is=title K2Node_CallFunction}
   > n19 = Set Common Part Merged Mesh Part Info(Common Part Merged Mesh Part Info=@n18)
   n20 = ShouldSplitTransparentPart!(MergedMeshPartInfo:"Atom Merged Mesh Part Info Structure (by ref)"=@n19) -> (Should Split Transparent Part:bool)
   > n21 = Branch(Condition=AND Boolean(A=ShouldDefineSeparateNaniteCommonParts!() -> (Separate Nanite Common Parts:bool), B=NOT Boolean(A=@n20)))
   n22 = Break Atom Merged Mesh Part Info!(AtomMergedMeshPartInfo=Get Common Part Merged Mesh Part Info() -> (Common Part Merged Mesh Part Info:Atom Merged Mesh Part Info Structure)) -> (
    GeometryLODs:Atom Primitive Geometry LODs Structure
    Transform
    bHasConstantOverrideVertexColor:bool=~
    VertexColor:Color Structure
    Materials:Array of Material Interface Object References
    PreventMergingMaterials
    MaterialFeatures:Array of Atom Material Features Structures
    MeshDetailLevel:EAtomMergedMeshPartDetailLevel Enum
    bPreserveUVs:bool=~
    bAllowMerging:bool=~
    bAllowSimplification=~
    bAllowApproximation:bool=~
    bIsOccluder:bool=~
    FilterLODLevel:int
    MeshName:String
    InstanceSubsetID:int
    ApproximationType:Bitmask Integer
   )
   n23 = Break Atom Primitive Geometry LODs!(AtomPrimitiveGeometryLODs=@n22) -> (
    GeometryLODs:Array of Atom Primitive Geometry Object References
    GeometryLODsMaterials:Array of Atom Primitive Geometry Material Assignments Structures
    HiResNaniteGeomery
    HiResNaniteGeomeryMaterials:Atom Primitive Geometry Material Assignments Structure
    Collision:Atom Primitive Collision Geometry Structure
    DetailedCollision
   )
   > n24 = Set Nanite Common Part Merged Mesh Part Info(Nanite Common Part Merged Mesh Part Info:Atom Merged Mesh Part Info Structure=Make Atom Merged Mesh Part Info!(
     GeometryLODs=Make Atom Primitive Geometry LODs!(
      GeometryLODs:Array of Atom Primitive Geometry Object References
      GeometryLODsMaterials:Array of Atom Primitive Geometry Material Assignments Structures
      HiResNaniteGeomery:Atom Primitive Geometry Object Reference=@n23.HiResNaniteGeomery
      HiResNaniteGeomeryMaterials:Atom Primitive Geometry Material Assignments Structure=@n23.HiResNaniteGeomeryMaterials
      Collision:Atom Primitive Collision Geometry Structure=@n23.Collision
      DetailedCollision:Atom Primitive Collision Geometry Structure=(AggGeom=(SphereElems=,BoxElems=,SphylElems=,ConvexElems=,TaperedCapsuleElems=,LevelSetElems=,SkinnedLevelSetElems=,MLLevelSetElems=,SkinnedTriangleMeshElems=))
     ) -> (AtomPrimitiveGeometryLODs:Atom Primitive Geometry LODs Structure) {is=title K2Node_MakeStruct}
     Transform=@n22.Transform
     bHasConstantOverrideVertexColor=@n22.bHasConstantOverrideVertexColor
     VertexColor=@n22.VertexColor
     Materials=@n22.Materials
     PreventMergingMaterials
     MaterialFeatures=@n22.MaterialFeatures
     MeshDetailLevel=@n22.MeshDetailLevel
     bPreserveUVs=@n22.bPreserveUVs
     bAllowMerging:bool=@n22.bAllowMerging
     bAllowSimplification=True
     bAllowApproximation:bool=@n22.bAllowApproximation
     bIsOccluder=@n22.bIsOccluder
     FilterLODLevel=-1
     MeshName=@n22.MeshName
     InstanceSubsetID:int=@n22.InstanceSubsetID
     ApproximationType:Bitmask Integer=@n22.ApproximationType
    ) -> (AtomMergedMeshPartInfo)) -> (Output_Get:Atom Merged Mesh Part Info Structure)
   > SetNaniteCommonPartInfo!(
    execute
    self
    Merged Mesh Part Info:"Atom Merged Mesh Part Info Structure (by ref)"=@n24
    Shell Merged Mesh Part Info:"Atom Merged Mesh Part Info Structure (by ref)"=Get Merged Mesh Part Info()
    Atom Model Part Info:"Atom Model Part Instance Info Structure (by ref)"=Get Atom Model Part Info()
   ) -> (then)
   > Add(
    TargetArray:Array of Atom Merged Mesh Part Info Structures=Get MergedMeshPartInfos()
    NewItem:"Atom Merged Mesh Part Info Structure (by ref)"=@n24
   ) {is=title K2Node_CallArrayFunction}
   n25 = Get Common Part Merged Mesh Part Info()
   > n26 = Set members in Atom Primitive Geometry LODs!(
    execute:Exec
    StructRef:"Atom Primitive Geometry LODs Structure (by ref)"=Break Atom Merged Mesh Part Info(AtomMergedMeshPartInfo=@n25) -> (bHasConstantOverrideVertexColor=~, bPreserveUVs=~, bAllowMerging=~, bAllowSimplification=~, bAllowApproximation=~, bIsOccluder=~)
    HiResNaniteGeomery:Atom Primitive Geometry Object Reference
   ) -> (then:Exec, StructOut:"Atom Primitive Geometry LODs Structure (by ref)") {is=title K2Node_SetFieldsInStruct}
   > n27 = Set members in Atom Merged Mesh Part Info!(execute, StructRef=@n25, GeometryLODs:Atom Primitive Geometry LODs Structure=@n26) -> (then, StructOut)
   Add(execute=[@n21.else, @n27.then], TargetArray=Get MergedMeshPartInfos(), NewItem=Get Common Part Merged Mesh Part Info())
   n28 = Break Atom Model Part Instance Info!(AtomModelPartInstanceInfo=Get Atom Model Part Info()) -> (Transforms:Array of Transforms, Group:String, ParentGroup:String, CommonParts, bIsUndersideVisible:bool=~)
   Set members in Atom Merged Mesh Part Info!(
    execute=@n5.then_4
    StructRef=Get Merged Mesh Part Info()
    Transform:Transform=ComposeTransforms(
     A="Get (a copy)"(Array:Array of Transforms=@n28) -> (Output:Transform)
     B=InvertTransform!(T:"Transform (by ref)"=@n2.Pivot) -> (ReturnValue:Transform)
    )
   ) -> (then, StructOut)
   n29 = Get Atom Model Part Info()
   n30 = Is Color Change Part!(PartInfo:"Atom Model Part Instance Info Structure (by ref)"=@n29) -> (Is Color Change Part:bool)
   n31 = Branch(
    execute=@n5.then_6
    Condition=OR Boolean!(
     A=GetIsProcessingPrimitive()
     B=NOT Boolean(A=@n30)
     C:bool=Has Color Change Baked in Geometry!(PartInfo:Atom Model Part Instance Info Structure=@n29) -> (HasColorChangeBakedInGeometry:bool)
    ) -> (ReturnValue)
   )
   > n32 = Add(TargetArray=Get MergedMeshPartInfos(), NewItem=Get Merged Mesh Part Info())
   n33 = Break Atom Model Part Instance Info!(AtomModelPartInstanceInfo=Get Atom Model Part Info()) -> (PartGuid:Atom Model Part Guid Structure, Primitive:Atom Primitive Object Reference, Group, ParentGroup)
   n34 = Branch(
    execute=[@n32.then, @n44.then]
    Condition=AND Boolean(
     A=Get bGenerateBakedFlexPrimitives() -> (bGenerateBakedFlexPrimitives:bool)
     B=IsFlexElement!(self:Atom Primitive Object Reference=@n33.Primitive) -> (ReturnValue:bool)
    )
   )
   n35 = Break Atom Merged Mesh Part Info(AtomMergedMeshPartInfo=Get Merged Mesh Part Info()) -> (bHasConstantOverrideVertexColor=~, bPreserveUVs=~, bAllowMerging=~, bAllowSimplification=~, bAllowApproximation=~, bIsOccluder=~)
   n36 = To String!(InModelPartGuid:"Atom Model Part Guid Structure (by ref)"=@n33) -> (ReturnValue:String)
   > Add(
    TargetArray=Get MergedMeshPartInfos()
    NewItem=Make Atom Merged Mesh Part Info(
     GeometryLODs=@n35
     Transform="0,0,0|0,0,0|1,1,1"
     bHasConstantOverrideVertexColor=@n35.bHasConstantOverrideVertexColor
     VertexColor=@n35.VertexColor
     Materials=@n35.Materials
     MaterialFeatures=@n35.MaterialFeatures
     MeshDetailLevel=@n35.MeshDetailLevel
     bPreserveUVs=@n35.bPreserveUVs
     bAllowMerging=@n35.bAllowMerging
     bAllowSimplification=True
     bAllowApproximation=@n35.bAllowApproximation
     bIsOccluder=@n35.bIsOccluder
     FilterLODLevel=-1
     MeshName=@n36
     InstanceSubsetID=@n35.InstanceSubsetID
     ApproximationType=@n35.ApproximationType
    )
   )
   > n37 = Add(TargetArray:Array of Strings=Get BakedFlexPrimitivesNames(), NewItem:"String (by ref)"=@n36)
   n38 = Break Atom Model Part Instance Info!(AtomModelPartInstanceInfo=Get Atom Model Part Info()) -> (PartGuid, Primitive, Decorations:Array of Atom Model Part Decoration Info Structures, Group, ParentGroup)
   n39 = GetIsProcessingPrimitive()
   Branch(
    execute=[@n37.then, @n34.else]
    Condition=AND Boolean!(
     A=Is Not Empty!(TargetArray:Array of Atom Model Part Decoration Info Structures=@n38.Decorations) -> (ReturnValue:bool)
     B=Get bGenerateDecoratedPrimitiveMesh() -> (bGenerateDecoratedPrimitiveMesh:bool)
     C:bool=NOT Boolean(A=@n39)
    ) -> (ReturnValue)
   )
   > n40 = RunPrimitiveProcessor!(
    execute
    Primitive:Atom Primitive Object Reference=@n38.Primitive
    PrimitiveSettings:"Atom On Process Primitive Settings Structure (by ref)"=Make Atom on Process Primitive Settings!(CustomLocation:String=GetProcessedPrimitivePath!() -> (Processed Primitive Path:String), bSupportDecorations:bool=true) -> (AtomOnProcessPrimitiveSettings:Atom On Process Primitive Settings Structure) {is=title K2Node_MakeStruct}
   ) -> (then, ReturnValue:Atom Processor Result Structure)
   n41 = Break Atom Processor Result!(AtomProcessorResult:"Atom Processor Result Structure (by ref)"=@n40) -> (bSuccess:bool=~, ProcessedObjects:Array of Object References, SharedAssets:Array of Object References) {is=title K2Node_BreakStruct}
   > Branch(Condition=AND Boolean!(A=@n41, B=Is Not Empty(TargetArray:Array of Object References=@n41.ProcessedObjects)) -> (ReturnValue))
   > n42 = Cast To StaticMesh(Object="Get (a copy)"(Array:Array of Object References=@n41.ProcessedObjects) -> (Output:Object Reference))
   > Add(
    TargetMap:Map of Atom Model Part Guid Structures to Static Mesh Object References=Get PrimitiveMeshOverrides()
    Key=@n38
    Value:"Static Mesh Object Reference (by ref)"=@n42
   ) {is=title K2Node_CallFunction}
   > RegisterPrimitiveMeshOverrideData!(
    execute
    self
    Primitive Insance Data Map:Map of Atom Model Part Guid Structures to Atom Primitive Instance Data Structures=Get PrimitiveInstanceDataMap()
    Part Guid:Atom Model Part Guid Structure=@n38
    Mesh Override:Static Mesh Object Reference=@n42
   ) -> (then)
   n43 = GetColorChangeMergedMeshPartInfos!(
    execute=@n31.else
    self
    Part Info:"Atom Model Part Instance Info Structure (by ref)"=Get Atom Model Part Info()
    MergedMeshPartInfo:"Atom Merged Mesh Part Info Structure (by ref)"=Get Merged Mesh Part Info()
   ) -> (then, MergedMeshPartInfos:Array of Atom Merged Mesh Part Info Structures)
   > n44 = Append Array!(
    execute
    TargetArray:Array of Atom Merged Mesh Part Info Structures=Get MergedMeshPartInfos()
    SourceArray:Array of Atom Merged Mesh Part Info Structures=@n43
   ) -> (then)
   SetPartOptimizationStrategy!(
    execute=@n5.then_5
    self
    Merged Mesh Part Info:"Atom Merged Mesh Part Info Structure (by ref)"=Get Merged Mesh Part Info()
    Atom Model Part Info:"Atom Model Part Instance Info Structure (by ref)"=Get Atom Model Part Info()
   ) -> (then)
   > SetPartMeshName!(
    execute
    self
    Merged Mesh Part Info:"Atom Merged Mesh Part Info Structure (by ref)"=Get Merged Mesh Part Info()
    Atom Model Part Info:"Atom Model Part Instance Info Structure (by ref)"=Get Atom Model Part Info()
   ) -> (then)
   n45 = Get Merged Mesh Part Info()
   > n46 = Branch(Condition=ShouldSplitTransparentPart(MergedMeshPartInfo=@n45))
   > n47 = Set members in Atom Primitive Geometry LODs(StructRef=Break Atom Merged Mesh Part Info!(AtomMergedMeshPartInfo=@n45) -> (GeometryLODs, PreventMergingMaterials, bAllowSimplification=~))
   > n48 = Set members in Atom Merged Mesh Part Info!(execute, StructRef=Get Merged Mesh Part Info(), GeometryLODs:Atom Primitive Geometry LODs Structure=@n47) -> (then, StructOut)
   n49 = FinalizeMergedMeshPart!(
    execute=[@n48.then, @n46.else]
    self
    MergedMeshPartInfo:"Atom Merged Mesh Part Info Structure (by ref)"=Get Merged Mesh Part Info()
    PartInfo:Atom Model Part Instance Info Structure=Get Atom Model Part Info()
   ) -> (then, Out MergedMeshPartInfo:Atom Merged Mesh Part Info Structure) {is=title K2Node_CallFunction}
   > Set Merged Mesh Part Info(Merged Mesh Part Info=@n49)
   > RegisterPrimitiveMaterialData!(
    execute
    self
    Primitive Insance Data Map:Map of Atom Model Part Guid Structures to Atom Primitive Instance Data Structures=Get PrimitiveInstanceDataMap()
    Part Info:"Atom Model Part Instance Info Structure (by ref)"=Get Atom Model Part Info()
    Mesh Part Info:"Atom Merged Mesh Part Info Structure (by ref)"=Get Merged Mesh Part Info()
   ) -> (then)
   > ProcessUncommonParts!(
    execute
    self
    Part Instance Info:"Atom Model Part Instance Info Structure (by ref)"=Get Atom Model Part Info()
    Merged Mesh Part Info:"Atom Merged Mesh Part Info Structure (by ref)"=Get Merged Mesh Part Info()
    Merged Mesh Part Infos:Array of Atom Merged Mesh Part Info Structures=Get MergedMeshPartInfos()
   ) -> (then)
   n50 = GetMaterialsForPart!(
    execute=@n5.then_3
    self
    Part Info:"Atom Model Part Instance Info Structure (by ref)"=Get Atom Model Part Info()
    MergedMeshPartInfo:Atom Merged Mesh Part Info Structure=Get Merged Mesh Part Info()
   ) -> (then, Material:Array of Material Interface Object References, MaterialFeatures:Array of Atom Material Features Structures)
   > Set members in Atom Merged Mesh Part Info!(
    execute
    StructRef=Get Merged Mesh Part Info()
    Materials:Array of Material Interface Object References=@n50
    MaterialFeatures:Array of Atom Material Features Structures=@n50.MaterialFeatures
   ) -> (then, StructOut)
   n51 = Get Merged Mesh Part Info()
   > n52 = GetPreserveUVsForPart!(
    execute
    self
    Atom Model Part Info:"Atom Model Part Instance Info Structure (by ref)"=Get Atom Model Part Info()
    Merged Mesh Part Info:"Atom Merged Mesh Part Info Structure (by ref)"=@n51
   ) -> (then, PreserveUVs:bool)
   > Set members in Atom Merged Mesh Part Info!(execute, StructRef=@n51, bPreserveUVs:bool=@n52) -> (then, StructOut)
   n53 = GetPrimitiveGeometry!(
    execute=@n5.then_1
    self
    Atom Model Part Info:"Atom Model Part Instance Info Structure (by ref)"=Get Atom Model Part Info()
    Geometry Map:Map of Strings to Atom Primitive Geometry LODs Structures=Get Geometry Map() -> (Geometry Map:Map of Strings to Atom Primitive Geometry LODs Structures)
    Atom Model:Atom Model Object Reference=Get Model() -> (Model:Atom Model Object Reference)
   ) -> (then, Geometry LODs:Atom Primitive Geometry LODs Structure, Geometry Unique Name:String)
   > n54 = Branch(Condition=Get Generate Nanite Mesh() -> (Generate Nanite Mesh:bool))
   > Set members in Atom Merged Mesh Part Info!(
    execute
    StructRef=Get Merged Mesh Part Info()
    GeometryLODs:Atom Primitive Geometry LODs Structure=@n53
    MeshDetailLevel:EAtomMergedMeshPartDetailLevel Enum=Standard
    bPreserveUVs=False
   ) -> (then, StructOut)
   n55 = Set members in Atom Primitive Geometry LODs!(
    execute=@n54.else
    StructRef=@n53
    HiResNaniteGeomery
    HiResNaniteGeomeryMaterials:Atom Primitive Geometry Material Assignments Structure=Make Atom Primitive Geometry Material Assignments()
   ) -> (then, StructOut)
   > Set members in Atom Merged Mesh Part Info(StructRef=Get Merged Mesh Part Info(), GeometryLODs=@n55, MeshDetailLevel=Standard, bPreserveUVs=False)
  ]}
 Get Primitive Geometry Container Collision {nodes [
   n1 = Get Primitive Geometry Container Collision!() -> (then, Part Info:Atom Model Part Instance Info Structure) {is=title K2Node_FunctionEntry}
   n2 = Break Atom Processor Geometry Style Settings(AtomProcessorGeometryStyleSettings=Get Geometry Style Settings For LOD(LOD=0) {is=title K2Node_CallFunction}) -> (
    ShouldBevel_8_#6=~
    BevelWidth_11_#7=~
    BevelAngleThreshold_13_#8=~
    UseImprovedBevelAlgorithm_37_#3=~
    OnlyBevelBlock-LikePrimitives_39_#4=~
    FakeBevel_23_#10=~
    FakeUnderside_25_#11=~
    AnalyticFakeUnderside_32_#12=~
    ForceFlatCap_35_#13=~
    FakeTechnicHole_41_#2=~
    AnalyticCommonParts_27_#14=~
   )
   n3 = GetGeometryContainerForBestExportStyle(
    self=Break Atom Model Part Instance Info!(AtomModelPartInstanceInfo=@n1) -> (Primitive)
    GeometryStylePriorityList=@n2
    LODIndex=@n2.PreferredGeometryLOD_30_#5
   )
   > Return Node!(
    execute
    Success:bool=AND Boolean(
     A=Get Prefer Primitive Geometry Container Collision() -> (Prefer Primitive Geometry Container Collision:bool)
     B=HasCollisionGeometry!(self:Atom Primitive Geometry Container Object Reference=@n3) -> (ReturnValue:bool)
    )
    Geometry Container:Atom Primitive Geometry Container Object Reference=@n3
   )
  ]}
 Get Unique Name for Common Part Geometry LODs {nodes [
   n1 = Get Unique Name for Common Part Geometry LODs!() -> (then, PartInfo:Atom Model Part Instance Info Structure, MergedMeshPartInfo:Atom Merged Mesh Part Info Structure, CommonPartIndex:int) {is=title K2Node_FunctionEntry}
   n2 = Break Atom Model Part Instance Info!(AtomModelPartInstanceInfo=@n1) -> (PartGuid:Atom Model Part Guid Structure, Primitive, CommonParts:Array of Atom Common Part And Transform Structures)
   > n3 = PrimitiveHasConsistentCommonPartStyles!(execute, Primitive:Atom Primitive Object Reference=@n2.Primitive) -> (then, ReturnValue:bool)
   > n4 = Branch(Condition=@n3)
   n5 = Enum to String!(Enumerator:EAtomCommonPartType Enum=Break Atom Common Part and Transform(AtomCommonPartAndTransform="Get (a ref)"(Array:Array of Atom Common Part And Transform Structures=@n2.CommonParts, Dimension 1=@n1.CommonPartIndex) -> (Output:"Atom Common Part And Transform Structure (by ref)"))) -> (ReturnValue:String)
   > Return Node!(execute, Name:String=@n5)
   n6 = Get CommonPartIndex() -> (CommonPartIndex:int=0)
   Return Node(execute=@n4.else, Name=Append!(A=@n5, B="To String (Integer)"(InInt=@n6), C:String=To String(InModelPartGuid=@n2)) -> (ReturnValue))
  ]}
 GetAtomModel {nodes [
   GetAtomModel() {is=title K2Node_FunctionEntry}
   > Return Node!(execute, Atom Model:Atom Model Object Reference=Get Model())
  ]}
 GetBlueprintContainerName {nodes [
   n1 = GetBlueprintContainerName!() -> (then, Model:Atom Model Object Reference, Parts Collection:"Atom Model Parts Collection Structure (by ref)") {is=title K2Node_FunctionEntry}
   > Return Node!(execute, Result:String=Append!(A=BP_, B=Break Atom Model Parts Collection(AtomModelPartsCollection=@n1.Parts Collection)) -> (ReturnValue))
  ]}
 GetBlueprintContainerPath {nodes [
   n1 = GetBlueprintContainerPath!() -> (then, Model:Atom Model Object Reference, Parts Collection:"Atom Model Parts Collection Structure (by ref)") {is=title K2Node_FunctionEntry}
   > n2 = Branch(Condition=GetIsProcessingPrimitive())
   > n3 = GetImporterSettings!(execute) -> (then, ReturnValue:Atom Model Importer Settings Object Reference)
   > Return Node(Result=Break Directory Path!(DirectoryPath:"Directory Path Structure (by ref)"=Get ProcessedPrimitiveContentFolder!(self:Atom Model Importer Settings Object Reference=@n3) -> (ProcessedPrimitiveContentFolder:Directory Path Structure)) -> (Path:String) {is=title K2Node_BreakStruct})
   Return Node(execute=@n2.else, Result=GetModelPath!(self:Atom Model Object Reference=@n1) -> (ReturnValue:String))
  ]}
 GetCollisionGeometryForCommonPart {nodes [
   GetCollisionGeometryForCommonPart!() -> (then, Geometry:Atom Primitive Geometry Object Reference) {is=title K2Node_FunctionEntry}
   > Return Node!(
    execute
    Collision Geometry:Atom Primitive Collision Geometry Structure=Make Atom Primitive Collision Geometry!() -> (AtomPrimitiveCollisionGeometry:Atom Primitive Collision Geometry Structure) {is=title K2Node_MakeStruct}
   )
  ]}
 GetColorChangeMergedMeshPartInfos {nodes [
   GetColorChangeMergedMeshPartInfos!() -> (then, Part Info:"Atom Model Part Instance Info Structure (by ref)", MergedMeshPartInfo:"Atom Merged Mesh Part Info Structure (by ref)") {is=title K2Node_FunctionEntry}
   n1 = Get ColorChange Geometry LODs() -> (ColorChange Geometry LODs:Array of Atom Primitive Geometry LODs Structures)
   > Clear(TargetArray:Array of Atom Primitive Geometry LODs Structures=@n1)
   > SplitGeometryLODsByPolygonGroup!(
    execute
    self
    Geometry LODs:"Atom Primitive Geometry LODs Structure (by ref)"=Break Atom Merged Mesh Part Info!(AtomMergedMeshPartInfo=Get MergedMeshPartInfo() -> (MergedMeshPartInfo:"Atom Merged Mesh Part Info Structure (by ref)")) -> (
     GeometryLODs
     bHasConstantOverrideVertexColor:bool=~
     PreventMergingMaterials
     MaterialFeatures:Array of Atom Material Features Structures
     bAllowSimplification=~
     bIsOccluder:bool=~
     FilterLODLevel:int
    )
    Output Geomery LODs Array:Array of Atom Primitive Geometry LODs Structures=@n1
   ) -> (then)
   > n2 = For Each Loop(Array:Array of Atom Primitive Geometry LODs Structures=Get ColorChange Geometry LODs()) -> (Array Element:Atom Primitive Geometry LODs Structure)
   Return Node!(
    execute=@n2.Completed
    MergedMeshPartInfos:Array of Atom Merged Mesh Part Info Structures=Get OutMergedMeshPartInfos() -> (OutMergedMeshPartInfos:Array of Atom Merged Mesh Part Info Structures)
   )
   n3 = Break Atom Primitive Geometry LODs!(AtomPrimitiveGeometryLODs=@n2) -> (GeometryLODs, HiResNaniteGeomery, DetailedCollision)
   > n4 = Branch(Condition=Is Valid Index!(TargetArray:Array of Atom Primitive Geometry Object References=@n3, IndexToTest:int) -> (ReturnValue:bool))
   > n5 = Set CurrentGeometry(CurrentGeometry:Atom Primitive Geometry Object Reference="Get (a copy)"(Array:Array of Atom Primitive Geometry Object References=@n3) -> (Output:Atom Primitive Geometry Object Reference)) -> (Output_Get:Atom Primitive Geometry Object Reference)
   n6 = GetMaterialsForPart(
    execute=[@n10.then, @n5.then]
    Part Info=Get Part Info() -> (Part Info:"Atom Model Part Instance Info Structure (by ref)")
    MergedMeshPartInfo=Get MergedMeshPartInfo()
   )
   n7 = Break Atom Model Part Instance Info!(AtomModelPartInstanceInfo=Get Part Info()) -> (Colors:Array of Atom Model Part Color Info Structures)
   > n8 = GetVertexColorForPart(
    Part Info=Get Part Info()
    Color Change Index="% (Integer)"(A=@n2.Array Index, B=Length(TargetArray:Array of Atom Model Part Color Info Structures=@n7)) -> (ReturnValue=0)
   )
   > n9 = Set members in Atom Merged Mesh Part Info!(
    execute
    StructRef=Copy(Item:Atom Merged Mesh Part Info Structure=Get MergedMeshPartInfo()) -> (ReturnValue:Atom Merged Mesh Part Info Structure)
    GeometryLODs=@n2
    VertexColor:Color Structure=@n8
    Materials:Array of Material Interface Object References=@n6
    MaterialFeatures:Array of Atom Material Features Structures=@n6.MaterialFeatures
   ) -> (then, StructOut)
   > Add(
    TargetArray:Array of Atom Merged Mesh Part Info Structures=Get OutMergedMeshPartInfos()
    NewItem:"Atom Merged Mesh Part Info Structure (by ref)"=@n9
   ) {is=title K2Node_CallArrayFunction}
   n10 = Set CurrentGeometry(execute=@n4.else, CurrentGeometry=@n3.HiResNaniteGeomery)
  ]}
 GetCommonPartColorChangeIndex {nodes [
   GetCommonPartColorChangeIndex!() -> (then, Part Info:Atom Model Part Instance Info Structure, Common Part Index:int) {is=title K2Node_FunctionEntry}
   > Return Node!(execute, Color Change Index:int=0)
  ]}
 GetCommonPartGeometry {nodes [
   n1 = GetCommonPartGeometry!() -> (
    then
    PartInfo:"Atom Model Part Instance Info Structure (by ref)"
    MergedMeshPartInfo:Atom Merged Mesh Part Info Structure
    CommonPartIndex:"Integer (by ref)"
   ) {is=title K2Node_FunctionEntry}
   n2 = "Get (a ref)"(
    Array=Break Atom Model Part Instance Info!(AtomModelPartInstanceInfo=@n1) -> (CommonParts:Array of Atom Common Part And Transform Structures)
    Dimension 1=@n1.CommonPartIndex
   )
   > n3 = GetGeometryLODsForCommonPart!(
    execute
    self
    PartInfo:Atom Model Part Instance Info Structure=@n1
    MergedMeshPartInfo:Atom Merged Mesh Part Info Structure=@n1.MergedMeshPartInfo
    Common Part Type:EAtomCommonPartType Enum=Break Atom Common Part and Transform(AtomCommonPartAndTransform=@n2)
   ) -> (then, Geometry LODs:Atom Primitive Geometry LODs Structure)
   > n4 = PrimitiveHasConsistentCommonPartStyles(Primitive=Break Atom Model Part Instance Info!(AtomModelPartInstanceInfo=@n1) -> (Primitive:Atom Primitive Object Reference))
   > n5 = Branch(Condition=@n4)
   > Return Node!(execute, GeometryLODs:Atom Primitive Geometry LODs Structure=@n3)
   n6 = Sequence!(execute=@n5.else) -> (then_0, then_1)
   > n7 = Branch(Condition=IsCommonPartUsedInGeometryStyle!(
     PartInfo:"Atom Model Part Instance Info Structure (by ref)"=Get PartInfo() -> (PartInfo:"Atom Model Part Instance Info Structure (by ref)")
     GeometryStyleSettings:"Atom Processor Geometry Style Settings Structure (by ref)"=Get Nanite Geometry Style Settings()
     CommonPart:Atom Common Part And Transform Structure=@n2
    ) -> (CommonPartUsed:bool))
   n8 = CreateEmptyAtomGeometry!(execute=@n7.else, NumberOfUVChannels:int=1) -> (then, ReturnValue:Atom Primitive Geometry Object Reference)
   > Set members in Atom Primitive Geometry LODs!(execute, StructRef=@n3, HiResNaniteGeomery=@n8) -> (then, StructOut)
   Set CommonPartLODs(
    execute=@n6.then_1
    CommonPartLODs:Array of Atom Primitive Geometry Object References=Break Atom Primitive Geometry LODs!(AtomPrimitiveGeometryLODs=@n3) -> (
     GeometryLODs
     GeometryLODsMaterials:Array of Atom Primitive Geometry Material Assignments Structures
     HiResNaniteGeomeryMaterials:Atom Primitive Geometry Material Assignments Structure
     DetailedCollision
    )
   ) -> (Output_Get:Array of Atom Primitive Geometry Object References)
   > n9 = For Each Loop(Array:Array of Atom Processor Geometry Style Settings Structures=Get LOD Geometry Style Settings()) -> (Array Element:Atom Processor Geometry Style Settings Structure)
   n10 = Set members in Atom Primitive Geometry LODs!(
    execute=@n9.Completed
    StructRef=@n3
    GeometryLODs:Array of Atom Primitive Geometry Object References=Get CommonPartLODs() -> (CommonPartLODs:Array of Atom Primitive Geometry Object References)
   ) -> (then, StructOut)
   > Return Node(GeometryLODs=@n10)
   n11 = Branch(execute=@n9.LoopBody, Condition=IsCommonPartUsedInGeometryStyle(PartInfo=Get PartInfo(), GeometryStyleSettings=@n9, CommonPart=@n2))
   n12 = CreateEmptyAtomGeometry(execute=@n11.else, NumberOfUVChannels=1)
   n13 = Get CommonPartLODs()
   n14 = For Loop(LastIndex=Get Geometry Style Settings Last LOD!(Part Info:Atom Model Part Instance Info Structure=Get PartInfo()) -> (Num Style Settings:int=0) {is=title K2Node_CallFunction})
   Set Array Elem(
    execute=@n12.then
    TargetArray:Array of Atom Primitive Geometry Object References=@n13
    Index="Min (Integer)"(A=Last Index(TargetArray:Array of Atom Primitive Geometry Object References=@n13) -> (ReturnValue=0), B=@n14) -> (ReturnValue=0)
    Item:"Atom Primitive Geometry Object Reference (by ref)"=@n12
    bSizeToFit=false
   )
   Get Geometry Style Settings For LOD(LOD=@n14)
  ]}
 GetCommonPartMeshSuffix {nodes [
   GetCommonPartMeshSuffix() {is=title K2Node_FunctionEntry}
   > Return Node!(execute, Common Part Mesh Suffix:String=CP)
  ]}
 GetCommonPartsGeometry {nodes [
   n1 = GetCommonPartsGeometry!() -> (
    then
    Part Info:"Atom Model Part Instance Info Structure (by ref)"
    MergedMeshPartInfo:Atom Merged Mesh Part Info Structure
    Common Part To Geometry LODs:Map of Strings to Atom Primitive Geometry LODs Structures
   ) {is=title K2Node_FunctionEntry}
   > n2 = For Each Loop(Array:Array of Atom Common Part And Transform Structures=Break Atom Model Part Instance Info!(AtomModelPartInstanceInfo=@n1) -> (Primitive, Group:String, ParentGroup:String, CommonParts:Array of Atom Common Part And Transform Structures).CommonParts) -> (Array Element:Atom Common Part And Transform Structure)
   > n3 = Get Unique Name for Common Part Geometry LODs(
    PartInfo=Get Part Info()
    MergedMeshPartInfo=Get MergedMeshPartInfo() -> (MergedMeshPartInfo:Atom Merged Mesh Part Info Structure)
    CommonPartIndex=@n2.Array Index
   ) {is=title K2Node_CallFunction}
   > n4 = Set CommonPartUniqueName(CommonPartUniqueName:String=@n3) -> (Output_Get:String)
   > n5 = Branch(Condition=Find(
     TargetMap=Get Common Part To Geometry LODs() -> (Common Part To Geometry LODs:Map of Strings to Atom Primitive Geometry LODs Structures)
     Key=@n4
    ).ReturnValue)
   n6 = GetCommonPartGeometry!(
    execute=@n5.else
    self
    PartInfo:"Atom Model Part Instance Info Structure (by ref)"=Get Part Info()
    MergedMeshPartInfo:Atom Merged Mesh Part Info Structure=Get MergedMeshPartInfo()
    CommonPartIndex:"Integer (by ref)"=@n2.Array Index
   ) -> (then, GeometryLODs:Atom Primitive Geometry LODs Structure) {is=title K2Node_CallFunction}
   > Set New Common Part LODs(New Common Part LODs:Atom Primitive Geometry LODs Structure=@n6) -> (Output_Get:Atom Primitive Geometry LODs Structure)
   n7 = ShouldSplitTransparentPart(MergedMeshPartInfo=Get MergedMeshPartInfo())
   > n8 = Branch(Condition=AND Boolean(A=Get Generate Nanite Mesh(), B=NOT Boolean(A=@n7)))
   n9 = Get New Common Part LODs() -> (New Common Part LODs:Atom Primitive Geometry LODs Structure)
   n10 = GetCollisionGeometryForCommonPart!(
    execute=[@n8.then, @n12.then]
    self
    Geometry:Atom Primitive Geometry Object Reference="Get (a copy)"(Array=Break Atom Primitive Geometry LODs!(AtomPrimitiveGeometryLODs=@n9) -> (
      GeometryLODs
      GeometryLODsMaterials
      HiResNaniteGeomery:Atom Primitive Geometry Object Reference
      HiResNaniteGeomeryMaterials
      Collision:Atom Primitive Collision Geometry Structure
      DetailedCollision
     ))
   ) -> (then, Collision Geometry:Atom Primitive Collision Geometry Structure) {is=title K2Node_CallFunction}
   > n11 = Set members in Atom Primitive Geometry LODs!(execute, StructRef=Get New Common Part LODs(), Collision:Atom Primitive Collision Geometry Structure=@n10) -> (then, StructOut)
   > Add(
    TargetMap:Map of Strings to Atom Primitive Geometry LODs Structures=Get Common Part To Geometry LODs()
    Key:"String (by ref)"=Get CommonPartUniqueName() -> (CommonPartUniqueName:String)
    Value:"Atom Primitive Geometry LODs Structure (by ref)"=@n11
   ) {is=title K2Node_CallFunction}
   n12 = Set members in Atom Primitive Geometry LODs!(execute=@n8.else, StructRef=Get New Common Part LODs(), HiResNaniteGeomery:Atom Primitive Geometry Object Reference) -> (then, StructOut)
  ]}
 GetCustomCommonPartMeshPath {nodes [
   GetCustomCommonPartMeshPath() {is=title K2Node_FunctionEntry}
   > Return Node!(execute, Common Part Mesh Path:String)
  ]}
 GetDecorationMaterial {nodes [
   n1 = GetDecorationMaterial!() -> (
    then
    Base Material:Material Interface Object Reference
    Decoration Infos:Array of Atom Model Part Decoration Info Structures
    New Material Path:String
    ColorInfo:Atom Model Part Color Info Structure
   ) {is=title K2Node_FunctionEntry}
   > n2 = Sequence!(execute) -> (then_0, then_1, then_2:Exec)
   > Set Material Name(Material Name:String=GetObjectName(Object=@n1)) -> (Output_Get:String)
   > n3 = For Each Loop(Array:Array of Atom Model Part Decoration Info Structures=@n1.Decoration Infos) -> (Array Element:Atom Model Part Decoration Info Structure)
   n4 = Break Atom Model Part Decoration Info!(AtomModelPartDecorationInfo=@n3) -> (Texture)
   > Set Material Name(Material Name=Append!(A=Get Material Name() -> (Material Name:String), B=_, C:String=GetObjectName(Object=@n4)) -> (ReturnValue))
   n5 = CreateMaterialInstanceFromBaseMaterial!(
    execute=@n2.then_1
    BaseMaterial:Material Interface Object Reference=@n1
    Subpath:String=@n1.New Material Path
    MaterialInstanceName:String=Get Material Name()
    bOverwriteExisting:bool=false
   ) -> (then, bAlreadyExists:bool, ReturnValue:Material Interface Object Reference)
   > n6 = Cast To MaterialInstanceConstant(Object=@n5.ReturnValue)
   > Set Material(Material:Material Instance Constant Object Reference=@n6) -> (Output_Get:Material Instance Constant Object Reference)
   > n7 = Branch(Condition=@n5)
   SetMaterialInstanceStaticSwitchParameterValue!(
    execute=@n7.else
    Instance:Material Instance Constant Object Reference=Get Material() -> (Material:Material Instance Constant Object Reference)
    ParameterName:Name=Uses Decorators
    Value:bool=true
    Association:EMaterialParameterAssociation Enum=GlobalParameter
    bUpdateMaterialInstance:bool=true
   ) -> (then, ReturnValue:bool)
   > n8 = For Each Loop(Array=@n1.Decoration Infos)
   > n9 = Branch(Condition="Equal (Integer)"(A=@n8.Array Index))
   n10 = Break Atom Model Part Decoration Info!(AtomModelPartDecorationInfo=@n8) -> (Material:Material Interface Object Reference, Texture, PrimitiveSurfaceName:String, PrimitiveSurfaceIndex:int)
   n11 = Get Material()
   > SetDecorationMaterialParameters!(
    execute
    self
    Material Instance:Material Instance Constant Object Reference=@n11
    Decoration Enable Param Name:String=Decorator ID 1
    Texture Param Name:String=Decorator 1
    UDIM Param Name:String=Decorator 1 UDIM
    Texture:Texture Object Reference=@n10.Texture
    UDIM:int=@n10.PrimitiveSurfaceIndex
   ) -> (then)
   n12 = Branch(execute=@n9.else, Condition="Equal (Integer)"(A=@n8.Array Index, B=1))
   > SetDecorationMaterialParameters(
    Material Instance=@n11
    Decoration Enable Param Name=Decorator ID 2
    Texture Param Name=Decorator 2
    UDIM Param Name=Decorator 2 UDIM
    Texture=@n10.Texture
    UDIM=@n10.PrimitiveSurfaceIndex
   )
   n13 = Branch(execute=@n12.else, Condition="Equal (Integer)"(A=@n8.Array Index, B=2))
   > SetDecorationMaterialParameters(
    Material Instance=@n11
    Decoration Enable Param Name=Decorator ID 3
    Texture Param Name=Decorator 3
    UDIM Param Name=Decorator 3 UDIM
    Texture=@n10.Texture
    UDIM=@n10.PrimitiveSurfaceIndex
   )
   Branch(execute=@n13.else, Condition="Equal (Integer)"(A=@n8.Array Index, B=3))
   > SetDecorationMaterialParameters(
    Material Instance=@n11
    Decoration Enable Param Name=Decorator ID 4
    Texture Param Name=Decorator 4
    UDIM Param Name=Decorator 4 UDIM
    Texture=@n10.Texture
    UDIM=@n10.PrimitiveSurfaceIndex
   )
   Return Node!(execute=@n2.then_2, Decoration Material:Material Interface Object Reference=Get Material())
  ]}
 GetDecorationMaterialPath {nodes [
   n1 = GetDecorationMaterialPath!() -> (then, Model Path:String) {is=title K2Node_FunctionEntry}
   > n2 = GetImporterSettings()
   > n3 = For Each Loop(Array:Array of Primitive Style Import Settings Structures=Get PrimitiveImportStyles!(self:Atom Model Importer Settings Object Reference=@n2) -> (PrimitiveImportStyles:Array of Primitive Style Import Settings Structures)) -> (Array Element:Primitive Style Import Settings Structure)
   Return Node!(execute=@n3.Completed, Material Path:String=Append!(A=@n1, B=/Materials) -> (ReturnValue))
   n4 = Break Primitive Style Import Settings!(PrimitiveStyleImportSettings:"Primitive Style Import Settings Structure (by ref)"=@n3) -> (
    Name:Name
    MeshTargetContentFolder:Directory Path Structure
    CombinedMeshTargetContentFolder:Directory Path Structure
    MaterialTargetContentFolder:Directory Path Structure
    TextureTargetContentFolder:Directory Path Structure
    StandardMaterial:Material Interface Soft Object Reference
    TransparentMaterial:Material Interface Soft Object Reference
    GlitterMaterial:Material Interface Soft Object Reference
    OpalescentMaterial:Material Interface Soft Object Reference
    StandardMaterialWithVertexColors:Material Interface Soft Object Reference
    TransparentMaterialWithVertexColors:Material Interface Soft Object Reference
    GlitterMaterialWithVertexColors:Material Interface Soft Object Reference
    OpalescentMaterialWithVertexColors:Material Interface Soft Object Reference
    StandardIndexColorMaterial:Material Interface Soft Object Reference
    MetalIndexColorMaterial:Material Interface Soft Object Reference
    TransparentIndexColorMaterial:Material Interface Soft Object Reference
    GlitterIndexColorMaterial:Material Interface Soft Object Reference
    OpalescentIndexColorMaterial:Material Interface Soft Object Reference
   ) {is=title K2Node_BreakStruct}
   n5 = Break Atom Processor Geometry Style Settings!(AtomProcessorGeometryStyleSettings=Get Geometry Style Settings For LOD(LOD=0)) -> (GeometryStylePreferences_4_#1, FakeTechnicHole_41_#2=~)
   > Branch(Condition="Equal (Name)"(A:"String (by ref)"="Get (a ref)"(Array:Array of Strings=@n5, Dimension 1=0) -> (Output:"String (by ref)"), B=@n4))
   > Return Node(Material Path=Break Directory Path(DirectoryPath=@n4.MaterialTargetContentFolder))
  ]}
 GetDefaultMaterialForType {nodes [
   n1 = GetDefaultMaterialForType!() -> (then, "Atom Material  Type":EAtomMaterialType Enum=Standard, Part Info:"Atom Model Part Instance Info Structure (by ref)") {is=title K2Node_FunctionEntry}
   > n2 = Switch on EAtomMaterialType!(execute, Selection:EAtomMaterialType Enum=@n1) -> (None:Exec, Standard:Exec, Transparent:Exec, Glitter:Exec, Opalescent:Exec, Metallic:Exec) {is=title K2Node_SwitchEnum}
   Return Node!(execute=@n2.Glitter, Material:Material Interface Object Reference=/AtomFramework/Materials/Game_Qual/MI_Primitive_GlitVC)
   Return Node(execute=@n2.Metallic, Material=/AtomFramework/Materials/Game_Qual/MI_Primitive_MetalVC)
   Return Node(execute=@n2.Opalescent, Material=/AtomFramework/Materials/Game_Qual/MI_Primitive_OpalVC)
   Return Node(execute=@n2.Standard, Material=/AtomFramework/Materials/Game_Qual/MI_Primitive_StandardVC)
   Return Node(execute=@n2.Transparent, Material=/AtomFramework/Materials/Game_Qual/MI_Primitive_TransVC)
  ]}
 GetDefaultMeshMergingSettings {nodes [
   GetDefaultMeshMergingSettings() {is=title K2Node_FunctionEntry}
   > Return Node!(
    execute
    Regular Mesh Merging Settings:Atom Mesh Merging Settings Structure=Get Mesh Merging Settings() -> (Mesh Merging Settings:Atom Mesh Merging Settings Structure)
   )
  ]}
 GetFlexGeometryWithStyleSettings {nodes [
   n1 = GetFlexGeometryWithStyleSettings!() -> (
    then
    Part Instance Info:"Atom Model Part Instance Info Structure (by ref)"
    Geometry Style Settings:"Atom Processor Geometry Style Settings Structure (by ref)"
   ) {is=title K2Node_FunctionEntry}
   n2 = Break Atom Model Part Instance Info!(AtomModelPartInstanceInfo=@n1) -> (Primitive, Transforms:Array of Transforms)
   > Set FlexRootTransformInverse(FlexRootTransformInverse:Transform=InvertTransform(T="Get (a copy)"(Array:Array of Transforms=@n2.Transforms) -> (Output:Transform))) -> (Output_Get:Transform)
   > n3 = For Each Loop(Array:Array of Transforms=@n2.Transforms) -> (Array Element:Transform)
   n4 = Break Atom Processor Geometry Style Settings!(AtomProcessorGeometryStyleSettings=@n1.Geometry Style Settings) -> (
    GeometryStylePreferences_4_#1
    PreferredGeometryLOD_30_#5:int
    UseImprovedBevelAlgorithm_37_#3:bool=~
    OnlyBevelBlock-LikePrimitives_39_#4:bool=~
    FakeTechnicHole_41_#2=~
   )
   n5 = Set Geometry(
    execute=@n3.Completed
    Geometry:Atom Primitive Geometry Object Reference=GetGeometrySection!(
     self:Atom Primitive Geometry Container Object Reference=GetGeometryContainerForBestExportStyle(self=@n2, GeometryStylePriorityList=@n4, LODIndex=@n4.PreferredGeometryLOD_30_#5)
     PrimitiveGeometrySectionFlag:Bitmask Integer=35
    ) -> (ReturnValue:Atom Primitive Geometry Object Reference)
   ) -> (Output_Get:Atom Primitive Geometry Object Reference)
   > n6 = BakeScale!(execute, self:Atom Primitive Geometry Object Reference=@n5, Scale:float=GetDefaultPrimitiveScale()) -> (then, ReturnValue:Atom Primitive Geometry Object Reference)
   > n7 = BakeTransforms!(
    execute
    self:Atom Primitive Geometry Object Reference=@n6
    Transforms:Array of Transform 3f Structures=Get Flex Transforms() -> (Flex Transforms:Array of Transforms)
   ) -> (then, ReturnValue:Atom Primitive Geometry Object Reference)
   > Return Node!(execute, Geometry Result:Atom Primitive Geometry Object Reference=@n7)
   Add(
    execute=@n3.LoopBody
    TargetArray:Array of Transforms=Get Flex Transforms()
    NewItem:"Transform (by ref)"=ComposeTransforms(A=@n3, B=Get FlexRootTransformInverse() -> (FlexRootTransformInverse:Transform))
   ) {is=title K2Node_CallArrayFunction}
   Get Part Instance Info() -> (Part Instance Info:"Atom Model Part Instance Info Structure (by ref)")
  ]}
 GetGeometricBevelSettings {nodes [
   n1 = GetGeometricBevelSettings!() -> (
    then
    Part Instance Info:"Atom Model Part Instance Info Structure (by ref)"
    Geometry Style Settings:"Atom Processor Geometry Style Settings Structure (by ref)"
   ) {is=title K2Node_FunctionEntry}
   n2 = Break Atom Processor Geometry Style Settings!(AtomProcessorGeometryStyleSettings=@n1.Geometry Style Settings) -> (
    BevelWidth_11_#7:double=~
    BevelAngleThreshold_13_#8:double=~
    OnlyBevelBlock-LikePrimitives_39_#4=~
    FakeTechnicHole_41_#2=~
   )
   > Return Node!(
    execute
    AtomGeometricBevelSettings:Atom Geometric Bevel Settings Structure=Make Atom Geometric Bevel Settings!(
     Technique:EAtomGeometricBevelTechnique Enum=NewAlgorithm
     BevelWidth:float=@n2
     BevelAngleThreshold:float=@n2.BevelAngleThreshold_13_#8
     bOnlyBevelBlockLikeGeometry:bool=@n2.OnlyBevelBlock-LikePrimitives_39_#4
     bIgnoreUVSeams:bool=True
     PartId:int=Break Atom Model Part Instance Info!(AtomModelPartInstanceInfo=@n1) -> (
      PartGuid:Atom Model Part Guid Structure
      PartId:int
      Primitive
      Transforms
      Decorations:Array of Atom Model Part Decoration Info Structures
      Colors:Array of Atom Model Part Color Info Structures
      SelectionSets:Array of Strings
      Group:String
      ParentGroup:String
      CommonParts:Array of Atom Common Part And Transform Structures
      bIsUndersideVisible:bool=~
     ).PartId
    ) -> (AtomGeometricBevelSettings:Atom Geometric Bevel Settings Structure) {is=title K2Node_MakeStruct}
   )
  ]}
 GetGeometryLODsForCommonPart {nodes [
   n1 = GetGeometryLODsForCommonPart!() -> (
    then
    PartInfo:Atom Model Part Instance Info Structure
    MergedMeshPartInfo:Atom Merged Mesh Part Info Structure
    Common Part Type:EAtomCommonPartType Enum=None
   ) {is=title K2Node_FunctionEntry}
   > n2 = Get Geometry Material Assignments For LOD!(execute, self, PartInfo:Atom Model Part Instance Info Structure=@n1, LOD:int=0, Color Change Index:int=0) -> (then, Material Assignments:Atom Primitive Geometry Material Assignments Structure) {is=title K2Node_CallFunction}
   n3 = Break Atom Primitive Geometry Material Assignments(AtomPrimitiveGeometryMaterialAssignments=@n2)
   > n4 = Set members in Atom Material Features!(
    execute
    StructRef="Get (a ref)"(Array:Array of Atom Material Features Structures=@n3.MaterialFeatures, Dimension 1=0) -> (Output:"Atom Material Features Structure (by ref)")
    UDIMDecorationMap=Make Map!() -> (Map:Map of Integers to Texture Object References) {is=title K2Node_MakeMap}
    NormalMap:Texture Object Reference
    AlphaMask:Texture Object Reference
   ) -> (then, StructOut)
   n5 = "Get (a ref)"(Array:Array of Strings=@n3, Dimension 1=0) -> (Output:"String (by ref)")
   > Set Default Common Part Material(Default Common Part Material:Atom Primitive Geometry Material Assignments Structure=Make Atom Primitive Geometry Material Assignments(
     MaterialSlotName=Make Array("[0]":String=@n5) -> (Array:Array of Strings)
     MaterialFeatures=Make Array("[0]":Atom Material Features Structure=@n4) -> (Array:Array of Atom Material Features Structures)
    )) -> (Output_Get:Atom Primitive Geometry Material Assignments Structure)
   > n6 = Sequence!(execute) -> (then_0, then_1)
   > Set UsedCommonPartStyle(UsedCommonPartStyle:String=Get Common Part Geometry Style!(Common Part Type:EAtomCommonPartType Enum=Get Common Part Type() -> (Common Part Type=None), bIsNanite:bool=false) -> (Geometry Style:String) {is=title K2Node_CallFunction}) -> (Output_Get:String)
   n7 = GetCustomCommonPartMeshPath() {is=title K2Node_CallFunction}
   n8 = Get Common Part Type() -> (Common Part Type=None)
   n9 = Get UsedCommonPartStyle() -> (UsedCommonPartStyle:String)
   > n10 = For Loop(LastIndex=int - int(A=GetNumberOfCommonPartLODs!(ExportStyleName:String=@n9, CommonPartType:EAtomCommonPartType Enum=@n8, CommonPartsMeshPath:String=@n7) -> (ReturnValue:int=0)))
   > n11 = CreateAtomGeometryFromCommonPart!(
    execute
    ExportStyleName:String=Get UsedCommonPartStyle()
    CommonPartType:EAtomCommonPartType Enum=Get Common Part Type() -> (Common Part Type=None)
    LodIndex:int=@n10
    CommonPartsMeshPath:String=GetCustomCommonPartMeshPath()
   ) -> (then, ReturnValue:Atom Primitive Geometry Object Reference)
   > n12 = BakeScale(self=@n11, Scale=GetDefaultPrimitiveScale())
   > n13 = SetMaterialName!(execute, self:Atom Primitive Geometry Object Reference=@n12, Name:String=Default, PolygonGroupIndex:int=0) -> (then, ReturnValue:Atom Primitive Geometry Object Reference)
   > Add(
    TargetArray:Array of Atom Primitive Geometry Object References=Get Common Part Geometry LODs() -> (Common Part Geometry LODs:Array of Atom Primitive Geometry Object References)
    NewItem:"Atom Primitive Geometry Object Reference (by ref)"=@n13
   )
   > Add(
    TargetArray:Array of Atom Primitive Geometry Material Assignments Structures=Get Common Part Geometry Materials() -> (Common Part Geometry Materials:Array of Atom Primitive Geometry Material Assignments Structures)
    NewItem:"Atom Primitive Geometry Material Assignments Structure (by ref)"=Get Default Common Part Material() -> (Default Common Part Material:Atom Primitive Geometry Material Assignments Structure)
   )
   n14 = Get Common Part Geometry LODs()
   n15 = DuplicateGeometry!(
    execute=@n6.then_1
    self:Atom Primitive Geometry Object Reference="Get (a copy)"(Array:Array of Atom Primitive Geometry Object References=@n14) -> (Output:Atom Primitive Geometry Object Reference)
   ) -> (then, ReturnValue:Atom Primitive Geometry Object Reference)
   > Return Node!(
    execute
    Geometry LODs:Atom Primitive Geometry LODs Structure=Make Atom Primitive Geometry LODs(
     GeometryLODs=@n14
     GeometryLODsMaterials=Get Common Part Geometry Materials()
     HiResNaniteGeomery=@n15
     HiResNaniteGeomeryMaterials=Get Default Common Part Material()
     Collision=(AggGeom=(SphereElems=,BoxElems=,SphylElems=,ConvexElems=,TaperedCapsuleElems=,LevelSetElems=,SkinnedLevelSetElems=,MLLevelSetElems=,SkinnedTriangleMeshElems=))
     DetailedCollision=(AggGeom=(SphereElems=,BoxElems=,SphylElems=,ConvexElems=,TaperedCapsuleElems=,LevelSetElems=,SkinnedLevelSetElems=,MLLevelSetElems=,SkinnedTriangleMeshElems=))
    )
   )
  ]}
 GetGeometryLODsForPart {nodes [
   n1 = GetGeometryLODsForPart!() -> (then, Part Info:"Atom Model Part Instance Info Structure (by ref)") {is=title K2Node_FunctionEntry}
   > n2 = Sequence!(execute) -> (then_0, then_1, then_2:Exec)
   > Set Last LOD(Last LOD:int=Get Geometry Style Settings Last LOD(Part Info=@n1) -> (Num Style Settings=0)) -> (Output_Get:int=0)
   > n3 = Branch(Condition=IsFlexElement(self=Break Atom Model Part Instance Info!(AtomModelPartInstanceInfo=@n1) -> (Primitive)))
   > n4 = Sequence!(execute) -> (then_0, then_1)
   > n5 = For Loop(LastIndex=Get Last LOD() -> (Last LOD:int=0))
   > n6 = GetFlexGeometryWithStyleSettings!(
    execute
    self
    Part Instance Info:"Atom Model Part Instance Info Structure (by ref)"=Get Part Info()
    Geometry Style Settings:"Atom Processor Geometry Style Settings Structure (by ref)"=Get Geometry Style Settings For LOD(LOD=@n5)
   ) -> (then, Geometry Result:Atom Primitive Geometry Object Reference) {is=title K2Node_CallFunction}
   > Add(
    TargetArray:Array of Atom Primitive Geometry Object References=Break Atom Primitive Geometry LODs(AtomPrimitiveGeometryLODs=Get AtomPrimitiveGeometryLODs() -> (AtomPrimitiveGeometryLODs:Atom Primitive Geometry LODs Structure))
    NewItem:"Atom Primitive Geometry Object Reference (by ref)"=@n6
   )
   n7 = GetFlexGeometryWithStyleSettings(execute=@n4.then_1, Part Instance Info=Get Part Info(), Geometry Style Settings=Get Nanite Geometry Style Settings())
   > Set members in Atom Primitive Geometry LODs(StructRef=Get AtomPrimitiveGeometryLODs(), HiResNaniteGeomery=@n7)
   n8 = Sequence(execute=@n3.else)
   > n9 = For Loop(LastIndex=Get Last LOD() -> (Last LOD=0))
   > n10 = GetGeometryWithStyleSettings!(
    execute
    self
    Part Instance Info:"Atom Model Part Instance Info Structure (by ref)"=Get Part Info()
    Geometry Style Settings:"Atom Processor Geometry Style Settings Structure (by ref)"=Get Geometry Style Settings For LOD(LOD=@n9)
    Bake Scale:bool=true
    MaterialName:String=Default
    Geometry Sections:int=Make Bitmask!(Bitmask:Bitmask Integer=0) -> (ReturnValue:Bitmask Integer) {is=title K2Node_BitmaskLiteral}
   ) -> (then, Geometry Result:Atom Primitive Geometry Object Reference)
   > Add(
    TargetArray=Break Atom Primitive Geometry LODs!(AtomPrimitiveGeometryLODs=Get AtomPrimitiveGeometryLODs()) -> (GeometryLODs, DetailedCollision)
    NewItem=@n10
   )
   n11 = GetGeometryWithStyleSettings(
    execute=@n8.then_1
    Part Instance Info=Get Part Info()
    Geometry Style Settings=Get Nanite Geometry Style Settings()
    Bake Scale=true
    MaterialName=Default
    Geometry Sections=Make Bitmask(Bitmask=0)
   )
   > Set members in Atom Primitive Geometry LODs(StructRef=Get AtomPrimitiveGeometryLODs(), HiResNaniteGeomery=@n11)
   n12 = Sequence(execute=@n2.then_1)
   > n13 = Get Geometry Material Assignments For LOD(PartInfo=Get Part Info(), LOD=0, Color Change Index=0)
   > Add(
    TargetArray:Array of Atom Primitive Geometry Material Assignments Structures=Break Atom Primitive Geometry LODs!(AtomPrimitiveGeometryLODs=Get AtomPrimitiveGeometryLODs()) -> (
     GeometryLODs
     GeometryLODsMaterials:Array of Atom Primitive Geometry Material Assignments Structures
     HiResNaniteGeomery:Atom Primitive Geometry Object Reference
     DetailedCollision
    ).GeometryLODsMaterials
    NewItem:"Atom Primitive Geometry Material Assignments Structure (by ref)"=@n13
   )
   n14 = Get Geometry Material Assignments For Nanite!(execute=@n12.then_1, self, PartInfo:Atom Model Part Instance Info Structure=Get Part Info(), Color Change Index:int=0) -> (then, Material Assignments:Atom Primitive Geometry Material Assignments Structure) {is=title K2Node_CallFunction}
   > Set members in Atom Primitive Geometry LODs!(execute, StructRef=Get AtomPrimitiveGeometryLODs(), HiResNaniteGeomeryMaterials:Atom Primitive Geometry Material Assignments Structure=@n14) -> (then, StructOut)
   Return Node(execute=@n2.then_2, Geometry LODs=Get AtomPrimitiveGeometryLODs())
  ]}
 GetGeometryLODsForUncommonPart {nodes [
   n1 = GetGeometryLODsForUncommonPart!() -> (
    then
    Part Info:Atom Model Part Instance Info Structure
    Uncommon Part Geometry:Atom Primitive Geometry And Transform Structure
    Is Nanite:bool=~
   ) {is=title K2Node_FunctionEntry}
   n2 = Break Atom Primitive Geometry and Transform!(AtomPrimitiveGeometryAndTransform:"Atom Primitive Geometry And Transform Structure (by ref)"=@n1.Uncommon Part Geometry) -> (AtomPrimitiveGeometry:Atom Primitive Geometry Object Reference, Transform:Transform 3f Structure) {is=title K2Node_BreakStruct}
   > n3 = BakeScale(self=@n2, Scale=GetDefaultPrimitiveScale())
   > Set Primitive Transform(Primitive Transform:Transform="Get (a ref)"(
     Array:Array of Transforms=Break Atom Model Part Instance Info!(AtomModelPartInstanceInfo=@n1) -> (Transforms:Array of Transforms)
     Dimension 1=0
    ) -> (Output:"Transform (by ref)")) -> (Output_Get:Transform)
   > n4 = Branch(Condition=Get Is Nanite() -> (Is Nanite:bool))
   n5 = BreakTransform(InTransform=@n2.Transform)
   n6 = vector * vector(A=@n5, B:float=GetDefaultPrimitiveScale()) -> (ReturnValue=~)
   n7 = Get Is Nanite()
   Return Node!(
    execute=[@n4.then, @n8.Completed]
    Geometry LODs=Make Atom Primitive Geometry LODs(
     GeometryLODs=Get GeometryLODs() -> (GeometryLODs:Array of Atom Primitive Geometry Object References)
     HiResNaniteGeomery=Select(Option 0:Atom Primitive Geometry Object Reference, Option 1:Atom Primitive Geometry Object Reference=@n3, Index=@n7) -> (ReturnValue:Atom Primitive Geometry Object Reference)
     HiResNaniteGeomeryMaterials=(MaterialSlotName=,MaterialFeatures=)
     Collision=Make Atom Primitive Collision Geometry()
     DetailedCollision=(AggGeom=(SphereElems=,BoxElems=,SphylElems=,ConvexElems=,TaperedCapsuleElems=,LevelSetElems=,SkinnedLevelSetElems=,MLLevelSetElems=,SkinnedTriangleMeshElems=))
    )
    Merged Mesh Transform:Transform=ComposeTransforms(A=MakeTransform(Location=@n6, Rotation=@n5.Rotation, Scale=@n5.Scale), B=Get Primitive Transform() -> (Primitive Transform:Transform))
   )
   n8 = For Loop(
    execute=@n4.else
    LastIndex=Get Geometry Style Settings Last LOD(Part Info=Get Part Info() -> (Part Info:Atom Model Part Instance Info Structure)) -> (Num Style Settings=0)
   )
   > Add(
    TargetArray:Array of Atom Primitive Geometry Object References=Get GeometryLODs()
    NewItem:"Atom Primitive Geometry Object Reference (by ref)"=@n3
   )
  ]}
 GetGeometryWithStyleSettings {nodes [
   n1 = GetGeometryWithStyleSettings!() -> (
    then
    Part Instance Info:"Atom Model Part Instance Info Structure (by ref)"
    Geometry Style Settings:"Atom Processor Geometry Style Settings Structure (by ref)"
    Bake Scale:bool=~
    MaterialName:String
    Geometry Sections:int
   ) {is=title K2Node_FunctionEntry}
   n2 = Break Atom Model Part Instance Info!(AtomModelPartInstanceInfo=@n1) -> (Primitive:Atom Primitive Object Reference)
   n3 = Break Atom Processor Geometry Style Settings!(AtomProcessorGeometryStyleSettings=@n1.Geometry Style Settings) -> (
    GeometryStylePreferences_4_#1:Array of Strings
    PreferredGeometryLOD_30_#5:int
    ShouldBevel_8_#6:bool=~
    BevelWidth_11_#7=~
    BevelAngleThreshold_13_#8=~
    UseImprovedBevelAlgorithm_37_#3:bool=~
    OnlyBevelBlock-LikePrimitives_39_#4=~
    NumberofSubdivisionsToApply_16_#9:int
    FakeTechnicHole_41_#2=~
   )
   > Set Geometry(Geometry=GetGeometrySection(
     self=GetGeometryContainerForBestExportStyle(self=@n2, GeometryStylePriorityList=@n3, LODIndex=@n3.PreferredGeometryLOD_30_#5)
     PrimitiveGeometrySectionFlag=Get Geometry Sections() -> (Geometry Sections:int=0)
    ))
   > n4 = Branch(Condition=@n3.ShouldBevel_8_#6)
   > n5 = ApplyGeometricBevels!(
    execute
    GeometryToBevel:Atom Primitive Geometry Object Reference=Get Geometry()
    BevelSettings:"Atom Geometric Bevel Settings Structure (by ref)"=GetGeometricBevelSettings!(
     Part Instance Info:"Atom Model Part Instance Info Structure (by ref)"=@n1
     Geometry Style Settings:"Atom Processor Geometry Style Settings Structure (by ref)"=@n1.Geometry Style Settings
    ) -> (AtomGeometricBevelSettings:Atom Geometric Bevel Settings Structure) {is=title K2Node_CallFunction}
   ) -> (then, ReturnValue:Atom Primitive Geometry Object Reference)
   n6 = Branch(execute=[@n5.then, @n4.else], Condition="Not Equal (Integer)"(A=@n3.NumberofSubdivisionsToApply_16_#9))
   > n7 = SubdivideGeometry!(
    execute
    GeometryToSubdivide:Atom Primitive Geometry Object Reference=Get Geometry()
    NumberOfSubdivisions:int=@n3.NumberofSubdivisionsToApply_16_#9
   ) -> (then, ReturnValue:Atom Primitive Geometry Object Reference)
   n8 = Branch(execute=[@n7.then, @n6.else], Condition=Get Bake Scale() -> (Bake Scale:bool))
   > n9 = BakeScale(self=Get Geometry(), Scale=GetDefaultPrimitiveScale())
   n10 = Get MaterialName() -> (MaterialName:String)
   n11 = Branch(execute=[@n9.then, @n8.else], Condition=IsEmpty(InString=@n10))
   Return Node!(execute=[@n12.then, @n11.then], Geometry Result:Atom Primitive Geometry Object Reference=Get Geometry())
   n12 = SetMaterialName(execute=@n11.else, self=Get Geometry(), Name=@n10, PolygonGroupIndex=0)
   Byte to Enum EPrimitiveGeometryComplexity!(Byte:Byte) -> (ReturnValue:EPrimitiveGeometryComplexity Enum) {is=title K2Node_CastByteToEnum}
  ]}
 GetHasOnlyNaniteCompatibleMaterials {nodes [
   n1 = GetHasOnlyNaniteCompatibleMaterials!() -> (then, MergedMeshPartInfo:"Atom Merged Mesh Part Info Structure (by ref)") {is=title K2Node_FunctionEntry}
   > n2 = For Each Loop(Array:Array of Material Interface Object References=Break Atom Merged Mesh Part Info!(AtomMergedMeshPartInfo=@n1) -> (Materials:Array of Material Interface Object References, PreventMergingMaterials, bAllowSimplification=~)) -> (Array Element:Material Interface Object Reference)
   Return Node!(execute=@n2.Completed, bOnlyHasOpaque:bool=true)
   n3 = GetBlendMode!(self:Material Interface Object Reference=@n2) -> (ReturnValue:EBlendMode Enum=BLEND_Opaque)
   > n4 = Branch(Condition=OR Boolean!(A="Equal (Enum)"(A:EBlendMode Enum=@n3, B:EBlendMode Enum=BLEND_Opaque), B="Equal (Enum)"(A=@n3, B=BLEND_Masked)) -> (ReturnValue))
   Return Node(execute=@n4.else, bOnlyHasOpaque=false)
  ]}
 GetIsProcessingPrimitive {nodes [
   GetIsProcessingPrimitive() {is=title K2Node_FunctionEntry}
   > Return Node!(
    execute
    Is Processing Primitive:bool=Parent: GetIsProcessingPrimitive!() -> (Is Processing Primitive:bool) {is=title K2Node_CallParentFunction}
   )
  ]}
 GetMaterialFactory {nodes [
   GetMaterialFactory() {is=title K2Node_FunctionEntry}
   > Return Node!(
    execute
    Material Factory:Atom Material Factory Object Reference=Get Material Factory() -> (Material Factory:Atom Material Factory Object Reference)
   )
  ]}
 GetMaterialFactoryClass {nodes [
   GetMaterialFactoryClass() {is=title K2Node_FunctionEntry}
   > Return Node!(execute, AtomMaterialFactoryClass:Atom Material Factory Class Reference=/Script/AtomModelImporterEditor.AtomMaterialDefaultFactory)
  ]}
 GetMaterialsForCommonPart {nodes [
   n1 = GetMaterialsForCommonPart!() -> (
    then
    Common Part Geometry LODs:"Atom Primitive Geometry LODs Structure (by ref)"
    Shell Merged Mesh Part Info:"Atom Merged Mesh Part Info Structure (by ref)"
    PartInfo:"Atom Model Part Instance Info Structure (by ref)"
    CommonPartIndex:int
   ) {is=title K2Node_FunctionEntry}
   > n2 = Branch(Condition=UseMaterialFactory())
   > n3 = Sequence()
   n4 = For Each Loop(
    Exec=@n3.then_1
    Array:Array of Atom Material Features Structures=Get MaterialFeatures() -> (MaterialFeatures:Array of Atom Material Features Structures)
   ) -> (Array Element:Atom Material Features Structure)
   Return Node!(
    execute=@n4.Completed
    Common Part Materials:Array of Material Interface Object References=Get Materials() -> (Materials:Array of Material Interface Object References)
    Material Features:Array of Atom Material Features Structures=Get MaterialFeatures()
   )
   > n5 = GetOrCreateMaterial!(
    execute
    self:Atom Material Factory Object Reference=Get Material Factory()
    MaterialFeatures:"Atom Material Features Structure (by ref)"=@n4
   ) -> (then, ReturnValue:Material Interface Object Reference)
   > Add(TargetArray:Array of Material Interface Object References=Get Materials(), NewItem:"Material Interface Object Reference (by ref)"=@n5)
   n6 = Break Atom Primitive Geometry LODs!(AtomPrimitiveGeometryLODs=Get Common Part Geometry LODs() -> (Common Part Geometry LODs:"Atom Primitive Geometry LODs Structure (by ref)")) -> (HiResNaniteGeomery, DetailedCollision)
   n7 = Branch(execute=@n3.then_0, Condition=AND Boolean(A=Get Generate Nanite Mesh(), B=IsValid(Object=@n6)))
   > n8 = Get CP Geometry Material Assignments For Nanite!(
    execute
    self
    Part Info:Atom Model Part Instance Info Structure=Get PartInfo()
    Common Part Index:int=Get CommonPartIndex() -> (CommonPartIndex=0)
    "Primitive  Merged Mesh Part Info":Atom Merged Mesh Part Info Structure=Get Shell Merged Mesh Part Info() -> (Shell Merged Mesh Part Info:"Atom Merged Mesh Part Info Structure (by ref)")
    Common Part Merged Mesh Part Info:Atom Primitive Geometry LODs Structure=Get Common Part Geometry LODs()
   ) -> (then, Material Assignments:Atom Primitive Geometry Material Assignments Structure) {is=title K2Node_CallFunction}
   > n9 = Append Array(
    TargetArray:Array of Atom Material Features Structures=Get MaterialFeatures()
    SourceArray:Array of Atom Material Features Structures=Break Atom Primitive Geometry Material Assignments(AtomPrimitiveGeometryMaterialAssignments=@n8).MaterialFeatures
   )
   n10 = For Each Loop(
    Exec=[@n9.then, @n7.else]
    Array:Array of Atom Primitive Geometry Object References=Break Atom Primitive Geometry LODs!(AtomPrimitiveGeometryLODs=Get Common Part Geometry LODs()) -> (GeometryLODs:Array of Atom Primitive Geometry Object References, DetailedCollision)
   ) -> (Array Element:Atom Primitive Geometry Object Reference)
   > n11 = Get CP Geometry Material Assignments For LOD!(
    execute
    self
    Part Info:Atom Model Part Instance Info Structure=Get PartInfo()
    Common Part Index:int=Get CommonPartIndex() -> (CommonPartIndex=0)
    "Primitive  Merged Mesh Part Info":Atom Merged Mesh Part Info Structure=Get Shell Merged Mesh Part Info()
    Common Part Geometry LODs:Atom Primitive Geometry LODs Structure=Get Common Part Geometry LODs()
    LOD:int=@n10.Array Index
   ) -> (then, Material Assignments:Atom Primitive Geometry Material Assignments Structure) {is=title K2Node_CallFunction}
   > Append Array(
    TargetArray=Get MaterialFeatures()
    SourceArray=Break Atom Primitive Geometry Material Assignments(AtomPrimitiveGeometryMaterialAssignments=@n11).MaterialFeatures
   )
   n12 = Break Atom Merged Mesh Part Info(AtomMergedMeshPartInfo=@n1.Shell Merged Mesh Part Info) -> (bAllowSimplification=~)
   Return Node(
    execute=@n2.else
    Common Part Materials=Make Array("[0]":Material Interface Object Reference="Get (a copy)"(Array:Array of Material Interface Object References=@n12) -> (Output:Material Interface Object Reference)) -> (Array:Array of Material Interface Object References)
   )
  ]}
 GetMaterialsForPart {nodes [
   PrintString(InString=GetMaterialsForParts)
   n1 = Get Materials()
   > n2 = For Each Loop(Array:Array of Material Interface Object References=@n1) -> (Array Element:Material Interface Object Reference)
   Return Node!(
    execute=[@n9.Completed, @n2.Completed]
    Material:Array of Material Interface Object References=@n1
    MaterialFeatures:Array of Atom Material Features Structures=Get Material Features() -> (Material Features:Array of Atom Material Features Structures)
   )
   > PrintString(InString=GetDisplayName(Object=@n2))
   n3 = GetMaterialsForPart!() -> (then, Part Info:"Atom Model Part Instance Info Structure (by ref)", MergedMeshPartInfo:Atom Merged Mesh Part Info Structure) {is=title K2Node_FunctionEntry}
   > Clear(TargetArray:Array of Material Interface Object References=Get Materials())
   n4 = Length(TargetArray=Break Atom Model Part Instance Info!(AtomModelPartInstanceInfo=@n3) -> (Colors:Array of Atom Model Part Color Info Structures))
   > Set MaxColorChangeID(MaxColorChangeID:int=SelectInt(A=int - int(A=@n4), B=0, bPickA=Has Color Change Baked in Geometry(PartInfo=@n3))) -> (Output_Get:int=0)
   n5 = Get Part Info() -> (Part Info:"Atom Model Part Instance Info Structure (by ref)")
   n6 = AND Boolean(
    A=Get Use Legacy Decoration Material() -> (Use Legacy Decoration Material:bool)
    B=Is Not Empty(TargetArray:Array of Atom Model Part Decoration Info Structures=Break Atom Model Part Instance Info!(AtomModelPartInstanceInfo=@n5) -> (Decorations:Array of Atom Model Part Decoration Info Structures))
   )
   > n7 = Branch(Condition=AND Boolean(A=UseMaterialFactory(), B=NOT Boolean(A=@n6)))
   > n8 = Sequence()
   n9 = For Each Loop(Exec=@n8.then_1, Array:Array of Atom Material Features Structures=Get Material Features()) -> (Array Element:Atom Material Features Structure)
   > n10 = GetOrCreateMaterial(self=Get Material Factory(), MaterialFeatures=@n9)
   > Add(TargetArray=Get Materials(), NewItem=@n10)
   n11 = Break Atom Primitive Geometry LODs!(AtomPrimitiveGeometryLODs=Break Atom Merged Mesh Part Info!(AtomMergedMeshPartInfo=Get MergedMeshPartInfo()) -> (GeometryLODs:Atom Primitive Geometry LODs Structure)) -> (GeometryLODs, HiResNaniteGeomery:Atom Primitive Geometry Object Reference)
   n12 = Branch(execute=@n8.then_0, Condition=AND Boolean(A=Get Generate Nanite Mesh(), B=IsValid(Object=@n11.HiResNaniteGeomery)))
   > n13 = For Loop(LastIndex=Get MaxColorChangeID() -> (MaxColorChangeID:int=0))
   > n14 = Get Geometry Material Assignments For Nanite(PartInfo=Get Part Info(), Color Change Index=@n13)
   > Append Array(
    TargetArray=Get Material Features()
    SourceArray=Break Atom Primitive Geometry Material Assignments(AtomPrimitiveGeometryMaterialAssignments=@n14).MaterialFeatures
   )
   n15 = For Each Loop(Exec=[@n13.Completed, @n12.else], Array:Array of Atom Primitive Geometry Object References=@n11) -> (Array Element:Atom Primitive Geometry Object Reference)
   > n16 = For Loop(LastIndex=Get MaxColorChangeID() -> (MaxColorChangeID=0))
   > n17 = Get Geometry Material Assignments For LOD(PartInfo=Get Part Info(), LOD=@n15.Array Index, Color Change Index=@n16)
   > Append Array(
    TargetArray=Get Material Features()
    SourceArray=Break Atom Primitive Geometry Material Assignments(AtomPrimitiveGeometryMaterialAssignments=@n17).MaterialFeatures
   )
   n18 = Sequence(execute=@n7.else)
   > Clear(TargetArray=Get Materials())
   n19 = Break Atom Model Part Instance Info!(AtomModelPartInstanceInfo=Get Part Info()) -> (Decorations, Colors:Array of Atom Model Part Color Info Structures)
   n20 = "Get (a ref)"(Array:Array of Atom Model Part Color Info Structures=@n19.Colors, Dimension 1=0) -> (Output:"Atom Model Part Color Info Structure (by ref)")
   > n21 = GetDefaultMaterialForType!(
    execute
    self
    "Atom Material  Type":EAtomMaterialType Enum=Break Atom Model Part Color Info!(AtomModelPartColorInfo=@n20) -> (Color:Color Structure, ColorId, MaterialType).MaterialType
    Part Info:"Atom Model Part Instance Info Structure (by ref)"=Get Part Info()
   ) -> (then, Material:Material Interface Object Reference) {is=title K2Node_CallFunction}
   n22 = Is Not Empty(TargetArray=@n19)
   n23 = NOT Boolean(A=GetIsProcessingPrimitive() {is=title K2Node_CallFunction})
   n24 = Get Support Decorations()
   > n25 = Branch(Condition=NOT Boolean(A=AND Boolean!(A=@n24, B=@n22, C:bool=@n23) -> (ReturnValue)))
   > Add(TargetArray=Get Materials(), NewItem=@n21)
   n26 = GetDecorationMaterialPath!(execute=@n25.else, self, Model Path:String=GetModelPath(self=GetAtomModel() {is=title K2Node_CallFunction})) -> (then, Material Path:String) {is=title K2Node_CallFunction}
   > n27 = GetDecorationMaterial!(
    execute
    self
    Base Material:Material Interface Object Reference=@n21
    Decoration Infos:Array of Atom Model Part Decoration Info Structures=@n19
    New Material Path:String=@n26
    ColorInfo:Atom Model Part Color Info Structure=@n20
   ) -> (then, Decoration Material:Material Interface Object Reference) {is=title K2Node_CallFunction}
   > Add(TargetArray=Get Materials(), NewItem=@n27)
   Return Node(execute=@n18.then_1, Material=Get Materials())
  ]}
 GetMergedMeshType {nodes [
   n1 = GetMergedMeshType!() -> (then, " Mesh Name":String) {is=title K2Node_FunctionEntry}
   > n2 = Branch(Condition=EndsWith(SourceString=@n1, InSuffix=GetTransparentMeshSuffix!(self) -> (Transparent Mesh Suffix:String)))
   > Return Node!(execute, Merged Mesh Type:EAtomMergedMeshType Enum=NewEnumerator1)
   n3 = Branch(
    execute=@n2.else
    Condition=EndsWith(
     SourceString="Get  Mesh Name"() -> (" Mesh Name":String)
     InSuffix=GetCommonPartMeshSuffix!(self) -> (Common Part Mesh Suffix:String) {is=title K2Node_CallFunction}
    )
   )
   > Return Node(Merged Mesh Type=NewEnumerator2)
   Return Node(execute=@n3.else, Merged Mesh Type=NewEnumerator0)
  ]}
 GetMeshMergingSettings {nodes [
   GetMeshMergingSettings!() -> (then, AtomModel:Atom Model Object Reference) {is=title K2Node_FunctionEntry}
   n1 = Get VertexColorDebugMode() -> (VertexColorDebugMode=None)
   > n2 = Branch(Condition="Not Equal (Enum)"(A=@n1, B=None))
   n3 = Copy(Item:Atom Mesh Merging Settings Structure=GetDefaultMeshMergingSettings!() -> (Regular Mesh Merging Settings:Atom Mesh Merging Settings Structure) {is=title K2Node_CallFunction}) -> (ReturnValue:Atom Mesh Merging Settings Structure)
   > n4 = Set members in Atom Mesh Merging Settings!(
    execute:Exec
    StructRef:"Atom Mesh Merging Settings Structure (by ref)"=@n3
    DebugUtilitySettings:Atom Mesh Merging Debug Utility Settings Structure=Make Atom Mesh Merging Debug Utility Settings!(VertexColorMappingMode:EAtomMeshMergingVertexColorMappingMode Enum=@n1) -> (AtomMeshMergingDebugUtilitySettings:Atom Mesh Merging Debug Utility Settings Structure) {is=title K2Node_MakeStruct}
   ) -> (then:Exec, StructOut:"Atom Mesh Merging Settings Structure (by ref)") {is=title K2Node_SetFieldsInStruct}
   > Return Node!(execute, Regular Mesh Merging Settings:Atom Mesh Merging Settings Structure=@n4)
   Return Node(execute=@n2.else, Regular Mesh Merging Settings=@n3)
  ]}
 GetNaniteMeshMergingSettings {nodes [
   GetNaniteMeshMergingSettings!() -> (then, AtomModel:Atom Model Object Reference) {is=title K2Node_FunctionEntry}
   > Return Node!(execute, Nanite Mesh Merging Settings:Atom Mesh Merging Settings Structure=Get Mesh Merging Settings())
  ]}
 GetOverrideMeshMergingSettings {nodes [
   GetOverrideMeshMergingSettings!() -> (then, Default Settings:"Atom Mesh Merging Settings Structure (by ref)") {is=title K2Node_FunctionEntry}
   > Return Node!(
    execute
    Mesh Merging Settings Map:Map of Strings to Atom Mesh Merging Settings Structures=Get Mesh Merging Settings() -> (Mesh Merging Settings:Map of Strings to Atom Mesh Merging Settings Structures)
   )
  ]}
 GetPartColorInfo {nodes [
   n1 = GetPartColorInfo!() -> (then, Part Info:Atom Model Part Instance Info Structure, Color Change Index:int) {is=title K2Node_FunctionEntry}
   n2 = Break Atom Model Part Color Info!(AtomModelPartColorInfo="Get (a ref)"(Array=Break Atom Model Part Instance Info!(AtomModelPartInstanceInfo=@n1) -> (Colors), Dimension 1=@n1.Color Change Index)) -> (ColorId, MaterialType)
   > Return Node!(execute, Color Id:int=@n2, Material Type:EAtomMaterialType Enum=@n2.MaterialType)
  ]}
 GetPartsCollectionConnectivity {nodes [
   GetPartsCollectionConnectivity() {is=title K2Node_FunctionEntry}
   > Return Node!(
    execute
    Parts Collection Connectivity:Atom Model Parts Collection Connectivity Object Reference=Get Parts Collection Connectivity() -> (Parts Collection Connectivity:Atom Model Parts Collection Connectivity Object Reference)
   )
  ]}
 GetPartsCollectionName {nodes [
   GetPartsCollectionName() {is=title K2Node_FunctionEntry}
   n1 = GetPartsCollectionConnectivity!() -> (Parts Collection Connectivity:Atom Model Parts Collection Connectivity Object Reference) {is=title K2Node_CallFunction}
   > Return Node!(
    execute
    Parts Collection Name:String=Break Atom Model Parts Collection(AtomModelPartsCollection=Get PartsCollection!(self:Atom Model Parts Collection Connectivity Object Reference=@n1) -> (PartsCollection:Atom Model Parts Collection Structure))
   )
  ]}
 GetPivotForPartsCollection {nodes [
   n1 = GetPivotForPartsCollection!() -> (then, Parts Collection:Atom Model Parts Collection Structure) {is=title K2Node_FunctionEntry}
   > Return Node!(execute, Pivot:Transform=Break Atom Model Parts Collection(AtomModelPartsCollection=@n1).Pivot)
  ]}
 GetPreserveUVsForPart {nodes [
   n1 = GetPreserveUVsForPart!() -> (
    then
    Atom Model Part Info:"Atom Model Part Instance Info Structure (by ref)"
    Merged Mesh Part Info:"Atom Merged Mesh Part Info Structure (by ref)"
   ) {is=title K2Node_FunctionEntry}
   n2 = Is Not Empty(TargetArray=Break Atom Model Part Instance Info!(AtomModelPartInstanceInfo=@n1) -> (Decorations:Array of Atom Model Part Decoration Info Structures, Group:String, ParentGroup:String))
   n3 = Get Support Decorations()
   > Return Node!(
    execute
    PreserveUVs:bool=OR Boolean(
     A=Break Atom Merged Mesh Part Info!(AtomMergedMeshPartInfo=@n1.Merged Mesh Part Info) -> (
      GeometryLODs
      Transform:Transform
      bHasConstantOverrideVertexColor:bool=~
      VertexColor:Color Structure
      Materials:Array of Material Interface Object References
      PreventMergingMaterials:Array of Material Interface Object References
      MaterialFeatures:Array of Atom Material Features Structures
      MeshDetailLevel:EAtomMergedMeshPartDetailLevel Enum
      bPreserveUVs:bool=~
      bAllowMerging:bool=~
      bAllowSimplification:bool=~
      bAllowApproximation:bool=~
      bIsOccluder:bool=~
      FilterLODLevel:int
      MeshName:String
      InstanceSubsetID:int
      ApproximationType:Bitmask Integer
     ).bPreserveUVs
     B=AND Boolean!(A=@n3, B=@n2) -> (ReturnValue)
    )
   )
  ]}
 GetPrimitiveGeometry {nodes [
   n1 = GetPrimitiveGeometry!() -> (
    then
    Atom Model Part Info:"Atom Model Part Instance Info Structure (by ref)"
    Geometry Map:Map of Strings to Atom Primitive Geometry LODs Structures
    Atom Model:Atom Model Object Reference
   ) {is=title K2Node_FunctionEntry}
   > n2 = GetUniqueNameForGeometryLODs!(execute, self, Part Info:Atom Model Part Instance Info Structure=@n1) -> (then, Name:String)
   > Set Unique Name(Unique Name:String=@n2) -> (Output_Get:String)
   n3 = Find(TargetMap=@n1.Geometry Map, Key=Get Unique Name() -> (Unique Name:String))
   > n4 = Branch(Condition=@n3.ReturnValue)
   > Return Node!(execute, Geometry LODs:Atom Primitive Geometry LODs Structure=@n3, Geometry Unique Name:String=Get Unique Name())
   n5 = GetGeometryLODsForPart!(execute=@n4.else, self, Part Info:"Atom Model Part Instance Info Structure (by ref)"=@n1) -> (then, Geometry LODs:Atom Primitive Geometry LODs Structure) {is=title K2Node_CallFunction}
   > Set New Geometry LODs(New Geometry LODs:Atom Primitive Geometry LODs Structure=@n5) -> (Output_Get:Atom Primitive Geometry LODs Structure)
   n6 = Get New Geometry LODs() -> (New Geometry LODs:Atom Primitive Geometry LODs Structure)
   > n7 = Get Collision Geometry for Part!(
    execute
    self
    Geometry:Atom Primitive Geometry Object Reference="Get (a copy)"(Array:Array of Atom Primitive Geometry Object References=Break Atom Primitive Geometry LODs!(AtomPrimitiveGeometryLODs=@n6) -> (GeometryLODs, DetailedCollision:Atom Primitive Collision Geometry Structure)) -> (Output:Atom Primitive Geometry Object Reference)
    CollisionType:EPrimitiveAutoCollisionType Enum=Get Auto Primitive Collision Type() -> (Auto Primitive Collision Type:EPrimitiveAutoCollisionType Enum=Box)
    PartInfo:Atom Model Part Instance Info Structure=@n1
   ) -> (then, Collision Geometry:Atom Primitive Collision Geometry Structure) {is=title K2Node_CallFunction}
   > n8 = Sequence()
   > Branch(Condition=GenerateDetailedCollision() {is=title K2Node_CallFunction})
   n9 = Break Atom Model Part Instance Info!(AtomModelPartInstanceInfo=@n1) -> (Primitive:Atom Primitive Object Reference, Transforms:Array of Transforms)
   > n10 = GeneratePrimitiveDetailedCollision!(
    execute
    Settings:"Primitive Source Collision Geometry Settings Structure (by ref)"=Make Primitive Source Collision Geometry Settings(
     Primitive=@n9
     PrimitiveScale=GetDefaultPrimitiveScale()
     BoneTransforms=@n9.Transforms
     CylinderColliderNumberOfSides=Get Cylinder Colliders Number Of Sides() -> (Side Count=0) {is=title K2Node_CallFunction}
     bSplitNarrowColliders=NeedSplittingNarrowColliders!() -> (bSplitNarrowColliders:bool)
    )
   ) -> (then, ReturnValue:Atom Primitive Collision Geometry Structure)
   > Set members in Atom Primitive Geometry LODs!(execute, StructRef=Get New Geometry LODs(), DetailedCollision:Atom Primitive Collision Geometry Structure=@n10) -> (then, StructOut)
   Set members in Atom Primitive Geometry LODs!(execute=@n8.then_1, StructRef=Get New Geometry LODs(), Collision:Atom Primitive Collision Geometry Structure=@n7) -> (then, StructOut)
   n11 = Get New Geometry LODs()
   n12 = Get Unique Name()
   > Add(TargetMap=@n1.Geometry Map, Key=@n12, Value=@n11) {is=title K2Node_CallFunction}
   > Return Node(Geometry LODs=@n11, Geometry Unique Name=@n12)
  ]}
 GetProcessedPrimitivePath {nodes [
   GetProcessedPrimitivePath() {is=title K2Node_FunctionEntry}
   > n1 = GetImporterSettings()
   n2 = Get ProcessedPrimitiveContentFolder(self=@n1)
   n3 = Get Custom Primitive Location() -> (Custom Primitive Location:String)
   > Return Node!(execute, Processed Primitive Path:String=SelectString(A=Break Directory Path(DirectoryPath=@n2), B=@n3, bPickA=IsEmpty(InString=@n3)))
  ]}
 GetStaticMeshName {nodes [
   n1 = GetStaticMeshName!() -> (then, Model:Atom Model Object Reference, Parts Collection:"Atom Model Parts Collection Structure (by ref)") {is=title K2Node_FunctionEntry}
   > n2 = Branch(Condition=AND Boolean(A=GetIsProcessingPrimitive(), B=Get Support Decorations()))
   n3 = Append(A=SM_, B=Break Atom Model Parts Collection(AtomModelPartsCollection=@n1.Parts Collection))
   > Return Node!(execute, Result:String=Append(A=@n3, B=_DECO))
   Return Node(execute=@n2.else, Result=@n3)
  ]}
 GetStaticMeshPath {nodes [
   n1 = GetStaticMeshPath!() -> (then, Model:Atom Model Object Reference, Parts Collection:"Atom Model Parts Collection Structure (by ref)") {is=title K2Node_FunctionEntry}
   > n2 = Branch(Condition=GetIsProcessingPrimitive())
   > Return Node(Result=GetProcessedPrimitivePath() {is=title K2Node_CallFunction})
   Return Node(execute=@n2.else, Result=GetModelPath(self=@n1))
  ]}
 GetTransparentMeshMergingSettings {nodes [
   GetTransparentMeshMergingSettings() {is=title K2Node_FunctionEntry}
   > Return Node!(
    execute
    Mesh Merging Settings:Atom Mesh Merging Settings Structure=Get Tranparent Mesh Merging Settings() -> (Tranparent Mesh Merging Settings:Atom Mesh Merging Settings Structure)
   )
  ]}
 GetTransparentMeshSuffix {nodes [
   GetTransparentMeshSuffix() {is=title K2Node_FunctionEntry}
   > Return Node!(execute, Transparent Mesh Suffix:String=Trans)
  ]}
 GetUniqueNameForGeometryLODs {nodes [
   n1 = GetUniqueNameForGeometryLODs!() -> (then, Part Info:Atom Model Part Instance Info Structure) {is=title K2Node_FunctionEntry}
   n2 = Break Atom Model Part Instance Info!(AtomModelPartInstanceInfo=@n1) -> (PartGuid:Atom Model Part Guid Structure, PartId:int, Primitive)
   n3 = Get Part Info() -> (Part Info:Atom Model Part Instance Info Structure)
   > n4 = Branch(Condition=OR Boolean(A=IsFlexElement(self=@n2.Primitive), B=Has Color Change Baked in Geometry(PartInfo=@n3)))
   n5 = Break Atom Model Part Guid!(AtomModelPartGuid:"Atom Model Part Guid Structure (by ref)"=@n2) -> (Guid:Guid Structure, PartIndex:int) {is=title K2Node_BreakStruct}
   n6 = Append(A=m, B="To String (Integer)"(InInt=@n2.PartId))
   > Return Node!(
    execute
    Name:String=Append!(
     A=@n6
     B=_
     C:String="To String (Guid)"!(InGuid:"Guid Structure (by ref)"=@n5) -> (ReturnValue:String)
     D:String=_
     E:String="To String (Integer)"(InInt=@n5.PartIndex)
    ) -> (ReturnValue)
   )
   Return Node(execute=@n4.else, Name=@n6)
  ]}
 GetVertexColorForCommonPart {nodes [
   n1 = GetVertexColorForCommonPart!() -> (
    then
    Part Info:"Atom Model Part Instance Info Structure (by ref)"
    Merged Mesh Part Info:"Atom Merged Mesh Part Info Structure (by ref)"
    Common Part Index:int
   ) {is=title K2Node_FunctionEntry}
   > n2 = GetVertexColorForPart(
    Part Info=@n1
    Color Change Index=GetCommonPartColorChangeIndex!(Part Info:Atom Model Part Instance Info Structure=@n1, Common Part Index:int=@n1.Common Part Index) -> (Color Change Index:int=0) {is=title K2Node_CallFunction}
   )
   > Return Node!(execute, Vertex Color:Color Structure=@n2)
   Break Atom Merged Mesh Part Info!(AtomMergedMeshPartInfo=@n1.Merged Mesh Part Info) -> (
    bHasConstantOverrideVertexColor=~
    VertexColor
    PreventMergingMaterials
    MaterialFeatures
    bAllowSimplification=~
    bIsOccluder=~
    FilterLODLevel
    MeshName
   )
  ]}
 GetVertexColorForPart {nodes [
   n1 = GetVertexColorForPart!() -> (then, Part Info:"Atom Model Part Instance Info Structure (by ref)", Color Change Index:int) {is=title K2Node_FunctionEntry}
   > n2 = Branch(Condition=UseRandomAlphaColorForPart!() -> (UseRandomAlpha:bool))
   n3 = Break Atom Model Part Instance Info!(AtomModelPartInstanceInfo=@n1) -> (PartGuid, Colors:Array of Atom Model Part Color Info Structures)
   n4 = Get Hash!(InModelPartGuid:"Atom Model Part Guid Structure (by ref)"=@n3) -> (ReturnValue:int=0) {is=title K2Node_CallFunction}
   n5 = Break Atom Model Part Color Info!(AtomModelPartColorInfo="Get (a ref)"(Array=@n3.Colors, Dimension 1=Get Color Change Index() -> (Color Change Index=0))) -> (Color:Color Structure, ColorId, MaterialType)
   n6 = "Min (Integer)"(A=@n5.ColorId, B=255) -> (ReturnValue=0)
   n7 = "Max (Integer)"(A=int - int(A=@n5.ColorId, B=255), B=0)
   n8 = Make Color!() -> (Color:Color Structure) {is=title K2Node_MakeStruct}
   n9 = Select!(
    VertexColor:Color Structure="To Color (LinearColor)"!(
     InLinearColor:Linear Color Structure="To LinearColor (Color)"!(InColor:Color Structure=@n5) -> (ReturnValue:Linear Color Structure)
     InUseSRGB:bool=false
    ) -> (ReturnValue:Color Structure)
    InstanceCustomData:Color Structure=@n8
    Material:Color Structure=@n8
    LUTVertexColor:Color Structure=Make Color!(
     B:Byte=0
     G:Byte="To Byte (Integer)"!(InInt:int=@n7) -> (ReturnValue:Byte=0)
     R:Byte="To Byte (Integer)"(InInt=@n6) -> (ReturnValue=0)
     A:Byte=0
    ) -> (Color)
    LUTInstanceCustomData:Color Structure=@n8
    LUTMaterial:Color Structure
    Index:EAtomMaterialColorSource Enum=Get Material Color Source() -> (Material Color Source=VertexColor)
   ) -> (ReturnValue:Color Structure)
   > n10 = Set members in Color!(
    execute:Exec
    StructRef:"Color Structure (by ref)"=@n9
    A:Byte="To Byte (Integer)"(InInt="% (Integer)"(A=@n4, B=256) -> (ReturnValue=0)) -> (ReturnValue=0)
   ) -> (then:Exec, StructOut:"Color Structure (by ref)") {is=title K2Node_SetFieldsInStruct}
   > Return Node!(execute, Color:Color Structure=@n10)
   Return Node(execute=@n2.else, Color=@n9)
  ]}
 Has Color Change Baked in Geometry {nodes [
   Has Color Change Baked in Geometry!() -> (then, PartInfo:Atom Model Part Instance Info Structure) {is=title K2Node_FunctionEntry}
   > Return Node!(execute, HasColorChangeBakedInGeometry:bool=false)
  ]}
 Has Custom Baked Vertex Color {nodes [
   n1 = Has Custom Baked Vertex Color!() -> (then, PartInfo:Atom Model Part Instance Info Structure) {is=title K2Node_FunctionEntry}
   > Return Node!(execute, HasCustomBakedVertexColor:bool=Has Color Change Baked in Geometry(PartInfo=@n1) {is=title K2Node_CallFunction})
  ]}
 Initialize Geometry Style Settings {nodes [
   Initialize Geometry Style Settings() {is=title K2Node_FunctionEntry}
   n1 = Get Geometry style() -> (Geometry style:String)
   > n2 = Branch(Condition=IsEmpty(InString=@n1))
   > n3 = Add(TargetArray:Array of Strings=Get Style Preferences() -> (Style Preferences:Array of Strings), NewItem:"String (by ref)"=Realtime) {is=title K2Node_CallArrayFunction}
   Set Nanite Geometry Style Settings(
    execute=[@n3.then, @n8.then, @n7.then]
    Nanite Geometry Style Settings:Atom Processor Geometry Style Settings Structure=Make Atom Processor Geometry Style Settings!(
     GeometryStylePreferences_4_#1:Array of Strings=Get Style Preferences()
     PreferredGeometryLOD_30_#5:int=0
     ShouldBevel_8_#6:bool=true
     UseImprovedBevelAlgorithm_37_#3:bool=False
     OnlyBevelBlock-LikePrimitives_39_#4:bool=True
     NumberofSubdivisionsToApply_16_#9:int=0
     AnalyticFakeUnderside_32_#12:bool=False
     ForceFlatCap_35_#13:bool=False
     FakeTechnicHole_41_#2:bool=False
    ) -> (AtomProcessorGeometryStyleSettings:Atom Processor Geometry Style Settings Structure) {is=title K2Node_MakeStruct}
   ) -> (Output_Get:Atom Processor Geometry Style Settings Structure)
   n4 = Get Style Preferences()
   n5 = Get Style Preferences()
   > Set LOD Geometry Style Settings(LOD Geometry Style Settings:Array of Atom Processor Geometry Style Settings Structures=Make Array!(
     "[0]":Atom Processor Geometry Style Settings Structure=Make Atom Processor Geometry Style Settings!(
      GeometryStylePreferences_4_#1=@n4
      PreferredGeometryLOD_30_#5=0
      ShouldBevel_8_#6=False
      UseImprovedBevelAlgorithm_37_#3=False
      OnlyBevelBlock-LikePrimitives_39_#4=True
      NumberofSubdivisionsToApply_16_#9=0
      FakeTechnicHole_41_#2=False
     ) -> (AtomProcessorGeometryStyleSettings)
     "[1]":Atom Processor Geometry Style Settings Structure=Make Atom Processor Geometry Style Settings(
      GeometryStylePreferences_4_#1=@n5
      PreferredGeometryLOD_30_#5=0
      ShouldBevel_8_#6=False
      UseImprovedBevelAlgorithm_37_#3=False
      OnlyBevelBlock-LikePrimitives_39_#4=True
      NumberofSubdivisionsToApply_16_#9=0
      FakeTechnicHole_41_#2=False
     )
    ) -> (Array:Array of Atom Processor Geometry Style Settings Structures)) -> (Output_Get:Array of Atom Processor Geometry Style Settings Structures)
   Add(execute=@n2.else, TargetArray=Get Style Preferences(), NewItem=@n1)
   n6 = Get Fallback Geometry style() -> (Fallback Geometry style:String)
   > n7 = Branch(Condition=IsEmpty(InString=@n6))
   n8 = Add(execute=@n7.else, TargetArray=Get Style Preferences(), NewItem=@n6)
  ]}
 Initialize Parts Collection Connectivity {nodes [
   n1 = Initialize Parts Collection Connectivity!() -> (then, Parts Collection:"Atom Model Parts Collection Structure (by ref)", Atom Model:Atom Model Object Reference) {is=title K2Node_FunctionEntry}
   > n2 = CreatePartsCollectionConnectivity!(
    execute
    PartsCollection:"Atom Model Parts Collection Structure (by ref)"=@n1
    Model:Atom Model Object Reference=@n1.Atom Model
    Scale:float=GetDefaultPrimitiveScale()
   ) -> (then, ReturnValue:Atom Model Parts Collection Connectivity Object Reference)
   > InitializeCommonParts!(
    execute
    PartsCollectionConnectivity:Atom Model Parts Collection Connectivity Object Reference=@n2
    CommonPartOptimizationFlags:Bitmask Integer=Get Common Part Optimizations() -> (Common Part Optimizations:Bitmask Integer=0)
   ) -> (then)
   > n3 = FilterCommonParts!(
    execute
    self
    InPartsCollection:"Atom Model Parts Collection Structure (by ref)"=Copy(Item:Atom Model Parts Collection Structure=Get PartsCollection(self=@n2)) -> (ReturnValue:Atom Model Parts Collection Structure)
   ) -> (then, OutPartsCollection:Atom Model Parts Collection Structure) {is=title K2Node_CallFunction}
   > n4 = CreatePartsCollectionConnectivity(PartsCollection=@n3, Model=Get Atom Model() -> (Atom Model:Atom Model Object Reference), Scale=GetDefaultPrimitiveScale())
   > Set Parts Collection Connectivity(Parts Collection Connectivity:Atom Model Parts Collection Connectivity Object Reference=@n4) -> (Output_Get:Atom Model Parts Collection Connectivity Object Reference)
   > Return Node!(execute)
  ]}
 Is Color Change Part {nodes [
   n1 = Is Color Change Part!() -> (then, PartInfo:"Atom Model Part Instance Info Structure (by ref)") {is=title K2Node_FunctionEntry}
   n2 = Break Atom Model Part Instance Info!(AtomModelPartInstanceInfo=@n1) -> (Colors)
   > Return Node!(execute, Is Color Change Part:bool="integer > integer"(A=Length(TargetArray=@n2), B=1))
  ]}
 IsCommonPartUsedInGeometryStyle {nodes [
   n1 = IsCommonPartUsedInGeometryStyle!() -> (
    then
    PartInfo:"Atom Model Part Instance Info Structure (by ref)"
    GeometryStyleSettings:"Atom Processor Geometry Style Settings Structure (by ref)"
    CommonPart:Atom Common Part And Transform Structure
   ) {is=title K2Node_FunctionEntry}
   n2 = Break Atom Processor Geometry Style Settings!(AtomProcessorGeometryStyleSettings=@n1.GeometryStyleSettings) -> (
    GeometryStylePreferences_4_#1
    PreferredGeometryLOD_30_#5
    UseImprovedBevelAlgorithm_37_#3=~
    OnlyBevelBlock-LikePrimitives_39_#4=~
    FakeTechnicHole_41_#2=~
   )
   n3 = GetExportStyleName(self=GetGeometryContainerForBestExportStyle(
     self=Break Atom Model Part Instance Info!(AtomModelPartInstanceInfo=@n1) -> (Primitive:Atom Primitive Object Reference)
     GeometryStylePriorityList=@n2
     LODIndex=@n2.PreferredGeometryLOD_30_#5
    ))
   > Return Node!(
    execute
    CommonPartUsed:bool=Contains Item(
     TargetArray:Array of Names=Break Atom Common Part and Transform!(AtomCommonPartAndTransform=@n1.CommonPart) -> (GeometryStyles)
     ItemToFind:"Name (by ref)"=String To Name(InString=@n3)
    )
   )
  ]}
 Make Merged Geometries {nodes [
   Make Merged Geometries!() -> (then, MergedMeshPartInfos:Array of Atom Merged Mesh Part Info Structures) {is=title K2Node_FunctionEntry}
   > n1 = GetMeshMergingSettings!(execute, self, AtomModel:Atom Model Object Reference=Get Model()) -> (then, Regular Mesh Merging Settings:Atom Mesh Merging Settings Structure) {is=title K2Node_CallFunction}
   > n2 = GetOverrideMeshMergingSettings!(execute, self, Default Settings:"Atom Mesh Merging Settings Structure (by ref)"=@n1) -> (then, Mesh Merging Settings Map:Map of Strings to Atom Mesh Merging Settings Structures) {is=title K2Node_CallFunction}
   > n3 = Branch(Condition=Get UseNewMergeAPI())
   n4 = Get MergedMeshPartInfos()
   > n5 = MakeOptimizedMergedGeometries_V2!(
    execute
    MergedMeshParts:Array of Atom Merged Mesh Part Info Structures=@n4
    DefaultMeshMergingSettings:"Atom Mesh Merging Settings Structure (by ref)"=@n1
    MeshMergingSettings:Map of Strings to Atom Mesh Merging Settings Structures=@n2
   ) -> (then, MergedGeometries:Map of Strings to Atom Merged Geometry Structures)
   > Return Node!(execute, MergedGeometries:Map of Strings to Atom Merged Geometry Structures=@n5)
   n6 = MakeOptimizedMergedGeometries!(
    execute=@n3.else
    MergedMeshParts:Array of Atom Merged Mesh Part Info Structures=@n4
    DefaultMeshMergingSettings:"Atom Mesh Merging Settings Structure (by ref)"=@n1
    MeshMergingSettings:Map of Strings to Atom Mesh Merging Settings Structures=@n2
   ) -> (then, MergedGeometries:Map of Strings to Atom Merged Geometry Structures)
   > Return Node(MergedGeometries=@n6)
  ]}
 NeedSplittingNarrowColliders {nodes [
   NeedSplittingNarrowColliders() {is=title K2Node_FunctionEntry}
   > Return Node!(execute, bSplitNarrowColliders:bool=false)
  ]}
 OnBuildPartsCollections {nodes [
   n1 = OnBuildPartsCollections!() -> (then, Base Parts Collection:"Atom Model Parts Collection Structure (by ref)", Model:Atom Model Object Reference) {is=title K2Node_FunctionEntry}
   n2 = Get SelectionSets!(self:Atom Model Object Reference=@n1.Model) -> (SelectionSets:Array of Atom Model Selection Set Structures)
   > n3 = Branch(Condition=AND Boolean(
     A=Get Separate Selection Sets() -> (Separate Selection Sets:bool)
     B=Is Not Empty(TargetArray:Array of Atom Model Selection Set Structures=@n2)
    ))
   > n4 = For Each Loop(Array:Array of Atom Model Selection Set Structures=@n2) -> (Array Element:Atom Model Selection Set Structure)
   Return Node!(
    execute=@n4.Completed
    Array of Filtered Parts Collections:Array of Atom Model Parts Collection Structures=Get Parts Collection() -> (Parts Collection:Array of Atom Model Parts Collection Structures)
   )
   n5 = Break Atom Model Parts Collection(AtomModelPartsCollection=@n1)
   n6 = "To String (Name)"(InName=Break Atom Model Selection Set!(AtomModelSelectionSet:"Atom Model Selection Set Structure (by ref)"=@n4) -> (PrimitiveIds:Set of Guid Structures, SelectionSetName:Name) {is=title K2Node_BreakStruct}.SelectionSetName)
   > n7 = Set members in Atom Model Parts Collection!(
    execute:Exec
    StructRef:"Atom Model Parts Collection Structure (by ref)"=FilterSelectionSet!(
     PartsCollectionToFilter:"Atom Model Parts Collection Structure (by ref)"=@n1
     SelectionSetName:String=@n6
     NewPartsCollectionName:String=Append!(A=@n5, B=_, C=@n6) -> (ReturnValue)
    ) -> (ReturnValue:Atom Model Parts Collection Structure)
    Pivot:Transform=@n5.Pivot
   ) -> (then:Exec, StructOut:"Atom Model Parts Collection Structure (by ref)") {is=title K2Node_SetFieldsInStruct}
   > Add(
    TargetArray:Array of Atom Model Parts Collection Structures=Get Parts Collection()
    NewItem:"Atom Model Parts Collection Structure (by ref)"=@n7
   )
   n8 = Branch(execute=@n3.else, Condition=Get Separate Selection Sets())
   > n9 = ReportProcessorWarning!(
    execute
    Message:"Text (by ref)"=MakeLiteralText(Value="LOCTEXT(\"Missing Selection Sets in Model and  Separate Selection Sets chosen. Falling back to single mesh for all.\")")
    Asset:Object Reference=@n1.Model
   ) -> (then)
   Return Node(
    execute=[@n9.then, @n8.else]
    Array of Filtered Parts Collections=Make Array!("[0]":Atom Model Parts Collection Structure=@n1) -> (Array:Array of Atom Model Parts Collection Structures)
   )
  ]}
 OnGetProcessModelTargetAssetPath {nodes [
   n1 = OnGetProcessModelTargetAssetPath!() -> (then, Model:Atom Model Object Reference, AtomModelPartsCollection:"Atom Model Parts Collection Structure (by ref)") {is=title K2Node_FunctionEntry}
   > Parent: OnGetProcessModelTargetAssetPath!(
    execute
    Model:Atom Model Object Reference=@n1
    AtomModelPartsCollection:"Atom Model Parts Collection Structure (by ref)"=@n1.AtomModelPartsCollection
   ) -> (then, ReturnValue:String) {is=title K2Node_CallParentFunction}
   > n2 = GetStaticMeshPath!(
    execute
    self
    Model:Atom Model Object Reference=@n1
    Parts Collection:"Atom Model Parts Collection Structure (by ref)"=@n1.AtomModelPartsCollection
   ) -> (then, Result:String) {is=title K2Node_CallFunction}
   > n3 = GetStaticMeshName!(
    execute
    self
    Model:Atom Model Object Reference=@n1
    Parts Collection:"Atom Model Parts Collection Structure (by ref)"=@n1.AtomModelPartsCollection
   ) -> (then, Result:String) {is=title K2Node_CallFunction}
   > Return Node!(
    execute
    ReturnValue:String=RemoveDuplicateSlashes!(InPath:String=Append!(A=@n2, B=/, C=@n3, D:String=., E:String=@n3) -> (ReturnValue)) -> (OutPath:String)
   )
  ]}
 OnGetProcessPrimitiveTargetAssetPath {nodes [
   n1 = OnGetProcessPrimitiveTargetAssetPath!() -> (
    then
    Model:Atom Model Object Reference
    Primitive:Atom Primitive Object Reference
    AtomModelPartsCollection:"Atom Model Parts Collection Structure (by ref)"
    Settings:"Atom On Process Primitive Settings Structure (by ref)"
   ) {is=title K2Node_FunctionEntry}
   > Parent: OnGetProcessPrimitiveTargetAssetPath!(
    execute
    Model:Atom Model Object Reference=@n1
    Primitive:Atom Primitive Object Reference=@n1.Primitive
    AtomModelPartsCollection:"Atom Model Parts Collection Structure (by ref)"=@n1.AtomModelPartsCollection
    Settings:"Atom On Process Primitive Settings Structure (by ref)"=@n1.Settings
   ) -> (then, ReturnValue:String) {is=title K2Node_CallParentFunction}
   n2 = Break Atom on Process Primitive Settings!(AtomOnProcessPrimitiveSettings:"Atom On Process Primitive Settings Structure (by ref)"=@n1.Settings) -> (CustomLocation:String, bSupportDecorations:bool=~) {is=title K2Node_BreakStruct}
   > Set Support Decorations(Support Decorations:bool=@n2.bSupportDecorations) -> (Output_Get:bool)
   > Set Custom Primitive Location(Custom Primitive Location:String=@n2) -> (Output_Get:String)
   > n3 = GetStaticMeshPath(Model=@n1, Parts Collection=@n1.AtomModelPartsCollection)
   > n4 = GetStaticMeshName(Model=@n1, Parts Collection=@n1.AtomModelPartsCollection)
   > Return Node(ReturnValue=RemoveDuplicateSlashes(InPath=Append(A=@n3, B=/, C=@n4, D=., E=@n4)))
  ]}
 OnPostCreateBlueprintContainer {nodes [OnPostCreateBlueprintContainer!() -> (then, BlueprintObject:Object Reference) {is=title K2Node_FunctionEntry}]}
 OnPostMeshMerge {nodes [
   n1 = OnPostMeshMerge!() -> (
    then
    PrimitiveGeometryLODs:"Atom Primitive Geometry LODs Structure (by ref)"
    Materials:Array of Material Interface Object References
    Mesh Name:String
   ) {is=title K2Node_FunctionEntry}
   > n2 = PerColorAlphaRemap!(
    execute
    self
    PrimitiveGeomLODs:Atom Primitive Geometry LODs Structure=@n1
    Materials:Array of Material Interface Object References=@n1.Materials
   ) -> (then, Out PrimitiveGeomLODs:Atom Primitive Geometry LODs Structure)
   > Return Node!(
    execute
    OutPrimitiveGeometryLODs:Atom Primitive Geometry LODs Structure=@n2
    OutMaterials:Array of Material Interface Object References=@n1.Materials
   )
  ]}
 OnProcessPartsCollection {nodes [
   n1 = OnProcessPartsCollection!() -> (then, Parts Collection:"Atom Model Parts Collection Structure (by ref)", Model:Atom Model Object Reference) {is=title K2Node_FunctionEntry}
   > Set Model(Model:Atom Model Object Reference=@n1.Model) -> (Output_Get:Atom Model Object Reference)
   > Branch(Condition=Is Not Empty(TargetArray:Array of Atom Model Part Instance Info Structures=Break Atom Model Parts Collection!(AtomModelPartsCollection=@n1) -> (Parts)))
   > Initialize Parts Collection Connectivity!(
    execute
    self
    Parts Collection:"Atom Model Parts Collection Structure (by ref)"=Get Parts Collection() -> (Parts Collection:"Atom Model Parts Collection Structure (by ref)")
    Atom Model:Atom Model Object Reference=Get Model()
   ) -> (then) {is=title K2Node_CallFunction}
   > n2 = Get Merged Parts Infos!(
    execute
    self
    AtomModelPartsCollection:"Atom Model Parts Collection Structure (by ref)"=Get PartsCollection(self=GetPartsCollectionConnectivity())
   ) -> (
    then
    Merged Mesh Part Infos:Array of Atom Merged Mesh Part Info Structures
    Baked Flex Primitives Names:Array of Strings
    Primitive Instance Data Map:Map of Atom Model Part Guid Structures to Atom Primitive Instance Data Structures
    Primitive Mesh Overrides:Map of Atom Model Part Guid Structures to Static Mesh Object References
   ) {is=title K2Node_CallFunction}
   > Set BakedFlexPrimitivesNames(BakedFlexPrimitivesNames:Array of Strings=@n2.Baked Flex Primitives Names) -> (Output_Get:Array of Strings)
   > Set PrimitiveInstanceDataMap(PrimitiveInstanceDataMap:Map of Atom Model Part Guid Structures to Atom Primitive Instance Data Structures=@n2.Primitive Instance Data Map) -> (Output_Get:Map of Atom Model Part Guid Structures to Atom Primitive Instance Data Structures)
   > Set PrimitiveMeshOverrides(PrimitiveMeshOverrides:Map of Atom Model Part Guid Structures to Static Mesh Object References=@n2.Primitive Mesh Overrides) -> (Output_Get:Map of Atom Model Part Guid Structures to Static Mesh Object References)
   > n3 = Make Merged Geometries!(execute, self, MergedMeshPartInfos:Array of Atom Merged Mesh Part Info Structures=@n2) -> (then, MergedGeometries:Map of Strings to Atom Merged Geometry Structures) {is=title K2Node_CallFunction}
   > n4 = Branch(Condition=AreAnyChildProcessorsActive!(self:Editor Atom Model Processor Object Reference) -> (ReturnValue:bool))
   > SetupChildProcessingPipeline!(
    execute
    self
    Baked Flex Primitives Names:Array of Strings=Get BakedFlexPrimitivesNames()
    Scene:"Atom Model Hierarchical Scene Structure (by ref)"=Get HierarchicalScene() -> (HierarchicalScene:Atom Model Hierarchical Scene Structure)
    Primitive Instance Data Map:Map of Atom Model Part Guid Structures to Atom Primitive Instance Data Structures=Get PrimitiveInstanceDataMap()
    Merged Geometries:Map of Strings to Atom Merged Geometry Structures=@n3
   ) -> (then)
   n5 = Sequence(execute=@n4.else)
   > n6 = Keys(TargetMap:Map of Strings to Atom Merged Geometry Structures=@n3) -> (Keys:Array of Strings)
   > n7 = For Each Loop(Array:Array of Strings=@n6) -> (Array Element:String)
   n8 = Break Atom Merged Geometry!(AtomMergedGeometry:"Atom Merged Geometry Structure (by ref)"=Find(TargetMap:Map of Strings to Atom Merged Geometry Structures=@n3, Key=@n7) -> (Value:Atom Merged Geometry Structure)) -> (GeometryLODs:Atom Primitive Geometry LODs Structure, Materials:Array of Material Interface Object References) {is=title K2Node_BreakStruct}
   > n9 = OnPostMeshMerge!(
    execute
    self
    PrimitiveGeometryLODs:"Atom Primitive Geometry LODs Structure (by ref)"=@n8
    Materials:Array of Material Interface Object References=@n8.Materials
    Mesh Name:String=@n7
   ) -> (then, OutPrimitiveGeometryLODs:Atom Primitive Geometry LODs Structure, OutMaterials:Array of Material Interface Object References) {is=title K2Node_CallFunction}
   > n10 = GetStaticMeshPath(Model=Get Model(), Parts Collection=Get PartsCollection(self=GetPartsCollectionConnectivity()))
   > n11 = GetStaticMeshName(Model=Get Model(), Parts Collection=Get PartsCollection(self=GetPartsCollectionConnectivity()))
   n12 = Get BakedFlexPrimitivesNames()
   > n13 = CreateStaticMeshFromGeometry!(
    execute
    GeometryLODs:"Atom Primitive Geometry LODs Structure (by ref)"=@n9
    NaniteHiResLOD:Atom Primitive Geometry Object Reference
    MaterialArray:Array of Material Interface Object References=@n9.OutMaterials
    Subpath:String=@n10
    StaticMeshName:String=SelectString(
     A=@n11
     B=FormatAdditionalStaticMeshName!(Base Static Mesh Name:String=@n11, Additional Mesh Name:String=@n7, Baked Flex Primitives Names:Array of Strings=@n12) -> (MeshName:String) {is=title K2Node_CallFunction}
     bPickA=IsEmpty(InString=@n7)
    )
    bPreserveExistingCollision:bool=Get Preserve Existing Collision() -> (Preserve Existing Collision:bool)
    bPreserveExistingSockets:bool=Get Preserve Existing Sockets() -> (Preserve Existing Sockets:bool)
    bGenerateMeshDistanceFields:bool=true
    TraceFlag:ECollisionTraceFlag Enum=CTF_UseSimpleAsComplex
    bSupportRaytracing:bool=true
   ) -> (then, ReturnValue:Static Mesh Object Reference)
   > Add(
    TargetMap:Map of Strings to Static Mesh Object References=Get NameMergedMeshMap() -> (NameMergedMeshMap:Map of Strings to Static Mesh Object References)
    Key=@n7
    Value:"Static Mesh Object Reference (by ref)"=@n13
   ) {is=title K2Node_CallFunction}
   > Add(
    TargetMap:Map of Strings to Atom Primitive Collision Geometry Structures=Get NameMergedMeshDetailedCollision() -> (NameMergedMeshDetailedCollision:Map of Strings to Atom Primitive Collision Geometry Structures)
    Key=@n7
    Value:"Atom Primitive Collision Geometry Structure (by ref)"=Break Atom Primitive Geometry LODs!(AtomPrimitiveGeometryLODs=@n9) -> (DetailedCollision)
   )
   > Branch(Condition=Contains Item(TargetArray:Array of Strings=Get BakedFlexPrimitivesNames(), ItemToFind:"String (by ref)"=@n7))
   n14 = To Atom Model Part Guid!(InString:String=@n7) -> (ReturnValue:Atom Model Part Guid Structure)
   > Add(
    TargetMap:Map of Atom Model Part Guid Structures to Static Mesh Object References=Get PrimitiveMeshOverrides()
    Key:"Atom Model Part Guid Structure (by ref)"=@n14
    Value:"Static Mesh Object Reference (by ref)"=@n13
   )
   > RegisterPrimitiveMeshOverrideData(Primitive Insance Data Map=Get PrimitiveInstanceDataMap(), Part Guid=@n14, Mesh Override=@n13)
   n15 = Keys(execute=@n7.Completed, TargetMap:Map of Strings to Static Mesh Object References=Get NameMergedMeshMap())
   > n16 = For Each Loop(Array=@n15)
   n17 = Find(TargetMap:Map of Strings to Static Mesh Object References=Get NameMergedMeshMap(), Key=@n16) -> (Value:Static Mesh Object Reference)
   > Add(
    TargetArray:Array of Object References=Get Processed Objects() -> (Processed Objects:Array of Object References)
    NewItem:"Object Reference (by ref)"=@n17
   ) {is=title K2Node_CallArrayFunction}
   > FinalizeMeshInternal!(
    execute
    self
    Static Mesh:Static Mesh Object Reference=@n17
    Merged Mesh Name:String=@n16
    Primitive Mesh Overrides:Map of Atom Model Part Guid Structures to Static Mesh Object References=Get PrimitiveMeshOverrides()
    Primitive Instance Data:Map of Atom Model Part Guid Structures to Atom Primitive Instance Data Structures=Get PrimitiveInstanceDataMap()
    Detailed Collision:Atom Primitive Collision Geometry Structure=Find(TargetMap:Map of Strings to Atom Primitive Collision Geometry Structures=Get NameMergedMeshDetailedCollision(), Key=@n16) -> (Value:Atom Primitive Collision Geometry Structure)
   ) -> (then) {is=title K2Node_CallFunction}
   n18 = GetIsProcessingPrimitive()
   n19 = Branch(
    execute=@n5.then_1
    Condition=AND Boolean(A=NOT Boolean(A=@n18), B=Get Generate Blueprint Container() -> (Generate Blueprint Container:bool))
   )
   > n20 = Keys(TargetMap=Get NameMergedMeshMap())
   > Add(
    TargetArray:Array of Atom Model Hierarchical Scene Node Structures=Break Atom Model Hierarchical Scene!(AtomModelHierarchicalScene:"Atom Model Hierarchical Scene Structure (by ref)"=Get HierarchicalScene()) -> (SceneNodes:Array of Atom Model Hierarchical Scene Node Structures) {is=title K2Node_BreakStruct}
    NewItem:"Atom Model Hierarchical Scene Node Structure (by ref)"=Make Atom Model Hierarchical Scene Node!(Name:String=Root, WorldTransform:Transform=MakeTransform(Rotation="0, 0, 0"), ParentIndex:int=-1, ItemNames:Array of Strings=@n20) -> (AtomModelHierarchicalSceneNode:Atom Model Hierarchical Scene Node Structure) {is=title K2Node_MakeStruct}
   )
   > n21 = GetBlueprintContainerPath!(
    execute
    self
    Model:Atom Model Object Reference=Get Model()
    Parts Collection:"Atom Model Parts Collection Structure (by ref)"=Get PartsCollection(self=GetPartsCollectionConnectivity())
   ) -> (then, Result:String) {is=title K2Node_CallFunction}
   > n22 = GetBlueprintContainerName!(
    execute
    self
    Model:Atom Model Object Reference=Get Model()
    Parts Collection:"Atom Model Parts Collection Structure (by ref)"=Get PartsCollection(self=GetPartsCollectionConnectivity())
   ) -> (then, Result:String) {is=title K2Node_CallFunction}
   > n23 = CreateBlueprintContainerFromAtomModelScene!(
    execute
    AssetPath:String=@n21
    AssetName:String=@n22
    ParentClass:Object Class Reference=Get Blueprint Container Base Class() -> (Blueprint Container Base Class:Actor Class Reference)
    Scene:"Atom Model Hierarchical Scene Structure (by ref)"=Get HierarchicalScene()
    StaticMeshes:Map of Strings to Static Mesh Object References=Get NameMergedMeshMap()
   ) -> (then, ReturnValue:Blueprint Object Reference)
   > OnPostCreateBlueprintContainer!(execute, self, BlueprintObject:Object Reference=@n23) -> (then) {is=title K2Node_CallFunction}
   > n24 = Add(TargetArray:Array of Object References=Get Processed Objects(), NewItem:"Object Reference (by ref)"=@n23)
   Return Node!(
    execute=[@n24.then, @n19.else]
    Result:Atom Processor Result Structure=Make Atom Processor Result!(bSuccess:bool=True, ProcessedObjects:Array of Object References=Get Processed Objects(), SharedAssets:Array of Object References) -> (AtomProcessorResult:Atom Processor Result Structure) {is=title K2Node_MakeStruct}
   )
  ]}
 OnProcessPrimitive {nodes [
   n1 = OnProcessPrimitive!() -> (
    then
    DummyModel:Atom Model Object Reference
    Primitive:Atom Primitive Object Reference
    AtomModelPartsCollection:"Atom Model Parts Collection Structure (by ref)"
    Settings:"Atom On Process Primitive Settings Structure (by ref)"
   ) {is=title K2Node_FunctionEntry}
   n2 = Break Atom on Process Primitive Settings(AtomOnProcessPrimitiveSettings=@n1.Settings) -> (bSupportDecorations=~)
   > Set Support Decorations(Support Decorations=@n2.bSupportDecorations)
   > Set Custom Primitive Location(Custom Primitive Location=@n2)
   > n3 = Parent: OnProcessPrimitive!(
    execute
    DummyModel:Atom Model Object Reference=@n1
    Primitive:Atom Primitive Object Reference=@n1.Primitive
    AtomModelPartsCollection:"Atom Model Parts Collection Structure (by ref)"=@n1.AtomModelPartsCollection
    Settings:"Atom On Process Primitive Settings Structure (by ref)"=@n1.Settings
   ) -> (then, ReturnValue:Atom Processor Result Structure) {is=title K2Node_CallParentFunction}
   > Return Node!(execute, ReturnValue:Atom Processor Result Structure=@n3)
  ]}
 PerColorAlphaRemap {nodes [
   n1 = PerColorAlphaRemap!() -> (then, PrimitiveGeomLODs:Atom Primitive Geometry LODs Structure, Materials:Array of Material Interface Object References) {is=title K2Node_FunctionEntry}
   > Set LocalGeomLODs(LocalGeomLODs:Array of Atom Primitive Geometry Object References=Break Atom Primitive Geometry LODs!(AtomPrimitiveGeometryLODs=@n1) -> (GeometryLODs:Array of Atom Primitive Geometry Object References, DetailedCollision)) -> (Output_Get:Array of Atom Primitive Geometry Object References)
   > n2 = GetMeshMergingSettings()
   n3 = Break Atom Mesh Merging Settings!(AtomMeshMergingSettings:"Atom Mesh Merging Settings Structure (by ref)"=@n2) -> (
    RequestedLODsSettings_NumLODs:int
    RequestedLODsSettings_LODGroup:Name
    CoarseLODsSettings_CoarseLODStrategy:EAtomMeshMergingCoarseApproximationStrategy Enum
    CoarseLODsSettings_NumCoarseLODs:int
    CoarseLODsSettings_CoarseLODBaseTolerance:float=~
    CoarseLODsSettings_CoarseLODMaxTriCountBase:int
    CoarseLODsSettings_CoarseApproximationDetailSize:float=~
    CoplanarMergeSettings_bMergeCoplanarFaces:bool=~
    CoplanarMergeSettings_MergeCoplanarFacesStartLOD:int
   ) {is=title K2Node_BreakStruct}
   n4 = Get LocalGeomLODs() -> (LocalGeomLODs:Array of Atom Primitive Geometry Object References)
   > n5 = Set NumLODs(NumLODs:int="Min (Integer)"(A=Length(TargetArray:Array of Atom Primitive Geometry Object References=@n4), B=@n3) -> (ReturnValue=0)) -> (Output_Get:int=0)
   > n6 = Set NumCoarseLODs(NumCoarseLODs:int=@n3.CoarseLODsSettings_NumCoarseLODs) -> (Output_Get:int=0)
   n7 = int - int(A=@n5, B=@n6)
   > Set StartLODIndex(StartLODIndex:int=SelectInt(A=@n3.CoplanarMergeSettings_MergeCoplanarFacesStartLOD, B="Max (Integer)"(A=@n7, B=0), bPickA=@n3.CoplanarMergeSettings_bMergeCoplanarFaces)) -> (Output_Get:int=0)
   > n8 = Branch(Condition=UseMaterialFactory())
   > n9 = For Each Loop(Array:Array of Material Interface Object References=Get Materials()) -> (Array Element:Material Interface Object Reference)
   > n10 = GetMaterialFeatures!(
    execute
    self:Atom Material Factory Object Reference=GetMaterialFactory() {is=title K2Node_CallFunction}
    Material:Material Interface Object Reference=@n9
   ) -> (then, ReturnValue:Atom Material Features Structure)
   > Branch(Condition="Equal (Enum)"(
     A:EAtomMaterialUVFeature Enum=Break Atom Material Features!(AtomMaterialFeatures:"Atom Material Features Structure (by ref)"=@n10) -> (
      TechnicNormalMap:Texture Object Reference
      TechnicMasks:Texture Object Reference
      LookUpTableAtlasTexture:Texture Object Reference
      UVFeature:EAtomMaterialUVFeature Enum
      TranslucencyType:EAtomMaterialTranslucencyType Enum
     ) {is=title K2Node_BreakStruct}.UVFeature
     B:EAtomMaterialUVFeature Enum=AnalyticalCommonParts
    ))
   > Add(
    TargetArray:Array of Integers=Get AnalyticalCommonPartMaterialIndices() -> (AnalyticalCommonPartMaterialIndices:Array of Integers)
    NewItem:"Integer (by ref)"=@n9.Array Index
   )
   n11 = For Loop(
    execute=[@n8.else, @n9.Completed]
    FirstIndex=Get StartLODIndex() -> (StartLODIndex:int=0)
    LastIndex=int - int(A=Get NumLODs() -> (NumLODs:int=0))
   )
   n12 = Set members in Atom Primitive Geometry LODs!(
    execute=@n11.Completed
    StructRef=Copy(Item:Atom Primitive Geometry LODs Structure=Get PrimitiveGeomLODs() -> (PrimitiveGeomLODs:Atom Primitive Geometry LODs Structure)) -> (ReturnValue:Atom Primitive Geometry LODs Structure)
    GeometryLODs:Array of Atom Primitive Geometry Object References=Get LocalGeomLODs()
   ) -> (then, StructOut)
   > Return Node!(execute, Out PrimitiveGeomLODs:Atom Primitive Geometry LODs Structure=@n12)
   Set LODIndex(execute=@n11.LoopBody, LODIndex:int=@n11) -> (Output_Get:int=0)
   > n13 = ConvertAtomPrimitiveGeometryToDynamicMesh!(
    execute
    AtomPrimitiveGeometry:Atom Primitive Geometry Object Reference="Get (a copy)"(Array=Get LocalGeomLODs(), Dimension 1=@n11)
    DynamicMesh:Dynamic Mesh Object Reference
   ) -> (then, ReturnValue:Dynamic Mesh Object Reference)
   > n14 = GetAllTriangleIDs!(execute, TargetMesh:Dynamic Mesh Object Reference=@n13) -> (then, TriangleIDList:Index List Structure, bHasTriangleIDGaps:bool, ReturnValue:Dynamic Mesh Object Reference)
   > n15 = For Loop(LastIndex=GetIndexListLastIndex!(IndexList:Index List Structure=@n14) -> (ReturnValue:int=0))
   n16 = ConvertDynamicMeshToAtomPrimitiveGeometry!(
    execute=@n15.Completed
    DynamicMesh:Dynamic Mesh Object Reference=@n14.ReturnValue
    AtomPrimitiveGeometry:Atom Primitive Geometry Object Reference
   ) -> (then, ReturnValue:Atom Primitive Geometry Object Reference)
   > Set Array Elem(TargetArray=Get LocalGeomLODs(), Index=Get LODIndex() -> (LODIndex:int=0), Item=@n16, bSizeToFit=false)
   n17 = GetIndexListItem!(IndexList:Index List Structure=@n14, Index:int=@n15) -> (bIsValidIndex:bool, ReturnValue:int=0)
   n18 = Branch(
    execute=@n15.LoopBody
    Condition=Contains Item(
     TargetArray:Array of Integers=Get AnalyticalCommonPartMaterialIndices()
     ItemToFind:"Integer (by ref)"=Get Material ID Of Triangle!(TargetMesh:Dynamic Mesh Object Reference=@n14.ReturnValue, TriangleID:int=@n17.ReturnValue) -> (bIsValidTriangle:bool, ReturnValue:int=0) {is=title K2Node_CallFunction}.ReturnValue
    )
   )
   n19 = GetTriangleVertexColors!(execute=@n18.else, TargetMesh:Dynamic Mesh Object Reference=@n14.ReturnValue, TriangleID:int=@n17.ReturnValue) -> (
    then
    Color1:Linear Color Structure
    Color2:Linear Color Structure
    Color3:Linear Color Structure
    bTriHasValidVertexColors:bool
    ReturnValue:Dynamic Mesh Object Reference
   )
   > n20 = Sequence()
   n21 = ConvertIndexArrayToMeshSelection!(
    execute=@n20.then_1
    TargetMesh:Dynamic Mesh Object Reference=@n19.ReturnValue
    IndexArray:Array of Integers=Make Array("[0]":int=@n17.ReturnValue) -> (Array:Array of Integers)
    SelectionType:EGeometryScriptMeshSelectionType Enum=Triangles
   ) -> (then, Selection:Mesh Selection Structure, ReturnValue:Dynamic Mesh Object Reference)
   > SetMeshSelectionVertexColor!(
    execute
    TargetMesh:Dynamic Mesh Object Reference=@n21.ReturnValue
    Selection:Mesh Selection Structure=@n21
    Color:Linear Color Structure=Get OutputColor() -> (OutputColor:Linear Color Structure)
    Flags:Geometry Script Color Flags Structure
    bCreateColorSeam:bool=false
   ) -> (then, ReturnValue:Dynamic Mesh Object Reference)
   n22 = Get NumCoarseLODs() -> (NumCoarseLODs:int=0)
   n23 = Get NumLODs() -> (NumLODs=0)
   n24 = Branch(execute=@n20.then_0, Condition="integer >= integer"(A=Get LODIndex() -> (LODIndex=0), B=int - int(A=@n23, B=@n22)))
   n25 = BreakColor(InColor=@n19)
   > Set OutputColor(OutputColor:Linear Color Structure=MakeColor!(R:float=@n25, G:float=@n25.G, B:float=@n25.B, A:float=0) -> (ReturnValue:Linear Color Structure)) -> (Output_Get:Linear Color Structure)
   n26 = float * float(A=Dot Product!(A:Vector=MakeVector(X=@n25, Y=@n25.G, Z=@n25.B), B:Vector="0.3,0.59,0.11") -> (ReturnValue:double), B=43758.5453) {is=title K2Node_PromotableOperator}
   Set OutputColor(execute=@n24.else, OutputColor=MakeColor(R=@n25, G=@n25.G, B=@n25.B, A=Fraction!(A:double=@n26) -> (ReturnValue:double)))
  ]}
 PreProcessorPipeline {nodes [
   n1 = PreProcessorPipeline!() -> (then, AtomModel:Atom Model Object Reference, PartsCollection:"Atom Model Parts Collection Structure (by ref)") {is=title K2Node_FunctionEntry}
   > Parent: PreProcessorPipeline!(execute, AtomModel:Atom Model Object Reference=@n1, PartsCollection:"Atom Model Parts Collection Structure (by ref)"=@n1.PartsCollection) -> (then) {is=title K2Node_CallParentFunction}
   > n2 = Branch(Condition=UseMaterialFactory())
   > n3 = Construct Object from Class(Class=GetMaterialFactoryClass!() -> (AtomMaterialFactoryClass:Atom Material Factory Class Reference) {is=title K2Node_CallFunction}) -> (ReturnValue:Atom Material Factory Object Reference)
   > n4 = Set Material Factory(Material Factory:Atom Material Factory Object Reference=@n3) -> (Output_Get:Atom Material Factory Object Reference)
   n5 = Branch(execute=[@n4.then, @n2.else], Condition=Get Use Geometry Style Settings() -> (Use Geometry Style Settings:bool))
   Return Node!(execute=[@n5.then, @n6.then])
   Initialize Geometry Style Settings(execute=@n5.else) {is=title K2Node_CallFunction}
   > n6 = Set Use Geometry Style Settings(Use Geometry Style Settings:bool=true) -> (Output_Get:bool)
  ]}
 ProcessUncommonParts {nodes [
   ProcessUncommonParts!() -> (
    then
    Part Instance Info:"Atom Model Part Instance Info Structure (by ref)"
    Merged Mesh Part Info:"Atom Merged Mesh Part Info Structure (by ref)"
    Merged Mesh Part Infos:Array of Atom Merged Mesh Part Info Structures
   ) {is=title K2Node_FunctionEntry}
   > n1 = Sequence()
   > Branch(Condition=Get Generate Nanite Mesh())
   n2 = Break Atom Model Part Instance Info!(AtomModelPartInstanceInfo=Get Part Instance Info()) -> (Primitive, Transforms:Array of Transforms)
   n3 = Break Atom Processor Geometry Style Settings!(AtomProcessorGeometryStyleSettings=Get Nanite Geometry Style Settings()) -> (
    GeometryStylePreferences_4_#1
    PreferredGeometryLOD_30_#5
    ShouldBevel_8_#6:bool=~
    BevelWidth_11_#7:double=~
    BevelAngleThreshold_13_#8:double=~
    UseImprovedBevelAlgorithm_37_#3=~
    OnlyBevelBlock-LikePrimitives_39_#4=~
    NumberofSubdivisionsToApply_16_#9:int
    FakeBevel_23_#10:bool=~
    FakeUnderside_25_#11:bool=~
    AnalyticFakeUnderside_32_#12:bool=~
    ForceFlatCap_35_#13:bool=~
    FakeTechnicHole_41_#2=~
    AnalyticCommonParts_27_#14:bool=~
   )
   > n4 = For Each Loop(Array:Array of Atom Primitive Geometry And Transform Structures=GetUncommonPartsGeometry!(
     self:Atom Primitive Geometry Container Object Reference=GetGeometryContainerForBestExportStyle(self=@n2, GeometryStylePriorityList=@n3, LODIndex=@n3.PreferredGeometryLOD_30_#5)
     ColorChangeIndex:int=-1
    ) -> (ReturnValue:Array of Atom Primitive Geometry And Transform Structures)) -> (Array Element:Atom Primitive Geometry And Transform Structure)
   > n5 = GetGeometryLODsForUncommonPart!(
    execute
    self
    Part Info:Atom Model Part Instance Info Structure=Get Part Instance Info()
    Uncommon Part Geometry:Atom Primitive Geometry And Transform Structure=@n4
    Is Nanite:bool=true
   ) -> (then, Geometry LODs:Atom Primitive Geometry LODs Structure, Merged Mesh Transform:Transform) {is=title K2Node_CallFunction}
   > n6 = GetVertexColorForPart(Part Info=Get Part Instance Info(), Color Change Index=0) {is=title K2Node_CallFunction}
   > n7 = Set members in Atom Merged Mesh Part Info!(
    execute
    StructRef=Copy(Item:Atom Merged Mesh Part Info Structure=Get Merged Mesh Part Info() -> (Merged Mesh Part Info:"Atom Merged Mesh Part Info Structure (by ref)")) -> (ReturnValue:Atom Merged Mesh Part Info Structure)
    GeometryLODs=@n5
    Transform:Transform=@n5.Merged Mesh Transform
    bHasConstantOverrideVertexColor:bool=True
    VertexColor=@n6
    MeshDetailLevel:EAtomMergedMeshPartDetailLevel Enum=Decorative
   ) -> (then, StructOut)
   > Add(
    TargetArray:Array of Atom Merged Mesh Part Info Structures=Get Merged Mesh Part Infos() -> (Merged Mesh Part Infos:Array of Atom Merged Mesh Part Info Structures)
    NewItem:"Atom Merged Mesh Part Info Structure (by ref)"=@n7
   )
   n8 = Break Atom Processor Geometry Style Settings(AtomProcessorGeometryStyleSettings=Get Geometry Style Settings For LOD(LOD=0)) -> (
    ShouldBevel_8_#6=~
    BevelWidth_11_#7=~
    BevelAngleThreshold_13_#8=~
    UseImprovedBevelAlgorithm_37_#3=~
    OnlyBevelBlock-LikePrimitives_39_#4=~
    FakeBevel_23_#10=~
    FakeUnderside_25_#11=~
    AnalyticFakeUnderside_32_#12=~
    ForceFlatCap_35_#13=~
    FakeTechnicHole_41_#2=~
    AnalyticCommonParts_27_#14=~
   )
   n9 = Break Atom Model Part Instance Info(AtomModelPartInstanceInfo=Get Part Instance Info())
   n10 = GetGeometryContainerForBestExportStyle(self=@n9, GeometryStylePriorityList=@n8, LODIndex=@n8.PreferredGeometryLOD_30_#5)
   Set GeometryContainer(execute=@n1.then_1, GeometryContainer:Atom Primitive Geometry Container Object Reference=@n10) -> (Output_Get:Atom Primitive Geometry Container Object Reference)
   > n11 = For Each Loop(Array=GetUncommonPartsGeometry(self=@n10, ColorChangeIndex=-1))
   > n12 = GetGeometryLODsForUncommonPart(Part Info=Get Part Instance Info(), Uncommon Part Geometry=@n11, Is Nanite=false)
   > n13 = ScaleUncommonParts!(
    execute
    self
    GeometryContainer:Atom Primitive Geometry Container Object Reference=Get GeometryContainer() -> (GeometryContainer:Atom Primitive Geometry Container Object Reference)
    GeometryLODs:Atom Primitive Geometry LODs Structure=@n12
   ) -> (then, OutGeometryLODs:Atom Primitive Geometry LODs Structure)
   > n14 = GetVertexColorForPart(Part Info=Get Part Instance Info(), Color Change Index=0)
   > n15 = Set members in Atom Merged Mesh Part Info(
    StructRef=Copy(Item=Get Merged Mesh Part Info())
    GeometryLODs=@n13
    Transform=@n12.Merged Mesh Transform
    bHasConstantOverrideVertexColor=True
    VertexColor=@n14
    MeshDetailLevel=Decorative
   )
   > Add(TargetArray=Get Merged Mesh Part Infos(), NewItem=@n15)
   "Get (a ref)"(Array:Array of Transforms=@n2.Transforms, Dimension 1=0) -> (Output:"Transform (by ref)")
   "Get (a ref)"(Array=@n9.Transforms, Dimension 1=0)
  ]}
 RegisterPrimitiveMaterialData {nodes [
   RegisterPrimitiveMaterialData!() -> (
    then
    Primitive Insance Data Map:Map of Atom Model Part Guid Structures to Atom Primitive Instance Data Structures
    Part Info:"Atom Model Part Instance Info Structure (by ref)"
    Mesh Part Info:"Atom Merged Mesh Part Info Structure (by ref)"
   ) {is=title K2Node_FunctionEntry}
   n1 = Find(
    TargetMap:Map of Atom Model Part Guid Structures to Atom Primitive Instance Data Structures=Get Primitive Insance Data Map() -> (Primitive Insance Data Map:Map of Atom Model Part Guid Structures to Atom Primitive Instance Data Structures)
    Key:"Atom Model Part Guid Structure (by ref)"=Break Atom Model Part Instance Info!(AtomModelPartInstanceInfo=Get Part Info() -> (Part Info:"Atom Model Part Instance Info Structure (by ref)")) -> (PartGuid:Atom Model Part Guid Structure)
   ) -> (Value:Atom Primitive Instance Data Structure)
   > Branch(Condition=@n1.ReturnValue)
   n2 = Break Atom Merged Mesh Part Info!(AtomMergedMeshPartInfo=Get Mesh Part Info() -> (Mesh Part Info:"Atom Merged Mesh Part Info Structure (by ref)")) -> (Materials:Array of Material Interface Object References, PreventMergingMaterials, MaterialFeatures, bAllowSimplification=~)
   > n3 = For Each Loop(Array:Array of Material Interface Object References=@n2) -> (Array Element:Material Interface Object Reference)
   n4 = Set members in Atom Primitive Instance Data!(
    execute:Exec=@n3.Completed
    StructRef:"Atom Primitive Instance Data Structure (by ref)"=@n1
    MaterialAssignments:Array of Atom Primitive Material Assignment Structures=Get MaterialAssigments() -> (MaterialAssigments:Array of Atom Primitive Material Assignment Structures)
   ) -> (then:Exec, StructOut:"Atom Primitive Instance Data Structure (by ref)") {is=title K2Node_SetFieldsInStruct}
   > Add(
    TargetMap:Map of Atom Model Part Guid Structures to Atom Primitive Instance Data Structures=Get Primitive Insance Data Map()
    Key=Break Atom Model Part Instance Info(AtomModelPartInstanceInfo=Get Part Info())
    Value:"Atom Primitive Instance Data Structure (by ref)"=@n4
   ) {is=title K2Node_CallFunction}
   n5 = Branch(
    execute=@n3.LoopBody
    Condition=Is Valid Index(TargetArray:Array of Atom Material Features Structures=@n2.MaterialFeatures, IndexToTest=@n3.Array Index)
   )
   n6 = Break Atom Material Override(AtomMaterialOverride=Break Atom Material Features!(AtomMaterialFeatures="Get (a ref)"(Array:Array of Atom Material Features Structures=@n2.MaterialFeatures, Dimension 1=@n3.Array Index) -> (Output:"Atom Material Features Structure (by ref)")) -> (TechnicNormalMap, TechnicMasks, LookUpTableAtlasTexture, TranslucencyType, MaterialOverride:Atom Material Override Structure).MaterialOverride) -> (bManualColorSource=~)
   n7 = Make Soft Reference!(Input:Material Interface Object Reference=@n3) -> (Output:Material Interface Soft Object Reference)
   > Add(
    TargetArray:Array of Atom Primitive Material Assignment Structures=Get MaterialAssigments()
    NewItem:"Atom Primitive Material Assignment Structure (by ref)"=Make Atom Primitive Material Assignment!(
     SlotName:Name=None
     Material:Material Interface Soft Object Reference=@n7
     bIsCustomMaterialOverride:bool=IsValid(Object=@n6)
     bHasManualColorSource:bool=@n6.bManualColorSource
    ) -> (AtomPrimitiveMaterialAssignment:Atom Primitive Material Assignment Structure) {is=title K2Node_MakeStruct}
   ) {is=title K2Node_CallArrayFunction}
   Add(
    execute=@n5.else
    TargetArray=Get MaterialAssigments()
    NewItem=Make Atom Primitive Material Assignment!(SlotName=None, Material=@n7) -> (AtomPrimitiveMaterialAssignment)
   )
  ]}
 RegisterPrimitiveMeshOverrideData {nodes [
   RegisterPrimitiveMeshOverrideData!() -> (
    then
    Primitive Insance Data Map:Map of Atom Model Part Guid Structures to Atom Primitive Instance Data Structures
    Part Guid:Atom Model Part Guid Structure
    Mesh Override:Static Mesh Object Reference
   ) {is=title K2Node_FunctionEntry}
   n1 = Find(TargetMap=Get Primitive Insance Data Map(), Key=Get Part Guid() -> (Part Guid:Atom Model Part Guid Structure))
   > Branch(Condition=@n1.ReturnValue)
   > n2 = Set members in Atom Primitive Instance Data!(
    execute
    StructRef=@n1
    MeshOverride:Static Mesh Soft Object Reference=Make Soft Reference(Input:Static Mesh Object Reference=Get Mesh Override() -> (Mesh Override:Static Mesh Object Reference)) -> (Output:Static Mesh Soft Object Reference)
   ) -> (then, StructOut)
   > Add(TargetMap=Get Primitive Insance Data Map(), Key=Get Part Guid(), Value=@n2) {is=title K2Node_CallFunction}
  ]}
 ScaleUncommonParts {nodes [
   ScaleUncommonParts!() -> (then, GeometryContainer:Atom Primitive Geometry Container Object Reference, GeometryLODs:Atom Primitive Geometry LODs Structure) {is=title K2Node_FunctionEntry}
   n1 = Break Atom Processor Geometry Style Settings(AtomProcessorGeometryStyleSettings=Get Geometry Style Settings For LOD(LOD=1)) -> (
    ShouldBevel_8_#6=~
    BevelWidth_11_#7=~
    BevelAngleThreshold_13_#8=~
    UseImprovedBevelAlgorithm_37_#3=~
    OnlyBevelBlock-LikePrimitives_39_#4=~
    FakeBevel_23_#10=~
    FakeUnderside_25_#11=~
    AnalyticFakeUnderside_32_#12=~
    ForceFlatCap_35_#13=~
    FakeTechnicHole_41_#2=~
    AnalyticCommonParts_27_#14=~
   )
   > n2 = Branch(Condition=OR Boolean(A=@n1.FakeUnderside_25_#11, B=@n1.ForceFlatCap_35_#13))
   > n3 = Sequence!(execute) -> (then_0, then_1, then_2:Exec)
   > n4 = For Each Loop(Array:Array of Atom Primitive Geometry And Transform Structures=GetCapsGeometry!(self:Atom Primitive Geometry Container Object Reference=Get GeometryContainer(), ColorChangeIndex:int=-1) -> (ReturnValue:Array of Atom Primitive Geometry And Transform Structures)) -> (Array Element:Atom Primitive Geometry And Transform Structure)
   > n5 = ConvertAtomPrimitiveGeometryToDynamicMesh(AtomPrimitiveGeometry=Break Atom Primitive Geometry and Transform(AtomPrimitiveGeometryAndTransform=@n4))
   > Add(
    TargetArray:Array of Box Structures=Get CapBounds() -> (CapBounds:Array of Box Structures)
    NewItem:"Box Structure (by ref)"=GetMeshBoundingBox!(TargetMesh:Dynamic Mesh Object Reference=@n5) -> (ReturnValue:Box Structure)
   ) {is=title K2Node_CallArrayFunction}
   n6 = Get GeometryLODs() -> (GeometryLODs:Atom Primitive Geometry LODs Structure)
   n7 = ConvertAtomPrimitiveGeometryToDynamicMesh(
    execute=@n3.then_1
    AtomPrimitiveGeometry="Get (a copy)"(
     Array=Break Atom Primitive Geometry LODs!(AtomPrimitiveGeometryLODs=@n6) -> (
      GeometryLODs
      GeometryLODsMaterials:Array of Atom Primitive Geometry Material Assignments Structures
      HiResNaniteGeomery:Atom Primitive Geometry Object Reference
      HiResNaniteGeomeryMaterials:Atom Primitive Geometry Material Assignments Structure
      Collision:Atom Primitive Collision Geometry Structure
      DetailedCollision
     )
     Dimension 1=1
    )
   )
   > n8 = Set DynMesh(DynMesh:Dynamic Mesh Object Reference=@n7) -> (Output_Get:Dynamic Mesh Object Reference)
   > n9 = Sequence()
   n10 = Break Box!(Box:"Box Structure (by ref)"=GetMeshBoundingBox(TargetMesh=@n8)) -> (Min:Vector=~, Max:Vector=~, IsValid:bool=~) {is=title K2Node_BreakStruct}
   n11 = BreakVector(InVec=Vector_GetAbs!(A:Vector=vector - vector(A=@n10, B=@n10.Max)) -> (ReturnValue:Vector))
   > n12 = Branch(Condition=AND Boolean(A="float > float"(A=@n11, B=@n11.Y), B="float > float"(A=@n11, B=@n11.Z)))
   > Set Scale Axis(Scale Axis:Vector="1,0,0") -> (Output_Get:Vector)
   n13 = Branch(execute=@n12.else, Condition=AND Boolean(A="float > float"(A=@n11.Y, B=@n11.Z), B="float > float"(A=@n11, B=@n11.Y)))
   > Set Scale Axis(Scale Axis="0,1,0")
   Set Scale Axis(execute=@n13.else, Scale Axis="0,0,1")
   n14 = Break Box(Box=GetMeshBoundingBox(TargetMesh=Get DynMesh() -> (DynMesh:Dynamic Mesh Object Reference))) -> (Min=~, Max=~, IsValid=~)
   n15 = Get Scale Axis() -> (Scale Axis:Vector)
   n16 = Dot Product(A=@n14.Max, B=@n15)
   n17 = Break Box(Box="Get (a copy)"(Array:Array of Box Structures=Get CapBounds()) -> (Output:Box Structure)) -> (Min=~, Max=~, IsValid=~)
   n18 = Get Scale Axis()
   n19 = Dot Product(A=@n17, B=@n18)
   n20 = Dot Product(A=@n14, B=@n15)
   n21 = Dot Product(A=@n17.Max, B=@n18)
   Branch(execute=@n9.then_1, Condition=OR Boolean(A="float > float"(A=@n16, B=@n19), B="float < float"(A=@n20, B=@n21)))
   > PrintString(InString="pin is interescting cap ")
   > Set PartIntersecting(PartIntersecting:bool=true) -> (Output_Get:bool)
   n22 = Branch(execute=@n3.then_2, Condition=Get PartIntersecting() -> (PartIntersecting:bool))
   n23 = Break Atom Primitive Geometry LODs(AtomPrimitiveGeometryLODs=Get GeometryLODs())
   > Add(
    TargetArray:Array of Atom Primitive Geometry Object References=Get AdjustedGeomLODs() -> (AdjustedGeomLODs:Array of Atom Primitive Geometry Object References)
    NewItem:"Atom Primitive Geometry Object Reference (by ref)"="Get (a copy)"(Array:Array of Atom Primitive Geometry Object References=@n23) -> (Output:Atom Primitive Geometry Object Reference)
   )
   > n24 = For Loop(FirstIndex=1, LastIndex=int - int(A=Length(TargetArray=@n23)))
   > n25 = ConvertAtomPrimitiveGeometryToDynamicMesh(AtomPrimitiveGeometry="Get (a copy)"(Array=@n23, Dimension 1=@n24))
   > n26 = ScaleMesh!(
    execute
    TargetMesh:Dynamic Mesh Object Reference=@n25
    Scale:Vector="0,0,0"
    ScaleOrigin:Vector="0,0,0"
    bFixOrientationForNegativeScale:bool=true
   ) -> (then, ReturnValue:Dynamic Mesh Object Reference)
   > n27 = ConvertDynamicMeshToAtomPrimitiveGeometry(DynamicMesh=@n26)
   > Add(TargetArray=Get AdjustedGeomLODs(), NewItem=@n27)
   n28 = Set members in Atom Primitive Geometry LODs(execute=@n24.Completed, StructRef=Get GeometryLODs(), GeometryLODs=Get AdjustedGeomLODs())
   > Return Node!(execute, OutGeometryLODs:Atom Primitive Geometry LODs Structure=@n28)
   Return Node(execute=@n22.else, OutGeometryLODs=Get GeometryLODs())
   Return Node(execute=@n2.else, OutGeometryLODs=Get GeometryLODs())
   Break Box(Box=GetMeshBoundingBox(TargetMesh=Get DynMesh())) -> (Min=~, Max=~, IsValid=~)
   n29 = Get Scale Axis()
   vector + vector(A=vector - vector(A="1,1,1", B=@n29), B=vector * vector(A=@n29, B=0.5) -> (ReturnValue=~))
  ]}
 SetCommonPartMeshName {nodes [
   n1 = SetCommonPartMeshName!() -> (
    then
    Merged Mesh Part Info:"Atom Merged Mesh Part Info Structure (by ref)"
    Atom Model Part Info:"Atom Model Part Instance Info Structure (by ref)"
    Common Part Index:int
    Common Part Type:"EAtomCommonPartType Enum (by ref)"=None
   ) {is=title K2Node_FunctionEntry}
   > AppendGroupFilterModeToMeshName!(
    execute
    self
    PartInfo:"Atom Merged Mesh Part Info Structure (by ref)"=@n1
    PartInstanceInfo:"Atom Model Part Instance Info Structure (by ref)"=@n1.Atom Model Part Info
   ) -> (then) {is=title K2Node_CallFunction}
   > Branch(Condition=ShouldSplitTransparentPart(MergedMeshPartInfo=@n1))
   n2 = Break Atom Merged Mesh Part Info!(AtomMergedMeshPartInfo=@n1) -> (
    GeometryLODs:Atom Primitive Geometry LODs Structure
    bHasConstantOverrideVertexColor:bool=~
    PreventMergingMaterials
    MaterialFeatures
    bAllowSimplification=~
    bIsOccluder:bool=~
    MeshName:String
   )
   > n3 = Set members in Atom Primitive Geometry LODs!(execute, StructRef=@n2, HiResNaniteGeomery:Atom Primitive Geometry Object Reference) -> (then, StructOut)
   > n4 = Branch(Condition=IsEmpty(InString=@n2.MeshName))
   > Set members in Atom Merged Mesh Part Info!(execute, StructRef=@n1, GeometryLODs=@n3, MeshName:String=GetTransparentMeshSuffix() {is=title K2Node_CallFunction}) -> (then, StructOut)
   Set members in Atom Merged Mesh Part Info(execute=@n4.else, StructRef=@n1, GeometryLODs=@n3, MeshName=Append!(A=@n2.MeshName, B=_, C=GetTransparentMeshSuffix()) -> (ReturnValue))
  ]}
 SetDecorationMaterialParameters {nodes [
   n1 = SetDecorationMaterialParameters!() -> (
    then
    Material Instance:Material Instance Constant Object Reference
    Decoration Enable Param Name:String
    Texture Param Name:String
    UDIM Param Name:String
    Texture:Texture Object Reference
    UDIM:int
   ) {is=title K2Node_FunctionEntry}
   > SetMaterialInstanceStaticSwitchParameterValue(
    Instance=@n1
    ParameterName=String To Name(InString=@n1.Decoration Enable Param Name)
    Value=true
    Association=GlobalParameter
    bUpdateMaterialInstance=true
   )
   > SetMaterialInstanceTextureParameterValue!(
    execute
    Instance:Material Instance Constant Object Reference=@n1
    ParameterName:Name=String To Name(InString=@n1.Texture Param Name)
    Value:Texture Object Reference=@n1.Texture
    Association:EMaterialParameterAssociation Enum=GlobalParameter
   ) -> (then, ReturnValue:bool)
   > SetMaterialInstanceScalarParameterValue!(
    execute
    Instance:Material Instance Constant Object Reference=@n1
    ParameterName:Name=String To Name(InString=@n1.UDIM Param Name)
    Value:float="To Float (Integer)"(InInt=@n1.UDIM)
    Association:EMaterialParameterAssociation Enum=GlobalParameter
   ) -> (then, ReturnValue:bool)
   > UpdateMaterialInstance!(execute, Instance:Material Instance Constant Object Reference=@n1) -> (then)
  ]}
 SetNaniteCommonPartInfo {nodes [SetNaniteCommonPartInfo!() -> (
    then
    Merged Mesh Part Info:"Atom Merged Mesh Part Info Structure (by ref)"
    Shell Merged Mesh Part Info:"Atom Merged Mesh Part Info Structure (by ref)"
    Atom Model Part Info:"Atom Model Part Instance Info Structure (by ref)"
   ) {is=title K2Node_FunctionEntry}]}
 SetPartMeshName {nodes [
   n1 = SetPartMeshName!() -> (
    then
    Merged Mesh Part Info:"Atom Merged Mesh Part Info Structure (by ref)"
    Atom Model Part Info:"Atom Model Part Instance Info Structure (by ref)"
   ) {is=title K2Node_FunctionEntry}
   > AppendGroupFilterModeToMeshName(PartInfo=@n1, PartInstanceInfo=@n1.Atom Model Part Info)
   > Branch(Condition=ShouldSplitTransparentPart(MergedMeshPartInfo=Get Merged Mesh Part Info()))
   n2 = Break Atom Merged Mesh Part Info!(AtomMergedMeshPartInfo=Get Merged Mesh Part Info()) -> (PreventMergingMaterials, bAllowSimplification=~, MeshName)
   > n3 = Branch(Condition=IsEmpty(InString=@n2.MeshName))
   > Set members in Atom Merged Mesh Part Info!(execute, StructRef=@n1, MeshName=GetTransparentMeshSuffix()) -> (then, StructOut)
   Set members in Atom Merged Mesh Part Info(execute=@n3.else, StructRef=@n1, MeshName=Append(A=@n2.MeshName, B=_, C=GetTransparentMeshSuffix()))
  ]}
 SetPartOptimizationStrategy {nodes [
   n1 = SetPartOptimizationStrategy!() -> (
    then
    Merged Mesh Part Info:"Atom Merged Mesh Part Info Structure (by ref)"
    Atom Model Part Info:"Atom Model Part Instance Info Structure (by ref)"
   ) {is=title K2Node_FunctionEntry}
   n2 = Break Atom Primitive Optimization Settings!(AtomPrimitiveOptimizationSettings:"Atom Primitive Optimization Settings Structure (by ref)"=Get OptimizationSettings!(self:Atom Primitive Object Reference=Break Atom Model Part Instance Info!(AtomModelPartInstanceInfo=@n1.Atom Model Part Info) -> (Primitive:Atom Primitive Object Reference, Group:String, ParentGroup:String)) -> (OptimizationSettings:Atom Primitive Optimization Settings Structure)) -> (
    OptimizationType:EAtomPrimitiveGeoOptimization Enum
    ApproximationShapeType:EAtomPrimitiveApproximationShapeType Enum
    bUseOptimizationAxisOverride:bool=~
    ApproximationAxisOverride:Vector=~
   ) {is=title K2Node_BreakStruct}
   > Branch(Condition="Not Equal (Enum)"(A:EAtomPrimitiveGeoOptimization Enum=@n2, B:EAtomPrimitiveGeoOptimization Enum=Default))
   > n3 = Switch on EAtomPrimitiveApproximationShapeType!(execute, Selection:EAtomPrimitiveApproximationShapeType Enum=@n2.ApproximationShapeType) -> (Auto:Exec, OrientedBox:Exec, ExtrudedConvexHull2D:Exec, ConvexHull3D:Exec, Extrusion:Exec) {is=title K2Node_SwitchEnum}
   > Set members in Atom Merged Mesh Part Info!(execute, StructRef=@n1, ApproximationType:Bitmask Integer=255) -> (then, StructOut)
   Set members in Atom Merged Mesh Part Info(execute=@n3.ConvexHull3D, StructRef=@n1, ApproximationType=232)
   Set members in Atom Merged Mesh Part Info(execute=@n3.ExtrudedConvexHull2D, StructRef=@n1, ApproximationType=228)
   Set members in Atom Merged Mesh Part Info(execute=@n3.Extrusion, StructRef=@n1, ApproximationType=240)
   Set members in Atom Merged Mesh Part Info(execute=@n3.OrientedBox, StructRef=@n1, ApproximationType=227)
  ]}
 SetupChildProcessingPipeline {nodes [
   n1 = SetupChildProcessingPipeline!() -> (
    then
    Baked Flex Primitives Names:Array of Strings
    Scene:"Atom Model Hierarchical Scene Structure (by ref)"
    Primitive Instance Data Map:Map of Atom Model Part Guid Structures to Atom Primitive Instance Data Structures
    Merged Geometries:Map of Strings to Atom Merged Geometry Structures
   ) {is=title K2Node_FunctionEntry}
   > n2 = Construct Object from Class(Class=/Script/AtomModelImporterEditor.AtomMergedGeometryContainer) -> (ReturnValue:Atom Merged Geometry Container Object Reference)
   > Set Scene!(execute, Scene:Atom Model Hierarchical Scene Structure=@n1.Scene, self:Atom Merged Geometry Container Object Reference=@n2) -> (then, Output_Get:Atom Model Hierarchical Scene Structure)
   > Set MergedGeometries!(
    execute
    MergedGeometries:Map of Strings to Atom Merged Geometry Structures=Get Merged Geometries() -> (Merged Geometries:Map of Strings to Atom Merged Geometry Structures)
    self:Atom Merged Geometry Container Object Reference=@n2
   ) -> (then, Output_Get:Map of Strings to Atom Merged Geometry Structures)
   > Set MaxLODs!(execute, MaxLODs:int=0, self:Atom Merged Geometry Container Object Reference=@n2) -> (then, Output_Get:int=0)
   > Set BakedFlexPrimitivesNames!(
    execute
    BakedFlexPrimitivesNames=Get Baked Flex Primitives Names() -> (Baked Flex Primitives Names:Array of Strings)
    self:Atom Merged Geometry Container Object Reference=@n2
   ) -> (then, Output_Get)
   > Set PrimitiveInstanceDataMap!(
    execute
    PrimitiveInstanceDataMap=Get Primitive Instance Data Map() -> (Primitive Instance Data Map:Map of Atom Model Part Guid Structures to Atom Primitive Instance Data Structures)
    self:Atom Merged Geometry Container Object Reference=@n2
   ) -> (then, Output_Get)
   > n3 = ExportProcessorToJson!(execute, Processor:Atom Model Processor Object Reference=Self-Reference()) -> (then, ReturnValue:String)
   > Set ProcessorImportJson!(execute, ProcessorImportJson:String=@n3, self:Atom Merged Geometry Container Object Reference=@n2) -> (then, Output_Get:String)
   > Set MaterialFactory!(
    execute
    MaterialFactory:Atom Material Factory Object Reference=GetMaterialFactory()
    self:Atom Merged Geometry Container Object Reference=@n2
   ) -> (then, Output_Get:Atom Material Factory Object Reference)
   > RegisterChildProcessorHelperObject!(execute, self:MP Base Object Reference, Input:Object Reference=@n2) -> (then)
  ]}
 ShouldDefineSeparateNaniteCommonParts {nodes [
   ShouldDefineSeparateNaniteCommonParts() {is=title K2Node_FunctionEntry}
   > Return Node!(execute, Separate Nanite Common Parts:bool=false)
  ]}
 ShouldSplitTransparentPart {nodes [
   n1 = ShouldSplitTransparentPart!() -> (then, MergedMeshPartInfo:"Atom Merged Mesh Part Info Structure (by ref)") {is=title K2Node_FunctionEntry}
   > n2 = GetHasOnlyNaniteCompatibleMaterials!(execute, self, MergedMeshPartInfo:"Atom Merged Mesh Part Info Structure (by ref)"=@n1) -> (then, bOnlyHasOpaque:bool) {is=title K2Node_CallFunction}
   > Return Node!(
    execute
    Should Split Transparent Part:bool=AND Boolean(A=NOT Boolean(A=@n2), B=Get SplitTransparentMeshes() -> (SplitTransparentMeshes:bool))
   )
  ]}
 SplitGeometryLODsByPolygonGroup {nodes [
   n1 = SplitGeometryLODsByPolygonGroup!() -> (
    then
    Geometry LODs:"Atom Primitive Geometry LODs Structure (by ref)"
    Output Geomery LODs Array:Array of Atom Primitive Geometry LODs Structures
   ) {is=title K2Node_FunctionEntry}
   n2 = Break Atom Primitive Geometry LODs(AtomPrimitiveGeometryLODs=@n1)
   > n3 = For Each Loop(Array:Array of Atom Primitive Geometry Object References=@n2) -> (Array Element:Atom Primitive Geometry Object Reference)
   n4 = Is Valid(exec=@n3.Completed, InputObject=@n2.HiResNaniteGeomery)
   > n5 = For Each Loop(Array=SplitByPolygonGroup!(self:Atom Primitive Geometry Object Reference=@n2.HiResNaniteGeomery, bIgnoreEmptyGroups:bool=true) -> (ReturnValue:Array of Atom Primitive Geometry Object References))
   Set members in Atom Primitive Geometry LODs!(
    execute=[@n5.Completed, @n4.Is Not Valid]
    StructRef="Get (a ref)"(
     Array:Array of Atom Primitive Geometry LODs Structures=Get Output Geomery LODs Array() -> (Output Geomery LODs Array:Array of Atom Primitive Geometry LODs Structures)
     Dimension 1=0
    ) -> (Output:"Atom Primitive Geometry LODs Structure (by ref)")
    Collision:Atom Primitive Collision Geometry Structure=@n2.Collision
   ) -> (then, StructOut)
   n6 = Get Output Geomery LODs Array()
   Resize!(
    execute=@n5.LoopBody
    TargetArray:Array of Atom Primitive Geometry LODs Structures=@n6
    Size:int="Max (Integer)"(A=Length(TargetArray:Array of Atom Primitive Geometry LODs Structures=@n6), B=int + int(A=@n5.Array Index))
   ) -> (then) {is=title K2Node_CallArrayFunction}
   > Set members in Atom Primitive Geometry LODs!(execute, StructRef="Get (a ref)"(Array=@n6, Dimension 1=@n5.Array Index), HiResNaniteGeomery:Atom Primitive Geometry Object Reference=@n5) -> (then, StructOut)
   n7 = For Each Loop(Exec=@n3.LoopBody, Array=SplitByPolygonGroup(self=@n3, bIgnoreEmptyGroups=true))
   n8 = Get Output Geomery LODs Array()
   n9 = Get Output Geomery LODs Array()
   > Resize(TargetArray=@n9, Size="Max (Integer)"(A=Length(TargetArray=@n8), B=int + int(A=@n7.Array Index)))
   > Set Array Elem(
    TargetArray=Break Atom Primitive Geometry LODs(AtomPrimitiveGeometryLODs="Get (a ref)"(Array=@n9, Dimension 1=@n7.Array Index))
    Index=@n3.Array Index
    Item=@n7
    bSizeToFit=true
   )
  ]}
 SplitTransparentMesh {nodes [
   n1 = Inputs!() -> (
    Exec:Exec
    IsTransparent:bool=~
    PartInfo:Atom Merged Mesh Part Info Structure
    TransPartInfos:Array of Atom Merged Mesh Part Info Structures
    OpaquePartInfos:Array of Atom Merged Mesh Part Info Structures
   )
   > n2 = Branch(Condition=@n1)
   > n3 = Add(
    TargetArray:Array of Atom Merged Mesh Part Info Structures=@n1.TransPartInfos
    NewItem:"Atom Merged Mesh Part Info Structure (by ref)"=@n1.PartInfo
   )
   Outputs!(Out:Exec=[@n4.then, @n3.then])
   n4 = Add(execute=@n2.else, TargetArray=@n1.OpaquePartInfos, NewItem=@n1.PartInfo)
  ]}
 UseMaterialFactory {nodes [
   UseMaterialFactory() {is=title K2Node_FunctionEntry}
   > Return Node!(execute, UseMaterialFactory:bool=false)
  ]}
 UseRandomAlphaColorForPart {nodes [
   UseRandomAlphaColorForPart() {is=title K2Node_FunctionEntry}
   n1 = UseMaterialFactory() {is=title K2Node_CallFunction}
   n2 = Get Material Color Source() -> (Material Color Source=VertexColor)
   > Return Node!(
    execute
    UseRandomAlpha:bool=OR Boolean!(
     A=NOT Boolean(A=@n1)
     B="Equal (Enum)"(A:EAtomMaterialColorSource Enum=@n2, B:EAtomMaterialColorSource Enum=VertexColor)
     C:bool="Equal (Enum)"(A=@n2, B=LUTVertexColor)
    ) -> (ReturnValue)
   )
  ]}
]
variables [
 Auto Primitive Collision Type {type=EPrimitiveAutoCollisionType, default=Box}
 Bake Common Parts {type=boolean}
 bGenerateBakedFlexPrimitives {type=boolean}
 bGenerateDecoratedPrimitiveMesh {type=boolean}
 Blueprint Container Base Class {type=Class, default=&Actor}
 bUseImportedPrimitiveCollision {type=boolean}
 Common Part Optimizations {type=integer, default=1}
 Common Part Scale {type=number, default=20}
 Common Part Style {type=string, default=Realtime}
 Custom Primitive Location {type=string, default=""}
 Extract Common Parts {type=boolean}
 Fallback Geometry style {type=string, default=""}
 Generate Blueprint Container {type=boolean}
 Generate Geometry Collection {type=boolean, default=true}
 Generate Nanite Mesh {type=boolean, default=true}
 Generate Static Meshes For Brick Geometry {type=boolean}
 Geometry Static Mesh Path {type=string, default=""}
 Geometry style {type=string, default=Realtime}
 GroupFilter {type=EAtomGroupFilterMode, default=Disabled}
 Invalid Group Prefixes {type="array<string>", default [_,Imported]}
 LOD Geometry Style Settings {type="array<AtomProcessorGeometryStyleSettings>", default []}
 Material Color Source {type=EAtomMaterialColorSource, default=LUTVertexColor}
 Material Factory {type=AtomMaterialFactory, default=null}
 Model {type=AtomModel, default=null}
 Nanite Geometry Style Settings {
  type=AtomProcessorGeometryStyleSettings
  default {
   analytic Common Parts=false
   analytic Fake Underside=false
   bevel Angle Threshold=60
   bevel Width=0.0125
   fake Bevel=false
   fake Technic Hole=false
   fake Underside=false
   force Flat Cap=false
   number of Subdivisions To Apply=0
   only Bevel Block-Like Primitives=true
   preferred Geometry LOD=0
   should Bevel=false
   use Improved Bevel Algorithm=false
   geometryStylePreferences []
  }
 }
 Parts Collection Connectivity {type=AtomModelPartsCollectionConnectivity, default=null}
 Parts Collection Name {type=string, default=""}
 Prefer Primitive Geometry Container Collision {type=boolean}
 Preserve Existing Collision {type=boolean, default=true}
 Preserve Existing Sockets {type=boolean, default=true}
 Remove All Knobs {type=boolean}
 Remove All Pins {type=boolean}
 Remove All Tubes {type=boolean}
 Remove Connected {type=boolean}
 Separate Selection Sets {type=boolean}
 SplitTransparentMeshes {type=boolean, default=true}
 Support Decorations {type=boolean, default=true}
 Use Geometry Style Settings {type=boolean}
 Use Legacy Decoration Material {type=boolean}
 UseNewMergeAPI {type=boolean}
 VertexColorDebugMode {type=EAtomMeshMergingVertexColorMappingMode, default=None}
]
