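# Editor utility base processor that optimizes merged meshes built from Atom brick-model parts.
# Finalize overrides use per-LOD geometry style settings (fake underside, caps, alpha mask, translucency) to set each part's LOD filter, merging, simplification and occluder flags.
# Geometry Script helpers append meshes, bake vertex colors and encode brick size and depth into polygroup UVs.
# Mesh finalization applies LOD settings and optional deprecated collision simplification; a helper derives LOD screen sizes from projected bounds.
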EditorUtilityBlueprint /AtomFramework/Editor/Processors/MP_MergedMeshOptimizerBase
graphs [
 AddMaterialAssignment {nodes [
   n1 = Inputs!() -> (
    Exec:Exec
    Material Assignment:Atom Primitive Geometry Material Assignments Structure
    Material Features:Atom Material Features Structure
    Slot Name:String
   )
   n2 = Break Atom Primitive Geometry Material Assignments!(AtomPrimitiveGeometryMaterialAssignments:"Atom Primitive Geometry Material Assignments Structure (by ref)"=@n1) -> (MaterialSlotName:Array of Strings, MaterialFeatures:Array of Atom Material Features Structures) {is=title K2Node_BreakStruct}
   > Add(
    TargetArray:Array of Atom Material Features Structures=@n2.MaterialFeatures
    NewItem:"Atom Material Features Structure (by ref)"=@n1.Material Features
   ) {is=title K2Node_CallArrayFunction}
   > Add(TargetArray:Array of Strings=@n2, NewItem:"String (by ref)"=@n1.Slot Name)
   > Outputs!(Out:Exec)
  ]}
 Analytic Common Parts Requested On Any LODs {nodes [
   Analytic Common Parts Requested On Any LODs() {is=title K2Node_FunctionEntry}
   > n1 = For Each Loop(Array:Array of Atom Processor Geometry Style Settings Structures=Get LOD Geometry Style Settings() -> (LOD Geometry Style Settings:Array of Atom Processor Geometry Style Settings Structures)) -> (Array Element:Atom Processor Geometry Style Settings Structure)
   Return Node!(execute=@n1.Completed, bAnalyticCommonParts:bool=false)
   > Branch(Condition=Break Atom Processor Geometry Style Settings!(AtomProcessorGeometryStyleSettings:"Atom Processor Geometry Style Settings Structure (by ref)"=@n1) -> (FakeTechnicHole_41_#1:bool=~, AnalyticCommonParts_27_#2:bool=~) {is=title K2Node_BreakStruct}.AnalyticCommonParts_27_#2)
   > Return Node(bAnalyticCommonParts=true)
  ]}
 AppendDynamicMeshes {nodes [
   n1 = AppendDynamicMeshes!() -> (
    then
    Target Mesh:Dynamic Mesh Object Reference
    Additional Meshes:Array of Dynamic Mesh Object References
    Additional Meshes Transforms:Array of Atom Primitive Geometry And Transform Structures
   ) {is=title K2Node_FunctionEntry}
   > n2 = For Each Loop(Array:Array of Dynamic Mesh Object References=@n1.Additional Meshes) -> (Array Element:Dynamic Mesh Object Reference)
   n3 = BreakTransform(InTransform=Break Atom Primitive Geometry and Transform!(AtomPrimitiveGeometryAndTransform:"Atom Primitive Geometry And Transform Structure (by ref)"="Get (a ref)"(Array:Array of Atom Primitive Geometry And Transform Structures=@n1.Additional Meshes Transforms, Dimension 1=@n2.Array Index) -> (Output:"Atom Primitive Geometry And Transform Structure (by ref)")) -> (AtomPrimitiveGeometry:Atom Primitive Geometry Object Reference, Transform:Transform 3f Structure) {is=title K2Node_BreakStruct}.Transform)
   > AppendMesh!(
    execute
    TargetMesh:Dynamic Mesh Object Reference=Get Target Mesh() -> (Target Mesh:Dynamic Mesh Object Reference)
    AppendMesh:Dynamic Mesh Object Reference=@n2
    AppendTransform_Location:Vector=@n3
    AppendTransform_Rotation:Rotator=@n3.Rotation
    AppendTransform_Scale:Vector=@n3.Scale
    bDeferChangeNotifications:bool=false
    AppendOptions:Geometry Script Append Mesh Options Structure=(CombineMode=EnableAllMatching)
   ) -> (then, ReturnValue:Dynamic Mesh Object Reference)
   n4 = Branch(
    execute=@n2.Completed
    Condition=Is Not Empty!(TargetArray:Array of Dynamic Mesh Object References=Get Additional Meshes() -> (Additional Meshes:Array of Dynamic Mesh Object References)) -> (ReturnValue:bool)
   )
   > n5 = WeldMeshEdges!(execute, TargetMesh:Dynamic Mesh Object Reference=Get Target Mesh(), WeldOptions_Tolerance:float=0, WeldOptions_bOnlyUniquePairs:bool=True) -> (then, ReturnValue:Dynamic Mesh Object Reference)
   > Return Node!(execute, Out Mesh:Dynamic Mesh Object Reference=@n5)
   Return Node(execute=@n4.else, Out Mesh=Get Target Mesh())
  ]}
 Are Optimizations Changing UVs {nodes [
   Are Optimizations Changing UVs!() -> (then, Applied Optimizations:Atom Processor Applied Optimizations Structure) {is=title K2Node_FunctionEntry}
   n1 = Get!() -> (
    Applied Optimizations_bApplyFakeBevels_3_#3:bool=~
    Applied Optimizations_bApplyFakeUnderside_9_#4:bool=~
    Applied Optimizations_bAnalyticFakeUnderside_10_#5:bool=~
    Applied Optimizations_bApplyFlatCap_11_#6:bool=~
    Applied Optimizations_bApplyTechnicCap_12_#7:bool=~
    Applied Optimizations_bApplyFakeTechnicConnector_22_#8:bool=~
    Applied Optimizations_bApplyFakeTranslucent_25_#9:bool=~
    Applied Optimizations_bApplyDitheredTranslucent_24_#10:bool=~
    Applied Optimizations_bApplyGlow_13_#11:bool=~
    Applied Optimizations_bUseNormalMap_16_#12:bool=~
    Applied Optimizations_bUseAlphaMask_17_#13:bool=~
    Applied Optimizations_ExtraOptmizationsToApply_20_#14:Instanced Struct Structure
   )
   > Return Node!(
    execute
    UVs Changed:bool=OR Boolean!(
     A=@n1
     B=@n1.Applied Optimizations_bApplyFakeUnderside_9_#4
     C:bool=@n1.Applied Optimizations_bApplyTechnicCap_12_#7
     D:bool=@n1.Applied Optimizations_bApplyFakeTranslucent_25_#9
     E:bool=@n1.Applied Optimizations_bApplyGlow_13_#11
     F:bool=@n1.Applied Optimizations_bUseNormalMap_16_#12
     G:bool=@n1.Applied Optimizations_bUseAlphaMask_17_#13
    ) -> (ReturnValue)
   )
  ]}
 Bake Color Change In Vertex Color {nodes [
   n1 = Bake Color Change In Vertex Color!() -> (
    then
    Part Info:Atom Model Part Instance Info Structure
    Color Change index:int
    Shell Mesh:Dynamic Mesh Object Reference
    Caps Meshes:Array of Dynamic Mesh Object References
    Technic Caps Meshes:Array of Dynamic Mesh Object References
    Technic Connectors Meshes:Array of Dynamic Mesh Object References
   ) {is=title K2Node_FunctionEntry}
   > n2 = Get Linear Vertex Color For Part!(execute, self, Part Info:Atom Model Part Instance Info Structure=@n1, Color Change Index:int=@n1.Color Change index) -> (then, Linear Vertex Color:Linear Color Structure) {is=title K2Node_CallFunction}
   > Set Vertex Color(Vertex Color:Linear Color Structure=@n2) -> (Output_Get:Linear Color Structure)
   > n3 = Sequence()
   > SetMeshConstantVertexColor!(
    execute
    TargetMesh:Dynamic Mesh Object Reference=Get Shell Mesh() -> (Shell Mesh:Dynamic Mesh Object Reference)
    Color:Linear Color Structure=Get Vertex Color() -> (Vertex Color:Linear Color Structure)
    Flags:Geometry Script Color Flags Structure
    bClearExisting:bool=true
   ) -> (then, ReturnValue:Dynamic Mesh Object Reference)
   > n4 = For Each Loop(Array=Get Caps Meshes() -> (Caps Meshes:Array of Dynamic Mesh Object References))
   > SetMeshConstantVertexColor(TargetMesh=@n4, Color=Get Vertex Color(), bClearExisting=true)
   n5 = For Each Loop(Exec=@n4.Completed, Array=Get Technic Caps Meshes() -> (Technic Caps Meshes:Array of Dynamic Mesh Object References))
   > SetMeshConstantVertexColor(TargetMesh=@n5, Color=Get Vertex Color(), bClearExisting=true)
   n6 = For Each Loop(Exec=@n5.Completed, Array=Get Technic Connectors Meshes() -> (Technic Connectors Meshes:Array of Dynamic Mesh Object References))
   > SetMeshConstantVertexColor(TargetMesh=@n6, Color=Get Vertex Color(), bClearExisting=true)
   Return Node!(
    execute=@n3.then_1
    Out Shell Mesh:Dynamic Mesh Object Reference=Get Shell Mesh()
    Out Caps Meshes:Array of Dynamic Mesh Object References=Get Caps Meshes()
    Out Technic Caps Meshes:Array of Dynamic Mesh Object References=Get Technic Caps Meshes()
    Out Technic Connectors Meshes:Array of Dynamic Mesh Object References=Get Technic Connectors Meshes()
   )
  ]}
 CalculateScreenSizeFromDistances {nodes [
   n1 = CalculateScreenSizeFromDistances!() -> (
    then
    StaticMesh:Static Mesh Object Reference
    Distances:"Array of Float (double-precision)s"
    Use LOD Size Reference:bool=~
    LOD Size Reference:double=~
   ) {is=title K2Node_FunctionEntry}
   > n2 = Sequence()
   n3 = Get BaseLODScreenSize() -> (BaseLODScreenSize:double)
   > Set DefaultBaseScreenSize(DefaultBaseScreenSize:double=float * float(A=1000, B="Clamp (Float)"(Value=@n3, Min=0.1, Max=0.99)) {is=title K2Node_PromotableOperator}) -> (Output_Get:double)
   > Set DefaultBaseScreenSizePowerScale(DefaultBaseScreenSizePowerScale:double="Clamp (Float)"(Value=Get BaseLODScreenSizeScaling() -> (BaseLODScreenSizeScaling:double), Min=0.1, Max=2)) -> (Output_Get:double)
   > Set LOD1Distance(LOD1Distance:double="Clamp (Float)"(Value=Get DefaultBaseScreenSize() -> (DefaultBaseScreenSize:double), Min=1, Max=10000)) -> (Output_Get:double)
   > Set DistanceScaling(DistanceScaling:double="Clamp (Float)"(Value=Get DefaultBaseScreenSizePowerScale() -> (DefaultBaseScreenSizePowerScale:double), Min=0.01, Max=5)) -> (Output_Get:double)
   n4 = BreakBoxSphereBounds!(InBoxSphereBounds:"Box Sphere Bounds Structure (by ref)"=GetBounds!(self:Static Mesh Object Reference=@n1) -> (ReturnValue:Box Sphere Bounds Structure)) -> (Origin:Vector, BoxExtent:Vector, SphereRadius:float)
   > Set SphereRadius(SphereRadius:double=@n4.SphereRadius) -> (Output_Get:double)
   > Set HalfFOV(HalfFOV:double=float / float(A=Get PI!() -> (ReturnValue:double) {is=title K2Node_CallFunction}, B=4)) -> (Output_Get:double)
   > Set CurLODDistance(CurLODDistance:double=Get LOD1Distance() -> (LOD1Distance:double)) -> (Output_Get:double)
   > n5 = Branch(Condition=Get Use LOD Size Reference() -> (Use LOD Size Reference:bool))
   n6 = Get LOD Size Reference() -> (LOD Size Reference:double)
   > Set LODSizeRatio(LODSizeRatio:double="Clamp (Float)"(Value=float / float(A=@n4.SphereRadius, B=@n6), Min=0.33, Max=1.2)) -> (Output_Get:double)
   Set LODSizeRatio(execute=@n5.else, LODSizeRatio=1)
   n7 = For Loop(execute=@n2.then_1, LastIndex=GetNumLODs(self=@n1))
   n8 = Branch(execute=@n7.Completed, Condition="float <= float"(A=Get NormalizationFactor() -> (NormalizationFactor:double), B=1) -> (ReturnValue=~))
   SetLodScreenSizes!(
    execute=[@n8.then, @n10.Completed]
    self:Static Mesh Editor Subsystem Object Reference=Get StaticMeshEditorSubsystem()
    StaticMesh:Static Mesh Object Reference=@n1
    ScreenSizes:"Array of Float (single-precision)s"=Get ScreenSizes() -> (ScreenSizes:"Array of Float (double-precision)s")
   ) -> (then, ReturnValue:bool)
   > Return Node!(execute)
   n9 = Get ScreenSizes()
   n10 = For Loop(execute=@n8.else, FirstIndex=1, LastIndex=int - int(A=Length(TargetArray:"Array of Float (double-precision)s"=@n9)))
   > Set Array Elem!(
    execute
    TargetArray:"Array of Float (double-precision)s"=Get ScreenSizes()
    Index:int=@n10
    Item:"Float (double-precision) (by ref)"=float / float(
     A="Get (a ref)"(Array:"Array of Float (double-precision)s"=@n9, Dimension 1=@n10) -> (Output:"Float (double-precision) (by ref)")
     B=Get NormalizationFactor()
    )
    bSizeToFit:bool=false
   ) -> (then)
   n11 = Branch(execute=@n7.LoopBody, Condition="Equal (Integer)"(A=@n7))
   > Add(TargetArray:"Array of Float (double-precision)s"=Get ScreenSizes(), NewItem:"Float (double-precision) (by ref)"=2)
   n12 = Branch(execute=@n11.else, Condition="integer < integer"(A=@n7, B=Length(TargetArray=@n1.Distances)))
   > n13 = Set LODDistance(LODDistance:double=float * float(A=Get LODSizeRatio() -> (LODSizeRatio:double), B="Get (a ref)"(Array=@n1.Distances, Dimension 1=@n7))) -> (Output_Get:double)
   Set CurLODDistance(
    execute=[@n26.then, @n13.then]
    CurLODDistance=float * float(A=Get CurLODDistance() -> (CurLODDistance:double), B=Get DistanceScaling() -> (DistanceScaling:double))
   )
   > n14 = Set!(
    execute
    PointToTest_X:double=0
    PointToTest_Y:double=0
    PointToTest_Z:double=Get LODDistance() -> (LODDistance:double)
    PointToTest_W:double=1
   ) -> (then, Output_Get:Vector 4 Structure)
   n15 = "Tan (Radians)"!(A:double=Get HalfFOV() -> (HalfFOV:double)) -> (ReturnValue:double)
   n16 = Get ScreenWidth() -> (ScreenWidth:double)
   n17 = Get ScreenHeight() -> (ScreenHeight:double)
   n18 = Make Matrix!(
    XPlane:Plane Structure=Make Plane!(W:double=0, X:double=float / float(A=1, B=@n15), Y:double=0, Z:double=0) -> (Plane:Plane Structure) {is=title K2Node_MakeStruct}
    YPlane:Plane Structure=Make Plane(W=0, X=0, Y=float / float!(A=@n16, B=@n15, C:double=@n17) -> (ReturnValue), Z=0)
    ZPlane:Plane Structure=Make Plane(W=1, X=0, Y=0, Z=1)
    WPlane:Plane Structure=Make Plane(W=0, X=0, Y=0, Z=-1)
   ) -> (Matrix:Matrix Structure) {is=title K2Node_MakeStruct}
   > n19 = Set ScreenPosition(ScreenPosition:Vector 4 Structure="Transform Vector4 (Matrix)"!(M:"Matrix Structure (by ref)"=@n18, V:Vector 4 Structure=@n14) -> (ReturnValue:Vector 4 Structure)) -> (Output_Get:Vector 4 Structure)
   n20 = Get ScreenWidth()
   > n21 = Set ScreenMultiple(ScreenMultiple:double=float * float(
     A=float / float!(A=@n20, B=2) -> (ReturnValue)
     B=Break Matrix!(Matrix:"Matrix Structure (by ref)"=@n18) -> (
      XPlane_W:double=~
      XPlane_X:double=~
      XPlane_Y:double=~
      XPlane_Z:double=~
      YPlane:Plane Structure
      ZPlane:Plane Structure
      WPlane:Plane Structure
     ) {is=title K2Node_BreakStruct}.XPlane_X
    )) -> (Output_Get:double)
   n22 = BreakVector4!(InVec:"Vector 4 Structure (by ref)"=@n19) -> (X:double, Y:double, Z:double, W:double)
   n23 = Get SphereRadius() -> (SphereRadius:double)
   > Set ScreenRadius(ScreenRadius:double=float / float(A=float * float(A=@n21, B=@n23), B="Max (Float)"(A=@n22.W, B=1))) -> (Output_Get:double)
   > Set ScreenArea(ScreenArea:double=float * float(A=Get ScreenWidth(), B=Get ScreenHeight())) -> (Output_Get:double)
   n24 = Get ScreenRadius() -> (ScreenRadius:double)
   > Set BoundsArea(BoundsArea:double=float * float!(A=Get PI(), B=@n24, C:double=@n24) -> (ReturnValue)) -> (Output_Get:double)
   n25 = float / float(A=Get BoundsArea() -> (BoundsArea:double), B=Get ScreenArea() -> (ScreenArea:double))
   > Add(TargetArray=Get ScreenSizes(), NewItem=@n25)
   > Branch(Condition="Equal (Integer)"(A=@n7, B=1))
   > Set NormalizationFactor(NormalizationFactor:double=@n25) -> (Output_Get:double)
   n26 = Set LODDistance(execute=@n12.else, LODDistance=float * float!(A:int=@n7, B=Get CurLODDistance()) -> (ReturnValue))
   "Clamp (Float)"(Value=@n25, Min=0, Max=1)
  ]}
 Compute Glow Position {nodes [
   n1 = Compute Glow Position!() -> (then, PrimitiveTransform:Transform, Primitive:Atom Primitive Object Reference) {is=title K2Node_FunctionEntry}
   n2 = GetDefaultPrimitiveScale!() -> (ReturnValue:float)
   > Return Node!(
    execute
    ReturnValue:Vector=TransformLocation!(
     T:"Transform (by ref)"=@n1
     Location:Vector=GetCenterOfMass!(self:Atom Primitive Object Reference=@n1.Primitive, Scale:float=@n2) -> (ReturnValue:Vector)
    ) -> (ReturnValue:Vector)
   )
  ]}
 EventGraph {links [], nodes []}
 FT_EncodeDepth {nodes [
   n1 = FT_EncodeDepth!() -> (then, DynamicMesh:Dynamic Mesh Object Reference, PolygroupID:int, UV Channel:int) {is=title K2Node_FunctionEntry}
   > n2 = ConvertIndexSetToMeshSelection!(
    execute
    TargetMesh:Dynamic Mesh Object Reference=@n1
    IndexSet:Set of Integers=Make Set!("[0]":int=@n1.PolygroupID) -> (Set:Set of Integers) {is=title K2Node_MakeSet}
    SelectionType:EGeometryScriptMeshSelectionType Enum=Polygroups
   ) -> (then, Selection:Mesh Selection Structure, ReturnValue:Dynamic Mesh Object Reference)
   > Set Selection(Selection:Mesh Selection Structure=@n2) -> (Output_Get:Mesh Selection Structure)
   > n3 = Sequence()
   > n4 = GetMeshUVSizeInfo!(
    execute
    TargetMesh:Dynamic Mesh Object Reference=Get DynamicMesh() -> (DynamicMesh:Dynamic Mesh Object Reference)
    UVSetIndex:int=Get UV Channel() -> (UV Channel:int=0)
    Selection:Mesh Selection Structure=Get Selection() -> (Selection:Mesh Selection Structure)
    bOnlyIncludeValidUVTris:bool=true
   ) -> (
    then
    MeshArea:double
    UVArea:double
    MeshBounds:Box Structure
    UVBounds_Min:Vector 2D Structure
    UVBounds_Max:Vector 2D Structure
    UVBounds_bIsValid:bool=~
    bIsValidUVSet:bool
    bFoundUnsetUVs:bool
    ReturnValue:Dynamic Mesh Object Reference
   )
   n5 = BreakVector2D(InVec=@n4.UVBounds_Max)
   n6 = "To Float (Integer)"(InInt=Floor(A=@n5.Y) {is=title K2Node_CallFunction})
   n7 = "To Float (Integer)"(InInt=Floor(A=@n5))
   > TranslateMeshUVs!(
    execute
    TargetMesh:Dynamic Mesh Object Reference=Get DynamicMesh()
    UVSetIndex:int=Get UV Channel() -> (UV Channel=0)
    Translation:Vector 2D Structure=vector2d * vector2d(A:Vector 2D Structure=MakeVector2D(X=@n7, Y=@n6), B:float=-1) -> (ReturnValue:Vector 2D Structure) {is=title K2Node_PromotableOperator}
    Selection:Mesh Selection Structure=Get Selection()
   ) -> (then, ReturnValue:Dynamic Mesh Object Reference)
   n8 = BreakVector(InVec=FT_GetBrickSize!(DynamicMesh:Dynamic Mesh Object Reference=Get DynamicMesh()) -> (ReturnValue:Vector))
   n9 = BreakVector(InVec=GetPredominantNormalFromPolygroup!(self, DynamicMesh:Dynamic Mesh Object Reference=Get DynamicMesh(), PolygroupID:int=Get PolygroupID() -> (PolygroupID:int=0)) -> (Normal:Vector))
   n10 = "Not Equal (Float)"(A="To Float (Integer)"(InInt=Round(A=@n9) {is=title K2Node_CallFunction}))
   n11 = "To Float (Integer)"(InInt=Round(A=@n9.Y))
   n12 = TranslateMeshUVs!(
    execute=@n3.then_1
    TargetMesh=Get DynamicMesh()
    UVSetIndex=Get UV Channel() -> (UV Channel=0)
    Translation_X:double=0
    Translation_Y:double=SelectFloat(A=@n8.Y, B=SelectFloat(A=@n8, B=@n8.Z, bPickA=@n10), bPickA="Not Equal (Float)"(A=@n11))
    Selection=Get Selection()
   ) -> (then, ReturnValue)
   > Return Node!(execute, Target Mesh:Dynamic Mesh Object Reference=@n12)
  ]}
 FT_EncodeSize {nodes [
   n1 = FT_EncodeSize!() -> (then, DynamicMesh:Dynamic Mesh Object Reference, PolygroupID:int, UV Channel:int) {is=title K2Node_FunctionEntry}
   > n2 = ConvertIndexSetToMeshSelection(TargetMesh=@n1, IndexSet=Make Set("[0]"=@n1.PolygroupID), SelectionType=Polygroups)
   > Set Selection(Selection=@n2)
   > n3 = Sequence!(execute) -> (then_0, then_1, then_2:Exec)
   > n4 = GetMeshUVSizeInfo(TargetMesh=Get DynamicMesh(), UVSetIndex=Get UV Channel() -> (UV Channel=0), Selection=Get Selection(), bOnlyIncludeValidUVTris=true) -> (UVBounds_bIsValid=~)
   > n5 = TranslateMeshUVs!(
    execute
    TargetMesh=Get DynamicMesh()
    UVSetIndex=Get UV Channel() -> (UV Channel=0)
    Translation:Vector 2D Structure=vector2d * vector2d(A=@n4.UVBounds_Min, B:double=-1)
    Selection=Get Selection()
   ) -> (then, ReturnValue)
   > n6 = GetMeshUVSizeInfo(TargetMesh=@n5, UVSetIndex=Get UV Channel() -> (UV Channel=0), Selection=Get Selection(), bOnlyIncludeValidUVTris=true) -> (UVBounds_bIsValid=~)
   > TranslateMeshUVs(
    TargetMesh=Get DynamicMesh()
    UVSetIndex=Get UV Channel() -> (UV Channel=0)
    Translation=vector2d * vector2d(
     A=vector2d - vector2d(A=1, B:Vector 2D Structure=@n6.UVBounds_Max) -> (ReturnValue:Vector 2D Structure) {is=title K2Node_PromotableOperator}
     B:float=0.5
    )
    Selection=Get Selection()
   )
   FT_SetNewScaleUV!(
    execute=@n3.then_1
    self
    DynamicMesh:Dynamic Mesh Object Reference=Get DynamicMesh()
    PolygroupID:int=Get PolygroupID() -> (PolygroupID=0)
    Selection:Mesh Selection Structure=Get Selection()
    UV Channel:int=Get UV Channel() -> (UV Channel=0)
   ) -> (then, Target Mesh:Dynamic Mesh Object Reference)
   n7 = GetMeshUVSizeInfo(
    execute=@n3.then_2
    TargetMesh=Get DynamicMesh()
    UVSetIndex=Get UV Channel() -> (UV Channel=0)
    Selection=Get Selection()
    bOnlyIncludeValidUVTris=true
   ) -> (UVBounds_bIsValid=~)
   n8 = vector2d * vector2d!(
    A=GetAbs2D!(A:Vector 2D Structure=vector2d - vector2d(A:Vector 2D Structure=@n7.UVBounds_Max, B=@n7.UVBounds_Min)) -> (ReturnValue:Vector 2D Structure)
    B=32
   ) -> (ReturnValue_X:double, ReturnValue_Y:double)
   > n9 = TranslateMeshUVs!(
    execute
    TargetMesh=Get DynamicMesh()
    UVSetIndex=Get UV Channel() -> (UV Channel=0)
    Translation_X:double="To Float (Integer)"(InInt=Ceil(A=@n8) {is=title K2Node_CallFunction})
    Translation_Y:double="To Float (Integer)"(InInt=Ceil(A=@n8.ReturnValue_Y))
    Selection=Get Selection()
   ) -> (then, ReturnValue)
   > Return Node(Target Mesh=@n9)
  ]}
 FT_GetBrickSize {nodes [
   n1 = FT_GetBrickSize!() -> (then, DynamicMesh:Dynamic Mesh Object Reference) {is=title K2Node_FunctionEntry}
   n2 = GetMeshBoundingBox!(TargetMesh:Dynamic Mesh Object Reference=@n1) -> (ReturnValue_Min:Vector=~, ReturnValue_Max:Vector=~, ReturnValue_IsValid:bool=~)
   n3 = vector * vector!(A=vector - vector(A=@n2.ReturnValue_Max, B=@n2), B:int=100) -> (ReturnValue_X:double, ReturnValue_Y:double, ReturnValue_Z:double)
   n4 = Find(
    TargetMap:"Map of Integers to Float (double-precision)s"=Get SizeChart() -> (SizeChart:"Map of Integers to Float (double-precision)s")
    Key:"Integer (by ref)"=Round(A=@n3.ReturnValue_Z)
   ) -> (Value:double)
   n5 = "To Float (Integer)"(InInt=Ceil(A=float * float(A:double=@n3.ReturnValue_Z, B=0.0312)))
   n6 = Find(TargetMap=Get SizeChart(), Key=Round(A=@n3))
   n7 = "To Float (Integer)"(InInt=Ceil(A=float * float(A=@n3, B=0.0312)))
   n8 = Find(TargetMap=Get SizeChart(), Key=Round(A=@n3.ReturnValue_Y))
   n9 = "To Float (Integer)"(InInt=Ceil(A=float * float(A=@n3.ReturnValue_Y, B=0.0312)))
   > Return Node!(
    execute
    ReturnValue:Vector=MakeVector(
     X=SelectFloat(A=@n6, B=@n7, bPickA=@n6.ReturnValue)
     Y=SelectFloat(A=@n8, B=@n9, bPickA=@n8.ReturnValue)
     Z=SelectFloat(A=@n4, B=@n5, bPickA=@n4.ReturnValue)
    )
   )
  ]}
 FT_SetNewScaleUV {nodes [
   FT_SetNewScaleUV!() -> (then, DynamicMesh:Dynamic Mesh Object Reference, PolygroupID:int, Selection:Mesh Selection Structure, UV Channel:int) {is=title K2Node_FunctionEntry}
   > n1 = Sequence!(execute) -> (then_0, then_1)
   n2 = Get Scale() -> (Scale:Vector 2D Structure)
   n3 = ScaleMeshUVs!(
    execute=@n1.then_1
    TargetMesh:Dynamic Mesh Object Reference=Get DynamicMesh()
    UVSetIndex:int=Get UV Channel() -> (UV Channel=0)
    Scale:Vector 2D Structure=vector2d / vector2d(A:Vector 2D Structure=vector2d * vector2d!(A=@n2, B:Vector 2D Structure=(X=1,Y=-1)) -> (ReturnValue:Vector 2D Structure), B:float=32) -> (ReturnValue:Vector 2D Structure) {is=title K2Node_PromotableOperator}
    ScaleOrigin:Vector 2D Structure=(X=0.5,Y=0.5)
    Selection:Mesh Selection Structure=Get Selection()
   ) -> (then, ReturnValue:Dynamic Mesh Object Reference)
   > Return Node!(execute, Target Mesh:Dynamic Mesh Object Reference=@n3)
   n4 = BreakVector(InVec=GetPredominantNormalFromPolygroup(DynamicMesh=Get DynamicMesh(), PolygroupID=Get PolygroupID() -> (PolygroupID=0)))
   n5 = "To Float (Integer)"(InInt=Round(A=@n4.Y))
   n6 = BreakVector(InVec=FT_GetBrickSize(DynamicMesh=Get DynamicMesh()) {is=title K2Node_CallFunction})
   n7 = MakeVector2D(X=@n6.Y, Y=@n6.Z)
   n8 = MakeVector2D(X=@n6, Y=@n6.Y)
   n9 = "Not Equal (Float)"(A="To Float (Integer)"(InInt=Round(A=@n4)))
   Set Scale(
    execute=@n1.then_0
    Scale:Vector 2D Structure=Select(
     Option 0:Vector 2D Structure=Select(Option 0=@n8, Option 1=@n7, Index=@n9)
     Option 1:Vector 2D Structure=MakeVector2D(X=@n6, Y=@n6.Z)
     Index="Not Equal (Float)"(A=@n5)
    ) -> (ReturnValue:Vector 2D Structure)
   ) -> (Output_Get:Vector 2D Structure)
  ]}
 Fake Underside Requested On Any LODs {nodes [
   Fake Underside Requested On Any LODs() {is=title K2Node_FunctionEntry}
   > n1 = For Each Loop(Array:Array of Atom Processor Geometry Style Settings Structures=Get LOD Geometry Style Settings()) -> (Array Element:Atom Processor Geometry Style Settings Structure)
   Return Node!(execute=@n1.Completed, bFakeUndersideRequested:bool=false)
   > Branch(Condition=Break Atom Processor Geometry Style Settings!(AtomProcessorGeometryStyleSettings=@n1) -> (FakeUnderside_25_#15:bool=~, FakeTechnicHole_41_#1=~))
   > Return Node(bFakeUndersideRequested=true)
  ]}
 Features Applied On Any LOD {nodes [
   n1 = Features Applied On Any LOD!() -> (then, Part Info:Atom Model Part Instance Info Structure, Expected Features:Atom Processor Applied Optimizations Structure) {is=title K2Node_FunctionEntry}
   > n2 = Get Geometry Style Settings For Part!(execute, self, Part Info:Atom Model Part Instance Info Structure=@n1) -> (then, Geometry Styles:Array of Atom Processor Geometry Style Settings Structures) {is=title K2Node_CallFunction}
   > n3 = For Each Loop with Break!(Exec:Exec, Array:Array of Atom Processor Geometry Style Settings Structures=@n2, Break:Exec=@n6.then) -> (LoopBody:Exec, Array Element:Atom Processor Geometry Style Settings Structure, Array Index:int, Completed:Exec)
   Return Node!(execute=@n3.Completed, Result:bool=Get Found Match() -> (Found Match:bool))
   n4 = Break Atom Processor Applied Optimizations!(AtomProcessorAppliedOptimizations:"Atom Processor Applied Optimizations Structure (by ref)"=Get Expected Features() -> (Expected Features:Atom Processor Applied Optimizations Structure)) -> (
    bApplyFakeBevels_3_#3:bool=~
    bApplyFakeUnderside_9_#4:bool=~
    bAnalyticFakeUnderside_10_#5:bool=~
    bApplyFlatCap_11_#6:bool=~
    bApplyTechnicCap_12_#7:bool=~
    bApplyFakeTechnicConnector_22_#8:bool=~
    bApplyFakeTranslucent_25_#9:bool=~
    bApplyDitheredTranslucent_24_#10:bool=~
    bApplyGlow_13_#11:bool=~
    bUseNormalMap_16_#12:bool=~
    bUseAlphaMask_17_#13:bool=~
    ExtraOptmizationsToApply_20_#14:Instanced Struct Structure
   ) {is=title K2Node_BreakStruct}
   n5 = Get Optimizations To Apply For Geometry Style!(
    Part Info:Atom Model Part Instance Info Structure=@n1
    Geometry Style Settings:Atom Processor Geometry Style Settings Structure=@n3
    ColorChangeIndex:int=0
   ) -> (
    Applied Optimizations_bApplyFakeBevels_3_#3:bool=~
    Applied Optimizations_bApplyFakeUnderside_9_#4:bool=~
    Applied Optimizations_bAnalyticFakeUnderside_10_#5:bool=~
    Applied Optimizations_bApplyFlatCap_11_#6:bool=~
    Applied Optimizations_bApplyTechnicCap_12_#7:bool=~
    Applied Optimizations_bApplyFakeTechnicConnector_22_#8:bool=~
    Applied Optimizations_bApplyFakeTranslucent_25_#9:bool=~
    Applied Optimizations_bApplyDitheredTranslucent_24_#10:bool=~
    Applied Optimizations_bApplyGlow_13_#11:bool=~
    Applied Optimizations_bUseNormalMap_16_#12:bool=~
    Applied Optimizations_bUseAlphaMask_17_#13:bool=~
    Applied Optimizations_ExtraOptmizationsToApply_20_#14:Instanced Struct Structure
   ) {is=title K2Node_CallFunction}
   > Branch(Condition=OR Boolean!(
     A=AND Boolean(A=@n5, B=@n4)
     B=AND Boolean(
      A=@n5.Applied Optimizations_bApplyFakeUnderside_9_#4
      B=@n4.bApplyFakeUnderside_9_#4
     )
     C=AND Boolean(
      A=@n5.Applied Optimizations_bAnalyticFakeUnderside_10_#5
      B=@n4.bAnalyticFakeUnderside_10_#5
     )
     D=AND Boolean(A=@n5.Applied Optimizations_bApplyFlatCap_11_#6, B=@n4.bApplyFlatCap_11_#6)
     E=AND Boolean(A=@n5.Applied Optimizations_bApplyTechnicCap_12_#7, B=@n4.bApplyTechnicCap_12_#7)
     F=AND Boolean(
      A=@n5.Applied Optimizations_bApplyFakeTechnicConnector_22_#8
      B=@n4.bApplyFakeTechnicConnector_22_#8
     )
     G=AND Boolean(
      A=@n5.Applied Optimizations_bApplyFakeTranslucent_25_#9
      B=@n4.bApplyFakeTranslucent_25_#9
     )
     H:bool=AND Boolean(
      A=@n5.Applied Optimizations_bApplyDitheredTranslucent_24_#10
      B=@n4.bApplyDitheredTranslucent_24_#10
     )
     I:bool=AND Boolean(A=@n5.Applied Optimizations_bApplyGlow_13_#11, B=@n4.bApplyGlow_13_#11)
     J:bool=AND Boolean(A=@n5.Applied Optimizations_bUseNormalMap_16_#12, B=@n4.bUseNormalMap_16_#12)
     K:bool=AND Boolean(A=@n5.Applied Optimizations_bUseAlphaMask_17_#13, B=@n4.bUseAlphaMask_17_#13)
    ) -> (ReturnValue))
   > n6 = Set Found Match(Found Match:bool=true) -> (Output_Get:bool)
  ]}
 FinalizeCommonPartMergedMeshPartInfo {nodes [
   n1 = FinalizeCommonPartMergedMeshPartInfo!() -> (
    then
    CommonPartMergedMeshPartInfo:"Atom Merged Mesh Part Info Structure (by ref)"
    PrimitivePartInfo:"Atom Model Part Instance Info Structure (by ref)"
    CommonPartIndex:int=0
   ) {is=title K2Node_FunctionEntry}
   > n2 = Parent: FinalizeCommonPartMergedMeshPartInfo!(
    execute
    CommonPartMergedMeshPartInfo:"Atom Merged Mesh Part Info Structure (by ref)"=@n1
    PrimitivePartInfo:"Atom Model Part Instance Info Structure (by ref)"=@n1.PrimitivePartInfo
    CommonPartIndex:int=@n1.CommonPartIndex
   ) -> (then, FinalizedPartInfo:Atom Merged Mesh Part Info Structure) {is=title K2Node_CallParentFunction}
   > Set Finalized Part Info(Finalized Part Info:Atom Merged Mesh Part Info Structure=@n2) -> (Output_Get:Atom Merged Mesh Part Info Structure)
   > n3 = Sequence!(execute) -> (then_0, then_1, then_2:Exec)
   Is Common Part Affected by Cap Optimization!(
    Exec:"Exec (by ref)"=@n3.then_0
    PartInfo:"Atom Model Part Instance Info Structure (by ref)"=@n1.PrimitivePartInfo
    CommonPartIndex:"Integer (by ref)"=@n1.CommonPartIndex
   ) -> (True:Exec, False:Exec) {is=title K2Node_MacroInstance}
   > n4 = For Each Loop with Break(Array=Get LOD Geometry Style Settings(), Break=@n7.then)
   n5 = Break Atom Processor Geometry Style Settings!(AtomProcessorGeometryStyleSettings=@n4) -> (
    FakeUnderside_25_#15=~
    ForceFlatCap_35_#16:bool=~
    FakeTechnicHole_41_#1=~
   )
   n6 = Break Atom Processor Applied Optimizations!(AtomProcessorAppliedOptimizations=Get Optimizations To Apply For LOD!(
     Part Info:Atom Model Part Instance Info Structure=Get PrimitivePartInfo() -> (PrimitivePartInfo:"Atom Model Part Instance Info Structure (by ref)")
     LOD:int=@n4.Array Index
    ) -> (Applied Optimizations:Atom Processor Applied Optimizations Structure) {is=title K2Node_CallFunction}) -> (
    bApplyFakeUnderside_9_#4=~
    bApplyFlatCap_11_#6=~
    bApplyFakeTranslucent_25_#9=~
    bApplyDitheredTranslucent_24_#10=~
   )
   > Branch(Condition=OR Boolean!(
     A=AND Boolean(A=@n5, B=@n6)
     B=AND Boolean(A=@n5.ForceFlatCap_35_#16, B=@n6.bApplyFlatCap_11_#6)
    ) -> (ReturnValue))
   > n7 = Set members in Atom Merged Mesh Part Info!(
    execute:Exec
    StructRef:"Atom Merged Mesh Part Info Structure (by ref)"=Get Finalized Part Info() -> (Finalized Part Info:Atom Merged Mesh Part Info Structure)
    FilterLODLevel:int=@n4.Array Index
   ) -> (then:Exec, StructOut:"Atom Merged Mesh Part Info Structure (by ref)") {is=title K2Node_SetFieldsInStruct}
   Branch(execute=@n3.then_1, Condition=Analytic Common Parts Requested On Any LODs!() -> (bAnalyticCommonParts:bool) {is=title K2Node_CallFunction})
   > Set members in Atom Merged Mesh Part Info!(execute, StructRef=Get Finalized Part Info(), bPreserveUVs:bool=true, bAllowMerging:bool=false, bAllowApproximation:bool=false) -> (then, StructOut)
   Return Node!(execute=@n3.then_2, FinalizedPartInfo:Atom Merged Mesh Part Info Structure=Get Finalized Part Info())
  ]}
 FinalizeMergedMeshPart {nodes [
   n1 = FinalizeMergedMeshPart!() -> (then, MergedMeshPartInfo:"Atom Merged Mesh Part Info Structure (by ref)", PartInfo:Atom Model Part Instance Info Structure) {is=title K2Node_FunctionEntry}
   > n2 = Parent: FinalizeMergedMeshPart!(
    execute
    MergedMeshPartInfo:"Atom Merged Mesh Part Info Structure (by ref)"=@n1
    PartInfo:Atom Model Part Instance Info Structure=@n1.PartInfo
   ) -> (then, Out MergedMeshPartInfo:Atom Merged Mesh Part Info Structure) {is=title K2Node_CallParentFunction}
   > Set Finalized Merged Mesh Part(Finalized Merged Mesh Part:Atom Merged Mesh Part Info Structure=@n2) -> (Output_Get:Atom Merged Mesh Part Info Structure)
   > n3 = Sequence()
   > n4 = Get Geometry Style Settings For Part(Part Info=Get PartInfo() -> (PartInfo:Atom Model Part Instance Info Structure))
   > n5 = For Each Loop with Break(Array=@n4, Break=@n12.then)
   n6 = Get PartInfo()
   n7 = Break Atom Processor Applied Optimizations!(AtomProcessorAppliedOptimizations=Get Optimizations To Apply For Geometry Style!(Part Info=@n6, Geometry Style Settings=@n5, ColorChangeIndex=0) -> (Applied Optimizations:Atom Processor Applied Optimizations Structure)) -> (bUseAlphaMask_17_#13:bool=~)
   n8 = Get MergedMeshPartInfo() -> (MergedMeshPartInfo:"Atom Merged Mesh Part Info Structure (by ref)")
   > Branch(Condition=OR Boolean!(
     A=@n7
     B=IsDitheredTranslucentPart!(Part Info:Atom Model Part Instance Info Structure=@n6, Color Change Index:int=0) -> (IsDitheredTranslucent:bool)
     C:bool=ShouldSplitTransparentPart!(MergedMeshPartInfo:"Atom Merged Mesh Part Info Structure (by ref)"=@n8) -> (Should Split Transparent Part:bool)
    ) -> (ReturnValue))
   n9 = Get Finalized Merged Mesh Part() -> (Finalized Merged Mesh Part:Atom Merged Mesh Part Info Structure)
   n10 = Break Atom Merged Mesh Part Info!(AtomMergedMeshPartInfo:"Atom Merged Mesh Part Info Structure (by ref)"=@n9) -> (bAllowSimplification:bool=~, bAllowApproximation:bool=~) {is=title K2Node_BreakStruct}
   n11 = NOT Boolean(A=@n7)
   > n12 = Set members in Atom Merged Mesh Part Info!(
    execute
    StructRef=@n9
    bAllowSimplification:bool=AND Boolean(A=@n10, B=@n11)
    bAllowApproximation=AND Boolean(A=@n10.bAllowApproximation, B=@n11)
    bIsOccluder:bool=false
   ) -> (then, StructOut)
   Return Node!(execute=@n3.then_2, Out MergedMeshPartInfo:Atom Merged Mesh Part Info Structure=Get Finalized Merged Mesh Part())
   n13 = For Each Loop(
    Exec=@n3.then_1
    Array:Array of Material Interface Object References=Break Atom Merged Mesh Part Info!(AtomMergedMeshPartInfo=Get MergedMeshPartInfo()) -> (Materials:Array of Material Interface Object References)
   ) -> (Array Element:Material Interface Object Reference)
   Set members in Atom Merged Mesh Part Info!(
    execute=@n13.Completed
    StructRef=Get Finalized Merged Mesh Part()
    PreventMergingMaterials:Array of Material Interface Object References=Get MaterialsWithTextures() -> (MaterialsWithTextures:Array of Material Interface Object References)
   ) -> (then, StructOut)
   Branch(
    execute=@n13.LoopBody
    Condition=IsPreventMergingMaterial!(Material:Material Interface Object Reference=@n13) -> (Is Prevent Merging Material:bool)
   )
   > Add Unique(
    TargetArray:Array of Material Interface Object References=Get MaterialsWithTextures()
    NewItem:"Material Interface Object Reference (by ref)"=@n13
   )
  ]}
 FinalizeMesh {nodes [
   n1 = FinalizeMesh!() -> (then, Static Mesh:Static Mesh Object Reference) {is=title K2Node_FunctionEntry}
   > Parent: FinalizeMesh!(execute, Static Mesh:Static Mesh Object Reference=@n1) -> (then) {is=title K2Node_CallParentFunction}
   > SetLODSettings!(execute, self, StaticMesh:Static Mesh Object Reference=@n1) -> (then)
   n2 = GetMergedMeshType!(" Mesh Name":String=GetObjectName(Object=Get Static Mesh() -> (Static Mesh:Static Mesh Object Reference))) -> (Merged Mesh Type:EAtomMergedMeshType Enum=NewEnumerator0)
   Branch(Condition=AND Boolean(
     A=Get Optimize Mesh Collision() -> (Optimize Mesh Collision:bool)
     B="Not Equal (Enum)"!(A:EAtomMergedMeshType Enum=@n2, B:EAtomMergedMeshType Enum=NewEnumerator2) -> (ReturnValue:bool)
    ))
   > Generate Simplify Collision DEPRECATED!(execute, self, StaticMesh:Static Mesh Object Reference=Get Static Mesh()) -> (then)
  ]}
 FixPolygroupBoxProjectionUVs {nodes [
   n1 = FixPolygroupBoxProjectionUVs!() -> (then, DynamicMesh:Dynamic Mesh Object Reference, Polygon Group Index:int, Fake Bevels Channel:int) {is=title K2Node_FunctionEntry}
   > Set PolygroupIndexSet(PolygroupIndexSet:Set of Integers=Make Set("[0]"=@n1.Polygon Group Index)) -> (Output_Get:Set of Integers)
   > n2 = ConvertIndexSetToMeshSelection(TargetMesh=Get DynamicMesh(), IndexSet=Get PolygroupIndexSet() -> (PolygroupIndexSet:Set of Integers), SelectionType=Polygroups)
   > n3 = GetMeshUVSizeInfo!(
    execute
    TargetMesh=Get DynamicMesh()
    UVSetIndex=Get Fake Bevels Channel() -> (Fake Bevels Channel:int=0)
    Selection=@n2
    bOnlyIncludeValidUVTris=true
   ) -> (then, MeshArea, UVArea, MeshBounds, UVBounds:Box 2D Structure, bIsValidUVSet, bFoundUnsetUVs, ReturnValue)
   > TranslateMeshUVs!(
    execute
    TargetMesh=Get DynamicMesh()
    UVSetIndex=Get Fake Bevels Channel() -> (Fake Bevels Channel=0)
    Translation:Vector 2D Structure=vector2d * vector2d(
     A=Break Box 2D!(Box2D:"Box 2D Structure (by ref)"=@n3.UVBounds) -> (Min:Vector 2D Structure, Max:Vector 2D Structure, bIsValid:bool=~) {is=title K2Node_BreakStruct}
     B:double=-1
    )
    Selection=@n2
   ) -> (then, ReturnValue)
  ]}
 Generate Simplify Collision DEPRECATED {nodes [
   n1 = Generate Simplify Collision DEPRECATED!() -> (then, StaticMesh:Static Mesh Object Reference) {is=title K2Node_FunctionEntry}
   > n2 = StaticMeshHasCustomizedCollision!(execute, StaticMeshAsset:Static Mesh Object Reference=@n1) -> (then, ReturnValue:bool)
   > n3 = Branch(Condition=@n2)
   > Return Node!(execute)
   n4 = GetSimpleCollisionFromStaticMesh!(execute=@n3.else, StaticMesh:Static Mesh Object Reference=Get StaticMesh!() -> (StaticMesh)) -> (then, ReturnValue:Geometry Script Simple Collision Structure)
   > n5 = Branch(Condition=Get Debug - Bypass Mesh Convex Merge() -> (Debug - Bypass Mesh Convex Merge:bool))
   > n6 = SetSimpleCollisionOfStaticMesh!(
    execute
    SimpleCollision:"Geometry Script Simple Collision Structure (by ref)"=@n4
    StaticMesh:Static Mesh Object Reference=Get StaticMesh()
    Options_bEmitTransaction:bool=True
    StaticMeshCollisionOptions_bMarkAsCustomized:bool=false
   ) -> (then)
   Return Node(execute=[@n11.then, @n6.then])
   n7 = Get Force Shape Count() -> (Force Shape Count:int=0)
   n8 = "Clamp (Integer)"(Value=@n7, Min=0, Max=1) -> (ReturnValue=0)
   n9 = Get Mesh Protection - Min Sample Radius() -> (Mesh Protection - Min Sample Radius:double)
   n10 = MergeSimpleCollisionShapes!(
    execute=@n5.else
    SimpleCollision:"Geometry Script Simple Collision Structure (by ref)"=@n4
    MergeOptions_MaxShapeCount:int=@n7
    MergeOptions_ErrorTolerance:double=0
    MergeOptions_MinThicknessTolerance:double=0
    MergeOptions_bConsiderAllPossibleMerges:bool=False
    MergeOptions_PrecomputedNegativeSpace:Sphere Covering Structure=()
    MergeOptions_bComputeNegativeSpace:bool="To Boolean (Integer)"!(InInt:int=int - int(A=1, B=@n8)) -> (ReturnValue:bool)
    MergeOptions_ComputeNegativeSpaceOptions_SampleMethod:ENegativeSpaceSampleMethod Enum=VoxelSearch
    MergeOptions_ComputeNegativeSpaceOptions_bRequireSearchSampleCoverage:bool=true
    MergeOptions_ComputeNegativeSpaceOptions_bOnlyConnectedToHull:bool=False
    MergeOptions_ComputeNegativeSpaceOptions_MaxVoxelsPerDim:int=128
    MergeOptions_ComputeNegativeSpaceOptions_TargetNumSamples:int=10
    MergeOptions_ComputeNegativeSpaceOptions_MinSampleSpacing:double=@n9
    MergeOptions_ComputeNegativeSpaceOptions_NegativeSpaceTolerance:double=Get Mesh Protection - Tolerance() -> (Mesh Protection - Tolerance:double)
    MergeOptions_ComputeNegativeSpaceOptions_ToleranceDirectionalScales:Array of Negative Space Directional Tolerance Scale Structures
    MergeOptions_ComputeNegativeSpaceOptions_MinRadius:double=@n9
    MergeOptions_ShapeToHullTriangulation:Geometry Script Simple Collision Triangulation Options Structure=(SphereStepsPerSide=4,CapsuleHemisphereSteps=5,CapsuleCircleSteps=8,bApproximateLevelSetsWithCubes=False)
   ) -> (then, bHasMerged:bool, ReturnValue:Geometry Script Simple Collision Structure)
   > SimplifyConvexHulls!(
    execute
    SimpleCollision:"Geometry Script Simple Collision Structure (by ref)"=@n10.ReturnValue
    SimplifyOptions_SimplificationMethod:EGeometryScriptConvexHullSimplifyMethod Enum=MeshQSlim
    SimplifyOptions_SimplificationDistanceThreshold:float=Get Simplification Distance() -> (Simplification Distance:double)
    SimplifyOptions_SimplificationAngleThreshold:float=10
    SimplifyOptions_MinTargetFaceCount:int=Get Simplification Min Face Count() -> (Simplification Min Face Count:int=0)
   ) -> (then, bHasSimplified:bool)
   > SimplifyConvexHulls(
    SimpleCollision=@n10.ReturnValue
    SimplifyOptions_SimplificationMethod=AngleTolerance
    SimplifyOptions_SimplificationDistanceThreshold=10
    SimplifyOptions_SimplificationAngleThreshold=10
    SimplifyOptions_MinTargetFaceCount=12
   )
   > ApproximateConvexHullsWithSimplerCollisionShapes!(
    execute
    SimpleCollision:"Geometry Script Simple Collision Structure (by ref)"=@n10.ReturnValue
    ApproximateOptions_bFitSpheres:bool=false
    ApproximateOptions_bFitBoxes:bool=True
    ApproximateOptions_bFitCapsules:bool=False
    ApproximateOptions_DistanceThreshold:float=1
    ApproximateOptions_VolumeDiffThreshold_Fraction:float=0.01
   ) -> (then, bHasApproximated:bool)
   > n11 = SetSimpleCollisionOfStaticMesh(
    SimpleCollision=@n10.ReturnValue
    StaticMesh=Get StaticMesh()
    Options_bEmitTransaction=True
    StaticMeshCollisionOptions_bMarkAsCustomized=false
   )
  ]}
 Generate Simplify Collision For Mesh Containers {nodes [
   n1 = Generate Simplify Collision For Mesh Containers!() -> (then, MergedGeo:"Atom Merged Geometry Structure (by ref)") {is=title K2Node_FunctionEntry}
   > n2 = Set GeometryLODs(GeometryLODs:Atom Primitive Geometry LODs Structure=Break Atom Merged Geometry!(AtomMergedGeometry:"Atom Merged Geometry Structure (by ref)"=@n1) -> (GeometryLODs:Atom Primitive Geometry LODs Structure, Materials:Array of Material Interface Object References) {is=title K2Node_BreakStruct}) -> (Output_Get:Atom Primitive Geometry LODs Structure)
   n3 = Break Atom Primitive Geometry LODs!(AtomPrimitiveGeometryLODs:"Atom Primitive Geometry LODs Structure (by ref)"=@n2) -> (Collision:Atom Primitive Collision Geometry Structure) {is=title K2Node_BreakStruct}
   n4 = IsAtomPrimitiveCollisionValid!(AtomCollision:"Atom Primitive Collision Geometry Structure (by ref)"=@n3) -> (ReturnValue:bool)
   > n5 = Branch(Condition=OR Boolean!(A=Get Debug - Bypass Mesh Convex Merge(), B=NOT Boolean(A=@n4)) -> (ReturnValue))
   > Return Node!(execute, StructOut:"Atom Merged Geometry Structure (by ref)"=@n1)
   n6 = Get Force Shape Count() -> (Force Shape Count=0)
   n7 = "Clamp (Integer)"(Value=@n6, Min=0, Max=1) -> (ReturnValue=0)
   n8 = Get Mesh Protection - Min Sample Radius()
   n9 = MergeSimpleCollisionShapes(
    execute=@n5.else
    SimpleCollision=Atom Primitive Collision to Geometry Script Collision!(AtomCollision:"Atom Primitive Collision Geometry Structure (by ref)"=@n3) -> (ReturnValue:Geometry Script Simple Collision Structure)
    MergeOptions_MaxShapeCount=@n6
    MergeOptions_ErrorTolerance=0
    MergeOptions_MinThicknessTolerance=0
    MergeOptions_bConsiderAllPossibleMerges=False
    MergeOptions_PrecomputedNegativeSpace=()
    MergeOptions_bComputeNegativeSpace="To Boolean (Integer)"(InInt=int - int(A=1, B=@n7))
    MergeOptions_ComputeNegativeSpaceOptions_SampleMethod=VoxelSearch
    MergeOptions_ComputeNegativeSpaceOptions_bRequireSearchSampleCoverage=true
    MergeOptions_ComputeNegativeSpaceOptions_bOnlyConnectedToHull=False
    MergeOptions_ComputeNegativeSpaceOptions_MaxVoxelsPerDim=128
    MergeOptions_ComputeNegativeSpaceOptions_TargetNumSamples=10
    MergeOptions_ComputeNegativeSpaceOptions_MinSampleSpacing=@n8
    MergeOptions_ComputeNegativeSpaceOptions_NegativeSpaceTolerance=Get Mesh Protection - Tolerance()
    MergeOptions_ComputeNegativeSpaceOptions_MinRadius=@n8
    MergeOptions_ShapeToHullTriangulation=(SphereStepsPerSide=4,CapsuleHemisphereSteps=5,CapsuleCircleSteps=8,bApproximateLevelSetsWithCubes=False)
   )
   > SimplifyConvexHulls(
    SimpleCollision=@n9.ReturnValue
    SimplifyOptions_SimplificationMethod=MeshQSlim
    SimplifyOptions_SimplificationDistanceThreshold=Get Simplification Distance()
    SimplifyOptions_SimplificationAngleThreshold=10
    SimplifyOptions_MinTargetFaceCount=Get Simplification Min Face Count() -> (Simplification Min Face Count=0)
   )
   > SimplifyConvexHulls(
    SimpleCollision=@n9.ReturnValue
    SimplifyOptions_SimplificationMethod=AngleTolerance
    SimplifyOptions_SimplificationDistanceThreshold=10
    SimplifyOptions_SimplificationAngleThreshold=10
    SimplifyOptions_MinTargetFaceCount=12
   )
   > n10 = Branch(Condition=@n9)
   > n11 = ApproximateConvexHullsWithSimplerCollisionShapes(
    SimpleCollision=@n9.ReturnValue
    ApproximateOptions_bFitSpheres=false
    ApproximateOptions_bFitBoxes=True
    ApproximateOptions_bFitCapsules=False
    ApproximateOptions_DistanceThreshold=1
    ApproximateOptions_VolumeDiffThreshold_Fraction=0.01
   )
   n12 = Set members in Atom Primitive Geometry LODs!(
    execute:Exec=[@n11.then, @n10.else]
    StructRef:"Atom Primitive Geometry LODs Structure (by ref)"=Get GeometryLODs() -> (GeometryLODs:Atom Primitive Geometry LODs Structure)
    Collision:Atom Primitive Collision Geometry Structure=Geometry Script Collision to Atom Primitive Collision!(GeometryScriptSimpleCollision:"Geometry Script Simple Collision Structure (by ref)"=@n9.ReturnValue) -> (ReturnValue:Atom Primitive Collision Geometry Structure)
   ) -> (then:Exec, StructOut:"Atom Primitive Geometry LODs Structure (by ref)") {is=title K2Node_SetFieldsInStruct}
   > n13 = Set members in Atom Merged Geometry!(execute:Exec, StructRef:"Atom Merged Geometry Structure (by ref)"=@n1, GeometryLODs:Atom Primitive Geometry LODs Structure=@n12) -> (then:Exec, StructOut:"Atom Merged Geometry Structure (by ref)") {is=title K2Node_SetFieldsInStruct}
   > Return Node(StructOut=@n13)
  ]}
 Get CP Geometry Material Assignments For Geometry Style {nodes [
   n1 = Get CP Geometry Material Assignments For Geometry Style!() -> (
    then
    Part Info:Atom Model Part Instance Info Structure
    Common Part Index:int=0
    Primitive Material Assignments:Atom Primitive Geometry Material Assignments Structure
    Common Part Geometry:Atom Primitive Geometry Object Reference
    Geometry Style Settings:Atom Processor Geometry Style Settings Structure
    Style Name:String
   ) {is=title K2Node_FunctionEntry}
   > n2 = Parent: Get CP Geometry Material Assignments For Geometry Style!(
    execute
    Part Info:Atom Model Part Instance Info Structure=@n1
    Common Part Index:int=@n1.Common Part Index
    Primitive Material Assignments:Atom Primitive Geometry Material Assignments Structure=@n1.Primitive Material Assignments
    Common Part Geometry:Atom Primitive Geometry Object Reference=@n1.Common Part Geometry
    Geometry Style Settings:Atom Processor Geometry Style Settings Structure=@n1.Geometry Style Settings
    Style Name:String=@n1.Style Name
   ) -> (then, Material Assignments:Atom Primitive Geometry Material Assignments Structure) {is=title K2Node_CallParentFunction}
   > Set MaterialAssignments(MaterialAssignments:Atom Primitive Geometry Material Assignments Structure=@n2) -> (Output_Get:Atom Primitive Geometry Material Assignments Structure)
   > n3 = Branch(Condition=Analytic Common Parts Requested On Any LODs())
   n4 = Get Part Info() -> (Part Info:Atom Model Part Instance Info Structure)
   n5 = GetPartColorInfo!(Part Info:Atom Model Part Instance Info Structure=@n4, Color Change Index:int=0) -> (Color Id:int=0, Material Type:EAtomMaterialType Enum=None)
   > n6 = Add Material Assignment!(
    Exec:Exec
    Material Assignment:Atom Primitive Geometry Material Assignments Structure=Get MaterialAssignments() -> (MaterialAssignments:Atom Primitive Geometry Material Assignments Structure)
    Material Features:Atom Material Features Structure=Make Atom Material Features!(
     MaterialType:EAtomMaterialType Enum=@n5.Material Type
     UDIMDecorationMap:Map of Integers to Texture Object References
     NormalMap:Texture Object Reference
     AlphaMask:Texture Object Reference
     TechnicNormalMap:Texture Object Reference
     TechnicMasks:Texture Object Reference
     LookUpTableAtlasTexture:Texture Object Reference=Get LUT Atlas Override!() -> (AsTexture 2D:Texture 2D Object Reference) {is=title K2Node_CallFunction}
     ColorSource:EAtomMaterialColorSource Enum=Get Material Color Source() -> (Material Color Source:EAtomMaterialColorSource Enum=VertexColor)
     ColorId:int=@n5
     UVFeature:EAtomMaterialUVFeature Enum=AnalyticalCommonParts
     TranslucencyType:EAtomMaterialTranslucencyType Enum=Get Translucency Type For Part!(Part Info:Atom Model Part Instance Info Structure=@n4) -> (Translucency Type:EAtomMaterialTranslucencyType Enum=Translucent) {is=title K2Node_CallFunction}
     MaterialOverride:Atom Material Override Structure=(Material=None,bManualColorSource=False)
     AdditionalFeatures:Array of Instanced Struct Structures
    ) -> (AtomMaterialFeatures:Atom Material Features Structure) {is=title K2Node_MakeStruct}
    Slot Name:String=CP_Analytical
   ) -> (Out:Exec) {is=title K2Node_MacroInstance}
   Return Node!(execute=[@n3.else, @n6.Out], Material Assignments:Atom Primitive Geometry Material Assignments Structure=Get MaterialAssignments())
  ]}
 Get Collision Geometry for Part {nodes [
   n1 = Get Collision Geometry for Part!() -> (
    then
    Geometry:Atom Primitive Geometry Object Reference
    CollisionType:EPrimitiveAutoCollisionType Enum=Box
    PartInfo:Atom Model Part Instance Info Structure
   ) {is=title K2Node_FunctionEntry}
   > n2 = Parent: Get Collision Geometry for Part!(
    execute
    Geometry:Atom Primitive Geometry Object Reference=@n1
    CollisionType:EPrimitiveAutoCollisionType Enum=@n1.CollisionType
    PartInfo:Atom Model Part Instance Info Structure=@n1.PartInfo
   ) -> (then, Collision Geometry:Atom Primitive Collision Geometry Structure) {is=title K2Node_CallParentFunction}
   > InjectPresetCollisionSettings()
   n3 = NOT Boolean(A=Get Primitive Geometry Container Collision!(Part Info:Atom Model Part Instance Info Structure=@n1.PartInfo) -> (Success:bool, Geometry Container:Atom Primitive Geometry Container Object Reference) {is=title K2Node_CallFunction})
   n4 = Get Optimize Primitive Collision() -> (Optimize Primitive Collision:bool)
   > n5 = Branch(Condition=AND Boolean(A=IsAtomPrimitiveCollisionValid(AtomCollision=@n2), B=AND Boolean(A=@n4, B=@n3)))
   n6 = Get Brick Protection - Min Sample Radius() -> (Brick Protection - Min Sample Radius:double)
   > n7 = MergeSimpleCollisionShapes(
    SimpleCollision=Atom Primitive Collision to Geometry Script Collision(AtomCollision=@n2)
    MergeOptions_MaxShapeCount=0
    MergeOptions_ErrorTolerance=0
    MergeOptions_MinThicknessTolerance=0
    MergeOptions_bConsiderAllPossibleMerges=False
    MergeOptions_PrecomputedNegativeSpace=()
    MergeOptions_bComputeNegativeSpace=true
    MergeOptions_ComputeNegativeSpaceOptions_SampleMethod=VoxelSearch
    MergeOptions_ComputeNegativeSpaceOptions_bRequireSearchSampleCoverage=true
    MergeOptions_ComputeNegativeSpaceOptions_bOnlyConnectedToHull=False
    MergeOptions_ComputeNegativeSpaceOptions_MaxVoxelsPerDim=128
    MergeOptions_ComputeNegativeSpaceOptions_TargetNumSamples=10
    MergeOptions_ComputeNegativeSpaceOptions_MinSampleSpacing=@n6
    MergeOptions_ComputeNegativeSpaceOptions_NegativeSpaceTolerance=Get Brick Protection - Tolerance() -> (Brick Protection - Tolerance:double)
    MergeOptions_ComputeNegativeSpaceOptions_MinRadius=@n6
    MergeOptions_ShapeToHullTriangulation=(SphereStepsPerSide=4,CapsuleHemisphereSteps=5,CapsuleCircleSteps=8,bApproximateLevelSetsWithCubes=False)
   )
   > n8 = GetSimpleCollisionShapeCount!(execute, SimpleCollision:"Geometry Script Simple Collision Structure (by ref)"=@n7.ReturnValue) -> (then, ReturnValue:int=0)
   > n9 = MergeSimpleCollisionShapes!(
    execute
    SimpleCollision=@n7.ReturnValue
    MergeOptions_MaxShapeCount=@n8
    MergeOptions_ErrorTolerance=0
    MergeOptions_MinThicknessTolerance=Get Brick Protection - Min Thickness() -> (Brick Protection - Min Thickness:double)
    MergeOptions_bConsiderAllPossibleMerges=False
    MergeOptions_PrecomputedNegativeSpace=()
    MergeOptions_bComputeNegativeSpace=False
    MergeOptions_ComputeNegativeSpaceOptions:Compute Negative Space Options Structure=(SampleMethod=Uniform,bRequireSearchSampleCoverage=False,bOnlyConnectedToHull=False,MaxVoxelsPerDim=128,TargetNumSamples=50,MinSampleSpacing=1,NegativeSpaceTolerance=2,ToleranceDirectionalScales=,MinRadius=10)
    MergeOptions_ShapeToHullTriangulation=(SphereStepsPerSide=4,CapsuleHemisphereSteps=5,CapsuleCircleSteps=8,bApproximateLevelSetsWithCubes=False)
   ) -> (then, bHasMerged, ReturnValue)
   > SimplifyConvexHulls(
    SimpleCollision=@n9.ReturnValue
    SimplifyOptions_SimplificationMethod=MeshQSlim
    SimplifyOptions_SimplificationDistanceThreshold=0.5
    SimplifyOptions_SimplificationAngleThreshold=10
    SimplifyOptions_MinTargetFaceCount=12
   )
   > Return Node!(
    execute
    Collision Geometry:Atom Primitive Collision Geometry Structure=Geometry Script Collision to Atom Primitive Collision(GeometryScriptSimpleCollision=@n9.ReturnValue)
   )
   Return Node(execute=@n5.else, Collision Geometry=@n2)
  ]}
 Get Geometry LODs For Part - Main Pipeline {nodes [
   n1 = Get Geometry LODs For Part - Main Pipeline!() -> (
    then
    Part Info:Atom Model Part Instance Info Structure
    Applied Optimizations:Atom Processor Applied Optimizations Structure
    GeometryContainer:Atom Primitive Geometry Container Object Reference
    Color Change Index:int
    Shell Mesh:Dynamic Mesh Object Reference
    Caps Meshes:Array of Dynamic Mesh Object References
    Technic Caps Meshes:Array of Dynamic Mesh Object References
    Technic Connectors Meshes:Array of Dynamic Mesh Object References
   ) {is=title K2Node_FunctionEntry}
   > Set ShellDynamicMesh(ShellDynamicMesh:Dynamic Mesh Object Reference=@n1.Shell Mesh) -> (Output_Get:Dynamic Mesh Object Reference)
   > Set CapsDynamicMeshes(CapsDynamicMeshes:Array of Dynamic Mesh Object References=@n1.Caps Meshes) -> (Output_Get:Array of Dynamic Mesh Object References)
   > Set TechnicCapsDynamicMeshes(TechnicCapsDynamicMeshes:Array of Dynamic Mesh Object References=@n1.Technic Caps Meshes) -> (Output_Get:Array of Dynamic Mesh Object References)
   > Set TechnicConnectorsDynamicMeshes(TechnicConnectorsDynamicMeshes:Array of Dynamic Mesh Object References=@n1.Technic Connectors Meshes) -> (Output_Get:Array of Dynamic Mesh Object References)
   > n2 = Branch(Condition=Has Color Change Baked in Geometry!(PartInfo:Atom Model Part Instance Info Structure=@n1) -> (HasColorChangeBakedInGeometry:bool))
   > n3 = Bake Color Change In Vertex Color!(
    execute
    self
    Part Info:Atom Model Part Instance Info Structure=Get Part Info()
    Color Change index:int=Get Color Change Index() -> (Color Change Index:int=0)
    Shell Mesh:Dynamic Mesh Object Reference=Get ShellDynamicMesh() -> (ShellDynamicMesh:Dynamic Mesh Object Reference)
    Caps Meshes:Array of Dynamic Mesh Object References=Get CapsDynamicMeshes() -> (CapsDynamicMeshes:Array of Dynamic Mesh Object References)
    Technic Caps Meshes:Array of Dynamic Mesh Object References=Get TechnicCapsDynamicMeshes() -> (TechnicCapsDynamicMeshes:Array of Dynamic Mesh Object References)
    Technic Connectors Meshes:Array of Dynamic Mesh Object References=Get TechnicConnectorsDynamicMeshes() -> (TechnicConnectorsDynamicMeshes:Array of Dynamic Mesh Object References)
   ) -> (
    then
    Out Shell Mesh:Dynamic Mesh Object Reference
    Out Caps Meshes:Array of Dynamic Mesh Object References
    Out Technic Caps Meshes:Array of Dynamic Mesh Object References
    Out Technic Connectors Meshes:Array of Dynamic Mesh Object References
   ) {is=title K2Node_CallFunction}
   > Set ShellDynamicMesh(ShellDynamicMesh=@n3)
   > Set CapsDynamicMeshes(CapsDynamicMeshes=@n3.Out Caps Meshes)
   > Set TechnicCapsDynamicMeshes(TechnicCapsDynamicMeshes=@n3.Out Technic Caps Meshes)
   > n4 = Set TechnicConnectorsDynamicMeshes(TechnicConnectorsDynamicMeshes=@n3.Out Technic Connectors Meshes)
   n5 = Branch(
    execute=[@n4.then, @n2.else]
    Condition=Break Atom Processor Applied Optimizations!(AtomProcessorAppliedOptimizations=Get Applied Optimizations() -> (Applied Optimizations:Atom Processor Applied Optimizations Structure)) -> (bApplyFakeTranslucent_25_#9:bool=~)
   )
   > n6 = Process Fake Translucent!(
    execute
    self
    Part Info:Atom Model Part Instance Info Structure=Get Part Info()
    Applied Optimizations:Atom Processor Applied Optimizations Structure=Get Applied Optimizations()
    Geometry Container:Atom Primitive Geometry Container Object Reference=Get GeometryContainer() -> (GeometryContainer:Atom Primitive Geometry Container Object Reference)
    DynamicMesh:Dynamic Mesh Object Reference=Get ShellDynamicMesh()
    Size Channel:int=0
    DepthChannel:int=1
   ) -> (then, UpdatedMesh:Dynamic Mesh Object Reference)
   > n7 = Set ShellDynamicMesh(ShellDynamicMesh=@n6)
   n8 = Branch(
    execute=[@n7.then, @n5.else]
    Condition=Break Atom Processor Applied Optimizations!(AtomProcessorAppliedOptimizations=Get Applied Optimizations()) -> (bApplyFakeBevels_3_#3:bool=~)
   )
   > n9 = Process Fake Bevels!(
    execute
    self
    Part Info:Atom Model Part Instance Info Structure=Get Part Info()
    Applied Optimizations:Atom Processor Applied Optimizations Structure=Get Applied Optimizations()
    Geometry Container:Atom Primitive Geometry Container Object Reference=Get GeometryContainer()
    Mesh:Dynamic Mesh Object Reference=Get ShellDynamicMesh()
    Fake Bevels Channel:int=0
   ) -> (then, UpdatedMesh:Dynamic Mesh Object Reference)
   > n10 = Set ShellDynamicMesh(ShellDynamicMesh=@n9)
   n11 = Branch(
    execute=[@n10.then, @n8.else]
    Condition=Break Atom Processor Applied Optimizations!(AtomProcessorAppliedOptimizations=Get Applied Optimizations()) -> (bApplyFakeUnderside_9_#4:bool=~)
   )
   > n12 = Process Fake Underside!(
    execute
    self
    Part Info:Atom Model Part Instance Info Structure=Get Part Info()
    Applied Optimizations:Atom Processor Applied Optimizations Structure=Get Applied Optimizations()
    Geometry Container:Atom Primitive Geometry Container Object Reference=Get GeometryContainer()
    Cap Meshes:Array of Dynamic Mesh Object References=Get CapsDynamicMeshes()
    Underside Channel:int=0
   ) -> (then, Updated Meshes:Array of Dynamic Mesh Object References)
   > n13 = Set CapsDynamicMeshes(CapsDynamicMeshes=@n12)
   n14 = Branch(
    execute=[@n13.then, @n11.else]
    Condition=Break Atom Processor Applied Optimizations!(AtomProcessorAppliedOptimizations=Get Applied Optimizations()) -> (bApplyTechnicCap_12_#7:bool=~)
   )
   > n15 = Process Fake Technic Hole!(
    execute
    self
    Part Info:Atom Model Part Instance Info Structure=Get Part Info()
    Applied Optimizations:Atom Processor Applied Optimizations Structure=Get Applied Optimizations()
    Geometry Container:Atom Primitive Geometry Container Object Reference=Get GeometryContainer()
    Technic Caps Meshes:Array of Dynamic Mesh Object References=Get TechnicCapsDynamicMeshes()
   ) -> (then, Out Technic Caps:Array of Dynamic Mesh Object References)
   > n16 = Set TechnicCapsDynamicMeshes(TechnicCapsDynamicMeshes=@n15)
   n17 = Branch(
    execute=[@n16.then, @n14.else]
    Condition=Break Atom Processor Applied Optimizations!(AtomProcessorAppliedOptimizations=Get Applied Optimizations()) -> (bApplyFakeTechnicConnector_22_#8:bool=~)
   )
   > n18 = Process Fake Technic Connector!(
    execute
    self
    Part Info:Atom Model Part Instance Info Structure=Get Part Info()
    Applied Optimizations:Atom Processor Applied Optimizations Structure=Get Applied Optimizations()
    Geometry Container:Atom Primitive Geometry Container Object Reference=Get GeometryContainer()
    Shell Mesh:Dynamic Mesh Object Reference=Get Shell Mesh()
    Technic Connectors Meshes:Array of Dynamic Mesh Object References=Get TechnicConnectorsDynamicMeshes()
   ) -> (then, Out Shell Mesh:Dynamic Mesh Object Reference, Out Technic Connectors Meshes:Array of Dynamic Mesh Object References)
   > Set ShellDynamicMesh(ShellDynamicMesh=@n18)
   > n19 = Set TechnicConnectorsDynamicMeshes(TechnicConnectorsDynamicMeshes=@n18.Out Technic Connectors Meshes)
   n20 = Branch(
    execute=[@n19.then, @n17.else]
    Condition=Break Atom Processor Applied Optimizations!(AtomProcessorAppliedOptimizations=Get Applied Optimizations()) -> (bApplyGlow_13_#11:bool=~)
   )
   > n21 = Process Glow!(
    execute
    self
    Part Info:Atom Model Part Instance Info Structure=Get Part Info()
    Applied Optimizations:Atom Processor Applied Optimizations Structure=Get Applied Optimizations()
    Geometry Container:Atom Primitive Geometry Container Object Reference=Get GeometryContainer()
    Dynamic Mesh:Dynamic Mesh Object Reference=Get ShellDynamicMesh()
    UV Channel XY:int=1
    UV Channel Z:int=2
   ) -> (then, Updated Mesh:Dynamic Mesh Object Reference)
   > n22 = Set ShellDynamicMesh(ShellDynamicMesh=@n21)
   Return Node!(
    execute=[@n22.then, @n20.else]
    Out Shell Mesh:Dynamic Mesh Object Reference=Get ShellDynamicMesh()
    Out Caps Meshes:Array of Dynamic Mesh Object References=Get CapsDynamicMeshes()
    Out Technic Caps Meshes:Array of Dynamic Mesh Object References=Get TechnicCapsDynamicMeshes()
    Out Technic Connectors Meshes:Array of Dynamic Mesh Object References=Get TechnicConnectorsDynamicMeshes()
   )
  ]}
 Get Geometry LODs For Part - Post Pipeline {nodes [
   n1 = Get Geometry LODs For Part - Post Pipeline!() -> (
    then
    Part Info:Atom Model Part Instance Info Structure
    Applied Optimizations:Atom Processor Applied Optimizations Structure
    Geometry Container:Atom Primitive Geometry Container Object Reference
    Color Change Index:int
    Shell Mesh:Dynamic Mesh Object Reference
    Caps Meshes:Array of Dynamic Mesh Object References
    Technic Caps Meshes:Array of Dynamic Mesh Object References
    Technic Connectors Meshes:Array of Dynamic Mesh Object References
   ) {is=title K2Node_FunctionEntry}
   > Return Node(
    Out Shell Mesh=@n1.Shell Mesh
    Out Caps Meshes=@n1.Caps Meshes
    Out Technic Caps Meshes=@n1.Technic Caps Meshes
    Out Technic Connectors Meshes=@n1.Technic Connectors Meshes
   )
  ]}
 Get Geometry Material Assignments For Geometry Style {nodes [
   Get Geometry Material Assignments For Geometry Style!() -> (
    then
    PartInfo:Atom Model Part Instance Info Structure
    Geometry Style Settings:Atom Processor Geometry Style Settings Structure
    StyleName:String
    Color Change Index:int=0
   ) {is=title K2Node_FunctionEntry}
   > n1 = Sequence!(execute) -> (then_0, then_1, then_2, then_3:Exec, then_4:Exec, then_5:Exec)
   n2 = Get Optimizations To Apply For Geometry Style!(
    Part Info=Get PartInfo()
    Geometry Style Settings=Get Geometry Style Settings() -> (Geometry Style Settings:Atom Processor Geometry Style Settings Structure)
    ColorChangeIndex=Get Color Change Index() -> (Color Change Index=0)
   ) -> (
    Applied Optimizations_bApplyFakeBevels_3_#3:bool=~
    Applied Optimizations_bApplyFakeUnderside_9_#4:bool=~
    Applied Optimizations_bAnalyticFakeUnderside_10_#5:bool=~
    Applied Optimizations_bApplyFlatCap_11_#6:bool=~
    Applied Optimizations_bApplyTechnicCap_12_#7:bool=~
    Applied Optimizations_bApplyFakeTechnicConnector_22_#8:bool=~
    Applied Optimizations_bApplyFakeTranslucent_25_#9:bool=~
    Applied Optimizations_bApplyDitheredTranslucent_24_#10:bool=~
    Applied Optimizations_bApplyGlow_13_#11:bool=~
    Applied Optimizations_bUseNormalMap_16_#12:bool=~
    Applied Optimizations_bUseAlphaMask_17_#13:bool=~
    Applied Optimizations_ExtraOptmizationsToApply_20_#14:Instanced Struct Structure
   )
   n3 = Branch(
    execute=@n1.then_2
    Condition=AND Boolean(
     A=OR Boolean(
      A=@n2.Applied Optimizations_bUseNormalMap_16_#12
      B=@n2.Applied Optimizations_bUseAlphaMask_17_#13
     )
     B=Get Use Geometry Style Settings() -> (Use Geometry Style Settings:bool)
    )
   )
   n4 = Break Atom Model Part Instance Info!(AtomModelPartInstanceInfo:"Atom Model Part Instance Info Structure (by ref)"=Get PartInfo()) -> (Primitive:Atom Primitive Object Reference) {is=title K2Node_BreakStruct}
   n5 = Break Atom Processor Geometry Style Settings!(AtomProcessorGeometryStyleSettings=Get Geometry Style Settings()) -> (
    GeometryStylePreferences_4_#17:Array of Strings
    PreferredGeometryLOD_30_#18:int
    ShouldBevel_8_#19:bool=~
    BevelWidth_11_#20:double=~
    BevelAngleThreshold_13_#21:double=~
    UseImprovedBevelAlgorithm_37_#22:bool=~
    OnlyBevelBlock-LikePrimitives_39_#23:bool=~
    NumberofSubdivisionsToApply_16_#24:int
    FakeBevel_23_#25:bool=~
    FakeUnderside_25_#15=~
    AnalyticFakeUnderside_32_#26:bool=~
    ForceFlatCap_35_#16=~
    FakeTechnicHole_41_#1=~
    AnalyticCommonParts_27_#2:bool=~
   )
   n6 = GetGeometryContainerForBestExportStyle!(
    self:Atom Primitive Object Reference=@n4
    GeometryStylePriorityList:Array of Strings=@n5
    LODIndex:int=@n5.PreferredGeometryLOD_30_#18
   ) -> (ReturnValue:Atom Primitive Geometry Container Object Reference)
   n7 = GetLOD!(self:Atom Primitive Geometry Container Object Reference=@n6) -> (ReturnValue:int=0)
   n8 = GetExportStyleName!(self:Atom Primitive Geometry Container Object Reference=@n6) -> (ReturnValue:String)
   n9 = Get PartInfo()
   n10 = GetPartColorInfo(Part Info=@n9, Color Change Index=Get Color Change Index() -> (Color Change Index=0)) -> (Color Id=0, Material Type=None)
   n11 = Make Atom Material Features(
    MaterialType=@n10.Material Type
    UDIMDecorationMap=Get UDIMDecorationMap() -> (UDIMDecorationMap:Map of Integers to Texture Object References)
    NormalMap=LoadDetailTextureForExportStyleAndSection!(
     self:Atom Primitive Object Reference=@n4
     ExportStyle:String=@n8
     LOD:int=@n7
     TextureType:EAtomPrimitiveDetailTextureType Enum=NormalMap
     Section:EPrimitiveGeometrySectionFlag Enum=Shell
    ) -> (ReturnValue:Texture Object Reference)
    AlphaMask=LoadDetailTextureForExportStyleAndSection(self=@n4, ExportStyle=@n8, LOD=@n7, TextureType=AlphaMask, Section=Shell)
    LookUpTableAtlasTexture=Get LUT Atlas Override()
    ColorSource=Get Material Color Source() -> (Material Color Source=VertexColor)
    ColorId=@n10
    UVFeature=Select(
     Option 0:EAtomMaterialUVFeature Enum=Literal enum EAtomMaterialUVFeature!(Enum:EAtomMaterialUVFeature Enum=None) -> (ReturnValue:EAtomMaterialUVFeature Enum) {is=title K2Node_EnumLiteral}
     Option 1:EAtomMaterialUVFeature Enum=Literal enum EAtomMaterialUVFeature(Enum=FakeTranslucent)
     Index=@n2.Applied Optimizations_bApplyFakeTranslucent_25_#9
    ) -> (ReturnValue:EAtomMaterialUVFeature Enum)
    TranslucencyType=Get Translucency Type For Part(Part Info=@n9) -> (Translucency Type=Translucent)
    MaterialOverride=(Material=None,bManualColorSource=False)
   )
   n12 = Select(
    Option 0:Atom Material Features Structure=@n11
    Option 1:Atom Material Features Structure=Get MaterialOverride() -> (MaterialOverride:Atom Material Features Structure)
    Index=Get bHasMaterialOverride() -> (bHasMaterialOverride:bool)
   ) -> (ReturnValue:Atom Material Features Structure)
   n13 = Append(A=Get StyleName() -> (StyleName:String), B=_DetailTexture)
   > Set MaterialAssignments(MaterialAssignments=Make Atom Primitive Geometry Material Assignments!(
     MaterialSlotName:Array of Strings=Make Array("[0]":String=@n13) -> (Array:Array of Strings)
     MaterialFeatures:Array of Atom Material Features Structures=Make Array("[0]":Atom Material Features Structure=@n12) -> (Array:Array of Atom Material Features Structures)
    ) -> (AtomPrimitiveGeometryMaterialAssignments:Atom Primitive Geometry Material Assignments Structure) {is=title K2Node_MakeStruct})
   n14 = Get PartInfo()
   n15 = GetPartColorInfo(Part Info=@n14, Color Change Index=Get Color Change Index() -> (Color Change Index=0)) -> (Color Id=0, Material Type=None)
   n16 = Literal enum EAtomMaterialUVFeature(Enum=FakeTranslucent)
   n17 = Select(
    Option 0:EAtomMaterialUVFeature Enum=Literal enum EAtomMaterialUVFeature(Enum=FakeBevels)
    Option 1:EAtomMaterialUVFeature Enum=Literal enum EAtomMaterialUVFeature(Enum=FakeBevels)
    Index=@n2
   ) -> (ReturnValue:EAtomMaterialUVFeature Enum)
   n18 = Select(
    Option 0:Atom Material Features Structure=Make Atom Material Features(
     MaterialType=@n15.Material Type
     UDIMDecorationMap=Get UDIMDecorationMap()
     LookUpTableAtlasTexture=Get LUT Atlas Override()
     ColorSource=Get Material Color Source() -> (Material Color Source=VertexColor)
     ColorId=@n15
     UVFeature=Select(
      Option 0:EAtomMaterialUVFeature Enum=@n17
      Option 1:EAtomMaterialUVFeature Enum=@n16
      Index=@n2.Applied Optimizations_bApplyFakeTranslucent_25_#9
     ) -> (ReturnValue:EAtomMaterialUVFeature Enum)
     TranslucencyType=Get Translucency Type For Part(Part Info=@n14) -> (Translucency Type=Translucent)
     MaterialOverride=(Material=None,bManualColorSource=False)
    )
    Option 1:Atom Material Features Structure=Get MaterialOverride()
    Index=Get bHasMaterialOverride()
   ) -> (ReturnValue:Atom Material Features Structure)
   n19 = Append(A=Get StyleName(), B=_Default)
   Set MaterialAssignments(
    execute=@n3.else
    MaterialAssignments=Make Atom Primitive Geometry Material Assignments(
     MaterialSlotName=Make Array("[0]":String=@n19) -> (Array:Array of Strings)
     MaterialFeatures=Make Array("[0]":Atom Material Features Structure=@n18) -> (Array:Array of Atom Material Features Structures)
    )
   )
   n20 = Break Atom Model Part Instance Info!(AtomModelPartInstanceInfo=Get PartInfo()) -> (Colors:Array of Atom Model Part Color Info Structures)
   > n21 = For Loop(FirstIndex=1, LastIndex=int - int(A=Length(TargetArray:Array of Atom Model Part Color Info Structures=@n20)))
   > Add Material Assignment(
    Material Assignment=Get MaterialAssignments()
    Material Features=@n18
    Slot Name=Append!(A=Get StyleName(), B=_Default, C:String="To String (Integer)"(InInt=@n21)) -> (ReturnValue)
   )
   n22 = Break Atom Processor Geometry Style Settings!(AtomProcessorGeometryStyleSettings=Get Geometry Style Settings()) -> (
    UseImprovedBevelAlgorithm_37_#22=~
    OnlyBevelBlock-LikePrimitives_39_#23=~
    FakeUnderside_25_#15=~
    AnalyticFakeUnderside_32_#26=~
    FakeTechnicHole_41_#1=~
   )
   Branch(
    execute=@n1.then_3
    Condition=AND Boolean(A=@n2.Applied Optimizations_bApplyFakeUnderside_9_#4, B=@n22.FakeUnderside_25_#15)
   )
   n23 = Get PartInfo()
   n24 = GetPartColorInfo(Part Info=@n23, Color Change Index=Get Color Change Index() -> (Color Change Index=0)) -> (Color Id=0, Material Type=None)
   n25 = Get Translucency Type For Part(Part Info=@n23) -> (Translucency Type=Translucent)
   n26 = Get LUT Atlas Override()
   n27 = Select(Option 0=FakeUnderside, Option 1=FakeAnalyticalUnderside, Index=@n22.AnalyticFakeUnderside_32_#26)
   n28 = Get Material Color Source() -> (Material Color Source=VertexColor)
   > Add Material Assignment(
    Material Assignment=Get MaterialAssignments()
    Material Features=Select(
     Option 0:Atom Material Features Structure=Make Atom Material Features(
      MaterialType=@n24.Material Type
      LookUpTableAtlasTexture=@n26
      ColorSource=@n28
      ColorId=@n24
      UVFeature=@n27
      TranslucencyType=@n25
      MaterialOverride=(Material=None,bManualColorSource=False)
     )
     Option 1:Atom Material Features Structure=Get MaterialOverride()
     Index=Get bHasMaterialOverride()
    ) -> (ReturnValue:Atom Material Features Structure)
    Slot Name=Append!(
     A=Get StyleName()
     B=SelectString(A=_AnalyticalFakeUnderside, B=_FakeUnderside, bPickA=@n22.AnalyticFakeUnderside_32_#26)
    ) -> (ReturnValue)
   )
   Return Node!(execute=@n1.then_5, Material Assignments:Atom Primitive Geometry Material Assignments Structure=Get MaterialAssignments())
   n29 = Get PartInfo()
   n30 = Break Atom Model Part Color Info!(AtomModelPartColorInfo:"Atom Model Part Color Info Structure (by ref)"="Get (a ref)"(
     Array:Array of Atom Model Part Color Info Structures=Break Atom Model Part Instance Info(AtomModelPartInstanceInfo=@n29)
     Dimension 1=Get Color Change Index() -> (Color Change Index=0)
    ) -> (Output:"Atom Model Part Color Info Structure (by ref)")) -> (ColorId:int, MaterialType:EAtomMaterialType Enum) {is=title K2Node_BreakStruct}
   n31 = Find(
    TargetMap:Map of Integers to Atom Material Override Structures=Get Material ID Overrides!() -> (Material ID Overrides:Map of Integers to Atom Material Override Structures) {is=title K2Node_CallFunction}
    Key=@n30
   ) -> (Value:Atom Material Override Structure)
   Branch(execute=@n1.then_0, Condition=@n31.ReturnValue)
   n32 = Break Atom Material Override!(AtomMaterialOverride:"Atom Material Override Structure (by ref)"=@n31) -> (Material:Material Interface Object Reference, bManualColorSource:bool=~) {is=title K2Node_BreakStruct}
   n33 = Get Material Color Source() -> (Material Color Source=VertexColor)
   > Set MaterialOverride(MaterialOverride:Atom Material Features Structure=Make Atom Material Features(
     MaterialType=None
     LookUpTableAtlasTexture=Get LUT Atlas Override()
     ColorSource=Select(Option 0:EAtomMaterialColorSource Enum=@n33, Option 1:EAtomMaterialColorSource Enum=Material, Index=@n32.bManualColorSource) -> (ReturnValue:EAtomMaterialColorSource Enum)
     ColorId=@n30
     UVFeature=None
     TranslucencyType=Translucent
     MaterialOverride=@n31
    )) -> (Output_Get:Atom Material Features Structure)
   > Set bHasMaterialOverride(bHasMaterialOverride:bool=true) -> (Output_Get:bool)
   n34 = GetIsProcessingPrimitive!() -> (Is Processing Primitive:bool)
   Branch(execute=@n1.then_1, Condition=AND Boolean(A=Get Support Decorations() -> (Support Decorations:bool), B=NOT Boolean(A=@n34)))
   > n35 = For Each Loop(Array:Array of Atom Model Part Decoration Info Structures=Break Atom Model Part Instance Info!(AtomModelPartInstanceInfo=Get PartInfo()) -> (Decorations:Array of Atom Model Part Decoration Info Structures)) -> (Array Element:Atom Model Part Decoration Info Structure)
   n36 = Break Atom Model Part Decoration Info!(AtomModelPartDecorationInfo:"Atom Model Part Decoration Info Structure (by ref)"=@n35) -> (Texture:Texture Object Reference, PrimitiveSurfaceIndex:int) {is=title K2Node_BreakStruct}
   > Add!(
    execute
    TargetMap:Map of Integers to Texture Object References=Get UDIMDecorationMap()
    Key:"Integer (by ref)"=@n36.PrimitiveSurfaceIndex
    Value:"Texture Object Reference (by ref)"=@n36
   ) -> (then) {is=title K2Node_CallFunction}
   Branch(
    execute=@n1.then_4
    Condition=AND Boolean(A=@n2.Applied Optimizations_bApplyTechnicCap_12_#7, B=@n22.FakeTechnicHole_41_#1)
   )
   n37 = Break Atom Processor Geometry Style Settings!(AtomProcessorGeometryStyleSettings=Get Geometry Style Settings()) -> (
    GeometryStylePreferences_4_#17:Array of Strings
    PreferredGeometryLOD_30_#18:int
    ShouldBevel_8_#19:bool=~
    BevelWidth_11_#20:double=~
    BevelAngleThreshold_13_#21:double=~
    UseImprovedBevelAlgorithm_37_#22=~
    OnlyBevelBlock-LikePrimitives_39_#23=~
    NumberofSubdivisionsToApply_16_#24:int
    FakeBevel_23_#25:bool=~
    FakeUnderside_25_#15=~
    AnalyticFakeUnderside_32_#26=~
    ForceFlatCap_35_#16:bool=~
    FakeTechnicHole_41_#1=~
    AnalyticCommonParts_27_#2:bool=~
   )
   n38 = Break Atom Model Part Instance Info!(AtomModelPartInstanceInfo=Get PartInfo()) -> (Primitive:Atom Primitive Object Reference)
   > n39 = For Each Loop(Array:Array of Atom Primitive Geometry And Transform Structures=GetTechnicCapsGeometry!(
     self:Atom Primitive Geometry Container Object Reference=GetGeometryContainerForBestExportStyle(self=@n38, GeometryStylePriorityList=@n37, LODIndex=@n37.PreferredGeometryLOD_30_#18)
     ColorChangeIndex:int=Get Color Change Index() -> (Color Change Index=0)
    ) -> (ReturnValue:Array of Atom Primitive Geometry And Transform Structures)) -> (Array Element:Atom Primitive Geometry And Transform Structure)
   > n40 = GetTechnicHoleMaterialFeatures!(
    execute
    self
    Part Info:Atom Model Part Instance Info Structure=Get PartInfo()
    Geometry Style Settings:Atom Processor Geometry Style Settings Structure=Get Geometry Style Settings()
    Style Name:String=Get StyleName()
    Technic Cap Index:int=@n39.Array Index
   ) -> (then, Material Features:Atom Material Features Structure, Slot Name:String)
   > Add Material Assignment(
    Material Assignment=Get MaterialAssignments()
    Material Features=Select(
     Option 0:Atom Material Features Structure=@n40
     Option 1:Atom Material Features Structure=Get MaterialOverride()
     Index=Get bHasMaterialOverride()
    ) -> (ReturnValue:Atom Material Features Structure)
    Slot Name=@n40.Slot Name
   )
   Is Empty!(TargetMap:Map of Integers to Texture Object References=Get UDIMDecorationMap()) -> (ReturnValue:bool) {is=title K2Node_CallFunction}
   Add Material Assignment(Material Assignment=Get MaterialAssignments(), Material Features=@n11, Slot Name=@n13)
  ]}
 Get Geometry Material Assignments For LOD {nodes [
   n1 = Get Geometry Material Assignments For LOD!() -> (then, PartInfo:Atom Model Part Instance Info Structure, LOD:int=0, Color Change Index:int=0) {is=title K2Node_FunctionEntry}
   > n2 = Get Geometry Style Settings For Part(Part Info=@n1)
   > n3 = Get Geometry Material Assignments For Geometry Style!(
    execute
    self:MP Merged Mesh Base Object Reference
    PartInfo:Atom Model Part Instance Info Structure=@n1
    Geometry Style Settings:Atom Processor Geometry Style Settings Structure="Get (a ref)"(Array:Array of Atom Processor Geometry Style Settings Structures=@n2, Dimension 1=@n1.LOD) -> (Output:"Atom Processor Geometry Style Settings Structure (by ref)")
    StyleName:String=Append(A=LOD, B="To String (Integer)"(InInt=@n1.LOD))
    Color Change Index:int=@n1.Color Change Index
   ) -> (then, Material Assignments:Atom Primitive Geometry Material Assignments Structure) {is=title K2Node_CallFunction}
   > Return Node(Material Assignments=@n3)
  ]}
 Get Geometry Style Settings For Part {nodes [
   n1 = Get Geometry Style Settings For Part!() -> (then, Part Info:Atom Model Part Instance Info Structure) {is=title K2Node_FunctionEntry}
   > Set LODSettingsForPart(LODSettingsForPart:Array of Atom Processor Geometry Style Settings Structures=Get LOD Geometry Style Settings()) -> (Output_Get:Array of Atom Processor Geometry Style Settings Structures)
   > n2 = Set HasGlow(HasGlow:bool=IsGlowPart!(PartInfo:Atom Model Part Instance Info Structure=@n1) -> (ShouldHaveGlow:bool)) -> (Output_Get:bool)
   > n3 = Branch(Condition=@n2)
   n4 = Get LODSettingsForPart() -> (LODSettingsForPart:Array of Atom Processor Geometry Style Settings Structures)
   > n5 = For Each Loop(Array:Array of Atom Processor Geometry Style Settings Structures=@n4) -> (Array Element:Atom Processor Geometry Style Settings Structure)
   n6 = Get Part Info()
   > n7 = Branch(Condition=NOT Boolean(A=Is Geometry Style Compatible With Part Features!(
      Part Info:"Atom Model Part Instance Info Structure (by ref)"=@n6
      Geometry Style Settings:"Atom Processor Geometry Style Settings Structure (by ref)"=@n5
     ) -> (Is Compatible:bool)))
   > n8 = Branch(Condition="Not Equal (Integer)"(A=Get LastNonAlphaLOD() -> (LastNonAlphaLOD:int=0), B=-1))
   Set members in Atom Processor Geometry Style Settings!(
    execute:Exec=[@n8.then, @n10.then]
    StructRef:"Atom Processor Geometry Style Settings Structure (by ref)"="Get (a ref)"(Array=@n4, Dimension 1=@n5.Array Index)
    PreferredGeometryLOD_30_#18:int=Get LastNonAlphaLOD() -> (LastNonAlphaLOD=0)
   ) -> (then:Exec, StructOut:"Atom Processor Geometry Style Settings Structure (by ref)") {is=title K2Node_SetFieldsInStruct}
   n9 = Break Atom Model Part Instance Info!(AtomModelPartInstanceInfo=Get Part Info()) -> (PartId:int)
   PrintString(
    execute=@n8.else
    InString=Append!(
     A="Brick "
     B="To String (Integer)"(InInt=@n9)
     C:String=" is used with with non compatible feature but doesn't have a compatible LOD. Likely this brick uses alpha masks on all LODs."
    ) -> (ReturnValue)
    bPrintToScreen=false
   )
   > n10 = Set LastNonAlphaLOD(LastNonAlphaLOD:int=0) -> (Output_Get:int=0)
   Set LastNonAlphaLOD(
    execute=@n7.else
    LastNonAlphaLOD=Break Atom Processor Geometry Style Settings(AtomProcessorGeometryStyleSettings=@n5) -> (
     ShouldBevel_8_#19=~
     BevelWidth_11_#20=~
     BevelAngleThreshold_13_#21=~
     UseImprovedBevelAlgorithm_37_#22=~
     OnlyBevelBlock-LikePrimitives_39_#23=~
     FakeBevel_23_#25=~
     FakeUnderside_25_#15=~
     AnalyticFakeUnderside_32_#26=~
     ForceFlatCap_35_#16=~
     FakeTechnicHole_41_#1=~
     AnalyticCommonParts_27_#2=~
    ).PreferredGeometryLOD_30_#18
   ) -> (Output_Get=0)
   n11 = Branch(
    execute=[@n5.Completed, @n3.else]
    Condition=AND Boolean(A=Get UseNewMergeAPI() -> (UseNewMergeAPI:bool), B=Get Append Additional Source LODs() -> (Append Additional Source LODs:bool))
   )
   n12 = Get Part Info()
   n13 = Break Atom Model Part Instance Info!(AtomModelPartInstanceInfo=@n12) -> (Primitive:Atom Primitive Object Reference)
   n14 = Get LOD Geometry Style Settings()
   n15 = "Get (a ref)"(Array=@n14, Dimension 1=Last Index!(TargetArray:Array of Atom Processor Geometry Style Settings Structures=@n14) -> (ReturnValue=0))
   n16 = Break Atom Processor Geometry Style Settings(AtomProcessorGeometryStyleSettings=@n15) -> (
    ShouldBevel_8_#19=~
    BevelWidth_11_#20=~
    BevelAngleThreshold_13_#21=~
    UseImprovedBevelAlgorithm_37_#22=~
    OnlyBevelBlock-LikePrimitives_39_#23=~
    FakeBevel_23_#25=~
    FakeUnderside_25_#15=~
    AnalyticFakeUnderside_32_#26=~
    ForceFlatCap_35_#16=~
    FakeTechnicHole_41_#1=~
    AnalyticCommonParts_27_#2=~
   )
   n17 = GetGeometryContainerForBestExportStyle(self=@n13, GeometryStylePriorityList=@n16, LODIndex=@n16.PreferredGeometryLOD_30_#18)
   > n18 = For Loop(
    FirstIndex=int + int(A=GetLOD(self=@n17) -> (ReturnValue=0))
    LastIndex=GetExportStyleMaxLOD!(self:Atom Primitive Object Reference=@n13, ExportStyle:String=GetExportStyleName(self=@n17)) -> (ReturnValue:int=0)
   )
   Return Node!(execute=[@n11.else, @n18.Completed], Geometry Styles:Array of Atom Processor Geometry Style Settings Structures=Get LODSettingsForPart())
   > Branch(Condition=Is Geometry Style Compatible With Part Features(Part Info=@n12, Geometry Style Settings=@n15))
   > n19 = Set members in Atom Processor Geometry Style Settings(
    StructRef=Copy!(Item:Atom Processor Geometry Style Settings Structure=@n15) -> (ReturnValue:Atom Processor Geometry Style Settings Structure) {is=title K2Node_Copy}
    PreferredGeometryLOD_30_#18=@n18
   )
   > Add(
    TargetArray:Array of Atom Processor Geometry Style Settings Structures=Get LODSettingsForPart()
    NewItem:"Atom Processor Geometry Style Settings Structure (by ref)"=@n19
   ) {is=title K2Node_CallArrayFunction}
  ]}
 Get Linear Vertex Color For Part {nodes [
   n1 = Get Linear Vertex Color For Part!() -> (then, Part Info:Atom Model Part Instance Info Structure, Color Change Index:int) {is=title K2Node_FunctionEntry}
   > n2 = GetVertexColorForPart!(
    execute
    self:MP Merged Mesh Base Object Reference
    Part Info:"Atom Model Part Instance Info Structure (by ref)"=@n1
    Color Change Index:int=@n1.Color Change Index
   ) -> (then, Color:Color Structure)
   n3 = Break Color!(Color:"Color Structure (by ref)"=@n2) -> (B:Byte, G:Byte, R:Byte, A:Byte) {is=title K2Node_BreakStruct}
   n4 = "To Float (Byte)"!(InByte:Byte=@n3.A) -> (ReturnValue:double)
   n5 = "To Float (Byte)"(InByte=@n3)
   n6 = "To Float (Byte)"(InByte=@n3.G)
   n7 = "To Float (Byte)"(InByte=@n3.R)
   > Return Node!(
    execute
    Linear Vertex Color:Linear Color Structure=MakeColor!(
     R:float=float / float(A=@n7, B=255)
     G:float=float / float(A=@n6, B=255)
     B:float=float / float(A=@n5, B=255)
     A:float=float / float(A=@n4, B=255)
    ) -> (ReturnValue:Linear Color Structure)
   )
  ]}
 Get Nanite Geometry For Part - Main Pipeline {nodes [
   n1 = Get Nanite Geometry For Part - Main Pipeline!() -> (
    then
    Part Info:Atom Model Part Instance Info Structure
    Applied Optimizations:Atom Processor Applied Optimizations Structure
    GeometryContainer:Atom Primitive Geometry Container Object Reference
    Color Change Index:int
    Shell Mesh:Dynamic Mesh Object Reference
   ) {is=title K2Node_FunctionEntry}
   > Set ShellDynamicMesh(ShellDynamicMesh=@n1.Shell Mesh)
   > n2 = Branch(Condition=Has Color Change Baked in Geometry(PartInfo=@n1))
   > n3 = Bake Color Change In Vertex Color(Part Info=Get Part Info(), Color Change index=Get Color Change Index() -> (Color Change Index=0), Shell Mesh=Get ShellDynamicMesh())
   > n4 = Set ShellDynamicMesh(ShellDynamicMesh=@n3)
   n5 = Branch(
    execute=[@n4.then, @n2.else]
    Condition=Break Atom Processor Applied Optimizations!(AtomProcessorAppliedOptimizations=Get Applied Optimizations()) -> (bApplyFakeTranslucent_25_#9:bool=~)
   )
   > n6 = Process Fake Translucent(
    Part Info=Get Part Info()
    Applied Optimizations=Get Applied Optimizations()
    Geometry Container=Get GeometryContainer()
    DynamicMesh=Get ShellDynamicMesh()
    Size Channel=0
    DepthChannel=1
   )
   > n7 = Set ShellDynamicMesh(ShellDynamicMesh=@n6)
   n8 = Branch(
    execute=[@n7.then, @n5.else]
    Condition=Break Atom Processor Applied Optimizations!(AtomProcessorAppliedOptimizations=Get Applied Optimizations()) -> (bApplyGlow_13_#11:bool=~)
   )
   > n9 = Process Glow(
    Part Info=Get Part Info()
    Applied Optimizations=Get Applied Optimizations()
    Geometry Container=Get GeometryContainer()
    Dynamic Mesh=Get ShellDynamicMesh()
    UV Channel XY=1
    UV Channel Z=2
   )
   > n10 = Set ShellDynamicMesh(ShellDynamicMesh=@n9)
   Return Node!(execute=[@n10.then, @n8.else], Out Shell Mesh:Dynamic Mesh Object Reference=Get ShellDynamicMesh())
  ]}
 Get Nanite Geometry For Part - Post Pipeline {nodes [
   n1 = Get Nanite Geometry For Part - Post Pipeline!() -> (
    then
    Part Info:Atom Model Part Instance Info Structure
    Applied Optimizations:Atom Processor Applied Optimizations Structure
    Geometry Container:Atom Primitive Geometry Container Object Reference
    Color Change Index:int
    Shell Mesh:Dynamic Mesh Object Reference
   ) {is=title K2Node_FunctionEntry}
   > Return Node(Out Shell Mesh=@n1.Shell Mesh)
  ]}
 Get Optimizations To Apply For Geometry Style {nodes [
   n1 = Get Optimizations To Apply For Geometry Style!() -> (
    then
    Part Info:Atom Model Part Instance Info Structure
    Geometry Style Settings:Atom Processor Geometry Style Settings Structure
    ColorChangeIndex:int
   ) {is=title K2Node_FunctionEntry}
   n2 = Get Part Supported Features For Geometry Style!(
    PartInstance:"Atom Model Part Instance Info Structure (by ref)"=@n1
    Geometry Style Settings:Atom Processor Geometry Style Settings Structure=@n1.Geometry Style Settings
    Color Change Index:int=@n1.ColorChangeIndex
   ) -> (
    Supported Features_FakeBevels_1_#27:bool=~
    Supported Features_FakeUnderside_3_#28:bool=~
    Supported Features_ForceFlatCap_11_#29:bool=~
    Supported Features_TechnicHoles_18_#30:bool=~
    Supported Features_TechnicConnectors_22_#31:bool=~
    Supported Features_NormalMap_7_#32:bool=~
    Supported Features_AlphaMask_9_#33:bool=~
    Supported Features_FakeTransluscent_5_#34:bool=~
    Supported Features_DitheredTranslucent_20_#35:bool=~
    Supported Features_Glow_13_#36:bool=~
    Supported Features_ExtraFeatures_16_#37:Instanced Struct Structure
   ) {is=title K2Node_CallFunction}
   n3 = Break Atom Processor Geometry Style Settings!(AtomProcessorGeometryStyleSettings=@n1.Geometry Style Settings) -> (
    UseImprovedBevelAlgorithm_37_#22=~
    OnlyBevelBlock-LikePrimitives_39_#23=~
    FakeBevel_23_#25=~
    FakeUnderside_25_#15=~
    AnalyticFakeUnderside_32_#26=~
    ForceFlatCap_35_#16=~
    FakeTechnicHole_41_#1=~
   )
   n4 = AND Boolean(A=@n2.Supported Features_FakeUnderside_3_#28, B=@n3.FakeUnderside_25_#15)
   n5 = AND Boolean(A=@n2.Supported Features_TechnicHoles_18_#30, B=@n3.FakeTechnicHole_41_#1)
   n6 = AND Boolean!(
    A=NOT Boolean(A=@n5)
    B=@n2.Supported Features_TechnicHoles_18_#30
    C:bool=@n3.ForceFlatCap_35_#16
   ) -> (ReturnValue)
   n7 = AND Boolean(
    A=NOT Boolean(A=@n4)
    B=@n2.Supported Features_ForceFlatCap_11_#29
    C=@n3.ForceFlatCap_35_#16
   )
   > Return Node!(
    execute
    Applied Optimizations:Atom Processor Applied Optimizations Structure=Make Atom Processor Applied Optimizations!(
     bApplyFakeBevels_3_#3:bool=AND Boolean!(A=@n2, B=@n3.FakeBevel_23_#25) -> (ReturnValue)
     bApplyFakeUnderside_9_#4:bool=@n4
     bAnalyticFakeUnderside_10_#5:bool=AND Boolean(A=@n4, B=@n3.AnalyticFakeUnderside_32_#26)
     bApplyFlatCap_11_#6:bool=OR Boolean(A=@n7, B=@n6)
     bApplyTechnicCap_12_#7:bool=@n5
     bApplyFakeTechnicConnector_22_#8:bool=AND Boolean(A=@n2.Supported Features_TechnicConnectors_22_#31, B=@n3.FakeTechnicHole_41_#1)
     bApplyFakeTranslucent_25_#9:bool=@n2.Supported Features_FakeTransluscent_5_#34
     bApplyDitheredTranslucent_24_#10:bool=@n2.Supported Features_DitheredTranslucent_20_#35
     bApplyGlow_13_#11:bool=@n2.Supported Features_Glow_13_#36
     bUseNormalMap_16_#12:bool=@n2.Supported Features_NormalMap_7_#32
     bUseAlphaMask_17_#13:bool=@n2.Supported Features_AlphaMask_9_#33
     ExtraOptmizationsToApply_20_#14:Instanced Struct Structure=None
    ) -> (AtomProcessorAppliedOptimizations:Atom Processor Applied Optimizations Structure) {is=title K2Node_MakeStruct}
   )
  ]}
 Get Optimizations To Apply For LOD {nodes [
   n1 = Get Optimizations To Apply For LOD!() -> (then, Part Info:Atom Model Part Instance Info Structure, LOD:int) {is=title K2Node_FunctionEntry}
   > Return Node(Applied Optimizations=Get Optimizations To Apply For Geometry Style!(
     Part Info=@n1
     Geometry Style Settings=Get Geometry Style Settings For LOD!(LOD:int=@n1.LOD) -> (Style Settings:Atom Processor Geometry Style Settings Structure) {is=title K2Node_CallFunction}
     ColorChangeIndex=0
    ) -> (Applied Optimizations:Atom Processor Applied Optimizations Structure) {is=title K2Node_CallFunction})
  ]}
 Get Optimizations To Apply For Nanite {nodes [
   n1 = Get Optimizations To Apply For Nanite!() -> (then, Part Info:Atom Model Part Instance Info Structure, Color Change Index:int) {is=title K2Node_FunctionEntry}
   > Return Node(Applied Optimizations=Get Optimizations To Apply For Geometry Style(
     Part Info=@n1
     Geometry Style Settings=Get Nanite Geometry Style Settings() -> (Nanite Geometry Style Settings:Atom Processor Geometry Style Settings Structure)
     ColorChangeIndex=@n1.Color Change Index
    ))
  ]}
 Get Part Supported Features For Geometry Style {nodes [
   n1 = Get Part Supported Features For Geometry Style!() -> (
    then
    PartInstance:"Atom Model Part Instance Info Structure (by ref)"
    Geometry Style Settings:Atom Processor Geometry Style Settings Structure
    Color Change Index:int
   ) {is=title K2Node_FunctionEntry}
   n2 = Break Atom Model Part Instance Info!(AtomModelPartInstanceInfo=@n1) -> (Primitive, Decorations:Array of Atom Model Part Decoration Info Structures)
   > Set MainGroupId(MainGroupId:int=Get AtomMainGroupId!(self:Atom Primitive Object Reference=@n2) -> (AtomMainGroupId:int=0)) -> (Output_Get:int=0)
   > Set DesignName(DesignName:Name=Get DesignName!(self:Atom Primitive Object Reference=@n2) -> (DesignName:Name=None)) -> (Output_Get:Name=None)
   > n3 = Set PartId(PartId:int=Get PartId!(self:Atom Primitive Object Reference=@n2) -> (PartId:int=0)) -> (Output_Get:int=0)
   n4 = Find(
    TargetMap:Map of Integers to Atom Primitive Features Overrides Structures=Get Primitive Features Overrides!(self:BP Atom Primitives Features Overrides Object Reference=Get Primitives Features Overrides() -> (Primitives Features Overrides:BP Atom Primitives Features Overrides Object Reference)) -> (Primitive Features Overrides:Map of Integers to Atom Primitive Features Overrides Structures)
    Key=@n3
   ) -> (Value:Atom Primitive Features Overrides Structure)
   > n5 = Branch(Condition=@n4.ReturnValue)
   > n6 = Set Primitive Features Overrides(Primitive Features Overrides:Atom Primitive Features Overrides Structure=@n4) -> (Output_Get:Atom Primitive Features Overrides Structure)
   n7 = Break Atom Processor Geometry Style Settings!(AtomProcessorGeometryStyleSettings=Get Geometry Style Settings()) -> (
    GeometryStylePreferences_4_#17:Array of Strings
    PreferredGeometryLOD_30_#18:int
    ShouldBevel_8_#19:bool=~
    BevelWidth_11_#20:double=~
    BevelAngleThreshold_13_#21:double=~
    UseImprovedBevelAlgorithm_37_#22=~
    OnlyBevelBlock-LikePrimitives_39_#23=~
    NumberofSubdivisionsToApply_16_#24:int
    FakeBevel_23_#25=~
    FakeUnderside_25_#15=~
    AnalyticFakeUnderside_32_#26=~
    ForceFlatCap_35_#16=~
    FakeTechnicHole_41_#1=~
    AnalyticCommonParts_27_#2:bool=~
   )
   n8 = GetGeometryContainerForBestExportStyle(self=@n2, GeometryStylePriorityList=@n7, LODIndex=@n7.PreferredGeometryLOD_30_#18)
   n9 = Get Color Change Index() -> (Color Change Index=0)
   n10 = Is Not Empty(TargetArray:Array of Atom Primitive Geometry And Transform Structures=GetCapsGeometry!(self:Atom Primitive Geometry Container Object Reference=@n8, ColorChangeIndex:int=@n9) -> (ReturnValue:Array of Atom Primitive Geometry And Transform Structures))
   n11 = Get Geometry Style Settings()
   n12 = Get PartInstance() -> (PartInstance:"Atom Model Part Instance Info Structure (by ref)")
   n13 = IsDitheredTranslucentPart(Part Info=Get PartInstance(), Color Change Index=Get Color Change Index() -> (Color Change Index=0))
   n14 = GetSectionTextureName!(
    self:Atom Primitive Geometry Container Object Reference=@n8
    SectionType:EPrimitiveGeometrySectionFlag Enum=Shell
    TextureType:EAtomPrimitiveDetailTextureType Enum=NormalMap
   ) -> (OutTextureName:String, ReturnValue:bool)
   n15 = GetSectionTextureName(self=@n8, SectionType=Shell, TextureType=AlphaMask)
   n16 = NOT Boolean(A=@n15.ReturnValue)
   n17 = IsFakeTranslucentPart!(PartInfo:Atom Model Part Instance Info Structure=Get PartInstance(), Color Change Index:int=@n9) -> (Is Fake Transluscent Part:bool)
   n18 = NOT Boolean(A=@n17)
   n19 = IsGlowPart(PartInfo=Get PartInstance())
   n20 = NOT Boolean(A=@n19)
   n21 = Break Atom Primitive Features Overrides!(AtomPrimitiveFeaturesOverrides:"Atom Primitive Features Overrides Structure (by ref)"=Get Primitive Features Overrides!() -> (Primitive Features Overrides:Atom Primitive Features Overrides Structure)) -> (
    OverrideFakeBevels_10_#38:bool=~
    FakeBevelsEnabled_11_#39:bool=~
    OverrideFakeUnderside_13_#40:bool=~
    FakeUndersideEnabled_12_#41:bool=~
    GeometryBevelMethod_18_#42:EAtomGeometricBevelTechnique Enum
   ) {is=title K2Node_BreakStruct}
   n22 = "To String (Name)"(InName=Get DesignName!() -> (DesignName=None))
   n23 = Get MainGroupInclusionList() -> (MainGroupInclusionList:Array of Integers)
   n24 = Get MainGroupId() -> (MainGroupId:int=0)
   n25 = Contains(SearchIn=@n22, Substring=deg)
   n26 = OR Boolean(
    A=AND Boolean(A=Contains Item(TargetArray:Array of Integers=@n23, ItemToFind:"Integer (by ref)"=@n24), B=NOT Boolean(A=@n25))
    B=Contains(SearchIn=@n22, Substring=roof)
   )
   n27 = AND Boolean(A=NOT Boolean(A=@n21), B=@n26)
   n28 = AND Boolean(A=@n21, B=@n21.FakeBevelsEnabled_11_#39)
   n29 = AND Boolean(A=@n21.OverrideFakeUnderside_13_#40, B=@n21.FakeUndersideEnabled_12_#41)
   n30 = AND Boolean(A=NOT Boolean(A=@n21.OverrideFakeUnderside_13_#40), B=@n26)
   n31 = IsGlowCompatibleWithFakeTranslucent!(Part Info:Atom Model Part Instance Info Structure=Get PartInstance()) -> (IsCompatible:bool)
   Return Node!(
    execute=[@n5.else, @n6.then]
    Supported Features_FakeBevels_1_#27:bool=AND Boolean!(A=true, B=OR Boolean(A=@n27, B=@n28), C:bool=@n16, D:bool=@n18, E:bool=@n20, F:bool=NOT Boolean(A=@n14.ReturnValue)) -> (ReturnValue)
    Supported Features_FakeUnderside_3_#28:bool=AND Boolean!(A=OR Boolean(A=@n30, B=@n29), B=@n10, C=@n18, D=@n20, E=NOT Boolean(A=@n13)) -> (ReturnValue)
    Supported Features_ForceFlatCap_11_#29:bool=@n10
    Supported Features_TechnicHoles_18_#30:bool=Has Fake Technic Hole For Geometry Style!(Part Info:Atom Model Part Instance Info Structure=@n12, Geometry Style Settings:Atom Processor Geometry Style Settings Structure=@n11) -> (Has Fake Technic Hole:bool)
    Supported Features_TechnicConnectors_22_#31:bool=Has Technic Connector Geometry For Geometry Style!(Part Info:Atom Model Part Instance Info Structure=@n12, Geometry Style Settings:Atom Processor Geometry Style Settings Structure=@n11) -> (Has Technic Connector:bool)
    Supported Features_NormalMap_7_#32:bool=@n14.ReturnValue
    Supported Features_AlphaMask_9_#33:bool=@n15.ReturnValue
    Supported Features_FakeTransluscent_5_#34:bool=AND Boolean!(A=@n17, B=OR Boolean(A=@n20, B=@n31)) -> (ReturnValue)
    Supported Features_DitheredTranslucent_20_#35:bool=@n13
    Supported Features_Glow_13_#36:bool=AND Boolean(A=@n19, B=@n16)
    Supported Features_ExtraFeatures_16_#37:Instanced Struct Structure=None
   )
   Is Empty!(TargetArray:Array of Atom Model Part Decoration Info Structures=@n2.Decorations) -> (ReturnValue:bool) {is=title K2Node_CallArrayFunction}
  ]}
 Get Part Supported Features For LOD {nodes [
   n1 = Get Part Supported Features For LOD!() -> (then, PartInstance:"Atom Model Part Instance Info Structure (by ref)", LOD:int, Color Change Index:int) {is=title K2Node_FunctionEntry}
   > Return Node!(
    execute
    Supported Features:Atom Primitive Supported Features Structure=Get Part Supported Features For Geometry Style!(PartInstance=@n1, Geometry Style Settings=Get Geometry Style Settings For LOD(LOD=@n1.LOD), Color Change Index=@n1.Color Change Index) -> (Supported Features:Atom Primitive Supported Features Structure) {is=title K2Node_CallFunction}
   )
  ]}
 Get Part Supported Features For Nanite {nodes [
   n1 = Get Part Supported Features For Nanite!() -> (then, PartInstance:"Atom Model Part Instance Info Structure (by ref)", Color Change Index:int) {is=title K2Node_FunctionEntry}
   > n2 = Set members in Atom Processor Geometry Style Settings!(
    execute
    StructRef=Copy(Item=Get Nanite Geometry Style Settings())
    FakeBevel_23_#25:bool=False
    FakeUnderside_25_#15:bool=False
    AnalyticFakeUnderside_32_#26:bool=False
    ForceFlatCap_35_#16:bool=False
    FakeTechnicHole_41_#1:bool=False
    AnalyticCommonParts_27_#2:bool=False
   ) -> (then, StructOut)
   > Return Node(Supported Features=Get Part Supported Features For Geometry Style(PartInstance=@n1, Geometry Style Settings=@n2, Color Change Index=@n1.Color Change Index))
  ]}
 Get Polygroup UVs Translate and Scale {nodes [
   Get Polygroup UVs Translate and Scale!() -> (then, DynamicMesh:Dynamic Mesh Object Reference, PolygroupID:int, BrickSize:Vector=~) {is=title K2Node_FunctionEntry}
   > n1 = Sequence!(execute) -> (then_0, then_1)
   n2 = BreakVector(InVec=GetPredominantNormalFromPolygroup(DynamicMesh=Get DynamicMesh(), PolygroupID=Get PolygroupID() -> (PolygroupID=0)))
   n3 = Round(A=@n2)
   > n4 = Branch(Condition="Not Equal (Float)"(A="To Float (Integer)"(InInt=@n3)))
   n5 = BreakVector(InVec=Get BrickSize() -> (BrickSize:Vector))
   > Set BrickScaleInUVSpace(BrickScaleInUVSpace:Vector 2D Structure=MakeVector2D(X=@n5.Y, Y=@n5.Z)) -> (Output_Get:Vector 2D Structure)
   n6 = Round(A=@n2.Y)
   n7 = Branch(execute=@n4.else, Condition="Not Equal (Float)"(A="To Float (Integer)"(InInt=@n6)))
   n8 = BreakVector(InVec=Get BrickSize())
   > Set BrickScaleInUVSpace(BrickScaleInUVSpace=MakeVector2D(X=@n8, Y=@n8.Z))
   n9 = Round(A=@n2.Z)
   Branch(execute=@n7.else, Condition="Not Equal (Float)"(A="To Float (Integer)"(InInt=@n9)))
   n10 = BreakVector(InVec=Get BrickSize())
   > Set BrickScaleInUVSpace(BrickScaleInUVSpace=MakeVector2D(X=@n10, Y=@n10.Y))
   n11 = BreakVector2D(InVec=vector2d - vector2d(A=Get BrickScaleInUVSpace() -> (BrickScaleInUVSpace:Vector 2D Structure), B=(X=0,Y=0)))
   n12 = vector2d - vector2d(A=Get BrickScaleInUVSpace(), B=(X=1,Y=1))
   Return Node!(
    execute=@n1.then_1
    UVsScale:Vector 2D Structure=vector2d / vector2d!(A_X:double=@n11, A_Y:double=float * float(A=@n11.Y, B=-1), B:double=32) -> (ReturnValue)
    UVsTranslate:Vector 2D Structure=vector2d + vector2d(A:Vector 2D Structure=vector2d * vector2d(A=@n12, B:Vector 2D Structure=(X=1,Y=-1)), B:Vector 2D Structure=(X=0,Y=-1)) -> (ReturnValue:Vector 2D Structure) {is=title K2Node_PromotableOperator}
   )
  ]}
 Get Shell Geometry Section Flags And Filter {nodes [
   n1 = Get Shell Geometry Section Flags And Filter!() -> (then, Applied Optimizations:Atom Processor Applied Optimizations Structure, Color Change Index:int) {is=title K2Node_FunctionEntry}
   n2 = Break Atom Processor Applied Optimizations!(AtomProcessorAppliedOptimizations=@n1) -> (
    bApplyFakeUnderside_9_#4:bool=~
    bAnalyticFakeUnderside_10_#5:bool=~
    bApplyFlatCap_11_#6:bool=~
    bApplyTechnicCap_12_#7:bool=~
   )
   n3 = Make Bitmask!(Bitmask:Bitmask Integer=35) -> (ReturnValue:Bitmask Integer) {is=title K2Node_BitmaskLiteral}
   n4 = Make Bitmask(Bitmask=41)
   n5 = OR Boolean(A=@n2, B=@n2.bApplyTechnicCap_12_#7)
   n6 = Make Bitmask(Bitmask=41)
   n7 = SelectInt(A=0, B=Make Bitmask(Bitmask=64), bPickA=@n2.bApplyTechnicCap_12_#7)
   n8 = SelectInt(A=0, B=Make Bitmask(Bitmask=4), bPickA=@n2.bAnalyticFakeUnderside_10_#5)
   > Set ShellSectionFlags(ShellSectionFlags:int=SelectInt(
     A=Bitwise OR!(A:int=@n6, B:int=@n8, C:int=@n7) -> (ReturnValue:int=0) {is=title K2Node_CommutativeAssociativeBinaryOperator}
     B=SelectInt(A=@n4, B=@n3, bPickA=@n5)
     bPickA=@n2.bApplyFlatCap_11_#6
    )) -> (Output_Get:int=0)
   n9 = Get Applied Optimizations()
   n10 = Make Array!(
    "[0]":String=GetGeometryOtherPartName!(OtherPartType:EAtomPrimitiveGeometryOtherPartType Enum=TechnicCrossAxle) -> (ReturnValue:String)
    "[1]":String=GetGeometryOtherPartName(OtherPartType=TechnicPeg)
   ) -> (Array:Array of Strings)
   n11 = Get Color Change Index() -> (Color Change Index=0)
   > Return Node!(
    execute
    Shell Section Flags:int=Get ShellSectionFlags() -> (ShellSectionFlags:int=0)
    Shell Section Filter:Atom Primitive Geometry Section Filter Params Structure=Select(
     Option 0:Atom Primitive Geometry Section Filter Params Structure=Make Atom Primitive Geometry Section Filter Params!(ColorChangeIndex:int=@n11) -> (AtomPrimitiveGeometrySectionFilterParams:Atom Primitive Geometry Section Filter Params Structure) {is=title K2Node_MakeStruct}
     Option 1:Atom Primitive Geometry Section Filter Params Structure=Make Atom Primitive Geometry Section Filter Params!(OtherPartsExcludeList:Array of Strings=@n10, ColorChangeIndex=@n11) -> (AtomPrimitiveGeometrySectionFilterParams)
     Index=Break Atom Processor Applied Optimizations!(AtomProcessorAppliedOptimizations=@n9) -> (bApplyTechnicCap_12_#7=~)
    ) -> (ReturnValue:Atom Primitive Geometry Section Filter Params Structure)
   )
   Make Bitmask(Bitmask=45)
  ]}
 Get Split Analytic Common Part Mesh {nodes [
   Get Split Analytic Common Part Mesh() {is=title K2Node_FunctionEntry}
   > Return Node!(execute, Split Analytic Common Part Mesh:bool=false)
  ]}
 Get Translucency Type For Part {nodes [
   Get Translucency Type For Part!() -> (then, Part Info:Atom Model Part Instance Info Structure) {is=title K2Node_FunctionEntry}
   > Return Node!(
    execute
    Translucency Type:EAtomMaterialTranslucencyType Enum=Get Translucency Type() -> (Translucency Type:EAtomMaterialTranslucencyType Enum=Translucent)
   )
  ]}
 Get Triangle Area {nodes [
   n1 = Get Triangle Area!() -> (then, DynamicMesh:Dynamic Mesh Object Reference, TriangleID:int) {is=title K2Node_FunctionEntry}
   n2 = GetTrianglePositions!(TargetMesh:Dynamic Mesh Object Reference=@n1, TriangleID:int=@n1.TriangleID) -> (bIsValidTriangle:bool, Vertex1:Vector, Vertex2:Vector, Vertex3:Vector)
   > Set A(A:double="Distance (Vector)"!(V1:Vector=@n2.Vertex1, V2:Vector=@n2.Vertex2) -> (ReturnValue:double)) -> (Output_Get:double)
   > Set B(B:double="Distance (Vector)"(V1=@n2.Vertex1, V2=@n2.Vertex3)) -> (Output_Get:double)
   > Set C(C:double="Distance (Vector)"(V1=@n2.Vertex2, V2=@n2.Vertex3)) -> (Output_Get:double)
   n3 = Get A() -> (A:double)
   n4 = Get B() -> (B:double)
   n5 = Get C() -> (C:double)
   > Set SemiPerimeter(SemiPerimeter:double=float / float(A=float + float!(A=@n3, B=@n4, C:double=@n5) -> (ReturnValue), B=2)) -> (Output_Get:double)
   n6 = float - float(A=Get SemiPerimeter() -> (SemiPerimeter:double), B=Get C())
   n7 = float - float(A=Get SemiPerimeter(), B=Get A())
   n8 = float - float(A=Get SemiPerimeter(), B=Get B())
   n9 = Get SemiPerimeter()
   > Return Node!(execute, Area:double=Sqrt!(A:double=float * float!(A=@n9, B=@n7, C:double=@n8, D:double=@n6) -> (ReturnValue)) -> (ReturnValue:double))
  ]}
 Get Unique Name for Common Part Geometry LODs {nodes [
   n1 = Get Unique Name for Common Part Geometry LODs!() -> (then, PartInfo:Atom Model Part Instance Info Structure, MergedMeshPartInfo:Atom Merged Mesh Part Info Structure, CommonPartIndex:int=0) {is=title K2Node_FunctionEntry}
   > n2 = Parent: Get Unique Name for Common Part Geometry LODs!(
    execute
    PartInfo:Atom Model Part Instance Info Structure=@n1
    MergedMeshPartInfo:Atom Merged Mesh Part Info Structure=@n1.MergedMeshPartInfo
    CommonPartIndex:int=@n1.CommonPartIndex
   ) -> (then, Name:String) {is=title K2Node_CallParentFunction}
   > n3 = Branch(Condition=ShouldSplitTransparentPart(MergedMeshPartInfo=@n1.MergedMeshPartInfo))
   > Return Node!(
    execute
    Name:String=Append(A=@n2, B=_, C=GetTransparentMeshSuffix!(self:MP Merged Mesh Base Object Reference) -> (Transparent Mesh Suffix:String))
   )
   n4 = Get PartInfo()
   n5 = Branch(
    execute=@n3.else
    Condition=Break Atom Processor Applied Optimizations!(AtomProcessorAppliedOptimizations=Get Optimizations To Apply For LOD(Part Info=@n4, LOD=0) {is=title K2Node_CallFunction}) -> (
     bApplyFakeTranslucent_25_#9:bool=~
     bApplyDitheredTranslucent_24_#10:bool=~
     bApplyGlow_13_#11:bool=~
    ).bApplyGlow_13_#11
   )
   n6 = Break Atom Model Part Instance Info!(AtomModelPartInstanceInfo=Get PartInfo()) -> (PartGuid:Atom Model Part Guid Structure)
   n7 = Get CommonPartIndex() -> (CommonPartIndex:int=0)
   > Return Node(Name=Append(
     A=@n2
     B=To String!(InModelPartGuid:"Atom Model Part Guid Structure (by ref)"=@n6) -> (ReturnValue:String)
     C="To String (Integer)"(InInt=@n7)
    ))
   n8 = Is Common Part Affected by Cap Optimization(Exec=@n5.else, PartInfo=Get PartInfo(), CommonPartIndex=Get CommonPartIndex() -> (CommonPartIndex=0))
   Return Node(execute=@n8.False, Name=@n2)
   > Return Node(Name=Append!(A=@n2, B=_FakeUnder) -> (ReturnValue))
  ]}
 GetAtomBrickSize {nodes [
   n1 = GetAtomBrickSize!() -> (then, DynamicMesh:Dynamic Mesh Object Reference) {is=title K2Node_FunctionEntry}
   n2 = vector * vector(A:double=3.125, B:Vector=GetBrickSize!(DynamicMesh:Dynamic Mesh Object Reference=@n1) -> (BrickSize:Vector))
   n3 = Ceil(A=@n2.ReturnValue_Z)
   n4 = Ceil(A=@n2)
   n5 = Ceil(A=@n2.ReturnValue_Y)
   > Return Node!(
    execute
    BrickSize:Vector=MakeVector(X="To Float (Integer)"(InInt=@n4), Y="To Float (Integer)"(InInt=@n5), Z="To Float (Integer)"(InInt=@n3))
   )
  ]}
 GetBrickSize {nodes [
   n1 = GetBrickSize!() -> (then, DynamicMesh:Dynamic Mesh Object Reference) {is=title K2Node_FunctionEntry}
   n2 = Break Box!(Box:"Box Structure (by ref)"=GetMeshBoundingBox!(TargetMesh=@n1) -> (ReturnValue:Box Structure)) -> (Min:Vector=~, Max:Vector=~, IsValid:bool=~) {is=title K2Node_BreakStruct}
   > Return Node(BrickSize=vector - vector(A=@n2.Max, B=@n2))
  ]}
 GetCenterDimension {nodes [
   n1 = GetMeshBoundingBox!(TargetMesh=Inputs!() -> (TargetMesh:Dynamic Mesh Object Reference)) -> (ReturnValue_Min:Vector=~, ReturnValue_Max:Vector=~, ReturnValue_IsValid:bool=~)
   n2 = vector - vector(A=@n1.ReturnValue_Max, B=@n1)
   n3 = vector + vector!(A=@n1, B=vector * vector!(A:Vector=@n2, B:float=0.5) -> (ReturnValue:Vector=~)) -> (ReturnValue_X:double, ReturnValue_Y:double, ReturnValue_Z:double)
   Outputs!(
    ReturnValue:Vector=@n2
    Center_X:double=@n3
    Center_Y:double=@n3.ReturnValue_Y
    Center_Z:double=BreakVector(InVec=@n1).Z
    IsDeep:bool="float > float"(A=BreakVector(InVec=@n2).Z, B=0.32)
   )
  ]}
 GetGeometricBevelSettings {nodes [
   n1 = GetGeometricBevelSettings!() -> (
    then
    Part Instance Info:"Atom Model Part Instance Info Structure (by ref)"
    Geometry Style Settings:"Atom Processor Geometry Style Settings Structure (by ref)"
   ) {is=title K2Node_FunctionEntry}
   n2 = Break Atom Model Part Instance Info!(AtomModelPartInstanceInfo=@n1) -> (
    PartGuid
    PartId:int
    Primitive:Atom Primitive Object Reference
    Transforms:Array of Transforms
    Decorations:Array of Atom Model Part Decoration Info Structures
    Colors:Array of Atom Model Part Color Info Structures
    SelectionSets:Array of Strings
    Group:String
    ParentGroup:String
    CommonParts:Array of Atom Common Part And Transform Structures
    bIsUndersideVisible:bool=~
   )
   n3 = Get Primitive Features Overrides!(self:BP Atom Primitives Features Overrides Object Reference=Get Primitives Features Overrides()) -> (Primitive Features Overrides:Map of Integers to Atom Primitive Features Overrides Structures)
   > n4 = Set members in Atom Geometric Bevel Settings!(
    execute:Exec
    StructRef:"Atom Geometric Bevel Settings Structure (by ref)"=Parent: GetGeometricBevelSettings!(
     Part Instance Info:"Atom Model Part Instance Info Structure (by ref)"=@n1
     Geometry Style Settings:"Atom Processor Geometry Style Settings Structure (by ref)"=@n1.Geometry Style Settings
    ) -> (AtomGeometricBevelSettings:Atom Geometric Bevel Settings Structure) {is=title K2Node_CallParentFunction}
    Technique:EAtomGeometricBevelTechnique Enum=Break Atom Primitive Features Overrides!(AtomPrimitiveFeaturesOverrides=Find(TargetMap=@n3, Key=@n2.PartId)) -> (GeometryBevelMethod_18_#42)
   ) -> (then:Exec, StructOut:"Atom Geometric Bevel Settings Structure (by ref)") {is=title K2Node_SetFieldsInStruct}
   > Return Node!(execute, AtomGeometricBevelSettings:Atom Geometric Bevel Settings Structure=@n4)
  ]}
 GetGeometryLODsForCommonPart {nodes [
   n1 = GetGeometryLODsForCommonPart!() -> (
    then
    PartInfo:Atom Model Part Instance Info Structure
    MergedMeshPartInfo:Atom Merged Mesh Part Info Structure
    Common Part Type:EAtomCommonPartType Enum=None
   ) {is=title K2Node_FunctionEntry}
   > n2 = GetHasOnlyNaniteCompatibleMaterials!(
    execute
    self:MP Merged Mesh Base Object Reference
    MergedMeshPartInfo:"Atom Merged Mesh Part Info Structure (by ref)"=@n1.MergedMeshPartInfo
   ) -> (then, bOnlyHasOpaque:bool)
   > n3 = Set OnlyHasOpaque(OnlyHasOpaque:bool=@n2) -> (Output_Get:bool)
   > n4 = Set IsGlow(IsGlow:bool=Break Atom Processor Applied Optimizations(AtomProcessorAppliedOptimizations=Get Optimizations To Apply For LOD(Part Info=@n1, LOD=0)) -> (
     bApplyFakeTranslucent_25_#9=~
     bApplyDitheredTranslucent_24_#10=~
     bApplyGlow_13_#11=~
    ).bApplyGlow_13_#11) -> (Output_Get:bool)
   > n5 = Branch(Condition=AND Boolean!(
     A=NOT Boolean(A=@n4)
     B=@n3
     C:bool=Get Split Analytic Common Part Mesh!() -> (Split Analytic Common Part Mesh:bool) {is=title K2Node_CallFunction}
    ) -> (ReturnValue))
   > n6 = Sequence()
   n7 = Get Common Part Type() -> (Common Part Type:EAtomCommonPartType Enum=None)
   n8 = CreateAtomGeometryFromCommonPart!(
    execute=[@n6.then_1, @n24.Completed]
    ExportStyleName:String=Get Common Part Geometry Style!(Common Part Type:EAtomCommonPartType Enum=@n7, bIsNanite:bool=true) -> (Geometry Style:String) {is=title K2Node_CallFunction}
    CommonPartType:EAtomCommonPartType Enum=@n7
    LodIndex:int=0
    CommonPartsMeshPath:String=GetCustomCommonPartMeshPath!(self:MP Merged Mesh Base Object Reference) -> (Common Part Mesh Path:String)
   ) -> (then, ReturnValue:Atom Primitive Geometry Object Reference)
   > Set CommonPartGeometry(CommonPartGeometry:Atom Primitive Geometry Object Reference=@n8) -> (Output_Get:Atom Primitive Geometry Object Reference)
   > n9 = Branch(Condition=Get IsGlow() -> (IsGlow:bool))
   > n10 = ConvertAtomPrimitiveGeometryToDynamicMesh!(
    execute
    AtomPrimitiveGeometry:Atom Primitive Geometry Object Reference=Get CommonPartGeometry() -> (CommonPartGeometry:Atom Primitive Geometry Object Reference)
    DynamicMesh:Dynamic Mesh Object Reference
   ) -> (then, ReturnValue:Dynamic Mesh Object Reference)
   > n11 = Process Glow(Part Info=Get PartInfo(), Dynamic Mesh=@n10, UV Channel XY=1, UV Channel Z=2)
   > n12 = ConvertDynamicMeshToAtomPrimitiveGeometry!(execute, DynamicMesh:Dynamic Mesh Object Reference=@n11, AtomPrimitiveGeometry:Atom Primitive Geometry Object Reference) -> (then, ReturnValue:Atom Primitive Geometry Object Reference)
   > n13 = Set CommonPartGeometry(CommonPartGeometry=@n12)
   n14 = SetMaterialName!(
    execute=[@n13.then, @n9.else]
    self:Atom Primitive Geometry Object Reference=Get CommonPartGeometry()
    Name:String=Default
    PolygonGroupIndex:int=0
   ) -> (then, ReturnValue:Atom Primitive Geometry Object Reference)
   > n15 = BakeScale!(execute, self:Atom Primitive Geometry Object Reference=@n14, Scale:float=Get Common Part Scale() -> (Common Part Scale:double)) -> (then, ReturnValue:Atom Primitive Geometry Object Reference)
   > Return Node!(
    execute
    Geometry LODs:Atom Primitive Geometry LODs Structure=Make Atom Primitive Geometry LODs!(
     GeometryLODs:Array of Atom Primitive Geometry Object References=Get Common Part Geometry LODs() -> (Common Part Geometry LODs:Array of Atom Primitive Geometry Object References)
     HiResNaniteGeomery:Atom Primitive Geometry Object Reference=@n15
     DetailedCollision:Atom Primitive Collision Geometry Structure=(AggGeom=(SphereElems=,BoxElems=,SphylElems=,ConvexElems=,TaperedCapsuleElems=,LevelSetElems=,SkinnedLevelSetElems=,MLLevelSetElems=,SkinnedTriangleMeshElems=))
    ) -> (AtomPrimitiveGeometryLODs:Atom Primitive Geometry LODs Structure) {is=title K2Node_MakeStruct}
   )
   n16 = Get Common Part Type() -> (Common Part Type=None)
   n17 = CreateAtomGeometryFromCommonPart(
    execute=[@n38.then, @n6.then_0]
    ExportStyleName=MakeLiteralString(Value=Analytic)
    CommonPartType=@n16
    LodIndex=0
    CommonPartsMeshPath=GetCustomCommonPartMeshPath()
   )
   > n18 = ConvertAtomPrimitiveGeometryToDynamicMesh(AtomPrimitiveGeometry=@n17)
   > n19 = RemapMaterialIDs!(execute, TargetMesh:Dynamic Mesh Object Reference=@n18, FromMaterialID:int=0, ToMaterialID:int=1) -> (then, ReturnValue:Dynamic Mesh Object Reference)
   > n20 = ConvertDynamicMeshToAtomPrimitiveGeometry(DynamicMesh=@n19)
   > n21 = SetMaterialName(self=@n20, Name=Analytic, PolygonGroupIndex=1)
   > n22 = Set CommonPartGeometry(CommonPartGeometry=@n21)
   n23 = BakeScale(execute=[@n22.then, @n35.then, @n33.then], self=Get CommonPartGeometry(), Scale=Get Common Part Scale())
   n24 = For Each Loop(Exec=@n5.else, Array=Get LOD Geometry Style Settings())
   Set Array Elem(
    execute=@n23.then
    TargetArray:Array of Atom Primitive Geometry Object References=Get Common Part Geometry LODs()
    Index=@n24.Array Index
    Item:"Atom Primitive Geometry Object Reference (by ref)"=@n23
    bSizeToFit=true
   )
   n25 = Break Atom Processor Geometry Style Settings!(AtomProcessorGeometryStyleSettings=@n24) -> (AnalyticCommonParts_27_#2=~)
   n26 = GetCustomCommonPartMeshPath()
   n27 = CreateAtomGeometryFromCommonPart(
    execute=@n38.else
    ExportStyleName=Get Common Part Geometry Style(Common Part Type=@n16, bIsNanite=false)
    CommonPartType=@n16
    LodIndex=@n24.Array Index
    CommonPartsMeshPath=@n26
   )
   > n28 = Branch(Condition=Get IsGlow())
   > n29 = ConvertAtomPrimitiveGeometryToDynamicMesh(AtomPrimitiveGeometry=@n27)
   > n30 = Process Glow(Part Info=Get PartInfo(), Dynamic Mesh=@n29, UV Channel XY=1, UV Channel Z=2)
   > n31 = ConvertDynamicMeshToAtomPrimitiveGeometry(DynamicMesh=@n30)
   > n32 = SetMaterialName(self=@n31, Name=Default, PolygonGroupIndex=0)
   > n33 = Set CommonPartGeometry(CommonPartGeometry=@n32)
   n34 = SetMaterialName(execute=@n28.else, self=@n27, Name=Default, PolygonGroupIndex=0)
   > n35 = Set CommonPartGeometry(CommonPartGeometry=@n34)
   n36 = NOT Boolean(A=Get IsGlow())
   n37 = AND Boolean(A=Get OnlyHasOpaque() -> (OnlyHasOpaque:bool), B=@n36, C=@n25)
   n38 = Branch(execute=@n24.LoopBody, Condition=@n37)
  ]}
 GetGeometryLODsForPart {nodes [
   n1 = GetGeometryLODsForPart!() -> (then, Part Info:"Atom Model Part Instance Info Structure (by ref)") {is=title K2Node_FunctionEntry}
   > n2 = Branch(Condition=IsFlexElement!(self:Atom Primitive Object Reference=Break Atom Model Part Instance Info!(AtomModelPartInstanceInfo=@n1) -> (Primitive)) -> (ReturnValue:bool))
   > n3 = Parent: GetGeometryLODsForPart!(
    execute
    Part Info:"Atom Model Part Instance Info Structure (by ref)"=Get Part Info() -> (Part Info:"Atom Model Part Instance Info Structure (by ref)")
   ) -> (then, Geometry LODs:Atom Primitive Geometry LODs Structure) {is=title K2Node_CallParentFunction}
   > Return Node(Geometry LODs=@n3)
   n4 = Get Part Info()
   n5 = Branch(execute=@n2.else, Condition=Has Color Change Baked in Geometry(PartInfo=@n4))
   > Set bMultipleColorSurfaces(bMultipleColorSurfaces:bool=true) -> (Output_Get:bool)
   > Set First Color Change index(First Color Change index:int=0) -> (Output_Get:int=0)
   n6 = Break Atom Model Part Instance Info!(AtomModelPartInstanceInfo=@n4) -> (Colors:Array of Atom Model Part Color Info Structures)
   > n7 = Set Max Color Change Index(Max Color Change Index:int=int - int(A=Length(TargetArray=@n6))) -> (Output_Get:int=0)
   Set PrimitiveGeometryLODs(
    execute=[@n87.then, @n7.then]
    PrimitiveGeometryLODs:Atom Primitive Geometry LODs Structure=Make Atom Primitive Geometry LODs!(DetailedCollision=(AggGeom=(SphereElems=,BoxElems=,SphylElems=,ConvexElems=,TaperedCapsuleElems=,LevelSetElems=,SkinnedLevelSetElems=,MLLevelSetElems=,SkinnedTriangleMeshElems=))) -> (AtomPrimitiveGeometryLODs)
   ) -> (Output_Get:Atom Primitive Geometry LODs Structure)
   > n8 = Get Geometry Style Settings For Part(Part Info=Get Part Info()) {is=title K2Node_CallFunction}
   > Set LODSettingsForPart(LODSettingsForPart=@n8)
   > n9 = CreateDynamicMeshPool!(execute) -> (then, ReturnValue:Dynamic Mesh Pool Object Reference)
   > Set DynamicMeshPool(DynamicMeshPool:Dynamic Mesh Pool Object Reference=@n9) -> (Output_Get:Dynamic Mesh Pool Object Reference)
   > n10 = Sequence!(execute) -> (then_0, then_1, then_2:Exec)
   n11 = For Each Loop(Exec=@n10.then_1, Array=Get LODSettingsForPart())
   > Set GeometryStyleSettings(GeometryStyleSettings:Atom Processor Geometry Style Settings Structure=@n11) -> (Output_Get:Atom Processor Geometry Style Settings Structure)
   > Set LOD(LOD:int=@n11.Array Index) -> (Output_Get:int=0)
   > n12 = RequestMesh!(execute, self:Dynamic Mesh Pool Object Reference=Get DynamicMeshPool() -> (DynamicMeshPool:Dynamic Mesh Pool Object Reference)) -> (then, ReturnValue:Dynamic Mesh Object Reference)
   > Set CurrentLODMergedMesh(CurrentLODMergedMesh:Dynamic Mesh Object Reference=@n12) -> (Output_Get:Dynamic Mesh Object Reference)
   > n13 = Branch(Condition=Get bMultipleColorSurfaces() -> (bMultipleColorSurfaces:bool))
   > n14 = For Loop(
    FirstIndex=Get First Color Change index() -> (First Color Change index:int=0)
    LastIndex=Get Max Color Change Index() -> (Max Color Change Index:int=0)
   )
   n15 = ConvertDynamicMeshToAtomPrimitiveGeometry(
    execute=[@n56.then_1, @n14.Completed]
    DynamicMesh=Get CurrentLODMergedMesh() -> (CurrentLODMergedMesh:Dynamic Mesh Object Reference)
    AtomPrimitiveGeometry=Get CurrentGeometry() -> (CurrentGeometry:Atom Primitive Geometry Object Reference)
   )
   > Set CurrentGeometry(CurrentGeometry:Atom Primitive Geometry Object Reference=@n15) -> (Output_Get:Atom Primitive Geometry Object Reference)
   > ReturnAllMeshes(self:Dynamic Mesh Pool Object Reference=Get DynamicMeshPool())
   > n16 = BakeScale(self=Get CurrentGeometry(), Scale=GetDefaultPrimitiveScale())
   > Add(
    TargetArray:Array of Atom Primitive Geometry Object References=Break Atom Primitive Geometry LODs!(AtomPrimitiveGeometryLODs=Get PrimitiveGeometryLODs() -> (PrimitiveGeometryLODs:Atom Primitive Geometry LODs Structure)) -> (
     GeometryLODs:Array of Atom Primitive Geometry Object References
     GeometryLODsMaterials:Array of Atom Primitive Geometry Material Assignments Structures
    )
    NewItem:"Atom Primitive Geometry Object Reference (by ref)"=@n16
   )
   n17 = Set Current Color Change Index(execute=@n14.LoopBody, Current Color Change Index:int=@n14) -> (Output_Get:int=0)
   > n18 = Set Current Color Change Filter(Current Color Change Filter:int=@n17) -> (Output_Get:int=0)
   n19 = Sequence!(execute=[@n56.then_0, @n18.then]) -> (then_0, then_1, then_2, then_3:Exec, then_4:Exec)
   n20 = AppendDynamicMeshes!(
    execute=@n19.then_3
    self
    Target Mesh:Dynamic Mesh Object Reference=Get ShellDynamicMesh()
    Additional Meshes:Array of Dynamic Mesh Object References=Get CapsDynamicMeshes()
    Additional Meshes Transforms:Array of Atom Primitive Geometry And Transform Structures=Get CapsGeometryAndTransform() -> (CapsGeometryAndTransform:Array of Atom Primitive Geometry And Transform Structures)
   ) -> (then, Out Mesh:Dynamic Mesh Object Reference) {is=title K2Node_CallFunction}
   > n21 = AppendDynamicMeshes(
    Target Mesh=@n20
    Additional Meshes=Get TechnicCapsDynamicMeshes()
    Additional Meshes Transforms=Get TechnicCapsGeometryAndTransform() -> (TechnicCapsGeometryAndTransform:Array of Atom Primitive Geometry And Transform Structures)
   )
   > n22 = AppendDynamicMeshes(
    Target Mesh=@n21
    Additional Meshes=Get TechnicConnectorsDynamicMeshes()
    Additional Meshes Transforms=Get TechnicConnectorsGeometryAndTransform() -> (TechnicConnectorsGeometryAndTransform:Array of Atom Primitive Geometry And Transform Structures)
   )
   > n23 = Set ShellDynamicMesh(ShellDynamicMesh=@n22)
   n24 = Break Atom Processor Applied Optimizations!(AtomProcessorAppliedOptimizations=Get Applied Optimizations()) -> (bApplyTechnicCap_12_#7:bool=~, bUseNormalMap_16_#12:bool=~)
   > n25 = Branch(Condition=OR Boolean(A=@n24, B=@n24.bUseNormalMap_16_#12))
   > n26 = ConvertDynamicMeshToAtomPrimitiveGeometry(DynamicMesh=@n23, AtomPrimitiveGeometry=Get CurrentGeometry())
   > n27 = CopyNormalMapsToUV0!(execute, self:Atom Primitive Geometry Object Reference=@n26) -> (then, ReturnValue:Atom Primitive Geometry Object Reference)
   > n28 = ConvertAtomPrimitiveGeometryToDynamicMesh(AtomPrimitiveGeometry=@n27, DynamicMesh=Get ShellDynamicMesh())
   AppendMesh!(
    execute=[@n28.then, @n25.else]
    TargetMesh=Get CurrentLODMergedMesh()
    AppendMesh=Get ShellDynamicMesh()
    AppendTransform:Transform
    bDeferChangeNotifications=false
    AppendOptions=(CombineMode=EnableAllMatching)
   ) -> (then, ReturnValue)
   > n29 = Get Geometry Material Assignments For LOD!(
    execute
    self:MP Merged Mesh Base Object Reference
    PartInfo:Atom Model Part Instance Info Structure=Get Part Info()
    LOD:int=Get LOD() -> (LOD:int=0)
    Color Change Index:int=Get Current Color Change Index() -> (Current Color Change Index:int=0)
   ) -> (then, Material Assignments:Atom Primitive Geometry Material Assignments Structure) {is=title K2Node_CallFunction}
   > Add(
    TargetArray:Array of Atom Primitive Geometry Material Assignments Structures=Break Atom Primitive Geometry LODs!(AtomPrimitiveGeometryLODs=Get PrimitiveGeometryLODs()) -> (GeometryLODsMaterials)
    NewItem:"Atom Primitive Geometry Material Assignments Structure (by ref)"=@n29
   )
   n30 = Break Atom Primitive Geometry Material Assignments(AtomPrimitiveGeometryMaterialAssignments=@n29)
   > Set MaterialSectionOffset(MaterialSectionOffset:int=int + int(
     A=Get MaterialSectionOffset() -> (MaterialSectionOffset:int=0)
     B=Length(TargetArray:Array of Atom Material Features Structures=@n30.MaterialFeatures)
    )) -> (Output_Get:int=0)
   Clear(execute=@n19.then_1, TargetArray:Array of Atom Primitive Geometry And Transform Structures=Get CapsGeometryAndTransform())
   > Clear(TargetArray:Array of Dynamic Mesh Object References=Get CapsDynamicMeshes())
   > Clear(TargetArray:Array of Atom Primitive Geometry And Transform Structures=Get TechnicCapsGeometryAndTransform())
   > Clear(TargetArray:Array of Dynamic Mesh Object References=Get TechnicCapsDynamicMeshes())
   > Clear(TargetArray=Get TechnicConnectorsDynamicMeshes())
   > Set ShellDynamicMesh()
   n31 = Break Atom Processor Geometry Style Settings!(AtomProcessorGeometryStyleSettings=Get GeometryStyleSettings() -> (GeometryStyleSettings:Atom Processor Geometry Style Settings Structure)) -> (
    GeometryStylePreferences_4_#17:Array of Strings
    PreferredGeometryLOD_30_#18:int
    ShouldBevel_8_#19:bool=~
    BevelWidth_11_#20:double=~
    BevelAngleThreshold_13_#21:double=~
    UseImprovedBevelAlgorithm_37_#22:bool=~
    OnlyBevelBlock-LikePrimitives_39_#23:bool=~
    NumberofSubdivisionsToApply_16_#24:int
    FakeBevel_23_#25:bool=~
    FakeUnderside_25_#15:bool=~
    AnalyticFakeUnderside_32_#26:bool=~
    ForceFlatCap_35_#16:bool=~
    FakeTechnicHole_41_#1:bool=~
    AnalyticCommonParts_27_#2=~
   )
   n32 = Get Part Info()
   > Set GeometryContainer(GeometryContainer:Atom Primitive Geometry Container Object Reference=GetGeometryContainerForBestExportStyle(
     self=Break Atom Model Part Instance Info!(AtomModelPartInstanceInfo=@n32) -> (Primitive:Atom Primitive Object Reference)
     GeometryStylePriorityList=@n31
     LODIndex=@n31.PreferredGeometryLOD_30_#18
    )) -> (Output_Get:Atom Primitive Geometry Container Object Reference)
   > Set CurrentGeometry(CurrentGeometry=GetGeometrySectionWithFilter!(
     self:Atom Primitive Geometry Container Object Reference=Get GeometryContainer()
     PrimitiveGeometrySectionFlag:Bitmask Integer=Get ShellSectionFlags() -> (ShellSectionFlags=0)
     FilterParams:"Atom Primitive Geometry Section Filter Params Structure (by ref)"=Get ShellSectionFilter() -> (ShellSectionFilter:Atom Primitive Geometry Section Filter Params Structure)
    ) -> (ReturnValue:Atom Primitive Geometry Object Reference))
   > n33 = RequestMesh(self=Get DynamicMeshPool())
   > n34 = Set ShellDynamicMesh(ShellDynamicMesh=@n33)
   > ConvertAtomPrimitiveGeometryToDynamicMesh(AtomPrimitiveGeometry=Get CurrentGeometry(), DynamicMesh=@n34)
   > n35 = Branch(Condition=Break Atom Processor Applied Optimizations!(AtomProcessorAppliedOptimizations=Get Applied Optimizations()) -> (bApplyFakeUnderside_9_#4:bool=~))
   > n36 = Set CapsGeometryAndTransform(CapsGeometryAndTransform:Array of Atom Primitive Geometry And Transform Structures=GetCapsGeometry(self=Get GeometryContainer(), ColorChangeIndex=Get Current Color Change Filter() -> (Current Color Change Filter:int=0))) -> (Output_Get:Array of Atom Primitive Geometry And Transform Structures)
   > n37 = For Each Loop(Array:Array of Atom Primitive Geometry And Transform Structures=@n36) -> (Array Element:Atom Primitive Geometry And Transform Structure)
   > n38 = RequestMesh(self=Get DynamicMeshPool())
   > n39 = ConvertAtomPrimitiveGeometryToDynamicMesh(
    AtomPrimitiveGeometry=Break Atom Primitive Geometry and Transform!(AtomPrimitiveGeometryAndTransform=@n37) -> (AtomPrimitiveGeometry)
    DynamicMesh=@n38
   )
   > Add(TargetArray:Array of Dynamic Mesh Object References=Get CapsDynamicMeshes(), NewItem:"Dynamic Mesh Object Reference (by ref)"=@n39)
   Branch(
    execute=[@n37.Completed, @n35.else]
    Condition=Break Atom Processor Applied Optimizations!(AtomProcessorAppliedOptimizations=Get Applied Optimizations()) -> (bApplyTechnicCap_12_#7:bool=~)
   )
   > n40 = Set TechnicCapsGeometryAndTransform(TechnicCapsGeometryAndTransform:Array of Atom Primitive Geometry And Transform Structures=GetTechnicCapsGeometry(self=Get GeometryContainer(), ColorChangeIndex=Get Current Color Change Filter() -> (Current Color Change Filter=0))) -> (Output_Get:Array of Atom Primitive Geometry And Transform Structures)
   > n41 = For Each Loop(Array=@n40)
   > n42 = RequestMesh(self=Get DynamicMeshPool())
   > n43 = ConvertAtomPrimitiveGeometryToDynamicMesh(AtomPrimitiveGeometry=Break Atom Primitive Geometry and Transform(AtomPrimitiveGeometryAndTransform=@n41), DynamicMesh=@n42)
   > Add(TargetArray=Get TechnicCapsDynamicMeshes(), NewItem=@n43)
   n44 = Set TechnicConnectorsGeometryAndTransform(
    execute=@n41.Completed
    TechnicConnectorsGeometryAndTransform:Array of Atom Primitive Geometry And Transform Structures=GetTechnicConnectorsGeometry!(
     self:Atom Primitive Geometry Container Object Reference=Get GeometryContainer()
     ColorChangeIndex:int=Get Current Color Change Filter() -> (Current Color Change Filter=0)
    ) -> (ReturnValue:Array of Atom Primitive Geometry And Transform Structures)
   ) -> (Output_Get:Array of Atom Primitive Geometry And Transform Structures)
   > n45 = For Each Loop(Array=@n44)
   > n46 = RequestMesh(self=Get DynamicMeshPool())
   > n47 = ConvertAtomPrimitiveGeometryToDynamicMesh(AtomPrimitiveGeometry=Break Atom Primitive Geometry and Transform(AtomPrimitiveGeometryAndTransform=@n45), DynamicMesh=@n46)
   > Add(TargetArray=Get TechnicConnectorsDynamicMeshes(), NewItem=@n47)
   n48 = Set Applied Optimizations(
    execute=@n19.then_0
    Applied Optimizations:Atom Processor Applied Optimizations Structure=Get Optimizations To Apply For Geometry Style(
     Part Info=Get Part Info()
     Geometry Style Settings=Get GeometryStyleSettings()
     ColorChangeIndex=Get Current Color Change Index() -> (Current Color Change Index=0)
    )
   ) -> (Output_Get:Atom Processor Applied Optimizations Structure)
   n49 = Get Shell Geometry Section Flags And Filter!(
    Applied Optimizations:Atom Processor Applied Optimizations Structure=@n48
    Color Change Index:int=Get Current Color Change Filter() -> (Current Color Change Filter=0)
   ) -> (Shell Section Flags:int=0, Shell Section Filter:Atom Primitive Geometry Section Filter Params Structure) {is=title K2Node_CallFunction}
   > Set ShellSectionFlags(ShellSectionFlags=@n49) -> (Output_Get=0)
   > Set ShellSectionFilter(ShellSectionFilter:Atom Primitive Geometry Section Filter Params Structure=@n49.Shell Section Filter) -> (Output_Get:Atom Primitive Geometry Section Filter Params Structure)
   n50 = Get Applied Optimizations()
   > Set bUVChanged(bUVChanged:bool=OR Boolean(
     A=Are Optimizations Changing UVs!(Applied Optimizations:Atom Processor Applied Optimizations Structure=@n50) -> (UVs Changed:bool) {is=title K2Node_CallFunction}
     B=Get bUVChanged() -> (bUVChanged:bool)
    )) -> (Output_Get:bool)
   n51 = Get Geometry LODs For Part - Main Pipeline!(
    execute=@n19.then_2
    self
    Part Info:Atom Model Part Instance Info Structure=Get Part Info()
    Applied Optimizations:Atom Processor Applied Optimizations Structure=Get Applied Optimizations()
    GeometryContainer:Atom Primitive Geometry Container Object Reference=Get GeometryContainer()
    Color Change Index:int=Get Current Color Change Index() -> (Current Color Change Index=0)
    Shell Mesh:Dynamic Mesh Object Reference=Get ShellDynamicMesh()
    Caps Meshes:Array of Dynamic Mesh Object References=Get CapsDynamicMeshes()
    Technic Caps Meshes:Array of Dynamic Mesh Object References=Get TechnicCapsDynamicMeshes()
    Technic Connectors Meshes:Array of Dynamic Mesh Object References=Get TechnicConnectorsDynamicMeshes()
   ) -> (
    then
    Out Shell Mesh:Dynamic Mesh Object Reference
    Out Caps Meshes:Array of Dynamic Mesh Object References
    Out Technic Caps Meshes:Array of Dynamic Mesh Object References
    Out Technic Connectors Meshes:Array of Dynamic Mesh Object References
   ) {is=title K2Node_CallFunction}
   > n52 = Get Geometry LODs For Part - Post Pipeline!(
    execute
    self
    Part Info:Atom Model Part Instance Info Structure=Get Part Info()
    Applied Optimizations:Atom Processor Applied Optimizations Structure=Get Applied Optimizations()
    Geometry Container:Atom Primitive Geometry Container Object Reference=Get GeometryContainer()
    Color Change Index:int=Get Current Color Change Index() -> (Current Color Change Index=0)
    Shell Mesh:Dynamic Mesh Object Reference=@n51
    Caps Meshes:Array of Dynamic Mesh Object References=@n51.Out Caps Meshes
    Technic Caps Meshes:Array of Dynamic Mesh Object References=@n51.Out Technic Caps Meshes
    Technic Connectors Meshes:Array of Dynamic Mesh Object References=@n51.Out Technic Connectors Meshes
   ) -> (
    then
    Out Shell Mesh:Dynamic Mesh Object Reference
    Out Caps Meshes:Array of Dynamic Mesh Object References
    Out Technic Caps Meshes:Array of Dynamic Mesh Object References
    Out Technic Connectors Meshes:Array of Dynamic Mesh Object References
   ) {is=title K2Node_CallFunction}
   > Set ShellDynamicMesh(ShellDynamicMesh=@n52)
   > Set CapsDynamicMeshes(CapsDynamicMeshes=@n52.Out Caps Meshes)
   > Set TechnicCapsDynamicMeshes(TechnicCapsDynamicMeshes=@n52.Out Technic Caps Meshes)
   > Set TechnicConnectorsDynamicMeshes(TechnicConnectorsDynamicMeshes=@n52.Out Technic Connectors Meshes)
   > Set CurrentMaterialOffset(CurrentMaterialOffset:int=0) -> (Output_Get:int=0)
   > Offset Material Sections!(
    Exec:Exec
    Mesh:Dynamic Mesh Object Reference=Get ShellDynamicMesh()
    Material Sections Offset:int=Get MaterialSectionOffset() -> (MaterialSectionOffset=0)
   ) -> (Out:Exec, Target Mesh:Dynamic Mesh Object Reference) {is=title K2Node_MacroInstance}
   > n53 = Branch(Condition=Is Not Empty(TargetArray:Array of Dynamic Mesh Object References=Get CapsDynamicMeshes()))
   > Increment Int(Value=Get CurrentMaterialOffset() -> (CurrentMaterialOffset:int=0))
   > n54 = For Each Loop(Array:Array of Dynamic Mesh Object References=Get CapsDynamicMeshes()) -> (Array Element:Dynamic Mesh Object Reference)
   > ClearMaterialIDs!(
    execute
    TargetMesh:Dynamic Mesh Object Reference=@n54
    ClearValue:int=int + int(A=Get MaterialSectionOffset() -> (MaterialSectionOffset=0), B=Get CurrentMaterialOffset() -> (CurrentMaterialOffset=0))
   ) -> (then, ReturnValue:Dynamic Mesh Object Reference)
   Increment Int(" "=[@n54.Completed, @n53.else], Value=Get CurrentMaterialOffset() -> (CurrentMaterialOffset=0))
   > n55 = For Each Loop(Array=Get TechnicCapsDynamicMeshes())
   > Offset Material Sections(
    Mesh=@n55
    Material Sections Offset=int + int(A=Get MaterialSectionOffset() -> (MaterialSectionOffset=0), B=Get CurrentMaterialOffset() -> (CurrentMaterialOffset=0))
   )
   n56 = Sequence!(execute=@n13.else) -> (then_0, then_1)
   Branch(execute=@n10.then_0, Condition=Get Generate Nanite Mesh() -> (Generate Nanite Mesh:bool))
   > n57 = RequestMesh(self=Get DynamicMeshPool())
   > Set CurrentLODMergedMesh(CurrentLODMergedMesh=@n57)
   > n58 = Branch(Condition=Get bMultipleColorSurfaces())
   > n59 = For Loop(
    FirstIndex=Get First Color Change index() -> (First Color Change index=0)
    LastIndex=Get Max Color Change Index() -> (Max Color Change Index=0)
   )
   n60 = WeldMeshEdges(execute=[@n85.then_1, @n59.Completed], TargetMesh=Get CurrentLODMergedMesh(), WeldOptions_Tolerance=0, WeldOptions_bOnlyUniquePairs=True)
   > n61 = ConvertDynamicMeshToAtomPrimitiveGeometry(
    DynamicMesh=@n60
    AtomPrimitiveGeometry=Get Nanite Primitive Geometry() -> (Nanite Primitive Geometry:Atom Primitive Geometry Object Reference)
   )
   > n62 = BakeScale(self=@n61, Scale=GetDefaultPrimitiveScale())
   > Set members in Atom Primitive Geometry LODs!(execute, StructRef=Get PrimitiveGeometryLODs(), HiResNaniteGeomery:Atom Primitive Geometry Object Reference=@n62) -> (then, StructOut)
   > ReturnAllMeshes(self=Get DynamicMeshPool())
   n63 = Set Current Color Change Index(execute=@n59.LoopBody, Current Color Change Index=@n59) -> (Output_Get=0)
   > n64 = Set Current Color Change Filter(Current Color Change Filter=@n63) -> (Output_Get=0)
   n65 = Get Part Info()
   n66 = Set Nanite Applied Optimizations(
    execute=[@n85.then_0, @n64.then]
    Nanite Applied Optimizations:Atom Processor Applied Optimizations Structure=Get Optimizations To Apply For Nanite!(
     Part Info:Atom Model Part Instance Info Structure=@n65
     Color Change Index:int=Get Current Color Change Index() -> (Current Color Change Index=0)
    ) -> (Applied Optimizations:Atom Processor Applied Optimizations Structure) {is=title K2Node_CallFunction}
   ) -> (Output_Get:Atom Processor Applied Optimizations Structure)
   n67 = Break Atom Processor Geometry Style Settings!(AtomProcessorGeometryStyleSettings=Get Nanite Geometry Style Settings()) -> (GeometryStylePreferences_4_#17, PreferredGeometryLOD_30_#18)
   > n68 = Set Nanite Geometry Container(Nanite Geometry Container:Atom Primitive Geometry Container Object Reference=GetGeometryContainerForBestExportStyle(
     self=Break Atom Model Part Instance Info(AtomModelPartInstanceInfo=@n65)
     GeometryStylePriorityList=@n67
     LODIndex=@n67.PreferredGeometryLOD_30_#18
    )) -> (Output_Get:Atom Primitive Geometry Container Object Reference)
   n69 = Get Shell Geometry Section Flags And Filter(Applied Optimizations=@n66, Color Change Index=Get Current Color Change Filter() -> (Current Color Change Filter=0)) -> (Shell Section Flags=0)
   > Set Nanite Primitive Geometry(Nanite Primitive Geometry:Atom Primitive Geometry Object Reference=GetGeometrySectionWithFilter(self=@n68, PrimitiveGeometrySectionFlag=@n69, FilterParams=@n69.Shell Section Filter)) -> (Output_Get:Atom Primitive Geometry Object Reference)
   n70 = Get Nanite Geometry Style Settings()
   > n71 = Branch(Condition=Break Atom Processor Geometry Style Settings!(AtomProcessorGeometryStyleSettings=@n70) -> (ShouldBevel_8_#19:bool=~))
   > n72 = ApplyGeometricBevels!(
    execute
    GeometryToBevel:Atom Primitive Geometry Object Reference=Get Nanite Primitive Geometry()
    BevelSettings:"Atom Geometric Bevel Settings Structure (by ref)"=GetGeometricBevelSettings!(
     Part Instance Info:"Atom Model Part Instance Info Structure (by ref)"=Get Part Info()
     Geometry Style Settings:"Atom Processor Geometry Style Settings Structure (by ref)"=@n70
    ) -> (AtomGeometricBevelSettings:Atom Geometric Bevel Settings Structure) {is=title K2Node_CallFunction}
   ) -> (then, ReturnValue:Atom Primitive Geometry Object Reference)
   n73 = Break Atom Processor Geometry Style Settings!(AtomProcessorGeometryStyleSettings=Get Nanite Geometry Style Settings()) -> (NumberofSubdivisionsToApply_16_#24:int)
   n74 = Branch(execute=[@n72.then, @n71.else], Condition="Not Equal (Integer)"(A=@n73))
   > n75 = SubdivideGeometry!(execute, GeometryToSubdivide:Atom Primitive Geometry Object Reference=Get Nanite Primitive Geometry(), NumberOfSubdivisions:int=@n73) -> (then, ReturnValue:Atom Primitive Geometry Object Reference)
   n76 = RequestMesh(execute=[@n75.then, @n74.else], self=Get DynamicMeshPool())
   > n77 = ConvertAtomPrimitiveGeometryToDynamicMesh(AtomPrimitiveGeometry=Get Nanite Primitive Geometry(), DynamicMesh=@n76)
   > n78 = Set NaniteDynamicMesh(NaniteDynamicMesh:Dynamic Mesh Object Reference=@n77) -> (Output_Get:Dynamic Mesh Object Reference)
   > n79 = Get Nanite Geometry For Part - Main Pipeline!(
    execute
    self
    Part Info:Atom Model Part Instance Info Structure=Get Part Info()
    Applied Optimizations:Atom Processor Applied Optimizations Structure=Get Nanite Applied Optimizations() -> (Nanite Applied Optimizations:Atom Processor Applied Optimizations Structure)
    GeometryContainer:Atom Primitive Geometry Container Object Reference=Get Nanite Geometry Container() -> (Nanite Geometry Container:Atom Primitive Geometry Container Object Reference)
    Color Change Index:int=Get Current Color Change Index() -> (Current Color Change Index=0)
    Shell Mesh:Dynamic Mesh Object Reference=@n78
   ) -> (then, Out Shell Mesh:Dynamic Mesh Object Reference) {is=title K2Node_CallFunction}
   > n80 = Get Nanite Geometry For Part - Post Pipeline!(
    execute
    self
    Part Info:Atom Model Part Instance Info Structure=Get Part Info()
    Applied Optimizations:Atom Processor Applied Optimizations Structure=Get Nanite Applied Optimizations()
    Geometry Container:Atom Primitive Geometry Container Object Reference=Get Nanite Geometry Container()
    Color Change Index:int=Get Current Color Change Index() -> (Current Color Change Index=0)
    Shell Mesh:Dynamic Mesh Object Reference=@n79
   ) -> (then, Out Shell Mesh:Dynamic Mesh Object Reference) {is=title K2Node_CallFunction}
   > n81 = Set NaniteDynamicMesh(NaniteDynamicMesh=@n80)
   > n82 = Offset Material Sections(Mesh=@n81, Material Sections Offset=Get MaterialSectionOffset() -> (MaterialSectionOffset=0))
   > AppendMesh(TargetMesh=Get CurrentLODMergedMesh(), AppendMesh=@n82, bDeferChangeNotifications=false, AppendOptions=(CombineMode=EnableAllMatching))
   > n83 = Get Geometry Material Assignments For Nanite!(
    execute
    self:MP Merged Mesh Base Object Reference
    PartInfo:Atom Model Part Instance Info Structure=Get Part Info()
    Color Change Index:int=Get Current Color Change Index() -> (Current Color Change Index=0)
   ) -> (then, Material Assignments:Atom Primitive Geometry Material Assignments Structure) {is=title K2Node_CallFunction}
   n84 = Break Atom Primitive Geometry Material Assignments(AtomPrimitiveGeometryMaterialAssignments=@n83)
   > Set MaterialSectionOffset(MaterialSectionOffset=int + int(A=Get MaterialSectionOffset() -> (MaterialSectionOffset=0), B=Length(TargetArray=@n84.MaterialFeatures))) -> (Output_Get=0)
   n85 = Sequence(execute=@n58.else)
   n86 = GetUniqueNameForGeometryLODs!(execute=@n10.then_2, self:MP Merged Mesh Base Object Reference, Part Info:Atom Model Part Instance Info Structure=Get Part Info()) -> (then, Name:String)
   > Add(
    TargetMap:Map of Strings to Booleans=Get UVChangedMap() -> (UVChangedMap:Map of Strings to Booleans)
    Key:"String (by ref)"=@n86
    Value:"Boolean (by ref)"=Get bUVChanged()
   ) {is=title K2Node_CallFunction}
   > Return Node(Geometry LODs=Get PrimitiveGeometryLODs())
   Set bMultipleColorSurfaces(execute=@n5.else, bMultipleColorSurfaces=false)
   > Set Current Color Change Index(Current Color Change Index=0) -> (Output_Get=0)
   > n87 = Set Current Color Change Filter(Current Color Change Filter=-1) -> (Output_Get=0)
  ]}
 GetLODPresets {nodes [
   GetLODPresets() {is=title K2Node_FunctionEntry}
   > n1 = Branch(Condition=Get Use Custom LOD Settings() -> (Use Custom LOD Settings:bool))
   Return Node!(
    execute=[@n5.then, @n3.then, @n1.then]
    Presets:Atom Processor LODSettings Structure=Get Custom LOD Settings() -> (Custom LOD Settings:Atom Processor LODSettings Structure)
   )
   n2 = Is Valid(exec=@n1.else, InputObject=Get LOD Presets Data() -> (LOD Presets Data:BP Atom Processor LODPreset Object Reference))
   n3 = PrintString(execute=@n2.Is Not Valid, InString="LODPresetsData asset not found or invalid, using CustomLODSettings as fallback.")
   n4 = For Each Loop(
    Exec=@n2.Is Valid
    Array:Array of Atom Processor LODSettings Structures=Get LOD Presets!(self:BP Atom Processor LODPreset Object Reference=Get LOD Presets Data()) -> (LOD Presets:Array of Atom Processor LODSettings Structures)
   ) -> (Array Element:Atom Processor LODSettings Structure)
   n5 = PrintString(execute=@n4.Completed, InString="Couldn't find matching LOD preset in LODPresetData, using CustomLODSettings as fallback.")
   Branch(
    execute=@n4.LoopBody
    Condition="Equal (Enum)"(
     A:EAtomProcessorLODPreset Enum=Get Selected LOD Preset() -> (Selected LOD Preset:EAtomProcessorLODPreset Enum=NewEnumerator0)
     B:EAtomProcessorLODPreset Enum=Break Atom Processor LODSettings!(AtomProcessorLODSettings:"Atom Processor LODSettings Structure (by ref)"=@n4) -> (
      Name_26_#43:EAtomProcessorLODPreset Enum
      LODDistances_33_#44:"Array of Float (double-precision)s"
      LODTriangleBudgets_35_#45:Array of Integers
      QualityLevelMinLOD_18_#46:Map of EPerQualityLevels Enums to Integers
      UseLODSizeReference_31_#47:bool=~
      LODSizeReference_34_#48:double=~
     ) {is=title K2Node_BreakStruct}
    )
   )
   > Return Node(Presets=@n4)
  ]}
 GetMeshMergingSettings {nodes [
   n1 = GetMeshMergingSettings!() -> (then, AtomModel:Atom Model Object Reference) {is=title K2Node_FunctionEntry}
   > n2 = Parent: GetMeshMergingSettings!(execute, AtomModel:Atom Model Object Reference=@n1) -> (then, Regular Mesh Merging Settings:Atom Mesh Merging Settings Structure) {is=title K2Node_CallParentFunction}
   > n3 = GetLODPresets!(execute, self) -> (then, Presets:Atom Processor LODSettings Structure) {is=title K2Node_CallFunction}
   n4 = Break Atom Mesh Merging Settings!(AtomMeshMergingSettings:"Atom Mesh Merging Settings Structure (by ref)"=@n2) -> (
    CopiedLODsSettings:Atom Mesh Merging Copied LODs Settings Structure
    ApproximateLODsSettings:Atom Mesh Merging Approximate LODs Settings Structure
    CommonPartsSettings:Atom Mesh Merging Common Parts Settings Structure
    CoplanarMergeSettings:Atom Mesh Merging Coplanar Merge Settings Structure
    BudgetSettings:Atom Mesh Merging Budget Settings Structure
   ) {is=title K2Node_BreakStruct}
   n5 = Break Atom Processor LODSettings!(AtomProcessorLODSettings=@n3) -> (LODTriangleBudgets_35_#45)
   > Set members in Atom Mesh Merging Budget Settings!(
    execute:Exec
    StructRef:"Atom Mesh Merging Budget Settings Structure (by ref)"=@n4.BudgetSettings
    bEnforceLODBudgets:bool=Is Not Empty(TargetArray:Array of Integers=@n5)
    bManualBudgetPerLOD:bool=true
    TriangleBudgetPerLOD:Array of Integers=@n5
   ) -> (then:Exec, StructOut:"Atom Mesh Merging Budget Settings Structure (by ref)") {is=title K2Node_SetFieldsInStruct}
   > n6 = Set NumberOfSourceLODs(NumberOfSourceLODs:int=Length(TargetArray:Array of Atom Processor Geometry Style Settings Structures=Get LOD Geometry Style Settings())) -> (Output_Get:int=0)
   > Set members in Atom Mesh Merging Coplanar Merge Settings!(
    execute:Exec
    StructRef:"Atom Mesh Merging Coplanar Merge Settings Structure (by ref)"=@n4.CoplanarMergeSettings
    MergeCoplanarFacesStartLOD:int=@n6
   ) -> (then:Exec, StructOut:"Atom Mesh Merging Coplanar Merge Settings Structure (by ref)") {is=title K2Node_SetFieldsInStruct}
   > Set members in Atom Mesh Merging Common Parts Settings!(
    execute:Exec
    StructRef:"Atom Mesh Merging Common Parts Settings Structure (by ref)"=@n4.CommonPartsSettings
    ApproximateCommonPartLODs:int=Select(Option 0:int=1, Option 1:int=0, Index=Analytic Common Parts Requested On Any LODs()) -> (ReturnValue:int)
   ) -> (then:Exec, StructOut:"Atom Mesh Merging Common Parts Settings Structure (by ref)") {is=title K2Node_SetFieldsInStruct}
   > Set members in Atom Mesh Merging Approximate LODs Settings!(
    execute:Exec
    StructRef:"Atom Mesh Merging Approximate LODs Settings Structure (by ref)"=@n4.ApproximateLODsSettings
    ApproximationSourceLOD:int=int - int(A=@n6)
   ) -> (then:Exec, StructOut:"Atom Mesh Merging Approximate LODs Settings Structure (by ref)") {is=title K2Node_SetFieldsInStruct}
   > Set members in Atom Mesh Merging Copied LODs Settings!(execute:Exec, StructRef:"Atom Mesh Merging Copied LODs Settings Structure (by ref)"=@n4, NumCopiedLODs:int=@n6) -> (then:Exec, StructOut:"Atom Mesh Merging Copied LODs Settings Structure (by ref)") {is=title K2Node_SetFieldsInStruct}
   > Return Node!(execute, Regular Mesh Merging Settings:Atom Mesh Merging Settings Structure=@n2)
  ]}
 GetOverrideMeshMergingSettings {nodes [
   n1 = GetOverrideMeshMergingSettings!() -> (then, Default Settings:"Atom Mesh Merging Settings Structure (by ref)") {is=title K2Node_FunctionEntry}
   > n2 = Parent: GetOverrideMeshMergingSettings!(execute, Default Settings:"Atom Mesh Merging Settings Structure (by ref)"=@n1) -> (then, Mesh Merging Settings Map:Map of Strings to Atom Mesh Merging Settings Structures) {is=title K2Node_CallParentFunction}
   > Add(
    TargetMap:Map of Strings to Atom Mesh Merging Settings Structures=@n2
    Key=CP
    Value:"Atom Mesh Merging Settings Structure (by ref)"=Make Atom Mesh Merging Settings!(
     RequestedLODsSettings:Atom Mesh Merging Requested LODs Settings Structure=Make Atom Mesh Merging Requested LODs Settings!(NumLODs:int=1, LODGroup:Name=None) -> (AtomMeshMergingRequestedLODsSettings:Atom Mesh Merging Requested LODs Settings Structure) {is=title K2Node_MakeStruct}
     CopiedLODsSettings:Atom Mesh Merging Copied LODs Settings Structure=Make Atom Mesh Merging Copied LODs Settings!(NumCopiedLODs:int=1) -> (AtomMeshMergingCopiedLODsSettings:Atom Mesh Merging Copied LODs Settings Structure) {is=title K2Node_MakeStruct}
     SimplifiedLODsSettings:Atom Mesh Merging Simplified LODs Settings Structure=Make Atom Mesh Merging Simplified LODs Settings!(
      NumSimplifiedLODs:int=0
      SimplifyBaseTolerance:float=0.5
      SimplifyLODLevelToleranceScale:float=2
      bSimplifyPreserveCorners:bool=True
      SimplifySharpEdgeAngleDeg:float=44
      SimplifyMinSalientDimension:float=4
     ) -> (AtomMeshMergingSimplifiedLODsSettings:Atom Mesh Merging Simplified LODs Settings Structure) {is=title K2Node_MakeStruct}
     CoarseLODsSettings:Atom Mesh Merging Coarse LODs Settings Structure=Make Atom Mesh Merging Coarse LODs Settings!(
      CoarseLODStrategy:EAtomMeshMergingCoarseApproximationStrategy Enum=Automatic
      NumCoarseLODs:int=0
      CoarseLODBaseTolerance:float=1
      CoarseLODMaxTriCountBase:int=64
      CoarseApproximationDetailSize:float=10
     ) -> (AtomMeshMergingCoarseLODsSettings:Atom Mesh Merging Coarse LODs Settings Structure) {is=title K2Node_MakeStruct}
     HiddenFacesRemovalSettings:Atom Mesh Merging Hidden Faces Removal Settings Structure=Make Atom Mesh Merging Hidden Faces Removal Settings!(
      RemoveHiddenFacesMethod:EAtomMeshMergingRemoveHiddenFacesMode Enum=None
      RemoveHiddenStartLOD:int=1
      RemoveHiddenSamplingDensity:float=1
      bDoubleSidedHiddenRemoval:bool=False
     ) -> (AtomMeshMergingHiddenFacesRemovalSettings:Atom Mesh Merging Hidden Faces Removal Settings Structure) {is=title K2Node_MakeStruct}
     BudgetSettings:Atom Mesh Merging Budget Settings Structure=Make Atom Mesh Merging Budget Settings!(
      bEnforceLODBudgets:bool=false
      TriangleBudget:int=4500
      PartLODPromotionBudgetMultiplier:float=2
      bManualBudgetPerLOD:bool=False
      TriangleBudgetPerLOD:Array of Integers
     ) -> (AtomMeshMergingBudgetSettings:Atom Mesh Merging Budget Settings Structure) {is=title K2Node_MakeStruct}
    ) -> (AtomMeshMergingSettings:Atom Mesh Merging Settings Structure) {is=title K2Node_MakeStruct}
   )
   > Return Node!(execute, Mesh Merging Settings Map:Map of Strings to Atom Mesh Merging Settings Structures=@n2)
  ]}
 GetPredominantNormalFromPolygroup {nodes [
   n1 = GetPredominantNormalFromPolygroup!() -> (then, DynamicMesh:Dynamic Mesh Object Reference, PolygroupID:int) {is=title K2Node_FunctionEntry}
   > Get Triangles In PolyGroup!(
    execute
    TargetMesh:Dynamic Mesh Object Reference=@n1
    GroupLayer:Geometry Script Group Layer Structure
    PolygroupID:int=@n1.PolygroupID
    TriangleIDsOut:"Index List Structure (by ref)"=Get TriangleIDs() -> (TriangleIDs:Index List Structure)
   ) -> (then, ReturnValue:Dynamic Mesh Object Reference) {is=title K2Node_CallFunction}
   > Set TriangleCumulativeArea(TriangleCumulativeArea:double=0) -> (Output_Get:double)
   > n2 = ConvertIndexListToArray!(execute, IndexList:Index List Structure=Get TriangleIDs()) -> (then, IndexArray:Array of Integers)
   > n3 = For Each Loop(Array:Array of Integers=@n2) -> (Array Element:int)
   n4 = GetTriangleFaceNormal!(TargetMesh:Dynamic Mesh Object Reference=Get DynamicMesh(), TriangleID:int=@n3) -> (bIsValidTriangle:bool, ReturnValue:Vector)
   n5 = Get Triangle Area!(DynamicMesh:Dynamic Mesh Object Reference=Get DynamicMesh(), TriangleID:int=@n3) -> (Area:double=~) {is=title K2Node_MacroInstance}
   > Set TrianglePredominantNormal(TrianglePredominantNormal:Vector=vector + vector!(
     A=Get TrianglePredominantNormal() -> (TrianglePredominantNormal:Vector)
     B=vector * vector(A=@n4.ReturnValue, B:double=@n5) -> (ReturnValue=~)
    ) -> (ReturnValue:Vector)) -> (Output_Get:Vector)
   > Set TriangleCumulativeArea(TriangleCumulativeArea=float + float!(A=Get TriangleCumulativeArea() -> (TriangleCumulativeArea:double), B=@n5) -> (ReturnValue))
   n6 = Get TriangleCumulativeArea()
   n7 = "Equal (Float)"(A=@n6)
   Return Node!(
    execute=@n3.Completed
    Normal:Vector=vector / vector(A:Vector=Get TrianglePredominantNormal(), B=SelectFloat(A=1, B=@n6, bPickA=@n7)) -> (ReturnValue:Vector)
   )
  ]}
 GetPreserveUVsForPart {nodes [
   n1 = GetPreserveUVsForPart!() -> (
    then
    Atom Model Part Info:"Atom Model Part Instance Info Structure (by ref)"
    Merged Mesh Part Info:"Atom Merged Mesh Part Info Structure (by ref)"
   ) {is=title K2Node_FunctionEntry}
   > n2 = Parent: GetPreserveUVsForPart!(
    execute
    Atom Model Part Info:"Atom Model Part Instance Info Structure (by ref)"=@n1
    Merged Mesh Part Info:"Atom Merged Mesh Part Info Structure (by ref)"=@n1.Merged Mesh Part Info
   ) -> (then, PreserveUVs:bool) {is=title K2Node_CallParentFunction}
   > n3 = GetUniqueNameForGeometryLODs(Part Info=@n1)
   n4 = Find(TargetMap:Map of Strings to Booleans=Get UVChangedMap(), Key:"String (by ref)"=@n3) -> (Value:bool)
   > n5 = Branch(Condition=@n4.ReturnValue)
   > n6 = Set bPreserveUVs(bPreserveUVs:bool=@n4) -> (Output_Get:bool)
   Return Node!(execute=[@n5.else, @n6.then], PreserveUVs:bool=OR Boolean(A=@n2, B=Get bPreserveUVs() -> (bPreserveUVs:bool)))
  ]}
 GetTechnicHoleMaterialFeatures {nodes [
   n1 = GetTechnicHoleMaterialFeatures!() -> (
    then
    Part Info:Atom Model Part Instance Info Structure
    Geometry Style Settings:Atom Processor Geometry Style Settings Structure
    Style Name:String
    Technic Cap Index:int
   ) {is=title K2Node_FunctionEntry}
   n2 = Break Atom Model Part Instance Info(AtomModelPartInstanceInfo=Get Part Info() -> (Part Info:Atom Model Part Instance Info Structure))
   n3 = Break Atom Processor Geometry Style Settings!(AtomProcessorGeometryStyleSettings=Get Geometry Style Settings()) -> (
    GeometryStylePreferences_4_#17:Array of Strings
    PreferredGeometryLOD_30_#18:int
    ShouldBevel_8_#19:bool=~
    BevelWidth_11_#20:double=~
    BevelAngleThreshold_13_#21:double=~
    UseImprovedBevelAlgorithm_37_#22:bool=~
    OnlyBevelBlock-LikePrimitives_39_#23:bool=~
    NumberofSubdivisionsToApply_16_#24
    FakeBevel_23_#25:bool=~
    FakeUnderside_25_#15:bool=~
    AnalyticFakeUnderside_32_#26:bool=~
    ForceFlatCap_35_#16:bool=~
    FakeTechnicHole_41_#1:bool=~
    AnalyticCommonParts_27_#2:bool=~
   )
   n4 = GetGeometryContainerForBestExportStyle(self=@n2, GeometryStylePriorityList=@n3, LODIndex=@n3.PreferredGeometryLOD_30_#18)
   n5 = GetLOD(self=@n4) -> (ReturnValue=0)
   n6 = GetExportStyleName(self=@n4)
   n7 = Get Part Info()
   n8 = GetPartColorInfo(Part Info=@n7, Color Change Index=0) -> (Color Id=0, Material Type=None)
   > Return Node!(
    execute
    Material Features:Atom Material Features Structure=Make Atom Material Features(
     MaterialType=@n8.Material Type
     TechnicNormalMap=LoadDetailTextureForExportStyleAndSection(self=@n2, ExportStyle=@n6, LOD=@n5, TextureType=NormalMap, Section=TechnicHoleCap)
     TechnicMasks=LoadDetailTextureForExportStyleAndSection(self=@n2, ExportStyle=@n6, LOD=@n5, TextureType=TechnicStateMask, Section=TechnicHoleCap)
     LookUpTableAtlasTexture=Get LUT Atlas Override()
     ColorSource=Get Material Color Source() -> (Material Color Source=VertexColor)
     ColorId=@n8
     UVFeature=FakeTechnicHole
     TranslucencyType=Get Translucency Type For Part(Part Info=@n7) -> (Translucency Type=Translucent) {is=title K2Node_CallFunction}
     MaterialOverride=(Material=None,bManualColorSource=False)
    )
    Slot Name:String=Append(A=@n1.Style Name, B=_TechnicHole)
   )
  ]}
 GetTriangleArea {nodes [
   n1 = Inputs!() -> (DynamicMesh:Dynamic Mesh Object Reference, TriangleID:int)
   n2 = GetTrianglePositions(TargetMesh=@n1, TriangleID=@n1.TriangleID)
   n3 = Cross Product!(A:Vector=vector - vector(A=@n2.Vertex2, B=@n2.Vertex1), B:Vector=vector - vector(A=@n2.Vertex3, B=@n2.Vertex1)) -> (ReturnValue:Vector)
   Outputs!(Area:double=float * float!(A=Vector Length(A=@n3), B=0.5) -> (ReturnValue))
  ]}
 GetUndersideMaterialReplacementIndex {nodes [
   n1 = GetUndersideMaterialReplacementIndex!() -> (then, Underside Material Feature:Atom Material Features Structure, Materials:Array of Material Interface Object References) {is=title K2Node_FunctionEntry}
   > n2 = For Each Loop(Array:Array of Material Interface Object References=@n1.Materials) -> (Array Element:Material Interface Object Reference)
   Return Node!(execute=@n2.Completed, Material Replacement Index:int=0)
   > n3 = GetMaterialFeatures!(
    execute
    self:Atom Material Factory Object Reference=GetMaterialFactory!() -> (Material Factory:Atom Material Factory Object Reference)
    Material:Material Interface Object Reference=@n2
   ) -> (then, ReturnValue:Atom Material Features Structure)
   n4 = Break Atom Material Features!(AtomMaterialFeatures:"Atom Material Features Structure (by ref)"=@n3) -> (UVFeature:EAtomMaterialUVFeature Enum, AdditionalFeatures:Array of Instanced Struct Structures) {is=title K2Node_BreakStruct}
   n5 = "Equal (Enum)"(A:EAtomMaterialUVFeature Enum=@n4, B:EAtomMaterialUVFeature Enum=FakeBevels)
   n6 = "Equal (Enum)"(A=@n4, B=None)
   n7 = Break Atom Material Features!(AtomMaterialFeatures=@n1) -> (
    MaterialType:EAtomMaterialType Enum
    UDIMDecorationMap:Map of Integers to Texture Object References
    NormalMap:Texture Object Reference
    AlphaMask:Texture Object Reference
    TechnicNormalMap:Texture Object Reference
    TechnicMasks:Texture Object Reference
    LookUpTableAtlasTexture:Texture Object Reference
    ColorSource:EAtomMaterialColorSource Enum
    ColorId:int
    UVFeature
    TranslucencyType:EAtomMaterialTranslucencyType Enum
    MaterialOverride:Atom Material Override Structure
    AdditionalFeatures
   )
   n8 = Length(TargetArray:Array of Instanced Struct Structures=@n7.AdditionalFeatures)
   n9 = Length(TargetArray=@n4.AdditionalFeatures)
   > Branch(Condition=AND Boolean!(A=OR Boolean(A=@n6, B=@n5), B="Equal (Integer)"(A=@n8, B=@n9)) -> (ReturnValue))
   > n10 = For Each Loop(Array:Array of Instanced Struct Structures=@n7.AdditionalFeatures) -> (Array Element:Instanced Struct Structure)
   n11 = "Get (a ref)"(Array:Array of Instanced Struct Structures=@n4.AdditionalFeatures, Dimension 1=@n10.Array Index) -> (Output:"Instanced Struct Structure (by ref)")
   > Set AllAdditionalFeaturesEqual(AllAdditionalFeaturesEqual:bool=AND Boolean(
     A=Equal(A:"Instanced Struct Structure (by ref)"=@n10, B:"Instanced Struct Structure (by ref)"=@n11) -> (ReturnValue:bool=~) {is=title K2Node_PromotableOperator}
     B=Get AllAdditionalFeaturesEqual() -> (AllAdditionalFeaturesEqual:bool)
    )) -> (Output_Get:bool)
   n12 = Break Atom Material Features(AtomMaterialFeatures=@n3)
   n13 = Break Atom Material Override(AtomMaterialOverride=@n12.MaterialOverride) -> (bManualColorSource=~)
   Branch(
    execute=@n10.Completed
    Condition=AND Boolean!(
     A="Equal (Enum)"(A:EAtomMaterialType Enum=@n7, B:EAtomMaterialType Enum=@n12)
     B="Equal (Object)"(A=@n7.LookUpTableAtlasTexture, B=@n12.LookUpTableAtlasTexture)
     C:bool="Equal (Enum)"(A:EAtomMaterialColorSource Enum=@n7.ColorSource, B:EAtomMaterialColorSource Enum=@n12.ColorSource)
     D:bool="Equal (Enum)"(A:EAtomMaterialTranslucencyType Enum=@n7.TranslucencyType, B:EAtomMaterialTranslucencyType Enum=@n12.TranslucencyType)
     E:bool="Equal (Object)"(A=@n13)
     F:bool=Get AllAdditionalFeaturesEqual()
    ) -> (ReturnValue)
   )
   > Return Node(Material Replacement Index=@n2.Array Index)
  ]}
 GetUniqueNameForGeometryLODs {nodes [
   n1 = GetUniqueNameForGeometryLODs!() -> (then, Part Info:Atom Model Part Instance Info Structure) {is=title K2Node_FunctionEntry}
   > n2 = Parent: GetUniqueNameForGeometryLODs!(execute, Part Info:Atom Model Part Instance Info Structure=@n1) -> (then, Name:String) {is=title K2Node_CallParentFunction}
   n3 = Break Atom Processor Applied Optimizations!(AtomProcessorAppliedOptimizations=Get Optimizations To Apply For LOD(Part Info=@n1, LOD=0)) -> (
    bApplyFakeTranslucent_25_#9:bool=~
    bApplyDitheredTranslucent_24_#10:bool=~
    bApplyGlow_13_#11:bool=~
   )
   > n4 = Branch(Condition=@n3)
   > Return Node!(execute, Name:String=Append(A=@n2, B=_FakeTransluscent))
   n5 = Branch(execute=@n4.else, Condition=@n3.bApplyDitheredTranslucent_24_#10)
   > Return Node(Name=Append(A=@n2, B=_Dithered))
   n6 = Branch(execute=@n5.else, Condition=@n3.bApplyGlow_13_#11)
   n7 = Break Atom Model Part Instance Info!(AtomModelPartInstanceInfo=@n1) -> (PartGuid:Atom Model Part Guid Structure)
   > Return Node(Name=Append!(A=@n2, B=_Glow_, C:String=To String(InModelPartGuid=@n7)) -> (ReturnValue))
   Return Node(execute=@n6.else, Name=@n2)
  ]}
 GetVertexColorForCommonPart {nodes [
   n1 = GetVertexColorForCommonPart!() -> (
    then
    Part Info:"Atom Model Part Instance Info Structure (by ref)"
    Merged Mesh Part Info:"Atom Merged Mesh Part Info Structure (by ref)"
    Common Part Index:int=0
   ) {is=title K2Node_FunctionEntry}
   > n2 = Parent: GetVertexColorForCommonPart!(
    execute
    Part Info:"Atom Model Part Instance Info Structure (by ref)"=@n1
    Merged Mesh Part Info:"Atom Merged Mesh Part Info Structure (by ref)"=@n1.Merged Mesh Part Info
    Common Part Index:int=@n1.Common Part Index
   ) -> (then, Vertex Color:Color Structure) {is=title K2Node_CallParentFunction}
   > n3 = Branch(Condition=Analytic Common Parts Requested On Any LODs())
   n4 = "Get (a ref)"(
    Array:Array of Atom Common Part And Transform Structures=Break Atom Model Part Instance Info!(AtomModelPartInstanceInfo=@n1) -> (CommonParts:Array of Atom Common Part And Transform Structures)
    Dimension 1=@n1.Common Part Index
   ) -> (Output:"Atom Common Part And Transform Structure (by ref)")
   n5 = Break Color(Color=@n2)
   > Return Node!(
    execute
    Vertex Color:Color Structure=Make Color!(
     B:Byte=@n5
     G:Byte=@n5.G
     R:Byte=@n5.R
     A:Byte=Break Atom Common Part and Transform!(AtomCommonPartAndTransform:"Atom Common Part And Transform Structure (by ref)"=@n4) -> (Type:EAtomCommonPartType Enum, Transform:Transform, GeometryStyles:Array of Names) {is=title K2Node_BreakStruct}
    ) -> (Color:Color Structure) {is=title K2Node_MakeStruct}
   )
   Return Node(execute=@n3.else, Vertex Color=@n2)
  ]}
 Has Color Change Baked in Geometry {nodes [
   n1 = Has Color Change Baked in Geometry!() -> (then, PartInfo:Atom Model Part Instance Info Structure) {is=title K2Node_FunctionEntry}
   n2 = IsFlexElement(self=Break Atom Model Part Instance Info!(AtomModelPartInstanceInfo=@n1) -> (Primitive:Atom Primitive Object Reference))
   n3 = GetIsProcessingPrimitive()
   > Return Node!(
    execute
    HasColorChangeBakedInGeometry:bool=AND Boolean!(
     A=Is Color Change Part!(PartInfo:"Atom Model Part Instance Info Structure (by ref)"=@n1) -> (Is Color Change Part:bool)
     B=NOT Boolean(A=@n2)
     C=NOT Boolean(A=@n3)
     D=Get Use Improved Color Change Processing() -> (Use Improved Color Change Processing:bool)
    ) -> (ReturnValue)
   )
  ]}
 Has Custom Baked Vertex Color {nodes [
   n1 = Has Custom Baked Vertex Color!() -> (then, PartInfo:Atom Model Part Instance Info Structure) {is=title K2Node_FunctionEntry}
   > n2 = Features Applied On Any LOD!(
    execute
    self
    Part Info:Atom Model Part Instance Info Structure=@n1
    Expected Features:Atom Processor Applied Optimizations Structure=Make Atom Processor Applied Optimizations!(bApplyTechnicCap_12_#7=true) -> (AtomProcessorAppliedOptimizations)
   ) -> (then, Result:bool) {is=title K2Node_CallFunction}
   > Return Node!(
    execute
    HasCustomBakedVertexColor:bool=OR Boolean(
     A=Parent: Has Custom Baked Vertex Color!(PartInfo:Atom Model Part Instance Info Structure=@n1) -> (HasCustomBakedVertexColor:bool) {is=title K2Node_CallParentFunction}
     B=@n2
    )
   )
  ]}
 Has Fake Technic Hole For Geometry Style {nodes [
   n1 = Has Fake Technic Hole For Geometry Style!() -> (then, Part Info:Atom Model Part Instance Info Structure, Geometry Style Settings:Atom Processor Geometry Style Settings Structure) {is=title K2Node_FunctionEntry}
   n2 = Break Atom Model Part Instance Info(AtomModelPartInstanceInfo=@n1)
   n3 = Break Atom Processor Geometry Style Settings!(AtomProcessorGeometryStyleSettings=@n1.Geometry Style Settings) -> (GeometryStylePreferences_4_#17, PreferredGeometryLOD_30_#18)
   > Return Node!(
    execute
    Has Fake Technic Hole:bool=GetSectionTextureName(
     self=GetGeometryContainerForBestExportStyle(self=@n2, GeometryStylePriorityList=@n3, LODIndex=@n3.PreferredGeometryLOD_30_#18)
     SectionType=TechnicHoleCap
     TextureType=NormalMap
    ).ReturnValue
   )
  ]}
 Has Technic Connector Geometry For Geometry Style {nodes [
   n1 = Has Technic Connector Geometry For Geometry Style!() -> (then, Part Info:Atom Model Part Instance Info Structure, Geometry Style Settings:Atom Processor Geometry Style Settings Structure) {is=title K2Node_FunctionEntry}
   n2 = Break Atom Model Part Instance Info(AtomModelPartInstanceInfo=@n1)
   n3 = GetNumTechnicPinConnectors!(self:Atom Primitive Object Reference=@n2) -> (ReturnValue:int=0)
   > n4 = Branch(Condition="integer > integer"(A=@n3))
   > n5 = For Loop(LastIndex=int - int(A=@n3))
   n6 = GetConnectionFieldForTechnicPinConnector!(self:Atom Primitive Object Reference=@n2, ConnectorIndex:int=@n5) -> (ConnectivityFieldReference:Typed Connectivity Field Reference Structure, ReturnValue:bool)
   > n7 = Branch(Condition=@n6.ReturnValue)
   > n8 = GetTechnicPinConnectivityStates!(
    execute
    self:Atom Model Parts Collection Connectivity Object Reference=GetPartsCollectionConnectivity!() -> (Parts Collection Connectivity:Atom Model Parts Collection Connectivity Object Reference)
    TechnicPinPartConnectivityRef:"Atom Part Connectivity Field Reference Structure (by ref)"=Make Atom Part Connectivity Field Reference!(PartInfo:Atom Model Part Instance Info Structure=Get Part Info(), ConnectivityField:Typed Connectivity Field Reference Structure=@n6) -> (AtomPartConnectivityFieldReference:Atom Part Connectivity Field Reference Structure) {is=title K2Node_MakeStruct}
   ) -> (
    then
    ConnectedPartsConnectivityRef:Array of Atom Part Connectivity Field Reference Structures
    ReturnValue:Array of EFieldConnectResult Enums=NoConnection
   )
   > n9 = For Each Loop(Array:Array of EFieldConnectResult Enums=@n8.ReturnValue) -> (Array Element:EFieldConnectResult Enum)
   > Branch(Condition=Select!(
     NoConnection:bool=false
     Rejection:bool=false
     FixedConnection:bool=true
     FreeConnection:bool=true
     HingeConnection:bool=true
     CardanConnection:bool=false
     BallConnection:bool=false
     PrismaticConnection:bool=false
     CylindricalConnection:bool=true
     PrismaticAPerpendicularHingeConnection:bool=false
     PrismaticBPerpendicularHingeConnection:bool=false
     ConnectResultMax:bool=false
     FirstConnection:bool=false
     LastConnection:bool=false
     Index:EFieldConnectResult Enum=@n9
    ) -> (ReturnValue:bool))
   n10 = "Get (a ref)"(Array:Array of Atom Part Connectivity Field Reference Structures=@n8, Dimension 1=@n9.Array Index) -> (Output:"Atom Part Connectivity Field Reference Structure (by ref)")
   > Return Node!(
    execute
    Has Technic Connector:bool=Has Fake Technic Hole For Geometry Style(
     Part Info=Break Atom Part Connectivity Field Reference!(AtomPartConnectivityFieldReference:"Atom Part Connectivity Field Reference Structure (by ref)"=@n10) -> (PartInfo:Atom Model Part Instance Info Structure, ConnectivityField:Typed Connectivity Field Reference Structure) {is=title K2Node_BreakStruct}
     Geometry Style Settings=Get Geometry Style Settings()
    ) {is=title K2Node_CallFunction}
   )
   PrintString(execute=@n7.else, InString=Has Fake Technic Connector For Geometry: Could not access technic pin connector.)
   Return Node(execute=@n4.else, Has Technic Connector=false)
   n11 = Break Atom Processor Geometry Style Settings(AtomProcessorGeometryStyleSettings=@n1.Geometry Style Settings)
   n12 = GetGeometryContainerForBestExportStyle(
    self=Break Atom Model Part Instance Info(AtomModelPartInstanceInfo=@n1)
    GeometryStylePriorityList=@n11
    LODIndex=@n11.PreferredGeometryLOD_30_#18
   )
   Return Node(Has Technic Connector=Is Not Empty(TargetArray:Array of Atom Primitive Geometry And Transform Structures=GetTechnicConnectorsGeometry(self=@n12, ColorChangeIndex=-1)))
  ]}
 Initialize Geometry Style Settings {nodes [
   Initialize Geometry Style Settings() {is=title K2Node_FunctionEntry}
   > Parent: Initialize Geometry Style Settings() {is=title K2Node_CallParentFunction}
   n1 = Get Source LOD Settings() -> (Source LOD Settings:Array of Merged Mesh Source LODSettings Structures)
   > Branch(Condition=Is Not Empty(TargetArray:Array of Merged Mesh Source LODSettings Structures=@n1))
   > Clear(TargetArray:Array of Atom Processor Geometry Style Settings Structures=Get LOD Geometry Style Settings())
   n2 = Get Geometry style() -> (Geometry style:String)
   > n3 = Branch(Condition=IsEmpty(InString=@n2))
   > n4 = Add(TargetArray:Array of Strings=Get Style Preferences() -> (Style Preferences:Array of Strings), NewItem:"String (by ref)"=Realtime) {is=title K2Node_CallArrayFunction}
   n5 = For Each Loop(Exec=[@n4.then, @n9.then, @n8.then], Array:Array of Merged Mesh Source LODSettings Structures=@n1) -> (Array Element:Merged Mesh Source LODSettings Structure)
   n6 = Break Merged Mesh Source LODSettings!(MergedMeshSourceLODSettings:"Merged Mesh Source LODSettings Structure (by ref)"=@n5) -> (
    FakeBevel_22_#49:bool=~
    FakeUnderside_21_#50:bool=~
    FakeAnalyticUnderside_26_#51:bool=~
    AnalyticCommonParts_24_#52:bool=~
    ForceFlatCap_28_#53:bool=~
   ) {is=title K2Node_BreakStruct}
   > Add(
    TargetArray:Array of Atom Processor Geometry Style Settings Structures=Get LOD Geometry Style Settings()
    NewItem:"Atom Processor Geometry Style Settings Structure (by ref)"=Make Atom Processor Geometry Style Settings!(
     GeometryStylePreferences_4_#17:Array of Strings=Get Style Preferences()
     PreferredGeometryLOD_30_#18:int=0
     ShouldBevel_8_#19:bool=False
     BevelWidth_11_#20:double=0.0125
     BevelAngleThreshold_13_#21:double=60
     UseImprovedBevelAlgorithm_37_#22:bool=False
     OnlyBevelBlock-LikePrimitives_39_#23:bool=True
     NumberofSubdivisionsToApply_16_#24:int=0
     FakeBevel_23_#25:bool=@n6
     FakeUnderside_25_#15:bool=@n6.FakeUnderside_21_#50
     AnalyticFakeUnderside_32_#26:bool=@n6.FakeAnalyticUnderside_26_#51
     ForceFlatCap_35_#16:bool=@n6.ForceFlatCap_28_#53
     FakeTechnicHole_41_#1:bool=False
     AnalyticCommonParts_27_#2:bool=@n6.AnalyticCommonParts_24_#52
    ) -> (AtomProcessorGeometryStyleSettings:Atom Processor Geometry Style Settings Structure) {is=title K2Node_MakeStruct}
   )
   Add(execute=@n3.else, TargetArray:Array of Strings=Get Style Preferences(), NewItem:"String (by ref)"=@n2)
   n7 = Get Fallback Geometry style() -> (Fallback Geometry style:String)
   > n8 = Branch(Condition=IsEmpty(InString=@n7))
   n9 = Add(execute=@n8.else, TargetArray=Get Style Preferences(), NewItem=@n7)
  ]}
 InjectPresetCollisionSettings {nodes [InjectPresetCollisionSettings() {is=title K2Node_FunctionEntry}]}
 Is Geometry Style Compatible With Part Features {nodes [
   n1 = Is Geometry Style Compatible With Part Features!() -> (
    then
    Part Info:"Atom Model Part Instance Info Structure (by ref)"
    Geometry Style Settings:"Atom Processor Geometry Style Settings Structure (by ref)"
   ) {is=title K2Node_FunctionEntry}
   n2 = IsGlowPart(PartInfo=@n1)
   n3 = Break Atom Processor Geometry Style Settings!(AtomProcessorGeometryStyleSettings=@n1.Geometry Style Settings) -> (
    GeometryStylePreferences_4_#17
    PreferredGeometryLOD_30_#18
    ShouldBevel_8_#19:bool=~
    BevelWidth_11_#20:double=~
    BevelAngleThreshold_13_#21:double=~
    UseImprovedBevelAlgorithm_37_#22:bool=~
    OnlyBevelBlock-LikePrimitives_39_#23:bool=~
    NumberofSubdivisionsToApply_16_#24:int
    FakeBevel_23_#25:bool=~
    FakeUnderside_25_#15:bool=~
    AnalyticFakeUnderside_32_#26:bool=~
    ForceFlatCap_35_#16:bool=~
    FakeTechnicHole_41_#1:bool=~
    AnalyticCommonParts_27_#2:bool=~
   )
   n4 = GetSectionTextureName(
    self=GetGeometryContainerForBestExportStyle(
     self=Break Atom Model Part Instance Info(AtomModelPartInstanceInfo=@n1)
     GeometryStylePriorityList=@n3
     LODIndex=@n3.PreferredGeometryLOD_30_#18
    )
    SectionType=Shell
    TextureType=AlphaMask
   )
   > Return Node!(execute, Is Compatible:bool=OR Boolean(A=NOT Boolean(A=@n2), B=NOT Boolean(A=@n4.ReturnValue)))
  ]}
 Is Glow Position Mergeable {nodes [
   Is Glow Position Mergeable!() -> (then, GlowPosition:Vector=~) {is=title K2Node_FunctionEntry}
   > n1 = Branch(Condition="Not Equal (Float)"(A=Get Glow Position Merge Distance() -> (Glow Position Merge Distance:double)))
   > n2 = For Each Loop(Array:Array of Vectors=Get GlowLocations() -> (GlowLocations:Array of Vectors)) -> (Array Element:Vector)
   Return Node!(execute=[@n1.else, @n2.Completed], GlowIndex:int=-1, IsGlowPositionMergeable:bool=false)
   n3 = Get GlowPosition() -> (GlowPosition:Vector)
   > Branch(Condition="float <= float"(A="Distance (Vector)"(V1=@n2, V2=@n3), B=Get Glow Position Merge Distance()) -> (ReturnValue=~))
   > Return Node(GlowIndex=@n2.Array Index, IsGlowPositionMergeable=true)
  ]}
 IsCommonPartAffectedByCapOptimization {nodes [
   n1 = Inputs!() -> (Exec:"Exec (by ref)", PartInfo:"Atom Model Part Instance Info Structure (by ref)", CommonPartIndex:"Integer (by ref)")
   n2 = Break Atom Model Part Instance Info!(AtomModelPartInstanceInfo=@n1.PartInfo) -> (CommonParts:Array of Atom Common Part And Transform Structures)
   n3 = GetCommonPartCategoryFromType!(CommonPartType:EAtomCommonPartType Enum=Break Atom Common Part and Transform!(AtomCommonPartAndTransform="Get (a ref)"(Array:Array of Atom Common Part And Transform Structures=@n2, Dimension 1=@n1.CommonPartIndex) -> (Output:"Atom Common Part And Transform Structure (by ref)")) -> (Type, GeometryStyles)) -> (ReturnValue:EAtomCommonPartCategory Enum=None)
   n4 = Branch(
    execute=@n1
    Condition=OR Boolean(A="Equal (Enum)"(A:EAtomCommonPartCategory Enum=@n3, B:EAtomCommonPartCategory Enum=Pin), B="Equal (Enum)"(A=@n3, B=Tube))
   )
   > n5 = Features Applied On Any LOD(
    Part Info=@n1.PartInfo
    Expected Features=Make Atom Processor Applied Optimizations!(
     bApplyFakeBevels_3_#3:bool=False
     bApplyFakeUnderside_9_#4:bool=true
     bAnalyticFakeUnderside_10_#5:bool=False
     bApplyFlatCap_11_#6:bool=true
     bApplyTechnicCap_12_#7=False
     bApplyFakeTechnicConnector_22_#8:bool=False
     bApplyFakeTranslucent_25_#9:bool=False
     bApplyDitheredTranslucent_24_#10:bool=False
     bApplyGlow_13_#11:bool=False
     bUseNormalMap_16_#12:bool=False
     bUseAlphaMask_17_#13:bool=False
     ExtraOptmizationsToApply_20_#14:Instanced Struct Structure=None
    ) -> (AtomProcessorAppliedOptimizations)
   )
   > n6 = Branch(Condition=@n5)
   > Outputs!(True:Exec, False:Exec=[@n6.else, @n4.else])
  ]}
 IsDitheredTranslucentPart {nodes [
   n1 = IsDitheredTranslucentPart!() -> (then, Part Info:Atom Model Part Instance Info Structure, Color Change Index:int) {is=title K2Node_FunctionEntry}
   n2 = Get Translucency Type For Part(Part Info=@n1) -> (Translucency Type=Translucent)
   n3 = GetPartColorInfo(Part Info=@n1, Color Change Index=@n1.Color Change Index) -> (Color Id=0, Material Type=None)
   > Return Node!(
    execute
    IsDitheredTranslucent:bool=AND Boolean!(
     A="Equal (Enum)"(A:EAtomMaterialTranslucencyType Enum=@n2, B:EAtomMaterialTranslucencyType Enum=Dithering)
     B=Select!(
      None:bool=false
      Standard:bool=false
      Transparent:bool=true
      Glitter:bool=true
      Opalescent:bool=true
      Metallic:bool=false
      Index:EAtomMaterialType Enum=@n3.Material Type
     ) -> (ReturnValue)
    ) -> (ReturnValue)
   )
  ]}
 IsFakeTranslucentPart {nodes [
   n1 = IsFakeTranslucentPart!() -> (then, PartInfo:Atom Model Part Instance Info Structure, Color Change Index:int) {is=title K2Node_FunctionEntry}
   n2 = GetPartColorInfo(Part Info=@n1, Color Change Index=@n1.Color Change Index) -> (Color Id=0, Material Type=None)
   n3 = Find(TargetMap:Map of Integers to Atom Material Override Structures=Get Material ID Overrides(), Key:"Integer (by ref)"=@n2) -> (Value:Atom Material Override Structure)
   n4 = Get Translucency Type For Part(Part Info=@n1) -> (Translucency Type=Translucent)
   > n5 = Branch(Condition=AND Boolean(A="Equal (Enum)"(A=@n4, B=FakeTranslucent), B=NOT Boolean(A=@n3.ReturnValue)))
   > Return Node!(
    execute
    Is Fake Transluscent Part:bool=Select(None=false, Standard=false, Transparent=true, Glitter=true, Opalescent=true, Metallic=false, Index=@n2.Material Type)
   )
   Return Node(execute=@n5.else, Is Fake Transluscent Part=false)
  ]}
 IsGlowCompatibleWithFakeTranslucent {nodes [
   IsGlowCompatibleWithFakeTranslucent!() -> (then, Part Info:Atom Model Part Instance Info Structure) {is=title K2Node_FunctionEntry}
   > Return Node!(execute, IsCompatible:bool=false)
  ]}
 IsGlowPart {nodes [
   n1 = IsGlowPart!() -> (then, PartInfo:Atom Model Part Instance Info Structure) {is=title K2Node_FunctionEntry}
   > Return Node!(
    execute
    ShouldHaveGlow:bool=Contains(SearchIn=Break Atom Model Part Instance Info!(AtomModelPartInstanceInfo=@n1) -> (Group:String), Substring=_Glow)
   )
  ]}
 IsPreventMergingMaterial {nodes [
   n1 = IsPreventMergingMaterial!() -> (then, Material:Material Interface Object Reference) {is=title K2Node_FunctionEntry}
   > n2 = GetMaterialFeatures(self=GetMaterialFactory(), Material=@n1)
   n3 = Break Atom Material Features!(AtomMaterialFeatures=@n2) -> (UDIMDecorationMap, NormalMap, AlphaMask, UVFeature)
   > Return Node!(
    execute
    Is Prevent Merging Material:bool=OR Boolean!(
     A=Is Not Empty!(TargetMap:Map of Integers to Texture Object References=@n3) -> (ReturnValue:bool) {is=title K2Node_CallFunction}
     B=IsValid(Object=@n3.NormalMap)
     C:bool=IsValid(Object=@n3.AlphaMask)
     D:bool="Equal (Enum)"(A:EAtomMaterialUVFeature Enum=@n3.UVFeature, B:EAtomMaterialUVFeature Enum=FakeTechnicHole)
     E:bool="Equal (Enum)"(A=@n3.UVFeature, B=FakeTechnicHoleParallax)
    ) -> (ReturnValue)
   )
  ]}
 Make Merged Geometries {nodes [
   n1 = Make Merged Geometries!() -> (then, MergedMeshPartInfos:Array of Atom Merged Mesh Part Info Structures) {is=title K2Node_FunctionEntry}
   > n2 = Parent: Make Merged Geometries!(execute, MergedMeshPartInfos:Array of Atom Merged Mesh Part Info Structures=@n1) -> (then, MergedGeometries:Map of Strings to Atom Merged Geometry Structures) {is=title K2Node_CallParentFunction}
   > n3 = Set UpdatedGeos(UpdatedGeos:Map of Strings to Atom Merged Geometry Structures=@n2) -> (Output_Get:Map of Strings to Atom Merged Geometry Structures)
   > n4 = Keys!(execute, TargetMap:Map of Strings to Atom Merged Geometry Structures=@n3) -> (then, Keys:Array of Strings)
   > n5 = For Each Loop(Array:Array of Strings=@n4) -> (Array Element:String)
   Return Node!(
    execute=@n5.Completed
    MergedGeometries:Map of Strings to Atom Merged Geometry Structures=Get UpdatedGeos() -> (UpdatedGeos:Map of Strings to Atom Merged Geometry Structures)
   )
   n6 = Find(TargetMap:Map of Strings to Atom Merged Geometry Structures=@n2, Key:"String (by ref)"=@n5) -> (Value:Atom Merged Geometry Structure)
   n7 = "Not Equal (Enum)"(A=GetMergedMeshType(" Mesh Name"=@n5) -> (Merged Mesh Type=NewEnumerator0), B=NewEnumerator2)
   n8 = Get Optimize Mesh Collision()
   > Branch(Condition=AND Boolean(A=@n6.ReturnValue, B=AND Boolean(A=@n8, B=@n7)))
   > n9 = Generate Simplify Collision For Mesh Containers!(execute, self, MergedGeo:"Atom Merged Geometry Structure (by ref)"=@n6) -> (then, StructOut:Atom Merged Geometry Structure) {is=title K2Node_CallFunction}
   > Add(TargetMap:Map of Strings to Atom Merged Geometry Structures=Get UpdatedGeos(), Key=@n5, Value:"Atom Merged Geometry Structure (by ref)"=@n9) {is=title K2Node_CallFunction}
  ]}
 OffsetMaterialSections {nodes [
   n1 = Inputs!() -> (Exec:Exec, Mesh:Dynamic Mesh Object Reference, Material Sections Offset:int)
   n2 = Local Integer!() -> (Variable:int) {is=title K2Node_TemporaryVariable}
   > Assign!(
    execute:Exec
    Variable:int=@n2
    Value:int=int * int(A=GetMaxMaterialID!(TargetMesh:Dynamic Mesh Object Reference=@n1) -> (bHasMaterialIDs:bool, ReturnValue:int=0).ReturnValue, B=-1)
   ) -> (then:Exec) {is=title K2Node_AssignmentStatement}
   > n3 = For Loop(FirstIndex=@n2)
   Outputs!(Out:Exec=@n3.Completed, Target Mesh:Dynamic Mesh Object Reference=@n1)
   n4 = int * int(A=@n3, B=-1)
   > RemapMaterialIDs(TargetMesh=@n1, FromMaterialID=@n4, ToMaterialID=int + int(A=@n4, B=@n1.Material Sections Offset))
  ]}
 OnPostMeshMerge {nodes [
   n1 = OnPostMeshMerge!() -> (
    then
    PrimitiveGeometryLODs:"Atom Primitive Geometry LODs Structure (by ref)"
    Materials:Array of Material Interface Object References
    Mesh Name:String
   ) {is=title K2Node_FunctionEntry}
   > n2 = Parent: OnPostMeshMerge!(
    execute
    PrimitiveGeometryLODs:"Atom Primitive Geometry LODs Structure (by ref)"=@n1
    Materials:Array of Material Interface Object References=@n1.Materials
    Mesh Name:String=@n1.Mesh Name
   ) -> (then, OutPrimitiveGeometryLODs:Atom Primitive Geometry LODs Structure, OutMaterials:Array of Material Interface Object References) {is=title K2Node_CallParentFunction}
   > n3 = RemoveUndersideOnLowerLODs!(
    execute
    self
    AtomPrimitiveGeometryLODs:"Atom Primitive Geometry LODs Structure (by ref)"=@n2
    MaterialList:Array of Material Interface Object References=@n2.OutMaterials
   ) -> (then, ReturnValue:Atom Primitive Geometry LODs Structure)
   > n4 = WeldLODs!(
    execute
    self
    AtomPrimitiveGeometryLODs:"Atom Primitive Geometry LODs Structure (by ref)"=@n3
    FromLodIndex:int=Length(TargetArray:Array of Atom Processor Geometry Style Settings Structures=Get LOD Geometry Style Settings())
   ) -> (then, ReturnValue:Atom Primitive Geometry LODs Structure)
   > Return Node!(
    execute
    OutPrimitiveGeometryLODs:Atom Primitive Geometry LODs Structure=@n4
    OutMaterials:Array of Material Interface Object References=@n2.OutMaterials
   )
  ]}
 OnProcessPartsCollection {nodes [
   n1 = OnProcessPartsCollection!() -> (then, Parts Collection:"Atom Model Parts Collection Structure (by ref)", Model:Atom Model Object Reference) {is=title K2Node_FunctionEntry}
   > n2 = Sequence()
   n3 = Parent: OnProcessPartsCollection!(execute=@n2.then_1, Parts Collection:"Atom Model Parts Collection Structure (by ref)"=@n1, Model:Atom Model Object Reference=@n1.Model) -> (then, Result:Atom Processor Result Structure) {is=title K2Node_CallParentFunction}
   > Return Node!(execute, Result:Atom Processor Result Structure=@n3)
   Clear(execute=@n2.then_0, TargetArray:Array of Vectors=Get GlowLocations())
   > n4 = For Each Loop(Array:Array of Atom Model Part Instance Info Structures=Break Atom Model Parts Collection!(AtomModelPartsCollection:"Atom Model Parts Collection Structure (by ref)"=Get Parts Collection() -> (Parts Collection:"Atom Model Parts Collection Structure (by ref)")) -> (Parts:Array of Atom Model Part Instance Info Structures) {is=title K2Node_BreakStruct}) -> (Array Element:Atom Model Part Instance Info Structure)
   > Branch(Condition=IsGlowPart(PartInfo=@n4) {is=title K2Node_CallFunction})
   > RegisterAndMergeGlowTransform!(execute, self, PartInfo:Atom Model Part Instance Info Structure=@n4) -> (then, GlowMerged:bool)
  ]}
 PreProcessorPipeline {nodes [
   n1 = PreProcessorPipeline!() -> (then, AtomModel:Atom Model Object Reference, PartsCollection:"Atom Model Parts Collection Structure (by ref)") {is=title K2Node_FunctionEntry}
   > Parent: PreProcessorPipeline!(execute, AtomModel:Atom Model Object Reference=@n1, PartsCollection:"Atom Model Parts Collection Structure (by ref)"=@n1.PartsCollection) -> (then) {is=title K2Node_CallParentFunction}
   > n2 = Branch(Condition=Is Empty(TargetArray:Array of Atom Processor Geometry Style Settings Structures=Get LOD Geometry Style Settings()))
   n3 = Make Array("[0]"=Game, "[1]"=Realtime)
   > Set LOD Geometry Style Settings(LOD Geometry Style Settings:Array of Atom Processor Geometry Style Settings Structures=Make Array!("[0]":Atom Processor Geometry Style Settings Structure=Make Atom Processor Geometry Style Settings(
      GeometryStylePreferences_4_#17=@n3
      PreferredGeometryLOD_30_#18=0
      ShouldBevel_8_#19=False
      BevelWidth_11_#20=0.0125
      BevelAngleThreshold_13_#21=60
      UseImprovedBevelAlgorithm_37_#22=False
      OnlyBevelBlock-LikePrimitives_39_#23=True
      NumberofSubdivisionsToApply_16_#24=0
      FakeBevel_23_#25=False
      FakeUnderside_25_#15=False
      AnalyticFakeUnderside_32_#26=False
      ForceFlatCap_35_#16=False
      FakeTechnicHole_41_#1=False
      AnalyticCommonParts_27_#2=False
     )) -> (Array:Array of Atom Processor Geometry Style Settings Structures)) -> (Output_Get:Array of Atom Processor Geometry Style Settings Structures)
   > n4 = ReportProcessorWarning!(
    execute
    Message:"Text (by ref)"=MakeLiteralText(Value="LOCTEXT(\"\\\"LOD Geometry Style Settings\\\" cannot be empty. A default entry was added to the array to make sure the merged mesh has source geometry.\")")
    Asset:Object Reference=Get AtomModel() -> (AtomModel:Atom Model Object Reference)
   ) -> (then)
   Return Node!(execute=[@n4.then, @n2.else])
  ]}
 Process Fake Bevels {nodes [
   n1 = Process Fake Bevels!() -> (
    then
    Part Info:Atom Model Part Instance Info Structure
    Applied Optimizations:Atom Processor Applied Optimizations Structure
    Geometry Container:Atom Primitive Geometry Container Object Reference
    Mesh:Dynamic Mesh Object Reference
    Fake Bevels Channel:int
   ) {is=title K2Node_FunctionEntry}
   > Set DynamicMesh(DynamicMesh:Dynamic Mesh Object Reference=@n1.Mesh) -> (Output_Get:Dynamic Mesh Object Reference)
   > SetMeshUVsFromBoxProjection!(
    execute
    TargetMesh:Dynamic Mesh Object Reference=Get DynamicMesh()
    UVSetIndex:int=Get Fake Bevels Channel() -> (Fake Bevels Channel=0)
    BoxTransform:Transform
    Selection:Mesh Selection Structure
    MinIslandTriCount:int=2
   ) -> (then, ReturnValue:Dynamic Mesh Object Reference)
   > TranslateMeshUVs(TargetMesh=Get DynamicMesh(), UVSetIndex=Get Fake Bevels Channel() -> (Fake Bevels Channel=0), Translation=(X=0,Y=2))
   > SetFakeBevelsUVs!(
    execute
    self
    DynamicMesh:Dynamic Mesh Object Reference=Get DynamicMesh()
    Fake Bevels Channel:int=Get Fake Bevels Channel() -> (Fake Bevels Channel=0)
   ) -> (then)
   > Return Node!(execute, UpdatedMesh:Dynamic Mesh Object Reference=Get DynamicMesh())
  ]}
 Process Fake Technic Connector {nodes [
   n1 = Process Fake Technic Connector!() -> (
    then
    Part Info:Atom Model Part Instance Info Structure
    Applied Optimizations:Atom Processor Applied Optimizations Structure
    Geometry Container:Atom Primitive Geometry Container Object Reference
    Shell Mesh:Dynamic Mesh Object Reference
    Technic Connectors Meshes:Array of Dynamic Mesh Object References
   ) {is=title K2Node_FunctionEntry}
   Return Node!(
    execute=[@n1.then, @n2.Completed]
    Out Shell Mesh:Dynamic Mesh Object Reference=@n1.Shell Mesh
    Out Technic Connectors Meshes:Array of Dynamic Mesh Object References=@n1.Technic Connectors Meshes
   )
   n2 = For Loop(LastIndex=int - int(A=Length(TargetArray:Array of Dynamic Mesh Object References=@n1.Technic Connectors Meshes)))
   n3 = GetConnectionFieldForTechnicConnectorGeometry!(
    self:Atom Primitive Object Reference=Break Atom Model Part Instance Info!(AtomModelPartInstanceInfo=@n1) -> (Primitive:Atom Primitive Object Reference)
    Container:Atom Primitive Geometry Container Object Reference=Get Geometry Container() -> (Geometry Container:Atom Primitive Geometry Container Object Reference)
    ConnectorGeometryIndex:int=@n2
   ) -> (ConnectivityFieldReference:Typed Connectivity Field Reference Structure, ReturnValue:bool)
   > Branch(Condition=@n3.ReturnValue)
   > n4 = GetTechnicPinConnectivityStates(
    self=GetPartsCollectionConnectivity()
    TechnicPinPartConnectivityRef=Make Atom Part Connectivity Field Reference(PartInfo=Get Part Info(), ConnectivityField=@n3)
   ) -> (ReturnValue=NoConnection)
   > n5 = For Each Loop(Array:Array of EFieldConnectResult Enums=@n4.ReturnValue) -> (Array Element:EFieldConnectResult Enum)
   > Branch(Condition=Select!(
     NoConnection:bool=false
     Rejection:bool=false
     FixedConnection:bool=true
     FreeConnection:bool=true
     HingeConnection:bool=true
     CardanConnection:bool=false
     BallConnection:bool=false
     PrismaticConnection:bool=false
     CylindricalConnection:bool=true
     PrismaticAPerpendicularHingeConnection:bool=false
     PrismaticBPerpendicularHingeConnection:bool=false
     ConnectResultMax:bool=false
     FirstConnection:bool=false
     LastConnection:bool=false
     Index:EFieldConnectResult Enum=@n5
    ) -> (ReturnValue))
   > Reroute Node()
   PrintString(InString=Hello)
   PrintString(InString=Process Fake Technic Connector: Could not access technic pin connector.)
  ]}
 Process Fake Technic Hole {nodes [
   n1 = Process Fake Technic Hole!() -> (
    then
    Part Info:Atom Model Part Instance Info Structure
    Applied Optimizations:Atom Processor Applied Optimizations Structure
    Geometry Container:Atom Primitive Geometry Container Object Reference
    Technic Caps Meshes:Array of Dynamic Mesh Object References
   ) {is=title K2Node_FunctionEntry}
   n2 = PrintString(InString="To String (Integer)"(InInt=Length(TargetArray=@n1.Technic Caps Meshes)))
   Return Node!(execute=[@n2.then, @n3.Completed], Out Technic Caps:Array of Dynamic Mesh Object References=@n1.Technic Caps Meshes)
   n3 = For Each Loop(Exec=@n1.then, Array:Array of Dynamic Mesh Object References=@n1.Technic Caps Meshes) -> (Array Element:Dynamic Mesh Object Reference)
   n4 = GetConnectionFieldForTechnicCapGeometry!(
    self:Atom Primitive Object Reference=Break Atom Model Part Instance Info(AtomModelPartInstanceInfo=@n1)
    Container:Atom Primitive Geometry Container Object Reference=@n1.Geometry Container
    CapIndex:int=@n3.Array Index
   ) -> (ConnectivityFieldReference:Typed Connectivity Field Reference Structure, ReturnValue:bool)
   Branch(Condition=@n4.ReturnValue)
   > n5 = GetTechnicHoleConnectivityState!(
    execute
    self:Atom Model Parts Collection Connectivity Object Reference=GetPartsCollectionConnectivity()
    TechnicHolePartConnectivityRef:"Atom Part Connectivity Field Reference Structure (by ref)"=Make Atom Part Connectivity Field Reference(PartInfo=Get Part Info(), ConnectivityField=@n4)
   ) -> (then, ConnectedPartConnectivityRef:Atom Part Connectivity Field Reference Structure, ReturnValue:EFieldConnectResult Enum=NoConnection)
   PrintString(execute=[@n5.then, @n6.then], InString=Enum to String!(Enumerator:EFieldConnectResult Enum=@n5.ReturnValue) -> (ReturnValue:String))
   n6 = Branch(Condition=Select(
     NoConnection=false
     Rejection=false
     FixedConnection=true
     FreeConnection=false
     HingeConnection=false
     CardanConnection=false
     BallConnection=false
     PrismaticConnection=false
     CylindricalConnection=true
     PrismaticAPerpendicularHingeConnection=false
     PrismaticBPerpendicularHingeConnection=false
     ConnectResultMax=false
     FirstConnection=false
     LastConnection=false
     Index=@n5.ReturnValue
    ))
   RemapMaterialIDs(TargetMesh=@n3, FromMaterialID=0, ToMaterialID=@n3.Array Index)
   n7 = Break Atom Part Connectivity Field Reference(AtomPartConnectivityFieldReference=@n5)
   n8 = Break Atom Model Part Instance Info!(AtomModelPartInstanceInfo=@n7) -> (
    PartGuid:Atom Model Part Guid Structure
    PartId:int
    Primitive
    Transforms:Array of Transforms
    Decorations:Array of Atom Model Part Decoration Info Structures
    Colors:Array of Atom Model Part Color Info Structures
    SelectionSets:Array of Strings
    Group:String
    ParentGroup:String
    CommonParts:Array of Atom Common Part And Transform Structures
    bIsUndersideVisible:bool=~
   )
   n9 = GetPartColorInfo(Part Info=@n7, Color Change Index=0) -> (Color Id=0, Material Type=None)
   PrintString(InString=Append!(A="Connected brick: ", B="To String (Integer)"(InInt=@n8.PartId), C=" with color: ", D:String="To String (Integer)"(InInt=@n9)) -> (ReturnValue))
   Reroute Node(InputPin:Dynamic Mesh Object Reference=@n3) -> (OutputPin:Dynamic Mesh Object Reference)
  ]}
 Process Fake Translucent {nodes [
   n1 = Process Fake Translucent!() -> (
    then
    Part Info:Atom Model Part Instance Info Structure
    Applied Optimizations:Atom Processor Applied Optimizations Structure
    Geometry Container:Atom Primitive Geometry Container Object Reference
    DynamicMesh:Dynamic Mesh Object Reference
    Size Channel:int
    DepthChannel:int=1
   ) {is=title K2Node_FunctionEntry}
   n2 = int + int(A="Max (Integer)"(A=@n1.Size Channel, B=@n1.DepthChannel))
   > n3 = Branch(Condition="integer < integer"(
     A=Get Num UV Channels!(TargetMesh:Dynamic Mesh Object Reference=@n1.DynamicMesh) -> (ReturnValue:int=0) {is=title K2Node_CallFunction}
     B=@n2
    ))
   > n4 = Set Num UV Channels!(execute, TargetMesh:Dynamic Mesh Object Reference=@n1.DynamicMesh, NumUVSets:int=@n2) -> (then, ReturnValue:Dynamic Mesh Object Reference) {is=title K2Node_CallFunction}
   Set BrickSize(execute=[@n4.then, @n3.else], BrickSize:Vector=GetBrickSize(DynamicMesh=@n1.DynamicMesh) {is=title K2Node_CallFunction}) -> (Output_Get:Vector)
   > Set AtomBrickSize(AtomBrickSize:Vector=GetAtomBrickSize!(DynamicMesh:Dynamic Mesh Object Reference=@n1.DynamicMesh) -> (BrickSize:Vector) {is=title K2Node_CallFunction}) -> (Output_Get:Vector)
   > n5 = Sequence!(execute) -> (then_0, then_1, then_2:Exec, then_3:Exec)
   > n6 = For Each Loop(Array:Array of Vectors=Get NormalArray() -> (NormalArray:Array of Vectors)) -> (Array Element:Vector)
   > n7 = SelectMeshElementsByNormalAngle!(
    execute
    TargetMesh:Dynamic Mesh Object Reference=Get DynamicMesh()
    Normal:Vector=@n6
    MaxAngleDeg:double=88
    SelectionType:EGeometryScriptMeshSelectionType Enum=Triangles
    bInvert:bool=false
    MinNumTrianglePoints:int=0
   ) -> (then, Selection:Mesh Selection Structure, ReturnValue:Dynamic Mesh Object Reference)
   n8 = BreakVector(InVec=Get BrickSize())
   n9 = Get RotationArray() -> (RotationArray:Array of Rotators)
   n10 = Get AtomBrickSize() -> (AtomBrickSize:Vector)
   n11 = BreakVector(InVec=vector * vector(A=vector - vector(A=@n10, B:double=1), B=-0.4) -> (ReturnValue=~))
   n12 = float / float(A=@n8.Y, B=2)
   n13 = float / float(A=@n8, B=2)
   > n14 = SetMeshUVsFromBoxProjection(
    TargetMesh=@n7.ReturnValue
    UVSetIndex=Get Size Channel() -> (Size Channel:int=0)
    BoxTransform=MakeTransform!(
     Location_X:double=float - float(A=@n11, B=@n13)
     Location_Y:double=float - float(A=@n11.Y, B=@n12)
     Location_Z:double=@n8.Z
     Rotation="Get (a ref)"(Array:Array of Rotators=@n9, Dimension 1=@n6.Array Index) -> (Output:"Rotator (by ref)")
     Scale_X:double=@n8
     Scale_Y:double=@n8.Y
     Scale_Z:double=@n8.Z
    ) -> (ReturnValue)
    Selection=@n7
    MinIslandTriCount=1
   )
   > SplitMeshBowties!(execute, TargetMesh:Dynamic Mesh Object Reference=@n14, bMeshBowties:bool=true, bAttributeBowties:bool=true) -> (then, ReturnValue:Dynamic Mesh Object Reference)
   n15 = Convert UV Islands To PolyGroups!(
    execute=@n5.then_1
    TargetMesh:Dynamic Mesh Object Reference=Get DynamicMesh()
    GroupLayer:Geometry Script Group Layer Structure
    UVLayer:int=Get Size Channel() -> (Size Channel=0)
   ) -> (then, ReturnValue:Dynamic Mesh Object Reference)
   > Get PolyGroup IDs In Mesh!(
    execute
    TargetMesh:Dynamic Mesh Object Reference=@n15
    GroupLayer:Geometry Script Group Layer Structure
    PolygroupIDsOut:"Index List Structure (by ref)"=Get Polygroups() -> (Polygroups:Index List Structure)
   ) -> (then, ReturnValue:Dynamic Mesh Object Reference) {is=title K2Node_CallFunction}
   > n16 = ConvertIndexListToArray(IndexList=Get Polygroups())
   > n17 = For Each Loop(Array:Array of Integers=@n16) -> (Array Element:int)
   > FT_EncodeSize!(
    execute
    self
    DynamicMesh:Dynamic Mesh Object Reference=Get DynamicMesh()
    PolygroupID:int=@n17
    UV Channel:int=Get Size Channel() -> (Size Channel=0)
   ) -> (then, Target Mesh:Dynamic Mesh Object Reference) {is=title K2Node_CallFunction}
   n18 = Get DepthChannel() -> (DepthChannel:int=0)
   n19 = Copy UV Channel!(
    execute=@n5.then_2
    TargetMesh:Dynamic Mesh Object Reference=Get DynamicMesh()
    FromUVSet:int=Get Size Channel() -> (Size Channel=0)
    ToUVSet:int=@n18
   ) -> (then, ReturnValue:Dynamic Mesh Object Reference)
   > n20 = Convert UV Islands To PolyGroups(TargetMesh=@n19, UVLayer=@n18)
   > Get PolyGroup IDs In Mesh(TargetMesh=@n20, PolygroupIDsOut=Get Polygroups())
   > n21 = ConvertIndexListToArray(IndexList=Get Polygroups())
   > n22 = For Each Loop(Array=@n21)
   > FT_EncodeDepth!(
    execute
    self
    DynamicMesh:Dynamic Mesh Object Reference=Get DynamicMesh()
    PolygroupID:int=@n22
    UV Channel:int=Get DepthChannel() -> (DepthChannel=0)
   ) -> (then, Target Mesh:Dynamic Mesh Object Reference) {is=title K2Node_CallFunction}
   Return Node!(execute=@n5.then_3, UpdatedMesh:Dynamic Mesh Object Reference=Get DynamicMesh())
  ]}
 Process Fake Underside {nodes [
   Process Fake Underside!() -> (
    then
    Part Info:Atom Model Part Instance Info Structure
    Applied Optimizations:Atom Processor Applied Optimizations Structure
    Geometry Container:Atom Primitive Geometry Container Object Reference
    Cap Meshes:Array of Dynamic Mesh Object References
    Underside Channel:int
   ) {is=title K2Node_FunctionEntry}
   > n1 = Sequence()
   n2 = For Each Loop(Exec=@n1.then_2, Array:Array of Dynamic Mesh Object References=Get Cap Meshes() -> (Cap Meshes:Array of Dynamic Mesh Object References)) -> (Array Element:Dynamic Mesh Object Reference)
   Return Node!(execute=@n2.Completed, Updated Meshes:Array of Dynamic Mesh Object References=Get Cap Meshes())
   > n3 = Is Valid(InputObject=@n2)
   PrintString(execute=@n3.Is Not Valid, InString=Append!(A="Invalid cap mesh at index: ", B="To String (Integer)"(InInt=@n2.Array Index)) -> (ReturnValue))
   Set CurrentCapDynamicMesh(execute=@n3.Is Valid, CurrentCapDynamicMesh:Dynamic Mesh Object Reference=@n2) -> (Output_Get:Dynamic Mesh Object Reference)
   > n4 = Sequence()
   n5 = Get Applied Optimizations()
   ClearMaterialIDs(
    execute=@n4.then_3
    TargetMesh=Get CurrentCapDynamicMesh() -> (CurrentCapDynamicMesh:Dynamic Mesh Object Reference)
    ClearValue=SelectInt(
     A=1
     B=2
     bPickA=Break Atom Processor Applied Optimizations!(AtomProcessorAppliedOptimizations=@n5) -> (bAnalyticFakeUnderside_10_#5:bool=~)
    )
   )
   n6 = Branch(execute=@n4.then_2, Condition=Get IsTShape() -> (IsTShape:bool))
   > n7 = SetMeshUVsFromBoxProjection!(
    execute
    TargetMesh=Get CurrentCapDynamicMesh()
    UVSetIndex=Get Underside Channel() -> (Underside Channel:int=0)
    BoxTransform_Location:Vector="0,-26.8,-13.4"
    BoxTransform_Rotation:Rotator="0, 0, 0"
    BoxTransform_Scale_X:double=0.8
    BoxTransform_Scale_Y:double=12.8
    BoxTransform_Scale_Z:double=12.8
    Selection
    MinIslandTriCount=2
   ) -> (then, ReturnValue)
   > Set CurrentCapDynamicMesh(CurrentCapDynamicMesh=@n7)
   n8 = AND Boolean(A=Get Deep() -> (Deep:bool), B=Get IsX1Width() -> (IsX1Width:bool))
   n9 = Get!() -> (tile_X:double, tile_Y:double)
   n10 = vector + vector!(
    A_X:double=float * float(A=@n9, B=0.4)
    A_Y:double=float * float(A=@n9.tile_Y, B=0.4)
    A_Z:double=0
    B_X:double=float * float(A=@n9, B=12.8)
    B_Y:double=float * float(A=float - float(A=@n9.tile_Y, B=1), B=12.8)
    B_Z:double=0
   ) -> (ReturnValue)
   n11 = Select!(
    Option 0:Vector=@n10
    Option 1:Vector=vector + vector!(A:Vector=Select(Option 0="0,-12.8,0", Option 1="-12.8,0,0", Index=false), B:Vector=@n10) -> (ReturnValue)
    Index:bool=Get Is1x1() -> (Is1x1:bool)
   ) -> (ReturnValue:Vector)
   n12 = vector + vector(A=vector * vector(A=@n11, B:Vector="1,-1,0") -> (ReturnValue=~), B="0,-25.6,0")
   n13 = SetMeshUVsFromBoxProjection(
    execute=@n6.else
    TargetMesh=Get CurrentCapDynamicMesh()
    UVSetIndex=Get Underside Channel() -> (Underside Channel=0)
    BoxTransform_Location=vector + vector(A=Get ProjectionLocation() -> (ProjectionLocation:Vector), B=Select(Option 0=@n11, Option 1=@n12, Index=@n8))
    BoxTransform_Rotation="0, 0, 0"
    BoxTransform_Scale_X=12.8
    BoxTransform_Scale_Y=12.8
    BoxTransform_Scale_Z=0.8
    MinIslandTriCount=2
   )
   > n14 = Set CurrentCapDynamicMesh(CurrentCapDynamicMesh=@n13)
   > Branch(Condition=Get IsCorner() -> (IsCorner:bool))
   > ApplyMeshMirror!(
    execute
    TargetMesh:Dynamic Mesh Object Reference=@n14
    MirrorFrame_Location:Vector=Get CornerCenter() -> (CornerCenter:Vector)
    MirrorFrame_Rotation:Rotator="0,45,90"
    MirrorFrame_Scale:Vector="1,1,1"
    Options_bApplyPlaneCut:bool=True
    Options_bFlipCutSide:bool=true
    Options_bWeldAlongPlane:bool=True
   ) -> (then, ReturnValue:Dynamic Mesh Object Reference)
   n15 = Get Center Dimension!(TargetMesh:Dynamic Mesh Object Reference=Get CurrentCapDynamicMesh()) -> (ReturnValue:Vector=~, Center:Vector=~, IsDeep:bool=~) {is=title K2Node_MacroInstance}
   n16 = BreakVector(InVec=@n15)
   Set!(execute=@n4.then_1, LegoDimension_X:double=@n16, LegoDimension_Y:double=@n16.Y) -> (then, Output_Get:Vector 2D Structure)
   > Set ProjectionLocation(ProjectionLocation:Vector=@n15.Center) -> (Output_Get:Vector)
   > Set Deep(Deep:bool=false) -> (Output_Get:bool)
   n17 = Get LegoDimension() -> (LegoDimension:Vector 2D Structure)
   n18 = "To Rounded (Vector2D)"!(A:Vector 2D Structure=vector2d * vector2d(A=@n17, B=(X=1.25,Y=1.25))) -> (ReturnValue:Vector 2D Structure)
   > Set tile(tile:Vector 2D Structure=@n18) -> (Output_Get:Vector 2D Structure)
   > Set Is1x1(Is1x1:bool="Equal (Vector2D)"!(A:Vector 2D Structure=@n18, B:Vector 2D Structure=(X=1,Y=1), ErrorTolerance:float=0.1) -> (ReturnValue:bool=~) {is=title K2Node_PromotableOperator}) -> (Output_Get:bool)
   n19 = BreakVector2D(InVec=@n18)
   n20 = Get Is1x1()
   > n21 = Set IsX1Width(IsX1Width:bool=AND Boolean(A="Nearly Equal (Float)"(A=@n19, B=1, ErrorTolerance=0.1), B=NOT Boolean(A=@n20))) -> (Output_Get:bool)
   Branch(execute=[@n43.then, @n4.then_0], Condition=Get IsCorner())
   > n22 = Branch(Condition=Get IsUShape() -> (IsUShape:bool))
   n23 = Get CurrentCapDynamicMesh()
   n24 = GetMeshBoundingBox(TargetMesh=@n23) -> (ReturnValue_Min=~, ReturnValue_Max=~, ReturnValue_IsValid=~)
   n25 = Get Center Dimension(TargetMesh=@n23) -> (ReturnValue=~, Center=~, IsDeep=~)
   n26 = BreakVector(InVec=@n25)
   n27 = Select(Option 0=MakeVector(X=@n26, Y=@n26, Z=@n26.Z), Option 1=MakeVector(X=@n26.Y, Y=@n26.Y, Z=@n26.Z), Index="float >= float"(A=@n26, B=@n26.Y))
   > Set CornerCenter(CornerCenter:Vector=vector + vector(A=@n24, B=vector / vector(A=@n27, B=2))) -> (Output_Get:Vector)
   > n28 = ApplyMeshPlaneCut!(
    execute
    TargetMesh:Dynamic Mesh Object Reference=Get CurrentCapDynamicMesh()
    CutFrame_Location:Vector=vector + vector(A=vector * vector(A=@n27, B:float=0.99) -> (ReturnValue=~), B=@n24)
    CutFrame_Rotation:Rotator="0,0,90"
    CutFrame_Scale:Vector="1,1,1"
    Options_bFillHoles:bool=false
    Options_HoleFillMaterialID:int=-1
    Options_bFillSpans:bool=false
    Options_bFlipCutSide:bool=False
    Options_UVWorldDimension:float=1
   ) -> (then, ReturnValue:Dynamic Mesh Object Reference)
   ApplyMeshPlaneCut(
    execute=[@n28.then, @n29.then]
    TargetMesh=Get CurrentCapDynamicMesh()
    CutFrame_Location=Get CornerCenter()
    CutFrame_Rotation="0,45,90"
    CutFrame_Scale="1,1,1"
    Options_bFillHoles=false
    Options_HoleFillMaterialID=-1
    Options_bFillSpans=false
    Options_bFlipCutSide=true
    Options_UVWorldDimension=1
   )
   n29 = Set CornerCenter(execute=@n22.else, CornerCenter=@n25.Center)
   Set UShapePartIDsRemap(
    execute=@n1.then_0
    UShapePartIDsRemap:Map of Integers to Integers=Make Map!(Key 0:int=73831, Value 0:int=2420) -> (Map:Map of Integers to Integers) {is=title K2Node_MakeMap}
   ) -> (Output_Get:Map of Integers to Integers)
   n30 = Break Atom Model Part Instance Info!(AtomModelPartInstanceInfo=Get Part Info()) -> (PartId)
   > Set IsCorner(IsCorner:bool=Contains Item(TargetArray=Get CornerPartsIDs() -> (CornerPartsIDs:Array of Integers), ItemToFind=@n30)) -> (Output_Get:bool)
   > Set IsUShape(IsUShape:bool=Contains!(
     TargetMap:Map of Integers to Integers=Get UShapePartIDsRemap() -> (UShapePartIDsRemap:Map of Integers to Integers)
     Key:"Integer (by ref)"=@n30
    ) -> (ReturnValue)) -> (Output_Get:bool)
   > Set IsTShape(IsTShape:bool=Contains Item(TargetArray=Get TShapePartsIDs() -> (TShapePartsIDs:Array of Integers), ItemToFind=@n30)) -> (Output_Get:bool)
   Branch(Condition=Get IsCorner())
   > n31 = Set CornerCenter(CornerCenter=Get Center Dimension(TargetMesh=Get CurrentCapDynamicMesh()) -> (ReturnValue=~, Center=~, IsDeep=~).Center)
   > ApplyMeshPlaneCut(
    TargetMesh=Get CurrentCapDynamicMesh()
    CutFrame_Location=@n31
    CutFrame_Rotation="0,45,90"
    CutFrame_Scale="1,1,1"
    Options_bFillHoles=false
    Options_HoleFillMaterialID=-1
    Options_bFillSpans=false
    Options_bFlipCutSide=true
    Options_UVWorldDimension=1
   )
   Branch(Condition=Get IsUShape())
   n32 = Get CurrentCapDynamicMesh()
   n33 = vector - vector(
    A=vector * vector(A=Get CornerCenter(), B=2.3333) -> (ReturnValue=~)
    B:Vector=GetMeshBoundingBox(TargetMesh=@n32) -> (ReturnValue_Min=~, ReturnValue_Max=~, ReturnValue_IsValid=~)
   )
   > ApplyMeshMirror!(
    execute
    TargetMesh=@n32
    MirrorFrame_Location=@n33
    MirrorFrame_Rotation="0,0,90"
    MirrorFrame_Scale="1,1,1"
    Options:Geometry Script Mesh Mirror Options Structure
   ) -> (then, ReturnValue)
   Reroute Node(InputPin:Exec) -> (OutputPin:Exec)
   > AppendMesh(
    TargetMesh=Get CapDynamicMesh() -> (CapDynamicMesh:Dynamic Mesh Object Reference)
    AppendMesh=Get CurrentCapDynamicMesh()
    AppendTransform=MakeTransform!(Location:Vector, Rotation="0, 0, 0", Scale:Vector) -> (ReturnValue)
    bDeferChangeNotifications=false
    AppendOptions=(CombineMode=EnableAllMatching)
   )
   Reroute Node()
   n34 = Get!() -> (AppendTransform_Location:Vector, AppendTransform_Rotation:Rotator, AppendTransform_Scale:Vector)
   > n35 = AppendMesh!(
    execute
    TargetMesh=Get PrimitiveDynamicMesh() -> (PrimitiveDynamicMesh:Dynamic Mesh Object Reference)
    AppendMesh=Get CapDynamicMesh()
    AppendTransform_Location:Vector=@n34
    AppendTransform_Rotation:Rotator=@n34.AppendTransform_Rotation
    AppendTransform_Scale:Vector=@n34.AppendTransform_Scale
    bDeferChangeNotifications=false
    AppendOptions=(CombineMode=EnableAllMatching)
   ) -> (then, ReturnValue)
   > WeldMeshEdges(TargetMesh=@n35, WeldOptions_Tolerance=0, WeldOptions_bOnlyUniquePairs=True)
   > n36 = Branch(Condition=Get IsUShape())
   > n37 = ApplyMeshMirror(
    TargetMesh=Get PrimitiveDynamicMesh()
    MirrorFrame_Location=vector + vector(A=Get CornerCenter(), B="0,0.4,0")
    MirrorFrame_Rotation="0,0,90"
    MirrorFrame_Scale="1,1,1"
   )
   > n38 = TransformMesh!(
    execute
    TargetMesh:Dynamic Mesh Object Reference=@n37
    Transform_Location:Vector="-0.8,-0.8,0"
    Transform_Rotation:Rotator="0, 0, 0"
    Transform_Scale:Vector="1,1,1"
    bFixOrientationForNegativeScale:bool=true
   ) -> (then, ReturnValue:Dynamic Mesh Object Reference)
   ReturnAllMeshes(execute=[@n38.then, @n36.else], self=Get DynamicMeshPool())
   > n39 = ConvertDynamicMeshToAtomPrimitiveGeometry(DynamicMesh=Get PrimitiveDynamicMesh())
   > Return Node(Updated Meshes=@n39)
   Reroute Node()
   n40 = Break Atom Primitive Geometry and Transform!(AtomPrimitiveGeometryAndTransform=Reroute Node(InputPin:"Atom Primitive Geometry And Transform Structure (by ref)") -> (OutputPin:"Atom Primitive Geometry And Transform Structure (by ref)")) -> (AtomPrimitiveGeometry, Transform_Location:Vector, Transform_Rotation:Rotator, Transform_Scale:Vector)
   > Set!(
    execute
    AppendTransform_Location:Vector=@n40.Transform_Location
    AppendTransform_Rotation:Rotator=@n40.Transform_Rotation
    AppendTransform_Scale:Vector=@n40.Transform_Scale
   ) -> (then, Output_Get:Transform)
   > n41 = SetMaterialName(self=@n40, Name=Underside, PolygonGroupIndex=0)
   > n42 = ConvertAtomPrimitiveGeometryToDynamicMesh(AtomPrimitiveGeometry=@n41)
   > n43 = Set CurrentCapDynamicMesh(CurrentCapDynamicMesh=@n42)
   ApplyMeshMirror!(
    execute
    TargetMesh=@n32
    MirrorFrame_Location=@n33
    MirrorFrame_Rotation="0,0,90"
    MirrorFrame_Scale="1,1,1"
    Options_bApplyPlaneCut:bool=True
    Options_bFlipCutSide:bool=true
    Options_bWeldAlongPlane:bool=True
   ) -> (then, ReturnValue)
   PrintString(InString=Append!(A="1x1: ", B="To String (Boolean)"(InBool=@n20), C:String=" Is 1x: ", D:String="To String (Boolean)"(InBool=@n21)) -> (ReturnValue))
   GetDefaultPrimitiveScale()
   vector2d / vector2d!(A:Vector 2D Structure=@n17, B:Vector 2D Structure=(X=1.6,Y=1.6)) -> (ReturnValue)
   Get AppendTransform() -> (AppendTransform:Transform)
   Get Deep()
  ]}
 Process Glow {nodes [
   n1 = Process Glow!() -> (
    then
    Part Info:Atom Model Part Instance Info Structure
    Applied Optimizations:Atom Processor Applied Optimizations Structure
    Geometry Container:Atom Primitive Geometry Container Object Reference
    Dynamic Mesh:Dynamic Mesh Object Reference
    UV Channel XY:int=1
    UV Channel Z:int=2
   ) {is=title K2Node_FunctionEntry}
   n2 = Break Atom Model Part Instance Info!(AtomModelPartInstanceInfo=@n1) -> (Primitive:Atom Primitive Object Reference, Transforms:Array of Transforms)
   > Set Primitive(Primitive:Atom Primitive Object Reference=@n2) -> (Output_Get:Atom Primitive Object Reference)
   > Set Primitive Transform(Primitive Transform:Transform="Get (a ref)"(Array:Array of Transforms=@n2.Transforms, Dimension 1=0) -> (Output:"Transform (by ref)")) -> (Output_Get:Transform)
   > n3 = Sequence()
   Return Node!(execute=@n3.then_3, Updated Mesh:Dynamic Mesh Object Reference=Get Dynamic Mesh() -> (Dynamic Mesh:Dynamic Mesh Object Reference))
   n4 = ConvertIndexListToArray(
    execute=@n3.then_2
    IndexList=GetTrianglesByMaterialID!(TargetMesh:Dynamic Mesh Object Reference=Get Dynamic Mesh(), MaterialID:int=0) -> (TriangleIDList:Index List Structure, ReturnValue:Dynamic Mesh Object Reference)
   )
   > n5 = For Each Loop(Array:Array of Integers=@n4) -> (Array Element:int)
   > n6 = SetMeshTriangleUVs!(
    execute
    TargetMesh:Dynamic Mesh Object Reference=Get Dynamic Mesh()
    UVSetIndex:int=Get UV Channel XY() -> (UV Channel XY:int=0)
    TriangleID:int=@n5
    UVs:Geometry Script UVTriangle Structure=Get GlowPositionUVsChannel1() -> (GlowPositionUVsChannel1:Geometry Script UVTriangle Structure)
    bDeferChangeNotifications:bool=false
   ) -> (then, bIsValidTriangle:bool, ReturnValue:Dynamic Mesh Object Reference)
   > SetMeshTriangleUVs(
    TargetMesh=@n6.ReturnValue
    UVSetIndex=Get UV Channel Z() -> (UV Channel Z:int=0)
    TriangleID=@n5
    UVs=Get GlowPositionUVsChannel2() -> (GlowPositionUVsChannel2:Geometry Script UVTriangle Structure)
    bDeferChangeNotifications=false
   )
   n7 = Sequence!(execute=@n3.then_0) -> (then_0, then_1)
   n8 = BreakVector(InVec=Get Glow Position() -> (Glow Position:Vector))
   n9 = MakeVector2D(X=@n8, Y=@n8.Y)
   Set GlowPositionUVsChannel1(
    execute=@n7.then_1
    GlowPositionUVsChannel1:Geometry Script UVTriangle Structure=Make Geometry Script UVTriangle!(UV0:Vector 2D Structure=@n9, UV1:Vector 2D Structure=@n9, UV2:Vector 2D Structure=@n9) -> (GeometryScriptUVTriangle:Geometry Script UVTriangle Structure) {is=title K2Node_MakeStruct}
   ) -> (Output_Get:Geometry Script UVTriangle Structure)
   n10 = MakeVector2D(X=@n8.Z, Y=0)
   > Set GlowPositionUVsChannel2(GlowPositionUVsChannel2:Geometry Script UVTriangle Structure=Make Geometry Script UVTriangle(UV0=@n10, UV1=@n10, UV2=@n10)) -> (Output_Get:Geometry Script UVTriangle Structure)
   n11 = Set Glow Position(
    execute=@n7.then_0
    Glow Position:Vector=Compute Glow Position!(
     PrimitiveTransform:Transform=Get Primitive Transform() -> (Primitive Transform:Transform)
     Primitive:Atom Primitive Object Reference=Get Primitive() -> (Primitive:Atom Primitive Object Reference)
    ) -> (ReturnValue:Vector) {is=title K2Node_CallFunction}
   ) -> (Output_Get:Vector)
   n12 = Is Glow Position Mergeable!(GlowPosition:Vector=@n11) -> (GlowIndex:int=0, IsGlowPositionMergeable:bool) {is=title K2Node_CallFunction}
   > Set MergedGlowIndex(MergedGlowIndex:int=@n12) -> (Output_Get:int=0)
   > Branch(Condition=@n12.IsGlowPositionMergeable)
   > Set Glow Position(Glow Position="Get (a ref)"(Array:Array of Vectors=Get GlowLocations(), Dimension 1=Get MergedGlowIndex() -> (MergedGlowIndex:int=0)) -> (Output:"Vector (by ref)"))
   n13 = Get Dynamic Mesh()
   n14 = int + int(A="Max (Integer)"(A=Get UV Channel XY() -> (UV Channel XY=0), B=Get UV Channel Z() -> (UV Channel Z=0)))
   Branch(execute=@n3.then_1, Condition="integer < integer"(A=Get Num UV Channels(TargetMesh=@n13) -> (ReturnValue=0), B=@n14))
   > Set Num UV Channels(TargetMesh=Get Dynamic Mesh(), NumUVSets=@n14)
  ]}
 RealignVertices {nodes [
   n1 = RealignVertices!() -> (then, DynamicMesh:Dynamic Mesh Object Reference, Options:Atom Processor Fix Vertices Options Structure) {is=title K2Node_FunctionEntry}
   n2 = Break Atom Processor Fix Vertices Options!(AtomProcessorFixVerticesOptions:"Atom Processor Fix Vertices Options Structure (by ref)"=@n1.Options) -> (
    Realignvertices_20_#54:bool=~
    DoXAxis_25_#55:bool=~
    DoYAxis_28_#56:bool=~
    DoZAxis_29_#57:bool=~
    ToleranceX_32_#58:double=~
    ToleranceY_35_#59:double=~
    ToleranceZ_36_#60:double=~
   ) {is=title K2Node_BreakStruct}
   > n3 = Branch(Condition=@n2)
   > n4 = RealignVerticesPerAxis!(
    execute
    self
    DynamicMesh:Dynamic Mesh Object Reference=@n1
    Do X Axis:bool=false
    Do Y Axis:bool=false
    Do Z Axis:bool=@n2.DoZAxis_29_#57
    Tolerance:double=@n2.ToleranceZ_36_#60
   ) -> (then, Target Mesh:Dynamic Mesh Object Reference)
   > n5 = RealignVerticesPerAxis(
    DynamicMesh=@n4
    Do X Axis=false
    Do Y Axis=@n2.DoYAxis_28_#56
    Do Z Axis=false
    Tolerance=@n2.ToleranceY_35_#59
   )
   > n6 = RealignVerticesPerAxis(
    DynamicMesh=@n5
    Do X Axis=@n2.DoXAxis_25_#55
    Do Y Axis=false
    Do Z Axis=false
    Tolerance=@n2.ToleranceX_32_#58
   )
   > Return Node!(execute, Target Mesh:Dynamic Mesh Object Reference=@n6)
   Return Node(execute=@n3.else, Target Mesh=Get DynamicMesh())
  ]}
 RealignVerticesPerAxis {nodes [
   n1 = RealignVerticesPerAxis!() -> (then, DynamicMesh:Dynamic Mesh Object Reference, Do X Axis:bool=~, Do Y Axis:bool=~, Do Z Axis:bool=~, Tolerance:double=0.2) {is=title K2Node_FunctionEntry}
   > n2 = Set AxisSwitch(AxisSwitch:Array of Booleans=Make Array!("[0]":bool=@n1.Do X Axis, "[1]":bool=@n1.Do Y Axis, "[2]":bool=@n1.Do Z Axis) -> (Array:Array of Booleans)) -> (Output_Get:Array of Booleans)
   > n3 = Branch(Condition=Contains Item(TargetArray:Array of Booleans=@n2, ItemToFind:"Boolean (by ref)"=true))
   > Clear!(execute, TargetMap:Map of Integers to Vector 4 Structures=Get VertexData() -> (VertexData:Map of Integers to Vector 4 Structures)) -> (then) {is=title K2Node_CallFunction}
   > n4 = GetAllVertexIDs!(execute, TargetMesh:Dynamic Mesh Object Reference=Get DynamicMesh()) -> (then, VertexIDList:Index List Structure, bHasVertexIDGaps:bool, ReturnValue:Dynamic Mesh Object Reference)
   > n5 = ConvertIndexListToArray(IndexList=@n4)
   > Set VertexIDs(VertexIDs:Array of Integers=@n5) -> (Output_Get:Array of Integers)
   > n6 = GetAllVertexPositions!(execute, TargetMesh:Dynamic Mesh Object Reference=Get DynamicMesh(), bSkipGaps:bool=false) -> (then, PositionList:Vector List Structure, bHasVertexIDGaps:bool, ReturnValue:Dynamic Mesh Object Reference)
   > n7 = ConvertVectorListToArray!(execute, VectorList:Vector List Structure=@n6) -> (then, VectorArray:Array of Vectors)
   > Set VertexPositions(VertexPositions:Array of Vectors=@n7) -> (Output_Get:Array of Vectors)
   > n8 = Sequence!(execute) -> (then_0, then_1, then_2:Exec)
   n9 = For Each Loop(Exec=@n8.then_1, Array:Array of Vectors=Get VertexPositions() -> (VertexPositions:Array of Vectors)) -> (Array Element:Vector)
   n10 = SelectComponent!(Vec3:Vector=@n9, XYZ:Array of Booleans=Get AxisSwitch() -> (AxisSwitch:Array of Booleans)) -> (ReturnValue:double)
   n11 = Get!() -> (PositionRange_X:double, PositionRange_Y:double)
   n12 = float - float(A=@n11, B=@n11.PositionRange_Y)
   n13 = Round(A=float * float(A=float / float(A=@n10, B=@n12), B:int=Get SliceCount() -> (SliceCount:int=0)))
   n14 = BreakVector(InVec=@n9)
   > Add(
    TargetMap:Map of Integers to Vector 4 Structures=Get VertexData()
    Key:"Integer (by ref)"="Get (a ref)"(Array:Array of Integers=Get VertexIDs() -> (VertexIDs:Array of Integers), Dimension 1=@n9.Array Index) -> (Output:"Integer (by ref)")
    Value:"Vector 4 Structure (by ref)"=MakeVector4!(X:double=@n14, Y:double=@n14.Y, Z:double=@n14.Z, W:double="To Float (Integer)"(InInt=@n13)) -> (ReturnValue:Vector 4 Structure)
   )
   n15 = For Each Loop(Exec=@n8.then_0, Array=Get VertexPositions())
   n16 = Get()
   n17 = float - float(A=@n16, B=@n16.PositionRange_Y)
   n18 = Get Tolerance() -> (Tolerance:double)
   Set SliceCount(execute=@n15.Completed, SliceCount:int=Round(A=float / float(A=@n17, B=@n18))) -> (Output_Get:int=0)
   n19 = SelectComponent(Vec3=@n15, XYZ=Get AxisSwitch())
   n20 = Get()
   Set!(
    execute=@n15.LoopBody
    PositionRange_X:double="Max (Float)"(A=@n19, B=@n20)
    PositionRange_Y:double="Min (Float)"(A:double=@n20.PositionRange_Y, B:double=@n19) -> (ReturnValue:double) {is=title K2Node_CommutativeAssociativeBinaryOperator}
   ) -> (then, Output_Get:Vector 2D Structure)
   n21 = Execute Python Script!(
    execute=@n8.then_2
    PythonScript="import unreal as ue\r\nfrom collections import defaultdict\r\n\r\n\r\ndef process_vertex_data(vertex_data, axis_switch):\r\n    ## Initialize\r\n    grouped_vertices = defaultdict(list)\r\n    output_map = ue.Map(int, ue.Vector)\r\n\r\n    for vertex_id in vertex_data.keys():  ## Group vertices per slice\r\n        v = vertex_data[vertex_id]  ## Vertex position in .xyz and vertex slice in .w\r\n        grouped_vertices[v.w].append({'x': v.x, 'y': v.y, 'z': v.z, 'vertex_id': vertex_id})  ## Group\r\n\r\n    for vertex_slice, vertex_info in grouped_vertices.items():  ## Process each group\r\n        n_count = defaultdict(int)  ## Count occurrences of each component in the group\r\n        for vertex in vertex_info:\r\n            for axis, switch in zip('xyz', axis_switch):\r\n                if switch:\r\n                    n_count[vertex[axis]] += 1\r\n\r\n        dominant_value = max(n_count, key=n_count.get)  ## Find the most common value in component\r\n\r\n        for vertex in vertex_info:   ## Set all components in the group to the dominant value\r\n            for axis, switch in zip('xyz', axis_switch):\r\n                if switch:\r\n                    vertex[axis] = dominant_value\r\n\r\n    for v_list in grouped_vertices.values():\r\n        for v in v_list:   ## Write new vertex data into the map\r\n            output_map[v['vertex_id']] = ue.Vector(v['x'], v['y'], v['z'])\r\n\r\n    return output_map\r\n\r\n\r\nvertex_data_out = process_vertex_data(vertex_data, axis_switch)"
    vertex_data:Map of Integers to Vector 4 Structures=Get VertexData()
    axis_switch:Array of Booleans=Get AxisSwitch()
   ) -> (then, ReturnValue, vertex_data_out:Map of Integers to Vectors) {inputs [vertex_data,axis_switch], outputs [vertex_data_out]}
   > n22 = Keys(TargetMap:Map of Integers to Vectors=@n21.vertex_data_out) -> (Keys:Array of Integers)
   > n23 = Values!(execute, TargetMap:Map of Integers to Vectors=@n21.vertex_data_out) -> (then, Values:Array of Vectors)
   > n24 = For Each Loop(Array:Array of Integers=@n22) -> (Array Element:int)
   > SetVertexPosition!(
    execute
    TargetMesh:Dynamic Mesh Object Reference=Get DynamicMesh()
    VertexID:int=@n24
    NewPosition:Vector="Get (a ref)"(Array:Array of Vectors=@n23, Dimension 1=@n24.Array Index) -> (Output:"Vector (by ref)")
    bDeferChangeNotifications:bool=false
   ) -> (then, bIsValidVertex:bool, ReturnValue:Dynamic Mesh Object Reference)
   Return Node(execute=@n24.Completed, Target Mesh=Get DynamicMesh())
   Return Node(execute=@n3.else, Target Mesh=Get DynamicMesh())
  ]}
 RegisterAndMergeGlowTransform {nodes [
   n1 = RegisterAndMergeGlowTransform!() -> (then, PartInfo:Atom Model Part Instance Info Structure) {is=title K2Node_FunctionEntry}
   n2 = Break Atom Model Part Instance Info!(AtomModelPartInstanceInfo=@n1) -> (Primitive, Transforms, Group:String, ParentGroup:String, CommonParts:Array of Atom Common Part And Transform Structures)
   > Set Primitive(Primitive=@n2)
   > Set PrimitiveTransform(PrimitiveTransform:Transform="Get (a ref)"(Array:Array of Transforms=@n2.Transforms, Dimension 1=0) -> (Output:"Transform (by ref)")) -> (Output_Get:Transform)
   > n3 = Sequence!(execute) -> (then_0, then_1)
   > Set NewGlowPosition(NewGlowPosition:Vector=Compute Glow Position(PrimitiveTransform=Get PrimitiveTransform() -> (PrimitiveTransform:Transform), Primitive=Get Primitive())) -> (Output_Get:Vector)
   n4 = Is Glow Position Mergeable(GlowPosition=Get NewGlowPosition() -> (NewGlowPosition:Vector)) -> (GlowIndex=0)
   Set PreviousGlowIndex(execute=@n3.then_1, PreviousGlowIndex:int=@n4) -> (Output_Get:int=0)
   > n5 = Branch(Condition=@n4.IsGlowPositionMergeable)
   n6 = "Get (a ref)"(Array:Array of Vectors=Get GlowLocations(), Dimension 1=Get PreviousGlowIndex() -> (PreviousGlowIndex:int=0)) -> (Output:"Vector (by ref)")
   n7 = Get NewGlowPosition()
   > Set Array Elem(
    TargetArray:Array of Vectors=Get GlowLocations()
    Index=Get PreviousGlowIndex() -> (PreviousGlowIndex=0)
    Item:"Vector (by ref)"=vector / vector(A=vector + vector(A=@n6, B=@n7), B=2)
    bSizeToFit=false
   )
   > Return Node!(execute, GlowMerged:bool=true)
   Add(execute=@n5.else, TargetArray:Array of Vectors=Get GlowLocations(), NewItem:"Vector (by ref)"=Get NewGlowPosition()) {is=title K2Node_CallArrayFunction}
   > Return Node(GlowMerged=false)
   Get PreviousGlowLocation() -> (PreviousGlowLocation:Vector)
  ]}
 RemoveUndersideOnLowerLODs {nodes [
   n1 = RemoveUndersideOnLowerLODs!() -> (
    then
    AtomPrimitiveGeometryLODs:"Atom Primitive Geometry LODs Structure (by ref)"
    MaterialList:Array of Material Interface Object References
   ) {is=title K2Node_FunctionEntry}
   > n2 = For Each Loop(Array:Array of Material Interface Object References=@n1.MaterialList) -> (Array Element:Material Interface Object Reference)
   n3 = For Each Loop(
    Exec=@n2.Completed
    Array:Array of Atom Primitive Geometry Object References=Break Atom Primitive Geometry LODs!(AtomPrimitiveGeometryLODs=@n1) -> (
     GeometryLODs:Array of Atom Primitive Geometry Object References
     GeometryLODsMaterials
     HiResNaniteGeomery:Atom Primitive Geometry Object Reference
     HiResNaniteGeomeryMaterials:Atom Primitive Geometry Material Assignments Structure
     Collision:Atom Primitive Collision Geometry Structure
     DetailedCollision:Atom Primitive Collision Geometry Structure
    )
   ) -> (Array Element:Atom Primitive Geometry Object Reference)
   n4 = Set members in Atom Primitive Geometry LODs!(
    execute=@n3.Completed
    StructRef=@n1
    GeometryLODs:Array of Atom Primitive Geometry Object References=Get LODArray() -> (LODArray:Array of Atom Primitive Geometry Object References)
   ) -> (then, StructOut)
   > Return Node!(execute, ReturnValue:"Atom Primitive Geometry LODs Structure (by ref)"=@n4)
   n5 = Get UndersideMaterialIndexReplacementMap() -> (UndersideMaterialIndexReplacementMap:Map of Integers to Integers)
   n6 = Length(TargetArray:Array of Atom Processor Geometry Style Settings Structures=Get LOD Geometry Style Settings())
   n7 = Branch(
    execute=@n3.LoopBody
    Condition=AND Boolean(A="integer >= integer"(A=@n3.Array Index, B=@n6), B=Is Not Empty(TargetMap:Map of Integers to Integers=@n5))
   )
   > n8 = ConvertAtomPrimitiveGeometryToDynamicMesh(AtomPrimitiveGeometry=@n3)
   > n9 = Keys(TargetMap:Map of Integers to Integers=Get UndersideMaterialIndexReplacementMap())
   > n10 = For Each Loop(Array:Array of Integers=@n9) -> (Array Element:int)
   n11 = ConvertDynamicMeshToAtomPrimitiveGeometry(execute=@n10.Completed, DynamicMesh=@n8)
   > Add(
    TargetArray:Array of Atom Primitive Geometry Object References=Get LODArray()
    NewItem:"Atom Primitive Geometry Object Reference (by ref)"=@n11
   )
   n12 = GetTrianglesByMaterialID(TargetMesh=@n8, MaterialID=@n10)
   Branch(execute=@n10.LoopBody, Condition="integer > integer"(A=GetIndexListLength!(IndexList:Index List Structure=@n12) -> (ReturnValue:int=0)))
   > n13 = ConvertIndexListToMeshSelection!(
    execute
    TargetMesh:Dynamic Mesh Object Reference=@n12.ReturnValue
    IndexList:Index List Structure=@n12
    SelectionType:EGeometryScriptMeshSelectionType Enum=Triangles
   ) -> (then, Selection:Mesh Selection Structure, ReturnValue:Dynamic Mesh Object Reference)
   n14 = GetMeshBoundingBox(TargetMesh=@n13.ReturnValue) -> (ReturnValue_Min=~, ReturnValue_Max=~, ReturnValue_IsValid=~)
   n15 = vector - vector(A=@n14, B=@n14.ReturnValue_Max)
   > n16 = SetMeshUVsFromBoxProjection!(
    execute
    TargetMesh=@n13.ReturnValue
    UVSetIndex=0
    BoxTransform_Location=vector - vector(A=@n14, B=@n15)
    BoxTransform_Rotation="0, 0, 0"
    BoxTransform_Scale:Vector=vector * vector(A=@n15, B=5) -> (ReturnValue=~)
    Selection=@n13
    MinIslandTriCount=2
   ) -> (then, ReturnValue)
   > Set Material ID For Mesh Selection!(
    execute
    TargetMesh:Dynamic Mesh Object Reference=@n16
    Selection:Mesh Selection Structure=@n13
    MaterialID:int=Find(TargetMap:Map of Integers to Integers=Get UndersideMaterialIndexReplacementMap(), Key:"Integer (by ref)"=@n10) -> (Value:int)
    bDeferChangeNotifications:bool=false
   ) -> (then, ReturnValue:Dynamic Mesh Object Reference) {is=title K2Node_CallFunction}
   Add(execute=@n7.else, TargetArray=Get LODArray(), NewItem=@n3)
   n17 = GetMaterialFeatures(execute=@n2.LoopBody, self=GetMaterialFactory(), Material=@n2)
   > Branch(Condition="Equal (Enum)"(A=Break Atom Material Features!(AtomMaterialFeatures=@n17) -> (UVFeature), B=FakeUnderside))
   > Add(
    TargetMap:Map of Integers to Integers=Get UndersideMaterialIndexReplacementMap()
    Key=@n2.Array Index
    Value:"Integer (by ref)"=GetUndersideMaterialReplacementIndex!(
     Underside Material Feature:Atom Material Features Structure=@n17
     Materials:Array of Material Interface Object References=Get MaterialList() -> (MaterialList:Array of Material Interface Object References)
    ) -> (Material Replacement Index:int=0) {is=title K2Node_CallFunction}
   ) {is=title K2Node_CallFunction}
   PrintString(
    InString=Append!(A="-- Force removed underside material on LOD  ", B="To String (Integer)"(InInt=@n3.Array Index)) -> (ReturnValue)
    bPrintToScreen=false
    Key=~
   )
  ]}
 SelectComponent {nodes [
   n1 = SelectComponent!() -> (then, Vec3:Vector=~, XYZ:Array of Booleans) {is=title K2Node_FunctionEntry}
   n2 = BreakVector(InVec=@n1)
   n3 = SelectFloat(A=@n2, B=@n2.Y, bPickA="Get (a ref)"(Array:Array of Booleans=@n1.XYZ, Dimension 1=0) -> (Output:"Boolean (by ref)"))
   n4 = "Get (a ref)"(Array=@n1.XYZ, Dimension 1=1)
   > Return Node(ReturnValue:double=SelectFloat(A=@n2.Z, B=SelectFloat(A=@n2.Y, B=@n3, bPickA=@n4), bPickA="Get (a ref)"(Array=@n1.XYZ, Dimension 1=2)))
  ]}
 SetCommonPartMeshName {nodes [
   n1 = SetCommonPartMeshName!() -> (
    then
    Merged Mesh Part Info:"Atom Merged Mesh Part Info Structure (by ref)"
    Atom Model Part Info:"Atom Model Part Instance Info Structure (by ref)"
    Common Part Index:int=0
    Common Part Type:"EAtomCommonPartType Enum (by ref)"=None
   ) {is=title K2Node_FunctionEntry}
   > n2 = Parent: SetCommonPartMeshName!(
    execute
    Merged Mesh Part Info:"Atom Merged Mesh Part Info Structure (by ref)"=@n1
    Atom Model Part Info:"Atom Model Part Instance Info Structure (by ref)"=@n1.Atom Model Part Info
    Common Part Index:int=@n1.Common Part Index
    Common Part Type:"EAtomCommonPartType Enum (by ref)"=@n1.Common Part Type
   ) -> (then) {is=title K2Node_CallParentFunction}
   n3 = Is Common Part Affected by Cap Optimization(Exec=@n2.then, PartInfo=@n1.Atom Model Part Info, CommonPartIndex=@n1.Common Part Index)
   n4 = Break Atom Merged Mesh Part Info!(AtomMergedMeshPartInfo=@n1) -> (MeshName:String)
   n5 = Break Atom Processor Applied Optimizations!(AtomProcessorAppliedOptimizations=Get Optimizations To Apply For LOD(Part Info=@n1.Atom Model Part Info, LOD=0)) -> (
    bApplyFakeTranslucent_25_#9:bool=~
    bApplyDitheredTranslucent_24_#10:bool=~
    bApplyGlow_13_#11:bool=~
   )
   n6 = Branch(
    execute=@n3.False
    Condition=AND Boolean!(A=Get Split Analytic Common Part Mesh(), B=IsEmpty(InString=@n4), C:bool=NOT Boolean(A=@n5.bApplyGlow_13_#11)) -> (ReturnValue)
   )
   > n7 = Set members in Atom Merged Mesh Part Info!(execute, StructRef=@n1, MeshName:String=CP) -> (then, StructOut)
   Return Node!(execute=[@n6.else, @n7.then])
  ]}
 SetFakeBevelsUVs {nodes [
   n1 = SetFakeBevelsUVs!() -> (then, DynamicMesh:Dynamic Mesh Object Reference, Fake Bevels Channel:int) {is=title K2Node_FunctionEntry}
   > n2 = Set MeshDimensions(MeshDimensions:Vector=GetBrickSize(DynamicMesh=@n1)) -> (Output_Get:Vector)
   n3 = vector * vector!(A:double=3.125, B:Vector=@n2) -> (ReturnValue_X:double, ReturnValue_Y:double, ReturnValue_Z:double)
   > Set!(
    execute
    ScaledDimensions_X:double="To Float (Integer)"(InInt=Ceil(A=@n3))
    ScaledDimensions_Y:double="To Float (Integer)"(InInt=Ceil(A=@n3.ReturnValue_Y))
    ScaledDimensions_Z:double="To Float (Integer)"(InInt=Ceil(A=@n3.ReturnValue_Z))
   ) -> (then, Output_Get:Vector)
   > n4 = For Each Loop(Array:Array of Vectors=Get NormalArray()) -> (Array Element:Vector)
   > n5 = SelectMeshElementsByNormalAngle(TargetMesh=Get DynamicMesh(), Normal=@n4, MaxAngleDeg=44, SelectionType=Triangles, bInvert=false, MinNumTrianglePoints=1)
   n6 = BreakVector(InVec=Get MeshDimensions() -> (MeshDimensions:Vector))
   n7 = Get RotationArray()
   n8 = float / float(A=@n6.Y, B=2)
   n9 = Get ScaledDimensions() -> (ScaledDimensions:Vector)
   n10 = BreakVector(InVec=vector * vector!(A:Vector=vector - vector(A=@n9, B:double=1), B:double=-0.4) -> (ReturnValue:Vector=~))
   n11 = float / float(A=@n6, B=2)
   > SetMeshUVsFromBoxProjection!(
    execute
    TargetMesh=@n5.ReturnValue
    UVSetIndex=Get Fake Bevels Channel() -> (Fake Bevels Channel=0)
    BoxTransform:Transform=MakeTransform!(
     Location_X:double=float - float(A=@n10, B=@n11)
     Location_Y:double=float - float(A=@n10.Y, B=@n8)
     Location_Z:double=@n6.Z
     Rotation="Get (a ref)"(Array:Array of Rotators=@n7, Dimension 1=@n4.Array Index) -> (Output:"Rotator (by ref)")
     Scale_X:double=@n6
     Scale_Y:double=@n6.Y
     Scale_Z:double=@n6.Z
    ) -> (ReturnValue)
    Selection=@n5
    MinIslandTriCount=1
   ) -> (then, ReturnValue)
   n12 = Convert UV Islands To PolyGroups(execute=@n4.Completed, TargetMesh=Get DynamicMesh(), UVLayer=Get Fake Bevels Channel() -> (Fake Bevels Channel=0))
   > Get PolyGroup IDs In Mesh(TargetMesh=@n12, PolygroupIDsOut=Get Polygroups())
   > n13 = ConvertIndexListToArray(IndexList=Get Polygroups())
   > n14 = For Each Loop(Array:Array of Integers=@n13) -> (Array Element:int)
   > FixPolygroupBoxProjectionUVs!(
    execute
    self
    DynamicMesh:Dynamic Mesh Object Reference=Get DynamicMesh()
    Polygon Group Index:int=@n14
    Fake Bevels Channel:int=Get Fake Bevels Channel() -> (Fake Bevels Channel=0)
   ) -> (then) {is=title K2Node_CallFunction}
   > n15 = Get Polygroup UVs Translate and Scale!(execute, self, DynamicMesh:Dynamic Mesh Object Reference=Get DynamicMesh(), PolygroupID:int=@n14, BrickSize:Vector=Get ScaledDimensions()) -> (then, UVsScale:Vector 2D Structure, UVsTranslate:Vector 2D Structure) {is=title K2Node_CallFunction}
   > n16 = ConvertIndexSetToMeshSelection(TargetMesh=Get DynamicMesh(), IndexSet=Make Set("[0]"=@n14), SelectionType=Polygroups)
   n17 = BreakVector2D(InVec=@n15)
   > n18 = ScaleMeshUVs!(
    execute
    TargetMesh=Get DynamicMesh()
    UVSetIndex=Get Fake Bevels Channel() -> (Fake Bevels Channel=0)
    Scale_X:double=@n17
    Scale_Y:double=@n17.Y
    ScaleOrigin
    Selection=@n16
   ) -> (then, ReturnValue)
   n19 = BreakVector2D(InVec=@n15.UVsTranslate)
   > TranslateMeshUVs!(
    execute
    TargetMesh=@n18
    UVSetIndex=Get Fake Bevels Channel() -> (Fake Bevels Channel=0)
    Translation_X:double=@n19
    Translation_Y:double=@n19.Y
    Selection=@n16
   ) -> (then, ReturnValue)
  ]}
 SetLODSettings {nodes [
   n1 = SetLODSettings!() -> (then, StaticMesh:Static Mesh Object Reference) {is=title K2Node_FunctionEntry}
   > n2 = GetLODPresets()
   n3 = Break Atom Processor LODSettings!(AtomProcessorLODSettings=@n2) -> (
    Name_26_#43:EAtomProcessorLODPreset Enum
    LODDistances_33_#44:"Array of Float (double-precision)s"
    LODTriangleBudgets_35_#45
    QualityLevelMinLOD_18_#46:Map of EPerQualityLevels Enums to Integers
    UseLODSizeReference_31_#47:bool=~
    LODSizeReference_34_#48:double=~
   )
   > CalculateScreenSizeFromDistances!(
    execute
    self
    StaticMesh:Static Mesh Object Reference=@n1
    Distances:"Array of Float (double-precision)s"=@n3.LODDistances_33_#44
    Use LOD Size Reference:bool=@n3.UseLODSizeReference_31_#47
    LOD Size Reference:double=@n3.LODSizeReference_34_#48
   ) -> (then) {is=title K2Node_CallFunction}
   > SetMinLODForQualityLevels!(
    execute
    self:Static Mesh Object Reference=@n1
    QualityLevelMinimumLODs:Map of EPerQualityLevels Enums to Integers=@n3.QualityLevelMinLOD_18_#46
    Default:int=0
   ) -> (then)
  ]}
 SetNaniteCommonPartInfo {nodes [
   n1 = SetNaniteCommonPartInfo!() -> (
    then
    Merged Mesh Part Info:"Atom Merged Mesh Part Info Structure (by ref)"
    Shell Merged Mesh Part Info:"Atom Merged Mesh Part Info Structure (by ref)"
    Atom Model Part Info:"Atom Model Part Instance Info Structure (by ref)"
   ) {is=title K2Node_FunctionEntry}
   > Parent: SetNaniteCommonPartInfo!(
    execute
    Merged Mesh Part Info:"Atom Merged Mesh Part Info Structure (by ref)"=@n1
    Shell Merged Mesh Part Info:"Atom Merged Mesh Part Info Structure (by ref)"=@n1.Shell Merged Mesh Part Info
    Atom Model Part Info:"Atom Model Part Instance Info Structure (by ref)"=@n1.Atom Model Part Info
   ) -> (then) {is=title K2Node_CallParentFunction}
   > n2 = Branch(Condition="Equal, Case Insensitive (String)"(A=Break Atom Merged Mesh Part Info(AtomMergedMeshPartInfo=@n1), B=CP))
   > n3 = Set members in Atom Merged Mesh Part Info(StructRef=@n1)
   Return Node(execute=[@n3.then, @n2.else])
  ]}
 ShouldDefineSeparateNaniteCommonParts {nodes [
   ShouldDefineSeparateNaniteCommonParts() {is=title K2Node_FunctionEntry}
   > Return Node!(execute, Separate Nanite Common Parts:bool=true)
  ]}
 UndersideCap Requested On Any LODs {nodes [
   UndersideCap Requested On Any LODs() {is=title K2Node_FunctionEntry}
   > n1 = For Each Loop(Array:Array of Atom Processor Geometry Style Settings Structures=Get LOD Geometry Style Settings()) -> (Array Element:Atom Processor Geometry Style Settings Structure)
   Return Node!(execute=@n1.Completed, bForceFlatCapRequested:bool=false)
   > Branch(Condition=Break Atom Processor Geometry Style Settings!(AtomProcessorGeometryStyleSettings=@n1) -> (ForceFlatCap_35_#16=~, FakeTechnicHole_41_#1=~))
   > Return Node(bForceFlatCapRequested=true)
  ]}
 UseMaterialFactory {nodes [
   UseMaterialFactory() {is=title K2Node_FunctionEntry}
   > Return Node!(execute, UseMaterialFactory:bool=true)
  ]}
 UseMaterialOverrideIfNeeded {nodes [
   Inputs!() -> (MaterialOverrides:bool=~)
   Outputs!()
  ]}
 WeldLODs {nodes [
   n1 = WeldLODs!() -> (then, AtomPrimitiveGeometryLODs:"Atom Primitive Geometry LODs Structure (by ref)", FromLodIndex:int) {is=title K2Node_FunctionEntry}
   > Set PrimitiveGeometry_L(PrimitiveGeometry_L:Atom Primitive Geometry LODs Structure=@n1) -> (Output_Get:Atom Primitive Geometry LODs Structure)
   > n2 = For Each Loop(Array:Array of Atom Primitive Geometry Object References=Break Atom Primitive Geometry LODs(AtomPrimitiveGeometryLODs=@n1)) -> (Array Element:Atom Primitive Geometry Object Reference)
   n3 = Set members in Atom Primitive Geometry LODs(
    execute=@n2.Completed
    StructRef=Get PrimitiveGeometry_L() -> (PrimitiveGeometry_L:Atom Primitive Geometry LODs Structure)
    GeometryLODs=Get LODs() -> (LODs:Array of Atom Primitive Geometry Object References)
   )
   > Return Node!(execute, ReturnValue:"Atom Primitive Geometry LODs Structure (by ref)"=@n3)
   n4 = Branch(execute=@n2.LoopBody, Condition="integer >= integer"(A=@n2.Array Index, B=Get FromLodIndex() -> (FromLodIndex:int=0)))
   > n5 = ConvertAtomPrimitiveGeometryToDynamicMesh(AtomPrimitiveGeometry=@n2)
   > n6 = RealignVertices!(
    execute
    self
    DynamicMesh:Dynamic Mesh Object Reference=@n5
    Options:Atom Processor Fix Vertices Options Structure=Get Vertex Realignment Options() -> (Vertex Realignment Options:Atom Processor Fix Vertices Options Structure)
   ) -> (then, Target Mesh:Dynamic Mesh Object Reference) {is=title K2Node_CallFunction}
   > n7 = SnapMeshOpenBoundaries!(
    execute
    TargetMesh:Dynamic Mesh Object Reference=@n6
    SnapOptions:Geometry Script Snap Boundaries Options Structure=Get Vertex Snap Options() -> (Vertex Snap Options:Geometry Script Snap Boundaries Options Structure)
   ) -> (then, ReturnValue:Dynamic Mesh Object Reference)
   > n8 = Branch(Condition=Get Weld Mesh Edges() -> (Weld Mesh Edges:bool))
   > n9 = WeldMeshEdges!(
    execute
    TargetMesh=@n7
    WeldOptions:Geometry Script Weld Edges Options Structure=Make Geometry Script Weld Edges Options!(Tolerance:float=2, bOnlyUniquePairs:bool=false) -> (GeometryScriptWeldEdgesOptions:Geometry Script Weld Edges Options Structure) {is=title K2Node_MakeStruct}
   ) -> (then, ReturnValue)
   > n10 = ConvertDynamicMeshToAtomPrimitiveGeometry(DynamicMesh=@n9)
   > Add(TargetArray=Get LODs(), NewItem=@n10) {is=title K2Node_CallArrayFunction}
   n11 = ConvertDynamicMeshToAtomPrimitiveGeometry(execute=@n8.else, DynamicMesh=@n7)
   > Add(TargetArray=Get LODs(), NewItem=@n11)
   Add(execute=@n4.else, TargetArray=Get LODs(), NewItem=@n2)
  ]}
]
variables [
 Append Additional Source LODs {type=boolean, default=true}
 Brick Protection - Min Sample Radius {type=number, default=6}
 Brick Protection - Min Thickness {type=number, default=2}
 Brick Protection - Tolerance {type=number, default=4}
 Custom LOD Settings {
  type=AtomProcessorLODSettings
  default {
   name=Medium Structure
   lOD Size Reference=0
   use LOD Size Reference=false
   lOD Distances [1,550,700,800,1200,1400,1600]
   lOD Triangle Budgets [4500]
   quality Level Min LOD {low=2, medium=1}
  }
 }
 Debug - Bypass Mesh Convex Merge {type=boolean}
 Force Shape Count {type=integer}
 Glow Position Merge Distance {type=number}
 GlowLocations {type="array<Vector>"}
 LOD Presets Data {type=BP_AtomProcessorLODPreset_C, default=&/DA_AtomProcessorLODPreset}
 Mesh Protection - Min Sample Radius {type=number, default=8}
 Mesh Protection - Tolerance {type=number, default=3}
 Optimize Mesh Collision {type=boolean, default=true}
 Optimize Primitive Collision {type=boolean, default=true}
 Primitives Features Overrides {type=BP_AtomPrimitivesFeaturesOverrides_C, default=&/DA_AtomPrimitivesFeaturesOverrides}
 Selected LOD Preset {type=EAtomProcessorLODPreset, default=Medium Structure}
 Simplification Distance {type=number, default=0.5}
 Simplification Min Face Count {type=integer, default=12}
 Source LOD Settings {
  type="array<MergedMeshSourceLODSettings>"
  default [
   {analyticCommonParts=true, fakeAnalyticUnderside=false, fakeBevel=true, fakeUnderside=false, forceFlatCap=false}
   {analyticCommonParts=true, fakeAnalyticUnderside=false, fakeBevel=true, fakeUnderside=true, forceFlatCap=true}
  ]
 }
 Translucency Type {type=EAtomMaterialTranslucencyType, default=Translucent}
 Use Custom LOD Settings {type=boolean}
 Use Improved Color Change Processing {type=boolean}
 UVChangedMap {type="map<boolean>", default {}}
 Vertex Realignment Options {
  type=AtomProcessorFixVerticesOptions
  default {do X Axis=true, do Y Axis=true, do Z Axis=true, realign vertices=false, tolerance X=0.2, tolerance Y=0.35, tolerance Z=0.2}
 }
 Vertex Snap Options {type=GeometryScriptSnapBoundariesOptions, default {bSnapToEdges=true, maxIterations=5, tolerance=0.1}}
 Weld Mesh Edges {type=boolean, default=true}
]
