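# Editor-side Atom model processor, named for skeletal meshes, that supplies geometry, materials, collision and finishing steps for converting brick-style models into mesh assets.
# Part LODs come from styled primitive geometry, with bevelled LOD0 or baked flex transforms, while common parts are cached by type.
# Materials follow part type, with decorations building parameterised material instances.
# Static mesh finalisation sets LOD screen sizes, sockets, import data and common-part caches.
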EditorUtilityBlueprint /AtomFramework/Editor/Processors/MP_SkeletalMesh
graphs [
 CreateGeometryCollection {nodes [
   CreateGeometryCollection!() -> (then, Proxy Mesh:Static Mesh Object Reference, Brick Static Mesh Path:String) {is=title K2Node_FunctionEntry}
   > Return Node!(execute, Geometry Collection:Geometry Collection Object Reference)
  ]}
 CreateNanitePartInfo {nodes [
   n1 = Inputs!() -> (Exec:Exec, MergedMeshPartInfo:Atom Merged Mesh Part Info Structure)
   n2 = Break Atom Merged Mesh Part Info!(AtomMergedMeshPartInfo:"Atom Merged Mesh Part Info Structure (by ref)"=@n1) -> (
    GeometryLODs:Atom Primitive Geometry LODs Structure
    Transform:Transform
    bHasConstantOverrideVertexColor:bool=~
    VertexColor:Color Structure
    Materials:Array of Material Interface Object References
    PreventMergingMaterials:Array of Material Interface Object References
    MaterialFeatures:Array of Atom Material Features Structures
    MeshDetailLevel:EAtomMergedMeshPartDetailLevel Enum
    bPreserveUVs:bool=~
    bAllowMerging:bool=~
    bAllowSimplification:bool=~
    bAllowApproximation:bool=~
    bIsOccluder:bool=~
    FilterLODLevel:int
    MeshName:String
    InstanceSubsetID:int
    ApproximationType:Bitmask Integer
   ) {is=title K2Node_BreakStruct}
   n3 = Break Atom Primitive Geometry LODs!(AtomPrimitiveGeometryLODs:"Atom Primitive Geometry LODs Structure (by ref)"=@n2) -> (
    GeometryLODs:Array of Atom Primitive Geometry Object References
    GeometryLODsMaterials:Array of Atom Primitive Geometry Material Assignments Structures
    HiResNaniteGeomery:Atom Primitive Geometry Object Reference
    HiResNaniteGeomeryMaterials:Atom Primitive Geometry Material Assignments Structure
    Collision:Atom Primitive Collision Geometry Structure
    DetailedCollision:Atom Primitive Collision Geometry Structure
   ) {is=title K2Node_BreakStruct}
   n4 = Make Array("[0]":Atom Primitive Geometry Object Reference=@n3.HiResNaniteGeomery) -> (Array:Array of Atom Primitive Geometry Object References)
   > Outputs!(
    Out:Exec
    NaniteMergedMeshPartInfo:Atom Merged Mesh Part Info Structure=Make Atom Merged Mesh Part Info!(
     GeometryLODs:Atom Primitive Geometry LODs Structure=Make Atom Primitive Geometry LODs!(
      GeometryLODs:Array of Atom Primitive Geometry Object References=@n4
      GeometryLODsMaterials:Array of Atom Primitive Geometry Material Assignments Structures
      HiResNaniteGeomery:Atom Primitive Geometry Object Reference
      HiResNaniteGeomeryMaterials:Atom Primitive Geometry Material Assignments Structure=(MaterialSlotName=,MaterialFeatures=)
      Collision:Atom Primitive Collision Geometry Structure=@n3.Collision
      DetailedCollision:Atom Primitive Collision Geometry Structure=(AggGeom=(SphereElems=,BoxElems=,SphylElems=,ConvexElems=,TaperedCapsuleElems=,LevelSetElems=,SkinnedLevelSetElems=,MLLevelSetElems=,SkinnedTriangleMeshElems=))
     ) -> (AtomPrimitiveGeometryLODs:Atom Primitive Geometry LODs Structure) {is=title K2Node_MakeStruct}
     Transform:Transform=@n2.Transform
     bHasConstantOverrideVertexColor:bool=True
     VertexColor:Color Structure=@n2.VertexColor
     Materials:Array of Material Interface Object References=@n2.Materials
     PreventMergingMaterials:Array of Material Interface Object References
     MaterialFeatures:Array of Atom Material Features Structures
     MeshDetailLevel:EAtomMergedMeshPartDetailLevel Enum=@n2.MeshDetailLevel
     bPreserveUVs:bool=@n2.bPreserveUVs
     bAllowMerging:bool=@n2.bAllowMerging
     bAllowSimplification:bool=True
     bAllowApproximation:bool=@n2.bAllowApproximation
     bIsOccluder:bool=True
     FilterLODLevel:int=-1
     MeshName:String
     InstanceSubsetID:int=@n2.InstanceSubsetID
     ApproximationType:Bitmask Integer=255
    ) -> (AtomMergedMeshPartInfo:Atom Merged Mesh Part Info Structure) {is=title K2Node_MakeStruct}
   )
  ]}
 CreatePartStaticMesh {nodes [
   n1 = CreatePartStaticMesh!() -> (
    then
    Part Info:"Atom Model Part Instance Info Structure (by ref)"
    MergedMeshPartInfo:"Atom Merged Mesh Part Info Structure (by ref)"
    Common Parts Geometry Map:Map of Strings to Atom Primitive Geometry LODs Structures
    Geometry Unique Name:"String (by ref)"
    StaticMeshPath:String
   ) {is=title K2Node_FunctionEntry}
   > Set New Geometry LODs(New Geometry LODs:Atom Primitive Geometry LODs Structure=Break Atom Merged Mesh Part Info(AtomMergedMeshPartInfo=@n1.MergedMeshPartInfo) -> (bHasConstantOverrideVertexColor=~, bPreserveUVs=~, bAllowMerging=~, bAllowSimplification=~, bAllowApproximation=~, bIsOccluder=~)) -> (Output_Get:Atom Primitive Geometry LODs Structure)
   > n2 = For Each Loop(Array:Array of Atom Common Part And Transform Structures=Break Atom Model Part Instance Info!(AtomModelPartInstanceInfo:"Atom Model Part Instance Info Structure (by ref)"=@n1) -> (Group:String, ParentGroup:String, CommonParts:Array of Atom Common Part And Transform Structures) {is=title K2Node_BreakStruct}.CommonParts) -> (Array Element:Atom Common Part And Transform Structure)
   n3 = CreateStaticMeshFromGeometry!(
    execute=@n2.Completed
    GeometryLODs:"Atom Primitive Geometry LODs Structure (by ref)"=Get New Geometry LODs() -> (New Geometry LODs:Atom Primitive Geometry LODs Structure)
    NaniteHiResLOD:Atom Primitive Geometry Object Reference
    MaterialArray:Array of Material Interface Object References=Make Array("[0]":Material Interface Object Reference) -> (Array:Array of Material Interface Object References)
    Subpath:String=@n1.StaticMeshPath
    StaticMeshName:String=@n1.Geometry Unique Name
    bPreserveExistingCollision:bool=false
    bPreserveExistingSockets:bool=false
    bGenerateMeshDistanceFields:bool=true
    TraceFlag:ECollisionTraceFlag Enum=CTF_UseSimpleAsComplex
    bSupportRaytracing:bool=true
   ) -> (then, ReturnValue:Static Mesh Object Reference)
   > Return Node!(execute, Static Mesh:Static Mesh Object Reference=@n3)
   n4 = Get Unique Name for Common Part Geometry LODs!(
    execute=@n2.LoopBody
    self
    PartInfo:Atom Model Part Instance Info Structure=Get Part Info() -> (Part Info:"Atom Model Part Instance Info Structure (by ref)")
    MergedMeshPartInfo:Atom Merged Mesh Part Info Structure=Get MergedMeshPartInfo() -> (MergedMeshPartInfo:"Atom Merged Mesh Part Info Structure (by ref)")
    CommonPartIndex:int=@n2.Array Index
   ) -> (then, Name:String) {is=title K2Node_CallFunction}
   > n5 = AppendAtomPrimitiveGeometryLODs!(
    execute
    BaseLODs:"Atom Primitive Geometry LODs Structure (by ref)"=Get New Geometry LODs()
    AppendLODs:"Atom Primitive Geometry LODs Structure (by ref)"=Find(
     TargetMap:Map of Strings to Atom Primitive Geometry LODs Structures=Get Common Parts Geometry Map() -> (Common Parts Geometry Map:Map of Strings to Atom Primitive Geometry LODs Structures)
     Key=@n4
    ) -> (Value:Atom Primitive Geometry LODs Structure)
    Transform:"Transform (by ref)"=Break Atom Common Part and Transform!(AtomCommonPartAndTransform:"Atom Common Part And Transform Structure (by ref)"=@n2) -> (Type:EAtomCommonPartType Enum, Transform:Transform, GeometryStyles:Array of Names) {is=title K2Node_BreakStruct}.Transform
   ) -> (then, ReturnValue:Atom Primitive Geometry LODs Structure)
   > Set New Geometry LODs(New Geometry LODs=@n5)
  ]}
 EventGraph {links [], nodes []}
 FinalizeCommonPartMergedMeshPartInfo {nodes [
   n1 = FinalizeCommonPartMergedMeshPartInfo!() -> (
    then
    CommonPartMergedMeshPartInfo:"Atom Merged Mesh Part Info Structure (by ref)"
    PrimitivePartInfo:"Atom Model Part Instance Info Structure (by ref)"
    CommonPartIndex:int
   ) {is=title K2Node_FunctionEntry}
   > Return Node!(execute, FinalizedPartInfo:Atom Merged Mesh Part Info Structure=@n1)
  ]}
 FinalizeMergedMeshPart {nodes [
   n1 = FinalizeMergedMeshPart!() -> (then, MergedMeshPartInfo:"Atom Merged Mesh Part Info Structure (by ref)") {is=title K2Node_FunctionEntry}
   > Return Node!(execute, Out MergedMeshPartInfo:Atom Merged Mesh Part Info Structure=@n1)
  ]}
 FinalizeMesh {nodes [
   n1 = FinalizeMesh!() -> (then, Static Mesh:Static Mesh Object Reference) {is=title K2Node_FunctionEntry}
   > CalculateMergedMeshScreenSizes!(
    execute
    LODSettings:"Atom Merged Mesh LODDistance Settings Structure (by ref)"=Make Atom Merged Mesh LODDistance Settings!(
     bOverrideLODScreenSizes:bool=true
     BaseLODScreenSize:float=0.25
     BaseLODScreenSizeScaling:float=1.05
     LODDistances:"Array of Float (single-precision)s"=Make Array!("[0]":float=1, "[1]":float=225, "[2]":float=300, "[3]":float=450, "[4]":float=700, "[5]":float=800, "[6]":float=1000, "[7]":float=1200) -> (Array:"Array of Float (single-precision)s")
    ) -> (AtomMergedMeshLODDistanceSettings:Atom Merged Mesh LODDistance Settings Structure) {is=title K2Node_MakeStruct}
    StaticMesh:Static Mesh Object Reference=@n1
   ) -> (then)
   > ApplyImportedSocketsToStaticMesh!(
    execute
    StaticMesh:Static Mesh Object Reference=@n1
    AtomModel:Atom Model Object Reference=GetAtomModel!(self) -> (Atom Model:Atom Model Object Reference)
   ) -> (then)
  ]}
 FinalizeStaticMesh {nodes [
   n1 = Inputs!() -> (
    Exec:"Exec (by ref)"
    AtomModel:"Atom Model Object Reference (by ref)"
    StaticMesh:"Static Mesh Object Reference (by ref)"
    CommonPartInstanceCache:Atom Common Part Instances Cache Structure
    MergedMeshName:String
    PrimitiveMeshOverrides:Map of Atom Model Part Guid Structures to Static Mesh Object References
   )
   FinalizeMesh!(execute=@n1, self, Static Mesh:Static Mesh Object Reference=@n1.StaticMesh) -> (then) {is=title K2Node_CallFunction}
   > n2 = Branch(Condition=GetIsProcessingPrimitive!() -> (Is Processing Primitive:bool))
   n3 = Self-Reference()
   > n4 = ApplyProcessorAssetImportDataToStaticMesh!(
    execute
    StaticMesh:Static Mesh Object Reference=@n1.StaticMesh
    Processor:Atom Model Processor Object Reference=@n3
    AtomModel:Atom Model Object Reference
   ) -> (then)
   Outputs!(Out=[@n4.then, @n8.then, @n7.then])
   AddAtomModelReferenceToStaticMesh!(
    execute=@n2.else
    StaticMesh:Static Mesh Object Reference=@n1.StaticMesh
    AtomModel:Atom Model Object Reference=@n1.AtomModel
    bSaveConnectivityState:bool=false
   ) -> (then)
   > ApplyProcessorAssetImportDataToStaticMesh(StaticMesh=@n1.StaticMesh, Processor=@n3, AtomModel=@n1.AtomModel)
   > n5 = Branch(Condition=Get Extract Common Parts() -> (Extract Common Parts:bool))
   > n6 = Set Common Part Instances Cache!(
    execute
    StaticMesh:Static Mesh Object Reference=@n1.StaticMesh
    CommonPartInstancesCache:"Atom Common Part Instances Cache Structure (by ref)"=@n1.CommonPartInstanceCache
   ) -> (then) {is=title K2Node_CallFunction}
   n7 = Branch(
    execute=[@n6.then, @n5.else]
    Condition=Contains!(
     TargetMap:Map of Atom Model Part Guid Structures to Static Mesh Object References=@n1.PrimitiveMeshOverrides
     Key:"Atom Model Part Guid Structure (by ref)"=To Atom Model Part Guid!(InString:String=@n1.MergedMeshName) -> (ReturnValue:Atom Model Part Guid Structure)
    ) -> (ReturnValue)
   )
   n8 = AddProcessorPrimitiveMeshOverrideToStaticMesh!(
    execute=@n7.else
    StaticMesh:Static Mesh Object Reference=@n1.StaticMesh
    PrimitiveMeshOverrides:Map of Atom Model Part Guid Structures to Static Mesh Object References=@n1.PrimitiveMeshOverrides
   ) -> (then)
  ]}
 Get Collision Geometry for Part {nodes [
   n1 = Get Collision Geometry for Part!() -> (
    then
    Geometry:Atom Primitive Geometry Object Reference
    CollisionType:EPrimitiveAutoCollisionType Enum=Box
    PartInfo:Atom Model Part Instance Info Structure
   ) {is=title K2Node_FunctionEntry}
   n2 = Break Atom Model Part Instance Info!(AtomModelPartInstanceInfo=@n1.PartInfo) -> (Primitive:Atom Primitive Object Reference, Transforms:Array of Transforms, Group, ParentGroup)
   > n3 = Branch(Condition=AND Boolean(
     A=Get bUseImportedCollisions!(self:Atom Primitive Object Reference=@n2) -> (bUseImportedCollisions:bool)
     B=Get bUseImportedPrimitiveCollision() -> (bUseImportedPrimitiveCollision:bool)
    ))
   > n4 = GetPrimitiveSourceCollision!(
    execute
    Settings:"Primitive Source Collision Geometry Settings Structure (by ref)"=Make Primitive Source Collision Geometry Settings!(
     Primitive:Atom Primitive Object Reference=@n2
     PrimitiveScale:float=GetDefaultPrimitiveScale!() -> (ReturnValue:float)
     BoneTransforms:Array of Transforms=@n2.Transforms
     CylinderColliderNumberOfSides:int=8
     bSplitNarrowColliders:bool=False
    ) -> (PrimitiveSourceCollisionGeometrySettings:Primitive Source Collision Geometry Settings Structure) {is=title K2Node_MakeStruct}
   ) -> (then, ReturnValue:Atom Primitive Collision Geometry Structure)
   > Return Node!(execute, Collision Geometry:Atom Primitive Collision Geometry Structure=@n4)
   n5 = GenerateCollisionGeometry!(execute=@n3.else, Geometry:Atom Primitive Geometry Object Reference=@n1, CollisionType:EPrimitiveAutoCollisionType Enum=@n1.CollisionType) -> (then, ReturnValue:Atom Primitive Collision Geometry Structure)
   > Return Node(Collision Geometry=@n5)
  ]}
 Get Unique Name for Common Part Geometry LODs {nodes [
   n1 = Get Unique Name for Common Part Geometry LODs!() -> (then, PartInfo:Atom Model Part Instance Info Structure, MergedMeshPartInfo:Atom Merged Mesh Part Info Structure, CommonPartIndex:int) {is=title K2Node_FunctionEntry}
   n2 = "Get (a ref)"(
    Array:Array of Atom Common Part And Transform Structures=Break Atom Model Part Instance Info!(AtomModelPartInstanceInfo=@n1) -> (Group, ParentGroup, CommonParts:Array of Atom Common Part And Transform Structures).CommonParts
    Dimension 1=@n1.CommonPartIndex
   ) -> (Output:"Atom Common Part And Transform Structure (by ref)")
   > Return Node!(
    execute
    Name:String=Enum to String!(Enumerator:EAtomCommonPartType Enum=Break Atom Common Part and Transform(AtomCommonPartAndTransform=@n2)) -> (ReturnValue:String)
   )
  ]}
 GetAtomModel {nodes [
   GetAtomModel() {is=title K2Node_FunctionEntry}
   > Return Node!(execute, Atom Model:Atom Model Object Reference=Get Model() -> (Model:Atom Model Object Reference))
  ]}
 GetBlueprintContainerName {nodes [
   n1 = GetBlueprintContainerName!() -> (then, Model:Atom Model Object Reference, Parts Collection:"Atom Model Parts Collection Structure (by ref)") {is=title K2Node_FunctionEntry}
   > Return Node!(
    execute
    Result:String=Append(
     A=BP_
     B=Break Atom Model Parts Collection!(AtomModelPartsCollection:"Atom Model Parts Collection Structure (by ref)"=@n1.Parts Collection) -> (Name:String, Pivot:Transform, Parts:Array of Atom Model Part Instance Info Structures) {is=title K2Node_BreakStruct}
    )
   )
  ]}
 GetBlueprintContainerPath {nodes [
   n1 = GetBlueprintContainerPath!() -> (then, Model:Atom Model Object Reference, Parts Collection:"Atom Model Parts Collection Structure (by ref)") {is=title K2Node_FunctionEntry}
   > n2 = Branch(Condition=GetIsProcessingPrimitive())
   > n3 = GetImporterSettings!(execute) -> (then, ReturnValue:Atom Model Importer Settings Object Reference)
   > Return Node(Result=Break Directory Path!(DirectoryPath:"Directory Path Structure (by ref)"=Get ProcessedPrimitiveContentFolder!(self:Atom Model Importer Settings Object Reference=@n3) -> (ProcessedPrimitiveContentFolder:Directory Path Structure)) -> (Path:String) {is=title K2Node_BreakStruct})
   Return Node(execute=@n2.else, Result=GetModelPath!(self:Atom Model Object Reference=@n1) -> (ReturnValue:String))
  ]}
 GetCollisionGeometryForCommonPart {nodes [
   GetCollisionGeometryForCommonPart!() -> (then, Geometry:Atom Primitive Geometry Object Reference) {is=title K2Node_FunctionEntry}
   > Return Node!(
    execute
    Collision Geometry:Atom Primitive Collision Geometry Structure=Make Atom Primitive Collision Geometry!() -> (AtomPrimitiveCollisionGeometry:Atom Primitive Collision Geometry Structure) {is=title K2Node_MakeStruct}
   )
  ]}
 GetCommonPartGeometry {nodes [
   n1 = GetCommonPartGeometry!() -> (
    then
    PartInfo:"Atom Model Part Instance Info Structure (by ref)"
    MergedMeshPartInfo:Atom Merged Mesh Part Info Structure
    CommonPartIndex:"Integer (by ref)"
   ) {is=title K2Node_FunctionEntry}
   n2 = Break Atom Model Part Instance Info(AtomModelPartInstanceInfo=@n1)
   > n3 = GetGeometryLODsForCommonPart!(
    execute
    self
    PartInfo:Atom Model Part Instance Info Structure=@n1
    MergedMeshPartInfo:Atom Merged Mesh Part Info Structure=@n1.MergedMeshPartInfo
    Common Part Type:EAtomCommonPartType Enum=Break Atom Common Part and Transform(AtomCommonPartAndTransform="Get (a ref)"(Array=@n2.CommonParts, Dimension 1=@n1.CommonPartIndex))
   ) -> (then, Geometry LODs:Atom Primitive Geometry LODs Structure)
   > Return Node!(execute, GeometryLODs:Atom Primitive Geometry LODs Structure=@n3)
  ]}
 GetCommonPartsGeometry {nodes [
   n1 = GetCommonPartsGeometry!() -> (
    then
    Part Info:"Atom Model Part Instance Info Structure (by ref)"
    MergedMeshPartInfo:Atom Merged Mesh Part Info Structure
    Common Part To Geometry LODs:Map of Strings to Atom Primitive Geometry LODs Structures
   ) {is=title K2Node_FunctionEntry}
   > n2 = For Each Loop(Array=Break Atom Model Part Instance Info!(AtomModelPartInstanceInfo=@n1) -> (Primitive:Atom Primitive Object Reference, Group, ParentGroup, CommonParts).CommonParts)
   > n3 = Get Unique Name for Common Part Geometry LODs(
    PartInfo=Get Part Info()
    MergedMeshPartInfo=Get MergedMeshPartInfo() -> (MergedMeshPartInfo:Atom Merged Mesh Part Info Structure)
    CommonPartIndex=@n2.Array Index
   ) {is=title K2Node_CallFunction}
   > n4 = Set CommonPartUniqueName(CommonPartUniqueName:String=@n3) -> (Output_Get:String)
   > n5 = Branch(Condition=Find(
     TargetMap=Get Common Part To Geometry LODs() -> (Common Part To Geometry LODs:Map of Strings to Atom Primitive Geometry LODs Structures)
     Key=@n4
    ).ReturnValue)
   n6 = GetCommonPartGeometry!(
    execute=@n5.else
    self
    PartInfo:"Atom Model Part Instance Info Structure (by ref)"=Get Part Info()
    MergedMeshPartInfo:Atom Merged Mesh Part Info Structure=Get MergedMeshPartInfo()
    CommonPartIndex:"Integer (by ref)"=@n2.Array Index
   ) -> (then, GeometryLODs:Atom Primitive Geometry LODs Structure) {is=title K2Node_CallFunction}
   > Set New Common Part LODs(New Common Part LODs:Atom Primitive Geometry LODs Structure=@n6) -> (Output_Get:Atom Primitive Geometry LODs Structure)
   > n7 = Branch(Condition=Get Generate Nanite Mesh() -> (Generate Nanite Mesh:bool))
   n8 = Get New Common Part LODs() -> (New Common Part LODs:Atom Primitive Geometry LODs Structure)
   n9 = GetCollisionGeometryForCommonPart!(
    execute=[@n7.then, @n11.then]
    self
    Geometry:Atom Primitive Geometry Object Reference="Get (a copy)"(Array:Array of Atom Primitive Geometry Object References=Break Atom Primitive Geometry LODs(AtomPrimitiveGeometryLODs=@n8)) -> (Output:Atom Primitive Geometry Object Reference)
   ) -> (then, Collision Geometry:Atom Primitive Collision Geometry Structure) {is=title K2Node_CallFunction}
   > n10 = Set members in Atom Primitive Geometry LODs!(
    execute:Exec
    StructRef:"Atom Primitive Geometry LODs Structure (by ref)"=Get New Common Part LODs()
    Collision:Atom Primitive Collision Geometry Structure=@n9
   ) -> (then:Exec, StructOut:"Atom Primitive Geometry LODs Structure (by ref)") {is=title K2Node_SetFieldsInStruct}
   > Add!(
    execute
    TargetMap:Map of Strings to Atom Primitive Geometry LODs Structures=Get Common Part To Geometry LODs()
    Key:"String (by ref)"=Get CommonPartUniqueName() -> (CommonPartUniqueName:String)
    Value:"Atom Primitive Geometry LODs Structure (by ref)"=@n10
   ) -> (then) {is=title K2Node_CallFunction}
   n11 = Set members in Atom Primitive Geometry LODs!(execute=@n7.else, StructRef=Get New Common Part LODs(), HiResNaniteGeomery:Atom Primitive Geometry Object Reference) -> (then, StructOut)
  ]}
 GetDecorationMaterial {nodes [
   n1 = GetDecorationMaterial!() -> (
    then
    Base Material:Material Interface Object Reference
    Decoration Infos:Array of Atom Model Part Decoration Info Structures
    New Material Path:String
   ) {is=title K2Node_FunctionEntry}
   > n2 = Sequence!(execute) -> (then_0, then_1, then_2:Exec)
   > Set Material Name(Material Name:String=GetObjectName(Object=@n1)) -> (Output_Get:String)
   > n3 = For Each Loop(Array:Array of Atom Model Part Decoration Info Structures=@n1.Decoration Infos) -> (Array Element:Atom Model Part Decoration Info Structure)
   n4 = Break Atom Model Part Decoration Info!(AtomModelPartDecorationInfo:"Atom Model Part Decoration Info Structure (by ref)"=@n3) -> (Texture:Texture Object Reference) {is=title K2Node_BreakStruct}
   > Set Material Name(Material Name=Append!(A=Get Material Name() -> (Material Name:String), B=_, C:String=GetObjectName(Object=@n4)) -> (ReturnValue))
   n5 = CreateMaterialInstanceFromBaseMaterial!(
    execute=@n2.then_1
    BaseMaterial:Material Interface Object Reference=@n1
    Subpath:String=@n1.New Material Path
    MaterialInstanceName:String=Get Material Name()
    bOverwriteExisting:bool=false
   ) -> (then, bAlreadyExists:bool, ReturnValue:Material Interface Object Reference)
   > n6 = Cast To MaterialInstanceConstant(Object=@n5.ReturnValue)
   > Set Material(Material:Material Instance Constant Object Reference=@n6) -> (Output_Get:Material Instance Constant Object Reference)
   > n7 = Branch(Condition=@n5)
   SetMaterialInstanceStaticSwitchParameterValue!(
    execute=@n7.else
    Instance:Material Instance Constant Object Reference=Get Material() -> (Material:Material Instance Constant Object Reference)
    ParameterName:Name=Uses Decorators
    Value:bool=true
    Association:EMaterialParameterAssociation Enum=GlobalParameter
    bUpdateMaterialInstance:bool=true
   ) -> (then, ReturnValue:bool)
   > n8 = For Each Loop(Array=@n1.Decoration Infos)
   > n9 = Branch(Condition="Equal (Integer)"(A=@n8.Array Index))
   n10 = Break Atom Model Part Decoration Info!(AtomModelPartDecorationInfo=@n8) -> (Material:Material Interface Object Reference, Texture, PrimitiveSurfaceName:String, PrimitiveSurfaceIndex:int)
   n11 = Get Material()
   > SetDecorationMaterialParameters!(
    execute
    self
    Material Instance:Material Instance Constant Object Reference=@n11
    Decoration Enable Param Name:String=Decorator ID 1
    Texture Param Name:String=Decorator 1
    UDIM Param Name:String=Decorator 1 UDIM
    Texture:Texture Object Reference=@n10.Texture
    UDIM:int=@n10.PrimitiveSurfaceIndex
   ) -> (then)
   n12 = Branch(execute=@n9.else, Condition="Equal (Integer)"(A=@n8.Array Index, B=1))
   > SetDecorationMaterialParameters(
    Material Instance=@n11
    Decoration Enable Param Name=Decorator ID 2
    Texture Param Name=Decorator 2
    UDIM Param Name=Decorator 2 UDIM
    Texture=@n10.Texture
    UDIM=@n10.PrimitiveSurfaceIndex
   )
   n13 = Branch(execute=@n12.else, Condition="Equal (Integer)"(A=@n8.Array Index, B=2))
   > SetDecorationMaterialParameters(
    Material Instance=@n11
    Decoration Enable Param Name=Decorator ID 3
    Texture Param Name=Decorator 3
    UDIM Param Name=Decorator 3 UDIM
    Texture=@n10.Texture
    UDIM=@n10.PrimitiveSurfaceIndex
   )
   Branch(execute=@n13.else, Condition="Equal (Integer)"(A=@n8.Array Index, B=3))
   > SetDecorationMaterialParameters(
    Material Instance=@n11
    Decoration Enable Param Name=Decorator ID 4
    Texture Param Name=Decorator 4
    UDIM Param Name=Decorator 4 UDIM
    Texture=@n10.Texture
    UDIM=@n10.PrimitiveSurfaceIndex
   )
   Return Node!(execute=@n2.then_2, Decoration Material:Material Interface Object Reference=Get Material())
  ]}
 GetDecorationMaterialPath {nodes [
   n1 = GetDecorationMaterialPath!() -> (then, Model Path:String) {is=title K2Node_FunctionEntry}
   > n2 = GetImporterSettings()
   > n3 = For Each Loop(Array:Array of Primitive Style Import Settings Structures=Get PrimitiveImportStyles!(self:Atom Model Importer Settings Object Reference=@n2) -> (PrimitiveImportStyles:Array of Primitive Style Import Settings Structures)) -> (Array Element:Primitive Style Import Settings Structure)
   Return Node!(execute=@n3.Completed, Material Path:String=Append!(A=@n1, B=/Materials) -> (ReturnValue))
   n4 = Break Primitive Style Import Settings!(PrimitiveStyleImportSettings:"Primitive Style Import Settings Structure (by ref)"=@n3) -> (
    Name:Name
    MeshTargetContentFolder:Directory Path Structure
    CombinedMeshTargetContentFolder:Directory Path Structure
    MaterialTargetContentFolder:Directory Path Structure
    TextureTargetContentFolder:Directory Path Structure
    StandardMaterial:Material Interface Soft Object Reference
    TransparentMaterial:Material Interface Soft Object Reference
    GlitterMaterial:Material Interface Soft Object Reference
    OpalescentMaterial:Material Interface Soft Object Reference
    StandardMaterialWithVertexColors:Material Interface Soft Object Reference
    TransparentMaterialWithVertexColors:Material Interface Soft Object Reference
    GlitterMaterialWithVertexColors:Material Interface Soft Object Reference
    OpalescentMaterialWithVertexColors:Material Interface Soft Object Reference
    StandardIndexColorMaterial:Material Interface Soft Object Reference
    MetalIndexColorMaterial:Material Interface Soft Object Reference
    TransparentIndexColorMaterial:Material Interface Soft Object Reference
    GlitterIndexColorMaterial:Material Interface Soft Object Reference
    OpalescentIndexColorMaterial:Material Interface Soft Object Reference
   ) {is=title K2Node_BreakStruct}
   > Branch(Condition="Equal (Name)"(A:String=Get Geometry style() -> (Geometry style:String), B=@n4))
   > Return Node(Material Path=Break Directory Path(DirectoryPath=@n4.MaterialTargetContentFolder))
  ]}
 GetDefaultMaterialForType {nodes [
   n1 = GetDefaultMaterialForType!() -> (
    then
    "Atom Material  Type":EAtomMaterialType Enum=Standard
    Part Info:"Atom Model Part Instance Info Structure (by ref)"
    Model:"Atom Model Object Reference (by ref)"
    Geometry:"Atom Primitive Geometry Object Reference (by ref)"
   ) {is=title K2Node_FunctionEntry}
   > n2 = Switch on EAtomMaterialType!(execute, Selection:EAtomMaterialType Enum=@n1) -> (None:Exec, Standard:Exec, Transparent:Exec, Glitter:Exec, Opalescent:Exec, Metallic:Exec) {is=title K2Node_SwitchEnum}
   Return Node!(execute=@n2.Glitter, Material:Material Interface Object Reference=/AtomFramework/Materials/Game_Qual/MI_Primitive_GlitVC)
   Return Node(execute=@n2.Metallic, Material=/AtomFramework/Materials/Game_Qual/MI_Primitive_MetalVC)
   Return Node(execute=@n2.Opalescent, Material=/AtomFramework/Materials/Game_Qual/MI_Primitive_OpalVC)
   Return Node(execute=@n2.Standard, Material=/AtomFramework/Materials/Game_Qual/MI_Primitive_StandardVC)
   Return Node(execute=@n2.Transparent, Material=/AtomFramework/Materials/Game_Qual/MI_Primitive_TransVC)
  ]}
 GetGeometryLODsForCommonPart {nodes [
   n1 = GetGeometryLODsForCommonPart!() -> (
    then
    PartInfo:Atom Model Part Instance Info Structure
    MergedMeshPartInfo:Atom Merged Mesh Part Info Structure
    Common Part Type:EAtomCommonPartType Enum=None
   ) {is=title K2Node_FunctionEntry}
   > n2 = Sequence!(execute) -> (then_0, then_1)
   n3 = Get Common Part Style() -> (Common Part Style:String)
   > n4 = For Loop(LastIndex=int - int(A=GetNumberOfCommonPartLODs!(ExportStyleName:String=@n3, CommonPartType:EAtomCommonPartType Enum=@n1.Common Part Type, CommonPartsMeshPath:String) -> (ReturnValue:int=0)))
   > n5 = CreateAtomGeometryFromCommonPart!(
    execute
    ExportStyleName:String=Get Common Part Style()
    CommonPartType:EAtomCommonPartType Enum=@n1.Common Part Type
    LodIndex:int=@n4
    CommonPartsMeshPath:String
   ) -> (then, ReturnValue:Atom Primitive Geometry Object Reference)
   > n6 = BakeScale!(execute, self:Atom Primitive Geometry Object Reference=@n5, Scale:float=Get Common Part Scale() -> (Common Part Scale:double)) -> (then, ReturnValue:Atom Primitive Geometry Object Reference)
   > n7 = SetMaterialName!(execute, self:Atom Primitive Geometry Object Reference=@n6, Name:String=Default, PolygonGroupIndex:int=0) -> (then, ReturnValue:Atom Primitive Geometry Object Reference)
   > Add(
    TargetArray:Array of Atom Primitive Geometry Object References=Get Common Part Geometry LODs() -> (Common Part Geometry LODs:Array of Atom Primitive Geometry Object References)
    NewItem:"Atom Primitive Geometry Object Reference (by ref)"=@n7
   ) {is=title K2Node_CallArrayFunction}
   n8 = Get Common Part Geometry LODs()
   n9 = DuplicateGeometry!(execute=@n2.then_1, self:Atom Primitive Geometry Object Reference="Get (a copy)"(Array=@n8)) -> (then, ReturnValue:Atom Primitive Geometry Object Reference)
   > Return Node!(
    execute
    Geometry LODs:Atom Primitive Geometry LODs Structure=Make Atom Primitive Geometry LODs(
     GeometryLODs=@n8
     HiResNaniteGeomery=@n9
     HiResNaniteGeomeryMaterials=(MaterialSlotName=,MaterialFeatures=)
     Collision=(AggGeom=(SphereElems=,BoxElems=,SphylElems=,ConvexElems=,TaperedCapsuleElems=,LevelSetElems=,SkinnedLevelSetElems=,MLLevelSetElems=,SkinnedTriangleMeshElems=))
     DetailedCollision=(AggGeom=(SphereElems=,BoxElems=,SphylElems=,ConvexElems=,TaperedCapsuleElems=,LevelSetElems=,SkinnedLevelSetElems=,MLLevelSetElems=,SkinnedTriangleMeshElems=))
    )
   )
  ]}
 GetGeometryLODsForPart {nodes [
   n1 = GetGeometryLODsForPart!() -> (then, Part Info:"Atom Model Part Instance Info Structure (by ref)", Model:Atom Model Object Reference) {is=title K2Node_FunctionEntry}
   > n2 = Set Part Instance Info(Part Instance Info:Atom Model Part Instance Info Structure=@n1) -> (Output_Get:Atom Model Part Instance Info Structure)
   > n3 = Branch(Condition=IsFlexElement!(self:Atom Primitive Object Reference=Break Atom Model Part Instance Info!(AtomModelPartInstanceInfo=@n2) -> (Primitive, Group, ParentGroup)) -> (ReturnValue:bool))
   n4 = Break Atom Model Part Instance Info!(AtomModelPartInstanceInfo=Get Part Instance Info() -> (Part Instance Info:Atom Model Part Instance Info Structure)) -> (
    Primitive
    Transforms:Array of Transforms
    Group
    ParentGroup
    CommonParts:Array of Atom Common Part And Transform Structures
    bIsUndersideVisible:bool=~
   )
   > Set FlexRootTransformInverse(FlexRootTransformInverse:Transform=InvertTransform!(T:"Transform (by ref)"="Get (a copy)"(Array:Array of Transforms=@n4.Transforms) -> (Output:Transform)) -> (ReturnValue:Transform)) -> (Output_Get:Transform)
   > n5 = For Each Loop(Array:Array of Transforms=@n4.Transforms) -> (Array Element:Transform)
   n6 = Get Geometry style()
   n7 = Get Fallback Geometry style() -> (Fallback Geometry style:String)
   n8 = Set LOD 0(
    execute=@n5.Completed
    LOD 0:Atom Primitive Geometry Object Reference=GetGeometrySection!(
     self:Atom Primitive Geometry Container Object Reference=GetGeometryContainerForExportStyle!(
      self:Atom Primitive Object Reference=@n4
      ExportStyleName:String=@n6
      FallbackExportStyleName:String=@n7
      LODIndex:int=0
      bUseFallbacks:bool=true
     ) -> (ReturnValue:Atom Primitive Geometry Container Object Reference)
     PrimitiveGeometrySectionFlag:Bitmask Integer=35
    ) -> (ReturnValue:Atom Primitive Geometry Object Reference)
   ) -> (Output_Get:Atom Primitive Geometry Object Reference)
   > n9 = BakeScale(self=@n8, Scale=GetDefaultPrimitiveScale())
   > BakeTransforms!(
    execute
    self:Atom Primitive Geometry Object Reference=@n9
    Transforms:Array of Transform 3f Structures=Get FlexTransforms() -> (FlexTransforms:Array of Transforms)
   ) -> (then, ReturnValue:Atom Primitive Geometry Object Reference)
   n10 = Get LOD 0() -> (LOD 0:Atom Primitive Geometry Object Reference)
   > n11 = DuplicateGeometry(self=@n10)
   > Return Node(Geometry LODs=Make Atom Primitive Geometry LODs(
     GeometryLODs=Make Array!("[0]":Atom Primitive Geometry Object Reference=@n10, "[1]":Atom Primitive Geometry Object Reference=@n10) -> (Array:Array of Atom Primitive Geometry Object References)
     HiResNaniteGeomery=@n11
     HiResNaniteGeomeryMaterials=(MaterialSlotName=,MaterialFeatures=)
     Collision=(AggGeom=(SphereElems=,BoxElems=,SphylElems=,ConvexElems=,TaperedCapsuleElems=,LevelSetElems=,SkinnedLevelSetElems=,MLLevelSetElems=,SkinnedTriangleMeshElems=))
     DetailedCollision=(AggGeom=(SphereElems=,BoxElems=,SphylElems=,ConvexElems=,TaperedCapsuleElems=,LevelSetElems=,SkinnedLevelSetElems=,MLLevelSetElems=,SkinnedTriangleMeshElems=))
    ))
   Add(
    execute=@n5.LoopBody
    TargetArray:Array of Transforms=Get FlexTransforms()
    NewItem:"Transform (by ref)"=ComposeTransforms!(A:"Transform (by ref)"=@n5, B:"Transform (by ref)"=Get FlexRootTransformInverse() -> (FlexRootTransformInverse:Transform)) -> (ReturnValue:Transform)
   )
   n12 = Sequence!(execute=@n3.else) -> (then_0, then_1, then_2:Exec)
   n13 = Break Atom Model Part Instance Info!(AtomModelPartInstanceInfo=Get Part Instance Info()) -> (Primitive, Group, ParentGroup, CommonParts, bIsUndersideVisible=~)
   n14 = Get Geometry style()
   n15 = Get Fallback Geometry style()
   > n16 = Set LOD 0(LOD 0=GetGeometrySection(
     self=GetGeometryContainerForExportStyle(self=@n13, ExportStyleName=@n14, FallbackExportStyleName=@n15, LODIndex=0, bUseFallbacks=true)
     PrimitiveGeometrySectionFlag=35
    ))
   > n17 = BevelGeometry!(
    execute
    GeometryToBevel:Atom Primitive Geometry Object Reference=@n16
    BevelWidth:float=0.0125
    AngleThresholdInDegrees:float=75
    bIgnoreUVSeams:bool=true
    bOnlyBevelBlockLikeGeometry:bool=true
    bUseNewAlgorithm:bool=false
   ) -> (then, ReturnValue:Atom Primitive Geometry Object Reference)
   > n18 = BakeScale(self=@n17, Scale=GetDefaultPrimitiveScale())
   > SetMaterialName(self=@n18, Name=Default, PolygonGroupIndex=0)
   n19 = Get Geometry style()
   n20 = Get Fallback Geometry style()
   n21 = Set LOD 1(
    execute=@n12.then_1
    LOD 1:Atom Primitive Geometry Object Reference=GetGeometrySection(
     self=GetGeometryContainerForExportStyle(self=@n13, ExportStyleName=@n19, FallbackExportStyleName=@n20, LODIndex=0, bUseFallbacks=true)
     PrimitiveGeometrySectionFlag=35
    )
   ) -> (Output_Get:Atom Primitive Geometry Object Reference)
   > n22 = BakeScale(self=@n21, Scale=GetDefaultPrimitiveScale())
   > SetMaterialName(self=@n22, Name=Default, PolygonGroupIndex=0)
   n23 = Get LOD 0()
   n24 = DuplicateGeometry(execute=@n12.then_2, self=@n23)
   n25 = Get LOD 1() -> (LOD 1:Atom Primitive Geometry Object Reference)
   > Return Node(Geometry LODs=Make Atom Primitive Geometry LODs(
     GeometryLODs=Make Array("[0]"=@n23, "[1]"=@n25)
     HiResNaniteGeomery=@n24
     HiResNaniteGeomeryMaterials=(MaterialSlotName=,MaterialFeatures=)
     Collision=(AggGeom=(SphereElems=,BoxElems=,SphylElems=,ConvexElems=,TaperedCapsuleElems=,LevelSetElems=,SkinnedLevelSetElems=,MLLevelSetElems=,SkinnedTriangleMeshElems=))
     DetailedCollision=(AggGeom=(SphereElems=,BoxElems=,SphylElems=,ConvexElems=,TaperedCapsuleElems=,LevelSetElems=,SkinnedLevelSetElems=,MLLevelSetElems=,SkinnedTriangleMeshElems=))
    ))
  ]}
 GetGeometryLODsForUncommonPart {nodes [
   n1 = GetGeometryLODsForUncommonPart!() -> (then, Geometry:Atom Primitive Geometry Object Reference) {is=title K2Node_FunctionEntry}
   > Return Node(Geometry LODs=Make Atom Primitive Geometry LODs(
     GeometryLODs=Make Array("[0]"=@n1, "[1]"=@n1)
     HiResNaniteGeomery=@n1
     HiResNaniteGeomeryMaterials=(MaterialSlotName=,MaterialFeatures=)
     Collision=Make Atom Primitive Collision Geometry()
     DetailedCollision=(AggGeom=(SphereElems=,BoxElems=,SphylElems=,ConvexElems=,TaperedCapsuleElems=,LevelSetElems=,SkinnedLevelSetElems=,MLLevelSetElems=,SkinnedTriangleMeshElems=))
    ))
  ]}
 GetHasOnlyNaniteCompatibleMaterials {nodes [
   n1 = GetHasOnlyNaniteCompatibleMaterials!() -> (then, MergedMeshPartInfo:"Atom Merged Mesh Part Info Structure (by ref)") {is=title K2Node_FunctionEntry}
   > n2 = For Each Loop(Array:Array of Material Interface Object References=Break Atom Merged Mesh Part Info(AtomMergedMeshPartInfo=@n1) -> (bHasConstantOverrideVertexColor=~, bPreserveUVs=~, bAllowMerging=~, bAllowSimplification=~, bAllowApproximation=~, bIsOccluder=~).Materials) -> (Array Element:Material Interface Object Reference)
   Return Node!(execute=@n2.Completed, bOnlyHasOpaque:bool=true)
   n3 = GetBlendMode!(self:Material Interface Object Reference=@n2) -> (ReturnValue:EBlendMode Enum=BLEND_Opaque)
   > n4 = Branch(Condition=OR Boolean(A="Equal (Enum)"(A:EBlendMode Enum=@n3, B:EBlendMode Enum=BLEND_Opaque), B="Equal (Enum)"(A=@n3, B=BLEND_Masked)))
   Return Node(execute=@n4.else, bOnlyHasOpaque=false)
  ]}
 GetIsProcessingPrimitive {nodes [
   GetIsProcessingPrimitive() {is=title K2Node_FunctionEntry}
   > Return Node!(
    execute
    Is Processing Primitive:bool=Parent: GetIsProcessingPrimitive!() -> (Is Processing Primitive:bool) {is=title K2Node_CallParentFunction}
   )
  ]}
 GetMaterialsForCommonPart {nodes [
   n1 = GetMaterialsForCommonPart!() -> (
    then
    Common Part Geometry LODs:"Atom Primitive Geometry LODs Structure (by ref)"
    Shell Merged Mesh Part Info:"Atom Merged Mesh Part Info Structure (by ref)"
    PartInfo:"Atom Model Part Instance Info Structure (by ref)"
    CommonPartIndex:int
   ) {is=title K2Node_FunctionEntry}
   n2 = Break Atom Merged Mesh Part Info!(AtomMergedMeshPartInfo=@n1.Shell Merged Mesh Part Info) -> (
    bHasConstantOverrideVertexColor=~
    Materials
    PreventMergingMaterials
    MaterialFeatures
    bAllowSimplification=~
    bIsOccluder=~
    FilterLODLevel
    MeshName
   )
   > Return Node!(
    execute
    Common Part Materials:Array of Material Interface Object References=Make Array!("[0]":Material Interface Object Reference="Get (a copy)"(Array:Array of Material Interface Object References=@n2.Materials) -> (Output:Material Interface Object Reference)) -> (Array:Array of Material Interface Object References)
   )
  ]}
 GetMaterialsForPart {nodes [
   n1 = GetMaterialsForPart!() -> (
    then
    Part Info:"Atom Model Part Instance Info Structure (by ref)"
    Geometry:Atom Primitive Geometry Object Reference
    Model:Atom Model Object Reference
   ) {is=title K2Node_FunctionEntry}
   > n2 = Sequence!(execute) -> (then_0, then_1)
   > Clear(TargetArray:Array of Material Interface Object References=Get Materials() -> (Materials:Array of Material Interface Object References))
   n3 = Break Atom Model Part Instance Info!(AtomModelPartInstanceInfo=@n1) -> (Decorations:Array of Atom Model Part Decoration Info Structures, Colors:Array of Atom Model Part Color Info Structures, Group, ParentGroup)
   > n4 = GetDefaultMaterialForType!(
    execute
    self
    "Atom Material  Type":EAtomMaterialType Enum=Break Atom Model Part Color Info!(AtomModelPartColorInfo:"Atom Model Part Color Info Structure (by ref)"="Get (a ref)"(Array:Array of Atom Model Part Color Info Structures=@n3.Colors, Dimension 1=0) -> (Output:"Atom Model Part Color Info Structure (by ref)")) -> (Color:Color Structure, ColorId:int, MaterialType:EAtomMaterialType Enum) {is=title K2Node_BreakStruct}.MaterialType
    Part Info:"Atom Model Part Instance Info Structure (by ref)"=@n1
    Model:"Atom Model Object Reference (by ref)"=@n1.Model
    Geometry:"Atom Primitive Geometry Object Reference (by ref)"=@n1.Geometry
   ) -> (then, Material:Material Interface Object Reference) {is=title K2Node_CallFunction}
   n5 = Get Support Decorations() -> (Support Decorations:bool)
   > n6 = Branch(Condition=OR Boolean(A=Is Empty!(TargetArray:Array of Atom Model Part Decoration Info Structures=@n3) -> (ReturnValue:bool), B=NOT Boolean(A=@n5)))
   > Add(TargetArray:Array of Material Interface Object References=Get Materials(), NewItem:"Material Interface Object Reference (by ref)"=@n4)
   n7 = GetDecorationMaterialPath!(execute=@n6.else, self, Model Path:String=GetModelPath(self=@n1.Model)) -> (then, Material Path:String) {is=title K2Node_CallFunction}
   > n8 = GetDecorationMaterial!(
    execute
    self
    Base Material:Material Interface Object Reference=@n4
    Decoration Infos:Array of Atom Model Part Decoration Info Structures=@n3
    New Material Path:String=@n7
   ) -> (then, Decoration Material:Material Interface Object Reference) {is=title K2Node_CallFunction}
   > Add(TargetArray=Get Materials(), NewItem=@n8)
   Return Node!(execute=@n2.then_1, Material:Array of Material Interface Object References=Get Materials())
  ]}
 GetMaterialsForPartWithColorChangeIndex {nodes [
   n1 = GetMaterialsForPartWithColorChangeIndex!() -> (
    then
    Part Info:"Atom Model Part Instance Info Structure (by ref)"
    Geometry:Atom Primitive Geometry Object Reference
    Model:Atom Model Object Reference
    Color Change Index:"Integer (by ref)"
   ) {is=title K2Node_FunctionEntry}
   > n2 = Sequence()
   > Clear(TargetArray=Get Materials())
   n3 = Break Atom Model Part Instance Info(AtomModelPartInstanceInfo=@n1)
   > n4 = GetDefaultMaterialForType(
    "Atom Material  Type"=Break Atom Model Part Color Info(AtomModelPartColorInfo="Get (a ref)"(Array=@n3.Colors, Dimension 1=@n1.Color Change Index)).MaterialType
    Part Info=@n1
    Model=@n1.Model
    Geometry=@n1.Geometry
   )
   n5 = Get Support Decorations()
   > n6 = Branch(Condition=OR Boolean(A=Is Empty(TargetArray=@n3), B=NOT Boolean(A=@n5)))
   > Add(TargetArray=Get Materials(), NewItem=@n4)
   n7 = GetDecorationMaterialPath(execute=@n6.else, Model Path=GetModelPath(self=@n1.Model))
   > n8 = GetDecorationMaterial(Base Material=@n4, Decoration Infos=@n3, New Material Path=@n7)
   > Add(TargetArray=Get Materials(), NewItem=@n8)
   Return Node(execute=@n2.then_1, Material=Get Materials())
  ]}
 GetMeshMergingSettings {nodes [
   n1 = GetMeshMergingSettings!() -> (then, AtomModel:Atom Model Object Reference) {is=title K2Node_FunctionEntry}
   > Return Node!(
    execute
    Regular Mesh Merging Settings:Atom Mesh Merging Settings Structure=Get Mesh Merging Settings() -> (Mesh Merging Settings:Atom Mesh Merging Settings Structure)
   )
   n2 = Branch(Condition=IsValid(Object=@n1))
   > n3 = Sequence!(execute) -> (then_0, then_1, then_2:Exec)
   n4 = Break Atom Model Asset Settings!(AtomModelAssetSettings:"Atom Model Asset Settings Structure (by ref)"=Get Settings!(self:Atom Model Object Reference=@n1) -> (Settings:Atom Model Asset Settings Structure)) -> (
    Scale:float=~
    CreateRigidElementComponents:bool=~
    OptimizationSettings:Atom Model Geometry Optimization Settings Structure
    Tags:Atom Model Tags Structure
   ) {is=title K2Node_BreakStruct}
   n5 = Break Atom Model Geometry Optimization Settings!(AtomModelGeometryOptimizationSettings:"Atom Model Geometry Optimization Settings Structure (by ref)"=@n4.OptimizationSettings) -> (bEnforceLODBudgets:bool=~, bUseTagBudget:bool=~, TriangleBudget:int, SimplifyBaseTolerance:double=~, OptimizeBaseTriCost:double=~) {is=title K2Node_BreakStruct}
   n6 = Break Atom Mesh Merging Settings!(AtomMeshMergingSettings:"Atom Mesh Merging Settings Structure (by ref)"=Get Mesh Merging Settings()) -> (
    SimplifiedLODsSettings:Atom Mesh Merging Simplified LODs Settings Structure
    ApproximateLODsSettings:Atom Mesh Merging Approximate LODs Settings Structure
    BudgetSettings:Atom Mesh Merging Budget Settings Structure
   ) {is=title K2Node_BreakStruct}
   > Set members in Atom Mesh Merging Simplified LODs Settings!(
    execute:Exec
    StructRef:"Atom Mesh Merging Simplified LODs Settings Structure (by ref)"=@n6
    SimplifyBaseTolerance:float=@n5.SimplifyBaseTolerance
   ) -> (then:Exec, StructOut:"Atom Mesh Merging Simplified LODs Settings Structure (by ref)") {is=title K2Node_SetFieldsInStruct}
   Set members in Atom Mesh Merging Approximate LODs Settings!(
    execute:Exec=@n3.then_1
    StructRef:"Atom Mesh Merging Approximate LODs Settings Structure (by ref)"=@n6.ApproximateLODsSettings
    OptimizeBaseTriCost:float=@n5.OptimizeBaseTriCost
   ) -> (then:Exec, StructOut:"Atom Mesh Merging Approximate LODs Settings Structure (by ref)") {is=title K2Node_SetFieldsInStruct}
   n7 = Set members in Atom Mesh Merging Budget Settings!(
    execute:Exec=@n3.then_2
    StructRef:"Atom Mesh Merging Budget Settings Structure (by ref)"=@n6.BudgetSettings
    bEnforceLODBudgets:bool=@n5
    TriangleBudget:int=@n5.TriangleBudget
   ) -> (then:Exec, StructOut:"Atom Mesh Merging Budget Settings Structure (by ref)") {is=title K2Node_SetFieldsInStruct}
   > n8 = Branch(Condition=@n5.bUseTagBudget)
   > n9 = For Each Loop with Break!(
    Exec:Exec
    Array:Array of Names=Break Atom Model Tags!(AtomModelTags:"Atom Model Tags Structure (by ref)"=@n4.Tags) -> (Tags:Array of Names) {is=title K2Node_BreakStruct}
    Break:Exec=@n11.then
   ) -> (LoopBody:Exec, Array Element:Name, Array Index:int, Completed:Exec)
   > n10 = Get Atom Model Tag Value!(execute, Tag:Name=@n9) -> (then, bFoundTag:bool, ReturnValue:int=0) {is=title K2Node_CallFunction}
   > Branch(Condition=@n10)
   > n11 = Set members in Atom Mesh Merging Budget Settings!(execute, StructRef=@n7, TriangleBudget=@n10.ReturnValue) -> (then, StructOut)
   Return Node(execute=[@n9.Completed, @n8.else], Regular Mesh Merging Settings=Get Mesh Merging Settings())
   Return Node(execute=@n2.else, Regular Mesh Merging Settings=Get Mesh Merging Settings())
  ]}
 GetNaniteMeshMergingSettings {nodes [
   GetNaniteMeshMergingSettings!() -> (then, AtomModel:Atom Model Object Reference) {is=title K2Node_FunctionEntry}
   > Return Node!(execute, Nanite Mesh Merging Settings:Atom Mesh Merging Settings Structure=Get Mesh Merging Settings())
  ]}
 GetOverrideMeshMergingSettings {nodes [
   GetOverrideMeshMergingSettings!() -> (then, Default Settings:"Atom Mesh Merging Settings Structure (by ref)") {is=title K2Node_FunctionEntry}
   > Return Node!(
    execute
    Mesh Merging Settings Map:Map of Strings to Atom Mesh Merging Settings Structures=Get Mesh Merging Settings() -> (Mesh Merging Settings:Map of Strings to Atom Mesh Merging Settings Structures)
   )
  ]}
 GetPartStaticMeshPath {nodes [
   n1 = GetPartStaticMeshPath!() -> (
    then
    Atom Model:Atom Model Object Reference
    Parts Collection:"Atom Model Parts Collection Structure (by ref)"
    Part Info:"Atom Merged Mesh Part Info Structure (by ref)"
   ) {is=title K2Node_FunctionEntry}
   n2 = Get Geometry Static Mesh Path() -> (Geometry Static Mesh Path:String)
   > n3 = Branch(Condition=IsEmpty(InString=@n2))
   > Return Node!(execute, PartStaticMeshPath:String=GetModelPath(self=@n1))
   Return Node(execute=@n3.else, PartStaticMeshPath=@n2)
  ]}
 GetPartsCollectionName {nodes [
   GetPartsCollectionName() {is=title K2Node_FunctionEntry}
   > Return Node!(execute, Parts Collection Name:String=Get Parts Collection Name() -> (Parts Collection Name:String))
  ]}
 GetPivotForPartsCollection {nodes [
   n1 = GetPivotForPartsCollection!() -> (then, Parts Collection:Atom Model Parts Collection Structure) {is=title K2Node_FunctionEntry}
   > Return Node!(execute, Pivot:Transform=Break Atom Model Parts Collection(AtomModelPartsCollection=@n1).Pivot)
  ]}
 GetPreserveUVsForPart {nodes [
   n1 = GetPreserveUVsForPart!() -> (
    then
    Atom Model Part Info:"Atom Model Part Instance Info Structure (by ref)"
    Merged Mesh Part Info:"Atom Merged Mesh Part Info Structure (by ref)"
   ) {is=title K2Node_FunctionEntry}
   n2 = Is Not Empty!(TargetArray:Array of Atom Model Part Decoration Info Structures=Break Atom Model Part Instance Info!(AtomModelPartInstanceInfo=@n1) -> (Decorations, Group, ParentGroup)) -> (ReturnValue:bool)
   n3 = Get Support Decorations()
   > Return Node!(
    execute
    PreserveUVs:bool=OR Boolean(
     A=Break Atom Merged Mesh Part Info!(AtomMergedMeshPartInfo=@n1.Merged Mesh Part Info) -> (
      GeometryLODs:Atom Primitive Geometry LODs Structure
      Transform:Transform
      bHasConstantOverrideVertexColor=~
      VertexColor:Color Structure
      Materials
      PreventMergingMaterials
      MaterialFeatures
      MeshDetailLevel:EAtomMergedMeshPartDetailLevel Enum
      bPreserveUVs:bool=~
      bAllowMerging:bool=~
      bAllowSimplification=~
      bAllowApproximation:bool=~
      bIsOccluder=~
      FilterLODLevel
      MeshName
      InstanceSubsetID:int
      ApproximationType:Bitmask Integer
     ).bPreserveUVs
     B=AND Boolean(A=@n3, B=@n2)
    )
   )
  ]}
 GetPrimitiveGeometry {nodes [
   n1 = GetPrimitiveGeometry!() -> (
    then
    Atom Model Part Info:"Atom Model Part Instance Info Structure (by ref)"
    Geometry Map:Map of Strings to Atom Primitive Geometry LODs Structures
    Atom Model:Atom Model Object Reference
   ) {is=title K2Node_FunctionEntry}
   > n2 = GetUniqueNameForGeometryLODs!(execute, self, Part Info:Atom Model Part Instance Info Structure=@n1) -> (then, Name:String)
   > Set Unique Name(Unique Name:String=@n2) -> (Output_Get:String)
   n3 = Find(TargetMap=@n1.Geometry Map, Key=Get Unique Name() -> (Unique Name:String))
   > n4 = Branch(Condition=@n3.ReturnValue)
   > Return Node!(execute, Geometry LODs:Atom Primitive Geometry LODs Structure=@n3, Geometry Unique Name:String=Get Unique Name())
   n5 = GetGeometryLODsForPart!(execute=@n4.else, self, Part Info:"Atom Model Part Instance Info Structure (by ref)"=@n1, Model:Atom Model Object Reference) -> (then, Geometry LODs:Atom Primitive Geometry LODs Structure) {is=title K2Node_CallFunction}
   > Set New Geometry LODs(New Geometry LODs=@n5)
   n6 = Get New Geometry LODs()
   > n7 = Get Collision Geometry for Part!(
    execute
    self
    Geometry:Atom Primitive Geometry Object Reference="Get (a copy)"(Array:Array of Atom Primitive Geometry Object References=Break Atom Primitive Geometry LODs!(AtomPrimitiveGeometryLODs=@n6) -> (GeometryLODs, GeometryLODsMaterials, HiResNaniteGeomery, HiResNaniteGeomeryMaterials, DetailedCollision)) -> (Output:Atom Primitive Geometry Object Reference)
    CollisionType:EPrimitiveAutoCollisionType Enum=Get Auto Primitive Collision Type() -> (Auto Primitive Collision Type:EPrimitiveAutoCollisionType Enum=Box)
    PartInfo:Atom Model Part Instance Info Structure=@n1
   ) -> (then, Collision Geometry:Atom Primitive Collision Geometry Structure) {is=title K2Node_CallFunction}
   > Set members in Atom Primitive Geometry LODs!(execute, StructRef=Get New Geometry LODs(), Collision:Atom Primitive Collision Geometry Structure=@n7) -> (then, StructOut)
   n8 = Get New Geometry LODs()
   n9 = Get Unique Name()
   > Add(TargetMap=@n1.Geometry Map, Key=@n9, Value=@n8) {is=title K2Node_CallFunction}
   > Return Node(Geometry LODs=@n8, Geometry Unique Name=@n9)
  ]}
 GetStaticMeshName {nodes [
   n1 = GetStaticMeshName!() -> (then, Model:Atom Model Object Reference, Parts Collection:"Atom Model Parts Collection Structure (by ref)") {is=title K2Node_FunctionEntry}
   > n2 = Branch(Condition=AND Boolean(A=GetIsProcessingPrimitive() {is=title K2Node_CallFunction}, B=Get Support Decorations()))
   n3 = Append(A=SM_, B=Break Atom Model Parts Collection(AtomModelPartsCollection=@n1.Parts Collection))
   > Return Node!(execute, Result:String=Append(A=@n3, B=_DECO))
   Return Node(execute=@n2.else, Result=@n3)
  ]}
 GetStaticMeshPath {nodes [
   n1 = GetStaticMeshPath!() -> (then, Model:Atom Model Object Reference, Parts Collection:"Atom Model Parts Collection Structure (by ref)") {is=title K2Node_FunctionEntry}
   > n2 = Branch(Condition=GetIsProcessingPrimitive())
   > n3 = GetImporterSettings()
   > Return Node(Result=Break Directory Path(DirectoryPath=Get ProcessedPrimitiveContentFolder(self=@n3)))
   Return Node(execute=@n2.else, Result=GetModelPath(self=@n1))
  ]}
 GetTransparentMeshMergingSettings {nodes [
   GetTransparentMeshMergingSettings() {is=title K2Node_FunctionEntry}
   > Return Node!(
    execute
    Mesh Merging Settings:Atom Mesh Merging Settings Structure=Get Tranparent Mesh Merging Settings() -> (Tranparent Mesh Merging Settings:Atom Mesh Merging Settings Structure)
   )
  ]}
 GetTransparentMeshName {nodes [
   n1 = GetTransparentMeshName!() -> (then, AtomModel:Atom Model Object Reference, Parts Collection:"Atom Model Parts Collection Structure (by ref)") {is=title K2Node_FunctionEntry}
   > n2 = GetStaticMeshName!(execute, self, Model:Atom Model Object Reference=@n1, Parts Collection:"Atom Model Parts Collection Structure (by ref)"=@n1.Parts Collection) -> (then, Result:String) {is=title K2Node_CallFunction}
   > Return Node!(execute, TransparentMeshName:String=Append(A=@n2, B=_Trans))
  ]}
 GetUniqueNameForGeometryLODs {nodes [
   n1 = GetUniqueNameForGeometryLODs!() -> (then, Part Info:Atom Model Part Instance Info Structure) {is=title K2Node_FunctionEntry}
   n2 = Break Atom Model Part Instance Info!(AtomModelPartInstanceInfo=@n1) -> (PartGuid:Atom Model Part Guid Structure, PartId:int, Primitive:Atom Primitive Object Reference, Group, ParentGroup)
   > n3 = Branch(Condition=IsFlexElement(self=@n2.Primitive))
   n4 = Break Atom Model Part Guid!(AtomModelPartGuid:"Atom Model Part Guid Structure (by ref)"=@n2) -> (Guid:Guid Structure, PartIndex:int) {is=title K2Node_BreakStruct}
   n5 = Append(A=m, B="To String (Integer)"(InInt=@n2.PartId))
   > Return Node!(
    execute
    Name:String=Append!(
     A=@n5
     B=_
     C:String="To String (Guid)"!(InGuid:"Guid Structure (by ref)"=@n4) -> (ReturnValue:String)
     D:String=_
     E:String="To String (Integer)"(InInt=@n4.PartIndex)
    ) -> (ReturnValue)
   )
   Return Node(execute=@n3.else, Name=@n5)
  ]}
 GetVertexColorForCommonPart {nodes [
   n1 = GetVertexColorForCommonPart!() -> (
    then
    Common Part Geometry LODs:"Atom Primitive Geometry LODs Structure (by ref)"
    Shell Merged Mesh Part Info:"Atom Merged Mesh Part Info Structure (by ref)"
    Common Part And Transform:"Atom Common Part And Transform Structure (by ref)"
   ) {is=title K2Node_FunctionEntry}
   > Return Node!(
    execute
    Vertex Color:Color Structure=Break Atom Merged Mesh Part Info!(AtomMergedMeshPartInfo=@n1.Shell Merged Mesh Part Info) -> (
     bHasConstantOverrideVertexColor=~
     VertexColor
     PreventMergingMaterials
     MaterialFeatures
     bAllowSimplification=~
     bIsOccluder=~
     FilterLODLevel
     MeshName
    ).VertexColor
   )
  ]}
 GetVertexColorForPart {nodes [
   n1 = GetVertexColorForPart!() -> (then, Part Info:"Atom Model Part Instance Info Structure (by ref)") {is=title K2Node_FunctionEntry}
   n2 = Break Atom Model Part Color Info(AtomModelPartColorInfo="Get (a ref)"(
     Array=Break Atom Model Part Instance Info!(AtomModelPartInstanceInfo=@n1) -> (
      Colors:Array of Atom Model Part Color Info Structures
      Group
      ParentGroup
      CommonParts:Array of Atom Common Part And Transform Structures
      bIsUndersideVisible:bool=~
     )
     Dimension 1=0
    ))
   > Return Node!(
    execute
    Color:Color Structure="To Color (LinearColor)"!(
     InLinearColor:Linear Color Structure="To LinearColor (Color)"!(InColor:Color Structure=@n2) -> (ReturnValue:Linear Color Structure)
     InUseSRGB:bool=false
    ) -> (ReturnValue:Color Structure)
   )
  ]}
 GetVertexColorForPartWithColorChangeIndex {nodes [
   n1 = GetVertexColorForPartWithColorChangeIndex!() -> (then, Part Info:"Atom Model Part Instance Info Structure (by ref)", Color Change Index:"Integer (by ref)") {is=title K2Node_FunctionEntry}
   n2 = Break Atom Model Part Color Info(AtomModelPartColorInfo="Get (a ref)"(Array=Break Atom Model Part Instance Info(AtomModelPartInstanceInfo=@n1) -> (bIsUndersideVisible=~), Dimension 1=@n1.Color Change Index))
   > Return Node(Color="To Color (LinearColor)"(InLinearColor="To LinearColor (Color)"(InColor=@n2), InUseSRGB=false))
  ]}
 LocalMergeExecutionWithData {nodes [
   n1 = Inputs!() -> (Option A:Exec, Option A Data:Wildcard, Option B:Exec, Option B Data:Wildcard)
   n2 = Local Boolean!() -> (Variable:bool=~) {is=title K2Node_TemporaryVariable}
   > n3 = Assign!(execute:Exec, Variable:bool=@n2, Value:bool) -> (then:Exec) {is=title K2Node_AssignmentStatement}
   Outputs!(
    Then:Exec=[@n4.then, @n3.then]
    Data:Wildcard=Select(Option 0:Wildcard=@n1, Option 1:Wildcard=@n1.Option B Data, Index=@n2) -> (ReturnValue:Wildcard)
   )
   n4 = Assign(execute=@n1.Option B, Variable=@n2, Value=true)
  ]}
 OnBuildPartsCollections {nodes [
   n1 = OnBuildPartsCollections!() -> (then, Base Parts Collection:"Atom Model Parts Collection Structure (by ref)", Model:Atom Model Object Reference) {is=title K2Node_FunctionEntry}
   n2 = Get SelectionSets!(self:Atom Model Object Reference=@n1.Model) -> (SelectionSets:Array of Atom Model Selection Set Structures)
   > n3 = Branch(Condition=AND Boolean(
     A=Get Separate Selection Sets() -> (Separate Selection Sets:bool)
     B=Is Not Empty(TargetArray:Array of Atom Model Selection Set Structures=@n2)
    ))
   > n4 = For Each Loop(Array:Array of Atom Model Selection Set Structures=@n2) -> (Array Element:Atom Model Selection Set Structure)
   Return Node!(
    execute=@n4.Completed
    Array of Filtered Parts Collections:Array of Atom Model Parts Collection Structures=Get Parts Collection() -> (Parts Collection:Array of Atom Model Parts Collection Structures)
   )
   > n5 = GetPivotForPartsCollection!(execute, self, Parts Collection:Atom Model Parts Collection Structure=@n1) -> (then, Pivot:Transform) {is=title K2Node_CallFunction}
   n6 = "To String (Name)"(InName=Break Atom Model Selection Set!(AtomModelSelectionSet:"Atom Model Selection Set Structure (by ref)"=@n4) -> (PrimitiveIds:Set of Guid Structures, SelectionSetName:Name) {is=title K2Node_BreakStruct}.SelectionSetName)
   n7 = Break Atom Model Parts Collection(AtomModelPartsCollection=@n1)
   > n8 = Set members in Atom Model Parts Collection!(
    execute:Exec
    StructRef:"Atom Model Parts Collection Structure (by ref)"=FilterSelectionSet!(
     PartsCollectionToFilter:"Atom Model Parts Collection Structure (by ref)"=@n1
     SelectionSetName:String=@n6
     NewPartsCollectionName:String=Append!(A=@n7, B=_, C=@n6) -> (ReturnValue)
    ) -> (ReturnValue:Atom Model Parts Collection Structure)
    Pivot:Transform=@n5
   ) -> (then:Exec, StructOut:"Atom Model Parts Collection Structure (by ref)") {is=title K2Node_SetFieldsInStruct}
   > Add(
    TargetArray:Array of Atom Model Parts Collection Structures=Get Parts Collection()
    NewItem:"Atom Model Parts Collection Structure (by ref)"=@n8
   ) {is=title K2Node_CallArrayFunction}
   n9 = Branch(execute=@n3.else, Condition=Get Separate Selection Sets())
   > n10 = ReportProcessorWarning!(
    execute
    Message:"Text (by ref)"=MakeLiteralText(Value="LOCTEXT(\"Missing Selection Sets in Model and  Separate Selection Sets chosen. Falling back to single mesh for all.\")")
    Asset:Object Reference=@n1.Model
   ) -> (then)
   Return Node(
    execute=[@n10.then, @n9.else]
    Array of Filtered Parts Collections=Make Array("[0]":Atom Model Parts Collection Structure=@n1) -> (Array:Array of Atom Model Parts Collection Structures)
   )
  ]}
 OnGetProcessModelTargetAssetPath {nodes [
   n1 = OnGetProcessModelTargetAssetPath!() -> (then, Model:Atom Model Object Reference, AtomModelPartsCollection:"Atom Model Parts Collection Structure (by ref)") {is=title K2Node_FunctionEntry}
   > Parent: OnGetProcessModelTargetAssetPath!(
    execute
    Model:Atom Model Object Reference=@n1
    AtomModelPartsCollection:"Atom Model Parts Collection Structure (by ref)"=@n1.AtomModelPartsCollection
   ) -> (then, ReturnValue:String) {is=title K2Node_CallParentFunction}
   > n2 = GetStaticMeshPath!(
    execute
    self
    Model:Atom Model Object Reference=@n1
    Parts Collection:"Atom Model Parts Collection Structure (by ref)"=@n1.AtomModelPartsCollection
   ) -> (then, Result:String) {is=title K2Node_CallFunction}
   > n3 = GetStaticMeshName(Model=@n1, Parts Collection=@n1.AtomModelPartsCollection)
   > Return Node!(
    execute
    ReturnValue:String=RemoveDuplicateSlashes!(InPath:String=Append!(A=@n2, B=/, C=@n3, D:String=., E:String=@n3) -> (ReturnValue)) -> (OutPath:String)
   )
  ]}
 OnGetProcessPrimitiveTargetAssetPath {nodes [
   n1 = OnGetProcessPrimitiveTargetAssetPath!() -> (
    then
    Model:Atom Model Object Reference
    Primitive:Atom Primitive Object Reference
    AtomModelPartsCollection:"Atom Model Parts Collection Structure (by ref)"
    Settings:"Atom On Process Primitive Settings Structure (by ref)"
   ) {is=title K2Node_FunctionEntry}
   > Parent: OnGetProcessPrimitiveTargetAssetPath!(
    execute
    Model:Atom Model Object Reference=@n1
    Primitive:Atom Primitive Object Reference=@n1.Primitive
    AtomModelPartsCollection:"Atom Model Parts Collection Structure (by ref)"=@n1.AtomModelPartsCollection
    Settings:"Atom On Process Primitive Settings Structure (by ref)"=@n1.Settings
   ) -> (then, ReturnValue:String) {is=title K2Node_CallParentFunction}
   > Set Support Decorations(Support Decorations:bool=Break Atom on Process Primitive Settings!(AtomOnProcessPrimitiveSettings:"Atom On Process Primitive Settings Structure (by ref)"=@n1.Settings) -> (CustomLocation:String, bSupportDecorations:bool=~) {is=title K2Node_BreakStruct}.bSupportDecorations) -> (Output_Get:bool)
   > n2 = GetStaticMeshPath(Model=@n1, Parts Collection=@n1.AtomModelPartsCollection)
   > n3 = GetStaticMeshName(Model=@n1, Parts Collection=@n1.AtomModelPartsCollection)
   > Return Node(ReturnValue=RemoveDuplicateSlashes(InPath=Append(A=@n2, B=/, C=@n3, D=., E=@n3)))
  ]}
 OnPostMeshMerge {nodes [
   n1 = OnPostMeshMerge!() -> (then, PrimitiveGeometryLODs:"Atom Primitive Geometry LODs Structure (by ref)", Materials:Array of Material Interface Object References) {is=title K2Node_FunctionEntry}
   > Return Node!(
    execute
    OutPrimitiveGeometryLODs:Atom Primitive Geometry LODs Structure=@n1
    OutMaterials:Array of Material Interface Object References=@n1.Materials
   )
  ]}
 OnProcessPartsCollection {nodes [
   n1 = GetMeshMergingSettings!(execute=@n30.Completed, self, AtomModel:Atom Model Object Reference=Get Model()) -> (then, Regular Mesh Merging Settings:Atom Mesh Merging Settings Structure) {is=title K2Node_CallFunction}
   > n2 = GetOverrideMeshMergingSettings!(execute, self, Default Settings:"Atom Mesh Merging Settings Structure (by ref)"=@n1) -> (then, Mesh Merging Settings Map:Map of Strings to Atom Mesh Merging Settings Structures) {is=title K2Node_CallFunction}
   > n3 = MakeOptimizedMergedGeometries!(
    execute
    MergedMeshParts:Array of Atom Merged Mesh Part Info Structures=Get Merged Mesh Part Infos() -> (Merged Mesh Part Infos:Array of Atom Merged Mesh Part Info Structures)
    DefaultMeshMergingSettings:"Atom Mesh Merging Settings Structure (by ref)"=@n1
    MeshMergingSettings:Map of Strings to Atom Mesh Merging Settings Structures=@n2
   ) -> (then, MergedGeometries:Map of Strings to Atom Merged Geometry Structures)
   n4 = Keys!(execute, TargetMap:Map of Strings to Atom Merged Geometry Structures=@n3) -> (then, Keys:Array of Strings)
   > n5 = For Each Loop(Array:Array of Strings=@n4) -> (Array Element:String)
   n6 = Break Atom Merged Geometry!(AtomMergedGeometry:"Atom Merged Geometry Structure (by ref)"=Find(TargetMap:Map of Strings to Atom Merged Geometry Structures=@n3, Key=@n5) -> (Value:Atom Merged Geometry Structure)) -> (GeometryLODs:Atom Primitive Geometry LODs Structure, Materials:Array of Material Interface Object References) {is=title K2Node_BreakStruct}
   > n7 = OnPostMeshMerge!(
    execute
    self
    PrimitiveGeometryLODs:"Atom Primitive Geometry LODs Structure (by ref)"=@n6
    Materials:Array of Material Interface Object References=@n6.Materials
   ) -> (then, OutPrimitiveGeometryLODs:Atom Primitive Geometry LODs Structure, OutMaterials:Array of Material Interface Object References) {is=title K2Node_CallFunction}
   > n8 = GetStaticMeshPath(Model=Get Model(), Parts Collection=Get Current Parts Collection() -> (Current Parts Collection:Atom Model Parts Collection Structure))
   > n9 = GetStaticMeshName(Model=Get Model(), Parts Collection=Get Current Parts Collection())
   > n10 = Branch(Condition=IsEmpty(InString=@n5))
   > n11 = Local Merge Execution with Data!(Option A:Exec, Option A Data:String=@n9, Option B:Exec=@n10.else, Option B Data:String=Append!(A=@n9, B=_, C=@n5) -> (ReturnValue)) -> (Then:Exec, Data:String) {is=title K2Node_MacroInstance}
   > n12 = CreateStaticMeshFromGeometry(
    GeometryLODs=@n7
    MaterialArray=@n7.OutMaterials
    Subpath=@n8
    StaticMeshName=@n11
    bPreserveExistingCollision=Get Preserve Existing Collision() -> (Preserve Existing Collision:bool)
    bPreserveExistingSockets=Get Preserve Existing Sockets() -> (Preserve Existing Sockets:bool)
    bGenerateMeshDistanceFields=true
    TraceFlag=CTF_UseSimpleAsComplex
    bSupportRaytracing=true
   )
   > Add(
    TargetMap:Map of Strings to Static Mesh Object References=Get NameMergedMeshMap() -> (NameMergedMeshMap:Map of Strings to Static Mesh Object References)
    Key=@n5
    Value:"Static Mesh Object Reference (by ref)"=@n12
   ) {is=title K2Node_CallFunction}
   > Branch(Condition=Contains Item(TargetArray:Array of Strings=Get BakedFlexPrimitivesNames() -> (BakedFlexPrimitivesNames:Array of Strings), ItemToFind:"String (by ref)"=@n5))
   > Add(
    TargetMap:Map of Atom Model Part Guid Structures to Static Mesh Object References=Get PrimitiveMeshOverrides() -> (PrimitiveMeshOverrides:Map of Atom Model Part Guid Structures to Static Mesh Object References)
    Key:"Atom Model Part Guid Structure (by ref)"=To Atom Model Part Guid(InString=@n5)
    Value=@n12
   )
   n13 = Keys(execute=@n5.Completed, TargetMap:Map of Strings to Static Mesh Object References=Get NameMergedMeshMap())
   > n14 = For Each Loop(Array=@n13)
   n15 = GetIsProcessingPrimitive()
   n16 = Branch(
    execute=@n14.Completed
    Condition=AND Boolean(A=NOT Boolean(A=@n15), B=Get Generate Blueprint Container() -> (Generate Blueprint Container:bool))
   )
   > n17 = Keys(TargetMap=Get NameMergedMeshMap())
   > Add(
    TargetArray:Array of Atom Model Hierarchical Scene Node Structures=Break Atom Model Hierarchical Scene!(AtomModelHierarchicalScene:"Atom Model Hierarchical Scene Structure (by ref)"=Get HierarchicalScene() -> (HierarchicalScene:Atom Model Hierarchical Scene Structure)) -> (SceneNodes:Array of Atom Model Hierarchical Scene Node Structures) {is=title K2Node_BreakStruct}
    NewItem:"Atom Model Hierarchical Scene Node Structure (by ref)"=Make Atom Model Hierarchical Scene Node!(Name:String=Root, WorldTransform:Transform=MakeTransform(Rotation="0, 0, 0"), ParentIndex:int=-1, ItemNames:Array of Strings=@n17) -> (AtomModelHierarchicalSceneNode:Atom Model Hierarchical Scene Node Structure) {is=title K2Node_MakeStruct}
   ) {is=title K2Node_CallArrayFunction}
   > n18 = GetBlueprintContainerPath!(
    execute
    self
    Model:Atom Model Object Reference=Get Model()
    Parts Collection:"Atom Model Parts Collection Structure (by ref)"=Get Current Parts Collection()
   ) -> (then, Result:String) {is=title K2Node_CallFunction}
   > n19 = GetBlueprintContainerName!(
    execute
    self
    Model:Atom Model Object Reference=Get Model()
    Parts Collection:"Atom Model Parts Collection Structure (by ref)"=Get Current Parts Collection()
   ) -> (then, Result:String) {is=title K2Node_CallFunction}
   > n20 = CreateBlueprintContainerFromAtomModelScene!(
    execute
    AssetPath:String=@n18
    AssetName:String=@n19
    ParentClass:Object Class Reference=Get Blueprint Container Base Class() -> (Blueprint Container Base Class:Actor Class Reference)
    Scene:"Atom Model Hierarchical Scene Structure (by ref)"=Get HierarchicalScene()
    StaticMeshes:Map of Strings to Static Mesh Object References=Get NameMergedMeshMap()
   ) -> (then, ReturnValue:Blueprint Object Reference)
   > n21 = Add(
    TargetArray:Array of Object References=Get Processed Objects() -> (Processed Objects:Array of Object References)
    NewItem:"Object Reference (by ref)"=@n20
   )
   Return Node!(
    execute=[@n21.then, @n16.else]
    Result:Atom Processor Result Structure=Make Atom Processor Result!(bSuccess:bool=True, ProcessedObjects:Array of Object References=Get Processed Objects(), SharedAssets:Array of Object References) -> (AtomProcessorResult:Atom Processor Result Structure) {is=title K2Node_MakeStruct}
   )
   n22 = Find(TargetMap:Map of Strings to Static Mesh Object References=Get NameMergedMeshMap(), Key=@n14) -> (Value:Static Mesh Object Reference)
   n23 = Add(execute=@n14.LoopBody, TargetArray=Get Processed Objects(), NewItem=@n22)
   Finalize Static Mesh!(
    Exec:"Exec (by ref)"=@n23.then
    AtomModel:"Atom Model Object Reference (by ref)"=Get Model()
    StaticMesh:"Static Mesh Object Reference (by ref)"=@n22
    CommonPartInstanceCache:Atom Common Part Instances Cache Structure=Get Common Part Instances Cache() -> (Common Part Instances Cache:Atom Common Part Instances Cache Structure)
    MergedMeshName:String=@n14
    PrimitiveMeshOverrides:Map of Atom Model Part Guid Structures to Static Mesh Object References=Get PrimitiveMeshOverrides()
   ) -> (Out:Exec) {is=title K2Node_MacroInstance}
   n24 = OnProcessPartsCollection!() -> (then, Parts Collection:"Atom Model Parts Collection Structure (by ref)", Model:Atom Model Object Reference) {is=title K2Node_FunctionEntry}
   > Set Model(Model:Atom Model Object Reference=@n24.Model) -> (Output_Get:Atom Model Object Reference)
   n25 = Break Atom Model Parts Collection(AtomModelPartsCollection=@n24)
   > Set Pivot(Pivot:Transform=@n25.Pivot) -> (Output_Get:Transform)
   > Set Parts Collection Name(Parts Collection Name:String=@n25) -> (Output_Get:String)
   > Set Local_BakeCommonParts(Local_BakeCommonParts:bool=NOT Boolean(A=Get Extract Common Parts())) -> (Output_Get:bool)
   > Branch(Condition=Is Not Empty(TargetArray:Array of Atom Model Part Instance Info Structures=@n25.Parts))
   > n26 = CreatePartsCollectionConnectivity!(
    execute
    PartsCollection:"Atom Model Parts Collection Structure (by ref)"=Get Parts Collection() -> (Parts Collection:"Atom Model Parts Collection Structure (by ref)")
    Model:Atom Model Object Reference=Get Model()
    Scale:float=GetDefaultPrimitiveScale()
   ) -> (then, ReturnValue:Atom Model Parts Collection Connectivity Object Reference)
   > n27 = Set PartsCollectionConnectivity(PartsCollectionConnectivity:Atom Model Parts Collection Connectivity Object Reference=@n26) -> (Output_Get:Atom Model Parts Collection Connectivity Object Reference)
   > InitializeCommonParts!(
    execute
    PartsCollectionConnectivity:Atom Model Parts Collection Connectivity Object Reference=@n27
    CommonPartOptimizationFlags:Bitmask Integer=Get Common Part Optimizations() -> (Common Part Optimizations:Bitmask Integer=0)
   ) -> (then)
   > n28 = Set Current Parts Collection(Current Parts Collection:Atom Model Parts Collection Structure=Get PartsCollection!(self:Atom Model Parts Collection Connectivity Object Reference=Get PartsCollectionConnectivity() -> (PartsCollectionConnectivity:Atom Model Parts Collection Connectivity Object Reference)) -> (PartsCollection:Atom Model Parts Collection Structure)) -> (Output_Get:Atom Model Parts Collection Structure)
   n29 = Break Atom Model Parts Collection(AtomModelPartsCollection=@n28)
   > n30 = For Each Loop(Array:Array of Atom Model Part Instance Info Structures=@n29.Parts) -> (Array Element:Atom Model Part Instance Info Structure)
   > n31 = Set Atom Model Part Info(Atom Model Part Info:Atom Model Part Instance Info Structure=@n30) -> (Output_Get:Atom Model Part Instance Info Structure)
   > Branch(Condition=IsValid(Object=Break Atom Model Part Instance Info!(AtomModelPartInstanceInfo=@n31) -> (Primitive:Atom Primitive Object Reference, Group, ParentGroup)))
   > n32 = Sequence!(execute) -> (then_0, then_1, then_2, then_3:Exec, then_4:Exec, then_5:Exec, then_6:Exec, then_7:Exec)
   n33 = GetVertexColorForPart!(
    execute=@n32.then_2
    self
    Part Info:"Atom Model Part Instance Info Structure (by ref)"=Get Atom Model Part Info() -> (Atom Model Part Info:Atom Model Part Instance Info Structure)
   ) -> (then, Color:Color Structure) {is=title K2Node_CallFunction}
   > Set members in Atom Merged Mesh Part Info!(
    execute:Exec
    StructRef:"Atom Merged Mesh Part Info Structure (by ref)"=Get Merged Mesh Part Info() -> (Merged Mesh Part Info:Atom Merged Mesh Part Info Structure)
    VertexColor:Color Structure=@n33
   ) -> (then:Exec, StructOut:"Atom Merged Mesh Part Info Structure (by ref)") {is=title K2Node_SetFieldsInStruct}
   Set Merged Mesh Part Info(
    execute=@n32.then_0
    Merged Mesh Part Info:Atom Merged Mesh Part Info Structure=Make Atom Merged Mesh Part Info!(
     bHasConstantOverrideVertexColor=True
     PreventMergingMaterials
     MaterialFeatures
     bAllowSimplification=True
     bIsOccluder=True
     FilterLODLevel=-1
     MeshName
    ) -> (AtomMergedMeshPartInfo)
   ) -> (Output_Get:Atom Merged Mesh Part Info Structure)
   n34 = Break Atom Model Part Instance Info!(AtomModelPartInstanceInfo=Get Atom Model Part Info()) -> (
    Transforms:Array of Transforms
    Group
    ParentGroup
    CommonParts:Array of Atom Common Part And Transform Structures
    bIsUndersideVisible:bool=~
   )
   Set members in Atom Merged Mesh Part Info!(
    execute=@n32.then_4
    StructRef=Get Merged Mesh Part Info()
    Transform:Transform=ComposeTransforms(A="Get (a copy)"(Array:Array of Transforms=@n34) -> (Output:Transform), B=InvertTransform(T=@n29.Pivot))
   ) -> (then, StructOut)
   GetCommonPartsGeometry!(
    execute=@n32.then_7
    self
    Part Info:"Atom Model Part Instance Info Structure (by ref)"=Get Atom Model Part Info()
    MergedMeshPartInfo:Atom Merged Mesh Part Info Structure=Get Merged Mesh Part Info()
    Common Part To Geometry LODs:Map of Strings to Atom Primitive Geometry LODs Structures=Get Common Parts Geometry Map()
   ) -> (then) {is=title K2Node_CallFunction}
   > n35 = For Each Loop(Array:Array of Atom Common Part And Transform Structures=Break Atom Model Part Instance Info!(AtomModelPartInstanceInfo=Get Atom Model Part Info()) -> (Group, ParentGroup, CommonParts).CommonParts) -> (Array Element:Atom Common Part And Transform Structure)
   n36 = Break Atom Common Part and Transform(AtomCommonPartAndTransform=@n35)
   > Set Common Part Type(Common Part Type:EAtomCommonPartType Enum=@n36) -> (Output_Get:EAtomCommonPartType Enum=None)
   n37 = Break Atom Merged Mesh Part Info!(AtomMergedMeshPartInfo=Get Merged Mesh Part Info()) -> (
    Transform:Transform
    bHasConstantOverrideVertexColor=~
    PreventMergingMaterials
    MaterialFeatures
    bAllowSimplification=~
    bIsOccluder=~
    FilterLODLevel
    MeshName
   )
   > n38 = Set Common Part Transform(Common Part Transform:Transform=ComposeTransforms(A=@n36.Transform, B=@n37)) -> (Output_Get:Transform)
   > n39 = Get Unique Name for Common Part Geometry LODs(PartInfo=Get Atom Model Part Info(), MergedMeshPartInfo=Get Merged Mesh Part Info(), CommonPartIndex=@n35.Array Index)
   n40 = Find(TargetMap:Map of Strings to Atom Primitive Geometry LODs Structures=Get Common Parts Geometry Map(), Key=@n39) -> (Value:Atom Primitive Geometry LODs Structure)
   > n41 = GetMaterialsForCommonPart!(
    execute
    self
    Common Part Geometry LODs:"Atom Primitive Geometry LODs Structure (by ref)"=@n40
    Shell Merged Mesh Part Info:"Atom Merged Mesh Part Info Structure (by ref)"=Get Merged Mesh Part Info()
    PartInfo:"Atom Model Part Instance Info Structure (by ref)"=Get Atom Model Part Info()
    CommonPartIndex:int=@n35.Array Index
   ) -> (then, Common Part Materials:Array of Material Interface Object References) {is=title K2Node_CallFunction}
   > n42 = GetVertexColorForCommonPart!(
    execute
    self
    Common Part Geometry LODs:"Atom Primitive Geometry LODs Structure (by ref)"=@n40
    Shell Merged Mesh Part Info:"Atom Merged Mesh Part Info Structure (by ref)"=Get Merged Mesh Part Info()
    Common Part And Transform:"Atom Common Part And Transform Structure (by ref)"=@n35
   ) -> (then, Vertex Color:Color Structure) {is=title K2Node_CallFunction}
   > n43 = Set Common Part Merged Mesh Part Info(Common Part Merged Mesh Part Info:Atom Merged Mesh Part Info Structure=Make Atom Merged Mesh Part Info!(
     GeometryLODs:Atom Primitive Geometry LODs Structure=@n40
     Transform:Transform=@n38
     bHasConstantOverrideVertexColor=True
     VertexColor:Color Structure=@n42
     Materials:Array of Material Interface Object References=@n41
     PreventMergingMaterials
     MaterialFeatures
     MeshDetailLevel:EAtomMergedMeshPartDetailLevel Enum=Decorative
     bPreserveUVs:bool=true
     bAllowSimplification=True
     bIsOccluder=True
     FilterLODLevel=-1
     MeshName=@n37.MeshName
    ) -> (AtomMergedMeshPartInfo)) -> (Output_Get:Atom Merged Mesh Part Info Structure)
   n44 = Get Atom Model Part Info()
   > SetCommonPartMeshName!(
    execute
    self
    Merged Mesh Part Info:"Atom Merged Mesh Part Info Structure (by ref)"=@n43
    Atom Model Part Info:"Atom Model Part Instance Info Structure (by ref)"=@n44
    Common Part Type:"EAtomCommonPartType Enum (by ref)"=Get Common Part Type() -> (Common Part Type:EAtomCommonPartType Enum=None)
   ) -> (then)
   > n45 = FinalizeCommonPartMergedMeshPartInfo!(
    execute
    self
    CommonPartMergedMeshPartInfo:"Atom Merged Mesh Part Info Structure (by ref)"=@n43
    PrimitivePartInfo:"Atom Model Part Instance Info Structure (by ref)"=@n44
    CommonPartIndex:int=@n35.Array Index
   ) -> (then, FinalizedPartInfo:Atom Merged Mesh Part Info Structure) {is=title K2Node_CallFunction}
   > n46 = Set Common Part Merged Mesh Part Info(Common Part Merged Mesh Part Info=@n45)
   n47 = ShouldSplitTransparentPart!(MergedMeshPartInfo:"Atom Merged Mesh Part Info Structure (by ref)"=@n46) -> (Should Split Transparent Part:bool)
   > n48 = Branch(Condition=AND Boolean(A=ShouldDefineSeparateNaniteCommonParts!() -> (Separate Nanite Common Parts:bool), B=NOT Boolean(A=@n47)))
   n49 = Break Atom Merged Mesh Part Info!(AtomMergedMeshPartInfo=Get Common Part Merged Mesh Part Info() -> (Common Part Merged Mesh Part Info:Atom Merged Mesh Part Info Structure)) -> (
    GeometryLODs:Atom Primitive Geometry LODs Structure
    Transform
    bHasConstantOverrideVertexColor=~
    VertexColor:Color Structure
    Materials:Array of Material Interface Object References
    PreventMergingMaterials
    MaterialFeatures
    MeshDetailLevel:EAtomMergedMeshPartDetailLevel Enum
    bPreserveUVs:bool=~
    bAllowMerging:bool=~
    bAllowSimplification=~
    bAllowApproximation:bool=~
    bIsOccluder=~
    FilterLODLevel
    MeshName
    InstanceSubsetID:int
    ApproximationType:Bitmask Integer
   )
   n50 = Break Atom Primitive Geometry LODs!(AtomPrimitiveGeometryLODs=@n49) -> (
    GeometryLODs
    GeometryLODsMaterials
    HiResNaniteGeomery
    HiResNaniteGeomeryMaterials
    Collision:Atom Primitive Collision Geometry Structure
    DetailedCollision
   )
   > n51 = Set Nanite Common Part Merged Mesh Part Info(Nanite Common Part Merged Mesh Part Info:Atom Merged Mesh Part Info Structure=Make Atom Merged Mesh Part Info!(
     GeometryLODs=Make Atom Primitive Geometry LODs(
      HiResNaniteGeomery=@n50.HiResNaniteGeomery
      HiResNaniteGeomeryMaterials=(MaterialSlotName=,MaterialFeatures=)
      Collision=@n50.Collision
      DetailedCollision=(AggGeom=(SphereElems=,BoxElems=,SphylElems=,ConvexElems=,TaperedCapsuleElems=,LevelSetElems=,SkinnedLevelSetElems=,MLLevelSetElems=,SkinnedTriangleMeshElems=))
     )
     Transform=@n49.Transform
     bHasConstantOverrideVertexColor=True
     VertexColor=@n49.VertexColor
     Materials=@n49.Materials
     PreventMergingMaterials
     MaterialFeatures
     MeshDetailLevel=@n49.MeshDetailLevel
     bPreserveUVs=@n49.bPreserveUVs
     bAllowMerging:bool=@n49.bAllowMerging
     bAllowSimplification=True
     bAllowApproximation:bool=@n49.bAllowApproximation
     bIsOccluder=True
     FilterLODLevel=-1
     MeshName=@n49.MeshName
     InstanceSubsetID:int=@n49.InstanceSubsetID
     ApproximationType:Bitmask Integer=@n49.ApproximationType
    ) -> (AtomMergedMeshPartInfo)) -> (Output_Get:Atom Merged Mesh Part Info Structure)
   > SetNaniteCommonPartInfo!(
    execute
    self
    Merged Mesh Part Info:"Atom Merged Mesh Part Info Structure (by ref)"=@n51
    Shell Merged Mesh Part Info:"Atom Merged Mesh Part Info Structure (by ref)"=Get Merged Mesh Part Info()
    Atom Model Part Info:"Atom Model Part Instance Info Structure (by ref)"=Get Atom Model Part Info()
   ) -> (then)
   > Add(
    TargetArray:Array of Atom Merged Mesh Part Info Structures=Get Merged Mesh Part Infos()
    NewItem:"Atom Merged Mesh Part Info Structure (by ref)"=@n51
   )
   n52 = Get Common Part Merged Mesh Part Info()
   > n53 = Set members in Atom Primitive Geometry LODs!(
    execute
    StructRef=Break Atom Merged Mesh Part Info(AtomMergedMeshPartInfo=@n52) -> (bHasConstantOverrideVertexColor=~, bPreserveUVs=~, bAllowMerging=~, bAllowSimplification=~, bAllowApproximation=~, bIsOccluder=~)
    HiResNaniteGeomery:Atom Primitive Geometry Object Reference
   ) -> (then, StructOut)
   > n54 = Set members in Atom Merged Mesh Part Info!(execute, StructRef=@n52, GeometryLODs:Atom Primitive Geometry LODs Structure=@n53) -> (then, StructOut)
   n55 = Branch(execute=[@n48.else, @n54.then], Condition=Get Local_BakeCommonParts() -> (Local_BakeCommonParts:bool))
   Add(execute=[@n57.then, @n55.then], TargetArray=Get Merged Mesh Part Infos(), NewItem=Get Common Part Merged Mesh Part Info())
   n56 = Get Common Part Merged Mesh Part Info()
   n57 = Clear(
    execute=@n55.else
    TargetArray:Array of Atom Primitive Geometry Object References=Break Atom Primitive Geometry LODs!(AtomPrimitiveGeometryLODs=Break Atom Merged Mesh Part Info!(AtomMergedMeshPartInfo=@n56) -> (
      GeometryLODs
      bHasConstantOverrideVertexColor=~
      PreventMergingMaterials
      MaterialFeatures
      bAllowSimplification=~
      bIsOccluder=~
      FilterLODLevel
      MeshName
      ApproximationType
     )) -> (GeometryLODs, GeometryLODsMaterials, HiResNaniteGeomeryMaterials, DetailedCollision)
   )
   n58 = Length(TargetArray:Array of Atom Model Part Color Info Structures=Break Atom Model Part Instance Info!(AtomModelPartInstanceInfo=Get Atom Model Part Info()) -> (Colors:Array of Atom Model Part Color Info Structures, Group, ParentGroup))
   n59 = Branch(execute=@n32.then_6, Condition=OR Boolean(A=GetIsProcessingPrimitive(), B="Equal (Integer)"(A=@n58, B=1)))
   > n60 = Add(TargetArray=Get Merged Mesh Part Infos(), NewItem=Get Merged Mesh Part Info())
   n61 = Break Atom Model Part Instance Info!(AtomModelPartInstanceInfo=Get Atom Model Part Info()) -> (PartGuid:Atom Model Part Guid Structure, Primitive:Atom Primitive Object Reference, Group, ParentGroup)
   n62 = Branch(
    execute=[@n60.then, @n72.Completed]
    Condition=AND Boolean(A=Get bGenerateBakedFlexPrimitives() -> (bGenerateBakedFlexPrimitives:bool), B=IsFlexElement(self=@n61.Primitive))
   )
   n63 = Break Atom Merged Mesh Part Info!(AtomMergedMeshPartInfo=Get Merged Mesh Part Info()) -> (
    GeometryLODs
    Transform:Transform
    bHasConstantOverrideVertexColor=~
    VertexColor:Color Structure
    Materials:Array of Material Interface Object References
    PreventMergingMaterials
    MaterialFeatures
    MeshDetailLevel:EAtomMergedMeshPartDetailLevel Enum
    bPreserveUVs:bool=~
    bAllowMerging:bool=~
    bAllowSimplification=~
    bAllowApproximation:bool=~
    bIsOccluder=~
    FilterLODLevel
    MeshName
    InstanceSubsetID:int
    ApproximationType
   )
   n64 = To String!(InModelPartGuid:"Atom Model Part Guid Structure (by ref)"=@n61) -> (ReturnValue:String)
   > Add(
    TargetArray=Get Merged Mesh Part Infos()
    NewItem=Make Atom Merged Mesh Part Info(
     GeometryLODs=@n63
     Transform="0,0,0|0,0,0|1,1,1"
     bHasConstantOverrideVertexColor=True
     VertexColor=@n63.VertexColor
     Materials=@n63.Materials
     MeshDetailLevel=@n63.MeshDetailLevel
     bPreserveUVs=@n63.bPreserveUVs
     bAllowMerging=@n63.bAllowMerging
     bAllowSimplification=True
     bAllowApproximation=@n63.bAllowApproximation
     bIsOccluder=True
     FilterLODLevel=-1
     MeshName=@n64
     InstanceSubsetID=@n63.InstanceSubsetID
     ApproximationType=@n63.ApproximationType
    )
   )
   > n65 = Add(TargetArray:Array of Strings=Get BakedFlexPrimitivesNames(), NewItem:"String (by ref)"=@n64)
   n66 = Break Atom Model Part Instance Info!(AtomModelPartInstanceInfo=Get Atom Model Part Info()) -> (PartGuid, Primitive, Decorations:Array of Atom Model Part Decoration Info Structures, Group, ParentGroup)
   n67 = GetIsProcessingPrimitive()
   Branch(
    execute=[@n65.then, @n62.else]
    Condition=AND Boolean!(
     A=Is Not Empty(TargetArray:Array of Atom Model Part Decoration Info Structures=@n66.Decorations)
     B=Get bGenerateDecoratedPrimitiveMesh() -> (bGenerateDecoratedPrimitiveMesh:bool)
     C:bool=NOT Boolean(A=@n67)
    ) -> (ReturnValue)
   )
   > n68 = RunPrimitiveProcessor!(
    execute
    Primitive:Atom Primitive Object Reference=@n66.Primitive
    PrimitiveSettings:"Atom On Process Primitive Settings Structure (by ref)"=Make Atom on Process Primitive Settings!(CustomLocation:String, bSupportDecorations:bool=true) -> (AtomOnProcessPrimitiveSettings:Atom On Process Primitive Settings Structure) {is=title K2Node_MakeStruct}
   ) -> (then, ReturnValue:Atom Processor Result Structure)
   n69 = Break Atom Processor Result!(AtomProcessorResult:"Atom Processor Result Structure (by ref)"=@n68) -> (bSuccess:bool=~, ProcessedObjects:Array of Object References, SharedAssets:Array of Object References) {is=title K2Node_BreakStruct}
   > Branch(Condition=AND Boolean!(A=@n69, B=Is Not Empty(TargetArray:Array of Object References=@n69.ProcessedObjects)) -> (ReturnValue))
   > n70 = Cast To StaticMesh(Object="Get (a copy)"(Array:Array of Object References=@n69.ProcessedObjects) -> (Output:Object Reference))
   > Add(TargetMap=Get PrimitiveMeshOverrides(), Key=@n66, Value=@n70) {is=title K2Node_CallFunction}
   n71 = Get ColorChange Geometry LODs() -> (ColorChange Geometry LODs:Array of Atom Primitive Geometry LODs Structures)
   Clear(execute=@n59.else, TargetArray:Array of Atom Primitive Geometry LODs Structures=@n71)
   > SplitGeometryLODsByPolygonGroup!(
    execute
    self
    Geometry LODs:"Atom Primitive Geometry LODs Structure (by ref)"=Break Atom Merged Mesh Part Info!(AtomMergedMeshPartInfo=Get Merged Mesh Part Info()) -> (
     GeometryLODs
     bHasConstantOverrideVertexColor=~
     PreventMergingMaterials
     MaterialFeatures
     bAllowSimplification=~
     bIsOccluder=~
     FilterLODLevel
    )
    Output Geomery LODs Array:Array of Atom Primitive Geometry LODs Structures=@n71
   ) -> (then)
   > n72 = For Each Loop(Array:Array of Atom Primitive Geometry LODs Structures=Get ColorChange Geometry LODs()) -> (Array Element:Atom Primitive Geometry LODs Structure)
   > n73 = GetMaterialsForPartWithColorChangeIndex!(
    execute
    self
    Part Info:"Atom Model Part Instance Info Structure (by ref)"=Get Atom Model Part Info()
    Geometry:Atom Primitive Geometry Object Reference="Get (a copy)"(Array:Array of Atom Primitive Geometry Object References=Break Atom Primitive Geometry LODs!(AtomPrimitiveGeometryLODs=@n72) -> (
      GeometryLODs
      GeometryLODsMaterials
      HiResNaniteGeomery:Atom Primitive Geometry Object Reference
      HiResNaniteGeomeryMaterials
      Collision:Atom Primitive Collision Geometry Structure
      DetailedCollision
     )) -> (Output:Atom Primitive Geometry Object Reference)
    Model:Atom Model Object Reference=Get Model()
    Color Change Index:"Integer (by ref)"=@n72.Array Index
   ) -> (then, Material:Array of Material Interface Object References) {is=title K2Node_CallFunction}
   > n74 = GetVertexColorForPartWithColorChangeIndex!(
    execute
    self
    Part Info:"Atom Model Part Instance Info Structure (by ref)"=Get Atom Model Part Info()
    Color Change Index:"Integer (by ref)"=@n72.Array Index
   ) -> (then, Color:Color Structure) {is=title K2Node_CallFunction}
   > n75 = Set members in Atom Merged Mesh Part Info!(
    execute
    StructRef=Get Merged Mesh Part Info()
    GeometryLODs=@n72
    VertexColor:Color Structure=@n74
    Materials:Array of Material Interface Object References=@n73
   ) -> (then, StructOut)
   > Add(
    TargetArray:Array of Atom Merged Mesh Part Info Structures=Get Merged Mesh Part Infos()
    NewItem:"Atom Merged Mesh Part Info Structure (by ref)"=@n75
   ) {is=title K2Node_CallArrayFunction}
   SetPartOptimizationStrategy!(
    execute=@n32.then_5
    self
    Merged Mesh Part Info:"Atom Merged Mesh Part Info Structure (by ref)"=Get Merged Mesh Part Info()
    Atom Model Part Info:"Atom Model Part Instance Info Structure (by ref)"=Get Atom Model Part Info()
   ) -> (then)
   > SetPartMeshName!(
    execute
    self
    Merged Mesh Part Info:"Atom Merged Mesh Part Info Structure (by ref)"=Get Merged Mesh Part Info()
    Atom Model Part Info:"Atom Model Part Instance Info Structure (by ref)"=Get Atom Model Part Info()
   ) -> (then)
   > n76 = FinalizeMergedMeshPart!(execute, self, MergedMeshPartInfo:"Atom Merged Mesh Part Info Structure (by ref)"=Get Merged Mesh Part Info()) -> (then, Out MergedMeshPartInfo:Atom Merged Mesh Part Info Structure) {is=title K2Node_CallFunction}
   > Set Merged Mesh Part Info(Merged Mesh Part Info=@n76)
   > ProcessUncommonParts!(
    execute
    self
    Part Instance Info:"Atom Model Part Instance Info Structure (by ref)"=Get Atom Model Part Info()
    Merged Mesh Part Info:"Atom Merged Mesh Part Info Structure (by ref)"=Get Merged Mesh Part Info()
    Merged Mesh Part Infos:Array of Atom Merged Mesh Part Info Structures=Get Merged Mesh Part Infos()
   ) -> (then)
   n77 = Break Atom Merged Mesh Part Info!(AtomMergedMeshPartInfo=Get Merged Mesh Part Info()) -> (
    GeometryLODs
    Transform:Transform
    bHasConstantOverrideVertexColor=~
    VertexColor:Color Structure
    Materials:Array of Material Interface Object References
    PreventMergingMaterials
    MaterialFeatures
    MeshDetailLevel:EAtomMergedMeshPartDetailLevel Enum
    bPreserveUVs:bool=~
    bAllowMerging:bool=~
    bAllowSimplification=~
    bAllowApproximation:bool=~
    bIsOccluder=~
    FilterLODLevel
    MeshName:String
    InstanceSubsetID:int
    ApproximationType:Bitmask Integer
   )
   n78 = GetMaterialsForPart!(
    execute=@n32.then_3
    self
    Part Info:"Atom Model Part Instance Info Structure (by ref)"=Get Atom Model Part Info()
    Geometry:Atom Primitive Geometry Object Reference="Get (a copy)"(Array=Break Atom Primitive Geometry LODs(AtomPrimitiveGeometryLODs=@n77))
    Model:Atom Model Object Reference=Get Model()
   ) -> (then, Material:Array of Material Interface Object References) {is=title K2Node_CallFunction}
   > Set members in Atom Merged Mesh Part Info!(execute, StructRef=Get Merged Mesh Part Info(), Materials=@n78) -> (then, StructOut)
   n79 = Get Merged Mesh Part Info()
   > n80 = GetPreserveUVsForPart!(
    execute
    self
    Atom Model Part Info:"Atom Model Part Instance Info Structure (by ref)"=Get Atom Model Part Info()
    Merged Mesh Part Info:"Atom Merged Mesh Part Info Structure (by ref)"=@n79
   ) -> (then, PreserveUVs:bool) {is=title K2Node_CallFunction}
   > Set members in Atom Merged Mesh Part Info!(execute, StructRef=@n79, bPreserveUVs:bool=@n80) -> (then, StructOut)
   n81 = GetPrimitiveGeometry!(
    execute=@n32.then_1
    self
    Atom Model Part Info:"Atom Model Part Instance Info Structure (by ref)"=Get Atom Model Part Info()
    Geometry Map:Map of Strings to Atom Primitive Geometry LODs Structures=Get Geometry Map() -> (Geometry Map:Map of Strings to Atom Primitive Geometry LODs Structures)
    Atom Model:Atom Model Object Reference=Get Model()
   ) -> (then, Geometry LODs:Atom Primitive Geometry LODs Structure, Geometry Unique Name:String) {is=title K2Node_CallFunction}
   > Set Geometry Unique Name(Geometry Unique Name:String=@n81.Geometry Unique Name) -> (Output_Get:String)
   > n82 = Branch(Condition=Get Generate Nanite Mesh())
   > Set members in Atom Merged Mesh Part Info!(
    execute
    StructRef=Get Merged Mesh Part Info()
    GeometryLODs:Atom Primitive Geometry LODs Structure=@n81
    MeshDetailLevel:EAtomMergedMeshPartDetailLevel Enum=Standard
    bPreserveUVs=False
   ) -> (then, StructOut)
   n83 = Set members in Atom Primitive Geometry LODs(execute=@n82.else, StructRef=@n81)
   > Set members in Atom Merged Mesh Part Info(StructRef=Get Merged Mesh Part Info(), GeometryLODs=@n83, MeshDetailLevel=Standard, bPreserveUVs=False)
   Branch(Condition=Get Generate Static Meshes For Brick Geometry() -> (Generate Static Meshes For Brick Geometry:bool))
   n84 = Get Merged Mesh Part Info()
   > n85 = GetPartStaticMeshPath!(
    execute
    self
    Atom Model:Atom Model Object Reference=Get Model()
    Parts Collection:"Atom Model Parts Collection Structure (by ref)"=Get Current Parts Collection()
    Part Info:"Atom Merged Mesh Part Info Structure (by ref)"=@n84
   ) -> (then, PartStaticMeshPath:String) {is=title K2Node_CallFunction}
   n86 = Break Atom Merged Mesh Part Info(AtomMergedMeshPartInfo=@n84) -> (bHasConstantOverrideVertexColor=~, bPreserveUVs=~, bAllowMerging=~, bAllowSimplification=~, bAllowApproximation=~, bIsOccluder=~)
   > n87 = DuplicateAtomPrimitiveGeometryLODs!(execute, GeometryLODs:"Atom Primitive Geometry LODs Structure (by ref)"=@n86) -> (then, ReturnValue:Atom Primitive Geometry LODs Structure)
   > CreatePartStaticMesh!(
    execute
    self
    Part Info:"Atom Model Part Instance Info Structure (by ref)"=Get Atom Model Part Info()
    MergedMeshPartInfo:"Atom Merged Mesh Part Info Structure (by ref)"=Make Atom Merged Mesh Part Info(
     GeometryLODs=@n87
     Transform=@n86.Transform
     bHasConstantOverrideVertexColor=True
     VertexColor=@n86.VertexColor
     Materials=@n86.Materials
     MeshDetailLevel=@n86.MeshDetailLevel
     bPreserveUVs=@n86.bPreserveUVs
     bAllowMerging=@n86.bAllowMerging
     bAllowSimplification=True
     bAllowApproximation=@n86.bAllowApproximation
     bIsOccluder=True
     FilterLODLevel=-1
     InstanceSubsetID=@n86.InstanceSubsetID
     ApproximationType=255
    )
    Common Parts Geometry Map:Map of Strings to Atom Primitive Geometry LODs Structures=Get Common Parts Geometry Map()
    Geometry Unique Name:"String (by ref)"=Get Geometry Unique Name() -> (Geometry Unique Name:String)
    StaticMeshPath:String=@n85
   ) -> (then, Static Mesh:Static Mesh Object Reference) {is=title K2Node_CallFunction}
   Branch(Condition=Get Extract Common Parts())
   > n88 = GetCommonPartStaticMesh!(
    execute
    ExportStyleName:String
    CommonPartType:EAtomCommonPartType Enum=Get Common Part Type() -> (Common Part Type=None)
    CommonPartMeshPath:String
   ) -> (then, ReturnValue:Static Mesh Object Reference)
   > n89 = Branch(Condition=Get SplitTransparentPart() -> (SplitTransparentPart:bool))
   n90 = Break Atom Model Part Instance Info!(AtomModelPartInstanceInfo=Get Atom Model Part Info()) -> (PartGuid, Colors:Array of Atom Model Part Color Info Structures, Group, ParentGroup)
   n91 = Break Atom Model Part Color Info(AtomModelPartColorInfo="Get (a ref)"(Array=@n90.Colors, Dimension 1=0))
   n92 = Get Common Part Transform() -> (Common Part Transform:Transform)
   > Add Common Part Instance to Cache!(
    execute
    Cache:"Atom Common Part Instances Cache Structure (by ref)"=Get TransCommonPartInstanceCache() -> (TransCommonPartInstanceCache:Atom Common Part Instances Cache Structure)
    Transform:"Transform (by ref)"=@n92
    Mesh:Static Mesh Object Reference=@n88
    Material:Material Interface Object Reference
    Guid:"Atom Model Part Guid Structure (by ref)"=@n90
    ColorId:int=@n91.ColorId
   ) -> (then)
   Add Common Part Instance to Cache(execute=@n89.else, Cache=Get Common Part Instances Cache(), Transform=@n92, Mesh=@n88, Guid=@n90, ColorId=@n91.ColorId)
   n93 = Keys(execute=@n109.Completed, TargetMap:Map of Strings to Atom Merged Geometry Structures=@n3)
   > n94 = For Each Loop(Array:Array of Strings=@n93) -> (Array Element:String)
   n95 = Break Atom Merged Geometry(AtomMergedGeometry=Find(TargetMap:Map of Strings to Atom Merged Geometry Structures=@n3, Key=@n94) -> (Value:Atom Merged Geometry Structure))
   > n96 = For Each Loop(Array:Array of Material Interface Object References=@n95.Materials) -> (Array Element:Material Interface Object Reference)
   > Add Unique(
    TargetArray:Array of Material Interface Object References=Get Skeletal Materials() -> (Skeletal Materials:Array of Material Interface Object References)
    NewItem:"Material Interface Object Reference (by ref)"=@n96
   )
   n97 = CreateEmptyAtomGeometry!(execute=@n109.LoopBody, NumberOfUVChannels:int=1) -> (then, ReturnValue:Atom Primitive Geometry Object Reference)
   Add(
    execute=@n106.then
    TargetArray:Array of Atom Primitive Geometry Object References=Break Atom Primitive Geometry LODs(AtomPrimitiveGeometryLODs=Get Skeletal Mesh Geometry() -> (Skeletal Mesh Geometry:Atom Primitive Geometry LODs Structure))
    NewItem:"Atom Primitive Geometry Object Reference (by ref)"=@n97
   )
   n98 = Get Model()
   n99 = GetModelName!(self:Atom Model Object Reference=@n98) -> (ReturnValue:String)
   CreateSkeletalMeshFromGeometry!(
    execute=@n94.Completed
    GeometryLODs:Array of Atom Primitive Geometry Object References=Break Atom Primitive Geometry LODs(AtomPrimitiveGeometryLODs=Get Skeletal Mesh Geometry())
    MaterialArray:Array of Material Interface Object References=Get Skeletal Materials()
    TemplateSkeleton:Skeleton Object Reference
    Subpath:String=GetModelPath(self=@n98)
    SkeletonName:String=Append!(A=SKEL_, B=@n99) -> (ReturnValue)
    MeshName:String=Append(A=SK_, B=@n99)
   ) -> (then, OutSkeletalMesh:Skeletal Mesh Object Reference, OutSkeleton:Skeleton Object Reference)
   n100 = Get Skeletal Mesh Geometry()
   n101 = For Each Loop(
    Exec=@n96.Completed
    Array:Array of Atom Primitive Geometry Object References=Break Atom Primitive Geometry LODs(AtomPrimitiveGeometryLODs=@n95)
   ) -> (Array Element:Atom Primitive Geometry Object Reference)
   Append!(
    execute=@n108.then
    self:Atom Primitive Geometry Object Reference="Get (a copy)"(Array=Break Atom Primitive Geometry LODs(AtomPrimitiveGeometryLODs=@n100), Dimension 1=@n101.Array Index)
    GeometryToAppend:Atom Primitive Geometry Object Reference=@n101
    Transform:"Transform 3f Structure (by ref)"=MakeTransform(Rotation="0, 0, 0")
   ) -> (then, ReturnValue:Atom Primitive Geometry Object Reference) {is=title K2Node_CallFunction}
   n102 = Values!(execute=@n3.then, TargetMap:Map of Strings to Atom Merged Geometry Structures=@n3) -> (then, Values:Array of Atom Merged Geometry Structures)
   n103 = Break Atom Merged Geometry(AtomMergedGeometry="Get (a ref)"(Array:Array of Atom Merged Geometry Structures=@n102, Dimension 1=0) -> (Output:"Atom Merged Geometry Structure (by ref)"))
   n104 = Break Atom Primitive Geometry LODs(AtomPrimitiveGeometryLODs=@n103)
   n105 = Break Atom Primitive Geometry LODs(AtomPrimitiveGeometryLODs=Get Skeletal Mesh Geometry())
   n106 = AddSkeletonToGeometryFromGroups!(
    execute=@n97.then
    AtomModel:Atom Model Object Reference=Get Model()
    PartsCollection:"Atom Model Parts Collection Structure (by ref)"=Get Parts Collection()
    Geometry:Atom Primitive Geometry Object Reference=@n97
   ) -> (then)
   n107 = GetBoneIndexForName!(self:Atom Primitive Geometry Object Reference="Get (a copy)"(Array=@n105), BoneName:String=@n94) -> (ReturnValue:int=0)
   n108 = AttachVerticesToBoneIndex!(execute=@n101.LoopBody, self:Atom Primitive Geometry Object Reference=@n101, BoneIndex:int=@n107) -> (then)
   n109 = For Each Loop(Exec=@n102.then, Array=@n104)
  ]}
 OnProcessPrimitive {nodes [
   n1 = OnProcessPrimitive!() -> (
    then
    DummyModel:Atom Model Object Reference
    Primitive:Atom Primitive Object Reference
    AtomModelPartsCollection:"Atom Model Parts Collection Structure (by ref)"
    Settings:"Atom On Process Primitive Settings Structure (by ref)"
   ) {is=title K2Node_FunctionEntry}
   > Set Support Decorations(Support Decorations=Break Atom on Process Primitive Settings(AtomOnProcessPrimitiveSettings=@n1.Settings) -> (bSupportDecorations=~).bSupportDecorations)
   > n2 = Parent: OnProcessPrimitive!(
    execute
    DummyModel:Atom Model Object Reference=@n1
    Primitive:Atom Primitive Object Reference=@n1.Primitive
    AtomModelPartsCollection:"Atom Model Parts Collection Structure (by ref)"=@n1.AtomModelPartsCollection
    Settings:"Atom On Process Primitive Settings Structure (by ref)"=@n1.Settings
   ) -> (then, ReturnValue:Atom Processor Result Structure) {is=title K2Node_CallParentFunction}
   > Return Node!(execute, ReturnValue:Atom Processor Result Structure=@n2)
  ]}
 PreProcessorPipeline {nodes [
   n1 = PreProcessorPipeline!() -> (then, AtomModel:Atom Model Object Reference, PartsCollection:"Atom Model Parts Collection Structure (by ref)") {is=title K2Node_FunctionEntry}
   > Parent: PreProcessorPipeline!(execute, AtomModel:Atom Model Object Reference=@n1, PartsCollection:"Atom Model Parts Collection Structure (by ref)"=@n1.PartsCollection) -> (then) {is=title K2Node_CallParentFunction}
   > n2 = Branch(Condition=GetIsProcessingPrimitive())
   > n3 = Set Extract Common Parts(Extract Common Parts:bool=false) -> (Output_Get:bool)
   Return Node!(execute=[@n2.else, @n3.then])
  ]}
 ProcessUncommonParts {nodes [
   n1 = ProcessUncommonParts!() -> (
    then
    Part Instance Info:"Atom Model Part Instance Info Structure (by ref)"
    Merged Mesh Part Info:"Atom Merged Mesh Part Info Structure (by ref)"
    Merged Mesh Part Infos:Array of Atom Merged Mesh Part Info Structures
   ) {is=title K2Node_FunctionEntry}
   n2 = Break Atom Model Part Instance Info!(AtomModelPartInstanceInfo=@n1) -> (Primitive:Atom Primitive Object Reference, Transforms:Array of Transforms, Group, ParentGroup)
   n3 = Get Geometry style()
   n4 = Get Fallback Geometry style()
   > n5 = For Each Loop(Array:Array of Atom Primitive Geometry And Transform Structures=GetUncommonPartsGeometry!(
     self:Atom Primitive Geometry Container Object Reference=GetGeometryContainerForExportStyle(self=@n2, ExportStyleName=@n3, FallbackExportStyleName=@n4, LODIndex=0, bUseFallbacks=true)
     ColorChangeIndex:int=-1
    ) -> (ReturnValue:Array of Atom Primitive Geometry And Transform Structures)) -> (Array Element:Atom Primitive Geometry And Transform Structure)
   n6 = Break Atom Primitive Geometry and Transform!(AtomPrimitiveGeometryAndTransform:"Atom Primitive Geometry And Transform Structure (by ref)"=@n5) -> (AtomPrimitiveGeometry:Atom Primitive Geometry Object Reference, Transform:Transform 3f Structure) {is=title K2Node_BreakStruct}
   n7 = GetDefaultPrimitiveScale()
   > n8 = BakeScale(self=@n6, Scale=@n7)
   > n9 = GetGeometryLODsForUncommonPart!(execute, self, Geometry:Atom Primitive Geometry Object Reference=@n8) -> (then, Geometry LODs:Atom Primitive Geometry LODs Structure) {is=title K2Node_CallFunction}
   > n10 = Branch(Condition=Get Generate Nanite Mesh())
   n11 = BreakTransform(InTransform=@n6.Transform)
   n12 = vector * vector(A=@n11, B:float=@n7) -> (ReturnValue=~)
   n13 = Set members in Atom Merged Mesh Part Info!(
    execute=[@n10.then, @n14.then]
    StructRef=Copy!(Item:Atom Merged Mesh Part Info Structure=@n1.Merged Mesh Part Info) -> (ReturnValue:Atom Merged Mesh Part Info Structure) {is=title K2Node_Copy}
    GeometryLODs=@n9
    Transform:Transform=ComposeTransforms(
     A=MakeTransform(Location=@n12, Rotation=@n11.Rotation, Scale=@n11.Scale)
     B="Get (a ref)"(Array:Array of Transforms=@n2.Transforms, Dimension 1=0) -> (Output:"Transform (by ref)")
    )
    MeshDetailLevel=Decorative
   ) -> (then, StructOut)
   > Add(
    TargetArray:Array of Atom Merged Mesh Part Info Structures=@n1.Merged Mesh Part Infos
    NewItem:"Atom Merged Mesh Part Info Structure (by ref)"=@n13
   )
   n14 = Set members in Atom Primitive Geometry LODs(execute=@n10.else, StructRef=@n9)
  ]}
 SetCommonPartMeshName {nodes [
   n1 = SetCommonPartMeshName!() -> (
    then
    Merged Mesh Part Info:"Atom Merged Mesh Part Info Structure (by ref)"
    Atom Model Part Info:"Atom Model Part Instance Info Structure (by ref)"
    Common Part Type:"EAtomCommonPartType Enum (by ref)"=None
   ) {is=title K2Node_FunctionEntry}
   > Branch(Condition=ShouldSplitTransparentPart(MergedMeshPartInfo=@n1))
   > n2 = Set members in Atom Primitive Geometry LODs(StructRef=Break Atom Merged Mesh Part Info(AtomMergedMeshPartInfo=@n1) -> (bHasConstantOverrideVertexColor=~, bPreserveUVs=~, bAllowMerging=~, bAllowSimplification=~, bAllowApproximation=~, bIsOccluder=~))
   > Set members in Atom Merged Mesh Part Info!(execute, StructRef=@n1, GeometryLODs=@n2, MeshName:String=Trans) -> (then, StructOut)
  ]}
 SetDecorationMaterialParameters {nodes [
   n1 = SetDecorationMaterialParameters!() -> (
    then
    Material Instance:Material Instance Constant Object Reference
    Decoration Enable Param Name:String
    Texture Param Name:String
    UDIM Param Name:String
    Texture:Texture Object Reference
    UDIM:int
   ) {is=title K2Node_FunctionEntry}
   > SetMaterialInstanceStaticSwitchParameterValue(
    Instance=@n1
    ParameterName=String To Name(InString=@n1.Decoration Enable Param Name)
    Value=true
    Association=GlobalParameter
    bUpdateMaterialInstance=true
   )
   > SetMaterialInstanceTextureParameterValue!(
    execute
    Instance:Material Instance Constant Object Reference=@n1
    ParameterName:Name=String To Name(InString=@n1.Texture Param Name)
    Value:Texture Object Reference=@n1.Texture
    Association:EMaterialParameterAssociation Enum=GlobalParameter
   ) -> (then, ReturnValue:bool)
   > SetMaterialInstanceScalarParameterValue!(
    execute
    Instance:Material Instance Constant Object Reference=@n1
    ParameterName:Name=String To Name(InString=@n1.UDIM Param Name)
    Value:float="To Float (Integer)"(InInt=@n1.UDIM)
    Association:EMaterialParameterAssociation Enum=GlobalParameter
   ) -> (then, ReturnValue:bool)
   > UpdateMaterialInstance!(execute, Instance:Material Instance Constant Object Reference=@n1) -> (then)
  ]}
 SetNaniteCommonPartInfo {nodes [SetNaniteCommonPartInfo!() -> (
    then
    Merged Mesh Part Info:"Atom Merged Mesh Part Info Structure (by ref)"
    Shell Merged Mesh Part Info:"Atom Merged Mesh Part Info Structure (by ref)"
    Atom Model Part Info:"Atom Model Part Instance Info Structure (by ref)"
   ) {is=title K2Node_FunctionEntry}]}
 SetPartMeshName {nodes [
   n1 = SetPartMeshName!() -> (
    then
    Merged Mesh Part Info:"Atom Merged Mesh Part Info Structure (by ref)"
    Atom Model Part Info:"Atom Model Part Instance Info Structure (by ref)"
   ) {is=title K2Node_FunctionEntry}
   > Set members in Atom Merged Mesh Part Info!(execute, StructRef=@n1, MeshName=Break Atom Model Part Instance Info!(AtomModelPartInstanceInfo=@n1.Atom Model Part Info) -> (Group)) -> (then, StructOut)
  ]}
 SetPartOptimizationStrategy {nodes [
   n1 = SetPartOptimizationStrategy!() -> (
    then
    Merged Mesh Part Info:"Atom Merged Mesh Part Info Structure (by ref)"
    Atom Model Part Info:"Atom Model Part Instance Info Structure (by ref)"
   ) {is=title K2Node_FunctionEntry}
   n2 = Break Atom Primitive Optimization Settings!(AtomPrimitiveOptimizationSettings:"Atom Primitive Optimization Settings Structure (by ref)"=Get OptimizationSettings!(self:Atom Primitive Object Reference=Break Atom Model Part Instance Info!(AtomModelPartInstanceInfo=@n1.Atom Model Part Info) -> (Primitive:Atom Primitive Object Reference, Group, ParentGroup:String)) -> (OptimizationSettings:Atom Primitive Optimization Settings Structure)) -> (
    OptimizationType:EAtomPrimitiveGeoOptimization Enum
    ApproximationShapeType:EAtomPrimitiveApproximationShapeType Enum
    bUseOptimizationAxisOverride:bool=~
    ApproximationAxisOverride:Vector=~
   ) {is=title K2Node_BreakStruct}
   > Branch(Condition="Not Equal (Enum)"!(A:EAtomPrimitiveGeoOptimization Enum=@n2, B:EAtomPrimitiveGeoOptimization Enum=Default) -> (ReturnValue:bool))
   > n3 = Switch on EAtomPrimitiveApproximationShapeType!(execute, Selection:EAtomPrimitiveApproximationShapeType Enum=@n2.ApproximationShapeType) -> (Auto:Exec, OrientedBox:Exec, ExtrudedConvexHull2D:Exec, ConvexHull3D:Exec, Extrusion:Exec) {is=title K2Node_SwitchEnum}
   > Set members in Atom Merged Mesh Part Info!(execute, StructRef=@n1, ApproximationType:Bitmask Integer=255) -> (then, StructOut)
   Set members in Atom Merged Mesh Part Info(execute=@n3.ConvexHull3D, StructRef=@n1, ApproximationType=232)
   Set members in Atom Merged Mesh Part Info(execute=@n3.ExtrudedConvexHull2D, StructRef=@n1, ApproximationType=228)
   Set members in Atom Merged Mesh Part Info(execute=@n3.Extrusion, StructRef=@n1, ApproximationType=240)
   Set members in Atom Merged Mesh Part Info(execute=@n3.OrientedBox, StructRef=@n1, ApproximationType=227)
  ]}
 ShouldDefineSeparateNaniteCommonParts {nodes [
   ShouldDefineSeparateNaniteCommonParts() {is=title K2Node_FunctionEntry}
   > Return Node!(execute, Separate Nanite Common Parts:bool=false)
  ]}
 ShouldSplitTransparentPart {nodes [
   n1 = ShouldSplitTransparentPart!() -> (then, MergedMeshPartInfo:"Atom Merged Mesh Part Info Structure (by ref)") {is=title K2Node_FunctionEntry}
   > n2 = GetHasOnlyNaniteCompatibleMaterials!(execute, self, MergedMeshPartInfo:"Atom Merged Mesh Part Info Structure (by ref)"=@n1) -> (then, bOnlyHasOpaque:bool) {is=title K2Node_CallFunction}
   > Return Node!(
    execute
    Should Split Transparent Part:bool=AND Boolean(A=NOT Boolean(A=@n2), B=Get SplitTransparentMeshes() -> (SplitTransparentMeshes:bool))
   )
  ]}
 SplitGeometryLODsByPolygonGroup {nodes [
   n1 = SplitGeometryLODsByPolygonGroup!() -> (
    then
    Geometry LODs:"Atom Primitive Geometry LODs Structure (by ref)"
    Output Geomery LODs Array:Array of Atom Primitive Geometry LODs Structures
   ) {is=title K2Node_FunctionEntry}
   n2 = Break Atom Primitive Geometry LODs(AtomPrimitiveGeometryLODs=@n1)
   > n3 = For Each Loop(Array:Array of Atom Primitive Geometry Object References=@n2) -> (Array Element:Atom Primitive Geometry Object Reference)
   n4 = For Each Loop(
    Exec=@n3.Completed
    Array=SplitByPolygonGroup!(self:Atom Primitive Geometry Object Reference=@n2.HiResNaniteGeomery, bIgnoreEmptyGroups:bool=true) -> (ReturnValue:Array of Atom Primitive Geometry Object References)
   )
   Set members in Atom Primitive Geometry LODs!(
    execute=@n4.Completed
    StructRef="Get (a ref)"(Array:Array of Atom Primitive Geometry LODs Structures=@n1.Output Geomery LODs Array, Dimension 1=0) -> (Output:"Atom Primitive Geometry LODs Structure (by ref)")
    Collision:Atom Primitive Collision Geometry Structure=@n2.Collision
   ) -> (then, StructOut)
   Set members in Atom Primitive Geometry LODs!(
    execute=@n4.LoopBody
    StructRef="Get (a ref)"(Array=@n1.Output Geomery LODs Array, Dimension 1=@n4.Array Index)
    HiResNaniteGeomery:Atom Primitive Geometry Object Reference=@n4
   ) -> (then, StructOut)
   n5 = For Each Loop(Exec=@n3.LoopBody, Array=SplitByPolygonGroup(self=@n3, bIgnoreEmptyGroups=true))
   > Resize!(
    execute
    TargetArray:Array of Atom Primitive Geometry LODs Structures=@n1.Output Geomery LODs Array
    Size:int="Max (Integer)"(A=Length(TargetArray:Array of Atom Primitive Geometry LODs Structures=@n1.Output Geomery LODs Array), B=int + int(A=@n5.Array Index))
   ) -> (then) {is=title K2Node_CallArrayFunction}
   > Set Array Elem!(
    execute
    TargetArray:Array of Atom Primitive Geometry Object References=Break Atom Primitive Geometry LODs(AtomPrimitiveGeometryLODs="Get (a ref)"(Array=@n1.Output Geomery LODs Array, Dimension 1=@n5.Array Index))
    Index:int=@n3.Array Index
    Item:"Atom Primitive Geometry Object Reference (by ref)"=@n5
    bSizeToFit:bool=true
   ) -> (then)
  ]}
 SplitTransparentMesh {nodes [
   n1 = Inputs!() -> (
    Exec:Exec
    IsTransparent:bool=~
    PartInfo:Atom Merged Mesh Part Info Structure
    TransPartInfos:Array of Atom Merged Mesh Part Info Structures
    OpaquePartInfos:Array of Atom Merged Mesh Part Info Structures
   )
   > n2 = Branch(Condition=@n1)
   > n3 = Add(TargetArray=@n1.TransPartInfos, NewItem=@n1.PartInfo)
   Outputs!(Out:Exec=[@n4.then, @n3.then])
   n4 = Add(execute=@n2.else, TargetArray=@n1.OpaquePartInfos, NewItem=@n1.PartInfo)
  ]}
]
variables [
 Auto Primitive Collision Type {type=EPrimitiveAutoCollisionType, default=Box}
 Bake Common Parts {type=boolean}
 bGenerateBakedFlexPrimitives {type=boolean}
 bGenerateDecoratedPrimitiveMesh {type=boolean}
 Blueprint Container Base Class {type=Class, default=&Actor}
 bUseImportedPrimitiveCollision {type=boolean}
 Common Part Optimizations {type=integer, default=1}
 Common Part Scale {type=number, default=20}
 Common Part Style {type=string, default=Realtime}
 Extract Common Parts {type=boolean}
 Fallback Geometry style {type=string, default=""}
 Generate Blueprint Container {type=boolean}
 Generate Geometry Collection {type=boolean, default=true}
 Generate Nanite Mesh {type=boolean}
 Generate Static Meshes For Brick Geometry {type=boolean}
 Geometry Static Mesh Path {type=string, default=""}
 Geometry style {type=string, default=Realtime}
 Model {type=AtomModel, default=null}
 Parts Collection Name {type=string, default=""}
 PartsCollectionConnectivity {type=AtomModelPartsCollectionConnectivity, default=null}
 Pivot {type=Transform, default {location=<0 0 0>, rotation=<rot 0 0 0>, scale=<1 1 1>}}
 Preserve Existing Collision {type=boolean, default=true}
 Preserve Existing Sockets {type=boolean, default=true}
 Remove All Knobs {type=boolean}
 Remove All Pins {type=boolean}
 Remove All Tubes {type=boolean}
 Remove Connected {type=boolean}
 Separate Selection Sets {type=boolean}
 SplitTransparentMeshes {type=boolean}
 Support Decorations {type=boolean, default=true}
]
